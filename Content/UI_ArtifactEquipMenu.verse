using { /Fortnite.com/Devices }
using { /Verse.org/Simulation }
using { /UnrealEngine.com/Temporary/Diagnostics }
using { /UnrealEngine.com/Temporary/UI }
using { /Fortnite.com/UI }
using { /UnrealEngine.com/Temporary/SpatialMath}
using { /Verse.org/Colors }
using { /Verse.org/Assets }

<#
UI_ArtifactEquipMenu := class():
    ToMessage<localizes>(String : string) : message = "{String}"
    CenteredAlignment : vector2 = vector2{X := 0.5, Y := 0.5}

    var BuiltCanvasMap : [player]ArtifactEquipData = map{}

    InitUI(InPlayer : player, ArtifactManager : V_ArtifactManager):void=
        ArtifactMenuCanvas := MakeArtifactEquipCanvas()
        # Set the players build canvas data
        if(set BuiltCanvasMap[InPlayer] = ArtifactMenuCanvas){}
        ArtifactMenuCanvas.ExitButton.OnClick().Subscribe(HandlePlayerExit)
        if(CanvasData := BuiltCanvasMap[InPlayer]):
            for(Index -> Display : CanvasData.ArtifactDisplays):
                Display.NextButton.OnClick().Subscribe(ChangeDisplayHandler{Next := true, Device := ArtifactManager, Slot := Index + 1}.HandlerFunction)
                Display.PrevButton.OnClick().Subscribe(ChangeDisplayHandler{Next := false, Device := ArtifactManager, Slot := Index + 1}.HandlerFunction)
                Display.EquipButton.OnClick().Subscribe(EquipHandler{Device := ArtifactManager, Slot := Index + 1}.HandlerFunction)

    # Add the UI to the player
    AddUI(InPlayer : player):void=
        if(PlayerUI := GetPlayerUI[InPlayer], CanvasData := BuiltCanvasMap[InPlayer]):
            CanvasData.Canvas.SetVisibility(widget_visibility.Visible)
            PlayerUI.AddWidget(CanvasData.Canvas, player_ui_slot{InputMode := ui_input_mode.All} )

    # When player presses the exit button
    HandlePlayerExit(WidgetMessage: widget_message):void=
        spawn{HideUI(WidgetMessage.Player)}

    HideUI(Player: player)<suspends>:void=
        Sleep(0.0) # So we're sure that the canvas has been added to BuiltCanvas map
        if(PlayerUI := GetPlayerUI[Player], CanvasData := BuiltCanvasMap[Player]):
            CanvasData.Canvas.SetVisibility(widget_visibility.Hidden)
            PlayerUI.RemoveWidget(CanvasData.Canvas)

    UpdateAllSlots(InPlayer : player, InPlayerStats : Player_Stats_Data):void=
        if(CanvasData := BuiltCanvasMap[InPlayer]):
            for(Index -> Display : CanvasData.ArtifactDisplays):
                if(EquippedArtifact := InPlayerStats.EquippedPlayerArtifacts[Index]):
                    UpdateSlot(InPlayer, InPlayerStats, Index + 1, EquippedArtifact.Rarity)
                else:
                    UpdateSlot(InPlayer, InPlayerStats, Index + 1, ArtifactRarity.Null)
        
    UpdateSlot(InPlayer : player, InPlayerStats : Player_Stats_Data, InSlot : int, InArtifactRarity : ArtifactRarity):void=
        if(CanvasData := BuiltCanvasMap[InPlayer]):
            for(Index -> Display : CanvasData.ArtifactDisplays):
                if(Index + 1 = InSlot):
                    if(InArtifactRarity = ArtifactRarity.Null):
                        Display.Stats.SetText(ToMessage("NO ARTIFACT EQUIPPED"))
                        Display.Icon.SetImage(GetArtifactImage(InArtifactRarity))
                    else:
                        FoundArtifact := GetArtifactFromPlayer(InPlayerStats, InArtifactRarity)
                        Display.Icon.SetImage(GetArtifactImage(FoundArtifact.Rarity))
                        Display.Name.SetText(ToMessage(GetArtifactRarityString(FoundArtifact.Rarity)))

                        var StatsString : string = ""
                        set StatsString += "Lv. {FoundArtifact.Level} - Pr. {FoundArtifact.Prestige}\n"
                        set StatsString += "\nStrength Bonus: {FormatFloatTo2DP(GetArtifactStat(FoundArtifact, ArtifactStats.ArtifactBonus))}%"
                        set StatsString += "\nCritical Rate: {FormatFloatTo2DP(GetArtifactStat(FoundArtifact, ArtifactStats.CriticalChance))}"
                        set StatsString += "\nCritical Damage: {FormatFloatTo2DP(GetArtifactStat(FoundArtifact, ArtifactStats.CriticalDamage))}%"
                        set StatsString += "\nFinal Strike: {FormatFloatTo2DP(GetArtifactStat(FoundArtifact, ArtifactStats.FinalStrike))}%"
                        
                        Display.Stats.SetText(ToMessage(StatsString))
                    set Display.CurrentArtifactRarity = InArtifactRarity

                    for(Artifact : InPlayerStats.EquippedPlayerArtifacts):
                        if(InArtifactRarity = Artifact.Rarity):
                            Display.EquipButton.SetText(ToMessage("EQUIPPED"))
                            DebugPrintArtifactStats(Artifact)
                            return
                    Display.EquipButton.SetText(ToMessage("EQUIP"))

    
    ToggleArtifactDisplayLocked(InArtifactEquipDisplay : ArtifactEquipDisplay, Lock : logic):void=
        if(Lock = false):
            InArtifactEquipDisplay.EquipButton.SetVisibility(widget_visibility.Visible)
            InArtifactEquipDisplay.NextButton.SetVisibility(widget_visibility.Visible)
            InArtifactEquipDisplay.PrevButton.SetVisibility(widget_visibility.Visible)
            InArtifactEquipDisplay.Icon.SetVisibility(widget_visibility.Visible)
            InArtifactEquipDisplay.Stats.SetVisibility(widget_visibility.Visible)
            InArtifactEquipDisplay.Name.SetVisibility(widget_visibility.Visible)

            InArtifactEquipDisplay.Locked.SetVisibility(widget_visibility.Collapsed)
            set InArtifactEquipDisplay.IsLocked = false
        else:
            InArtifactEquipDisplay.EquipButton.SetVisibility(widget_visibility.Collapsed)
            InArtifactEquipDisplay.NextButton.SetVisibility(widget_visibility.Collapsed)
            InArtifactEquipDisplay.PrevButton.SetVisibility(widget_visibility.Collapsed)
            InArtifactEquipDisplay.Icon.SetVisibility(widget_visibility.Hidden)
            InArtifactEquipDisplay.Stats.SetVisibility(widget_visibility.Collapsed)
            InArtifactEquipDisplay.Name.SetVisibility(widget_visibility.Collapsed)

            InArtifactEquipDisplay.Locked.SetVisibility(widget_visibility.Visible)
            set InArtifactEquipDisplay.IsLocked = true

    MakeArtifactDisplay():ArtifactEquipDisplay=
        EquipButton := button_quiet:
            DefaultText := ToMessage("EQUIP")
        NextButton := button_quiet:
            DefaultText := ToMessage(">")
        PrevButton := button_quiet:
            DefaultText := ToMessage("<")
        IconTexture := texture_block:
            DefaultImage := GetArtifactImage(ArtifactRarity.Null)

        LockedText := text_block:
            DefaultText := ToMessage("LOCKED")
            DefaultJustification := text_justification.Center
            DefaultShadowColor := NamedColors.Black
            DefaultShadowOffset := option{vector2{X:=2.0,Y:=2.0}}
        LockedText.SetTextColor(NamedColors.White)
        LockedText.SetShadowOpacity(0.8)
        LockedText.SetVisibility(widget_visibility.Collapsed)

        NameText := text_block:
            DefaultText := ToMessage("NO ARTIFACT")
            DefaultJustification := text_justification.Center
            DefaultShadowColor := NamedColors.Black
            DefaultShadowOffset := option{vector2{X:=2.0,Y:=2.0}}
        NameText.SetTextColor(NamedColors.Orange)
        NameText.SetShadowOpacity(0.8)
        StatText := text_block:
            DefaultText := ToMessage("NO ARTIFACT EQUIPPED")
            DefaultJustification := text_justification.Center
            DefaultShadowColor := NamedColors.Black
            DefaultShadowOffset := option{vector2{X:=2.0,Y:=2.0}}
        StatText.SetTextColor(NamedColors.White)
        StatText.SetShadowOpacity(0.8)

        IconOverlay := overlay:
            Slots := array:
                overlay_slot:
                    VerticalAlignment := vertical_alignment.Center
                    HorizontalAlignment := horizontal_alignment.Center
                    Padding := margin{Top := -8.0, Bottom := -8.0, Left := -6.0, Right := -6.0}
                    Widget := color_block:
                        DefaultColor := NamedColors.White
                        DefaultOpacity := 0.0
                        DefaultDesiredSize := vector2{X := 256.0, Y := 256.0}
                overlay_slot:
                    VerticalAlignment := vertical_alignment.Fill
                    HorizontalAlignment := horizontal_alignment.Fill
                    Widget := IconTexture

        ButtonStackBox := stack_box:
            Orientation := orientation.Horizontal
            Slots := 
                array:
                    stack_box_slot:
                        Widget := PrevButton
                        HorizontalAlignment := horizontal_alignment.Left
                        VerticalAlignment := vertical_alignment.Center
                    stack_box_slot:
                        Widget := EquipButton
                        Padding := margin{Left := 10.0, Right := 10.0}
                        HorizontalAlignment := horizontal_alignment.Center
                        VerticalAlignment := vertical_alignment.Center
                    stack_box_slot:
                        Widget := NextButton
                        HorizontalAlignment := horizontal_alignment.Right
                        VerticalAlignment := vertical_alignment.Center

        StackBox := stack_box:
            Orientation := orientation.Vertical
            Slots := 
                array:
                    stack_box_slot:
                        Widget := IconOverlay
                        HorizontalAlignment := horizontal_alignment.Center
                        VerticalAlignment := vertical_alignment.Center
                    stack_box_slot:
                        Widget := LockedText
                        HorizontalAlignment := horizontal_alignment.Center
                        VerticalAlignment := vertical_alignment.Center
                    stack_box_slot:
                        Widget := NameText
                        HorizontalAlignment := horizontal_alignment.Center
                        VerticalAlignment := vertical_alignment.Center
                    stack_box_slot:
                        Widget := StatText
                        Padding := margin{Top := 10.0}
                        HorizontalAlignment := horizontal_alignment.Center
                        VerticalAlignment := vertical_alignment.Center
                    stack_box_slot:
                        Widget := ButtonStackBox
                        Padding := margin{Top := 10.0}
                        HorizontalAlignment := horizontal_alignment.Center
                        VerticalAlignment := vertical_alignment.Center

        OutOverlay := overlay:
            Slots := array:
                overlay_slot:
                    VerticalAlignment := vertical_alignment.Center
                    HorizontalAlignment := horizontal_alignment.Center
                    Widget := color_block:
                        DefaultColor := NamedColors.White
                        DefaultOpacity := 0.4
                        DefaultDesiredSize := vector2{X:=400.0,Y:=850.0}
                overlay_slot:
                    VerticalAlignment := vertical_alignment.Center
                    HorizontalAlignment := horizontal_alignment.Center
                    Widget := StackBox

        return ArtifactEquipDisplay:
            IsLocked := false
            Widget := OutOverlay
            EquipButton := EquipButton
            NextButton := NextButton
            PrevButton := PrevButton
            Icon := IconTexture
            Name := NameText
            Stats := StatText
            Locked := LockedText
            CurrentArtifactRarity := ArtifactRarity.Null

    MakeArtifactEquipCanvas():ArtifactEquipData=
        ExitButton := button_loud:
            DefaultText := ToMessage("Exit")

        ArtifactDisplay1 := MakeArtifactDisplay()
        ArtifactDisplay2 := MakeArtifactDisplay()
        ArtifactDisplay3 := MakeArtifactDisplay()
        ArtifactDisplay4 := MakeArtifactDisplay()

        DisplayStackBox := stack_box:
            Orientation := orientation.Horizontal
            Slots := 
                array:
                    stack_box_slot:
                        Widget := ArtifactDisplay1.Widget
                        HorizontalAlignment := horizontal_alignment.Center
                        VerticalAlignment := vertical_alignment.Center
                    stack_box_slot:
                        Widget := ArtifactDisplay2.Widget
                        Padding := margin{Left := 10.0}
                        HorizontalAlignment := horizontal_alignment.Center
                        VerticalAlignment := vertical_alignment.Center
                    stack_box_slot:
                        Widget := ArtifactDisplay3.Widget
                        Padding := margin{Left := 10.0}
                        HorizontalAlignment := horizontal_alignment.Center
                        VerticalAlignment := vertical_alignment.Center
                    stack_box_slot:
                        Widget := ArtifactDisplay4.Widget
                        Padding := margin{Left := 10.0}
                        HorizontalAlignment := horizontal_alignment.Center
                        VerticalAlignment := vertical_alignment.Center

        BackgroundOverlay := overlay:
            Slots := array:
                overlay_slot:
                    VerticalAlignment := vertical_alignment.Fill
                    HorizontalAlignment := horizontal_alignment.Fill
                    Widget := texture_block:
                        DefaultImage := VerseExposed.T_UIBackground

        OutCanvas := canvas:
            Slots := array:
                canvas_slot:
                    Anchors := anchors{Minimum := vector2{X:=0.0, Y:=0.0}, Maximum := vector2{X:=1.0, Y:=1.0}}
                    Alignment := CenteredAlignment
                    Widget := BackgroundOverlay
                    ZOrder := 0
                canvas_slot:
                    Anchors := anchors{Minimum := vector2{X:=0.5, Y:=0.5}, Maximum := vector2{X:=0.5, Y:=0.5}}
                    Alignment := CenteredAlignment
                    Widget := DisplayStackBox
                    ZOrder := 1
                canvas_slot:
                    Anchors := anchors{Minimum := vector2{X:=0.05, Y:=0.95}, Maximum := vector2{X:=0.95, Y:=0.95}}
                    Alignment := CenteredAlignment
                    Widget := ExitButton
                    ZOrder := 1

        return ArtifactEquipData:
            Canvas := OutCanvas
            ExitButton := ExitButton
            ArtifactDisplays := array{ArtifactDisplay1, ArtifactDisplay2, ArtifactDisplay3, ArtifactDisplay4}

#>