# fill this file with your verse-script
using { /Fortnite.com/Devices }
using { /Verse.org/Simulation }
using { /UnrealEngine.com/Temporary/Diagnostics }
using { /UnrealEngine.com/Temporary/UI }
using { /Fortnite.com/UI }
using { /UnrealEngine.com/Temporary/SpatialMath}
using { /Verse.org/Colors }
using { /Verse.org/Assets }

GetArtifactImage(InArtifactRarity : ArtifactRarity):texture=
    case(InArtifactRarity):
        ArtifactRarity.MushroomCommon =>
            return VerseExposed.T_Artifact_B1_Common
        ArtifactRarity.MushroomUncommon =>
            return VerseExposed.T_Artifact_B1_Uncommon
        ArtifactRarity.MushroomRare =>
            return VerseExposed.T_Artifact_B1_Rare
        ArtifactRarity.MushroomEpic =>
            return VerseExposed.T_Artifact_B1_Epic
        ArtifactRarity.MushroomLegendary =>
            return VerseExposed.T_Artifact_B1_Legendary
        ArtifactRarity.LavaCommon =>
            return VerseExposed.T_Artifact_B2_Common
        ArtifactRarity.LavaUncommon =>
            return VerseExposed.T_Artifact_B2_Uncommon
        ArtifactRarity.LavaRare =>
            return VerseExposed.T_Artifact_B2_Rare
        ArtifactRarity.LavaEpic =>
            return VerseExposed.T_Artifact_B2_Epic
        ArtifactRarity.LavaLegendary =>
            return VerseExposed.T_Artifact_B2_Legendary
        ArtifactRarity.UnderworldCommon =>
            return VerseExposed.T_Artifact_B3_Common
        ArtifactRarity.UnderworldUncommon =>
            return VerseExposed.T_Artifact_B3_Uncommon
        ArtifactRarity.UnderworldRare =>
            return VerseExposed.T_Artifact_B3_Rare
        ArtifactRarity.UnderworldEpic =>
            return VerseExposed.T_Artifact_B3_Epic
        ArtifactRarity.UnderworldLegendary =>
            return VerseExposed.T_Artifact_B3_Legendary
        _ =>
            return VerseExposed.T_UnknownArtifact

GetArtifactRarityString(InArtifactRarity : ArtifactRarity):string=
    case(InArtifactRarity):
        ArtifactRarity.MushroomCommon =>
            return "Mushroom Common"
        ArtifactRarity.MushroomUncommon =>
            return "Mushroom Uncommon"
        ArtifactRarity.MushroomRare =>
            return "Mushroom Rare"
        ArtifactRarity.MushroomEpic =>
            return "Mushroom Epic"
        ArtifactRarity.MushroomLegendary =>
            return "Mushroom Legendary"
        ArtifactRarity.LavaCommon =>
            return "Lava Common"
        ArtifactRarity.LavaUncommon =>
            return "Lava Uncommon"
        ArtifactRarity.LavaRare =>
            return "Lava Rare"
        ArtifactRarity.LavaEpic =>
            return "Lava Epic"
        ArtifactRarity.LavaLegendary =>
            return "Lava Legendary"
        ArtifactRarity.UnderworldCommon =>
            return "Underworld Common"
        ArtifactRarity.UnderworldUncommon =>
            return "Underworld Uncommon"
        ArtifactRarity.UnderworldRare =>
            return "Underworld Rare"
        ArtifactRarity.UnderworldEpic =>
            return "Underworld Epic"
        ArtifactRarity.UnderworldLegendary =>
            return "Underworld Legendary"
        _ =>
            return "Null"

TriggerHandler := class() {
    Trigger : trigger_device
    PlayerRef : []player

    HandlerFunction(WidgetMessage : widget_message) : void = {
            Player := WidgetMessage.Player
            if(PlayerRef = array{Player}):
                Trigger.Trigger(Player)
    }
}

ChangeArtifactHandler := class() {
    Device : V_ArtifactManager
    Rarity : ArtifactRarity

    HandlerFunction(WidgetMessage : widget_message) : void = {
            Player := WidgetMessage.Player
            Device.ChangedDisplayedArtifact(Player, Rarity)
    }
}
ChangeResourceHandler := class(){
    Next : logic
    Device : V_ArtifactManager

    HandlerFunction(WidgetMessage : widget_message) : void = {
            Player := WidgetMessage.Player
            Device.ChangeDisplayedResource(Player, Next)
    }
}
SmeltHandler := class(){
    Device : V_ArtifactManager

    HandlerFunction(WidgetMessage : widget_message) : void = {
            Player := WidgetMessage.Player
            Device.SmeltButtonPressed(Player)
    }
}
PrestigeHandler := class(){
    Device : V_ArtifactManager

    HandlerFunction(WidgetMessage : widget_message) : void = {
            Player := WidgetMessage.Player
            Device.PrestigeButtonPressed(Player)
    }
}

ArtifactMenuCanvasData := class:
    Canvas : canvas

    var CurrentResourceType : PropResourceType
    var CurrentArtifactRarity : ArtifactRarity

    StoneCostText : text_block
    ResourceCostText : text_block
    ResourceCostImage : texture_block
    TiersText : text_block

    PlayerStoneText : text_block
    PlayerCopperOreText : text_block
    PlayerSilverOreText : text_block
    PlayerMalachiteOreText : text_block
    PlayerSpectroliteOreText : text_block
    PlayerObsidianOreText : text_block

    ArtifactCommon : ArtifactButtonData
    ArtifactUncommon : ArtifactButtonData
    ArtifactRare : ArtifactButtonData
    ArtifactEpic : ArtifactButtonData
    ArtifactLegendary : ArtifactButtonData

    ArtifactStatDatas : []ArtifactStatData

    ArtifactDisplayTexture : texture_block

    PrestigeButtonOverlay : overlay
    PrestigeStoneCostText : text_block
    PrestigeResourceCostText : text_block
    PrestigeResourceCostImage : texture_block

    CurrentResourceDisplay : texture_block
    NextResourceButton : text_button_base
    PrevResourceButton : text_button_base
    SmeltButton : text_button_base
    PrestigeButton : text_button_base

    MaxedWidget : widget

    ExitButton  : text_button_base

ArtifactInventoryData := class:
    Widget : widget
    Text : text_block

ArtifactButtonData := class:
    Widget : widget
    ArtifactImage : texture_block
    ArtifactTitleText : text_block
    ArtifactLevelText : text_block
    ArtifactRarityText : text_block
    Button : text_button_base

ArtifactStatData := class:
    Widget : widget
    Value : text_block

UI_ArtifactSmelterMenu := class():
    ToMessage<localizes>(String : string) : message = "{String}"
    CenteredAlignment : vector2 = vector2{X := 0.5, Y := 0.5}

    var BuiltCanvasMap : [player]ArtifactMenuCanvasData = map{}

    var ArtifactsInThisUI : []ArtifactRarity = array{}

    InitUI(InPlayer : player, Device : V_ArtifactManager,
    ArtifactRarity1 : ArtifactRarity, ArtifactRarity2 : ArtifactRarity, ArtifactRarity3 : ArtifactRarity, ArtifactRarity4 : ArtifactRarity, ArtifactRarity5 : ArtifactRarity):void=

        ArtifactMenuCanvas := MakeArtifactMenuCanvas()
        # Set the players build canvas data
        set ArtifactsInThisUI = array{ArtifactRarity1, ArtifactRarity2, ArtifactRarity3, ArtifactRarity4, ArtifactRarity5}
        if(set BuiltCanvasMap[InPlayer] = ArtifactMenuCanvas){}
        ArtifactMenuCanvas.ExitButton.OnClick().Subscribe(HandlePlayerExit)
        for(StatData : ArtifactMenuCanvas.ArtifactStatDatas):
            StatData.Widget.SetVisibility(widget_visibility.Hidden)
        
        ArtifactMenuCanvas.SmeltButton.OnClick().Subscribe(SmeltHandler{Device := Device}.HandlerFunction)
        ArtifactMenuCanvas.PrestigeButton.OnClick().Subscribe(PrestigeHandler{Device := Device}.HandlerFunction)
        ArtifactMenuCanvas.NextResourceButton.OnClick().Subscribe(ChangeResourceHandler{Device := Device, Next := true}.HandlerFunction)
        ArtifactMenuCanvas.PrevResourceButton.OnClick().Subscribe(ChangeResourceHandler{Device := Device, Next := false}.HandlerFunction)
        ArtifactMenuCanvas.ArtifactCommon.Button.OnClick().Subscribe(ChangeArtifactHandler{Device := Device, Rarity := ArtifactRarity1}.HandlerFunction)
        ArtifactMenuCanvas.ArtifactUncommon.Button.OnClick().Subscribe(ChangeArtifactHandler{Device := Device, Rarity := ArtifactRarity2}.HandlerFunction)
        ArtifactMenuCanvas.ArtifactRare.Button.OnClick().Subscribe(ChangeArtifactHandler{Device := Device, Rarity := ArtifactRarity3}.HandlerFunction)
        ArtifactMenuCanvas.ArtifactEpic.Button.OnClick().Subscribe(ChangeArtifactHandler{Device := Device, Rarity := ArtifactRarity4}.HandlerFunction)
        ArtifactMenuCanvas.ArtifactLegendary.Button.OnClick().Subscribe(ChangeArtifactHandler{Device := Device, Rarity := ArtifactRarity5}.HandlerFunction)

    # Add the UI to the player
    AddUI(InPlayer : player):void=
        if(PlayerUI := GetPlayerUI[InPlayer], CanvasData := BuiltCanvasMap[InPlayer]):
            CanvasData.Canvas.SetVisibility(widget_visibility.Visible)
            PlayerUI.AddWidget(CanvasData.Canvas, player_ui_slot{InputMode := ui_input_mode.All} )

    # When player presses the exit button
    HandlePlayerExit(WidgetMessage: widget_message):void=
        spawn{HideUI(WidgetMessage.Player)}

    HideUI(Player: player)<suspends>:void=
        Sleep(0.0) # So we're sure that the canvas has been added to BuiltCanvas map
        if(PlayerUI := GetPlayerUI[Player], CanvasData := BuiltCanvasMap[Player]):
            CanvasData.Canvas.SetVisibility(widget_visibility.Hidden)
            PlayerUI.RemoveWidget(CanvasData.Canvas)

    UpdateUI(InPlayer : player, InPlayerData : Player_Stats_Data, InPlayerStatManager : player_stats_manager, OwnedPlayerArtifacts : []Artifact_Data, LastSeenArtifactRarity : ArtifactRarity, InSelectedResource : PropResourceType, InBiome : Biome):void=
        if(CanvasData := BuiltCanvasMap[InPlayer]):
            Print("Update UI")
            CanvasData.PlayerStoneText.SetText(ToMessage(BigNumber_FormatDisplay(InPlayerData.PlayerStoneOreAmount)))
            CanvasData.PlayerCopperOreText.SetText(ToMessage(BigNumber_FormatDisplay(InPlayerData.PlayerCopperOreAmount)))
            CanvasData.PlayerSilverOreText.SetText(ToMessage(BigNumber_FormatDisplay(InPlayerData.PlayerSilverOreAmount)))
            CanvasData.PlayerMalachiteOreText.SetText(ToMessage(BigNumber_FormatDisplay(InPlayerData.PlayerMalachiteOreAmount)))
            CanvasData.PlayerSpectroliteOreText.SetText(ToMessage(BigNumber_FormatDisplay(InPlayerData.PlayerSpectroliteOreAmount)))
            CanvasData.PlayerObsidianOreText.SetText(ToMessage(BigNumber_FormatDisplay(InPlayerData.PlayerObsidianOreAmount)))

            TiersAndCostData := GetUIArtifactSmeltData(InSelectedResource, InBiome)

            CanvasData.TiersText.SetText(ToMessage(TiersAndCostData.RarityString))
            CanvasData.StoneCostText.SetText(ToMessage(BigNumber_FormatDisplay(TiersAndCostData.StoneCost)))
            CanvasData.ResourceCostText.SetText(ToMessage(BigNumber_FormatDisplay(TiersAndCostData.ResourceCost)))
            CanvasData.ResourceCostImage.SetImage(TiersAndCostData.ResourceImage)
            if(LastSeenArtifactRarity <> ArtifactRarity.Null):
                CanvasData.ArtifactDisplayTexture.SetImage(GetArtifactImage(LastSeenArtifactRarity))
            CanvasData.CurrentResourceDisplay.SetImage(TiersAndCostData.ResourceImage)

            set CanvasData.CurrentArtifactRarity = LastSeenArtifactRarity

            PrestigeCostData := GetArtifactPrestigeData(GetResourceFromRarity(CanvasData.CurrentArtifactRarity), InBiome)

            Artifact := GetArtifactFromPlayer(InPlayerData, LastSeenArtifactRarity)
            CanvasData.PrestigeStoneCostText.SetText(ToMessage(BigNumber_FormatDisplay(BigNumber_MultiplyWithAFloat(PrestigeCostData.StoneCost, (Artifact.Prestige + 1) * 100.0))))
            CanvasData.PrestigeResourceCostText.SetText(ToMessage(BigNumber_FormatDisplay(BigNumber_MultiplyWithAFloat(PrestigeCostData.ResourceCost, (Artifact.Prestige + 1) * 5.0))))
            CanvasData.PrestigeResourceCostImage.SetImage(PrestigeCostData.ResourceImage)

            UpdateCanAfford(InPlayer, InPlayerStatManager, InBiome)

            for(CurrentArtifact : OwnedPlayerArtifacts):
                for(ArtifactInUI : ArtifactsInThisUI):
                    if(CurrentArtifact.Rarity = ArtifactInUI):
                        UpdateArtifactButtons(InPlayer, CurrentArtifact)
                        if(CurrentArtifact.Rarity = LastSeenArtifactRarity):
                            UpdateArtifactStat(InPlayer, CurrentArtifact)
                            if(CurrentArtifact.Level = 30 and CurrentArtifact.Prestige < 3):
                                CanvasData.PrestigeButtonOverlay.SetVisibility(widget_visibility.Visible)
                            else:
                                CanvasData.PrestigeButtonOverlay.SetVisibility(widget_visibility.Hidden)
                                CanvasData.MaxedWidget.SetVisibility(widget_visibility.Hidden)
                                if(CurrentArtifact.Level = 30 and CurrentArtifact.Prestige = 3):
                                    Print("MAXED")
                                    CanvasData.MaxedWidget.SetVisibility(widget_visibility.Visible)
                                else:
                                    CanvasData.MaxedWidget.SetVisibility(widget_visibility.Hidden)


    UpdateUINoMainIcon(InPlayer : player, InPlayerStatManager : player_stats_manager, InPlayerData : Player_Stats_Data, OwnedPlayerArtifacts : []Artifact_Data, InSelectedResource : PropResourceType, InBiome : Biome):void=
        if(CanvasData := BuiltCanvasMap[InPlayer]):
            CanvasData.PlayerStoneText.SetText(ToMessage(BigNumber_FormatDisplay(InPlayerData.PlayerStoneOreAmount)))
            CanvasData.PlayerCopperOreText.SetText(ToMessage(BigNumber_FormatDisplay(InPlayerData.PlayerCopperOreAmount)))
            CanvasData.PlayerSilverOreText.SetText(ToMessage(BigNumber_FormatDisplay(InPlayerData.PlayerSilverOreAmount)))
            CanvasData.PlayerMalachiteOreText.SetText(ToMessage(BigNumber_FormatDisplay(InPlayerData.PlayerMalachiteOreAmount)))
            CanvasData.PlayerSpectroliteOreText.SetText(ToMessage(BigNumber_FormatDisplay(InPlayerData.PlayerSpectroliteOreAmount)))
            CanvasData.PlayerObsidianOreText.SetText(ToMessage(BigNumber_FormatDisplay(InPlayerData.PlayerObsidianOreAmount)))

            TiersAndCostData := GetUIArtifactSmeltData(InSelectedResource, InBiome)

            CanvasData.TiersText.SetText(ToMessage(TiersAndCostData.RarityString))
            CanvasData.StoneCostText.SetText(ToMessage(BigNumber_FormatDisplay(TiersAndCostData.StoneCost)))
            CanvasData.ResourceCostText.SetText(ToMessage(BigNumber_FormatDisplay(TiersAndCostData.ResourceCost)))
            CanvasData.ResourceCostImage.SetImage(TiersAndCostData.ResourceImage)
            CanvasData.CurrentResourceDisplay.SetImage(TiersAndCostData.ResourceImage)

            UpdateCanAfford(InPlayer, InPlayerStatManager, InBiome)

    UpdateCanAfford(InPlayer : player, InPlayerStatManager : player_stats_manager, InBiome : Biome):void=
        if(CanvasData := BuiltCanvasMap[InPlayer]):
            PrestigeData := GetArtifactPrestigeData(GetResourceFromRarity(CanvasData.CurrentArtifactRarity), InBiome)

            StoneAmount := InPlayerStatManager.GetPlayerStoneOreAmount(InPlayer)
            PrestigeResourceAmount := InPlayerStatManager.GetPlayerOreAmount(InPlayer, PrestigeData.ResourceType)

            PrestigeCost := PrestigeData.StoneCost
            PrestigeResourceCost := PrestigeData.ResourceCost
            PrestigeCanAffordStone := BigNumber_GreaterThanOrEquals(StoneAmount, PrestigeCost)
            PrestigeCanAffordResource := BigNumber_GreaterThanOrEquals(PrestigeResourceAmount, PrestigeResourceCost)
            #Check if we have enough ore for the next upgrade

            if(PrestigeCanAffordStone = true):
                CanvasData.PrestigeStoneCostText.SetTextColor(NamedColors.White)
            else:
                CanvasData.PrestigeStoneCostText.SetTextColor(NamedColors.Red)

            if(PrestigeCanAffordResource = true):
                CanvasData.PrestigeResourceCostText.SetTextColor(NamedColors.White)
            else:
                CanvasData.PrestigeResourceCostText.SetTextColor(NamedColors.Red)

            TiersAndCostData := GetUIArtifactSmeltData(CanvasData.CurrentResourceType, InBiome)

            ResourceAmount := InPlayerStatManager.GetPlayerOreAmount(InPlayer, CanvasData.CurrentResourceType)

            Cost := TiersAndCostData.StoneCost
            ResourceCost := TiersAndCostData.ResourceCost
            CanAffordStone := BigNumber_GreaterThanOrEquals(StoneAmount, Cost)
            CanAffordResource := BigNumber_GreaterThanOrEquals(ResourceAmount, ResourceCost)
            #Check if we have enough ore for the next upgrade

            if(CanAffordStone = true):
                CanvasData.StoneCostText.SetTextColor(NamedColors.White)
            else:
                CanvasData.StoneCostText.SetTextColor(NamedColors.Red)

            if(CanAffordResource = true):
                CanvasData.ResourceCostText.SetTextColor(NamedColors.White)
            else:
                CanvasData.ResourceCostText.SetTextColor(NamedColors.Red)

    UpdateArtifactButtons(InPlayer : player, InArtifact : Artifact_Data):void=
        if(CanvasData := BuiltCanvasMap[InPlayer]):
            var ButtonToUpdate : ArtifactButtonData = CanvasData.ArtifactCommon
            case(InArtifact.Rarity):
                ArtifactRarity.MushroomCommon =>
                    set ButtonToUpdate = CanvasData.ArtifactCommon
                ArtifactRarity.MushroomUncommon =>
                    set ButtonToUpdate = CanvasData.ArtifactUncommon
                ArtifactRarity.MushroomRare =>
                    set ButtonToUpdate = CanvasData.ArtifactRare
                ArtifactRarity.MushroomEpic =>
                    set ButtonToUpdate = CanvasData.ArtifactEpic
                ArtifactRarity.MushroomLegendary =>
                    set ButtonToUpdate = CanvasData.ArtifactLegendary
                ArtifactRarity.LavaCommon =>
                    set ButtonToUpdate = CanvasData.ArtifactCommon
                ArtifactRarity.LavaUncommon =>
                    set ButtonToUpdate = CanvasData.ArtifactUncommon
                ArtifactRarity.LavaRare =>
                    set ButtonToUpdate = CanvasData.ArtifactRare
                ArtifactRarity.LavaEpic =>
                    set ButtonToUpdate = CanvasData.ArtifactEpic
                ArtifactRarity.LavaLegendary =>
                    set ButtonToUpdate = CanvasData.ArtifactLegendary
                ArtifactRarity.UnderworldCommon =>
                    set ButtonToUpdate = CanvasData.ArtifactCommon
                ArtifactRarity.UnderworldUncommon =>
                    set ButtonToUpdate = CanvasData.ArtifactUncommon
                ArtifactRarity.UnderworldRare =>
                    set ButtonToUpdate = CanvasData.ArtifactRare
                ArtifactRarity.UnderworldEpic =>
                    set ButtonToUpdate = CanvasData.ArtifactEpic
                ArtifactRarity.UnderworldLegendary =>
                    set ButtonToUpdate = CanvasData.ArtifactLegendary
                _ =>
                    Print("NO ARTIFACT TO UPDATE")
                    return

            ButtonToUpdate.ArtifactLevelText.SetText(ToMessage("Lv. {InArtifact.Level} - Pr. {InArtifact.Prestige}"))
            ButtonToUpdate.ArtifactRarityText.SetText(ToMessage(GetArtifactRarityString(InArtifact.Rarity)))
            ButtonToUpdate.ArtifactImage.SetImage(GetArtifactImage(InArtifact.Rarity))

            ButtonToUpdate.ArtifactLevelText.SetVisibility(widget_visibility.Visible)
            ButtonToUpdate.ArtifactRarityText.SetVisibility(widget_visibility.Visible)

    UpdateArtifactStat(InPlayer : player, InArtifact : Artifact_Data):void=
        if(CanvasData := BuiltCanvasMap[InPlayer]):
            for(Index -> StatData : CanvasData.ArtifactStatDatas):
                StatData.Widget.SetVisibility(widget_visibility.Hidden)

            for(Index -> ArtifactStat : InArtifact.StatValues):
                if(StatData := CanvasData.ArtifactStatDatas[Index]):
                    StatData.Value.SetText(ToMessage("+{FormatFloatTo2DP(ArtifactStat)}%"))
                    StatData.Widget.SetVisibility(widget_visibility.Visible)


    MakeStatDisplay(InTitle : string, InValue : string):ArtifactStatData=
        Title := text_block:
            DefaultText := ToMessage("{InTitle}:")
            DefaultJustification := text_justification.Left
            DefaultShadowColor := NamedColors.Black
            DefaultShadowOffset := option{vector2{X:=2.0,Y:=2.0}}
        Title.SetTextColor(NamedColors.White)
        Title.SetShadowOpacity(0.8)
        Value := text_block:
            DefaultText := ToMessage("+{InValue}%")
            DefaultJustification := text_justification.Right
            DefaultShadowColor := NamedColors.Black
            DefaultShadowOffset := option{vector2{X:=2.0,Y:=2.0}}
        Value.SetTextColor(NamedColors.LightGreen)
        Value.SetShadowOpacity(0.8)

        StackBox := stack_box:
            Orientation := orientation.Horizontal
            Slots := 
                array:
                    stack_box_slot:
                        Widget := Title
                        HorizontalAlignment := horizontal_alignment.Left
                        VerticalAlignment := vertical_alignment.Center
                    stack_box_slot:
                        Widget := Value
                        Padding := margin{Left := 50.0}
                        HorizontalAlignment := horizontal_alignment.Right
                        VerticalAlignment := vertical_alignment.Center

        return ArtifactStatData:
            Widget := StackBox
            Value := Value


    MakeInventoryDisplay(InResourceImage : texture):ArtifactInventoryData=
        Text := text_block:
            DefaultText := ToMessage("")
            DefaultJustification := text_justification.Center
            DefaultShadowColor := NamedColors.Black
            DefaultShadowOffset := option{vector2{X:=2.0,Y:=2.0}}
        Text.SetTextColor(NamedColors.White)
        Text.SetShadowOpacity(0.8)

        IconOverlay := overlay:
            Slots := array:
                overlay_slot:
                    VerticalAlignment := vertical_alignment.Center
                    HorizontalAlignment := horizontal_alignment.Center
                    Padding := margin{Top := -8.0, Bottom := -8.0, Left := -6.0, Right := -6.0}
                    Widget := color_block:
                        DefaultColor := NamedColors.White
                        DefaultOpacity := 0.2
                        DefaultDesiredSize := vector2{X := 52.0, Y := 52.0}
                overlay_slot:
                    VerticalAlignment := vertical_alignment.Center
                    HorizontalAlignment := horizontal_alignment.Center
                    Widget := texture_block:
                        DefaultImage := InResourceImage
        
        Overlay := overlay:
            Slots := array:
                overlay_slot:
                    VerticalAlignment := vertical_alignment.Center
                    HorizontalAlignment := horizontal_alignment.Left
                    Widget := color_block:
                        DefaultColor := NamedColors.White
                        DefaultOpacity := 0.2
                        DefaultDesiredSize := vector2{X:=225.0,Y:=60.0}
                overlay_slot:
                    VerticalAlignment := vertical_alignment.Center
                    HorizontalAlignment := horizontal_alignment.Left
                    Padding := margin{Top := 10.0, Bottom := 10.0, Left := 10.0, Right := 10.0}
                    Widget := stack_box:
                        Orientation := orientation.Horizontal
                        Slots := 
                            array:
                                stack_box_slot:
                                    Widget := IconOverlay
                                    HorizontalAlignment := horizontal_alignment.Left
                                    VerticalAlignment := vertical_alignment.Center
                                stack_box_slot:
                                    Widget := Text
                                    Padding := margin{Left := 15.0}
                                    VerticalAlignment := vertical_alignment.Center
                                    HorizontalAlignment := horizontal_alignment.Left

        return ArtifactInventoryData:
            Widget := Overlay
            Text := Text

    MakeArtifactButtonDisplay(InTitle : string, InLevel : string, InRarity : string, InArtifactImage : texture):ArtifactButtonData=
        ArtifactButton := button_quiet:
            DefaultText := ToMessage("")

        ArtifactImage := texture_block:
            DefaultImage := InArtifactImage
            DefaultDesiredSize := vector2{X:= 128.0, Y:=128.0}
        
        ArtifactTitleText := text_block:
            DefaultText := ToMessage("{InTitle}")
            DefaultJustification := text_justification.Center
            DefaultShadowColor := NamedColors.Black
            DefaultShadowOffset := option{vector2{X:=2.0,Y:=2.0}}
        ArtifactTitleText.SetTextColor(NamedColors.White)
        ArtifactTitleText.SetShadowOpacity(0.8)
        ArtifactLevelText := text_block:
            DefaultText := ToMessage("{InLevel}")
            DefaultJustification := text_justification.Center
            DefaultShadowColor := NamedColors.Black
            DefaultShadowOffset := option{vector2{X:=2.0,Y:=2.0}}
        ArtifactLevelText.SetTextColor(NamedColors.White)
        ArtifactLevelText.SetShadowOpacity(0.8)
        ArtifactRarityText := text_block:
            DefaultText := ToMessage("{InRarity}")
            DefaultJustification := text_justification.Center
            DefaultShadowColor := NamedColors.Black
            DefaultShadowOffset := option{vector2{X:=2.0,Y:=2.0}}
        ArtifactRarityText.SetTextColor(NamedColors.Orange)
        ArtifactRarityText.SetShadowOpacity(0.8)

        TextStackBox := stack_box:
            Orientation := orientation.Vertical
            Slots := 
                array:
                    stack_box_slot:
                        Widget := ArtifactTitleText
                        Padding := margin{Left := 150.0, Bottom := 2.0}
                        HorizontalAlignment := horizontal_alignment.Left
                        VerticalAlignment := vertical_alignment.Center
                    stack_box_slot:
                        Widget := ArtifactLevelText
                        Padding := margin{Left := 150.0, Bottom := 2.0}
                        HorizontalAlignment := horizontal_alignment.Left
                        VerticalAlignment := vertical_alignment.Center
                    stack_box_slot:
                        Widget := ArtifactRarityText
                        Padding := margin{Left := 150.0, Bottom := 2.0}
                        HorizontalAlignment := horizontal_alignment.Left
                        VerticalAlignment := vertical_alignment.Center

        ArtifactOverlay := overlay:
            Slots := array:
                overlay_slot:
                    VerticalAlignment := vertical_alignment.Center
                    HorizontalAlignment := horizontal_alignment.Center
                    Widget := color_block:
                        DefaultDesiredSize := vector2{X:=650.0, Y:=100.0}
                        DefaultOpacity := 0.0
                overlay_slot:
                    VerticalAlignment := vertical_alignment.Center
                    HorizontalAlignment := horizontal_alignment.Left
                    Padding:= margin{Left := 16.0}
                    Widget := ArtifactImage
                overlay_slot:
                    VerticalAlignment := vertical_alignment.Fill
                    HorizontalAlignment := horizontal_alignment.Left
                    Widget := TextStackBox
                overlay_slot:
                    VerticalAlignment := vertical_alignment.Fill
                    HorizontalAlignment := horizontal_alignment.Fill
                    Widget := ArtifactButton

        return ArtifactButtonData:
            Widget := ArtifactOverlay
            ArtifactImage := ArtifactImage
            ArtifactTitleText := ArtifactTitleText
            ArtifactLevelText := ArtifactLevelText
            ArtifactRarityText := ArtifactRarityText
            Button := ArtifactButton

    MakeArtifactMenuCanvas():ArtifactMenuCanvasData=
        Print("Make Artifact")
        ExitButton := button_loud:
            DefaultText := ToMessage("Exit")
        SmeltButton := button_quiet:
            DefaultText := ToMessage("Smelt")
        PrestigeButton := button_quiet:
            DefaultText := ToMessage("")
        
        ArtifactCommonButton := button_quiet:
            DefaultText := ToMessage("")
        ArtifactUncommonButton := button_quiet:
            DefaultText := ToMessage("")
        ArtifactRareButton := button_quiet:
            DefaultText := ToMessage("")
        ArtifactEpicButton := button_quiet:
            DefaultText := ToMessage("")
        ArtifactLegendaryButton := button_quiet:
            DefaultText := ToMessage("")

        NextResourceButton := button_quiet:
            DefaultText := ToMessage(">")
        PrevResourceButton := button_quiet:
            DefaultText := ToMessage("<")

        StoneInventory := MakeInventoryDisplay(VerseExposed.T_Stone)
        CopperInventory := MakeInventoryDisplay(VerseExposed.T_CopperOreStroke)
        SilverInventory := MakeInventoryDisplay(VerseExposed.T_SilverOre)
        MalachiteInventory := MakeInventoryDisplay(VerseExposed.T_MalachiteOreStroke)
        SpectroliteInventory := MakeInventoryDisplay(VerseExposed.T_SpectroliteOreStroke)
        ObsidianInventory := MakeInventoryDisplay(VerseExposed.T_ObsidianOreStroke)

        ArtifactCommon := MakeArtifactButtonDisplay("Common Artifact", "Lv. 1", "Common", VerseExposed.T_UnknownArtifact)
        ArtifactCommon.ArtifactLevelText.SetVisibility(widget_visibility.Hidden)
        ArtifactCommon.ArtifactRarityText.SetVisibility(widget_visibility.Hidden)
        ArtifactUncommon := MakeArtifactButtonDisplay("Uncommon Artifact", "Lv. 1", "Uncommon", VerseExposed.T_UnknownArtifact)
        ArtifactUncommon.ArtifactLevelText.SetVisibility(widget_visibility.Hidden)
        ArtifactUncommon.ArtifactRarityText.SetVisibility(widget_visibility.Hidden)
        ArtifactRare := MakeArtifactButtonDisplay("Rare Artifact", "Lv. 1", "Rare", VerseExposed.T_UnknownArtifact)
        ArtifactRare.ArtifactLevelText.SetVisibility(widget_visibility.Hidden)
        ArtifactRare.ArtifactRarityText.SetVisibility(widget_visibility.Hidden)
        ArtifactEpic := MakeArtifactButtonDisplay("Epic Artifact", "Lv. 1", "Epic", VerseExposed.T_UnknownArtifact)
        ArtifactEpic.ArtifactLevelText.SetVisibility(widget_visibility.Hidden)
        ArtifactEpic.ArtifactRarityText.SetVisibility(widget_visibility.Hidden)
        ArtifactLegendary := MakeArtifactButtonDisplay("Legendary Artifact", "Lv. 1", "Legendary", VerseExposed.T_UnknownArtifact)
        ArtifactLegendary.ArtifactLevelText.SetVisibility(widget_visibility.Hidden)
        ArtifactLegendary.ArtifactRarityText.SetVisibility(widget_visibility.Hidden)
        
        StrengthStat := MakeStatDisplay("+ Strength", "5.00")
        CriticalRateStat := MakeStatDisplay("+ Crit Rate", "5.00")
        CriticalDamageStat := MakeStatDisplay("+ Crit Damage", "5.00")
        FinalStrikeStat := MakeStatDisplay("+ Final Hit", "5.00")

        <#ArtifactTitle := text_block:
            DefaultText := ToMessage("Artifacts")
            DefaultJustification := text_justification.Center
            DefaultShadowColor := NamedColors.Black
            DefaultShadowOffset := option{vector2{X:=2.0,Y:=2.0}}
        ArtifactTitle.SetTextColor(NamedColors.White)
        ArtifactTitle.SetShadowOpacity(0.8)#>
        
        InventoryTitle := text_block:
            DefaultText := ToMessage("Inventory")
            DefaultJustification := text_justification.Center
            DefaultShadowColor := NamedColors.Black
            DefaultShadowOffset := option{vector2{X:=2.0,Y:=2.0}}
        InventoryTitle.SetTextColor(NamedColors.White)
        InventoryTitle.SetShadowOpacity(0.8)

        StatsTitle := text_block:
            DefaultText := ToMessage("Stats")
            DefaultJustification := text_justification.Center
            DefaultShadowColor := NamedColors.Black
            DefaultShadowOffset := option{vector2{X:=2.0,Y:=2.0}}
        StatsTitle.SetTextColor(NamedColors.White)
        StatsTitle.SetShadowOpacity(0.8)

        TiersText := text_block:
            DefaultText := ToMessage("Tiers: ")
            DefaultJustification := text_justification.Center
            DefaultShadowColor := NamedColors.Black
            DefaultShadowOffset := option{vector2{X:=2.0,Y:=2.0}}
        TiersText.SetTextColor(NamedColors.Gray)
        TiersText.SetShadowOpacity(0.8)
        OreTiersText := text_block:
            DefaultText := ToMessage("Common - Rare")
            DefaultJustification := text_justification.Center
            DefaultShadowColor := NamedColors.Black
            DefaultShadowOffset := option{vector2{X:=2.0,Y:=2.0}}
        OreTiersText.SetTextColor(NamedColors.LightGreen)
        OreTiersText.SetShadowOpacity(0.8)


        CostText := text_block:
            DefaultText := ToMessage("Cost: ")
            DefaultJustification := text_justification.Center
            DefaultShadowColor := NamedColors.Black
            DefaultShadowOffset := option{vector2{X:=2.0,Y:=2.0}}
        CostText.SetTextColor(NamedColors.Gray)
        CostText.SetShadowOpacity(0.8)
        StoneCost := text_block:
            DefaultText := ToMessage("100.0M")
            DefaultJustification := text_justification.Center
            DefaultShadowColor := NamedColors.Black
            DefaultShadowOffset := option{vector2{X:=2.0,Y:=2.0}}
        StoneCost.SetTextColor(NamedColors.White)
        StoneCost.SetShadowOpacity(0.8)
        StoneCostStack := stack_box:
            Orientation := orientation.Horizontal
            Slots := 
                array:
                    stack_box_slot:
                        Widget := texture_block:
                            DefaultImage := VerseExposed.T_Stone
                        Padding := margin{Right := 10.0}
                        HorizontalAlignment := horizontal_alignment.Left
                        VerticalAlignment := vertical_alignment.Center
                    stack_box_slot:
                        Widget := StoneCost
                        HorizontalAlignment := horizontal_alignment.Center
                        VerticalAlignment := vertical_alignment.Center
        
        ResourceCostImage := texture_block:
            DefaultImage := VerseExposed.T_CopperOreStroke

        OreCost := text_block:
            DefaultText := ToMessage("1")
            DefaultJustification := text_justification.Center
            DefaultShadowColor := NamedColors.Black
            DefaultShadowOffset := option{vector2{X:=2.0,Y:=2.0}}
        OreCost.SetTextColor(NamedColors.White)
        OreCost.SetShadowOpacity(0.8)
        OreCostStack := stack_box:
            Orientation := orientation.Horizontal
            Slots := 
                array:
                    stack_box_slot:
                        Widget := ResourceCostImage
                        Padding := margin{Right := 10.0}
                        HorizontalAlignment := horizontal_alignment.Left
                        VerticalAlignment := vertical_alignment.Center
                    stack_box_slot:
                        Widget := OreCost
                        HorizontalAlignment := horizontal_alignment.Center
                        VerticalAlignment := vertical_alignment.Center

        CurrentResourceImage := texture_block:
            DefaultImage := VerseExposed.T_CopperOreStroke

        SmeltLeftStackBox := stack_box:
            Orientation := orientation.Vertical
            Slots := 
                array:
                    stack_box_slot:
                        Widget := TiersText
                        Padding := margin{Bottom := 2.5}
                        HorizontalAlignment := horizontal_alignment.Left
                        VerticalAlignment := vertical_alignment.Center
                    stack_box_slot:
                        Widget := CostText
                        Padding := margin{Bottom := 2.5}
                        HorizontalAlignment := horizontal_alignment.Left
                        VerticalAlignment := vertical_alignment.Center

        SmeltRightStackBox := stack_box:
            Orientation := orientation.Vertical
            Slots := 
                array:
                    stack_box_slot:
                        Widget := OreTiersText
                        Padding := margin{Bottom := 2.5}
                        HorizontalAlignment := horizontal_alignment.Left
                        VerticalAlignment := vertical_alignment.Center
                    stack_box_slot:
                        Widget := StoneCostStack
                        Padding := margin{Bottom := 2.5}
                        HorizontalAlignment := horizontal_alignment.Left
                        VerticalAlignment := vertical_alignment.Center
                    stack_box_slot:
                        Widget := OreCostStack
                        Padding := margin{Bottom := 2.5}
                        HorizontalAlignment := horizontal_alignment.Left
                        VerticalAlignment := vertical_alignment.Center

        SmeltTextStackBox := stack_box:
            Orientation := orientation.Horizontal
            Slots := 
                array:
                    stack_box_slot:
                        Widget := SmeltLeftStackBox
                        HorizontalAlignment := horizontal_alignment.Center
                        VerticalAlignment := vertical_alignment.Top
                    stack_box_slot:
                        Widget := SmeltRightStackBox
                        Padding := margin{Left := 100.0}
                        HorizontalAlignment := horizontal_alignment.Center
                        VerticalAlignment := vertical_alignment.Top

        NextResourceOverlay := overlay:
            Slots := array:
                overlay_slot:
                    VerticalAlignment := vertical_alignment.Center
                    HorizontalAlignment := horizontal_alignment.Center
                    Widget := color_block:
                        DefaultDesiredSize := vector2{X:=320.0, Y:=50.0}
                        DefaultOpacity := 0.0
                overlay_slot:
                    Widget := NextResourceButton
                    
                    HorizontalAlignment := horizontal_alignment.Fill
                    VerticalAlignment := vertical_alignment.Fill
            
        PrevResourceOverlay := overlay:
            Slots := array:
                overlay_slot:
                    VerticalAlignment := vertical_alignment.Center
                    HorizontalAlignment := horizontal_alignment.Center
                    Widget := color_block:
                        DefaultDesiredSize := vector2{X:=320.0, Y:=50.0}
                        DefaultOpacity := 0.0
                overlay_slot:
                    Widget := PrevResourceButton
                    HorizontalAlignment := horizontal_alignment.Fill
                    VerticalAlignment := vertical_alignment.Fill

        PrestigeText := text_block:
            DefaultText := ToMessage("PRESTIGE")
            DefaultJustification := text_justification.Center
            DefaultShadowColor := NamedColors.Black
            DefaultShadowOffset := option{vector2{X:=2.0,Y:=2.0}}
        PrestigeText.SetTextColor(NamedColors.White)
        PrestigeText.SetShadowOpacity(0.8)
        PrestigeStoneCost := text_block:
            DefaultText := ToMessage("100.0M")
            DefaultJustification := text_justification.Center
            DefaultShadowColor := NamedColors.Black
            DefaultShadowOffset := option{vector2{X:=2.0,Y:=2.0}}
        PrestigeStoneCost.SetTextColor(NamedColors.White)
        PrestigeStoneCost.SetShadowOpacity(0.8)
        PrestigeStoneCostStack := stack_box:
            Orientation := orientation.Horizontal
            Slots := 
                array:
                    stack_box_slot:
                        Widget := texture_block:
                            DefaultImage := VerseExposed.T_Stone
                        Padding := margin{Right := 10.0}
                        HorizontalAlignment := horizontal_alignment.Left
                        VerticalAlignment := vertical_alignment.Center
                    stack_box_slot:
                        Widget := PrestigeStoneCost
                        HorizontalAlignment := horizontal_alignment.Center
                        VerticalAlignment := vertical_alignment.Center
        
        PrestigeResourceCostImage := texture_block:
            DefaultImage := VerseExposed.T_CopperOreStroke

        PrestigeOreCost := text_block:
            DefaultText := ToMessage("1")
            DefaultJustification := text_justification.Center
            DefaultShadowColor := NamedColors.Black
            DefaultShadowOffset := option{vector2{X:=2.0,Y:=2.0}}
        PrestigeOreCost.SetTextColor(NamedColors.White)
        PrestigeOreCost.SetShadowOpacity(0.8)
        PrestigeOreCostStack := stack_box:
            Orientation := orientation.Horizontal
            Slots := 
                array:
                    stack_box_slot:
                        Widget := PrestigeResourceCostImage
                        Padding := margin{Right := 10.0}
                        HorizontalAlignment := horizontal_alignment.Left
                        VerticalAlignment := vertical_alignment.Center
                    stack_box_slot:
                        Widget := PrestigeOreCost
                        HorizontalAlignment := horizontal_alignment.Center


        PrestigeTextStack := stack_box:
            Orientation := orientation.Vertical
            Slots := 
                array:
                    stack_box_slot:
                        Widget := PrestigeStoneCostStack
                        Padding := margin{Bottom := 10.0}
                        HorizontalAlignment := horizontal_alignment.Center
                        VerticalAlignment := vertical_alignment.Center
                    stack_box_slot:
                        Widget := PrestigeOreCostStack
                        HorizontalAlignment := horizontal_alignment.Center

        PrestigeFullTextStack := stack_box:
            Orientation := orientation.Horizontal
            Slots := 
                array:
                    stack_box_slot:
                        Widget := PrestigeText
                        Padding := margin{Right := 20.0}
                        HorizontalAlignment := horizontal_alignment.Left
                        VerticalAlignment := vertical_alignment.Center
                    stack_box_slot:
                        Widget := PrestigeTextStack
                        HorizontalAlignment := horizontal_alignment.Right

        MaxedTexture := texture_block:
            DefaultImage := VerseExposed.T_Maxed
            DefaultDesiredSize := vector2{X:= 200.0, Y:= 100.0}

        MaxedTexture.SetVisibility(widget_visibility.Hidden)

        PrestigeButtonOverlay := overlay:
            Slots := array:
                overlay_slot:
                    VerticalAlignment := vertical_alignment.Center
                    HorizontalAlignment := horizontal_alignment.Center
                    Widget := color_block:
                        DefaultDesiredSize := vector2{X:=600.0, Y:=100.0}
                        DefaultOpacity := 0.0
                overlay_slot:
                    VerticalAlignment := vertical_alignment.Top
                    HorizontalAlignment := horizontal_alignment.Center
                    Padding := margin{Top := 10.0}
                    Widget := PrestigeFullTextStack
                overlay_slot:
                    Widget := PrestigeButton
                    HorizontalAlignment := horizontal_alignment.Fill
                    VerticalAlignment := vertical_alignment.Fill

        PrestigeButtonOverlay.SetVisibility(widget_visibility.Hidden)
        
        SmeltButtonOverlay := overlay:
            Slots := array:
                overlay_slot:
                    VerticalAlignment := vertical_alignment.Center
                    HorizontalAlignment := horizontal_alignment.Center
                    Widget := color_block:
                        DefaultDesiredSize := vector2{X:=600.0, Y:=85.0}
                        DefaultOpacity := 0.0
                overlay_slot:
                    Widget := SmeltButton
                    HorizontalAlignment := horizontal_alignment.Fill
                    VerticalAlignment := vertical_alignment.Fill

        SmeltResourceButtons := stack_box:
            Orientation := orientation.Horizontal
            Slots := 
                array:
                    stack_box_slot:
                        Widget := PrevResourceOverlay
                        Padding := margin{Right := 10.0}
                        HorizontalAlignment := horizontal_alignment.Left
                        VerticalAlignment := vertical_alignment.Center
                    stack_box_slot:
                        Widget := CurrentResourceImage
                        HorizontalAlignment := horizontal_alignment.Center
                        VerticalAlignment := vertical_alignment.Center
                    stack_box_slot:
                        Widget := NextResourceOverlay
                        Padding := margin{Left := 10.0}
                        HorizontalAlignment := horizontal_alignment.Right
                        VerticalAlignment := vertical_alignment.Center

        SmeltFullStackBox := stack_box:
            Orientation := orientation.Vertical
            Slots := 
                array:
                    stack_box_slot:
                        Widget := PrestigeButtonOverlay
                        HorizontalAlignment := horizontal_alignment.Center
                        VerticalAlignment := vertical_alignment.Center
                    stack_box_slot:
                        Widget := SmeltTextStackBox
                        Padding := margin{Top := 100.0}
                        HorizontalAlignment := horizontal_alignment.Fill
                        VerticalAlignment := vertical_alignment.Fill
                    stack_box_slot:
                        Widget := SmeltResourceButtons
                        Padding := margin{Top := 5.0}
                        HorizontalAlignment := horizontal_alignment.Center
                        VerticalAlignment := vertical_alignment.Center
                    stack_box_slot:
                        Widget := SmeltButtonOverlay
                        Padding := margin{Top := 5.0}
                        HorizontalAlignment := horizontal_alignment.Center
                        VerticalAlignment := vertical_alignment.Center

        StatTextStackBox := stack_box:
            Orientation := orientation.Vertical
            Slots := 
                array:
                    stack_box_slot:
                        Widget := StrengthStat.Widget
                        Padding := margin{Bottom := 2.5}
                        HorizontalAlignment := horizontal_alignment.Center
                        VerticalAlignment := vertical_alignment.Top
                    stack_box_slot:
                        Widget := CriticalRateStat.Widget
                        Padding := margin{Bottom := 2.5}
                        HorizontalAlignment := horizontal_alignment.Center
                        VerticalAlignment := vertical_alignment.Top
                    stack_box_slot:
                        Widget := CriticalDamageStat.Widget
                        Padding := margin{Bottom := 2.5}
                        HorizontalAlignment := horizontal_alignment.Center
                        VerticalAlignment := vertical_alignment.Top
                    stack_box_slot:
                        Widget := FinalStrikeStat.Widget
                        Padding := margin{Bottom := 2.5}
                        HorizontalAlignment := horizontal_alignment.Center
                        VerticalAlignment := vertical_alignment.Top

        ArtifactDisplayTexture := texture_block:
            DefaultImage := VerseExposed.T_UnknownArtifact
            DefaultDesiredSize := vector2{X:= 200.0, Y:=200.0}

        NoImageStatTextStackBox := stack_box:
            Orientation := orientation.Vertical
            Slots := 
                array:
                    stack_box_slot:
                        Widget := StatsTitle
                        Padding := margin{Bottom := 10.0}
                        HorizontalAlignment := horizontal_alignment.Center
                        VerticalAlignment := vertical_alignment.Top
                    stack_box_slot:
                        Widget := StatTextStackBox
                        HorizontalAlignment := horizontal_alignment.Center
                        VerticalAlignment := vertical_alignment.Top

        FullStatTextStackBox := stack_box:
            Orientation := orientation.Horizontal
            Slots := 
                array:
                    stack_box_slot:
                        Widget := ArtifactDisplayTexture
                        HorizontalAlignment := horizontal_alignment.Left
                        VerticalAlignment := vertical_alignment.Top
                    stack_box_slot:
                        Widget := NoImageStatTextStackBox
                        HorizontalAlignment := horizontal_alignment.Right
                        VerticalAlignment := vertical_alignment.Top

        ArtifactStackBox := stack_box:
            Orientation := orientation.Vertical
            Slots := 
                array:
                    <#stack_box_slot:
                        Widget := ArtifactTitle
                        Padding := margin{Bottom := 5.0}
                        HorizontalAlignment := horizontal_alignment.Center
                        VerticalAlignment := vertical_alignment.Center#>
                    stack_box_slot:
                        Widget := ArtifactCommon.Widget
                        Padding := margin{Bottom := 5.0}
                        HorizontalAlignment := horizontal_alignment.Center
                        VerticalAlignment := vertical_alignment.Center
                    stack_box_slot:
                        Widget := ArtifactUncommon.Widget
                        Padding := margin{Bottom := 5.0}
                        HorizontalAlignment := horizontal_alignment.Fill
                        VerticalAlignment := vertical_alignment.Fill
                    stack_box_slot:
                        Widget := ArtifactRare.Widget
                        Padding := margin{Bottom := 5.0}
                        HorizontalAlignment := horizontal_alignment.Fill
                        VerticalAlignment := vertical_alignment.Fill
                    stack_box_slot:
                        Widget := ArtifactEpic.Widget
                        Padding := margin{Bottom := 5.0}
                        HorizontalAlignment := horizontal_alignment.Fill
                        VerticalAlignment := vertical_alignment.Fill
                    stack_box_slot:
                        Widget := ArtifactLegendary.Widget
                        Padding := margin{Bottom := 5.0}
                        HorizontalAlignment := horizontal_alignment.Fill
                        VerticalAlignment := vertical_alignment.Fill

        InventoryStackBox := stack_box:
            Orientation := orientation.Vertical
            Slots := 
                array:
                    #stack_box_slot:
                        #Widget := InventoryTitle
                        #Padding := margin{Bottom := 5.0}
                        #HorizontalAlignment := horizontal_alignment.Center
                        #VerticalAlignment := vertical_alignment.Center
                    stack_box_slot:
                        Widget := StoneInventory.Widget
                        Padding := margin{Bottom := 5.0}
                        HorizontalAlignment := horizontal_alignment.Center
                        VerticalAlignment := vertical_alignment.Center
                    stack_box_slot:
                        Widget := CopperInventory.Widget
                        Padding := margin{Bottom := 5.0}
                        VerticalAlignment := vertical_alignment.Center
                        HorizontalAlignment := horizontal_alignment.Center
                    stack_box_slot:
                        Widget := SilverInventory.Widget
                        Padding := margin{Bottom := 5.0}
                        VerticalAlignment := vertical_alignment.Center
                        HorizontalAlignment := horizontal_alignment.Center
                    stack_box_slot:
                        Widget := MalachiteInventory.Widget
                        Padding := margin{Bottom := 5.0}
                        VerticalAlignment := vertical_alignment.Center
                        HorizontalAlignment := horizontal_alignment.Center
                    stack_box_slot:
                        Widget := SpectroliteInventory.Widget
                        Padding := margin{Bottom := 5.0}
                        VerticalAlignment := vertical_alignment.Center
                        HorizontalAlignment := horizontal_alignment.Center
                    stack_box_slot:
                        Widget := ObsidianInventory.Widget
                        Padding := margin{Bottom := 5.0}
                        VerticalAlignment := vertical_alignment.Center
                        HorizontalAlignment := horizontal_alignment.Center
        
        BackgroundOverlay := overlay:
            Slots := array:
                overlay_slot:
                    VerticalAlignment := vertical_alignment.Fill
                    HorizontalAlignment := horizontal_alignment.Fill
                    Widget := texture_block:
                        DefaultImage := VerseExposed.T_ArtifactMenu

        OutCanvas := canvas:
            Slots := array:
                canvas_slot:
                    Anchors := anchors{Minimum := vector2{X:=0.0, Y:=0.0}, Maximum := vector2{X:=1.0, Y:=1.0}}
                    Alignment := CenteredAlignment
                    Widget := BackgroundOverlay
                    ZOrder := 0
                canvas_slot:
                    Anchors := anchors{Minimum := vector2{X:=0.22, Y:=0.22}, Maximum := vector2{X:=0.5, Y:=0.85}}
                    Alignment := CenteredAlignment
                    Widget := ArtifactStackBox
                    ZOrder := 1
                canvas_slot:
                    Anchors := anchors{Minimum := vector2{X:=0.08, Y:=0.2}, Maximum := vector2{X:=0.08, Y:=0.85}}
                    Alignment := CenteredAlignment
                    Widget := InventoryStackBox
                    ZOrder := 1
                canvas_slot:
                    Anchors := anchors{Minimum := vector2{X:=0.58, Y:=0.45}, Maximum := vector2{X:=0.95, Y:=0.85}}
                    Alignment := CenteredAlignment
                    Widget := SmeltFullStackBox
                    ZOrder := 1
                canvas_slot:
                    Anchors := anchors{Minimum := vector2{X:=0.765, Y:=0.5}, Maximum := vector2{X:=0.765, Y:=0.5}}
                    Alignment := CenteredAlignment
                    Widget := MaxedTexture
                    ZOrder := 1
                canvas_slot:
                    Anchors := anchors{Minimum := vector2{X:=0.58, Y:=0.2}, Maximum := vector2{X:=0.9, Y:=0.6}}
                    Alignment := CenteredAlignment
                    Widget := FullStatTextStackBox
                    ZOrder := 1
                canvas_slot:
                    Anchors := anchors{Minimum := vector2{X:=0.4, Y:=0.95}, Maximum := vector2{X:=0.6, Y:=0.95}}
                    Alignment := CenteredAlignment
                    Widget := ExitButton
                    ZOrder := 1

        return ArtifactMenuCanvasData:
            Canvas := OutCanvas
            CurrentResourceType := PropResourceType.Copper
            CurrentArtifactRarity := ArtifactRarity.Null
            StoneCostText := StoneCost
            ResourceCostText := OreCost
            ResourceCostImage := ResourceCostImage
            TiersText := OreTiersText
            PlayerStoneText := StoneInventory.Text
            PlayerCopperOreText := CopperInventory.Text
            PlayerSilverOreText := SilverInventory.Text
            PlayerMalachiteOreText := MalachiteInventory.Text
            PlayerSpectroliteOreText := SpectroliteInventory.Text
            PlayerObsidianOreText := ObsidianInventory.Text
            ArtifactCommon := ArtifactCommon
            ArtifactUncommon := ArtifactUncommon
            ArtifactRare := ArtifactRare
            ArtifactEpic := ArtifactEpic
            ArtifactLegendary := ArtifactLegendary
            ArtifactStatDatas := array{StrengthStat, CriticalRateStat, CriticalDamageStat, FinalStrikeStat}
            ArtifactDisplayTexture := ArtifactDisplayTexture
            CurrentResourceDisplay := CurrentResourceImage
            PrestigeButtonOverlay := PrestigeButtonOverlay
            PrestigeStoneCostText := PrestigeStoneCost
            PrestigeResourceCostText := PrestigeOreCost
            PrestigeResourceCostImage := PrestigeResourceCostImage
            NextResourceButton := NextResourceButton
            PrevResourceButton := PrevResourceButton
            PrestigeButton := PrestigeButton
            MaxedWidget := MaxedTexture
            SmeltButton := SmeltButton
            ExitButton  := ExitButton