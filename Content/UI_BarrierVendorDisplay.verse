using { /Fortnite.com/Devices }
using { /Verse.org/Simulation }
using { /UnrealEngine.com/Temporary/Diagnostics }
using { /UnrealEngine.com/Temporary/UI }
using { /Fortnite.com/UI }
using { /UnrealEngine.com/Temporary/SpatialMath}
using { /Verse.org/Colors }
using { /Verse.org/Assets }


MakeBarrierVendorItem<constructor>(InImage : texture, InName : string, InCost : string,InTrigger : trigger_device, InPercentage : float) := BarrierVendorItemData:
    Image := InImage
    Name := InName
    Cost := InCost
    Trigger := InTrigger
    GatherPercentageIncrease := InPercentage

BarrierVendorItemData := class:
    Image : texture
    Name : string
    Cost : string
    Trigger : trigger_device
    GatherPercentageIncrease : float

BarrierVendorSlotData := class:
    Widget : widget
    Button : text_button_base
    Cost : text_base
    ButtonColor : texture_block

    var StoredTrigger : ?trigger_device = false

    OnButtonEvent(WidgetMessage : widget_message):void=
        if(bStoredTrigger := StoredTrigger?):
            bStoredTrigger.Trigger(WidgetMessage.Player)

BarrierVendorCanvasData := class:
    ExitButton : text_button_base
    BarrierVendorSlot : BarrierVendorSlotData
    Canvas : canvas

UI_BarrierVendorDisplay := class(creative_device):
    ToMessage<localizes>(String : string) : message = "{String}"

    var BuiltCanvasMap : [player]BarrierVendorCanvasData = map{}

    CenteredAlignment : vector2 = vector2{X := 0.5, Y := 0.5}

    # Initiliase the UI so it can added to the player much quicker
    InitUI(InPlayer : player, InBarrierVendorItem : BarrierVendorItemData):void=
        BarrierVendorCanvas := MakeBarrierVendorCanvas(InBarrierVendorItem)
        # Set the players build canvas data
        if(set BuiltCanvasMap[InPlayer] = BarrierVendorCanvas){}
        BarrierVendorCanvas.ExitButton.OnClick().Subscribe(HandlePlayerExit)

    # Add the UI to the player
    AddUI(InPlayer : player):void=
        if(PlayerUI := GetPlayerUI[InPlayer], CanvasData := BuiltCanvasMap[InPlayer]):
            CanvasData.Canvas.SetVisibility(widget_visibility.Visible)
            PlayerUI.AddWidget(CanvasData.Canvas, player_ui_slot{InputMode := ui_input_mode.All} )

    # When player presses the exit button
    HandlePlayerExit(WidgetMessage: widget_message):void=
        spawn{HideUI(WidgetMessage.Player)}

    HideUI(Player: player)<suspends>:void=
        Sleep(0.0) # So we're sure that the canvas has been added to BuiltCanvas map
        if(PlayerUI := GetPlayerUI[Player], CanvasData := BuiltCanvasMap[Player]):
            CanvasData.Canvas.SetVisibility(widget_visibility.Hidden)
            PlayerUI.RemoveWidget(CanvasData.Canvas)

    UpdateUI(InPlayer : player, OpenedBarrier : logic, CanAfford : logic):void=
        if(CanvasData := BuiltCanvasMap[InPlayer]):
            if(OpenedBarrier = false):
                if(CanAfford = false):
                    CanvasData.BarrierVendorSlot.Button.SetText(ToMessage("CAN'T AFFORD"))
                    CanvasData.BarrierVendorSlot.Button.SetEnabled(false)
                    CanvasData.BarrierVendorSlot.ButtonColor.SetTint(NamedColors.Grey)
                    CanvasData.BarrierVendorSlot.Cost.SetTextColor(NamedColors.Red)
                else:
                    CanvasData.BarrierVendorSlot.Button.SetText(ToMessage("PURCHASE"))
                    CanvasData.BarrierVendorSlot.Button.SetEnabled(true)
                    CanvasData.BarrierVendorSlot.ButtonColor.SetTint(NamedColors.Yellow)
                    CanvasData.BarrierVendorSlot.Cost.SetTextColor(NamedColors.White)
            else:
                CanvasData.BarrierVendorSlot.Button.SetText(ToMessage("✅"))
                CanvasData.BarrierVendorSlot.Button.SetEnabled(false)
                CanvasData.BarrierVendorSlot.ButtonColor.SetTint(NamedColors.Grey)
                CanvasData.BarrierVendorSlot.Cost.SetTextColor(NamedColors.White)

        

    MakeBarrierVendorCanvas(InBarrierVendorItem : BarrierVendorItemData):BarrierVendorCanvasData=
        ExitButton := button_loud:
            DefaultText := ToMessage("Exit")

        # Make an array with all of the Teleporter slots made
        CurrentBarrierVendorSlot := MakeBarrierVendorSlot(InBarrierVendorItem.Image, InBarrierVendorItem.Name, InBarrierVendorItem.Cost, InBarrierVendorItem.GatherPercentageIncrease)

        set CurrentBarrierVendorSlot.StoredTrigger = option{InBarrierVendorItem.Trigger}
        CurrentBarrierVendorSlot.Button.OnClick().Subscribe(CurrentBarrierVendorSlot.OnButtonEvent)
        # Create the Stack Box Slot
        CurrentBarrierVendorOverlay := overlay:
            Slots := array:
                overlay_slot:
                    VerticalAlignment := vertical_alignment.Center
                    HorizontalAlignment := horizontal_alignment.Center
                    Padding := margin{Left := 20.0, Right := 20.0}
                    Widget := CurrentBarrierVendorSlot.Widget

        BackgroundOverlay := overlay:
            Slots := array:
                overlay_slot:
                    VerticalAlignment := vertical_alignment.Fill
                    HorizontalAlignment := horizontal_alignment.Fill
                    Widget := texture_block:
                        DefaultImage := VerseExposed.T_UIBackground
        
        # Create the canvas for the whole widget
        Canvas := canvas:
            Slots := array:
                canvas_slot:
                    Anchors := anchors{Minimum := vector2{X:=0.0, Y:=0.0}, Maximum := vector2{X:=1.0, Y:=1.0}}
                    Alignment := CenteredAlignment
                    Widget := BackgroundOverlay
                canvas_slot:
                    Anchors := anchors{Minimum := vector2{X:=0.0, Y:=0.0}, Maximum := vector2{X:=1.0, Y:=1.0}}
                    Alignment := CenteredAlignment
                    Widget := CurrentBarrierVendorOverlay
                canvas_slot:
                    # Fix for controllers not being able to find buttons that are not perpendicular to the selected button possition
                    Anchors := anchors{Minimum := vector2{X:=0.4, Y:=0.95}, Maximum := vector2{X:=0.6, Y:=0.95}}
                    Alignment := CenteredAlignment
                    Widget := ExitButton

        return BarrierVendorCanvasData:
            ExitButton := ExitButton
            BarrierVendorSlot := CurrentBarrierVendorSlot
            Canvas := Canvas

    
    # Make a singular Teleporter slot
    MakeBarrierVendorSlot(InImage : texture, InName : string, InCost : string, InGatherPercentage : float):BarrierVendorSlotData=
        IconMargin := margin{Bottom := 175.0, Left := 10.0, Right := 10.0, Top := 30.0}
        PurchaseButton := button_quiet:
            DefaultText := ToMessage("Purchase")

        PurchaseButtonColor := texture_block:
            DefaultImage := VerseExposed.T_TeleporterButtonBackground
            DefaultDesiredSize := vector2{X:=512.0,Y:=54.0}
            DefaultTint := NamedColors.Yellow

        NameBlock := text_block:
            DefaultText := ToMessage(InName)
            DefaultJustification := text_justification.Center
            DefaultShadowColor := NamedColors.Black
            DefaultShadowOffset := option{vector2{X:=2.0,Y:=2.0}}
        NameBlock.SetTextColor(NamedColors.White)
        NameBlock.SetShadowOpacity(0.8)

        CostBlock := text_block:
            DefaultText := ToMessage(InCost)
            DefaultJustification := text_justification.Center
            DefaultShadowColor := NamedColors.Black
            DefaultShadowOffset := option{vector2{X:=2.0,Y:=2.0}}
        CostBlock.SetTextColor(NamedColors.White)
        CostBlock.SetShadowOpacity(0.8)

        Percentage := FormatFloatToNDP(InGatherPercentage, 1)
        GatherRate := text_block:
            DefaultText := ToMessage("+{Percentage}%")
            DefaultJustification := text_justification.Center
            DefaultShadowColor := NamedColors.Black
            DefaultShadowOffset := option{vector2{X:=2.0,Y:=2.0}}
        GatherRate.SetTextColor(NamedColors.LightGreen)
        GatherRate.SetShadowOpacity(0.8)
        PreGatherRateText := text_block:
            DefaultText := ToMessage("Gather Multiplier Bonus: ")
            DefaultJustification := text_justification.Center
            DefaultShadowColor := NamedColors.Black
            DefaultShadowOffset := option{vector2{X:=2.0,Y:=2.0}}
        PreGatherRateText.SetTextColor(NamedColors.White)
        PreGatherRateText.SetShadowOpacity(0.8)
        
        BackgroundColorBlock := color_block:
            DefaultColor := NamedColors.White
            DefaultOpacity := 0.4

        BackgroundColorBlock2 := color_block:
            DefaultColor := NamedColors.White
            DefaultOpacity := 0.4

        BorderSize := 1.0
        TopBorderColorBlock := color_block{DefaultColor := NamedColors.DarkBlue, DefaultDesiredSize := vector2{X := BorderSize, Y := BorderSize}}
        RightBorderColorBlock := color_block{DefaultColor := NamedColors.DarkBlue, DefaultDesiredSize := vector2{X := BorderSize, Y := BorderSize}}
        BottomBorderColorBlock := color_block{DefaultColor := NamedColors.DarkBlue, DefaultDesiredSize := vector2{X := BorderSize, Y := BorderSize}}
        LeftBorderColorBlock := color_block{DefaultColor := NamedColors.DarkBlue, DefaultDesiredSize := vector2{X := BorderSize, Y := BorderSize}}

        ButtonOverlay := overlay:
            Slots := array:
                overlay_slot:
                    VerticalAlignment := vertical_alignment.Fill
                    HorizontalAlignment := horizontal_alignment.Fill
                    Widget := color_block:
                        DefaultColor := NamedColors.White
                        DefaultOpacity := 0.0
                overlay_slot:
                    VerticalAlignment := vertical_alignment.Center
                    HorizontalAlignment := horizontal_alignment.Center
                    Widget := PurchaseButtonColor
                overlay_slot:
                    VerticalAlignment := vertical_alignment.Center
                    HorizontalAlignment := horizontal_alignment.Fill
                    Padding := margin{Top := 3.0, Bottom := 3.0, Left := 10.0, Right := 10.0}
                    Widget := PurchaseButton

        CostStack := stack_box :
            Orientation := orientation.Horizontal
            Slots := 
                array:
                    stack_box_slot:
                        VerticalAlignment := vertical_alignment.Center
                        HorizontalAlignment := horizontal_alignment.Center
                        Widget := texture_block:
                            DefaultImage := VerseExposed.T_Stone
                    stack_box_slot:
                        VerticalAlignment := vertical_alignment.Center
                        HorizontalAlignment := horizontal_alignment.Center
                        Widget := CostBlock

        GatherStack := stack_box :
            Orientation := orientation.Horizontal
            Slots := 
                array:
                    stack_box_slot:
                        VerticalAlignment := vertical_alignment.Center
                        HorizontalAlignment := horizontal_alignment.Center
                        Widget := PreGatherRateText
                        Padding := margin{Right := 5.0}
                    stack_box_slot:
                        VerticalAlignment := vertical_alignment.Center
                        HorizontalAlignment := horizontal_alignment.Center
                        Widget := GatherRate
    
        StackBoxSlots :=
            array:
                stack_box_slot:
                    Widget := NameBlock
                stack_box_slot:
                    Widget := CostStack
                stack_box_slot:
                    Widget := GatherStack
                stack_box_slot:
                    Padding := margin{Top := 10.0}
                    VerticalAlignment := vertical_alignment.Fill
                    HorizontalAlignment := horizontal_alignment.Fill
                    Widget := ButtonOverlay
        
        Overlay := overlay:
            Slots := array:
                overlay_slot:
                    VerticalAlignment := vertical_alignment.Fill
                    HorizontalAlignment := horizontal_alignment.Fill
                    Padding := margin{Bottom := -50.0}
                    Widget := BackgroundColorBlock
                overlay_slot:
                    VerticalAlignment := vertical_alignment.Bottom
                    HorizontalAlignment := horizontal_alignment.Fill
                    Padding := margin{Top := 50.0}
                    Widget := overlay:
                        Slots := array:
                            overlay_slot:
                                Padding := margin{Top := 50.0, Bottom := 0.0}
                                VerticalAlignment := vertical_alignment.Center
                                HorizontalAlignment := horizontal_alignment.Fill
                                Widget := stack_box:
                                    Orientation := orientation.Vertical
                                    Slots := StackBoxSlots
                overlay_slot:
                    VerticalAlignment := vertical_alignment.Fill
                    HorizontalAlignment := horizontal_alignment.Fill
                    Padding := IconMargin
                    Widget := texture_block:
                        DefaultImage := InImage
                        DefaultDesiredSize := vector2{X:=512.0,Y:=512.0}
                overlay_slot:
                    VerticalAlignment := vertical_alignment.Top
                    HorizontalAlignment := horizontal_alignment.Fill
                    Widget := TopBorderColorBlock
                overlay_slot:
                    VerticalAlignment := vertical_alignment.Bottom
                    HorizontalAlignment := horizontal_alignment.Fill
                    Padding := margin{Bottom := -50.0}
                    Widget := BottomBorderColorBlock
                overlay_slot:
                    VerticalAlignment := vertical_alignment.Fill
                    HorizontalAlignment := horizontal_alignment.Right
                    Padding := margin{Bottom := -50.0}
                    Widget := RightBorderColorBlock
                overlay_slot:
                    VerticalAlignment := vertical_alignment.Fill
                    HorizontalAlignment := horizontal_alignment.Left
                    Padding := margin{Bottom := -50.0}
                    Widget := LeftBorderColorBlock

        return BarrierVendorSlotData:
            Widget := Overlay
            Button := PurchaseButton
            ButtonColor := PurchaseButtonColor
            Cost := CostBlock