using { /Fortnite.com/Devices }
using { /Fortnite.com/Characters }
using { /Verse.org/Simulation }
using { /Verse.org/Random }
using { /UnrealEngine.com/Temporary/Diagnostics }
using { /UnrealEngine.com/Temporary/UI }
using { /Fortnite.com/UI }
using { /UnrealEngine.com/Temporary/SpatialMath}
using { /Verse.org/Colors }
using { /Verse.org/Colors/NamedColors }

UI_BossHealthBar := class():
    var iTotalHP<private> : float = 1000.0
    var iCurrentHP : float = 1000.0
    var iBossName<private> : string = "Bulgna The Ugly"

    BossBarColorBase<private> : color_block = color_block{
        DefaultColor := MakeColorFromHex("000000")
        DefaultDesiredSize := vector2{X:= 1200.0, Y := 40.0}} 

    var BossBarColorTop<private> : color_block = color_block{}  
    var Overlay<private> : overlay = overlay{}
        
    HPBarTitleTextBlock<private> : text_block = text_block{DefaultTextColor := MakeColorFromHex("FFFFFF")}

    HPTitleText<private><localizes>(HPText : string) : message = "{HPText}"

    Init(BossName : string, BossTotalHP: float):void=
        set iBossName = BossName
        set iTotalHP = BossTotalHP
        set iCurrentHP = BossTotalHP
        HPBarTitleTextBlock.SetText(HPTitleText("{iBossName}: {iCurrentHP}/{iTotalHP}"))
    <#
    TakeDamage(DamageAmount : float):void=
        set iCurrentHP -= DamageAmount
        if (iCurrentHP < 0.0):
            set iCurrentHP = 0.0

        HealthPercentage := iCurrentHP / iTotalHP
        BarSize := BossBarColorBase.GetDesiredSize()
        NewHealthBarWidth := BarSize.X * HealthPercentage
        if (Current := Int[iCurrentHP], Total := Int[iTotalHP]):
            HPBarTitleTextBlock.SetText(HPTitleText("{iBossName}: {Current}/{Total}"))
        
        UpdateHealthBar(NewHealthBarWidth)
    #>
    Heal(HealAmount : float):void=
        block:

    ShowUIForPlayer(Player : player):void=
        if (PlayerUI := GetPlayerUI[Player]):
            PlayerUI.AddWidget(CreateUI())

    ShowUIToAllPlayers():void=
        block:

    MyButton : button_device = button_device{}

    UpdateHealthBar(NewWidth : float):void=
        Overlay.RemoveWidget(HPBarTitleTextBlock)
        Overlay.RemoveWidget(BossBarColorTop)
        set BossBarColorTop = CreateTopHealthBar(NewWidth)

        Overlay.AddWidget(CreateOverlaySlot(BossBarColorTop, horizontal_alignment.Left))
        Overlay.AddWidget(CreateOverlaySlot(HPBarTitleTextBlock,horizontal_alignment.Center))
        
    CreateOverlaySlot<private>(TheWidget : widget, HAlignment : horizontal_alignment):overlay_slot=  
        overlay_slot:
            Widget := TheWidget 
            HorizontalAlignment := HAlignment

    CreateTopHealthBar<private>(Width: float):color_block=
        ColorBlock := color_block{
            DefaultColor := MakeColorFromHex("ff2a00")
            DefaultDesiredSize := vector2{X:= Width, Y := 40.0}} 
        
        set BossBarColorTop = ColorBlock

    CreateOverlay<private>() : overlay=
        TheOverlay := overlay:
            Slots := array:
                overlay_slot:
                    Widget := BossBarColorBase
                overlay_slot:
                    Widget := CreateTopHealthBar(1200.0)
                overlay_slot:
                    Widget := HPBarTitleTextBlock
        set Overlay = TheOverlay

    CreateUI<private>() : canvas=
        TheCanvas : canvas = canvas:
            Slots := array:
                canvas_slot:
                    Anchors := anchors{Minimum := vector2{X := 0.5, Y := 0.9}, Maximum := vector2{X := 0.5, Y := 0.9}}
                    Offsets := margin{Top := 0.0, Left := 0.0, Right := 0.0, Bottom := 100.0}
                    Alignment := vector2{X := 0.5, Y := 1.0}
                    Widget := stack_box:
                        Orientation := orientation.Vertical
                        Slots := array:
                            stack_box_slot:
                                Widget := CreateOverlay()

        return TheCanvas        