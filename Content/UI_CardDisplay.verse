
using { /Fortnite.com/Devices }
using { /Verse.org/Simulation }
using { /UnrealEngine.com/Temporary/Diagnostics }

using { /UnrealEngine.com/Temporary/UI }
using { /Fortnite.com/UI }
using { /Verse.org/Colors }
using { /UnrealEngine.com/Temporary/SpatialMath }

# See https://dev.epicgames.com/documentation/en-us/uefn/create-your-own-device-in-verse for how to create a verse device.
ChangeRarityHandler := class(){
    UICardDisplay : UI_CardDisplay
    PlayerStatsManager : player_stats_manager
    Rarity : int

    HandlerFunction(WidgetMessage : widget_message) : void = {
            Player := WidgetMessage.Player
            if(CanvasData := UICardDisplay.BuiltCanvasMap[Player]):
                set CanvasData.SelectedRarity = Rarity
                UICardDisplay.UpdateUI(Player, PlayerStatsManager)
    }
}

# Holds a reference to what item is in a slot and how much
CardSlot := class:
    Widget : widget

    LevelText : text_block
    Image : texture_block

CardDisplayData := class:
    Canvas : canvas
    var SelectedRarity : int

    CommonButton : text_button_base
    RareButton : text_button_base
    EpicButton : text_button_base
    LegendaryButton : text_button_base

    CardSlots : []CardSlot
    ExitButton : text_button_base

# A Verse-authored creative device that can be placed in a level
UI_CardDisplay := class():
    ToMessage<localizes>(String : string) : message = "{String}"
    CenteredAlignment : vector2 = vector2{X := 0.5, Y := 0.5}

    var BuiltCanvasMap : [player]CardDisplayData = map{}

    # Initialises the UI for the given player
    InitUI(InPlayer : player, InPlayerStatsManager : player_stats_manager):void=
        Print("Initilaise Card UI")
        Data := CreateCardDisplay()
        if(set BuiltCanvasMap[InPlayer] = Data, CanvasData := BuiltCanvasMap[InPlayer]):
            Print("Set CardData")
            CanvasData.ExitButton.OnClick().Subscribe(HandlePlayerExit)

            CanvasData.CommonButton.OnClick().Subscribe(ChangeRarityHandler{UICardDisplay := Self, PlayerStatsManager := InPlayerStatsManager, Rarity := 0}.HandlerFunction)
            CanvasData.RareButton.OnClick().Subscribe(ChangeRarityHandler{UICardDisplay := Self, PlayerStatsManager := InPlayerStatsManager, Rarity := 1}.HandlerFunction)
            CanvasData.EpicButton.OnClick().Subscribe(ChangeRarityHandler{UICardDisplay := Self, PlayerStatsManager := InPlayerStatsManager, Rarity := 2}.HandlerFunction)
            CanvasData.LegendaryButton.OnClick().Subscribe(ChangeRarityHandler{UICardDisplay := Self, PlayerStatsManager := InPlayerStatsManager, Rarity := 3}.HandlerFunction)

    # Add the UI to the player
    AddUI(InPlayer : player):void=
        if(PlayerUI := GetPlayerUI[InPlayer], CanvasData := BuiltCanvasMap[InPlayer]):
            CanvasData.Canvas.SetVisibility(widget_visibility.Visible)
            PlayerUI.AddWidget(CanvasData.Canvas, player_ui_slot{InputMode := ui_input_mode.All})
            
    # When player presses the exit button
    HandlePlayerExit(WidgetMessage: widget_message):void=
        HideUI(WidgetMessage.Player)

    HideUI(InPlayer: player):void=
        if(PlayerUI := GetPlayerUI[InPlayer], CanvasData := BuiltCanvasMap[InPlayer]):
            CanvasData.Canvas.SetVisibility(widget_visibility.Hidden)
            PlayerUI.RemoveWidget(CanvasData.Canvas)

    UpdateUI(InPlayer : player, InPlayerStatsManager : player_stats_manager):void=
        Print("Updating Card UI")
        if(CanvasData := BuiltCanvasMap[InPlayer], PlayerStats := InPlayerStatsManager.GetPlayerStats[InPlayer]):
            var SlotIndex : int = 0
            var CardsComingSoon : int = 0

            case (CanvasData.SelectedRarity):
                0=>
                    set CardsComingSoon = 5
                1=>
                    set CardsComingSoon = 7
                2=>
                    set CardsComingSoon = 4
                3=>
                    set CardsComingSoon = 6
                _=>
                    set CardsComingSoon = 0

            for(Index -> CurrentSlot : CanvasData.CardSlots):
                CurrentSlot.LevelText.SetVisibility(widget_visibility.Hidden)
                CurrentSlot.Image.SetImage(VerseExposed.T_CardBackUI)
                if(Index > CardsComingSoon - 1):
                    CurrentSlot.LevelText.SetVisibility(widget_visibility.Visible)
                    CurrentSlot.LevelText.SetText(ToMessage("Coming Soon"))

            for(CurrentCard : PlayerStats.OwnedCards):
                if(CurrentCard.Rarity = CanvasData.SelectedRarity):
                    if(CurrentSlot := CanvasData.CardSlots[SlotIndex]):
                        FoundCard := LookupCardDataSaved(CurrentCard)
                        Print("Card: {FoundCard.ID}")
                        CurrentSlot.LevelText.SetVisibility(widget_visibility.Visible)
                        CurrentSlot.Image.SetImage(FoundCard.Image)
                        CurrentSlot.LevelText.SetText(ToMessage("Lvl {FoundCard.Level}"))
                        set SlotIndex += 1

    # Creates the slot for the Card Display
    CreateCardSlot(InSlotID : int):CardSlot=
        OutImage := texture_block:
            DefaultImage := VerseExposed.T_CardBackUI

        OutQuantityText := text_block:
            DefaultText := ToMessage("Lvl 0")
            DefaultJustification := text_justification.Center
            DefaultShadowColor := NamedColors.Black
            DefaultShadowOffset := option{vector2{X:=2.0,Y:=2.0}}
        OutQuantityText.SetTextColor(NamedColors.White)
        OutQuantityText.SetShadowOpacity(0.8)

        OutOverlay := overlay:
            Slots := array:
                overlay_slot:
                    VerticalAlignment := vertical_alignment.Center
                    HorizontalAlignment := horizontal_alignment.Center
                    Widget := color_block:
                        DefaultColor := NamedColors.White
                        DefaultOpacity := 0.0
                        DefaultDesiredSize := vector2{X := 350.0, Y := 350.0}
                overlay_slot:
                    VerticalAlignment := vertical_alignment.Fill
                    HorizontalAlignment := horizontal_alignment.Fill
                    Widget := OutImage
                overlay_slot:
                    VerticalAlignment := vertical_alignment.Bottom
                    HorizontalAlignment := horizontal_alignment.Fill
                    Widget := OutQuantityText

        return CardSlot:
            Widget := OutOverlay
            LevelText := OutQuantityText
            Image := OutImage

    # Creates the grid for the Card
    CreateCardDisplay():CardDisplayData=
        Print("Creating Card Display")
        ExitButton := button_loud:
            DefaultText := ToMessage("Exit")
        # Create a 4x7 grid of Card slots in stack boxes
        var CardSlots : []CardSlot = array{}

        # Creating arrays to store the widgets to be placed in stack boxes
        var StackSlots1 : []stack_box_slot = array{}
        var StackSlots2 : []stack_box_slot = array{}
        var StackSlots3 : []stack_box_slot = array{}
        var StackSlots4 : []stack_box_slot = array{}
        var StackSlots5 : []stack_box_slot = array{}
        var StackSlots6 : []stack_box_slot = array{}

        for(IndexX := 0..4):
            for(IndexY := 0..1):
                CreatedCardSlot := CreateCardSlot((IndexX + 1) * (IndexY + 1))

                Slot := stack_box_slot:
                    Widget := CreatedCardSlot.Widget
                    Padding := margin{Top := 20.0, Bottom := 20.0}
                    HorizontalAlignment := horizontal_alignment.Center
                    VerticalAlignment := vertical_alignment.Center

                # Add to array with correct index
                if(IndexX = 0):
                    set StackSlots1 += array{Slot}
                else if(IndexX = 1):
                    set StackSlots2 += array{Slot}
                else if(IndexX = 2):
                    set StackSlots3 += array{Slot}
                else if(IndexX = 3):
                    set StackSlots4 += array{Slot}
                else if(IndexX = 4):
                    set StackSlots5 += array{Slot}
                else:
                    set StackSlots6 += array{Slot}
                
                set CardSlots += array{CreatedCardSlot}

        # Create the actual stack boxes that is displayed
        Column1StackBox := stack_box:
            Orientation := orientation.Vertical
            Slots := StackSlots1
        Column2StackBox := stack_box:
            Orientation := orientation.Vertical
            Slots := StackSlots2
        Column3StackBox := stack_box:
            Orientation := orientation.Vertical
            Slots := StackSlots3
        Column4StackBox := stack_box:
            Orientation := orientation.Vertical
            Slots := StackSlots4
        Column5StackBox := stack_box:
            Orientation := orientation.Vertical
            Slots := StackSlots5
        Column6StackBox := stack_box:
            Orientation := orientation.Vertical
            Slots := StackSlots6

        CardsStackBox := stack_box:
            Orientation := orientation.Horizontal
            Slots := array:
                stack_box_slot:
                    Widget := Column1StackBox
                    HorizontalAlignment := horizontal_alignment.Center
                    VerticalAlignment := vertical_alignment.Center
                stack_box_slot:
                    Widget := Column2StackBox
                    HorizontalAlignment := horizontal_alignment.Center
                    VerticalAlignment := vertical_alignment.Center
                stack_box_slot:
                    Widget := Column3StackBox
                    HorizontalAlignment := horizontal_alignment.Center
                    VerticalAlignment := vertical_alignment.Center                
                stack_box_slot:
                    Widget := Column4StackBox
                    HorizontalAlignment := horizontal_alignment.Center
                    VerticalAlignment := vertical_alignment.Center
                stack_box_slot:
                    Widget := Column5StackBox
                    HorizontalAlignment := horizontal_alignment.Center
                    VerticalAlignment := vertical_alignment.Center                
                stack_box_slot:
                    Widget := Column6StackBox
                    HorizontalAlignment := horizontal_alignment.Center
                    VerticalAlignment := vertical_alignment.Center

        CommonButton := button_loud:
            DefaultText := ToMessage("Common")
        RareButton := button_loud:
            DefaultText := ToMessage("Rare")
        EpicButton := button_loud:
            DefaultText := ToMessage("Epic")
        LegendaryButton := button_loud:
            DefaultText := ToMessage("Legendary")

        ButtonStackBox := stack_box:
            Orientation := orientation.Horizontal
            Slots := array:
                stack_box_slot:
                    Widget := CommonButton
                    Padding := margin{Right := 10.0}
                    HorizontalAlignment := horizontal_alignment.Fill
                    VerticalAlignment := vertical_alignment.Center
                stack_box_slot:
                    Widget := RareButton
                    Padding := margin{Right := 10.0}
                    HorizontalAlignment := horizontal_alignment.Fill
                    VerticalAlignment := vertical_alignment.Center
                stack_box_slot:
                    Widget := EpicButton
                    Padding := margin{Right := 10.0}
                    HorizontalAlignment := horizontal_alignment.Fill
                    VerticalAlignment := vertical_alignment.Center
                stack_box_slot:
                    Widget := LegendaryButton
                    HorizontalAlignment := horizontal_alignment.Fill
                    VerticalAlignment := vertical_alignment.Center

        OutStackBox:= stack_box:
            Orientation := orientation.Vertical
            Slots := array:
                stack_box_slot:
                    Widget := ButtonStackBox
                    Padding := margin{Top := 50.0}
                    HorizontalAlignment := horizontal_alignment.Center
                    VerticalAlignment := vertical_alignment.Center
                stack_box_slot:
                    Widget := CardsStackBox
                    Padding := margin{Top := 25.0}
                    HorizontalAlignment := horizontal_alignment.Center
                    VerticalAlignment := vertical_alignment.Center
                stack_box_slot:
                    Widget := ExitButton
                    Padding := margin{Top := 25.0}
                    HorizontalAlignment := horizontal_alignment.Center
                    VerticalAlignment := vertical_alignment.Center

        OutOverlay := overlay:
            Slots := array:
                overlay_slot:
                    VerticalAlignment := vertical_alignment.Fill
                    HorizontalAlignment := horizontal_alignment.Fill
                    Widget := OutStackBox

        BackgroundOverlay := overlay:
            Slots := array:
                overlay_slot:
                    VerticalAlignment := vertical_alignment.Fill
                    HorizontalAlignment := horizontal_alignment.Fill
                    Widget := texture_block:
                        DefaultImage := VerseExposed.T_UIBackground

        OutCanvas := canvas:
            Slots := array:
                canvas_slot:
                    Anchors := anchors{Minimum := vector2{X:=0.0, Y:=0.0}, Maximum := vector2{X:=1.0, Y:=1.0}}
                    Alignment := CenteredAlignment
                    Widget := BackgroundOverlay
                    ZOrder := 0
                canvas_slot:
                    Anchors := anchors{Minimum := vector2{X:=0.05, Y:=0.0}, Maximum := vector2{X:=0.95, Y:=1.0}}
                    Alignment := CenteredAlignment
                    Widget := OutOverlay
                    ZOrder := 1

        return CardDisplayData:
            Canvas := OutCanvas
            SelectedRarity := 0
            CardSlots := CardSlots
            CommonButton := CommonButton
            RareButton := RareButton
            EpicButton := EpicButton
            LegendaryButton := LegendaryButton
            ExitButton := ExitButton
