
using { /Fortnite.com/Devices }
using { /Verse.org/Simulation }
using { /UnrealEngine.com/Temporary/Diagnostics }
using { /UnrealEngine.com/Temporary/UI }
using { /Fortnite.com/UI }
using { /UnrealEngine.com/Temporary/SpatialMath}
using { /Verse.org/Colors }
using { /Verse.org/Assets }

ChargeDisplayCanvasData := class:
    Canvas : canvas
    MainSlot : ChargeSlotData

    var Shown : logic

ChargeSlotData := class:
    Widget : widget
    SlotText : text_block

# A Verse-authored creative device that can be placed in a level
UI_CriticalChargeDisplay := class(creative_device):
    ToMessage<localizes>(String : string) : message = "{String}"
    CenteredAlignment : vector2 = vector2{X := 0.5, Y := 0.5}

    var BuiltCanvasMap : [player]ChargeDisplayCanvasData = map{}
    var ResetRemoveEventMap : [player]event() = map{}
    
    # Initiliase the UI so it can added to the player much quicker
    InitUI(InPlayer : player):void=
        ChargeDisplayCanvas := MakeChargeStreamCanvas(VerseExposed.T_CritHitIcon, "")
        # Set the players build canvas data
        if(set BuiltCanvasMap[InPlayer] = ChargeDisplayCanvas, set ResetRemoveEventMap[InPlayer] = event(){}) {}

    # Add the UI to the player
    AddUI(InPlayer : player):void=
        if(PlayerUI := GetPlayerUI[InPlayer], CanvasData := BuiltCanvasMap[InPlayer]):
            if(CanvasData.Shown = false):
                CanvasData.Canvas.SetVisibility(widget_visibility.Visible)
                PlayerUI.AddWidget(CanvasData.Canvas)
                set CanvasData.Shown = true

    # Hides the UI for the given player
    HideUI(InPlayer: player)<suspends>:void=
        Sleep(0.0) # So we're sure that the canvas has been added to BuiltCanvas map
        if(PlayerUI := GetPlayerUI[InPlayer], CanvasData := BuiltCanvasMap[InPlayer]):
            if(CanvasData.Shown = true):
                CanvasData.Canvas.SetVisibility(widget_visibility.Hidden)
                PlayerUI.RemoveWidget(CanvasData.Canvas)
                set CanvasData.Shown = false

    AddCharge(InPlayer : player, InAmount : int):void=
        if(CanvasData := BuiltCanvasMap[InPlayer], PlayerRemoveEvent := ResetRemoveEventMap[InPlayer]):
            PlayerRemoveEvent.Signal()
            if(InAmount >= 50):
                CanvasData.MainSlot.SlotText.SetText(ToMessage("Charged!"))
            else:
                CanvasData.MainSlot.SlotText.SetText(ToMessage("{InAmount}/50"))
        spawn:
            RemoveUIOnTimer(InPlayer)

    # Removes the UI when the sleep succeeds
    RemoveUIOnTimer(InPlayer : player)<suspends>:void=
        if(PlayerRemoveEvent := ResetRemoveEventMap[InPlayer]):
            race:
                PlayerRemoveEvent.Await()
                block:
                    Sleep(5.0)
                    HideUI(InPlayer)

    # Makes the main display that shows how much total you have gained
    MakeChargeDisplaySlot(InImage : texture, InAmount : string):ChargeSlotData=
        # Variables for other parts of UI
        # Use these to copy and paste the settings into overlays
        BackgroundColorBlock := color_block:
            DefaultColor := NamedColors.White
            DefaultOpacity := 0.2

        BorderSize := 1.0
        BorderColorBlock := color_block:
            DefaultColor := NamedColors.White
            DefaultDesiredSize := vector2{X := BorderSize, Y := BorderSize}
    
        ImageWidget := texture_block:
            DefaultImage := InImage
        
        AmountText := text_block:
            DefaultText := ToMessage("{InAmount}/50")
            DefaultJustification := text_justification.Left
            DefaultShadowColor := NamedColors.Black
            DefaultShadowOffset := option{vector2{X:=2.0,Y:=2.0}}
        AmountText.SetTextColor(NamedColors.White)
        AmountText.SetShadowOpacity(0.8)
        
        IconOverlay := overlay:
            Slots := array:
                overlay_slot:
                    VerticalAlignment := vertical_alignment.Center
                    HorizontalAlignment := horizontal_alignment.Center
                    Widget := ImageWidget
        
        ResourceOverlay := overlay:
            Slots := array:
                overlay_slot:
                    VerticalAlignment := vertical_alignment.Center
                    HorizontalAlignment := horizontal_alignment.Left
                    Padding := margin{Top := 10.0, Bottom := 10.0, Left := 10.0, Right := 10.0}
                    Widget := stack_box:
                        Orientation := orientation.Horizontal
                        Slots := 
                            array:
                                stack_box_slot:
                                    Widget := IconOverlay
                                    HorizontalAlignment := horizontal_alignment.Left
                                    VerticalAlignment := vertical_alignment.Center
                                stack_box_slot:
                                    Widget := AmountText
                                    Padding := margin{Left := 15.0}
                                    VerticalAlignment := vertical_alignment.Center
                                    HorizontalAlignment := horizontal_alignment.Left
        
        return ChargeSlotData:
            Widget := ResourceOverlay
            SlotText := AmountText

    # Make the overall canvas for the 
    MakeChargeStreamCanvas(InImage : texture, InAmount : string):ChargeDisplayCanvasData=
        MainSlotData := MakeChargeDisplaySlot(InImage, InAmount)
        OutputCanvas := canvas:
            Slots := array:
                canvas_slot:
                    Anchors := anchors{Minimum := vector2{X:=0.755, Y:=0.7}, Maximum := vector2{X:=0.755, Y:=0.7}}
                    Alignment := CenteredAlignment
                    Widget := MainSlotData.Widget
                    SizeToContent := true

        return ChargeDisplayCanvasData:
            Canvas := OutputCanvas
            MainSlot := MainSlotData
            Shown := false
