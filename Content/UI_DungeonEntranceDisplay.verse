using { /Fortnite.com/Devices }
using { /Verse.org/Simulation }
using { /UnrealEngine.com/Temporary/Diagnostics }
using { /UnrealEngine.com/Temporary/UI }
using { /Fortnite.com/UI }
using { /UnrealEngine.com/Temporary/SpatialMath}
using { /Verse.org/Colors }
using { /Verse.org/Assets }
#We add the varialbes we want to add into a "Dungeon Item" here
MakeDungeonItem<constructor>(InImage : texture, InName : string, InTrigger : trigger_device, InDungeonDescription : string, InCostAmount : int) := DungeonItemData:
    Image := InImage
    Name := InName
    Amount := InCostAmount
    Trigger := InTrigger
    DungeonDescription := InDungeonDescription

#We create the item data here
DungeonItemData := class:
    Image : texture
    Name : string
    Amount : int
    Trigger : trigger_device
    DungeonDescription : string

DungeonSlotData := class:
    Widget : widget
    Button : text_button_base
    ButtonColor : texture_block

    var StoredTrigger : ?trigger_device = false

    OnButtonEvent(WidgetMessage : widget_message):void=
        if(bStoredTrigger := StoredTrigger?):
            bStoredTrigger.Trigger(WidgetMessage.Player)

DungeonCanvasData := class:
    ExitButton : text_button_base
    DungeonSlot : DungeonSlotData
    Canvas : canvas

# A Verse-authored creative device that can be placed in a level
UI_DungeonEntranceDisplay := class(creative_device):
    ToMessage<localizes>(String : string) : message = "{String}"
    var BuiltCanvasMap : [player]DungeonCanvasData = map{}
    CenteredAlignment : vector2 = vector2{X := 0.5, Y := 0.5}

    # Initiliase the UI so it can added to the player much quicker
    InitUI(InPlayer : player, InDungeonItem : DungeonItemData):void=
        DungeonCanvas := MakeDungeonCanvas(InDungeonItem)
        # Set the players build canvas data
        if(set BuiltCanvasMap[InPlayer] = DungeonCanvas){}
        DungeonCanvas.ExitButton.OnClick().Subscribe(HandlePlayerExit)

    # Add the UI to the player
    AddUI(InPlayer : player):void=
        if(PlayerUI := GetPlayerUI[InPlayer], CanvasData := BuiltCanvasMap[InPlayer]):
            CanvasData.Canvas.SetVisibility(widget_visibility.Visible)
            PlayerUI.AddWidget(CanvasData.Canvas, player_ui_slot{InputMode := ui_input_mode.All} )

    # When player presses the exit button
    HandlePlayerExit(WidgetMessage: widget_message):void=
        spawn{HideUI(WidgetMessage.Player)}

    HideUI(Player: player)<suspends>:void=
        Sleep(0.0) # So we're sure that the canvas has been added to BuiltCanvas map
        if(PlayerUI := GetPlayerUI[Player], CanvasData := BuiltCanvasMap[Player]):
            CanvasData.Canvas.SetVisibility(widget_visibility.Hidden)
            PlayerUI.RemoveWidget(CanvasData.Canvas)

    UpdateUI(InPlayer : player, Opened : logic):void=
        if(CanvasData := BuiltCanvasMap[InPlayer]):
            if(Opened = false):
                CanvasData.DungeonSlot.Button.SetText(ToMessage("Join"))
                CanvasData.DungeonSlot.Button.SetEnabled(true)
                CanvasData.DungeonSlot.ButtonColor.SetTint(NamedColors.Yellow)

    MakeDungeonCanvas(InDungeonItem : DungeonItemData):DungeonCanvasData=
        #Exit Button is just this =o
        ExitButton := button_quiet:
            DefaultText := ToMessage("Exit")

        # Make an array with all of the Teleporter slots made
        CurrentDungeonSlot := MakeDungeonSlot(InDungeonItem.Image, InDungeonItem.Name, InDungeonItem.DungeonDescription, InDungeonItem.Amount)

        set CurrentDungeonSlot.StoredTrigger = option{InDungeonItem.Trigger}
        CurrentDungeonSlot.Button.OnClick().Subscribe(CurrentDungeonSlot.OnButtonEvent)
        # Create the Stack Box Slot
        CurrentDungeonOverlay := overlay:
            Slots := array:
                overlay_slot:
                    VerticalAlignment := vertical_alignment.Center
                    HorizontalAlignment := horizontal_alignment.Center
                    Padding := margin{Left := 20.0, Right := 20.0}
                    Widget := CurrentDungeonSlot.Widget

        BackgroundOverlay := overlay:
            Slots := array:
                overlay_slot:
                    VerticalAlignment := vertical_alignment.Fill
                    HorizontalAlignment := horizontal_alignment.Fill
                    Widget := texture_block:
                        DefaultImage := VerseExposed.T_UIBackground
        
        # Create the canvas for the whole widget
        Canvas := canvas:
            Slots := array:
                #
                canvas_slot:
                    Anchors := anchors{Minimum := vector2{X:=0.0, Y:=0.0}, Maximum := vector2{X:=1.0, Y:=1.0}}
                    Alignment := CenteredAlignment
                    Widget := BackgroundOverlay
                canvas_slot:
                    Anchors := anchors{Minimum := vector2{X:=0.0, Y:=0.0}, Maximum := vector2{X:=1.0, Y:=1.0}}
                    Alignment := CenteredAlignment
                    Widget := CurrentDungeonOverlay
                #TODO: NOTE!! I can add more slots and position them here
                
                canvas_slot:
                    # Fix for controllers not being able to find buttons that are not perpendicular to the selected button possition
                    Anchors := anchors{Minimum := vector2{X:=0.4, Y:=0.95}, Maximum := vector2{X:=0.6, Y:=0.95}}
                    Alignment := CenteredAlignment
                    Widget := ExitButton

        return DungeonCanvasData:
            ExitButton := ExitButton
            DungeonSlot := CurrentDungeonSlot
            Canvas := Canvas

    # Make a singular Teleporter slot
    MakeDungeonSlot(InImage : texture, InName : string, DungeonDescription : string, CostAmount : int):DungeonSlotData=
        IconMargin := margin{Bottom := 10.0, Left := 10.0, Right := 10.0, Top := 30.0}
        PurchaseButton := button_quiet:
            DefaultText := ToMessage("Purchase")

        PurchaseButtonColor := texture_block:
            DefaultImage := VerseExposed.T_TeleporterButtonBackground
            DefaultDesiredSize := vector2{X:=512.0,Y:=54.0}
            DefaultTint := NamedColors.Yellow

        NameBlock := texture_block:
            DefaultImage := VerseExposed.T_Dungeon
            DefaultDesiredSize := vector2{X:=512.0,Y:=162.0}
            DefaultTint := NamedColors.Yellow
            <#
            DefaultText := ToMessage(InName)
            DefaultJustification := text_justification.Center
            DefaultShadowColor := NamedColors.Black
            DefaultShadowOffset := option{vector2{X:=2.0,Y:=2.0}}
        NameBlock.SetTextColor(NamedColors.White)
        NameBlock.SetShadowOpacity(0.8)
        #>

        #Dungeon Description
        DescriptionBlock := text_block:
            DefaultText := ToMessage(DungeonDescription)
            DefaultJustification := text_justification.Center
            DefaultShadowColor := NamedColors.Black
            DefaultShadowOffset := option{vector2{X:=2.0,Y:=2.0}}
        DescriptionBlock.SetTextColor(NamedColors.White)
        DescriptionBlock.SetShadowOpacity(0.8)

        #Dungeon Description
        RewardBlock := text_block:
            DefaultText := ToMessage("Reward X 100 Multiplier")
            DefaultJustification := text_justification.Center
            DefaultShadowColor := NamedColors.Black
            DefaultShadowOffset := option{vector2{X:=2.0,Y:=2.0}}
        RewardBlock.SetTextColor(NamedColors.White)
        RewardBlock.SetShadowOpacity(0.8)

        #Reward Images
        RewardImage1 := texture_block:
            DefaultImage := VerseExposed.T_IslandLootBonus_SmeltingOre
            DefaultDesiredSize := vector2{X:=256.0,Y:=256.0}
            #DefaultTint := NamedColors.Yellow
        RewardImage2 := texture_block:
            DefaultImage := VerseExposed.T_IslandLootBonus_Stone
            DefaultDesiredSize := vector2{X:=256.0,Y:=256.0}
            #DefaultTint := NamedColors.Yellow

        CostBlock := text_block:
            DefaultJustification := text_justification.Center
            DefaultShadowColor := NamedColors.Black
            DefaultShadowOffset := option{vector2{X:=2.0,Y:=2.0}}
        CostBlock.SetTextColor(NamedColors.White)
        CostBlock.SetShadowOpacity(0.8)

        #We define the text_blocks here and call via Widget := CostText [In this example]
        CostText := text_block:
            DefaultText := ToMessage("Cost: ")
            DefaultJustification := text_justification.Center
            DefaultShadowColor := NamedColors.Black
            DefaultShadowOffset := option{vector2{X:=2.0,Y:=2.0}}
        CostText.SetTextColor(NamedColors.White)
        CostText.SetShadowOpacity(0.8)
        CostAmountText := text_block:
            DefaultText := ToMessage("x{CostAmount}")
            DefaultJustification := text_justification.Center
            DefaultShadowColor := NamedColors.Black
            DefaultShadowOffset := option{vector2{X:=2.0,Y:=2.0}}
        CostAmountText.SetTextColor(NamedColors.White)
        CostAmountText.SetShadowOpacity(0.8)
        
        BackgroundColorBlock := color_block:
            DefaultColor := NamedColors.White
            DefaultOpacity := 0.4

        BackgroundColorBlock2 := color_block:
            DefaultColor := NamedColors.White
            DefaultOpacity := 0.4

        BorderSize := 1.0
        TopBorderColorBlock := color_block{DefaultColor := NamedColors.DarkBlue, DefaultDesiredSize := vector2{X := BorderSize, Y := BorderSize}}
        RightBorderColorBlock := color_block{DefaultColor := NamedColors.DarkBlue, DefaultDesiredSize := vector2{X := BorderSize, Y := BorderSize}}
        BottomBorderColorBlock := color_block{DefaultColor := NamedColors.DarkBlue, DefaultDesiredSize := vector2{X := BorderSize, Y := BorderSize}}
        LeftBorderColorBlock := color_block{DefaultColor := NamedColors.DarkBlue, DefaultDesiredSize := vector2{X := BorderSize, Y := BorderSize}}

        ButtonOverlay := overlay:
            Slots := array:
                overlay_slot:
                    VerticalAlignment := vertical_alignment.Fill
                    HorizontalAlignment := horizontal_alignment.Fill
                    Widget := color_block:
                        DefaultColor := NamedColors.White
                        DefaultOpacity := 0.0
                overlay_slot:
                    VerticalAlignment := vertical_alignment.Center
                    HorizontalAlignment := horizontal_alignment.Center
                    Widget := PurchaseButtonColor
                overlay_slot:
                    VerticalAlignment := vertical_alignment.Center
                    HorizontalAlignment := horizontal_alignment.Fill
                    Padding := margin{Top := 3.0, Bottom := 3.0, Left := 10.0, Right := 10.0}
                    Widget := PurchaseButton

        RewardsStack := stack_box :
            Orientation := orientation.Horizontal
            Slots :=
                array:
                    stack_box_slot:
                        #NOTE: We have to define the text in the block above so we can call it here
                        Widget := RewardImage1
                    stack_box_slot:
                        Widget := RewardImage2

        CostStack := stack_box :
            #Because this is orientated horizontally it will display next to the image
            Orientation := orientation.Horizontal
            Slots :=
                array:
                    stack_box_slot:
                        #TODO: Displaying the Cost Text next to the skull
                        #NOTE: We have to define the text in the block above so we can call it here
                        Widget := CostText
                    stack_box_slot:
                        VerticalAlignment := vertical_alignment.Center
                        HorizontalAlignment := horizontal_alignment.Center
                        Widget := texture_block:
                            DefaultImage := VerseExposed.T_Skull_01
                            #Added the size
                            DefaultDesiredSize := vector2{X:=64.0,Y:=64.0}
                    stack_box_slot:
                        VerticalAlignment := vertical_alignment.Center
                        HorizontalAlignment := horizontal_alignment.Center
                        Widget := CostAmountText
            
        StackBoxSlots :=
            array:
                #TODO: Testing [Removing the image we had in the middle - In this example the teleporter used to be here]
                <#
                stack_box_slot:
                    VerticalAlignment := vertical_alignment.Fill
                    HorizontalAlignment := horizontal_alignment.Fill
                    Padding := IconMargin
                    Widget := texture_block:
                        DefaultImage := InImage
                        DefaultDesiredSize := vector2{X:=512.0,Y:=256.0}
                #>
                #Will display the dungeon name
                stack_box_slot:
                    Widget := NameBlock 
                #Will display two images of rewards side by side
                stack_box_slot:
                    Widget := RewardBlock 
                stack_box_slot:
                    Widget := RewardsStack
                #Will display the Cost then a skull icon next to it
                stack_box_slot:
                    Widget := CostStack     
                #Description of the dungeon
                stack_box_slot:
                    Widget := DescriptionBlock
                #Display Purchase Button
                stack_box_slot:
                    Padding := margin{Top := 10.0}
                    VerticalAlignment := vertical_alignment.Fill
                    HorizontalAlignment := horizontal_alignment.Fill
                    Widget := ButtonOverlay

        #This should make a background image appear
        #TODO: Not appearing
        Overlay := overlay:
            Slots := array:
                overlay_slot:
                    VerticalAlignment := vertical_alignment.Fill
                    HorizontalAlignment := horizontal_alignment.Fill
                    Padding := margin{Bottom := 0.0, Left := 0.0, Right := 40.0, Top := 30.0} 
                    Widget := texture_block:
                        DefaultImage := InImage
                        DefaultDesiredSize := vector2{X:=1024.0,Y:=1024.0}
                
                overlay_slot:
                    Widget := stack_box:
                        Orientation := orientation.Vertical
                        Slots := StackBoxSlots

                <#
                    VerticalAlignment := vertical_alignment.Fill
                    HorizontalAlignment := horizontal_alignment.Fill
                    Widget := BackgroundColorBlock
                overlay_slot:
                    VerticalAlignment := vertical_alignment.Bottom
                    HorizontalAlignment := horizontal_alignment.Fill
                    Padding := margin{Top := 50.0}
                    Widget := overlay:
                        Slots := array:
                            overlay_slot:
                                Padding := margin{Top := 20.0, Bottom := 20.0}
                                VerticalAlignment := vertical_alignment.Center
                                HorizontalAlignment := horizontal_alignment.Fill
                                Widget := stack_box:
                                    Orientation := orientation.Vertical
                                    Slots := StackBoxSlots
                #4 Outlines
                overlay_slot:
                    VerticalAlignment := vertical_alignment.Top
                    HorizontalAlignment := horizontal_alignment.Fill
                    Widget := TopBorderColorBlock
                overlay_slot:
                    VerticalAlignment := vertical_alignment.Bottom
                    HorizontalAlignment := horizontal_alignment.Fill
                    Widget := BottomBorderColorBlock
                overlay_slot:
                    VerticalAlignment := vertical_alignment.Fill
                    HorizontalAlignment := horizontal_alignment.Right
                    Widget := RightBorderColorBlock
                overlay_slot:
                    VerticalAlignment := vertical_alignment.Fill
                    HorizontalAlignment := horizontal_alignment.Left
                    Widget := LeftBorderColorBlock
                #>

        return DungeonSlotData:
            Widget := Overlay
            Button := PurchaseButton
            ButtonColor := PurchaseButtonColor