using { /Fortnite.com/Devices }
using { /Verse.org/Simulation }
using { /UnrealEngine.com/Temporary/Diagnostics }
using { /UnrealEngine.com/Temporary/UI }
using { /Fortnite.com/UI }
using { /UnrealEngine.com/Temporary/SpatialMath}
using { /Verse.org/Colors }
using { /Verse.org/Assets }

UI_GnomeBuffDisplay := class():
    ToMessage<localizes>(String : string) : message = "{String}"
    CenteredAlignment : vector2 = vector2{X := 0.5, Y := 0.5}

    var BuiltCanvasMap : [player]AltarBuffDisplayCanvasData = map{}

    # Initiliase the UI so it can added to the player much quicker
    InitUI(InPlayer : player):void=
        # Format Timer
        TimeFormatted := FormattedDuration(0.0)
        AltarBuffDisplayCanvas := MakeAltarBuffCanvas(TimeFormatted)
        # Set the players build canvas data
        if(set BuiltCanvasMap[InPlayer] = AltarBuffDisplayCanvas) {}

    # Add the UI to the player
    AddUI(InPlayer : player):void=
        if(PlayerUI := GetPlayerUI[InPlayer], CanvasData := BuiltCanvasMap[InPlayer]):
            CanvasData.Canvas.SetVisibility(widget_visibility.Visible)
            PlayerUI.AddWidget(CanvasData.Canvas)
            

    HideUI(InPlayer: player)<suspends>:void=
        Sleep(0.0) # So we're sure that the canvas has been added to BuiltCanvas map
        if(PlayerUI := GetPlayerUI[InPlayer], CanvasData := BuiltCanvasMap[InPlayer]):
            CanvasData.Canvas.SetVisibility(widget_visibility.Hidden)
            PlayerUI.RemoveWidget(CanvasData.Canvas)
    
    # Upates a given player's UI
    UpdateUI(InPlayer : player,  InTime : float):void=
        if(PlayerUI := GetPlayerUI[InPlayer], CanvasData := BuiltCanvasMap[InPlayer]):
            # Format Timer then update
            TimeFormatted := FormattedDuration(InTime)
            CanvasData.TimeText.SetText(ToMessage("{TimeFormatted}"))

    # Make a singular resource slot
    MakeAltarBuffSlot(InImage : texture, InTime : string):ResourceSlotData=
        # Variables for other parts of UI
        # Use these to copy and paste the settings into overlays
        BackgroundColorBlock := color_block:
            DefaultColor := NamedColors.White
            DefaultOpacity := 0.2

        BorderSize := 1.0
        BorderColorBlock := color_block:
            DefaultColor := NamedColors.White
            DefaultDesiredSize := vector2{X := BorderSize, Y := BorderSize}
    
        CoinWidget := texture_block:
            DefaultImage := VerseExposed.T_Stone
        
        AmountText := text_block:
            DefaultText := ToMessage("{InTime}")
            DefaultJustification := text_justification.Left
            DefaultShadowColor := NamedColors.Black
            DefaultShadowOffset := option{vector2{X:=2.0,Y:=2.0}}
        AmountText.SetTextColor(NamedColors.White)
        AmountText.SetShadowOpacity(0.8)
        
        IconOverlay := overlay:
            Slots := array:
                overlay_slot:
                    VerticalAlignment := vertical_alignment.Center
                    HorizontalAlignment := horizontal_alignment.Center
                    Padding := margin{Top := -8.0, Bottom := -8.0, Left := -6.0, Right := -6.0}
                    Widget := color_block:
                        DefaultColor := NamedColors.Black
                        DefaultOpacity := 0.4
                        DefaultDesiredSize := vector2{X := 52.0, Y := 52.0}
                overlay_slot:
                    VerticalAlignment := vertical_alignment.Center
                    HorizontalAlignment := horizontal_alignment.Center
                    Widget := texture_block:
                        DefaultImage := InImage
        
        ResourceOverlay := overlay:
            Slots := array:
                overlay_slot:
                    VerticalAlignment := vertical_alignment.Center
                    HorizontalAlignment := horizontal_alignment.Left
                    Widget := color_block:
                        DefaultColor := NamedColors.Black
                        DefaultOpacity := 0.4
                        DefaultDesiredSize := vector2{X:=225.0,Y:=60.0}
                overlay_slot:
                    VerticalAlignment := vertical_alignment.Center
                    HorizontalAlignment := horizontal_alignment.Left
                    Padding := margin{Top := 10.0, Bottom := 10.0, Left := 10.0, Right := 10.0}
                    Widget := stack_box:
                        Orientation := orientation.Horizontal
                        Slots := 
                            array:
                                stack_box_slot:
                                    Widget := IconOverlay
                                    HorizontalAlignment := horizontal_alignment.Left
                                    VerticalAlignment := vertical_alignment.Center
                                stack_box_slot:
                                    Widget := AmountText
                                    Padding := margin{Left := 15.0}
                                    VerticalAlignment := vertical_alignment.Center
                                    HorizontalAlignment := horizontal_alignment.Left
        
        return ResourceSlotData:
            Widget := ResourceOverlay
            SlotText := AmountText

    # Make the Resource canvas and all widgets
    MakeAltarBuffCanvas(InTime : string):AltarBuffDisplayCanvasData=
        # Variables for other parts of UI
        # Use these to copy and paste the settings into overlays
        BackgroundColorBlock := color_block:
            DefaultColor := NamedColors.White
            DefaultOpacity := 0.2

        BorderSize := 1.0
        BorderColorBlock := color_block:
            DefaultColor := NamedColors.White
            DefaultDesiredSize := vector2{X := BorderSize, Y := BorderSize}
        
        # Make resource slots for all 5 resources
        TimeSlot := MakeAltarBuffSlot(VerseExposed.T_GnomeIcon, InTime)

        # Add the resource slots into a stack box to be displayed
        ResourcesBox :=
            array:
                stack_box_slot:
                    Widget := TimeSlot.Widget
                    Padding := margin{Top := 1.0, Bottom := 1.0}
                    VerticalAlignment := vertical_alignment.Center
                    HorizontalAlignment := horizontal_alignment.Left

        
        ResourcesOverlay := overlay:
            Slots := array:
                overlay_slot:
                    VerticalAlignment := vertical_alignment.Fill
                    HorizontalAlignment := horizontal_alignment.Fill
                    Widget := overlay:
                        Slots := array:
                            overlay_slot:
                                VerticalAlignment := vertical_alignment.Fill
                                HorizontalAlignment := horizontal_alignment.Fill
                                Widget := stack_box:
                                    Orientation := orientation.Vertical
                                    Slots := ResourcesBox
        OutputCanvas := canvas:
                    Slots := array:
                        canvas_slot:
                            Anchors := anchors{Minimum := vector2{X:=0.01, Y:=0.45}, Maximum := vector2{X:=0.20, Y:=0.65}} #vector2{X:=0.05, Y:=0.35}, Maximum := vector2{X:=0.20, Y:=0.55}}
                            Alignment := CenteredAlignment
                            Widget := ResourcesOverlay
                            SizeToContent := true

        return AltarBuffDisplayCanvasData:
            Canvas := OutputCanvas
            TimeText := TimeSlot.SlotText