
using { /Fortnite.com/Devices }
using { /Verse.org/Simulation }
using { /UnrealEngine.com/Temporary/Diagnostics }
using { /UnrealEngine.com/Temporary/UI }
using { /Fortnite.com/UI }
using { /UnrealEngine.com/Temporary/SpatialMath}
using { /Verse.org/Colors }
using { /Verse.org/Assets }

ItemStreamCanvasData := class:
    Canvas : canvas
    MainSlot : ItemSlotData
    var Shown : logic

    var FreeSlots : []ItemSlotStreamData

ItemSlotStreamData := class:
    Canvas : canvas
    MoveBox : color_block
    SlotText : text_block
    SlotImage : texture_block
    

ItemSlotData := class:
    Widget : widget
    SlotText : text_block
    SlotImage : texture_block
    var ResourceName : string 
    var CurrentAmount : big_number

UI_ItemStreamDisplay := class():
    ToMessage<localizes>(String : string) : message = "{String}"
    CenteredAlignment : vector2 = vector2{X := 0.5, Y := 0.5}

    var BuiltCanvasMap : [player]ItemStreamCanvasData = map{}
    var ResetRemoveEventMap : [player]event() = map{}
    
    # Initiliase the UI so it can added to the player much quicker
    InitUI(InPlayer : player):void=
        ItemStreamDisplayCanvas := MakeItemStreamCanvas("Name", VerseExposed.T_Stone, big_number{})
        # Set the players build canvas data
        if(set BuiltCanvasMap[InPlayer] = ItemStreamDisplayCanvas, set ResetRemoveEventMap[InPlayer] = event(){}) {}

    # Add the UI to the player
    AddUI(InPlayer : player):void=
        if(PlayerUI := GetPlayerUI[InPlayer], CanvasData := BuiltCanvasMap[InPlayer]):
            if(CanvasData.Shown = false):
                CanvasData.Canvas.SetVisibility(widget_visibility.Visible)
                PlayerUI.AddWidget(CanvasData.Canvas)
                set CanvasData.Shown = true

    # Hides the UI for the given player
    HideUI(InPlayer: player)<suspends>:void=
        Sleep(0.0) # So we're sure that the canvas has been added to BuiltCanvas map
        if(PlayerUI := GetPlayerUI[InPlayer], CanvasData := BuiltCanvasMap[InPlayer]):
            if(CanvasData.Shown = true):
                CanvasData.Canvas.SetVisibility(widget_visibility.Hidden)
                PlayerUI.RemoveWidget(CanvasData.Canvas)
                set CanvasData.MainSlot.CurrentAmount = big_number{}
                set CanvasData.Shown = false

    # Add an item to the stream
    AddToStream(InPlayer: player, InResourceName : string, InImage : texture, InAmount : big_number, IsCrit : logic, IsLastHit : logic):void=
        if(CanvasData := BuiltCanvasMap[InPlayer]):
            if(PlayerRemoveEvent := ResetRemoveEventMap[InPlayer]):
                PlayerRemoveEvent.Signal()
            AddUI(InPlayer)
            var NewAmount : big_number = big_number{}
            if(InResourceName <> CanvasData.MainSlot.ResourceName):
                set CanvasData.MainSlot.CurrentAmount = big_number{}
                set CanvasData.MainSlot.ResourceName = InResourceName
                CanvasData.MainSlot.SlotImage.SetImage(InImage)

            set NewAmount = BigNumber_Add(CanvasData.MainSlot.CurrentAmount, InAmount)
            

            set CanvasData.MainSlot.CurrentAmount = NewAmount
            CanvasData.MainSlot.SlotText.SetText(ToMessage("{BigNumber_FormatDisplay(CanvasData.MainSlot.CurrentAmount)}"))
        if(IsLastHit = true):
            spawn:
                AddMovingItemSlot(InPlayer, InImage, InAmount, 4.0, NamedColors.Orange)
        else if(IsCrit = true):
            spawn:
                AddMovingItemSlot(InPlayer, VerseExposed.T_CritHitIcon, InAmount, 4.0, NamedColors.Red)
        else:
            spawn:
                AddMovingItemSlot(InPlayer, InImage, InAmount, 4.0, NamedColors.White)
        spawn:
            RemoveUIOnTimer(InPlayer, 3.0)

    AddMovingItemSlot(InPlayer : player, InImage : texture, InAmount : big_number, InIncrement : float, TextColor : color)<suspends>:void=
        if(PlayerUI := GetPlayerUI[InPlayer], CanvasData := BuiltCanvasMap[InPlayer]):
            if(FirstSlot := CanvasData.FreeSlots[0]):
                FirstSlot.MoveBox.SetDesiredSize(vector2{X:=275.0,Y:=60.0})
                FirstSlot.SlotImage.SetImage(InImage)
                FirstSlot.SlotText.SetText(ToMessage("+ {BigNumber_FormatDisplay(InAmount)}"))
                FirstSlot.SlotText.SetTextColor(TextColor)
                # Remove from array until finished being used
                if(NewSlots := CanvasData.FreeSlots.RemoveElement[0]):
                    set CanvasData.FreeSlots = NewSlots
                FirstSlot.Canvas.SetVisibility(widget_visibility.Visible)
                PlayerUI.AddWidget(FirstSlot.Canvas)
                FirstSlot.SlotText.SetTextOpacity(1.0)
                FirstSlot.SlotText.SetShadowOpacity(1.0)
                var Increment : float = InIncrement
                for(X := 0..100):
                    NewSize := vector2{X:=275.0,Y:=60.0 + Increment}
                    FirstSlot.MoveBox.SetDesiredSize(NewSize)
                    TextOpacity := FirstSlot.SlotText.GetTextOpacity()
                    if(ClampedFloat := type{_X:float where 0.000000 <= _X, _X <= 1.000000}[TextOpacity - 0.01]):
                        FirstSlot.SlotText.SetTextOpacity(ClampedFloat)
                        FirstSlot.SlotText.SetShadowOpacity(ClampedFloat)
                    Sleep(0.02)
                    set Increment += InIncrement
                FirstSlot.Canvas.SetVisibility(widget_visibility.Hidden)
                PlayerUI.RemoveWidget(FirstSlot.Canvas)
                set CanvasData.FreeSlots += array{FirstSlot}
                
    # Removes the UI when the sleep succeeds
    RemoveUIOnTimer(InPlayer : player, Time : float)<suspends>:void=
        if(PlayerRemoveEvent := ResetRemoveEventMap[InPlayer]):
            race:
                PlayerRemoveEvent.Await()
                block:
                    Sleep(Time)
                    HideUI(InPlayer)
        
    # Make an ItemSlot that will update
    MakeItemStreamSlot(InImage : texture, InAmount : big_number):ItemSlotStreamData=

        ImageWidget := texture_block:
            DefaultImage := InImage
        
        AmountText := text_block:
            DefaultText := ToMessage("+ {BigNumber_FormatDisplay(InAmount)}")
            DefaultJustification := text_justification.Left
            DefaultShadowColor := NamedColors.Black
            DefaultShadowOffset := option{vector2{X:=2.0,Y:=2.0}}
        AmountText.SetTextColor(NamedColors.White)
        AmountText.SetShadowOpacity(0.8)

        MoveBox := color_block:
            DefaultColor := NamedColors.Black
            DefaultOpacity := 0.0
            DefaultDesiredSize := vector2{X:=275.0,Y:=60.0}
        
        IconOverlay := overlay:
            Slots := array:
                overlay_slot:
                    VerticalAlignment := vertical_alignment.Center
                    HorizontalAlignment := horizontal_alignment.Center
                    Widget := ImageWidget
        
        OverlayStackBox := stack_box:
            Orientation := orientation.Horizontal
            Slots := 
                array:
                    stack_box_slot:
                        Widget := IconOverlay
                        HorizontalAlignment := horizontal_alignment.Left
                        VerticalAlignment := vertical_alignment.Center
                    stack_box_slot:
                        Widget := AmountText
                        Padding := margin{Left := 15.0}
                        VerticalAlignment := vertical_alignment.Center
                        HorizontalAlignment := horizontal_alignment.Left

        OverlayWithBox := overlay:
            Slots := array:
                overlay_slot:
                    VerticalAlignment := vertical_alignment.Center
                    HorizontalAlignment := horizontal_alignment.Center
                    Widget := MoveBox
                overlay_slot:
                    VerticalAlignment := vertical_alignment.Top
                    HorizontalAlignment := horizontal_alignment.Center
                    Widget := OverlayStackBox

        OutputCanvas := canvas:
            Slots := array:
                canvas_slot:
                    Anchors := anchors{Minimum := vector2{X:=0.65, Y:=0.66}, Maximum := vector2{X:=0.65, Y:=0.66}}
                    Alignment := CenteredAlignment
                    Widget := OverlayWithBox
                    SizeToContent := true

        return ItemSlotStreamData:
            Canvas := OutputCanvas
            MoveBox := MoveBox
            SlotText := AmountText
            SlotImage := ImageWidget

    # Makes the main display that shows how much total you have gained
    MakeItemDisplaySlot(InResourceName : string, InImage : texture, InAmount : big_number):ItemSlotData=
        # Variables for other parts of UI
        # Use these to copy and paste the settings into overlays
        BackgroundColorBlock := color_block:
            DefaultColor := NamedColors.White
            DefaultOpacity := 0.2

        BorderSize := 1.0
        BorderColorBlock := color_block:
            DefaultColor := NamedColors.White
            DefaultDesiredSize := vector2{X := BorderSize, Y := BorderSize}
    
        ImageWidget := texture_block:
            DefaultImage := InImage
        
        AmountText := text_block:
            DefaultText := ToMessage("{BigNumber_FormatDisplay(InAmount)}")
            DefaultJustification := text_justification.Left
            DefaultShadowColor := NamedColors.Black
            DefaultShadowOffset := option{vector2{X:=2.0,Y:=2.0}}
        AmountText.SetTextColor(NamedColors.White)
        AmountText.SetShadowOpacity(0.8)
        
        IconOverlay := overlay:
            Slots := array:
                overlay_slot:
                    VerticalAlignment := vertical_alignment.Center
                    HorizontalAlignment := horizontal_alignment.Center
                    Padding := margin{Top := -8.0, Bottom := -8.0, Left := -6.0, Right := -6.0}
                    Widget := color_block:
                        DefaultColor := NamedColors.Black
                        DefaultOpacity := 0.4
                        DefaultDesiredSize := vector2{X := 52.0, Y := 52.0}
                overlay_slot:
                    VerticalAlignment := vertical_alignment.Center
                    HorizontalAlignment := horizontal_alignment.Center
                    Widget := ImageWidget
        
        ResourceOverlay := overlay:
            Slots := array:
                overlay_slot:
                    VerticalAlignment := vertical_alignment.Center
                    HorizontalAlignment := horizontal_alignment.Left
                    Widget := color_block:
                        DefaultColor := NamedColors.Black
                        DefaultOpacity := 0.4
                        DefaultDesiredSize := vector2{X:=225.0,Y:=60.0}
                overlay_slot:
                    VerticalAlignment := vertical_alignment.Center
                    HorizontalAlignment := horizontal_alignment.Left
                    Padding := margin{Top := 10.0, Bottom := 10.0, Left := 10.0, Right := 10.0}
                    Widget := stack_box:
                        Orientation := orientation.Horizontal
                        Slots := 
                            array:
                                stack_box_slot:
                                    Widget := IconOverlay
                                    HorizontalAlignment := horizontal_alignment.Left
                                    VerticalAlignment := vertical_alignment.Center
                                stack_box_slot:
                                    Widget := AmountText
                                    Padding := margin{Left := 15.0}
                                    VerticalAlignment := vertical_alignment.Center
                                    HorizontalAlignment := horizontal_alignment.Left
        
        return ItemSlotData:
            Widget := ResourceOverlay
            SlotText := AmountText
            SlotImage := ImageWidget
            ResourceName := InResourceName
            CurrentAmount := InAmount

    # Make the overall canvas for the 
    MakeItemStreamCanvas(InResourceName : string, InImage : texture, InAmount : big_number):ItemStreamCanvasData=
        MainSlotData := MakeItemDisplaySlot(InResourceName, InImage, InAmount)
        var FreeSlots : []ItemSlotStreamData = array{}
        for(Index := 0..19):
            set FreeSlots += array{MakeItemStreamSlot(InImage, InAmount)}

        OutputCanvas := canvas:
            Slots := array:
                canvas_slot:
                    Anchors := anchors{Minimum := vector2{X:=0.65, Y:=0.7}, Maximum := vector2{X:=0.65, Y:=0.7}}
                    Alignment := CenteredAlignment
                    Widget := MainSlotData.Widget
                    SizeToContent := true

        return ItemStreamCanvasData:
            Canvas := OutputCanvas
            MainSlot := MainSlotData
            FreeSlots := FreeSlots
            Shown := false
