
using { /Fortnite.com/Devices }
using { /Verse.org/Simulation }
using { /UnrealEngine.com/Temporary/Diagnostics }
using { /UnrealEngine.com/Temporary/UI }
using { /Fortnite.com/UI }
using { /UnrealEngine.com/Temporary/SpatialMath}
using { /Verse.org/Colors }
using { /Verse.org/Assets }

MakeNodeHealthBarData<constructor>(InNameText : string, InHealthText : string, InHealthPercentage : float, InPosition : vector3, IsBoss : logic):=NodeHealthBarData:
    NameText := InNameText
    HealthText := InHealthText
    HealthPercentage := InHealthPercentage
    Position := InPosition
    IsBoss := IsBoss

NodeHealthBarCanvasData := class:
    Canvas : canvas
    NameText : text_block
    HealthText : text_block
    BiomeBossText : text_block

    var Shown : logic

    var ResourceName : string

    var HealthPercentage : float

    HealthBlock : color_block
    ChaserBlock : color_block

    var PreviousPosition : vector3

NodeHealthBarData := class:
    NameText : string
    HealthText : string
    IsBoss : logic

    HealthPercentage : float

    Position : vector3

UI_NodeHealthBar := class():

    ToMessage<localizes>(String : string) : message = "{String}"
    CenteredAlignment : vector2 = vector2{X := 0.5, Y := 0.5}

    var BuiltCanvasMap : [player]NodeHealthBarCanvasData = map{}
    var ResetRemoveEventMap : [player]event() = map{}
    var UpdateEventMap : [player]event() = map{}


    # Initiliase the UI so it can added to the player much quicker
    InitUI(InPlayer : player):void=
        TempNodeHealthBarData := MakeNodeHealthBarData("Name", "100 / 100", 100.0, vector3{X:=0.0,Y:=0.0,Z:=0.0}, false)
        NodeHealthBarCanvas := MakeHealthBarCanvas(TempNodeHealthBarData)
        # Set the players build canvas data
        if(set BuiltCanvasMap[InPlayer] = NodeHealthBarCanvas, set ResetRemoveEventMap[InPlayer] = event(){}, set UpdateEventMap[InPlayer] = event(){}) {}

    # Add the UI to the player
    AddUI(InPlayer : player):void=
        if(PlayerUI := GetPlayerUI[InPlayer], CanvasData := BuiltCanvasMap[InPlayer]):
            if(CanvasData.Shown = false):
                CanvasData.Canvas.SetVisibility(widget_visibility.Visible)
                PlayerUI.AddWidget(CanvasData.Canvas)
                set CanvasData.Shown = true
            

    # Hides the UI for the given player
    HideUI(InPlayer: player)<suspends>:void=
        Sleep(0.0) # So we're sure that the canvas has been added to BuiltCanvas map
        if(PlayerUI := GetPlayerUI[InPlayer], CanvasData := BuiltCanvasMap[InPlayer]):
            if(CanvasData.Shown = true):
                CanvasData.Canvas.SetVisibility(widget_visibility.Hidden)
                PlayerUI.RemoveWidget(CanvasData.Canvas)
                set CanvasData.HealthPercentage = 100.0
                set CanvasData.Shown = false

    UpdateUI(InPlayer : player, InNodeHealthBarData : NodeHealthBarData):void=
        if(PlayerUpdateEvent := UpdateEventMap[InPlayer], PlayerRemoveEvent := ResetRemoveEventMap[InPlayer],CanvasData := BuiltCanvasMap[InPlayer]):
            PlayerUpdateEvent.Signal()
            PlayerRemoveEvent.Signal()
            if(CanvasData.ResourceName <> InNodeHealthBarData.NameText or CanvasData.PreviousPosition.X <> InNodeHealthBarData.Position.X or CanvasData.PreviousPosition.Y <> InNodeHealthBarData.Position.Y or CanvasData.PreviousPosition.Z <> InNodeHealthBarData.Position.Z):
                set CanvasData.HealthPercentage = 100.0
                set CanvasData.ResourceName = InNodeHealthBarData.NameText
                set CanvasData.PreviousPosition = InNodeHealthBarData.Position
            CanvasData.NameText.SetText(ToMessage("{InNodeHealthBarData.NameText}"))
            CanvasData.HealthText.SetText(ToMessage(InNodeHealthBarData.HealthText))
            if(InNodeHealthBarData.IsBoss = true):
                CanvasData.BiomeBossText.SetText(ToMessage("Boss:"))
            else:
                CanvasData.BiomeBossText.SetText(ToMessage("Resource:"))
            spawn:
                UpdateOverTime(InPlayer, InNodeHealthBarData)
            spawn:
                RemoveUIOnTimer(InPlayer)

    #Allows custom Boss functionality
    UpdateBossUI(InPlayer : player, InNodeHealthBarData : NodeHealthBarData):void=
        if(PlayerUpdateEvent := UpdateEventMap[InPlayer], <#PlayerRemoveEvent := ResetRemoveEventMap[InPlayer],#>CanvasData := BuiltCanvasMap[InPlayer]):
            PlayerUpdateEvent.Signal()
            #PlayerRemoveEvent.Signal()
            if(CanvasData.ResourceName <> InNodeHealthBarData.NameText or CanvasData.PreviousPosition.X <> InNodeHealthBarData.Position.X or CanvasData.PreviousPosition.Y <> InNodeHealthBarData.Position.Y or CanvasData.PreviousPosition.Z <> InNodeHealthBarData.Position.Z):
                set CanvasData.HealthPercentage = 100.0
                set CanvasData.ResourceName = InNodeHealthBarData.NameText
                set CanvasData.PreviousPosition = InNodeHealthBarData.Position
            CanvasData.NameText.SetText(ToMessage("{InNodeHealthBarData.NameText}"))
            CanvasData.HealthText.SetText(ToMessage(InNodeHealthBarData.HealthText))
            if(InNodeHealthBarData.IsBoss = true):
                CanvasData.BiomeBossText.SetText(ToMessage("Boss:"))
            else:
                CanvasData.BiomeBossText.SetText(ToMessage("Biome:"))
            spawn:
                UpdateOverTime(InPlayer, InNodeHealthBarData)

    # Updates the health bar over time with lerps
    UpdateOverTime(InPlayer : player, InNodeHealthBarData : NodeHealthBarData)<suspends>:void=
        if(PlayerUpdateEvent := UpdateEventMap[InPlayer], CanvasData := BuiltCanvasMap[InPlayer]):
            race:
                PlayerUpdateEvent.Await()
                block:
                    
                    var OldHealthPercentage : float = CanvasData.HealthPercentage
                    var NewHealthPercentage : float = InNodeHealthBarData.HealthPercentage
                    var OldChaserPercentage : float = CanvasData.HealthPercentage
                    set CanvasData.HealthPercentage = InNodeHealthBarData.HealthPercentage
                    loop:
                        if(not(NewHealthPercentage = OldHealthPercentage)):
                            if(NewHealthPercentage < OldHealthPercentage):
                                set OldHealthPercentage = Lerp(OldHealthPercentage, NewHealthPercentage, 0.4)
                                set OldChaserPercentage = Lerp(OldChaserPercentage, OldHealthPercentage, 0.2)
                        
                        BarWidth := 600.0
                        HealthBarWidthPercentage := BarWidth - (BarWidth * ((100.0 - OldHealthPercentage) * 0.01))
                        ChaserBarWidthPercentage := BarWidth - (BarWidth * ((100.0 - OldChaserPercentage) * 0.01))
                        CanvasData.HealthBlock.SetDesiredSize(vector2{X:=HealthBarWidthPercentage ,Y:=42.0})
                        CanvasData.ChaserBlock.SetDesiredSize(vector2{X:=ChaserBarWidthPercentage,Y:=42.0})
                        # Ends loop when OldHealthPercentage is within margin of error
                        if(OldChaserPercentage - NewHealthPercentage <= 0.05):
                            return
                        
                        Sleep(0.01)

    # Removes the UI when the sleep succeeds
    RemoveUIOnTimer(InPlayer : player)<suspends>:void=
        if(PlayerRemoveEvent := ResetRemoveEventMap[InPlayer]):
            race:
                PlayerRemoveEvent.Await()
                block:
                    Sleep(0.75)
                    HideUI(InPlayer)

    MakeHealthBarCanvas(InNodeHealthBarData : NodeHealthBarData):NodeHealthBarCanvasData=
        # Variables for other parts of UI
        # Use these to copy and paste the settings into overlays
        BackgroundColorBlock := color_block:
            DefaultColor := NamedColors.Black
            DefaultOpacity := 0.2
        BorderSize := 1.0
        BorderColorBlock := color_block:
            DefaultColor := NamedColors.White
            DefaultDesiredSize := vector2{X := BorderSize, Y := BorderSize}
        BarWidth := 600.0
        
        BarShadingBottomColorBlock := color_block:
            DefaultColor := NamedColors.Black
            DefaultOpacity := 0.4
            DefaultDesiredSize := vector2{X:= BarWidth,Y:=21.0}
        BarShadingTopColorBlock := color_block:
            DefaultColor := NamedColors.White
            DefaultOpacity := 0.2
            DefaultDesiredSize := vector2{X:= BarWidth,Y:=21.0}

        # Updating Color Blocks
        BarWidthPercentage := BarWidth * (InNodeHealthBarData.HealthPercentage * 0.01)

        HealthBarColorBlock := color_block:
            DefaultColor := NamedColors.Red
            DefaultDesiredSize := vector2{X:= BarWidthPercentage,Y:=42.0}
        ChaserColorBlock := color_block:
            DefaultColor := NamedColors.LightGray
            DefaultDesiredSize := vector2{X:= BarWidthPercentage,Y:=42.0}

        # Updating Text Blocks
        HealthText := text_block:
            DefaultText := ToMessage("{InNodeHealthBarData.HealthText}")
            DefaultJustification := text_justification.Left
            DefaultShadowColor := NamedColors.Black
            DefaultShadowOffset := option{vector2{X:=2.0,Y:=2.0}}
        HealthText.SetTextColor(NamedColors.White)
        HealthText.SetShadowOpacity(0.8)
        NameText := text_block:
            DefaultText := ToMessage("{InNodeHealthBarData.NameText}")
            DefaultJustification := text_justification.Left
            DefaultShadowColor := NamedColors.Black
            DefaultShadowOffset := option{vector2{X:=2.0,Y:=2.0}}
        NameText.SetTextColor(NamedColors.White)
        NameText.SetShadowOpacity(0.8)
        OreNameText := text_block:
            DefaultText := ToMessage("Biome:")
            DefaultJustification := text_justification.Left
            DefaultShadowColor := NamedColors.Black
            DefaultShadowOffset := option{vector2{X:=2.0,Y:=2.0}}
        OreNameText.SetTextColor(NamedColors.DarkRed)
        OreNameText.SetShadowOpacity(0.8)

        # Overlays
        NameStackBox := stack_box:
            Orientation := orientation.Horizontal
            Slots := array:
                stack_box_slot:
                    Widget := OreNameText
                    Padding := margin{Right := 5.0}
                    HorizontalAlignment := horizontal_alignment.Center
                    VerticalAlignment := vertical_alignment.Center
                stack_box_slot:
                    Widget := NameText
                    HorizontalAlignment := horizontal_alignment.Center
                    VerticalAlignment := vertical_alignment.Center
                    
        NameOverlay := overlay:
            Slots := array:
                overlay_slot:
                    VerticalAlignment := vertical_alignment.Center
                    HorizontalAlignment := horizontal_alignment.Center
                    Widget := NameStackBox

        HealthOverlay := overlay:
            Slots := array:
                overlay_slot:
                    VerticalAlignment := vertical_alignment.Center
                    HorizontalAlignment := horizontal_alignment.Center
                    Widget := HealthText

        HealthBarOverlay := overlay:
            Slots := array:
                overlay_slot:
                    VerticalAlignment := vertical_alignment.Center
                    HorizontalAlignment := horizontal_alignment.Center
                    Widget := color_block:
                        DefaultColor := NamedColors.Black
                        DefaultOpacity := 0.2
                        DefaultDesiredSize := vector2{X:= BarWidth,Y:=42.0}
                overlay_slot:
                    VerticalAlignment := vertical_alignment.Center
                    HorizontalAlignment := horizontal_alignment.Left
                    Widget := ChaserColorBlock
                overlay_slot:
                    VerticalAlignment := vertical_alignment.Center
                    HorizontalAlignment := horizontal_alignment.Left
                    Widget := HealthBarColorBlock
                overlay_slot:
                    VerticalAlignment := vertical_alignment.Bottom
                    HorizontalAlignment := horizontal_alignment.Left
                    Widget := BarShadingBottomColorBlock
                overlay_slot:
                    VerticalAlignment := vertical_alignment.Top
                    HorizontalAlignment := horizontal_alignment.Left
                    Widget := BarShadingTopColorBlock
                overlay_slot:
                    VerticalAlignment := vertical_alignment.Center
                    HorizontalAlignment := horizontal_alignment.Center
                    Widget := HealthText
                overlay_slot:
                    VerticalAlignment := vertical_alignment.Top
                    HorizontalAlignment := horizontal_alignment.Fill
                    Widget := color_block:
                        DefaultColor := NamedColors.Black
                        DefaultDesiredSize := vector2{X := BorderSize, Y := BorderSize}
                overlay_slot:
                    VerticalAlignment := vertical_alignment.Bottom
                    HorizontalAlignment := horizontal_alignment.Fill
                    Widget := color_block:
                        DefaultColor := NamedColors.Black
                        DefaultDesiredSize := vector2{X := BorderSize, Y := BorderSize}
                overlay_slot:
                    VerticalAlignment := vertical_alignment.Fill
                    HorizontalAlignment := horizontal_alignment.Right
                    Widget := color_block:
                        DefaultColor := NamedColors.Black
                        DefaultDesiredSize := vector2{X := BorderSize, Y := BorderSize}
                overlay_slot:
                    VerticalAlignment := vertical_alignment.Fill
                    HorizontalAlignment := horizontal_alignment.Left
                    Widget := color_block:
                        DefaultColor := NamedColors.Black
                        DefaultDesiredSize := vector2{X := BorderSize, Y := BorderSize}

        HealthBarStackBox := stack_box:
            Orientation := orientation.Vertical
            Slots := array:
                    stack_box_slot:
                        Widget := NameOverlay
                        HorizontalAlignment := horizontal_alignment.Center
                        VerticalAlignment := vertical_alignment.Center
                    stack_box_slot:
                        Widget := HealthBarOverlay
                        HorizontalAlignment := horizontal_alignment.Center
                        VerticalAlignment := vertical_alignment.Center

        OutputCanvas := canvas:
            Slots := array:
                canvas_slot:
                    Anchors := anchors{Minimum := vector2{X:=0.5, Y:=0.8}, Maximum := vector2{X:=0.5, Y:=0.8}}
                    Alignment := CenteredAlignment
                    Widget := HealthBarStackBox
                    SizeToContent := true

        return NodeHealthBarCanvasData:
            Canvas := OutputCanvas
            NameText := NameText
            BiomeBossText := OreNameText
            HealthText := HealthText
            ResourceName := InNodeHealthBarData.NameText
            Shown := false
            HealthPercentage := InNodeHealthBarData.HealthPercentage
            HealthBlock := HealthBarColorBlock
            ChaserBlock := ChaserColorBlock
            PreviousPosition := vector3{X:=0.0,Y:=0.0,Z:=0.0}
