using { /Fortnite.com/Devices }
using { /Verse.org/Simulation }
using { /UnrealEngine.com/Temporary/Diagnostics }
using { /UnrealEngine.com/Temporary/UI }
using { /Fortnite.com/UI }
using { /UnrealEngine.com/Temporary/SpatialMath}
using { /Verse.org/Colors }
using { /Verse.org/Assets }

OpenButtonHandler := class(){
    UIPlayerMenu : UI_PlayerMenu

    HandlerFunction(WidgetMessage : widget_message) : void = {
            Player := WidgetMessage.Player
            spawn:
                UIPlayerMenu.HideUI(Player)
    }
}

TeleportHUBHandler := class() {
    UIPlayerMenu : UI_PlayerMenu
    Teleporter : teleporter_device

    HandlerFunction(WidgetMessage : widget_message) : void = {
        Player := WidgetMessage.Player
        if(CanvasData := UIPlayerMenu.BuiltCanvasMap[Player]):
            if(CanvasData.CanTeleport = true):
                UIPlayerMenu.HandlePlayerTeleportHUB(WidgetMessage, Teleporter)
    }
}

# SHOWS THE PLAYER STATS MENU
OpenPlayerStatsHandler := class(){
    UIPlayerMenu : UI_PlayerMenu
    PlayerStatsManager : player_stats_manager
    Achievements : []achievement_data

    HandlerFunction(WidgetMessage : widget_message) : void = {
            Player := WidgetMessage.Player
            if(CanvasData := UIPlayerMenu.BuiltCanvasMap[Player]):
                CanvasData.CurrentOpenWidget.SetVisibility(widget_visibility.Collapsed)
                CanvasData.PlayerStatsMenu.Widget.SetVisibility(widget_visibility.Visible)
                set CanvasData.CurrentOpenWidget = CanvasData.PlayerStatsMenu.Widget
                CanvasData.Background.SetImage(VerseExposed.T_PlayerMenuStats)
                UIPlayerMenu.UpdateUI(Player, PlayerStatsManager, Achievements)
    }
}
OpenInventoryHandler := class(){
    UIPlayerMenu : UI_PlayerMenu
    PlayerStatsManager : player_stats_manager
    Achievements : []achievement_data

    HandlerFunction(WidgetMessage : widget_message) : void = {
            Player := WidgetMessage.Player
            if(CanvasData := UIPlayerMenu.BuiltCanvasMap[Player]):
                CanvasData.CurrentOpenWidget.SetVisibility(widget_visibility.Collapsed)
                CanvasData.Inventory.Widget.SetVisibility(widget_visibility.Visible)
                set CanvasData.CurrentOpenWidget = CanvasData.Inventory.Widget
                CanvasData.Background.SetImage(VerseExposed.T_PlayerMenuInventory)

                UIPlayerMenu.UpdateUI(Player, PlayerStatsManager, Achievements)
    }
}
OpenAchievementsHandler := class(){
    UIPlayerMenu : UI_PlayerMenu
    PlayerStatsManager : player_stats_manager
    Achievements : []achievement_data

    HandlerFunction(WidgetMessage : widget_message) : void = {
            Player := WidgetMessage.Player
            if(CanvasData := UIPlayerMenu.BuiltCanvasMap[Player]):
                CanvasData.CurrentOpenWidget.SetVisibility(widget_visibility.Collapsed)
                CanvasData.Achievements.OverallWidget.SetVisibility(widget_visibility.Visible)
                set CanvasData.CurrentOpenWidget = CanvasData.Achievements.OverallWidget
                CanvasData.Background.SetImage(VerseExposed.T_PlayerMenuAchievements)
                UIPlayerMenu.UpdateUI(Player, PlayerStatsManager, Achievements)
    }
}
OpenAchievementsSwitchHandler := class(){
    UIPlayerMenu : UI_PlayerMenu
    PlayerStatsManager : player_stats_manager
    Achievements : []achievement_data
    Type : AchievementType

    HandlerFunction(WidgetMessage : widget_message) : void = {
        Player := WidgetMessage.Player
        if(CanvasData := UIPlayerMenu.BuiltCanvasMap[Player]):
            CanvasData.Achievements.CurrentWidget.SetVisibility(widget_visibility.Hidden)
            case(Type):
                AchievementType.TotalStone =>
                    CanvasData.Achievements.TotalStone.SetVisibility(widget_visibility.Visible)
                    set CanvasData.Achievements.CurrentWidget = CanvasData.Achievements.TotalStone
                AchievementType.HighestStrength =>
                    CanvasData.Achievements.HeighestStrength.SetVisibility(widget_visibility.Visible)
                    set CanvasData.Achievements.CurrentWidget = CanvasData.Achievements.HeighestStrength
                AchievementType.RocksDestroyed =>
                    CanvasData.Achievements.RocksDestroyed.SetVisibility(widget_visibility.Visible)
                    set CanvasData.Achievements.CurrentWidget = CanvasData.Achievements.RocksDestroyed
                AchievementType.ArtifactsSmelted =>
                    CanvasData.Achievements.ArtifactsSmelted.SetVisibility(widget_visibility.Visible)
                    set CanvasData.Achievements.CurrentWidget = CanvasData.Achievements.ArtifactsSmelted
                AchievementType.DungeonsCompleted =>
                    CanvasData.Achievements.DungeonCompleted.SetVisibility(widget_visibility.Visible)
                    set CanvasData.Achievements.CurrentWidget = CanvasData.Achievements.DungeonCompleted
                AchievementType.Null =>
            set CanvasData.Achievements.Type = Type
            
        UIPlayerMenu.UpdateUI(Player, PlayerStatsManager, Achievements)
    }
}
OpenEquipArtifactHandler := class(){
    UIPlayerMenu : UI_PlayerMenu
    PlayerStatsManager : player_stats_manager
    Achievements : []achievement_data

    HandlerFunction(WidgetMessage : widget_message) : void = {
            Player := WidgetMessage.Player
            if(CanvasData := UIPlayerMenu.BuiltCanvasMap[Player]):
                CanvasData.CurrentOpenWidget.SetVisibility(widget_visibility.Collapsed)
                CanvasData.ArtifactEquip.Widget.SetVisibility(widget_visibility.Visible)
                set CanvasData.CurrentOpenWidget = CanvasData.ArtifactEquip.Widget
                CanvasData.Background.SetImage(VerseExposed.T_PlayerMenuArtifacts)
                UIPlayerMenu.UpdateUI(Player, PlayerStatsManager, Achievements)
    }
}

OpenPrestigeButtonHandler := class(){
    UIPlayerMenu : UI_PlayerMenu
    Prestige : V_PrestigeFunctionality

    HandlerFunction(WidgetMessage : widget_message) : void = {
            Player := WidgetMessage.Player
            spawn:
                UIPlayerMenu.HideUI(Player)
            Prestige.OpenPrestigeMenu(Player)
    }
}

OpenCardButtonHandler := class(){
    UIPlayerMenu : UI_PlayerMenu
    Card : V_CardManager

    HandlerFunction(WidgetMessage : widget_message) : void = {
            Player := WidgetMessage.Player
            spawn:
                UIPlayerMenu.HideUI(Player)
            Card.OpenCardMenu(Player)
    }
}

AltarButtonHandler := class(){
    UIPlayerMenu : UI_PlayerMenu
    Device : V_OfferingAltar

    HandlerFunction(WidgetMessage : widget_message) : void = {
            Player := WidgetMessage.Player
            spawn:
                UIPlayerMenu.HideUI(Player)
            Device.ActivateButton(Player)
    }
}

EquipHandler := class() {
    Device : V_ArtifactManager
    UIPlayerMenu : UI_PlayerMenu
    PlayerStatsManager : player_stats_manager
    Achievements : []achievement_data

    HandlerFunction(WidgetMessage : widget_message) : void = {
        Player := WidgetMessage.Player
        if(CanvasData := UIPlayerMenu.BuiltCanvasMap[Player]):
            Device.EquipArtifact(Player, CanvasData.ArtifactEquip.SlotSelected)
            UIPlayerMenu.UpdateUI(Player, PlayerStatsManager, Achievements)
    }
}

SelectSlotHandler := class(){
    Slot : int
    UIPlayerMenu : UI_PlayerMenu
    PlayerStatsManager : player_stats_manager
    Achievements : []achievement_data

    HandlerFunction(WidgetMessage : widget_message) : void = {
        Player := WidgetMessage.Player
        if(CanvasData := UIPlayerMenu.BuiltCanvasMap[Player]):
            if(CanvasData.ArtifactEquip.ArtifactSlots[Slot - 1].IsLocked = false):
                set CanvasData.ArtifactEquip.SlotSelected = Slot
                Print("SET SLOT : {Slot}")
                UIPlayerMenu.UpdateUI(Player, PlayerStatsManager, Achievements)
    }
}

ChangeBiomeHandler := class(){
    Next : logic
    UIPlayerMenu : UI_PlayerMenu
    PlayerStatsManager : player_stats_manager
    Achievements : []achievement_data

    HandlerFunction(WidgetMessage : widget_message) : void = {
            Player := WidgetMessage.Player
            if(CanvasData := UIPlayerMenu.BuiltCanvasMap[Player]):
                if(CanvasData.ArtifactEquip.BiomeSelected = Biome.Mushroom):
                    if(Next = true):
                        set CanvasData.ArtifactEquip.BiomeSelected = Biome.Lava
                    else:
                        set CanvasData.ArtifactEquip.BiomeSelected = Biome.Underworld
                else if(CanvasData.ArtifactEquip.BiomeSelected = Biome.Lava):
                    if(Next = true):
                        set CanvasData.ArtifactEquip.BiomeSelected = Biome.Underworld
                    else:
                        set CanvasData.ArtifactEquip.BiomeSelected = Biome.Mushroom
                else if(CanvasData.ArtifactEquip.BiomeSelected = Biome.Underworld):
                    if(Next = true):
                        set CanvasData.ArtifactEquip.BiomeSelected = Biome.Mushroom
                    else:
                        set CanvasData.ArtifactEquip.BiomeSelected = Biome.Lava
            
            if(PlayerData := PlayerStatsManager.GetPlayerStats[Player]):
                UIPlayerMenu.UpdateUI(Player, PlayerStatsManager, Achievements)
    }
}

ChangeDisplayHandler := class(){
    Index : int
    PlayerStatsManager : player_stats_manager
    UIPlayerMenu : UI_PlayerMenu

    HandlerFunction(WidgetMessage : widget_message) : void = {
            Player := WidgetMessage.Player
            var OutRarity : ArtifactRarity = ArtifactRarity.Null
            if(PlayerData := PlayerStatsManager.GetPlayerStats[Player]):
                if(CanvasData := UIPlayerMenu.BuiltCanvasMap[Player]):
                    if(Index = 0):
                        if(CanvasData.ArtifactEquip.BiomeSelected = Biome.Mushroom):
                            set OutRarity = ArtifactRarity.MushroomCommon
                        else if(CanvasData.ArtifactEquip.BiomeSelected = Biome.Lava):
                            set OutRarity = ArtifactRarity.LavaCommon
                        else if(CanvasData.ArtifactEquip.BiomeSelected = Biome.Underworld):
                            set OutRarity = ArtifactRarity.UnderworldCommon
                    if(Index = 1):
                        if(CanvasData.ArtifactEquip.BiomeSelected = Biome.Mushroom):
                            set OutRarity = ArtifactRarity.MushroomUncommon
                        else if(CanvasData.ArtifactEquip.BiomeSelected = Biome.Lava):
                            set OutRarity = ArtifactRarity.LavaUncommon
                        else if(CanvasData.ArtifactEquip.BiomeSelected = Biome.Underworld):
                            set OutRarity = ArtifactRarity.UnderworldUncommon
                    if(Index = 2):
                        if(CanvasData.ArtifactEquip.BiomeSelected = Biome.Mushroom):
                            set OutRarity = ArtifactRarity.MushroomRare
                        else if(CanvasData.ArtifactEquip.BiomeSelected = Biome.Lava):
                            set OutRarity = ArtifactRarity.LavaRare
                        else if(CanvasData.ArtifactEquip.BiomeSelected = Biome.Underworld):
                            set OutRarity = ArtifactRarity.UnderworldRare
                    if(Index = 3):
                        if(CanvasData.ArtifactEquip.BiomeSelected = Biome.Mushroom):
                            set OutRarity = ArtifactRarity.MushroomEpic
                        else if(CanvasData.ArtifactEquip.BiomeSelected = Biome.Lava):
                            set OutRarity = ArtifactRarity.LavaEpic
                        else if(CanvasData.ArtifactEquip.BiomeSelected = Biome.Underworld):
                            set OutRarity = ArtifactRarity.UnderworldEpic
                    if(Index = 4):
                        if(CanvasData.ArtifactEquip.BiomeSelected = Biome.Mushroom):
                            set OutRarity = ArtifactRarity.MushroomLegendary
                        else if(CanvasData.ArtifactEquip.BiomeSelected = Biome.Lava):
                            set OutRarity = ArtifactRarity.LavaLegendary
                        else if(CanvasData.ArtifactEquip.BiomeSelected = Biome.Underworld):
                            set OutRarity = ArtifactRarity.UnderworldLegendary

                    UIPlayerMenu.UpdateMainArtifact(Player, PlayerData, OutRarity)
                    set CanvasData.ArtifactEquip.SelectedArtifact = OutRarity
    }
}

AchievementDisplay := class:
    OverallWidget : widget
    var CurrentWidget : widget
    TotalStone : widget
    RocksDestroyed : widget
    DungeonCompleted : widget
    ArtifactsSmelted : widget
    HeighestStrength : widget

    TotalStoneWidgets : []text_block
    RocksDestroyedWidgets : []text_block
    DungeonCompletedWidgets : []text_block
    ArtifactsSmeltedWidgets : []text_block
    HeighestStrengthWidgets : []text_block

    TotalStoneButton : text_button_base
    RocksDestroyedButton : text_button_base
    DungeonCompletedButton : text_button_base
    ArtifactsSmeltedButton : text_button_base
    HeighestStrengthButton : text_button_base

    NameText : text_block
    DescriptionText : text_block
    ProgressText : text_block

    var Type : AchievementType

InventoryDisplay := class:
    Widget : widget
    Stone : text_block
    Prestige : text_block
    Copper : text_block
    Silver : text_block
    Malachite : text_block
    Spectrolite : text_block
    Obsidian : text_block
    Orb : text_block
    Gold : text_block
    Dust : text_block

ArtifactEquipDisplay := class:
    var IsLocked : logic
    Widget : widget
    Button : text_button_base
    Icon : texture_block
    SelectionBlock : color_block

    var CurrentArtifactRarity : ArtifactRarity

ArtifactEquipData := class:
    Widget : widget
    var SelectedArtifact : ArtifactRarity
    EquipButton : text_button_base
    BiomeText : text_block
    BiomeNext : text_button_base
    BiomePrev : text_button_base
    var SlotSelected : int
    var BiomeSelected : Biome

    MainIcon : texture_block
    StrengthStat : text_block
    CriticalRateStat : text_block
    CriticalDamageStat : text_block
    FinalStrikeStat : text_block

    SlotText : text_block
    EquippedIcon : texture_block
    EquippedStrengthStat : text_block
    EquippedCriticalRateStat : text_block
    EquippedCriticalDamageStat : text_block
    EquippedFinalStrikeStat : text_block

    ArtifactSlots : []ArtifactEquipDisplay
    ArtifactButtons : []ArtifactButtonData

MakePlayerMenuData<constructor>(InPlayerStatsManager : player_stats_manager):= PlayerMenuData:
    PlayerStatsManager := InPlayerStatsManager

PlayerMenuCanvas := class:
    Canvas : canvas
    var CurrentOpenWidget : widget

    ExitButton : text_button_base
    InventoryButton : text_button_base
    PlayerStatsButton : text_button_base
    AchievementsButton : text_button_base
    ArtifactButton : text_button_base
    PrestigeButton : text_button_base
    TeleportToHUBButton : text_button_base
    CardsButton : text_button_base
    AltarButton : text_button_base

    Background : texture_block
    #SettingsButton : text_button_base

    PlayerStatsMenu : PlayerStatsMenuData
    ArtifactEquip : ArtifactEquipData
    Inventory : InventoryDisplay
    Achievements : AchievementDisplay

    var CanTeleport : logic

PlayerStatsMenuData := class:
    Widget : widget

    PlayerStrengthText : text_block
    PlayerCardText : text_block
    PlayerArtifactText : text_block
    PlayerSwordText : text_block
    PlayerTotalStrengthText : text_block
    PlayerAreaMultiplierText : text_block
    PlayerAchievementMultiplierText : text_block
    PlayerPrestigeMultiplierText : text_block
    PlayerGatherRateText : text_block
    PlayerTimePlayedText : text_block

PlayerMenuData := class:
    PlayerStatsManager : player_stats_manager

IconTextData := class:
    Text : text_block
    Widget : widget

UI_PlayerMenu := class():
    ToMessage<localizes>(String : string) : message = "{String}"
    CenteredAlignment : vector2 = vector2{X := 0.5, Y := 0.5}

    var BuiltCanvasMap : [player]PlayerMenuCanvas = map{}

    # Initiliase the UI so it can added to the player much quicker
    InitUI(InPlayer : player, InPlayerMenuData : PlayerMenuData, InArtifactManager : V_ArtifactManager, InPrestige : V_PrestigeFunctionality, InCards : V_CardManager, InAchievements : []achievement_data, HUBTeleporter : teleporter_device, InAltar : V_OfferingAltar):void=
        DisplayCanvas := MakePlayerMenuCanvas(InPlayer, InPlayerMenuData, InAchievements)
        # Set the players build canvas data
        if(set BuiltCanvasMap[InPlayer] = DisplayCanvas, CanvasData := BuiltCanvasMap[InPlayer]):
            CanvasData.ArtifactButton.OnClick().Subscribe(OpenEquipArtifactHandler{UIPlayerMenu := Self, PlayerStatsManager := InPlayerMenuData.PlayerStatsManager, Achievements := InAchievements}.HandlerFunction)
            CanvasData.PlayerStatsButton.OnClick().Subscribe(OpenPlayerStatsHandler{UIPlayerMenu := Self, PlayerStatsManager := InPlayerMenuData.PlayerStatsManager, Achievements := InAchievements}.HandlerFunction)
            CanvasData.PrestigeButton.OnClick().Subscribe(OpenPrestigeButtonHandler{UIPlayerMenu := Self, Prestige := InPrestige}.HandlerFunction)
            CanvasData.InventoryButton.OnClick().Subscribe(OpenInventoryHandler{UIPlayerMenu := Self, PlayerStatsManager := InPlayerMenuData.PlayerStatsManager, Achievements := InAchievements}.HandlerFunction)
            CanvasData.CardsButton.OnClick().Subscribe(OpenCardButtonHandler{UIPlayerMenu := Self, Card := InCards}.HandlerFunction)
            CanvasData.AchievementsButton.OnClick().Subscribe(OpenAchievementsHandler{UIPlayerMenu := Self, PlayerStatsManager := InPlayerMenuData.PlayerStatsManager, Achievements := InAchievements}.HandlerFunction)
            CanvasData.TeleportToHUBButton.OnClick().Subscribe(TeleportHUBHandler{UIPlayerMenu := Self, Teleporter := HUBTeleporter}.HandlerFunction) 
            CanvasData.AltarButton.OnClick().Subscribe(AltarButtonHandler{UIPlayerMenu := Self, Device := InAltar}.HandlerFunction)

            CanvasData.ExitButton.OnClick().Subscribe(HandlePlayerExit)

            CanvasData.ArtifactEquip.BiomeNext.OnClick().Subscribe(ChangeBiomeHandler{Next := true, UIPlayerMenu := Self, PlayerStatsManager := InPlayerMenuData.PlayerStatsManager, Achievements := InAchievements}.HandlerFunction)
            CanvasData.ArtifactEquip.BiomePrev.OnClick().Subscribe(ChangeBiomeHandler{Next := false, UIPlayerMenu := Self, PlayerStatsManager := InPlayerMenuData.PlayerStatsManager, Achievements := InAchievements}.HandlerFunction)
            CanvasData.ArtifactEquip.EquipButton.OnClick().Subscribe(EquipHandler{Device := InArtifactManager, UIPlayerMenu := Self, PlayerStatsManager := InPlayerMenuData.PlayerStatsManager, Achievements := InAchievements}.HandlerFunction)

            for(Index -> Display : CanvasData.ArtifactEquip.ArtifactSlots):
                Display.Button.OnClick().Subscribe(SelectSlotHandler{Slot := Index + 1, UIPlayerMenu := Self, PlayerStatsManager := InPlayerMenuData.PlayerStatsManager, Achievements := InAchievements}.HandlerFunction)
            
            for(Index -> Display : CanvasData.ArtifactEquip.ArtifactButtons):
                Display.Button.OnClick().Subscribe(ChangeDisplayHandler{Index := Index, UIPlayerMenu := Self, PlayerStatsManager := InPlayerMenuData.PlayerStatsManager}.HandlerFunction)

            CanvasData.Achievements.TotalStoneButton.OnClick().Subscribe(OpenAchievementsSwitchHandler{UIPlayerMenu := Self, PlayerStatsManager := InPlayerMenuData.PlayerStatsManager, Achievements := InAchievements, Type := AchievementType.TotalStone}.HandlerFunction)
            CanvasData.Achievements.RocksDestroyedButton.OnClick().Subscribe(OpenAchievementsSwitchHandler{UIPlayerMenu := Self, PlayerStatsManager := InPlayerMenuData.PlayerStatsManager, Achievements := InAchievements, Type := AchievementType.RocksDestroyed}.HandlerFunction)
            CanvasData.Achievements.DungeonCompletedButton.OnClick().Subscribe(OpenAchievementsSwitchHandler{UIPlayerMenu := Self, PlayerStatsManager := InPlayerMenuData.PlayerStatsManager, Achievements := InAchievements, Type := AchievementType.DungeonsCompleted}.HandlerFunction)
            CanvasData.Achievements.ArtifactsSmeltedButton.OnClick().Subscribe(OpenAchievementsSwitchHandler{UIPlayerMenu := Self, PlayerStatsManager := InPlayerMenuData.PlayerStatsManager, Achievements := InAchievements, Type := AchievementType.ArtifactsSmelted}.HandlerFunction)
            CanvasData.Achievements.HeighestStrengthButton.OnClick().Subscribe(OpenAchievementsSwitchHandler{UIPlayerMenu := Self, PlayerStatsManager := InPlayerMenuData.PlayerStatsManager, Achievements := InAchievements, Type := AchievementType.HighestStrength}.HandlerFunction)

    # Add the UI to the player
    AddUI(InPlayer : player):void=
        if(PlayerUI := GetPlayerUI[InPlayer], CanvasData := BuiltCanvasMap[InPlayer]):
            CanvasData.Canvas.SetVisibility(widget_visibility.Visible)
            PlayerUI.AddWidget(CanvasData.Canvas, player_ui_slot{InputMode := ui_input_mode.All} )

    SetCanTeleportHUB(InPlayer : player, InCanTeleport : logic):void=
        if(CanvasData := BuiltCanvasMap[InPlayer]):
            set CanvasData.CanTeleport = InCanTeleport

    HandlePlayerTeleportHUB(WidgetMessage : widget_message, Teleporter : teleporter_device):void=
        #Teleport Player Code Here
        Teleporter.Teleport(WidgetMessage.Player)
        spawn{HideUI(WidgetMessage.Player)}

    # When player presses the exit button
    HandlePlayerExit(WidgetMessage: widget_message):void=
        spawn{HideUI(WidgetMessage.Player)}

    HideUI(InPlayer: player)<suspends>:void=
        Sleep(0.0) # So we're sure that the canvas has been added to BuiltCanvas map
        if(PlayerUI := GetPlayerUI[InPlayer], CanvasData := BuiltCanvasMap[InPlayer]):
            CanvasData.Canvas.SetVisibility(widget_visibility.Hidden)
            PlayerUI.RemoveWidget(CanvasData.Canvas)
    
    # Upates a given player's UI
    UpdateUI(InPlayer : player, PlayerStatsManager : player_stats_manager, InAchievements : []achievement_data):void=
        if(PlayerUI := GetPlayerUI[InPlayer], CanvasData := BuiltCanvasMap[InPlayer], PlayerStatsData := PlayerStatsManager.GetPlayerStats[InPlayer]):
            # Update Artifacts
            # Get slot level
            ArtifactSlotLevel := PlayerStatsManager.GetPlayerUpgradeLevel(InPlayer, PlayerUpgradeType.Prestige_ArtifactSlots)
            # Unlock the right amount of slots
            for(Index -> DisplaySlot : CanvasData.ArtifactEquip.ArtifactSlots):
                if(Index <= ArtifactSlotLevel):
                    ToggleArtifactDisplayLocked(DisplaySlot, false)
                else:
                    ToggleArtifactDisplayLocked(DisplaySlot, true)

            UpdateAllSlots(InPlayer, PlayerStatsData)
            UpdateMainArtifact(InPlayer, PlayerStatsData, CanvasData.ArtifactEquip.SelectedArtifact)
            if(CanvasData.ArtifactEquip.BiomeSelected = Biome.Mushroom):
                CanvasData.ArtifactEquip.BiomeText.SetText(ToMessage("Mushroom"))
            if(CanvasData.ArtifactEquip.BiomeSelected = Biome.Lava):
                CanvasData.ArtifactEquip.BiomeText.SetText(ToMessage("Lava"))
            if(CanvasData.ArtifactEquip.BiomeSelected = Biome.Underworld):
                CanvasData.ArtifactEquip.BiomeText.SetText(ToMessage("Underworld"))

            CanvasData.ArtifactEquip.SlotText.SetText(ToMessage("Slot {CanvasData.ArtifactEquip.SlotSelected}"))

            for(Index -> Slot : CanvasData.ArtifactEquip.ArtifactSlots):
                if(Index = CanvasData.ArtifactEquip.SlotSelected - 1):
                    Slot.SelectionBlock.SetVisibility(widget_visibility.Visible)
                else:
                    Slot.SelectionBlock.SetVisibility(widget_visibility.Hidden)

        # Update Player Stats
            CanvasData.PlayerStatsMenu.PlayerStrengthText.SetText(ToMessage(BigNumber_FormatDisplay(PlayerStatsData.HarvestRatePerHit)))

            var CardBonus : float = 0.0
            if(FoundCard := SearchForCard(8, InPlayer, PlayerStatsManager)?):
                BonusAmount := GetCardBonus(8, FoundCard.Level)
                set CardBonus = BonusAmount

            CanvasData.PlayerStatsMenu.PlayerCardText.SetText(ToMessage("{FormatFloatTo2DP(CardBonus)}%"))

            var ArtifactStrength : float = 0.0
            for(Index -> Artifact : PlayerStatsData.EquippedPlayerArtifacts):
                set ArtifactStrength += GetArtifactStat(Artifact, ArtifactStats.ArtifactBonus)
                for(SlotIndex -> Slot : CanvasData.ArtifactEquip.ArtifactSlots):
                    if(Artifact.EquippedSlot = SlotIndex + 1):
                        Slot.Icon.SetImage(GetArtifactImage(Artifact.Rarity))
                    if(Artifact.EquippedSlot = CanvasData.ArtifactEquip.SlotSelected):
                        CanvasData.ArtifactEquip.EquippedIcon.SetImage(GetArtifactImage(Artifact.Rarity))
                        CanvasData.ArtifactEquip.EquippedStrengthStat.SetText(ToMessage("+{FormatFloatTo2DP(GetArtifactStat(Artifact, ArtifactStats.ArtifactBonus))}%"))
                        CanvasData.ArtifactEquip.EquippedCriticalRateStat.SetText(ToMessage("+{FormatFloatTo2DP(GetArtifactStat(Artifact, ArtifactStats.CriticalChance))}%"))
                        CanvasData.ArtifactEquip.EquippedCriticalDamageStat.SetText(ToMessage("+{FormatFloatTo2DP(GetArtifactStat(Artifact, ArtifactStats.CriticalDamage))}%"))
                        CanvasData.ArtifactEquip.EquippedFinalStrikeStat.SetText(ToMessage("+{FormatFloatTo2DP(GetArtifactStat(Artifact, ArtifactStats.FinalStrike))}%"))

            CanvasData.PlayerStatsMenu.PlayerArtifactText.SetText(ToMessage("{FormatFloatTo2DP(ArtifactStrength)}%"))

            var SwordStrengthBonus : float = 0.0
            if(Sword := PlayerStatsData.OwnedPlayerSword[0], Bonus := Sword.BigNumber_StatValues[3]):
                set SwordStrengthBonus = BigNumber_ConvertBigNumberIntoFloat(Bonus)

            CanvasData.PlayerStatsMenu.PlayerSwordText.SetText(ToMessage("{FormatFloatTo2DP(SwordStrengthBonus)}%"))


            CardRate :=  BigNumber_Add(PlayerStatsData.HarvestRatePerHit, (BigNumber_MultiplyWithAFloat(PlayerStatsData.HarvestRatePerHit, CardBonus / 100.0)))
            BonusStrength := BigNumber_MultiplyWithAFloat(CardRate, ArtifactStrength / 100.0)
            HitRateArtifact := BigNumber_Add(CardRate, BonusStrength)
            HarvestRateSword := BigNumber_Add(HitRateArtifact, BigNumber_MultiplyWithAFloat(HitRateArtifact, (SwordStrengthBonus / 100.0)))

            CanvasData.PlayerStatsMenu.PlayerTotalStrengthText.SetText(ToMessage(BigNumber_FormatDisplay(HarvestRateSword)))

            PrestigeMultiplier := PlayerStatsData.PrestigeBonusHarvestRate
            AreaMultipler := PlayerStatsData.CaveUnlockPercent
            AchievementMultipler := PlayerStatsData.AchievementTotalStoneBonus + PlayerStatsData.AchievementTotalDungeonsCompletedBonus + PlayerStatsData.AchievementHighestStrengthBonus + PlayerStatsData.AchievementTotalRocksDestroyedBonus + PlayerStatsData.AchievementTotalArtifactsSmeltedBonus
            
            AreaBonus := BigNumber_Add(HarvestRateSword, BigNumber_MultiplyWithAFloat(HarvestRateSword, AreaMultipler / 100.0))
            AchievementBonus := BigNumber_Add(AreaBonus, BigNumber_MultiplyWithAFloat(AreaBonus, AchievementMultipler / 100.0))
            PrestigeBonus := BigNumber_Add(AchievementBonus, BigNumber_MultiplyWithAFloat(AchievementBonus, PrestigeMultiplier / 100.0))

            CanvasData.PlayerStatsMenu.PlayerAreaMultiplierText.SetText(ToMessage("{FormatFloatTo2DP(AreaMultipler)}%"))
            CanvasData.PlayerStatsMenu.PlayerAchievementMultiplierText.SetText(ToMessage("{FormatFloatTo2DP(AchievementMultipler)}%"))
            CanvasData.PlayerStatsMenu.PlayerPrestigeMultiplierText.SetText(ToMessage("{BigNumber_FormatDisplay(BigNumber_ConvertFloatIntoBigNumber(PrestigeMultiplier))}%"))
            CanvasData.PlayerStatsMenu.PlayerGatherRateText.SetText(ToMessage("{BigNumber_FormatDisplay(PrestigeBonus)}/Hit"))
            CanvasData.PlayerStatsMenu.PlayerTimePlayedText.SetText(ToMessage(FormattedPlayTime(PlayerStatsData.HoursPlayed, PlayerStatsData.MinutesPlayed)))

            # Update Inventory
            CanvasData.Inventory.Stone.SetText(ToMessage(BigNumber_FormatDisplay(PlayerStatsData.PlayerStoneOreAmount)))
            CanvasData.Inventory.Prestige.SetText(ToMessage(BigNumber_FormatDisplay(PlayerStatsData.PlayerPrestigeResourceAmount)))
            CanvasData.Inventory.Copper.SetText(ToMessage(BigNumber_FormatDisplay(PlayerStatsData.PlayerCopperOreAmount)))
            CanvasData.Inventory.Silver.SetText(ToMessage(BigNumber_FormatDisplay(PlayerStatsData.PlayerSilverOreAmount)))
            CanvasData.Inventory.Malachite.SetText(ToMessage(BigNumber_FormatDisplay(PlayerStatsData.PlayerMalachiteOreAmount)))
            CanvasData.Inventory.Spectrolite.SetText(ToMessage(BigNumber_FormatDisplay(PlayerStatsData.PlayerSpectroliteOreAmount)))
            CanvasData.Inventory.Obsidian.SetText(ToMessage(BigNumber_FormatDisplay(PlayerStatsData.PlayerObsidianOreAmount)))
            CanvasData.Inventory.Orb.SetText(ToMessage(BigNumber_FormatDisplay(PlayerStatsData.OrbAmount)))
            CanvasData.Inventory.Gold.SetText(ToMessage(BigNumber_FormatDisplay(PlayerStatsData.PlayerGoldCoinAmount)))
            CanvasData.Inventory.Dust.SetText(ToMessage(BigNumber_FormatDisplay(PlayerStatsData.DustAmount)))


            var AchievementToUse : achievement_data = achievement_data{}
            var BigNumberValue : big_number = big_number{}
            var CheckTexts : []text_block = array{}
            # Get achievement type
            for(Achievement : InAchievements):
                if(Achievement.Type = CanvasData.Achievements.Type):
                    case(CanvasData.Achievements.Type):
                        AchievementType.TotalStone =>
                            set AchievementToUse = Achievement
                            set BigNumberValue = PlayerStatsData.TotalStoneOreAmount
                            set CheckTexts = CanvasData.Achievements.TotalStoneWidgets
                        AchievementType.HighestStrength =>
                            set AchievementToUse = Achievement
                            set BigNumberValue = PlayerStatsData.HighestStrength
                            set CheckTexts = CanvasData.Achievements.HeighestStrengthWidgets
                        AchievementType.RocksDestroyed =>
                            set AchievementToUse = Achievement
                            set BigNumberValue = PlayerStatsData.TotalRocksDestroyed
                            set CheckTexts = CanvasData.Achievements.RocksDestroyedWidgets
                        AchievementType.ArtifactsSmelted =>
                            set AchievementToUse = Achievement
                            set BigNumberValue = PlayerStatsData.TotalArtifactsSmelted
                            set CheckTexts = CanvasData.Achievements.ArtifactsSmeltedWidgets
                        AchievementType.DungeonsCompleted =>
                            set AchievementToUse = Achievement
                            set BigNumberValue = PlayerStatsData.TotalDungeonsCompleted
                            set CheckTexts = CanvasData.Achievements.DungeonCompletedWidgets
                        AchievementType.Null =>

                    CanvasData.Achievements.NameText.SetText(ToMessage(Achievement.Name))
                    CanvasData.Achievements.DescriptionText.SetText(ToMessage(Achievement.Description))
                    CanvasData.Achievements.ProgressText.SetText(ToMessage("{BigNumber_FormatDisplay(BigNumberValue)}"))

                    # Compare the stat to the acheivement to see if it's done
                    for(Index := 0..Achievement.Levels.Length - 1):
                        if(Level := Achievement.Levels[Index]):
                            GreaterThan := BigNumber_GreaterThanOrEquals(BigNumberValue, Level)
                            if(GreaterThan = true):
                                if(CheckText := CheckTexts[Index]):
                                    CheckText.SetTextColor(NamedColors.LightGreen)
                            else:
                                if(CheckText := CheckTexts[Index]):
                                    CheckText.SetTextColor(NamedColors.Red)


    MakePlayerMenuCanvas(InPlayer : player, InPlayerMenuData : PlayerMenuData, InAchievements : []achievement_data):PlayerMenuCanvas=
        ExitButton := button_loud:
            DefaultText := ToMessage("Exit")
        InventoryButton := button_quiet:
            DefaultText := ToMessage("INVENTORY")
        PlayerStatsButton := button_quiet:
            DefaultText := ToMessage("PLAYER STATS")
        AchievementsButton := button_quiet:
            DefaultText := ToMessage("ACHIEVEMENTS")
        ArtifactButton := button_quiet:
            DefaultText := ToMessage("ARTIFACTS")
        PrestigeButton := button_quiet:
            DefaultText := ToMessage("PRESTIGE")
        TeleportToHUBButton := button_quiet:
            DefaultText := ToMessage("Teleport HUB")
        CardsButton := button_quiet:
            DefaultText := ToMessage("CARDS")
        AltarButton := button_quiet:
            DefaultText := ToMessage("Altar")
        #SettingsButton := button_quiet:
            #DefaultText := ToMessage("SETTTINGS")
        BackgroundBlock := texture_block:
            DefaultImage := VerseExposed.T_PlayerMenuInventory
        BorderSize := 2.0
        BorderColorBlock := color_block:
            DefaultColor := NamedColors.White
            DefaultDesiredSize := vector2{X := BorderSize, Y := BorderSize}

        PlayerStatsMenu := MakePlayerStatsPage(InPlayer, InPlayerMenuData.PlayerStatsManager)
        PlayerStatsMenu.Widget.SetVisibility(widget_visibility.Collapsed)
        ArtifactEquip := MakeArtifactEquipCanvas()
        ArtifactEquip.Widget.SetVisibility(widget_visibility.Collapsed)
        Achievements := MakeAchievementDisplay(InAchievements)
        Achievements.OverallWidget.SetVisibility(widget_visibility.Collapsed)

        Inventory := MakeInventoryPage()
        
        LeftButtonStackBox :=
            array:
                stack_box_slot:
                    Padding := margin {Top:= 10.0}
                    Widget := InventoryButton
                    VerticalAlignment := vertical_alignment.Fill
                    HorizontalAlignment := horizontal_alignment.Fill
                stack_box_slot:
                    Padding := margin {Top:= 10.0}
                    Widget := PlayerStatsButton
                    VerticalAlignment := vertical_alignment.Fill
                    HorizontalAlignment := horizontal_alignment.Fill
                stack_box_slot:
                    Padding := margin {Top:= 10.0}
                    Widget := AchievementsButton
                    VerticalAlignment := vertical_alignment.Fill
                    HorizontalAlignment := horizontal_alignment.Fill
                stack_box_slot:
                    Padding := margin {Top:= 10.0}
                    Widget := ArtifactButton
                    VerticalAlignment := vertical_alignment.Fill
                    HorizontalAlignment := horizontal_alignment.Fill
                stack_box_slot:
                    Padding := margin {Top:= 10.0}
                    Widget := PrestigeButton
                    VerticalAlignment := vertical_alignment.Fill
                    HorizontalAlignment := horizontal_alignment.Fill
                    #
                stack_box_slot:
                    Padding := margin {Top:= 10.0}
                    Widget := TeleportToHUBButton
                    VerticalAlignment := vertical_alignment.Fill
                    HorizontalAlignment := horizontal_alignment.Fill
                stack_box_slot:
                    Padding := margin {Top:= 10.0}
                    Widget := CardsButton
                    VerticalAlignment := vertical_alignment.Fill
                    HorizontalAlignment := horizontal_alignment.Fill
                stack_box_slot:
                    Padding := margin {Top:= 10.0}
                    Widget := AltarButton
                    VerticalAlignment := vertical_alignment.Fill
                    HorizontalAlignment := horizontal_alignment.Fill
                stack_box_slot:
                    Padding := margin {Top:= 20.0}
                    Widget := ExitButton
                    VerticalAlignment := vertical_alignment.Fill
                    HorizontalAlignment := horizontal_alignment.Fill

        LeftButtonOverlay := overlay:
            Slots := array:
                overlay_slot:
                    VerticalAlignment := vertical_alignment.Fill
                    HorizontalAlignment := horizontal_alignment.Fill
                    Widget := overlay:
                        Slots := array:
                            overlay_slot:
                                VerticalAlignment := vertical_alignment.Fill
                                HorizontalAlignment := horizontal_alignment.Fill
                                Widget := stack_box:
                                    Orientation := orientation.Vertical
                                    Slots := LeftButtonStackBox
        BackgroundOverlay := overlay:
            Slots := array:
                overlay_slot:
                    VerticalAlignment := vertical_alignment.Fill
                    HorizontalAlignment := horizontal_alignment.Fill
                    Widget := BackgroundBlock
                overlay_slot:
                    VerticalAlignment := vertical_alignment.Top
                    HorizontalAlignment := horizontal_alignment.Fill
                    Widget := color_block:
                        DefaultColor := NamedColors.DarkBlue
                        DefaultDesiredSize := vector2{X := BorderSize, Y := BorderSize}

        RightDisplayOverlay := overlay:
            Slots := array:
                overlay_slot:
                    VerticalAlignment := vertical_alignment.Fill
                    HorizontalAlignment := horizontal_alignment.Fill
                    Widget := PlayerStatsMenu.Widget
                overlay_slot:
                    VerticalAlignment := vertical_alignment.Fill
                    HorizontalAlignment := horizontal_alignment.Fill
                    Widget := ArtifactEquip.Widget
                overlay_slot:
                    VerticalAlignment := vertical_alignment.Fill
                    HorizontalAlignment := horizontal_alignment.Fill
                    Widget := Inventory.Widget
                overlay_slot:
                    VerticalAlignment := vertical_alignment.Fill
                    HorizontalAlignment := horizontal_alignment.Fill
                    Widget := Achievements.OverallWidget

        OutputCanvas := canvas:
            Slots := array:
                canvas_slot:
                    Anchors := anchors{Minimum := vector2{X:=0.0, Y:=0.0}, Maximum := vector2{X:=1.0, Y:=1.0}}
                    Alignment := CenteredAlignment
                    Widget := BackgroundOverlay
                    ZOrder := 0
                canvas_slot:
                    Anchors := anchors{Minimum := vector2{X:=0.03, Y:=0.22}, Maximum := vector2{X:=0.16, Y:=0.85}}
                    Alignment := CenteredAlignment
                    Widget := LeftButtonOverlay
                    ZOrder := 1
                canvas_slot:
                    Anchors := anchors{Minimum := vector2{X:=0.16, Y:=0.15}, Maximum := vector2{X:=0.98, Y:=0.85}}
                    Alignment := CenteredAlignment
                    Widget := RightDisplayOverlay
                    ZOrder := 1

        return PlayerMenuCanvas:
            Canvas := OutputCanvas
            CurrentOpenWidget := Inventory.Widget
            ExitButton := ExitButton
            InventoryButton := InventoryButton
            PlayerStatsButton := PlayerStatsButton
            AchievementsButton := AchievementsButton
            ArtifactButton := ArtifactButton
            PrestigeButton := PrestigeButton
            TeleportToHUBButton := TeleportToHUBButton
            CardsButton := CardsButton
            AltarButton := AltarButton
            #SettingsButton := SettingsButton
            PlayerStatsMenu := PlayerStatsMenu
            ArtifactEquip := ArtifactEquip
            Inventory := Inventory
            Achievements := Achievements
            Background := BackgroundBlock
            CanTeleport := true

    MakeIconText(InImage : texture, InText : string):IconTextData=
        Text := text_block:
            DefaultText := ToMessage("{InText}")
            DefaultJustification := text_justification.Left
            DefaultShadowColor := NamedColors.Black
            DefaultShadowOffset := option{vector2{X:=2.0,Y:=2.0}}
        Text.SetTextColor(NamedColors.White)
        Text.SetShadowOpacity(0.8)
        Icon := texture_block:
            DefaultImage := InImage
            DefaultDesiredSize := vector2{X:=32.0,Y:=32.0}

        Overlay := overlay:
            Slots := array:
                overlay_slot:
                    VerticalAlignment := vertical_alignment.Fill
                    HorizontalAlignment := horizontal_alignment.Fill
                    Widget := stack_box:
                        Orientation := orientation.Horizontal
                        Slots := array:
                            stack_box_slot:
                                Widget := Icon
                                VerticalAlignment := vertical_alignment.Center
                                HorizontalAlignment := horizontal_alignment.Center
                            stack_box_slot:
                                Padding := margin {Left:= 10.0}
                                Widget := Text
                                VerticalAlignment := vertical_alignment.Fill
                                HorizontalAlignment := horizontal_alignment.Left

        return IconTextData:
            Text := Text
            Widget := Overlay

    MakePlayerStatsPage(InPlayer : player, InPlayerStatsManager : player_stats_manager):PlayerStatsMenuData=

        BaseStrengthTitle := MakeIconText(VerseExposed.T_Achievements_Strength, "Base Strength")
        PlayerStrengthText := text_block:
            DefaultText := ToMessage("1")
            DefaultJustification := text_justification.Right
            DefaultShadowColor := NamedColors.Black
            DefaultShadowOffset := option{vector2{X:=2.0,Y:=2.0}}
        PlayerStrengthText.SetTextColor(NamedColors.White)
        PlayerStrengthText.SetShadowOpacity(0.8)

        CardTitle := MakeIconText(VerseExposed.T_CardIcon, "Card Bonus")
        PlayerCardText := text_block:
            DefaultText := ToMessage("1")
            DefaultJustification := text_justification.Right
            DefaultShadowColor := NamedColors.Black
            DefaultShadowOffset := option{vector2{X:=2.0,Y:=2.0}}
        PlayerCardText.SetTextColor(NamedColors.White)
        PlayerCardText.SetShadowOpacity(0.8)

        ArtifactTitle := MakeIconText(VerseExposed.T_Achievements_Artifacts, "Artifact Bonus")
        PlayerArtifactText := text_block:
            DefaultText := ToMessage("1")
            DefaultJustification := text_justification.Right
            DefaultShadowColor := NamedColors.Black
            DefaultShadowOffset := option{vector2{X:=2.0,Y:=2.0}}
        PlayerArtifactText.SetTextColor(NamedColors.White)
        PlayerArtifactText.SetShadowOpacity(0.8)

        SwordTitle := MakeIconText(VerseExposed.T_SwordWhite, "Sword Bonus")
        PlayerSwordText := text_block:
            DefaultText := ToMessage("1")
            DefaultJustification := text_justification.Right
            DefaultShadowColor := NamedColors.Black
            DefaultShadowOffset := option{vector2{X:=2.0,Y:=2.0}}
        PlayerSwordText.SetTextColor(NamedColors.White)
        PlayerSwordText.SetShadowOpacity(0.8)

        TotalStrengthTitle := MakeIconText(VerseExposed.T_Achievements_Strength, "Total Strength")
        PlayerTotalStrengthText := text_block:
            DefaultText := ToMessage("1")
            DefaultJustification := text_justification.Right
            DefaultShadowColor := NamedColors.Black
            DefaultShadowOffset := option{vector2{X:=2.0,Y:=2.0}}
        PlayerTotalStrengthText.SetTextColor(NamedColors.LightGreen)
        PlayerTotalStrengthText.SetShadowOpacity(0.8)

        AreaMultiplierTitle := MakeIconText(VerseExposed.T_AreaIcon, "Area Multiplier")
        PlayerAreaMultiplierText := text_block:
            DefaultText := ToMessage("0%")
            DefaultJustification := text_justification.Right
            DefaultShadowColor := NamedColors.Black
            DefaultShadowOffset := option{vector2{X:=2.0,Y:=2.0}}
        PlayerAreaMultiplierText.SetTextColor(NamedColors.White)
        PlayerAreaMultiplierText.SetShadowOpacity(0.8)

        AchievementMultiplierTitle:= MakeIconText(VerseExposed.T_AchievementIcon, "Achievement Multiplier")
        PlayerAchievementMultiplierText := text_block:
            DefaultText := ToMessage("0%")
            DefaultJustification := text_justification.Right
            DefaultShadowColor := NamedColors.Black
            DefaultShadowOffset := option{vector2{X:=2.0,Y:=2.0}}
        PlayerAchievementMultiplierText.SetTextColor(NamedColors.White)
        PlayerAchievementMultiplierText.SetShadowOpacity(0.8)

        PrestigeMultiplierTitle:= MakeIconText(VerseExposed.T_PrestigeIcon, "Prestige Multiplier")
        PlayerPrestigeMultiplierText := text_block:
            DefaultText := ToMessage("0%")
            DefaultJustification := text_justification.Right
            DefaultShadowColor := NamedColors.Black
            DefaultShadowOffset := option{vector2{X:=2.0,Y:=2.0}}
        PlayerPrestigeMultiplierText.SetTextColor(NamedColors.White)
        PlayerPrestigeMultiplierText.SetShadowOpacity(0.8)

        GatherRateTitle:= MakeIconText(VerseExposed.T_PickaxeWhite, "Total Gather Rate")
        PlayerGatherRateText := text_block:
            DefaultText := ToMessage("1/Hit")
            DefaultJustification := text_justification.Right
            DefaultShadowColor := NamedColors.Black
            DefaultShadowOffset := option{vector2{X:=2.0,Y:=2.0}}
        PlayerGatherRateText.SetTextColor(NamedColors.LightGreen)
        PlayerGatherRateText.SetShadowOpacity(0.8)

        PlayTimeTitle:= MakeIconText(VerseExposed.T_TimerIcon, "Total Time Played")
        PlayerTimePlayedText := text_block:
            DefaultText := ToMessage("00:00")
            DefaultJustification := text_justification.Right
            DefaultShadowColor := NamedColors.Black
            DefaultShadowOffset := option{vector2{X:=2.0,Y:=2.0}}
        PlayerTimePlayedText.SetTextColor(NamedColors.White)
        PlayerTimePlayedText.SetShadowOpacity(0.8)

        TitlesStackBox := stack_box:
            Orientation := orientation.Vertical
            Slots := array:
                stack_box_slot:
                    Padding := margin {Top:= 5.0}
                    Widget := BaseStrengthTitle.Widget
                    VerticalAlignment := vertical_alignment.Fill
                    HorizontalAlignment := horizontal_alignment.Fill
                stack_box_slot:
                    Padding := margin {Top:= 10.0}
                    Widget := CardTitle.Widget
                    VerticalAlignment := vertical_alignment.Fill
                    HorizontalAlignment := horizontal_alignment.Fill
                stack_box_slot:
                    Padding := margin {Top:= 10.0}
                    Widget := ArtifactTitle.Widget
                    VerticalAlignment := vertical_alignment.Fill
                    HorizontalAlignment := horizontal_alignment.Fill
                stack_box_slot:
                    Padding := margin {Top:= 10.0}
                    Widget := SwordTitle.Widget
                    VerticalAlignment := vertical_alignment.Fill
                    HorizontalAlignment := horizontal_alignment.Fill
                stack_box_slot:
                    Padding := margin {Top:= 30.0}
                    Widget := TotalStrengthTitle.Widget
                    VerticalAlignment := vertical_alignment.Fill
                    HorizontalAlignment := horizontal_alignment.Fill
                stack_box_slot:
                    Padding := margin {Top:= 70.0}
                    Widget := AreaMultiplierTitle.Widget
                    VerticalAlignment := vertical_alignment.Fill
                    HorizontalAlignment := horizontal_alignment.Fill
                stack_box_slot:
                    Padding := margin {Top:= 10.0}
                    Widget := AchievementMultiplierTitle.Widget
                    VerticalAlignment := vertical_alignment.Fill
                    HorizontalAlignment := horizontal_alignment.Fill
                stack_box_slot:
                    Padding := margin {Top:= 10.0}
                    Widget := PrestigeMultiplierTitle.Widget
                    VerticalAlignment := vertical_alignment.Fill
                    HorizontalAlignment := horizontal_alignment.Fill
                stack_box_slot:
                    Padding := margin {Top:= 30.0}
                    Widget := GatherRateTitle.Widget
                    VerticalAlignment := vertical_alignment.Fill
                    HorizontalAlignment := horizontal_alignment.Fill
                stack_box_slot:
                    Padding := margin {Top:= 65.0}
                    Widget := PlayTimeTitle.Widget
                    VerticalAlignment := vertical_alignment.Fill
                    HorizontalAlignment := horizontal_alignment.Fill
                

        StatsStackBox := stack_box:
            Orientation := orientation.Vertical
            Slots := array:
                stack_box_slot:
                    Padding := margin {Top:= 5.0}
                    Widget := PlayerStrengthText
                    VerticalAlignment := vertical_alignment.Fill
                    HorizontalAlignment := horizontal_alignment.Fill
                stack_box_slot:
                    Padding := margin {Top:= 10.0}
                    Widget := PlayerCardText
                    VerticalAlignment := vertical_alignment.Fill
                    HorizontalAlignment := horizontal_alignment.Fill
                stack_box_slot:
                    Padding := margin {Top:= 10.0}
                    Widget := PlayerArtifactText
                    VerticalAlignment := vertical_alignment.Fill
                    HorizontalAlignment := horizontal_alignment.Fill
                stack_box_slot:
                    Padding := margin {Top:= 10.0}
                    Widget := PlayerSwordText
                    VerticalAlignment := vertical_alignment.Fill
                    HorizontalAlignment := horizontal_alignment.Fill
                stack_box_slot:
                    Padding := margin {Top:= 30.0}
                    Widget := PlayerTotalStrengthText
                    VerticalAlignment := vertical_alignment.Fill
                    HorizontalAlignment := horizontal_alignment.Fill
                stack_box_slot:
                    Padding := margin {Top:= 70.0}
                    Widget := PlayerAreaMultiplierText
                    VerticalAlignment := vertical_alignment.Fill
                    HorizontalAlignment := horizontal_alignment.Fill
                stack_box_slot:
                    Padding := margin {Top:= 10.0}
                    Widget := PlayerAchievementMultiplierText
                    VerticalAlignment := vertical_alignment.Fill
                    HorizontalAlignment := horizontal_alignment.Fill
                stack_box_slot:
                    Padding := margin {Top:= 10.0}
                    Widget := PlayerPrestigeMultiplierText
                    VerticalAlignment := vertical_alignment.Fill
                    HorizontalAlignment := horizontal_alignment.Fill
                stack_box_slot:
                    Padding := margin {Top:= 30.0}
                    Widget := PlayerGatherRateText
                    VerticalAlignment := vertical_alignment.Fill
                    HorizontalAlignment := horizontal_alignment.Fill
                stack_box_slot:
                    Padding := margin {Top:= 65.0}
                    Widget := PlayerTimePlayedText
                    VerticalAlignment := vertical_alignment.Fill
                    HorizontalAlignment := horizontal_alignment.Fill

        StackBox := stack_box:
            Orientation := orientation.Horizontal
            Slots := array:
                stack_box_slot:
                    Padding := margin {Right:= 250.0}
                    Widget := TitlesStackBox
                    VerticalAlignment := vertical_alignment.Fill
                    HorizontalAlignment := horizontal_alignment.Fill
                stack_box_slot:
                    Padding := margin {Left:= 250.0}
                    Widget := StatsStackBox
                    VerticalAlignment := vertical_alignment.Fill
                    HorizontalAlignment := horizontal_alignment.Fill

        OutOverlay := overlay:
            Slots := array:
                overlay_slot:
                    VerticalAlignment := vertical_alignment.Fill
                    Padding := margin{Top := 160.0, Left := 300.0}
                    HorizontalAlignment := horizontal_alignment.Left
                    Widget := StackBox

        return PlayerStatsMenuData:
            Widget := OutOverlay
            PlayerStrengthText := PlayerStrengthText
            PlayerCardText := PlayerCardText
            PlayerArtifactText := PlayerArtifactText
            PlayerSwordText := PlayerSwordText
            PlayerTotalStrengthText := PlayerTotalStrengthText
            PlayerAreaMultiplierText := PlayerAreaMultiplierText
            PlayerAchievementMultiplierText := PlayerAchievementMultiplierText
            PlayerPrestigeMultiplierText := PlayerPrestigeMultiplierText
            PlayerGatherRateText := PlayerGatherRateText
            PlayerTimePlayedText := PlayerTimePlayedText

    UpdateAllSlots(InPlayer : player, InPlayerStats : Player_Stats_Data):void=
        if(CanvasData := BuiltCanvasMap[InPlayer]):
            for(Index -> Display : CanvasData.ArtifactEquip.ArtifactButtons):
                Display.ArtifactLevelText.SetText(ToMessage("No Artifact"))
                Display.ArtifactRarityText.SetText(ToMessage("No Artifact"))
                Display.ArtifactLevelText.SetVisibility(widget_visibility.Hidden)
                Display.ArtifactImage.SetImage(GetArtifactImage(ArtifactRarity.Null))

                for(Artifact : InPlayerStats.OwnedPlayerArtifacts):
                    if(CanvasData.ArtifactEquip.BiomeSelected = Biome.Mushroom):
                        if(Artifact.Rarity = ArtifactRarity.MushroomCommon or Artifact.Rarity = ArtifactRarity.MushroomUncommon or Artifact.Rarity = ArtifactRarity.MushroomRare or Artifact.Rarity = ArtifactRarity.MushroomEpic or Artifact.Rarity = ArtifactRarity.MushroomLegendary):
                            if(Artifact.RarityInt = Index):
                                Display.ArtifactLevelText.SetText(ToMessage("Lv. {Artifact.Level} - Pr. {Artifact.Prestige}"))
                                Display.ArtifactRarityText.SetText(ToMessage(GetArtifactRarityString(Artifact.Rarity)))
                                Display.ArtifactLevelText.SetVisibility(widget_visibility.Visible)
                                Display.ArtifactImage.SetImage(GetArtifactImage(Artifact.Rarity))
                    if(CanvasData.ArtifactEquip.BiomeSelected = Biome.Lava):
                        if(Artifact.Rarity = ArtifactRarity.LavaCommon or Artifact.Rarity = ArtifactRarity.LavaUncommon or Artifact.Rarity = ArtifactRarity.LavaRare or Artifact.Rarity = ArtifactRarity.LavaEpic or Artifact.Rarity = ArtifactRarity.LavaLegendary):
                            if(Artifact.RarityInt - 5 = Index):
                                Display.ArtifactLevelText.SetText(ToMessage("Lv. {Artifact.Level} - Pr. {Artifact.Prestige}"))
                                Display.ArtifactRarityText.SetText(ToMessage(GetArtifactRarityString(Artifact.Rarity)))
                                Display.ArtifactLevelText.SetVisibility(widget_visibility.Visible)
                                Display.ArtifactImage.SetImage(GetArtifactImage(Artifact.Rarity))
                    if(CanvasData.ArtifactEquip.BiomeSelected = Biome.Underworld):
                        if(Artifact.Rarity = ArtifactRarity.UnderworldCommon or Artifact.Rarity = ArtifactRarity.UnderworldUncommon or Artifact.Rarity = ArtifactRarity.UnderworldRare or Artifact.Rarity = ArtifactRarity.UnderworldEpic or Artifact.Rarity = ArtifactRarity.UnderworldLegendary):
                            if(Artifact.RarityInt - 10 = Index):
                                Display.ArtifactLevelText.SetText(ToMessage("Lv. {Artifact.Level} - Pr. {Artifact.Prestige}"))
                                Display.ArtifactRarityText.SetText(ToMessage(GetArtifactRarityString(Artifact.Rarity)))
                                Display.ArtifactLevelText.SetVisibility(widget_visibility.Visible)
                                Display.ArtifactImage.SetImage(GetArtifactImage(Artifact.Rarity))
                        
    UpdateMainArtifact(InPlayer : player, InPlayerStats : Player_Stats_Data, InArtifactRarity : ArtifactRarity):void=
        if(CanvasData := BuiltCanvasMap[InPlayer]):
            CanvasData.ArtifactEquip.MainIcon.SetImage(GetArtifactImage(ArtifactRarity.Null))
            CanvasData.ArtifactEquip.StrengthStat.SetText(ToMessage("+0.00%"))
            CanvasData.ArtifactEquip.CriticalRateStat.SetText(ToMessage("+0.00%"))
            CanvasData.ArtifactEquip.CriticalDamageStat.SetText(ToMessage("+0.00%"))
            CanvasData.ArtifactEquip.FinalStrikeStat.SetText(ToMessage("+0.00%"))

            for(Artifact : InPlayerStats.OwnedPlayerArtifacts):
                if(InArtifactRarity = Artifact.Rarity):
                    CanvasData.ArtifactEquip.MainIcon.SetImage(GetArtifactImage(InArtifactRarity))
                    CanvasData.ArtifactEquip.StrengthStat.SetText(ToMessage("+{FormatFloatTo2DP(GetArtifactStat(Artifact, ArtifactStats.ArtifactBonus))}%"))
                    CanvasData.ArtifactEquip.CriticalRateStat.SetText(ToMessage("+{FormatFloatTo2DP(GetArtifactStat(Artifact, ArtifactStats.CriticalChance))}%"))
                    CanvasData.ArtifactEquip.CriticalDamageStat.SetText(ToMessage("+{FormatFloatTo2DP(GetArtifactStat(Artifact, ArtifactStats.CriticalDamage))}%"))
                    CanvasData.ArtifactEquip.FinalStrikeStat.SetText(ToMessage("+{FormatFloatTo2DP(GetArtifactStat(Artifact, ArtifactStats.FinalStrike))}%"))

    ToggleArtifactDisplayLocked(InArtifactEquipDisplay : ArtifactEquipDisplay, Lock : logic):void=
        if(Lock = false):
            set InArtifactEquipDisplay.IsLocked = false
            InArtifactEquipDisplay.Icon.SetImage(VerseExposed.T_UnknownArtifact)
        else:
            set InArtifactEquipDisplay.IsLocked = true
            InArtifactEquipDisplay.Icon.SetImage(VerseExposed.T_LockIcon)

    MakeArtifactButtonDisplay(InTitle : string, InLevel : string, InRarity : string, InArtifactImage : texture):ArtifactButtonData=
        ArtifactButton := button_quiet:
            DefaultText := ToMessage("")

        ArtifactImage := texture_block:
            DefaultImage := InArtifactImage
            DefaultDesiredSize := vector2{X:= 128.0, Y:=128.0}
        
        ArtifactTitleText := text_block:
            DefaultText := ToMessage("{InTitle}")
            DefaultJustification := text_justification.Center
            DefaultShadowColor := NamedColors.Black
            DefaultShadowOffset := option{vector2{X:=2.0,Y:=2.0}}
        ArtifactTitleText.SetTextColor(NamedColors.White)
        ArtifactTitleText.SetShadowOpacity(0.8)
        ArtifactLevelText := text_block:
            DefaultText := ToMessage("{InLevel}")
            DefaultJustification := text_justification.Center
            DefaultShadowColor := NamedColors.Black
            DefaultShadowOffset := option{vector2{X:=2.0,Y:=2.0}}
        ArtifactLevelText.SetTextColor(NamedColors.White)
        ArtifactLevelText.SetShadowOpacity(0.8)
        ArtifactRarityText := text_block:
            DefaultText := ToMessage("{InRarity}")
            DefaultJustification := text_justification.Center
            DefaultShadowColor := NamedColors.Black
            DefaultShadowOffset := option{vector2{X:=2.0,Y:=2.0}}
        ArtifactRarityText.SetTextColor(NamedColors.Orange)
        ArtifactRarityText.SetShadowOpacity(0.8)

        TextStackBox := stack_box:
            Orientation := orientation.Vertical
            Slots := 
                array:
                    <#stack_box_slot:
                        Widget := ArtifactRarityText
                        Padding := margin{Left := 150.0, Bottom := 2.0}
                        HorizontalAlignment := horizontal_alignment.Left
                        VerticalAlignment := vertical_alignment.Center#>
                    stack_box_slot:
                        Widget := ArtifactRarityText
                        Padding := margin{Left := 150.0, Bottom := 2.0}
                        HorizontalAlignment := horizontal_alignment.Left
                        VerticalAlignment := vertical_alignment.Center
                    stack_box_slot:
                        Widget := ArtifactLevelText
                        Padding := margin{Left := 150.0, Bottom := 2.0}
                        HorizontalAlignment := horizontal_alignment.Left
                        VerticalAlignment := vertical_alignment.Center

        ArtifactOverlay := overlay:
            Slots := array:
                overlay_slot:
                    VerticalAlignment := vertical_alignment.Center
                    HorizontalAlignment := horizontal_alignment.Center
                    Widget := color_block:
                        DefaultDesiredSize := vector2{X:=550.0, Y:=100.0}
                        DefaultOpacity := 0.0
                overlay_slot:
                    VerticalAlignment := vertical_alignment.Center
                    HorizontalAlignment := horizontal_alignment.Left
                    Padding:= margin{Left := 16.0}
                    Widget := ArtifactImage
                overlay_slot:
                    VerticalAlignment := vertical_alignment.Fill
                    HorizontalAlignment := horizontal_alignment.Left
                    Widget := TextStackBox
                overlay_slot:
                    VerticalAlignment := vertical_alignment.Fill
                    HorizontalAlignment := horizontal_alignment.Fill
                    Widget := ArtifactButton

        return ArtifactButtonData:
            Widget := ArtifactOverlay
            ArtifactImage := ArtifactImage
            ArtifactTitleText := ArtifactTitleText
            ArtifactLevelText := ArtifactLevelText
            ArtifactRarityText := ArtifactRarityText
            Button := ArtifactButton

    MakeStatDisplay(InTitle : string, InValue : string):ArtifactStatData=
        Title := text_block:
            DefaultText := ToMessage("{InTitle}:")
            DefaultJustification := text_justification.Left
            DefaultShadowColor := NamedColors.Black
            DefaultShadowOffset := option{vector2{X:=2.0,Y:=2.0}}
        Title.SetTextColor(NamedColors.White)
        Title.SetShadowOpacity(0.8)
        Value := text_block:
            DefaultText := ToMessage("+{InValue}%")
            DefaultJustification := text_justification.Right
            DefaultShadowColor := NamedColors.Black
            DefaultShadowOffset := option{vector2{X:=2.0,Y:=2.0}}
        Value.SetTextColor(NamedColors.LightGreen)
        Value.SetShadowOpacity(0.8)

        StackBox := stack_box:
            Orientation := orientation.Horizontal
            Slots := 
                array:
                    stack_box_slot:
                        Widget := Title
                        HorizontalAlignment := horizontal_alignment.Left
                        VerticalAlignment := vertical_alignment.Center
                    stack_box_slot:
                        Widget := Value
                        Padding := margin{Left := 50.0}
                        HorizontalAlignment := horizontal_alignment.Right
                        VerticalAlignment := vertical_alignment.Center

        return ArtifactStatData:
            Widget := StackBox
            Value := Value

    MakeArtifactSlot():ArtifactEquipDisplay=
        SlotButton := button_quiet:
            DefaultText := ToMessage("")

        Icon := texture_block:
            DefaultImage := VerseExposed.T_UnknownArtifact

        SelectionBlock := color_block:
            DefaultColor := NamedColors.Yellow
            DefaultOpacity := 1.0
            DefaultDesiredSize := vector2{X := 100.0, Y := 100.0}

        SelectionBlock.SetVisibility(widget_visibility.Hidden)

        Overlay := overlay:
            Slots := array:
                overlay_slot:
                    VerticalAlignment := vertical_alignment.Center
                    HorizontalAlignment := horizontal_alignment.Center
                    Widget := SelectionBlock
                overlay_slot:
                    VerticalAlignment := vertical_alignment.Fill
                    HorizontalAlignment := horizontal_alignment.Fill
                    Padding := margin{Top := 6.0, Bottom := 6.0, Left := 6.0, Right := 6.0}
                    Widget := Icon
                overlay_slot:
                    VerticalAlignment := vertical_alignment.Fill
                    HorizontalAlignment := horizontal_alignment.Fill
                    Padding := margin{Top := -8.0, Bottom := -8.0, Left := -8.0, Right := -8.0}
                    Widget := SlotButton

        return ArtifactEquipDisplay:
            IsLocked := false
            Widget := Overlay
            Button := SlotButton
            SelectionBlock := SelectionBlock
            Icon := Icon
            CurrentArtifactRarity := ArtifactRarity.Null

    MakeArtifactEquipCanvas():ArtifactEquipData=

        ArtifcatButton1 := MakeArtifactButtonDisplay("Common Artifact", "Lv. 1", "Common", VerseExposed.T_UnknownArtifact)
        ArtifcatButton2 := MakeArtifactButtonDisplay("Common Artifact", "Lv. 1", "Common", VerseExposed.T_UnknownArtifact)
        ArtifcatButton3 := MakeArtifactButtonDisplay("Common Artifact", "Lv. 1", "Common", VerseExposed.T_UnknownArtifact)
        ArtifcatButton4 := MakeArtifactButtonDisplay("Common Artifact", "Lv. 1", "Common", VerseExposed.T_UnknownArtifact)
        ArtifcatButton5 := MakeArtifactButtonDisplay("Common Artifact", "Lv. 1", "Common", VerseExposed.T_UnknownArtifact)

        ArtifactButtonStackBox := stack_box:
            Orientation := orientation.Vertical
            Slots := 
                array:
                    stack_box_slot:
                        Widget := ArtifcatButton1.Widget
                        Padding := margin{Bottom := 5.0}
                        HorizontalAlignment := horizontal_alignment.Center
                        VerticalAlignment := vertical_alignment.Center
                    stack_box_slot:
                        Widget := ArtifcatButton2.Widget
                        Padding := margin{Bottom := 5.0}
                        HorizontalAlignment := horizontal_alignment.Center
                        VerticalAlignment := vertical_alignment.Center
                    stack_box_slot:
                        Widget := ArtifcatButton3.Widget
                        Padding := margin{Bottom := 5.0}
                        HorizontalAlignment := horizontal_alignment.Center
                        VerticalAlignment := vertical_alignment.Center
                    stack_box_slot:
                        Widget := ArtifcatButton4.Widget
                        Padding := margin{Bottom := 5.0}
                        HorizontalAlignment := horizontal_alignment.Center
                        VerticalAlignment := vertical_alignment.Center
                    stack_box_slot:
                        Widget := ArtifcatButton5.Widget
                        Padding := margin{Bottom := 5.0}
                        HorizontalAlignment := horizontal_alignment.Center
                        VerticalAlignment := vertical_alignment.Center

        BiomeText := text_block:
            DefaultText := ToMessage("Mushroom")
            DefaultJustification := text_justification.Center
            DefaultShadowColor := NamedColors.Black
            DefaultShadowOffset := option{vector2{X:=2.0,Y:=2.0}}
        BiomeText.SetTextColor(NamedColors.White)
        BiomeText.SetShadowOpacity(0.8)

        BiomeNextButton := button_quiet:
            DefaultText := ToMessage(">")
        BiomePrevButton := button_quiet:
            DefaultText := ToMessage("<")

        BiomeSelectionStack := stack_box:
            Orientation := orientation.Horizontal
            Slots := 
                array:
                    stack_box_slot:
                        Widget := BiomePrevButton
                        Padding := margin{Right := 5.0}
                        HorizontalAlignment := horizontal_alignment.Fill
                        VerticalAlignment := vertical_alignment.Center
                    stack_box_slot:
                        Widget := BiomeText
                        Padding := margin{Right := 5.0}
                        HorizontalAlignment := horizontal_alignment.Center
                        VerticalAlignment := vertical_alignment.Center
                    stack_box_slot:
                        Widget := BiomeNextButton
                        HorizontalAlignment := horizontal_alignment.Fill
                        VerticalAlignment := vertical_alignment.Center

        LeftStackBox := stack_box:
            Orientation := orientation.Vertical
            Slots := 
                array:
                    stack_box_slot:
                        Widget := BiomeSelectionStack
                        Padding := margin{Bottom := 10.0}
                        HorizontalAlignment := horizontal_alignment.Center
                        VerticalAlignment := vertical_alignment.Center
                    stack_box_slot:
                        Widget := ArtifactButtonStackBox
                        HorizontalAlignment := horizontal_alignment.Fill
                        VerticalAlignment := vertical_alignment.Center

        StrengthStat := MakeStatDisplay("+ Strength", "0.00")
        CriticalRateStat := MakeStatDisplay("+ Crit Rate", "0.00")
        CriticalDamageStat := MakeStatDisplay("+ Crit Damage", "0.00")
        FinalStrikeStat := MakeStatDisplay("+ Final Hit", "0.00")

        ArtifactDisplayTexture := texture_block:
            DefaultImage := VerseExposed.T_UnknownArtifact
            DefaultDesiredSize := vector2{X:= 200.0, Y:=200.0}

        EquippedStrengthStat := MakeStatDisplay("+ Strength", "0.00")
        EquippedCriticalRateStat := MakeStatDisplay("+ Crit Rate", "0.00")
        EquippedCriticalDamageStat := MakeStatDisplay("+ Crit Damage", "0.00")
        EquippedFinalStrikeStat := MakeStatDisplay("+ Final Hit", "0.00")

        EquippedArtifactDisplayTexture := texture_block:
            DefaultImage := VerseExposed.T_UnknownArtifact
            DefaultDesiredSize := vector2{X:= 200.0, Y:=200.0}

        SlotText := text_block:
            DefaultText := ToMessage("Slot 1")
            DefaultJustification := text_justification.Center
            DefaultShadowColor := NamedColors.Black
            DefaultShadowOffset := option{vector2{X:=2.0,Y:=2.0}}
        SlotText.SetTextColor(NamedColors.White)
        SlotText.SetShadowOpacity(0.8)

        SlotStatTextStackBox := stack_box:
            Orientation := orientation.Vertical
            Slots := 
                array:
                    stack_box_slot:
                        Widget := EquippedStrengthStat.Widget
                        Padding := margin{Bottom := 2.5}
                        HorizontalAlignment := horizontal_alignment.Center
                        VerticalAlignment := vertical_alignment.Top
                    stack_box_slot:
                        Widget := EquippedCriticalRateStat.Widget
                        Padding := margin{Bottom := 2.5}
                        HorizontalAlignment := horizontal_alignment.Center
                        VerticalAlignment := vertical_alignment.Top
                    stack_box_slot:
                        Widget := EquippedCriticalDamageStat.Widget
                        Padding := margin{Bottom := 2.5}
                        HorizontalAlignment := horizontal_alignment.Center
                        VerticalAlignment := vertical_alignment.Top
                    stack_box_slot:
                        Widget := EquippedFinalStrikeStat.Widget
                        Padding := margin{Bottom := 2.5}
                        HorizontalAlignment := horizontal_alignment.Center
                        VerticalAlignment := vertical_alignment.Top

        StatTextStackBox := stack_box:
            Orientation := orientation.Vertical
            Slots := 
                array:
                    stack_box_slot:
                        Widget := StrengthStat.Widget
                        Padding := margin{Bottom := 2.5}
                        HorizontalAlignment := horizontal_alignment.Center
                        VerticalAlignment := vertical_alignment.Top
                    stack_box_slot:
                        Widget := CriticalRateStat.Widget
                        Padding := margin{Bottom := 2.5}
                        HorizontalAlignment := horizontal_alignment.Center
                        VerticalAlignment := vertical_alignment.Top
                    stack_box_slot:
                        Widget := CriticalDamageStat.Widget
                        Padding := margin{Bottom := 2.5}
                        HorizontalAlignment := horizontal_alignment.Center
                        VerticalAlignment := vertical_alignment.Top
                    stack_box_slot:
                        Widget := FinalStrikeStat.Widget
                        Padding := margin{Bottom := 2.5}
                        HorizontalAlignment := horizontal_alignment.Center
                        VerticalAlignment := vertical_alignment.Top
                        
        FullStatTextStackBox := stack_box:
            Orientation := orientation.Horizontal
            Slots := 
                array:
                    stack_box_slot:
                        Widget := ArtifactDisplayTexture
                        HorizontalAlignment := horizontal_alignment.Left
                        VerticalAlignment := vertical_alignment.Top
                    stack_box_slot:
                        Widget := StatTextStackBox
                        Padding := margin{Left := 10.0}
                        HorizontalAlignment := horizontal_alignment.Right
                        VerticalAlignment := vertical_alignment.Top

        FullSlotStatTextStackBox := stack_box:
            Orientation := orientation.Horizontal
            Slots := 
                array:
                    stack_box_slot:
                        Widget := EquippedArtifactDisplayTexture
                        HorizontalAlignment := horizontal_alignment.Left
                        VerticalAlignment := vertical_alignment.Top
                    stack_box_slot:
                        Widget := SlotStatTextStackBox
                        Padding := margin{Left := 10.0}
                        HorizontalAlignment := horizontal_alignment.Right
                        VerticalAlignment := vertical_alignment.Top

        StatOverlay := overlay:
            Slots := array:
                overlay_slot:
                    VerticalAlignment := vertical_alignment.Center
                    HorizontalAlignment := horizontal_alignment.Center
                    Widget := color_block:
                        DefaultColor := NamedColors.White
                        DefaultOpacity := 0.0
                        DefaultDesiredSize := vector2{X := 750.0, Y := 250.0}
                overlay_slot:
                    VerticalAlignment := vertical_alignment.Fill
                    HorizontalAlignment := horizontal_alignment.Fill
                    Widget := FullStatTextStackBox

        SlotStatOverlay := overlay:
            Slots := array:
                overlay_slot:
                    VerticalAlignment := vertical_alignment.Center
                    HorizontalAlignment := horizontal_alignment.Center
                    Widget := color_block:
                        DefaultColor := NamedColors.White
                        DefaultOpacity := 0.0
                        DefaultDesiredSize := vector2{X := 750.0, Y := 250.0}
                overlay_slot:
                    VerticalAlignment := vertical_alignment.Fill
                    HorizontalAlignment := horizontal_alignment.Fill
                    Widget := FullSlotStatTextStackBox

        EquipButton := button_loud:
            DefaultText := ToMessage("EQUIP")

        SlotButton1 := MakeArtifactSlot()
        SlotButton2 := MakeArtifactSlot()
        SlotButton3 := MakeArtifactSlot()
        SlotButton4 := MakeArtifactSlot()

        SlotButtonStacks := stack_box:
            Orientation := orientation.Horizontal
            Slots := 
                array:
                    stack_box_slot:
                        Widget := SlotButton1.Widget
                        Padding := margin{Right := 20.0}
                        HorizontalAlignment := horizontal_alignment.Center
                        VerticalAlignment := vertical_alignment.Center
                    stack_box_slot:
                        Widget := SlotButton2.Widget
                        Padding := margin{Right := 20.0}
                        HorizontalAlignment := horizontal_alignment.Center
                        VerticalAlignment := vertical_alignment.Center
                    stack_box_slot:
                        Widget := SlotButton3.Widget
                        Padding := margin{Right := 20.0}
                        HorizontalAlignment := horizontal_alignment.Center
                        VerticalAlignment := vertical_alignment.Center
                    stack_box_slot:
                        Widget := SlotButton4.Widget
                        HorizontalAlignment := horizontal_alignment.Center
                        VerticalAlignment := vertical_alignment.Center

        RightStackBox := stack_box:
            Orientation := orientation.Vertical
            Slots := 
                array:
                    stack_box_slot:
                        Widget := StatOverlay
                        Padding := margin{Bottom := 10.0}
                        HorizontalAlignment := horizontal_alignment.Fill
                        VerticalAlignment := vertical_alignment.Center
                    stack_box_slot:
                        Widget := EquipButton
                        Padding := margin{Bottom := 20.0}
                        HorizontalAlignment := horizontal_alignment.Fill
                        VerticalAlignment := vertical_alignment.Center
                    stack_box_slot:
                        Widget := SlotButtonStacks
                        Padding := margin{Bottom := 20.0}
                        HorizontalAlignment := horizontal_alignment.Center
                        VerticalAlignment := vertical_alignment.Center
                    stack_box_slot:
                        Widget := SlotText
                        Padding := margin{Bottom := 10.0}
                        HorizontalAlignment := horizontal_alignment.Fill
                        VerticalAlignment := vertical_alignment.Center
                    stack_box_slot:
                        Widget := SlotStatOverlay
                        Padding := margin{Bottom := 10.0}
                        HorizontalAlignment := horizontal_alignment.Fill
                        VerticalAlignment := vertical_alignment.Center
                        
        FullStack := stack_box:
            Orientation := orientation.Horizontal
            Slots := 
                array:
                    stack_box_slot:
                        Widget := LeftStackBox
                        HorizontalAlignment := horizontal_alignment.Fill
                        VerticalAlignment := vertical_alignment.Center
                    stack_box_slot:
                        Widget := RightStackBox
                        Padding := margin{Top := 25.0, Left := 50.0}
                        HorizontalAlignment := horizontal_alignment.Fill
                        VerticalAlignment := vertical_alignment.Top

        OutOverlay := overlay:
            Slots := array:
                overlay_slot:
                    VerticalAlignment := vertical_alignment.Center
                    HorizontalAlignment := horizontal_alignment.Left
                    Padding := margin{Top := 50.0, Left := 200.0}
                    Widget := FullStack

        return ArtifactEquipData:
            Widget := OutOverlay
            SelectedArtifact := ArtifactRarity.Null
            EquipButton := EquipButton
            BiomeText := BiomeText
            BiomeNext := BiomeNextButton
            BiomePrev := BiomePrevButton
            SlotSelected := 1
            BiomeSelected := Biome.Mushroom
            MainIcon := ArtifactDisplayTexture
            StrengthStat := StrengthStat.Value
            CriticalRateStat := CriticalRateStat.Value
            CriticalDamageStat := CriticalDamageStat.Value
            FinalStrikeStat := FinalStrikeStat.Value
            SlotText := SlotText
            EquippedIcon := EquippedArtifactDisplayTexture
            EquippedStrengthStat := EquippedStrengthStat.Value
            EquippedCriticalRateStat := EquippedCriticalRateStat.Value
            EquippedCriticalDamageStat := EquippedCriticalDamageStat.Value
            EquippedFinalStrikeStat := EquippedFinalStrikeStat.Value
            ArtifactSlots := array{SlotButton1, SlotButton2, SlotButton3, SlotButton4}
            ArtifactButtons := array{ArtifcatButton1, ArtifcatButton2, ArtifcatButton3, ArtifcatButton4, ArtifcatButton5}

    MakeInventoryDisplay(InResourceImage : texture):ArtifactInventoryData=
        Text := text_block:
            DefaultText := ToMessage("")
            DefaultJustification := text_justification.Center
            DefaultShadowColor := NamedColors.Black
            DefaultShadowOffset := option{vector2{X:=2.0,Y:=2.0}}
        Text.SetTextColor(NamedColors.White)
        Text.SetShadowOpacity(0.8)

        IconOverlay := overlay:
            Slots := array:
                overlay_slot:
                    VerticalAlignment := vertical_alignment.Center
                    HorizontalAlignment := horizontal_alignment.Center
                    Padding := margin{Top := -8.0, Bottom := -8.0, Left := -6.0, Right := -6.0}
                    Widget := color_block:
                        DefaultColor := NamedColors.White
                        DefaultOpacity := 0.2
                        DefaultDesiredSize := vector2{X := 52.0, Y := 52.0}
                overlay_slot:
                    VerticalAlignment := vertical_alignment.Center
                    HorizontalAlignment := horizontal_alignment.Center
                    Widget := texture_block:
                        DefaultImage := InResourceImage
        
        Overlay := overlay:
            Slots := array:
                overlay_slot:
                    VerticalAlignment := vertical_alignment.Center
                    HorizontalAlignment := horizontal_alignment.Left
                    Widget := color_block:
                        DefaultColor := NamedColors.White
                        DefaultOpacity := 0.2
                        DefaultDesiredSize := vector2{X:= 400.0,Y:=60.0}
                overlay_slot:
                    VerticalAlignment := vertical_alignment.Center
                    HorizontalAlignment := horizontal_alignment.Fill
                    Padding := margin{Top := 10.0, Bottom := 10.0, Left := 10.0, Right := 10.0}
                    Widget := stack_box:
                        Orientation := orientation.Horizontal
                        Slots :=
                            array:
                                stack_box_slot:
                                    Widget := IconOverlay
                                    HorizontalAlignment := horizontal_alignment.Left
                                    VerticalAlignment := vertical_alignment.Center
                                stack_box_slot:
                                    Widget := Text
                                    Padding := margin{Left := 15.0}
                                    VerticalAlignment := vertical_alignment.Center
                                    HorizontalAlignment := horizontal_alignment.Right

        return ArtifactInventoryData:
            Widget := Overlay
            Text := Text

    MakeInventoryPage():InventoryDisplay=
        StoneInventory := MakeInventoryDisplay(VerseExposed.T_Stone)
        PrestigeInventory := MakeInventoryDisplay(VerseExposed.T_RainbowCrystal)
        CopperInventory := MakeInventoryDisplay(VerseExposed.T_CopperOreStroke)
        SilverInventory := MakeInventoryDisplay(VerseExposed.T_SilverOre)
        MalachiteInventory := MakeInventoryDisplay(VerseExposed.T_MalachiteOreStroke)
        SpectroliteInventory := MakeInventoryDisplay(VerseExposed.T_SpectroliteOreStroke)
        ObsidianInventory := MakeInventoryDisplay(VerseExposed.T_ObsidianOreStroke)
        OrbInventory := MakeInventoryDisplay(VerseExposed.T_Orb)
        DustInventory := MakeInventoryDisplay(VerseExposed.T_GlitterDust)
        GoldInventory := MakeInventoryDisplay(VerseExposed.T_GoldCoin)

        StackBox := stack_box:
            Orientation := orientation.Vertical
            Slots := 
                array:
                    stack_box_slot:
                        Widget := StoneInventory.Widget
                        HorizontalAlignment := horizontal_alignment.Center
                        VerticalAlignment := vertical_alignment.Center
                    stack_box_slot:
                        Widget := PrestigeInventory.Widget
                        HorizontalAlignment := horizontal_alignment.Center
                        VerticalAlignment := vertical_alignment.Center
                    stack_box_slot:
                        Widget := CopperInventory.Widget
                        HorizontalAlignment := horizontal_alignment.Center
                        VerticalAlignment := vertical_alignment.Center
                    stack_box_slot:
                        Widget := SilverInventory.Widget
                        HorizontalAlignment := horizontal_alignment.Center
                        VerticalAlignment := vertical_alignment.Center
                    stack_box_slot:
                        Widget := MalachiteInventory.Widget
                        HorizontalAlignment := horizontal_alignment.Center
                        VerticalAlignment := vertical_alignment.Center
                    stack_box_slot:
                        Widget := SpectroliteInventory.Widget
                        HorizontalAlignment := horizontal_alignment.Center
                        VerticalAlignment := vertical_alignment.Center
                    stack_box_slot:
                        Widget := ObsidianInventory.Widget
                        HorizontalAlignment := horizontal_alignment.Center
                        VerticalAlignment := vertical_alignment.Center
                    stack_box_slot:
                        Widget := OrbInventory.Widget
                        HorizontalAlignment := horizontal_alignment.Center
                        VerticalAlignment := vertical_alignment.Center
                    stack_box_slot:
                        Widget := GoldInventory.Widget
                        HorizontalAlignment := horizontal_alignment.Center
                        VerticalAlignment := vertical_alignment.Center
                    stack_box_slot:
                        Widget := DustInventory.Widget
                        HorizontalAlignment := horizontal_alignment.Center
                        VerticalAlignment := vertical_alignment.Center
                        

        Overlay := overlay:
            Slots := array:
                overlay_slot:
                    VerticalAlignment := vertical_alignment.Center
                    HorizontalAlignment := horizontal_alignment.Center
                    Widget := StackBox

        return InventoryDisplay:
            Widget := Overlay
            Stone := StoneInventory.Text
            Prestige := PrestigeInventory.Text
            Copper := CopperInventory.Text
            Silver := SilverInventory.Text
            Malachite := MalachiteInventory.Text
            Spectrolite := SpectroliteInventory.Text
            Obsidian := ObsidianInventory.Text
            Orb := OrbInventory.Text
            Gold := GoldInventory.Text
            Dust := DustInventory.Text

    SplitStringBy4(InQuestName : string):string=
        var FirstHalf : string = ""
        var SecondHalf : string = ""

        var InFirst : logic = true
        for(Index -> Character : InQuestName):
            if(Index = 4):
                set InFirst = false

            if(InFirst = true):
                set FirstHalf += array{Character}
            else:
                set SecondHalf += array{Character}

        OutString := FirstHalf + "\n" + SecondHalf

        return OutString

    MakeAchievementDisplay(InAchievements : []achievement_data):AchievementDisplay=
        Print("Making Achievement Display")
        StoneButton := button_quiet:
            DefaultText := ToMessage("Stone")
        StrengthButton := button_quiet:
            DefaultText := ToMessage("Strength")
        ArtifactButton := button_quiet:
            DefaultText := ToMessage("Artifact")
        RocksButton := button_quiet:
            DefaultText := ToMessage("Rocks")
        DungeonsButton := button_quiet:
            DefaultText := ToMessage("Dungeons")

        var StoneWidgets : widget = overlay{}
        var StrengthWidgets : widget = overlay{}
        var ArtifactsWidgets : widget = overlay{}
        var RocksWidgets : widget = overlay{}
        var DungeonsWidgets : widget = overlay{}

        var StoneWidgetsArray : []text_block = array{}
        var StrengthWidgetsArray : []text_block = array{}
        var ArtifactsWidgetsArray : []text_block = array{}
        var RocksWidgetsArray : []text_block = array{}
        var DungeonsWidgetsArray : []text_block = array{}

        NameText := text_block:
            DefaultText := ToMessage("NAME HERE")
            DefaultJustification := text_justification.Left
            DefaultShadowColor := NamedColors.Black
            DefaultShadowOffset := option{vector2{X:=2.0,Y:=2.0}}
        NameText.SetTextColor(NamedColors.White)
        NameText.SetShadowOpacity(0.8)

        DescriptionText := text_block:
            DefaultText := ToMessage("DESCRIPTION HERE")
            DefaultJustification := text_justification.Left
            DefaultShadowColor := NamedColors.Black
            DefaultShadowOffset := option{vector2{X:=2.0,Y:=2.0}}
        DescriptionText.SetTextColor(NamedColors.LightGrey)
        DescriptionText.SetShadowOpacity(0.8)

        ProgressText := text_block:
            DefaultText := ToMessage("Achievement Progress:")
            DefaultJustification := text_justification.Left
            DefaultShadowColor := NamedColors.Black
            DefaultShadowOffset := option{vector2{X:=2.0,Y:=2.0}}
        ProgressText.SetTextColor(NamedColors.LightGrey)
        ProgressText.SetShadowOpacity(0.8)
        ProgressTitleText := text_block:
            DefaultText := ToMessage("0")
            DefaultJustification := text_justification.Left
            DefaultShadowColor := NamedColors.Black
            DefaultShadowOffset := option{vector2{X:=2.0,Y:=2.0}}
        ProgressTitleText.SetTextColor(NamedColors.LightGreen)
        ProgressTitleText.SetShadowOpacity(0.8)

        ProgressTextStack := stack_box:
            Orientation := orientation.Horizontal
            Slots := 
                array:
                    stack_box_slot:
                        VerticalAlignment := vertical_alignment.Fill
                        HorizontalAlignment := horizontal_alignment.Fill
                        Padding := margin {Right := 10.0}
                        Widget := ProgressText
                    stack_box_slot:
                        VerticalAlignment := vertical_alignment.Fill
                        HorizontalAlignment := horizontal_alignment.Fill
                        Padding := margin {Right := 10.0}
                        Widget := ProgressTitleText

        ButtonStack := stack_box:
            Orientation := orientation.Horizontal
            Slots := 
                array:
                    stack_box_slot:
                        Widget := StoneButton
                        Padding := margin {Right := 25.0}
                        HorizontalAlignment := horizontal_alignment.Center
                        VerticalAlignment := vertical_alignment.Center
                    stack_box_slot:
                        Padding := margin {Right := 25.0}
                        Widget := StrengthButton
                        HorizontalAlignment := horizontal_alignment.Center
                        VerticalAlignment := vertical_alignment.Center
                    stack_box_slot:
                        Padding := margin {Right := 25.0}
                        Widget := ArtifactButton
                        HorizontalAlignment := horizontal_alignment.Center
                        VerticalAlignment := vertical_alignment.Center
                    stack_box_slot:
                        Padding := margin {Right := 25.0}
                        Widget := RocksButton
                        HorizontalAlignment := horizontal_alignment.Center
                        VerticalAlignment := vertical_alignment.Center
                    stack_box_slot:
                        Widget := DungeonsButton
                        HorizontalAlignment := horizontal_alignment.Center
                        VerticalAlignment := vertical_alignment.Center

        for(Achievement : InAchievements):
            var OverallWidget : widget = overlay{}
            var Texts : []text_block = array{}
            var TempWidgets1 : []stack_box_slot = array{}
            var TempWidgets2 : []stack_box_slot = array{}
            var TempWidgets3 : []stack_box_slot = array{}
            var TempWidgets4 : []stack_box_slot = array{}
            for(Index -> Level : Achievement.Levels):

                # Image, Text in a widget with color block
                AmountText := text_block:
                    DefaultText := ToMessage("{SplitStringBy4(BigNumber_FormatDisplay(Level))}")
                    DefaultJustification := text_justification.Center
                    DefaultShadowColor := NamedColors.Black
                    DefaultShadowOffset := option{vector2{X:=2.0,Y:=2.0}}
                AmountText.SetTextColor(NamedColors.Grey)
                AmountText.SetShadowOpacity(0.8)

                StackBoxSlot := stack_box_slot:
                    VerticalAlignment := vertical_alignment.Fill
                    HorizontalAlignment := horizontal_alignment.Fill
                    Padding := margin {Right := 25.0}
                    Widget := overlay:
                        Slots := array:
                            overlay_slot:
                                VerticalAlignment := vertical_alignment.Center
                                HorizontalAlignment := horizontal_alignment.Center
                                Widget := color_block:
                                    DefaultColor := NamedColors.Black
                                    DefaultOpacity := 0.2
                                    DefaultDesiredSize := vector2 {X := 100.0, Y:= 100.0}
                            overlay_slot:
                                VerticalAlignment := vertical_alignment.Center
                                HorizontalAlignment := horizontal_alignment.Center
                                Widget := texture_block:
                                    DefaultImage := Achievement.Image
                                    DefaultDesiredSize := vector2 {X := 90.0, Y:= 90.0}
                            overlay_slot:
                                VerticalAlignment := vertical_alignment.Center
                                HorizontalAlignment := horizontal_alignment.Center
                                Widget := AmountText
                
                set Texts += array{AmountText}
                if(Index <= 9):
                    set TempWidgets1 += array{StackBoxSlot}
                else if (Index > 9 and Index <= 19):
                    set TempWidgets2 += array{StackBoxSlot}
                else if (Index > 19 and Index <= 29):
                    set TempWidgets3 += array{StackBoxSlot}
                else if (Index > 29):
                    set TempWidgets4 += array{StackBoxSlot}

                StackBox1 := stack_box:
                    Orientation := orientation.Horizontal
                    Slots := TempWidgets1
                StackBox2 := stack_box:
                    Orientation := orientation.Horizontal
                    Slots := TempWidgets2
                StackBox3 := stack_box:
                    Orientation := orientation.Horizontal
                    Slots := TempWidgets3
                StackBox4 := stack_box:
                    Orientation := orientation.Horizontal
                    Slots := TempWidgets4
                    
                TotalStack := stack_box:
                    Orientation := orientation.Vertical
                    Slots := array:
                        stack_box_slot:
                            VerticalAlignment := vertical_alignment.Fill
                            HorizontalAlignment := horizontal_alignment.Fill
                            Padding := margin {Bottom := 25.0}
                            Widget := StackBox1
                        stack_box_slot:
                            VerticalAlignment := vertical_alignment.Fill
                            HorizontalAlignment := horizontal_alignment.Fill
                            Padding := margin {Bottom := 25.0}
                            Widget := StackBox2
                        stack_box_slot:
                            VerticalAlignment := vertical_alignment.Fill
                            HorizontalAlignment := horizontal_alignment.Fill
                            Padding := margin {Bottom := 25.0}
                            Widget := StackBox3
                        stack_box_slot:
                            VerticalAlignment := vertical_alignment.Fill
                            HorizontalAlignment := horizontal_alignment.Fill
                            Widget := StackBox4

                set OverallWidget = TotalStack

            case(Achievement.Type):
                AchievementType.TotalStone =>
                    set StoneWidgets = OverallWidget
                    set StoneWidgetsArray = Texts
                AchievementType.HighestStrength =>
                    set StrengthWidgets = OverallWidget
                    set StrengthWidgetsArray = Texts
                AchievementType.RocksDestroyed =>
                    set RocksWidgets = OverallWidget
                    set RocksWidgetsArray = Texts
                AchievementType.ArtifactsSmelted =>
                    set ArtifactsWidgets = OverallWidget
                    set ArtifactsWidgetsArray = Texts
                AchievementType.DungeonsCompleted =>
                    set DungeonsWidgets = OverallWidget
                    set DungeonsWidgetsArray = Texts
                AchievementType.Null =>

        StrengthWidgets.SetVisibility(widget_visibility.Hidden)
        RocksWidgets.SetVisibility(widget_visibility.Hidden)
        ArtifactsWidgets.SetVisibility(widget_visibility.Hidden)
        DungeonsWidgets.SetVisibility(widget_visibility.Hidden)

        AllAchievementsOverlay := overlay:
            Slots := array:
                overlay_slot:
                    VerticalAlignment := vertical_alignment.Top
                    HorizontalAlignment := horizontal_alignment.Center
                    Widget := StoneWidgets
                overlay_slot:
                    VerticalAlignment := vertical_alignment.Top
                    HorizontalAlignment := horizontal_alignment.Center
                    Widget := StrengthWidgets
                overlay_slot:
                    VerticalAlignment := vertical_alignment.Top
                    HorizontalAlignment := horizontal_alignment.Center
                    Widget := RocksWidgets
                overlay_slot:
                    VerticalAlignment := vertical_alignment.Top
                    HorizontalAlignment := horizontal_alignment.Center
                    Widget := ArtifactsWidgets
                overlay_slot:
                    VerticalAlignment := vertical_alignment.Top
                    HorizontalAlignment := horizontal_alignment.Center
                    Widget := DungeonsWidgets

        OutStack := stack_box:
            Orientation := orientation.Vertical
            Slots := array:
                stack_box_slot:
                    VerticalAlignment := vertical_alignment.Fill
                    HorizontalAlignment := horizontal_alignment.Fill
                    Padding := margin {Bottom := 25.0}
                    Widget := ButtonStack
                stack_box_slot:
                    VerticalAlignment := vertical_alignment.Fill
                    HorizontalAlignment := horizontal_alignment.Left
                    Padding := margin {Bottom := 5.0}
                    Widget := NameText
                stack_box_slot:
                    VerticalAlignment := vertical_alignment.Fill
                    HorizontalAlignment := horizontal_alignment.Left
                    Padding := margin {Bottom := 5.0}
                    Widget := DescriptionText
                stack_box_slot:
                    VerticalAlignment := vertical_alignment.Fill
                    HorizontalAlignment := horizontal_alignment.Left
                    Padding := margin {Bottom := 25.0}
                    Widget := ProgressTextStack
                stack_box_slot:
                    VerticalAlignment := vertical_alignment.Fill
                    HorizontalAlignment := horizontal_alignment.Fill
                    Widget := AllAchievementsOverlay

        OutOverlay := overlay:
            Slots := array:
                overlay_slot:
                    VerticalAlignment := vertical_alignment.Center
                    HorizontalAlignment := horizontal_alignment.Left
                    Padding := margin {Left := 250.0}
                    Widget := OutStack

        return AchievementDisplay:
            OverallWidget := OutOverlay
            CurrentWidget := StoneWidgets
            TotalStone := StoneWidgets
            RocksDestroyed := RocksWidgets
            DungeonCompleted := DungeonsWidgets
            ArtifactsSmelted := ArtifactsWidgets
            HeighestStrength := StrengthWidgets
            TotalStoneWidgets := StoneWidgetsArray
            RocksDestroyedWidgets := RocksWidgetsArray
            DungeonCompletedWidgets := DungeonsWidgetsArray
            ArtifactsSmeltedWidgets := ArtifactsWidgetsArray
            HeighestStrengthWidgets := StrengthWidgetsArray
            TotalStoneButton := StoneButton
            RocksDestroyedButton := RocksButton
            DungeonCompletedButton := DungeonsButton
            ArtifactsSmeltedButton := ArtifactButton
            HeighestStrengthButton := StrengthButton
            NameText := NameText
            DescriptionText := DescriptionText
            ProgressText := ProgressTitleText
            Type := AchievementType.TotalStone