using { /Fortnite.com/Devices }
using { /Verse.org/Simulation }
using { /UnrealEngine.com/Temporary/Diagnostics }
using { /UnrealEngine.com/Temporary/UI }
using { /Fortnite.com/UI }
using { /UnrealEngine.com/Temporary/SpatialMath}
using { /Verse.org/Colors }
using { /Verse.org/Assets }

MakePlayerUpgradeItem<constructor>(InImage : texture, InLevelInt : int, InLevel : string, InCurrentBonus : float, InNextBonus : float, InName : string, InDescription : string, InCost : string, InTrigger : trigger_device, InDisplayPercentage : logic) := PlayerUpgradeItemData:
    Image := InImage
    LevelInt := InLevelInt
    Level := InLevel
    CurrentBonus := InCurrentBonus
    NextBonus := InNextBonus
    Name := InName
    Description := InDescription
    Cost := InCost
    Trigger := InTrigger
    DisplayPercentage := InDisplayPercentage

PlayerUpgradeItemData := class:
    Image : texture
    LevelInt : int
    Level : string
    CurrentBonus : float
    NextBonus : float
    Name : string
    Description : string
    Cost : string
    Trigger : trigger_device
    DisplayPercentage : logic

PlayerUpgradeSlotData := class:
    Name : string
    Level : text_base
    Cost : text_base
    CurrentBonus : text_base
    NextBonus : text_base
    ResourceTexture : texture_block
    Widget : widget
    Button : text_button_base
    ButtonColor : texture_block

PlayerUpgradeCanvasData := class:
    ExitButton : text_button_base
    PlayerUpgradeSlots : []PlayerUpgradeSlotData
    StoneAmount : text_base
    PrestigeAmount : text_base
    Canvas : canvas

UI_PlayerUpgradeVendor := class():
    ToMessage<localizes>(String : string) : message = "{String}"

    var BuiltCanvasMap : [player]PlayerUpgradeCanvasData = map{}

    CenteredAlignment : vector2 = vector2{X := 0.5, Y := 0.5}

    # Initiliase the UI so it can added to the player much quicker
    InitUI(InPlayer : player, InPlayerUpgradeItems : []PlayerUpgradeItemData, InVendorTexture : texture):void=
        PlayerUpgradeCanvas := MakePlayerUpgradeCanvas(InPlayer, InPlayerUpgradeItems, InVendorTexture)
        # Set the players build canvas data
        if(set BuiltCanvasMap[InPlayer] = PlayerUpgradeCanvas){}
        PlayerUpgradeCanvas.ExitButton.OnClick().Subscribe(HandlePlayerExit)

    # Add the UI to the player
    AddUI(InPlayer : player):void=
        if(PlayerUI := GetPlayerUI[InPlayer], CanvasData := BuiltCanvasMap[InPlayer]):
            CanvasData.Canvas.SetVisibility(widget_visibility.Visible)
            PlayerUI.AddWidget(CanvasData.Canvas, player_ui_slot{InputMode := ui_input_mode.All} )

    # When player presses the exit button
    HandlePlayerExit(WidgetMessage: widget_message):void=
        spawn{HideUI(WidgetMessage.Player)}

    HideUI(Player: player)<suspends>:void=
        Sleep(0.0) # So we're sure that the canvas has been added to BuiltCanvas map
        if(PlayerUI := GetPlayerUI[Player], CanvasData := BuiltCanvasMap[Player]):
            CanvasData.Canvas.SetVisibility(widget_visibility.Hidden)
            PlayerUI.RemoveWidget(CanvasData.Canvas)

    UpdateUI(InPlayer : player, InPlayerUpgradeItems : []PlayerUpgradeItemData, InPlayerData : Player_Stats_Data, InCosts : []big_number, UseFortniteResource : logic, UsePrestigeResource : logic, NeedFirstUpgrade : logic):void=
        Print("Update")
        var FirstUpgraded : logic = false
        if(CanvasData := BuiltCanvasMap[InPlayer]):
            CanvasData.StoneAmount.SetText(ToMessage(BigNumber_FormatDisplay(InPlayerData.PlayerStoneOreAmount)))
            CanvasData.PrestigeAmount.SetText(ToMessage(BigNumber_FormatDisplay(InPlayerData.PlayerPrestigeResourceAmount)))
            for(NewItem : InPlayerUpgradeItems):
                for(Index -> Slot : CanvasData.PlayerUpgradeSlots):
                    if(NewItem.Name = Slot.Name):
                        Slot.Level.SetText(ToMessage(NewItem.Level))
                        Slot.Cost.SetText(ToMessage(NewItem.Cost))
                        if(NewItem.DisplayPercentage = true):
                            Slot.CurrentBonus.SetText(ToMessage("{FormatFloatTo2DP(NewItem.CurrentBonus)}%"))
                            Slot.NextBonus.SetText(ToMessage(" (+ {FormatFloatTo2DP(NewItem.NextBonus - NewItem.CurrentBonus)}%)"))
                        else:
                            Slot.CurrentBonus.SetText(ToMessage(FormatFloatTo2DP(NewItem.CurrentBonus)))
                            Slot.NextBonus.SetText(ToMessage(" (+ {FormatFloatTo2DP(NewItem.NextBonus - NewItem.CurrentBonus)})"))
                        if(UseFortniteResource = true):
                            Slot.ResourceTexture.SetImage(VerseExposed.T_GoldCoin)
                        else:
                            Slot.ResourceTexture.SetImage(VerseExposed.T_Stone)
                        if(UsePrestigeResource = true):
                            Slot.ResourceTexture.SetImage(VerseExposed.T_RainbowCrystal)
                        else:
                            Slot.ResourceTexture.SetImage(VerseExposed.T_Stone)
        if(FirstPlayerUpgradeItem := InPlayerUpgradeItems[0], NeedFirstUpgrade = true):
            if(FirstPlayerUpgradeItem.LevelInt > 0):
                set FirstUpgraded = true
        
        ShouldUpgradesShow(InPlayer, InPlayerData, InCosts, FirstUpgraded, NeedFirstUpgrade, UsePrestigeResource)

    ShouldUpgradesShow(InPlayer : player, InPlayerData : Player_Stats_Data, InCosts : []big_number, FirstUpgraded : logic, NeedFirstUpgrade : logic, UsePrestigeResource : logic):void=
        Print("Checking For Show")
        if(CanvasData := BuiltCanvasMap[InPlayer]):
            for(Index -> Slot : CanvasData.PlayerUpgradeSlots):
                if(Cost := InCosts[Index]):
                    var ResourceAmount : big_number = big_number{}
                    if(UsePrestigeResource = true):
                        set ResourceAmount = InPlayerData.PlayerPrestigeResourceAmount
                    else:
                        set ResourceAmount = InPlayerData.PlayerStoneOreAmount
                    BNCost := Cost
                    CanAfford := BigNumber_GreaterThanOrEquals(ResourceAmount, BNCost)
                    if(CanAfford = true):
                        Slot.ButtonColor.SetTint(NamedColors.Yellow)
                        Slot.Cost.SetTextColor(NamedColors.White)
                        Slot.Button.SetText(ToMessage("PURCHASE"))
                    else:
                        Slot.ButtonColor.SetTint(NamedColors.Grey)
                        Slot.Cost.SetTextColor(NamedColors.Red)
                        Slot.Button.SetText(ToMessage("CAN'T AFFORD"))
                
                    if(NeedFirstUpgrade = true and FirstUpgraded = false and Index <> 0):
                        Slot.ButtonColor.SetTint(NamedColors.Grey)
                        Slot.Button.SetText(ToMessage("NEED FIRST UPGRADE"))
                        Slot.Widget.SetVisibility(widget_visibility.Collapsed)
                    else if(Index <> 0 and NeedFirstUpgrade = true and FirstUpgraded = true):
                        Slot.Widget.SetVisibility(widget_visibility.Visible)

                    if(Index = 0 and NeedFirstUpgrade = true and FirstUpgraded = true):
                        #Slot.Widget.SetVisibility(widget_visibility.Collapsed)
                        #Slot.Button.SetVisibility(widget_visibility.Hidden)
                        Slot.Button.SetEnabled(false)
                    else if (Index = 0 and NeedFirstUpgrade = true and FirstUpgraded = false):
                        #lot.Widget.SetVisibility(widget_visibility.Visible)
                        #Slot.Button.SetVisibility(widget_visibility.Visible)
                        Slot.Button.SetEnabled(true)

                CostText := Slot.Cost.GetText()
                if(CostText = "MAXED"):
                    Slot.Button.SetText(ToMessage("MAXED"))
                    Slot.NextBonus.SetVisibility(widget_visibility.Collapsed)
                else:
                    Slot.NextBonus.SetVisibility(widget_visibility.Visible)

    MakePlayerUpgradeCanvas(InPlayer : player, InPlayerUpgradeItems : []PlayerUpgradeItemData, InVendorTexture : texture):PlayerUpgradeCanvasData=
        ExitButton := button_loud:
            DefaultText := ToMessage("Exit")

        # Make an array with all of the vendor slots made
        var VendorSlotsArray : []PlayerUpgradeSlotData = array{}
        var StackBoxSlots : []stack_box_slot = array{}

        # Set up PlayerUpgradeSlots
        for(Index -> CurrentItem : InPlayerUpgradeItems):
            CurrentPlayerUpgradeSlot := MakePlayerUpgradeSlot(CurrentItem.Image, CurrentItem.Level, CurrentItem.Name, CurrentItem.Description, CurrentItem.Cost, CurrentItem.CurrentBonus, CurrentItem.NextBonus)

            CurrentPlayerUpgradeSlot.Button.OnClick().Subscribe(TriggerHandler{Trigger := CurrentItem.Trigger, PlayerRef := array{InPlayer}}.HandlerFunction)
            set VendorSlotsArray += array{CurrentPlayerUpgradeSlot}

        # Format the Slots based on the amount of them
        case(VendorSlotsArray.Length):
            1 =>
                if(VendorSlot1 := VendorSlotsArray[0]):
                    StackBoxSlot1 := stack_box_slot:
                        VerticalAlignment := vertical_alignment.Center
                        HorizontalAlignment := horizontal_alignment.Center
                        Widget := VendorSlot1.Widget
                    set StackBoxSlots += array{StackBoxSlot1}
            2 =>
                if(VendorSlot1 := VendorSlotsArray[0], VendorSlot2 := VendorSlotsArray[1]):
                    StackBoxSlot1 := stack_box_slot:
                        Padding := margin{Left := 50.0, Right := 50.0}
                        VerticalAlignment := vertical_alignment.Center
                        HorizontalAlignment := horizontal_alignment.Center
                        Widget := VendorSlot1.Widget
                    StackBoxSlot2 := stack_box_slot:
                        Padding := margin{Left := 50.0, Right := 50.0}
                        VerticalAlignment := vertical_alignment.Center
                        HorizontalAlignment := horizontal_alignment.Center
                        Widget := VendorSlot2.Widget
                    set StackBoxSlots += array{StackBoxSlot1, StackBoxSlot2}
            3 =>
                if(VendorSlot1 := VendorSlotsArray[0], VendorSlot2 := VendorSlotsArray[1], VendorSlot3 := VendorSlotsArray[2]):
                    StackBoxSlot1 := stack_box_slot:
                        Padding := margin{Left := 30.0, Right := 30.0}
                        VerticalAlignment := vertical_alignment.Center
                        HorizontalAlignment := horizontal_alignment.Center
                        Widget := VendorSlot1.Widget
                    StackBoxSlot2 := stack_box_slot:
                        Padding := margin{Left := 30.0, Right := 30.0}
                        VerticalAlignment := vertical_alignment.Center
                        HorizontalAlignment := horizontal_alignment.Center
                        Widget := VendorSlot2.Widget
                    StackBoxSlot3 := stack_box_slot:
                        Padding := margin{Left := 30.0, Right := 30.0}
                        VerticalAlignment := vertical_alignment.Center
                        HorizontalAlignment := horizontal_alignment.Center
                        Widget := VendorSlot3.Widget
                    set StackBoxSlots += array{StackBoxSlot1, StackBoxSlot2, StackBoxSlot3}
            4 =>
                if(VendorSlot1 := VendorSlotsArray[0], VendorSlot2 := VendorSlotsArray[1], VendorSlot3 := VendorSlotsArray[2], VendorSlot4 := VendorSlotsArray[3]):
                    StackBoxSlot1 := stack_box_slot:
                        Padding := margin{Left := 5.0, Right := 5.0}
                        VerticalAlignment := vertical_alignment.Center
                        HorizontalAlignment := horizontal_alignment.Center
                        Widget := VendorSlot1.Widget
                    StackBoxSlot2 := stack_box_slot:
                        Padding := margin{Left := 5.0, Right := 5.0}
                        VerticalAlignment := vertical_alignment.Center
                        HorizontalAlignment := horizontal_alignment.Center
                        Widget := VendorSlot2.Widget
                    StackBoxSlot3 := stack_box_slot:
                        Padding := margin{Left := 5.0, Right := 5.0}
                        VerticalAlignment := vertical_alignment.Center
                        HorizontalAlignment := horizontal_alignment.Center
                        Widget := VendorSlot3.Widget
                    StackBoxSlot4 := stack_box_slot:
                        Padding := margin{Left := 5.0, Right := 5.0}
                        VerticalAlignment := vertical_alignment.Center
                        HorizontalAlignment := horizontal_alignment.Center
                        Widget := VendorSlot4.Widget
                    set StackBoxSlots += array{StackBoxSlot1, StackBoxSlot2, StackBoxSlot3, StackBoxSlot4}
            _=>
                Print("VendorSlotsArray is not the correct length")

        BorderSize := 1.0
        PlayerStoneText := text_block:
            DefaultText := ToMessage("")
            DefaultJustification := text_justification.Center
            DefaultShadowColor := NamedColors.Black
            DefaultShadowOffset := option{vector2{X:=2.0,Y:=2.0}}
        PlayerStoneText.SetTextColor(NamedColors.White)
        PlayerStoneText.SetShadowOpacity(0.8)
        PlayerPrestigeText := text_block:
            DefaultText := ToMessage("")
            DefaultJustification := text_justification.Center
            DefaultShadowColor := NamedColors.Black
            DefaultShadowOffset := option{vector2{X:=2.0,Y:=2.0}}
        PlayerPrestigeText.SetTextColor(NamedColors.White)
        PlayerPrestigeText.SetShadowOpacity(0.8)
            
        TitleTexture := texture_block:
            DefaultImage := InVendorTexture
            DefaultDesiredSize := vector2{X:=512.0,Y:=512.0}

        TitleOverlay := overlay:
            Slots := array:
                overlay_slot:
                    VerticalAlignment := vertical_alignment.Center
                    HorizontalAlignment := horizontal_alignment.Center
                    Widget := TitleTexture

        StoneIconOverlay := overlay:
            Slots := array:
                overlay_slot:
                    VerticalAlignment := vertical_alignment.Center
                    HorizontalAlignment := horizontal_alignment.Center
                    Padding := margin{Top := -8.0, Bottom := -8.0, Left := -6.0, Right := -6.0}
                    Widget := color_block:
                        DefaultColor := NamedColors.White
                        DefaultOpacity := 0.4
                        DefaultDesiredSize := vector2{X := 52.0, Y := 52.0}
                overlay_slot:
                    VerticalAlignment := vertical_alignment.Center
                    HorizontalAlignment := horizontal_alignment.Center
                    Widget := texture_block:
                        DefaultImage := VerseExposed.T_Stone
        
        PlayerStoneOverlay := overlay:
            Slots := array:
                overlay_slot:
                    VerticalAlignment := vertical_alignment.Center
                    HorizontalAlignment := horizontal_alignment.Left
                    Widget := color_block:
                        DefaultColor := NamedColors.White
                        DefaultOpacity := 0.4
                        DefaultDesiredSize := vector2{X:=225.0,Y:=60.0}
                overlay_slot:
                    VerticalAlignment := vertical_alignment.Center
                    HorizontalAlignment := horizontal_alignment.Left
                    Padding := margin{Top := 10.0, Bottom := 10.0, Left := 10.0, Right := 10.0}
                    Widget := stack_box:
                        Orientation := orientation.Horizontal
                        Slots := 
                            array:
                                stack_box_slot:
                                    Widget := StoneIconOverlay
                                    HorizontalAlignment := horizontal_alignment.Left
                                    VerticalAlignment := vertical_alignment.Center
                                stack_box_slot:
                                    Widget := PlayerStoneText
                                    Padding := margin{Left := 15.0}
                                    VerticalAlignment := vertical_alignment.Center
                                    HorizontalAlignment := horizontal_alignment.Left

        PrestigeIconOverlay := overlay:
            Slots := array:
                overlay_slot:
                    VerticalAlignment := vertical_alignment.Center
                    HorizontalAlignment := horizontal_alignment.Center
                    Padding := margin{Top := -8.0, Bottom := -8.0, Left := -6.0, Right := -6.0}
                    Widget := color_block:
                        DefaultColor := NamedColors.White
                        DefaultOpacity := 0.4
                        DefaultDesiredSize := vector2{X := 52.0, Y := 52.0}
                overlay_slot:
                    VerticalAlignment := vertical_alignment.Center
                    HorizontalAlignment := horizontal_alignment.Center
                    Widget := texture_block:
                        DefaultImage := VerseExposed.T_RainbowCrystal
        
        PlayerPrestigeOverlay := overlay:
            Slots := array:
                overlay_slot:
                    VerticalAlignment := vertical_alignment.Center
                    HorizontalAlignment := horizontal_alignment.Left
                    Widget := color_block:
                        DefaultColor := NamedColors.White
                        DefaultOpacity := 0.4
                        DefaultDesiredSize := vector2{X:=225.0,Y:=60.0}
                overlay_slot:
                    VerticalAlignment := vertical_alignment.Center
                    HorizontalAlignment := horizontal_alignment.Left
                    Padding := margin{Top := 10.0, Bottom := 10.0, Left := 10.0, Right := 10.0}
                    Widget := stack_box:
                        Orientation := orientation.Horizontal
                        Slots := 
                            array:
                                stack_box_slot:
                                    Widget := PrestigeIconOverlay
                                    HorizontalAlignment := horizontal_alignment.Left
                                    VerticalAlignment := vertical_alignment.Center
                                stack_box_slot:
                                    Widget := PlayerPrestigeText
                                    Padding := margin{Left := 15.0}
                                    VerticalAlignment := vertical_alignment.Center
                                    HorizontalAlignment := horizontal_alignment.Left
        
        # Create the Stack Box Slot
        CurrentPlayerUpgradeOverlay := overlay:
            Slots := array:
                overlay_slot:
                    VerticalAlignment := vertical_alignment.Fill
                    HorizontalAlignment := horizontal_alignment.Fill
                    Widget := overlay:
                        Slots := array:
                            overlay_slot:
                                VerticalAlignment := vertical_alignment.Center
                                HorizontalAlignment := horizontal_alignment.Center
                                Widget := stack_box:
                                    Orientation := orientation.Horizontal
                                    Slots := StackBoxSlots

        BackgroundOverlay := overlay:
            Slots := array:
                overlay_slot:
                    VerticalAlignment := vertical_alignment.Fill
                    HorizontalAlignment := horizontal_alignment.Fill
                    Widget := texture_block:
                        DefaultImage := VerseExposed.T_UIBackground
        
        # Create the canvas for the whole widget
        Canvas := canvas:
            Slots := array:
                canvas_slot:
                    Anchors := anchors{Minimum := vector2{X:=0.0, Y:=0.0}, Maximum := vector2{X:=1.0, Y:=1.0}}
                    Alignment := CenteredAlignment
                    Widget := BackgroundOverlay
                canvas_slot:
                    Anchors := anchors{Minimum := vector2{X:=0.5, Y:=0.1}, Maximum := vector2{X:=0.5, Y:=0.1}}
                    Alignment := CenteredAlignment
                    Widget := TitleOverlay
                canvas_slot:
                    Anchors := anchors{Minimum := vector2{X:=0.0, Y:=0.0}, Maximum := vector2{X:=1.0, Y:=1.0}}
                    Alignment := CenteredAlignment
                    Widget := CurrentPlayerUpgradeOverlay
                canvas_slot:
                    # Fix for controllers not being able to find buttons that are not perpendicular to the selected button possition
                    Anchors := anchors{Minimum := vector2{X:=0.4, Y:=0.95}, Maximum := vector2{X:=0.6, Y:=0.95}}
                    Alignment := CenteredAlignment
                    Widget := ExitButton
                canvas_slot:
                    Anchors := anchors{Minimum := vector2{X:=0.06255, Y:=0.0475}, Maximum := vector2{X:=0.2, Y:=0.175}}
                    Alignment := CenteredAlignment
                    Widget := PlayerStoneOverlay
                    SizeToContent := true
                    ZOrder := 1
                canvas_slot:
                    Anchors := anchors{Minimum := vector2{X:=0.06255, Y:=0.175}, Maximum := vector2{X:=0.2, Y:=0.175}}
                    Alignment := CenteredAlignment
                    Widget := PlayerPrestigeOverlay
                    SizeToContent := true
                    ZOrder := 1

        return PlayerUpgradeCanvasData:
            ExitButton := ExitButton
            PlayerUpgradeSlots := VendorSlotsArray
            StoneAmount := PlayerStoneText
            PrestigeAmount := PlayerPrestigeText
            Canvas := Canvas
    
    # Make a singular Teleporter slot
    MakePlayerUpgradeSlot(InImage : texture, InLevel : string, InName : string, InDescription : string, InCost : string, InCurrentBonus : float, InNextBonus : float):PlayerUpgradeSlotData=
        IconMargin := margin{Bottom := 175.0, Left := 10.0, Right := 10.0, Top := 30.0}
        PurchaseButton := button_quiet:
            DefaultText := ToMessage("Purchase")

        PurchaseButtonColor := texture_block:
            DefaultImage := VerseExposed.T_PlayerUpgradeButtonBackground
            DefaultDesiredSize := vector2{X:=429.0,Y:=54.0}
            DefaultTint := NamedColors.Yellow

        LevelBlock := text_block:
            DefaultText := ToMessage(InLevel)
            DefaultJustification := text_justification.Center
            DefaultShadowColor := NamedColors.Black
            DefaultShadowOffset := option{vector2{X:=2.0,Y:=2.0}}
        LevelBlock.SetTextColor(NamedColors.White)
        LevelBlock.SetShadowOpacity(0.8)

        NameBlock := text_block:
            DefaultText := ToMessage(InName)
            DefaultJustification := text_justification.Center
            DefaultShadowColor := NamedColors.Black
            DefaultShadowOffset := option{vector2{X:=2.0,Y:=2.0}}
        NameBlock.SetTextColor(NamedColors.White)
        NameBlock.SetShadowOpacity(0.8)

        DescriptionBlock := text_block:
            DefaultText := ToMessage(InDescription)
            DefaultJustification := text_justification.Center
            DefaultShadowColor := NamedColors.Black
            DefaultShadowOffset := option{vector2{X:=2.0,Y:=2.0}}
        DescriptionBlock.SetTextColor(NamedColors.Grey)
        DescriptionBlock.SetShadowOpacity(0.8)

        CurrentBonusText := text_block:
            DefaultText := ToMessage(FormatFloatTo2DP(InCurrentBonus))
            DefaultJustification := text_justification.Center
            DefaultShadowColor := NamedColors.Black
            DefaultShadowOffset := option{vector2{X:=2.0,Y:=2.0}}
        CurrentBonusText.SetTextColor(NamedColors.White)
        CurrentBonusText.SetShadowOpacity(0.8)

        NextBonusText := text_block:
            DefaultText := ToMessage(" (+ {FormatFloatTo2DP(InNextBonus)})")
            DefaultJustification := text_justification.Center
            DefaultShadowColor := NamedColors.Black
            DefaultShadowOffset := option{vector2{X:=2.0,Y:=2.0}}
        NextBonusText.SetTextColor(NamedColors.LightGreen)
        NextBonusText.SetShadowOpacity(0.8)

        CostBlock := text_block:
            DefaultText := ToMessage(InCost)
            DefaultJustification := text_justification.Center
            DefaultShadowColor := NamedColors.Black
            DefaultShadowOffset := option{vector2{X:=2.0,Y:=2.0}}
        CostBlock.SetTextColor(NamedColors.White)
        CostBlock.SetShadowOpacity(0.8)

        ResourceTexture := texture_block:
            DefaultImage := VerseExposed.T_Stone
        
        BackgroundColorBlock := color_block:
            DefaultColor := NamedColors.White
            DefaultOpacity := 0.4

        BackgroundColorBlock2 := color_block:
            DefaultColor := NamedColors.White
            DefaultOpacity := 0.4

        BorderSize := 1.0
        TopBorderColorBlock := color_block{DefaultColor := NamedColors.DarkBlue, DefaultDesiredSize := vector2{X := BorderSize, Y := BorderSize}}
        RightBorderColorBlock := color_block{DefaultColor := NamedColors.DarkBlue, DefaultDesiredSize := vector2{X := BorderSize, Y := BorderSize}}
        BottomBorderColorBlock := color_block{DefaultColor := NamedColors.DarkBlue, DefaultDesiredSize := vector2{X := BorderSize, Y := BorderSize}}
        LeftBorderColorBlock := color_block{DefaultColor := NamedColors.DarkBlue, DefaultDesiredSize := vector2{X := BorderSize, Y := BorderSize}}

        ButtonOverlay := overlay:
            Slots := array:
                overlay_slot:
                    VerticalAlignment := vertical_alignment.Fill
                    HorizontalAlignment := horizontal_alignment.Fill
                    Widget := color_block:
                        DefaultColor := NamedColors.White
                        DefaultOpacity := 0.0
                overlay_slot:
                    VerticalAlignment := vertical_alignment.Center
                    HorizontalAlignment := horizontal_alignment.Center
                    Widget := PurchaseButtonColor
                overlay_slot:
                    VerticalAlignment := vertical_alignment.Center
                    HorizontalAlignment := horizontal_alignment.Fill
                    Padding := margin{Top := 3.0, Bottom := 3.0, Left := 10.0, Right := 10.0}
                    Widget := PurchaseButton

        BonusStack := stack_box :
            Orientation := orientation.Horizontal
            Slots := 
                array:
                    stack_box_slot:
                        VerticalAlignment := vertical_alignment.Center
                        HorizontalAlignment := horizontal_alignment.Center
                        Widget := CurrentBonusText
                    stack_box_slot:
                        VerticalAlignment := vertical_alignment.Center
                        HorizontalAlignment := horizontal_alignment.Center
                        Widget := NextBonusText

        CostStack := stack_box :
            Orientation := orientation.Horizontal
            Slots := 
                array:
                    stack_box_slot:
                        VerticalAlignment := vertical_alignment.Center
                        HorizontalAlignment := horizontal_alignment.Center
                        Widget := ResourceTexture
                    stack_box_slot:
                        VerticalAlignment := vertical_alignment.Center
                        HorizontalAlignment := horizontal_alignment.Center
                        Widget := CostBlock

        ImageOverlay := overlay:
            Slots := array:
                overlay_slot:
                    VerticalAlignment := vertical_alignment.Fill
                    HorizontalAlignment := horizontal_alignment.Fill
                    Padding := margin{Bottom := 10.0, Left := 10.0, Right := 10.0, Top := 10.0}
                    Widget := texture_block:
                        DefaultImage := InImage
                        DefaultDesiredSize := vector2{X:=256.0,Y:=256.0}
    
        StackBoxSlots :=
            array:
                stack_box_slot:
                    Widget := LevelBlock
                stack_box_slot:
                    Widget := ImageOverlay
                stack_box_slot:
                    Widget := NameBlock
                stack_box_slot:
                    Widget := DescriptionBlock
                stack_box_slot:
                    Widget := color_block:
                        DefaultColor := NamedColors.DarkBlue
                        DefaultDesiredSize := vector2{X := 450.0, Y := BorderSize}
                stack_box_slot:
                    Widget := BonusStack
                stack_box_slot:
                    Widget := color_block:
                        DefaultColor := NamedColors.DarkBlue
                        DefaultDesiredSize := vector2{X := 450.0, Y := BorderSize}
                stack_box_slot:
                    Widget := CostStack
                stack_box_slot:
                    Padding := margin{Top := 10.0}
                    VerticalAlignment := vertical_alignment.Fill
                    HorizontalAlignment := horizontal_alignment.Fill
                    Widget := ButtonOverlay
        
        Overlay := overlay:
            Slots := array:
                overlay_slot:
                    VerticalAlignment := vertical_alignment.Fill
                    HorizontalAlignment := horizontal_alignment.Fill
                    Widget := BackgroundColorBlock
                overlay_slot:
                    VerticalAlignment := vertical_alignment.Bottom
                    HorizontalAlignment := horizontal_alignment.Fill
                    Padding := margin{Top := 10.0}
                    Widget := overlay:
                        Slots := array:
                            overlay_slot:
                                VerticalAlignment := vertical_alignment.Center
                                HorizontalAlignment := horizontal_alignment.Fill
                                Widget := stack_box:
                                    Orientation := orientation.Vertical
                                    Slots := StackBoxSlots
                overlay_slot:
                    VerticalAlignment := vertical_alignment.Top
                    HorizontalAlignment := horizontal_alignment.Fill
                    Widget := TopBorderColorBlock
                overlay_slot:
                    VerticalAlignment := vertical_alignment.Bottom
                    HorizontalAlignment := horizontal_alignment.Fill
                    Widget := BottomBorderColorBlock
                overlay_slot:
                    VerticalAlignment := vertical_alignment.Fill
                    HorizontalAlignment := horizontal_alignment.Right
                    Widget := RightBorderColorBlock
                overlay_slot:
                    VerticalAlignment := vertical_alignment.Fill
                    HorizontalAlignment := horizontal_alignment.Left
                    Widget := LeftBorderColorBlock

        return PlayerUpgradeSlotData:
            Name := InName
            Level := LevelBlock
            Cost := CostBlock
            CurrentBonus := CurrentBonusText
            NextBonus := NextBonusText
            ResourceTexture := ResourceTexture
            Widget := Overlay
            Button := PurchaseButton
            ButtonColor := PurchaseButtonColor