using { /Fortnite.com/Devices }
using { /Verse.org/Simulation }
using { /UnrealEngine.com/Temporary/Diagnostics }
using { /UnrealEngine.com/Temporary/UI }
using { /Fortnite.com/UI }
using { /UnrealEngine.com/Temporary/SpatialMath}
using { /Verse.org/Colors }
using { /Verse.org/Assets }

PrestigeMenuCanvas := class:
    Canvas : canvas
    ExitButton : text_button_base
    PrestigeButton : text_button_base
    PrestigeButtonColor : texture_block

    TotalStoneText : text_block
    TotalPrestigeText : text_block
    TotalPrestigeBonusText : text_block

    PrestigeTrigger : trigger_device

    OnPrestigeButtonEvent(WidgetMessage : widget_message):void=
        Print("Prestige Button Press")
        PrestigeTrigger.Trigger(WidgetMessage.Player)

UI_PrestigeMenu := class():
    ToMessage<localizes>(String : string) : message = "{String}"
    CenteredAlignment : vector2 = vector2{X := 0.5, Y := 0.5}

    var BuiltCanvasMap : [player]PrestigeMenuCanvas = map{}

    InitUI(InPlayer : player, InPrestigeTrigger : trigger_device):void=
        DisplayCanvas := MakePrestigeMenuCanvas(InPrestigeTrigger)
        # Set the players build canvas data
        if(set BuiltCanvasMap[InPlayer] = DisplayCanvas, CanvasData := BuiltCanvasMap[InPlayer]):
            CanvasData.PrestigeButton.OnClick().Subscribe(CanvasData.OnPrestigeButtonEvent)
            CanvasData.ExitButton.OnClick().Subscribe(HandlePlayerExit)

    # Add the UI to the player
    AddUI(InPlayer : player):void=
        if(PlayerUI := GetPlayerUI[InPlayer], CanvasData := BuiltCanvasMap[InPlayer]):
            CanvasData.Canvas.SetVisibility(widget_visibility.Visible)
            PlayerUI.AddWidget(CanvasData.Canvas, player_ui_slot{InputMode := ui_input_mode.All} )

    # When player presses the exit button
    HandlePlayerExit(WidgetMessage: widget_message):void=
        spawn{HideUI(WidgetMessage.Player)}

    HideUI(InPlayer: player)<suspends>:void=
        Sleep(0.0) # So we're sure that the canvas has been added to BuiltCanvas map
        if(PlayerUI := GetPlayerUI[InPlayer], CanvasData := BuiltCanvasMap[InPlayer]):
            CanvasData.Canvas.SetVisibility(widget_visibility.Hidden)
            PlayerUI.RemoveWidget(CanvasData.Canvas)
    
    # Upates a given player's UI
    UpdateUI(InPlayer : player, TotalStoneAmount : big_number, PrestigeAmount : big_number, PrestigeBonusAmount : big_number, NeededAmount : big_number, CanPrestige : logic):void=
        if(PlayerUI := GetPlayerUI[InPlayer], CanvasData := BuiltCanvasMap[InPlayer]):
            if(CanPrestige = false):
                CanvasData.TotalStoneText.SetTextColor(NamedColors.DarkRed)
            else:
                CanvasData.TotalStoneText.SetTextColor(NamedColors.White)

            CanvasData.TotalStoneText.SetText(ToMessage("{BigNumber_FormatDisplay(TotalStoneAmount)} / {BigNumber_FormatDisplay(NeededAmount)}"))
            CanvasData.TotalPrestigeText.SetText(ToMessage("+ {BigNumber_FormatDisplay(PrestigeAmount)}"))
            CanvasData.TotalPrestigeBonusText.SetText(ToMessage("+ {BigNumber_FormatDisplay(PrestigeBonusAmount)}% Gather Rate"))

    MakePrestigeMenuCanvas(InPrestigeTrigger : trigger_device):PrestigeMenuCanvas=
        ExitButton := button_loud:
            DefaultText := ToMessage("EXIT")
        PrestigeButton := button_quiet:
            DefaultText := ToMessage("PRESTIGE")

        PrestigeButtonColor := texture_block:
            DefaultImage := VerseExposed.T_VendorButtonBackground
            DefaultDesiredSize := vector2{X:=494.0,Y:=54.0}
            DefaultTint := NamedColors.Yellow

        BorderSize := 1.0
        BorderColorBlock := color_block:
            DefaultColor := NamedColors.White
            DefaultDesiredSize := vector2{X := BorderSize, Y := BorderSize}

        PrestigeTitle := text_block:
            DefaultText := ToMessage("Prestige")
            DefaultJustification := text_justification.Center
            DefaultShadowColor := NamedColors.Black
            DefaultShadowOffset := option{vector2{X:=2.0,Y:=2.0}}
        PrestigeTitle.SetTextColor(NamedColors.White)
        PrestigeTitle.SetShadowOpacity(0.8)
        RedText := text_block:
            DefaultText := ToMessage("Restart your progress, losing strength and mining abilities.")
            DefaultJustification := text_justification.Center
            DefaultShadowColor := NamedColors.Black
            DefaultShadowOffset := option{vector2{X:=2.0,Y:=2.0}}
        RedText.SetTextColor(NamedColors.Red)
        RedText.SetShadowOpacity(0.8)
        GreenText := text_block:
            DefaultText := ToMessage("Gain Prestige Currency and a PERMANENT increase in power!")
            DefaultJustification := text_justification.Center
            DefaultShadowColor := NamedColors.Black
            DefaultShadowOffset := option{vector2{X:=2.0,Y:=2.0}}
        GreenText.SetTextColor(NamedColors.LightGreen)
        GreenText.SetShadowOpacity(0.8)

        KeepText := text_block:
            DefaultText := ToMessage("KEEP")
            DefaultJustification := text_justification.Center
            DefaultShadowColor := NamedColors.Black
            DefaultShadowOffset := option{vector2{X:=2.0,Y:=2.0}}
        KeepText.SetTextColor(NamedColors.White)
        KeepText.SetShadowOpacity(0.8)
        LostText := text_block:
            DefaultText := ToMessage("LOSE")
            DefaultJustification := text_justification.Center
            DefaultShadowColor := NamedColors.Black
            DefaultShadowOffset := option{vector2{X:=2.0,Y:=2.0}}
        LostText.SetTextColor(NamedColors.White)
        LostText.SetShadowOpacity(0.8)
        GainText := text_block:
            DefaultText := ToMessage("GAIN")
            DefaultJustification := text_justification.Center
            DefaultShadowColor := NamedColors.Black
            DefaultShadowOffset := option{vector2{X:=2.0,Y:=2.0}}
        GainText.SetTextColor(NamedColors.White)
        GainText.SetShadowOpacity(0.8)

        KeepBlockText := text_block:
            DefaultText := ToMessage("General Upgrades\nWeapon Upgrades\nArtifacts\nAltar Upgrades\nAchievements\nOres")
            DefaultJustification := text_justification.Center
            DefaultShadowColor := NamedColors.Black
            DefaultShadowOffset := option{vector2{X:=2.0,Y:=2.0}}
        KeepBlockText.SetTextColor(NamedColors.LightGreen)
        KeepBlockText.SetShadowOpacity(0.8)
        LostBlockText := text_block:
            DefaultText := ToMessage("Strength Upgrades\nAreas Unlocked\nMining Abilities\nCurrent Stone")
            DefaultJustification := text_justification.Center
            DefaultShadowColor := NamedColors.Black
            DefaultShadowOffset := option{vector2{X:=2.0,Y:=2.0}}
        LostBlockText.SetTextColor(NamedColors.Red)
        LostBlockText.SetShadowOpacity(0.8)


        TotalStoneTitleText := text_block:
            DefaultText := ToMessage("Total Stone Earned")
            DefaultJustification := text_justification.Center
            DefaultShadowColor := NamedColors.Black
            DefaultShadowOffset := option{vector2{X:=2.0,Y:=2.0}}
        TotalStoneTitleText.SetTextColor(NamedColors.White)
        TotalStoneTitleText.SetShadowOpacity(0.8)
        TotalStoneText := text_block:
            DefaultText := ToMessage("10/10")
            DefaultJustification := text_justification.Center
            DefaultShadowColor := NamedColors.Black
            DefaultShadowOffset := option{vector2{X:=2.0,Y:=2.0}}
        TotalStoneText.SetTextColor(NamedColors.White)
        TotalStoneText.SetShadowOpacity(0.8)

        TotalPrestigeTitleText := text_block:
            DefaultText := ToMessage("Prestige Resource Earned")
            DefaultJustification := text_justification.Center
            DefaultShadowColor := NamedColors.Black
            DefaultShadowOffset := option{vector2{X:=2.0,Y:=2.0}}
        TotalPrestigeTitleText.SetTextColor(NamedColors.White)
        TotalPrestigeTitleText.SetShadowOpacity(0.8)
        TotalPrestigeText := text_block:
            DefaultText := ToMessage("10")
            DefaultJustification := text_justification.Center
            DefaultShadowColor := NamedColors.Black
            DefaultShadowOffset := option{vector2{X:=2.0,Y:=2.0}}
        TotalPrestigeText.SetTextColor(NamedColors.LightGreen)
        TotalPrestigeText.SetShadowOpacity(0.8)

        TotalPrestigeBonusTitleText := text_block:
            DefaultText := ToMessage("Prestige Bonus Earned")
            DefaultJustification := text_justification.Center
            DefaultShadowColor := NamedColors.Black
            DefaultShadowOffset := option{vector2{X:=2.0,Y:=2.0}}
        TotalPrestigeBonusTitleText.SetTextColor(NamedColors.White)
        TotalPrestigeBonusTitleText.SetShadowOpacity(0.8)
        TotalPrestigeBonusText := text_block:
            DefaultText := ToMessage("10")
            DefaultJustification := text_justification.Center
            DefaultShadowColor := NamedColors.Black
            DefaultShadowOffset := option{vector2{X:=2.0,Y:=2.0}}
        TotalPrestigeBonusText.SetTextColor(NamedColors.LightGreen)
        TotalPrestigeBonusText.SetShadowOpacity(0.8)

        TopStackBox := stack_box:
            Orientation := orientation.Vertical
            Slots := 
                array:
                    stack_box_slot:
                        Widget := PrestigeTitle
                        HorizontalAlignment := horizontal_alignment.Center
                        VerticalAlignment := vertical_alignment.Center
                    stack_box_slot:
                        Widget := RedText
                        HorizontalAlignment := horizontal_alignment.Center
                        VerticalAlignment := vertical_alignment.Center
                    stack_box_slot:
                        Widget := GreenText
                        HorizontalAlignment := horizontal_alignment.Center
                        VerticalAlignment := vertical_alignment.Center

        KeepBox := overlay:
            Slots := array:
                overlay_slot:
                    VerticalAlignment := vertical_alignment.Fill
                    HorizontalAlignment := horizontal_alignment.Fill
                    Widget := color_block:
                        DefaultColor := NamedColors.White
                        DefaultOpacity := 0.4
                        DefaultDesiredSize := vector2{X:=425.0,Y:=425.0}
                overlay_slot:
                    VerticalAlignment := vertical_alignment.Center
                    HorizontalAlignment := horizontal_alignment.Center
                    Widget := KeepBlockText
                overlay_slot:
                    Widget := color_block:
                        DefaultColor := NamedColors.DarkBlue
                        DefaultDesiredSize := vector2{X := BorderSize, Y := BorderSize}
                    HorizontalAlignment := horizontal_alignment.Left
                    VerticalAlignment := vertical_alignment.Fill
                overlay_slot:
                    Widget := color_block:
                        DefaultColor := NamedColors.DarkBlue
                        DefaultDesiredSize := vector2{X := BorderSize, Y := BorderSize}
                    HorizontalAlignment := horizontal_alignment.Right
                    VerticalAlignment := vertical_alignment.Fill
                overlay_slot:
                    Widget := color_block:
                        DefaultColor := NamedColors.DarkBlue
                        DefaultDesiredSize := vector2{X := BorderSize, Y := BorderSize}
                    HorizontalAlignment := horizontal_alignment.Fill
                    VerticalAlignment := vertical_alignment.Top
                overlay_slot:
                    Widget := color_block:
                        DefaultColor := NamedColors.DarkBlue
                        DefaultDesiredSize := vector2{X := BorderSize, Y := BorderSize}
                    HorizontalAlignment := horizontal_alignment.Fill
                    VerticalAlignment := vertical_alignment.Bottom

        KeepStack := stack_box:
            Orientation := orientation.Vertical
            Slots := array:
                    stack_box_slot:
                        Widget := KeepText
                        HorizontalAlignment := horizontal_alignment.Center
                        VerticalAlignment := vertical_alignment.Center
                    stack_box_slot:
                        Widget := KeepBox
                        HorizontalAlignment := horizontal_alignment.Center
                        VerticalAlignment := vertical_alignment.Center

        LoseBox := overlay:
            Slots := array:
                overlay_slot:
                    VerticalAlignment := vertical_alignment.Fill
                    HorizontalAlignment := horizontal_alignment.Fill
                    Widget := color_block:
                        DefaultColor := NamedColors.White
                        DefaultOpacity := 0.4
                        DefaultDesiredSize := vector2{X:=425.0,Y:=425.0}
                overlay_slot:
                    VerticalAlignment := vertical_alignment.Center
                    HorizontalAlignment := horizontal_alignment.Center
                    Widget := LostBlockText
                overlay_slot:
                    Widget := color_block:
                        DefaultColor := NamedColors.DarkBlue
                        DefaultDesiredSize := vector2{X := BorderSize, Y := BorderSize}
                    HorizontalAlignment := horizontal_alignment.Left
                    VerticalAlignment := vertical_alignment.Fill
                overlay_slot:
                    Widget := color_block:
                        DefaultColor := NamedColors.DarkBlue
                        DefaultDesiredSize := vector2{X := BorderSize, Y := BorderSize}
                    HorizontalAlignment := horizontal_alignment.Right
                    VerticalAlignment := vertical_alignment.Fill
                overlay_slot:
                    Widget := color_block:
                        DefaultColor := NamedColors.DarkBlue
                        DefaultDesiredSize := vector2{X := BorderSize, Y := BorderSize}
                    HorizontalAlignment := horizontal_alignment.Fill
                    VerticalAlignment := vertical_alignment.Top
                overlay_slot:
                    Widget := color_block:
                        DefaultColor := NamedColors.DarkBlue
                        DefaultDesiredSize := vector2{X := BorderSize, Y := BorderSize}
                    HorizontalAlignment := horizontal_alignment.Fill
                    VerticalAlignment := vertical_alignment.Bottom

        LoseStack := stack_box:
            Orientation := orientation.Vertical
            Slots := array:
                    stack_box_slot:
                        Widget := LostText
                        HorizontalAlignment := horizontal_alignment.Center
                        VerticalAlignment := vertical_alignment.Center
                    stack_box_slot:
                        Widget := LoseBox
                        HorizontalAlignment := horizontal_alignment.Center
                        VerticalAlignment := vertical_alignment.Center

        PlayerTotalStoneStack := stack_box:
            Orientation := orientation.Horizontal
            Slots := 
                array:
                    stack_box_slot:
                        Widget := texture_block:
                            DefaultImage := VerseExposed.T_Stone
                        HorizontalAlignment := horizontal_alignment.Center
                        VerticalAlignment := vertical_alignment.Center
                    stack_box_slot:
                        Widget := TotalStoneText
                        HorizontalAlignment := horizontal_alignment.Center
                        VerticalAlignment := vertical_alignment.Center

        TotalMineStoneStack := stack_box:
            Orientation := orientation.Vertical
            Slots := 
                array:
                    stack_box_slot:
                        Widget := TotalStoneTitleText
                        HorizontalAlignment := horizontal_alignment.Left
                        VerticalAlignment := vertical_alignment.Center
                    stack_box_slot:
                        Widget := PlayerTotalStoneStack
                        HorizontalAlignment := horizontal_alignment.Center
                        VerticalAlignment := vertical_alignment.Center

        TotalMineStoneOverlay := overlay:
            Slots := array:
                overlay_slot:
                    VerticalAlignment := vertical_alignment.Fill
                    HorizontalAlignment := horizontal_alignment.Fill
                    Widget := color_block:
                        DefaultColor := NamedColors.White
                        DefaultOpacity := 0.4
                overlay_slot:
                    VerticalAlignment := vertical_alignment.Center
                    HorizontalAlignment := horizontal_alignment.Center
                    Padding := margin{Left := 10.0, Right := 10.0}
                    Widget := TotalMineStoneStack
                overlay_slot:
                    Widget := color_block:
                        DefaultColor := NamedColors.DarkBlue
                        DefaultDesiredSize := vector2{X := BorderSize, Y := BorderSize}
                    HorizontalAlignment := horizontal_alignment.Left
                    VerticalAlignment := vertical_alignment.Fill
                overlay_slot:
                    Widget := color_block:
                        DefaultColor := NamedColors.DarkBlue
                        DefaultDesiredSize := vector2{X := BorderSize, Y := BorderSize}
                    HorizontalAlignment := horizontal_alignment.Right
                    VerticalAlignment := vertical_alignment.Fill
                overlay_slot:
                    Widget := color_block:
                        DefaultColor := NamedColors.DarkBlue
                        DefaultDesiredSize := vector2{X := BorderSize, Y := BorderSize}
                    HorizontalAlignment := horizontal_alignment.Fill
                    VerticalAlignment := vertical_alignment.Top
                overlay_slot:
                    Widget := color_block:
                        DefaultColor := NamedColors.DarkBlue
                        DefaultDesiredSize := vector2{X := BorderSize, Y := BorderSize}
                    HorizontalAlignment := horizontal_alignment.Fill
                    VerticalAlignment := vertical_alignment.Bottom
        
        PlayerPrestigeStack := stack_box:
            Orientation := orientation.Horizontal
            Slots := 
                array:
                    stack_box_slot:
                        Widget := texture_block:
                            DefaultImage := VerseExposed.T_RainbowCrystal
                        HorizontalAlignment := horizontal_alignment.Center
                        VerticalAlignment := vertical_alignment.Center
                    stack_box_slot:
                        Widget := TotalPrestigeText
                        HorizontalAlignment := horizontal_alignment.Center
                        VerticalAlignment := vertical_alignment.Center

        TotalPrestigeStack := stack_box:
            Orientation := orientation.Vertical
            Slots := 
                array:
                    stack_box_slot:
                        Widget := TotalPrestigeTitleText
                        HorizontalAlignment := horizontal_alignment.Left
                        VerticalAlignment := vertical_alignment.Center
                    stack_box_slot:
                        Widget := PlayerPrestigeStack
                        HorizontalAlignment := horizontal_alignment.Center
                        VerticalAlignment := vertical_alignment.Center

        PlayerPrestigeBonusStack := stack_box:
            Orientation := orientation.Horizontal
            Slots := 
                array:
                    stack_box_slot:
                        Widget := texture_block:
                            DefaultImage := VerseExposed.T_PickaxeWhite
                        HorizontalAlignment := horizontal_alignment.Center
                        VerticalAlignment := vertical_alignment.Center
                    stack_box_slot:
                        Widget := TotalPrestigeBonusText
                        HorizontalAlignment := horizontal_alignment.Center
                        VerticalAlignment := vertical_alignment.Center

        TotalPrestigeBonusStack := stack_box:
            Orientation := orientation.Vertical
            Slots := 
                array:
                    stack_box_slot:
                        Widget := TotalPrestigeBonusTitleText
                        HorizontalAlignment := horizontal_alignment.Left
                        VerticalAlignment := vertical_alignment.Center
                    stack_box_slot:
                        Widget := PlayerPrestigeBonusStack
                        HorizontalAlignment := horizontal_alignment.Center
                        VerticalAlignment := vertical_alignment.Center

        GainBoxStack := stack_box:
            Orientation := orientation.Vertical
            Slots := array:
                stack_box_slot:
                    Widget := TotalPrestigeStack
                    HorizontalAlignment := horizontal_alignment.Center
                    VerticalAlignment := vertical_alignment.Center
                stack_box_slot:
                    Widget := TotalPrestigeBonusStack
                    HorizontalAlignment := horizontal_alignment.Center
                    VerticalAlignment := vertical_alignment.Center
                
        GainBox := overlay:
            Slots := array:
                overlay_slot:
                    VerticalAlignment := vertical_alignment.Fill
                    HorizontalAlignment := horizontal_alignment.Fill
                    Widget := color_block:
                        DefaultColor := NamedColors.White
                        DefaultOpacity := 0.4
                        DefaultDesiredSize := vector2{X:=425.0,Y:=425.0}
                overlay_slot:
                    VerticalAlignment := vertical_alignment.Center
                    HorizontalAlignment := horizontal_alignment.Center
                    Widget := GainBoxStack
                overlay_slot:
                    Widget := color_block:
                        DefaultColor := NamedColors.DarkBlue
                        DefaultDesiredSize := vector2{X := BorderSize, Y := BorderSize}
                    HorizontalAlignment := horizontal_alignment.Left
                    VerticalAlignment := vertical_alignment.Fill
                overlay_slot:
                    Widget := color_block:
                        DefaultColor := NamedColors.DarkBlue
                        DefaultDesiredSize := vector2{X := BorderSize, Y := BorderSize}
                    HorizontalAlignment := horizontal_alignment.Right
                    VerticalAlignment := vertical_alignment.Fill
                overlay_slot:
                    Widget := color_block:
                        DefaultColor := NamedColors.DarkBlue
                        DefaultDesiredSize := vector2{X := BorderSize, Y := BorderSize}
                    HorizontalAlignment := horizontal_alignment.Fill
                    VerticalAlignment := vertical_alignment.Top
                overlay_slot:
                    Widget := color_block:
                        DefaultColor := NamedColors.DarkBlue
                        DefaultDesiredSize := vector2{X := BorderSize, Y := BorderSize}
                    HorizontalAlignment := horizontal_alignment.Fill
                    VerticalAlignment := vertical_alignment.Bottom

        GainStack := stack_box:
            Orientation := orientation.Vertical
            Slots := array:
                    stack_box_slot:
                        Widget := GainText
                        HorizontalAlignment := horizontal_alignment.Center
                        VerticalAlignment := vertical_alignment.Center
                    stack_box_slot:
                        Widget := GainBox
                        HorizontalAlignment := horizontal_alignment.Center
                        VerticalAlignment := vertical_alignment.Center

        BoxesStack := stack_box:
            Orientation := orientation.Horizontal
            Slots := array:
                    stack_box_slot:
                        Widget := KeepStack
                        Padding := margin{Right := 10.0, Left := 10.0}
                        HorizontalAlignment := horizontal_alignment.Center
                        VerticalAlignment := vertical_alignment.Center
                    stack_box_slot:
                        Widget := LoseStack
                        Padding := margin{Right := 10.0, Left := 10.0}
                        HorizontalAlignment := horizontal_alignment.Center
                        VerticalAlignment := vertical_alignment.Center
                    stack_box_slot:
                        Widget := GainStack
                        Padding := margin{Right := 10.0, Left := 10.0}
                        HorizontalAlignment := horizontal_alignment.Center
                        VerticalAlignment := vertical_alignment.Center

        PrestigeButtonOverlay := overlay:
            Slots := array:
                overlay_slot:
                    VerticalAlignment := vertical_alignment.Center
                    HorizontalAlignment := horizontal_alignment.Center
                    Widget := PrestigeButtonColor
                overlay_slot:
                    VerticalAlignment := vertical_alignment.Center
                    HorizontalAlignment := horizontal_alignment.Fill
                    Widget := PrestigeButton

        FullStack := stack_box:
            Orientation := orientation.Vertical
            Slots := array:
                    stack_box_slot:
                        Widget := TopStackBox
                        Padding := margin{Top := 10.0, Bottom := 10.0}
                        HorizontalAlignment := horizontal_alignment.Center
                        VerticalAlignment := vertical_alignment.Center
                    stack_box_slot:
                        Widget := color_block:
                            DefaultColor := NamedColors.DarkBlue
                            DefaultDesiredSize := vector2{X := BorderSize, Y := BorderSize}
                        HorizontalAlignment := horizontal_alignment.Fill
                        VerticalAlignment := vertical_alignment.Fill
                    stack_box_slot:
                        Widget := BoxesStack
                        Padding := margin{Top := 10.0, Bottom := 10.0}
                        HorizontalAlignment := horizontal_alignment.Center
                        VerticalAlignment := vertical_alignment.Center
                    stack_box_slot:
                        Widget := TotalMineStoneOverlay
                        Padding := margin{Top := 10.0, Bottom := 10.0}
                        HorizontalAlignment := horizontal_alignment.Center
                        VerticalAlignment := vertical_alignment.Center
                    stack_box_slot:
                        Widget := PrestigeButtonOverlay
                        Padding := margin{Top := 10.0, Bottom := 10.0}
                        HorizontalAlignment := horizontal_alignment.Center
                        VerticalAlignment := vertical_alignment.Center

        OverallOverlay := overlay:
            Slots := array:
                overlay_slot:
                    VerticalAlignment := vertical_alignment.Fill
                    HorizontalAlignment := horizontal_alignment.Fill
                    Widget := overlay:
                        Slots := array:
                            overlay_slot:
                                VerticalAlignment := vertical_alignment.Fill
                                HorizontalAlignment := horizontal_alignment.Fill
                                Widget := FullStack
        
        BackgroundOverlay := overlay:
            Slots := array:
                overlay_slot:
                    VerticalAlignment := vertical_alignment.Fill
                    HorizontalAlignment := horizontal_alignment.Fill
                    Widget := texture_block:
                        DefaultImage := VerseExposed.T_UIBackground

        OutputCanvas := canvas:
            Slots := array:
                canvas_slot:
                    Anchors := anchors{Minimum := vector2{X:=0.0, Y:=0.0}, Maximum := vector2{X:=1.0, Y:=1.0}}
                    Alignment := CenteredAlignment
                    Widget := BackgroundOverlay
                    ZOrder := 0
                canvas_slot:
                    Anchors := anchors{Minimum := vector2{X:=0.0, Y:=0.05}, Maximum := vector2{X:=1.0, Y:=0.8}}
                    Alignment := CenteredAlignment
                    Widget := OverallOverlay
                    ZOrder := 1
                canvas_slot:
                    Anchors := anchors{Minimum := vector2{X:=0.5, Y:=0.85}, Maximum := vector2{X:=0.5, Y:=0.85}}
                    Alignment := CenteredAlignment
                    Widget := ExitButton
                    ZOrder := 1

        return PrestigeMenuCanvas:
            Canvas := OutputCanvas
            ExitButton := ExitButton
            PrestigeButton := PrestigeButton
            PrestigeButtonColor := PrestigeButtonColor
            TotalStoneText := TotalStoneText
            TotalPrestigeText := TotalPrestigeText
            TotalPrestigeBonusText := TotalPrestigeBonusText
            PrestigeTrigger := InPrestigeTrigger