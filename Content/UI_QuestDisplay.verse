
using { /Fortnite.com/Devices }
using { /Verse.org/Simulation }
using { /UnrealEngine.com/Temporary/Diagnostics }
using { /UnrealEngine.com/Temporary/UI }
using { /Fortnite.com/UI }
using { /UnrealEngine.com/Temporary/SpatialMath}
using { /Verse.org/Colors }
using { /Verse.org/Assets }

MakeQuestDisplayData<constructor>(InQuestLineName : string, InCurrentQuestName : string, InCurrentQuestProgress : string, InShowCost : logic):= QuestDisplayData:
    QuestLineName := InQuestLineName
    CurrentQuestName := InCurrentQuestName
    CurrentQuestProgress := InCurrentQuestProgress
    ShowCost := InShowCost

QuestDisplayData := class():
    QuestLineName : string
    CurrentQuestName : string
    CurrentQuestProgress : string
    ShowCost : logic

QuestDisplayCanvasData := class():
    Canvas : canvas
    QuestLineText : text_block
    CurrentQuestNameText : text_block
    CurrentQuestProgressText : text_block
    CostText : text_block
    CostWidget : widget

# A Verse-authored creative device that can be placed in a level
UI_QuestDisplay := class():

    ToMessage<localizes>(String : string) : message = "{String}"
    CenteredAlignment : vector2 = vector2{X := 0.5, Y := 0.5}

    var BuiltCanvasMap : [player]QuestDisplayCanvasData = map{}

    # Initiliase the UI so it can added to the player much quicker
    InitUI(InPlayer : player, InQuestDisplayData : QuestDisplayData):void=
        block:
        QuestDisplayCanvas := MakeQuestDisplayCanvas(InQuestDisplayData)
        # Set the players build canvas data
        if(set BuiltCanvasMap[InPlayer] = QuestDisplayCanvas) {}

    # Add the UI to the player
    AddUI(InPlayer : player):void=
        if(PlayerUI := GetPlayerUI[InPlayer], CanvasData := BuiltCanvasMap[InPlayer]):
            CanvasData.Canvas.SetVisibility(widget_visibility.Visible)
            PlayerUI.AddWidget(CanvasData.Canvas)
    
    HideUI(InPlayer: player)<suspends>:void=
        Sleep(0.0) # So we're sure that the canvas has been added to BuiltCanvas map
        if(PlayerUI := GetPlayerUI[InPlayer], CanvasData := BuiltCanvasMap[InPlayer]):
            CanvasData.Canvas.SetVisibility(widget_visibility.Hidden)
            PlayerUI.RemoveWidget(CanvasData.Canvas)
    
    # Upates a given player's UI
    UpdateUI(InPlayer : player, InQuestLineName : string, InQuestName : string, InQuestProgress : string, InQuestShowCost : logic, InQuestCost : string, InCanAfford : logic):void=
        if(PlayerUI := GetPlayerUI[InPlayer], CanvasData := BuiltCanvasMap[InPlayer]):

            CanvasData.QuestLineText.SetText(ToMessage(InQuestLineName))
            CanvasData.CurrentQuestNameText.SetText(ToMessage(SplitName(InQuestName)))
            CanvasData.CurrentQuestProgressText.SetText(ToMessage(InQuestProgress))
            if(InQuestShowCost = true):
                if(InCanAfford = true):
                    CanvasData.CostText.SetTextColor(NamedColors.LightGreen)
                else:
                    CanvasData.CostText.SetTextColor(NamedColors.Red)
                CanvasData.CostWidget.SetVisibility(widget_visibility.Visible)
                CanvasData.CostText.SetText(ToMessage(InQuestCost))
            else:
                CanvasData.CostWidget.SetVisibility(widget_visibility.Hidden)


    SplitName(InQuestName : string):string=
        var FirstHalf : string = ""
        var SecondHalf : string = ""

        var InFirst : logic = true
        for(Character : InQuestName):
            if(array{Character} = "|"):
                set InFirst = false

            if(array{Character} <> "|"):
                if(InFirst = true):
                    set FirstHalf += array{Character}
                else:
                    set SecondHalf += array{Character}

        OutString := FirstHalf + "\n" + SecondHalf

        return OutString


    MakeQuestDisplayCanvas(InQuestDisplayData : QuestDisplayData):QuestDisplayCanvasData=
        # Variables for other parts of UI
        # Use these to copy and paste the settings into overlays
        BackgroundColorBlock := color_block:
            DefaultColor := NamedColors.White
            DefaultOpacity := 0.2
        BorderSize := 1.0
        BorderColorBlock := color_block:
            DefaultColor := NamedColors.White
            DefaultDesiredSize := vector2{X := BorderSize, Y := BorderSize}

        # Updating Text Blocks
        QuestLineNameText := text_block:
            DefaultText := ToMessage("{InQuestDisplayData.QuestLineName}")
            DefaultJustification := text_justification.Left
            DefaultShadowColor := NamedColors.Black
            DefaultShadowOffset := option{vector2{X:=2.0,Y:=2.0}}
        QuestLineNameText.SetTextColor(NamedColors.Cyan)
        QuestLineNameText.SetShadowOpacity(0.8)
        QuestNameText := text_block:
            DefaultText := ToMessage("{InQuestDisplayData.CurrentQuestName}")
            DefaultJustification := text_justification.Left
            DefaultShadowColor := NamedColors.Black
            DefaultShadowOffset := option{vector2{X:=2.0,Y:=2.0}}
        QuestNameText.SetTextColor(NamedColors.White)
        QuestNameText.SetShadowOpacity(0.8)
        QuestProgressText := text_block:
            DefaultText := ToMessage("{InQuestDisplayData.CurrentQuestProgress}")
            DefaultJustification := text_justification.Right
            DefaultShadowColor := NamedColors.Black
            DefaultShadowOffset := option{vector2{X:=2.0,Y:=2.0}}
        QuestProgressText.SetTextColor(NamedColors.White)
        QuestProgressText.SetShadowOpacity(0.8)
        QuestProgressText.SetShadowOpacity(0.8)
        CostTitleText := text_block:
            DefaultText := ToMessage("Cost:")
            DefaultJustification := text_justification.Left
            DefaultShadowColor := NamedColors.Black
            DefaultShadowOffset := option{vector2{X:=2.0,Y:=2.0}}
        CostTitleText.SetTextColor(NamedColors.White)
        CostTitleText.SetShadowOpacity(0.8)
        CostText := text_block:
            DefaultText := ToMessage("1.00K")
            DefaultJustification := text_justification.Left
            DefaultShadowColor := NamedColors.Black
            DefaultShadowOffset := option{vector2{X:=2.0,Y:=2.0}}
        CostText.SetTextColor(NamedColors.Red)
        CostText.SetShadowOpacity(0.8)

        # UI Construction
        CostTextStack := stack_box:
            Orientation := orientation.Horizontal
            Slots :=
                array:
                    stack_box_slot:
                        VerticalAlignment := vertical_alignment.Center
                        HorizontalAlignment := horizontal_alignment.Center
                        Padding := margin{Left := 5.0, Right := 10.0} 
                        Widget := CostTitleText
                    stack_box_slot:
                        VerticalAlignment := vertical_alignment.Center
                        HorizontalAlignment := horizontal_alignment.Center
                        Padding := margin{Left := 10.0, Right := 5.0}
                        Widget := texture_block:
                            DefaultImage := VerseExposed.T_Stone
                    stack_box_slot:
                        VerticalAlignment := vertical_alignment.Center
                        HorizontalAlignment := horizontal_alignment.Center
                        Padding := margin{Left := 5.0, Right := 10.0} 
                        Widget := CostText

        QuestNameProgressOverlay := overlay:
            Slots := array:
                overlay_slot:
                    VerticalAlignment := vertical_alignment.Center
                    HorizontalAlignment := horizontal_alignment.Left
                    Widget := QuestNameText
                overlay_slot:
                    VerticalAlignment := vertical_alignment.Center
                    HorizontalAlignment := horizontal_alignment.Right
                    Widget := QuestProgressText

        QuestStackBox := stack_box:
            Orientation := orientation.Vertical
            Slots :=
                array:
                    stack_box_slot:
                        VerticalAlignment := vertical_alignment.Center
                        HorizontalAlignment := horizontal_alignment.Left
                        Padding := margin{Top := 10.0,Bottom := 10.0}
                        Widget := QuestLineNameText
                    stack_box_slot:
                        VerticalAlignment := vertical_alignment.Center
                        HorizontalAlignment := horizontal_alignment.Fill
                        Widget := color_block:
                            DefaultColor := NamedColors.Black
                            DefaultOpacity := 0.4
                            DefaultDesiredSize := vector2{X := BorderSize, Y := BorderSize}
                    stack_box_slot:
                        VerticalAlignment := vertical_alignment.Center
                        HorizontalAlignment := horizontal_alignment.Fill
                        Padding := margin{Top := 10.0,Bottom := 2.0}
                        Widget := QuestNameProgressOverlay
                    stack_box_slot:
                        VerticalAlignment := vertical_alignment.Center
                        HorizontalAlignment := horizontal_alignment.Left
                        Padding := margin{Top := 2.0,Bottom := 5.0}
                        Widget := CostTextStack

        OverallOverlay := overlay:
            Slots := array:
                overlay_slot:
                    VerticalAlignment := vertical_alignment.Fill
                    HorizontalAlignment := horizontal_alignment.Fill
                    Widget := color_block:
                        DefaultColor := NamedColors.Black
                        DefaultOpacity := 0.4
                overlay_slot:
                    VerticalAlignment := vertical_alignment.Fill
                    HorizontalAlignment := horizontal_alignment.Fill
                    Padding := margin{Left := 10.0, Right := 10.0}
                    Widget := QuestStackBox

        OutputCanvas := canvas:
            Slots := array:
                canvas_slot:
                    Anchors := anchors{Minimum := vector2{X:=0.725, Y:=0.4}, Maximum := vector2{X:=0.99, Y:=0.585}}
                    Alignment := CenteredAlignment
                    Widget := OverallOverlay
                    SizeToContent := true

        return QuestDisplayCanvasData:
            Canvas := OutputCanvas
            QuestLineText := QuestLineNameText
            CurrentQuestNameText := QuestNameText
            CurrentQuestProgressText := QuestProgressText
            CostText := CostText
            CostWidget := CostTextStack