
using { /Fortnite.com/Devices }
using { /Verse.org/Simulation }
using { /UnrealEngine.com/Temporary/Diagnostics }
using { /UnrealEngine.com/Temporary/UI }
using { /Fortnite.com/UI }
using { /UnrealEngine.com/Temporary/SpatialMath}
using { /Verse.org/Colors }
using { /Verse.org/Assets }

MakeResourceDisplayData<constructor>(InStoneAmount : big_number, InAlterAmount : int, DungeonAmount : int):= ResourceDisplayData:
    StoneAmount := InStoneAmount
    DungeonToken := DungeonAmount
    AlterToken := InAlterAmount

ResourceDisplayCanvasData := class:
    Canvas : canvas

    # Player resource amounts
    StoneAmount : text_block
    #SilverAmount : text_block
    #GoldAmount : text_block
    #DiamondAmount : text_block
    CardsAmount : text_block
    AlterTokenAmount : text_block
    AlterTokenWidget : widget
    DungeonTokenAmount : text_block
    DungeonTokenWidget : widget
    PrestigeAmount : text_block

    #AlterTimer : text_block
    #DungeonTimer : text_block

ResourceDisplayData := class:
    StoneAmount : big_number
    AlterToken : int = 1
    DungeonToken : int = 0
    AlterTimer : float = 0.0
    DungeonTimer : float = 0.0

ResourceSlotData := class:
    Widget : widget
    SlotText : text_block

UI_ResourceDisplay := class():
    ToMessage<localizes>(String : string) : message = "{String}"
    CenteredAlignment : vector2 = vector2{X := 0.5, Y := 0.5}

    var BuiltCanvasMap : [player]ResourceDisplayCanvasData = map{}

    # Initiliase the UI so it can added to the player much quicker
    InitUI(InPlayer : player, InResourceDisplayData : ResourceDisplayData):void=
        ResourceDisplayCanvas := MakeResourceCanvas(InResourceDisplayData)
        # Set the players build canvas data
        if(set BuiltCanvasMap[InPlayer] = ResourceDisplayCanvas) {}

    # Add the UI to the player
    AddUI(InPlayer : player):void=
        if(PlayerUI := GetPlayerUI[InPlayer], CanvasData := BuiltCanvasMap[InPlayer]):
            CanvasData.Canvas.SetVisibility(widget_visibility.Visible)
            PlayerUI.AddWidget(CanvasData.Canvas)
            

    HideUI(InPlayer: player)<suspends>:void=
        Sleep(0.0) # So we're sure that the canvas has been added to BuiltCanvas map
        if(PlayerUI := GetPlayerUI[InPlayer], CanvasData := BuiltCanvasMap[InPlayer]):
            CanvasData.Canvas.SetVisibility(widget_visibility.Hidden)
            PlayerUI.RemoveWidget(CanvasData.Canvas)

    HideDungeonUI(InPlayer : player):void=
        if(ResourceDisplayCanvas := BuiltCanvasMap[InPlayer]):
            ResourceDisplayCanvas.DungeonTokenWidget.SetVisibility(widget_visibility.Hidden)

    ShowDungeonUI(InPlayer : player):void=
        if(ResourceDisplayCanvas := BuiltCanvasMap[InPlayer]):
            ResourceDisplayCanvas.DungeonTokenWidget.SetVisibility(widget_visibility.Visible)

    HideAlterUI(InPlayer : player):void=
        if(ResourceDisplayCanvas := BuiltCanvasMap[InPlayer]):
            ResourceDisplayCanvas.AlterTokenWidget.SetVisibility(widget_visibility.Hidden)

    ShowAlterUI(InPlayer : player):void=
        if(ResourceDisplayCanvas := BuiltCanvasMap[InPlayer]):
            ResourceDisplayCanvas.AlterTokenWidget.SetVisibility(widget_visibility.Visible)
            
    # Upates a given player's UI
    UpdateUI(InPlayer : player, InStoneAmount : string, InCardsAmount : string, InAlterAmount : string, InDungeonAmount : string, bAlterTimer : ?float, bDungeonTimer : ?float, InPrestigeAmount : string):void=
        if(PlayerUI := GetPlayerUI[InPlayer], CanvasData := BuiltCanvasMap[InPlayer]):
            CanvasData.StoneAmount.SetText(ToMessage("{InStoneAmount}"))
            CanvasData.PrestigeAmount.SetText(ToMessage("{InPrestigeAmount}"))
            CanvasData.CardsAmount.SetText(ToMessage("{InCardsAmount}"))
            var AlterTimerTxt : string = "Ready"
            if (AlterTimer := bAlterTimer?):
                set AlterTimerTxt = MakeTimerText(AlterTimer)
            var DungeonTimerTxt : string = "Ready"
            if (DungeonTimer := bDungeonTimer?):
                set DungeonTimerTxt = MakeTimerText(DungeonTimer)
            CanvasData.AlterTokenAmount.SetText(ToMessage("{InAlterAmount}  {AlterTimerTxt}"))
            CanvasData.DungeonTokenAmount.SetText(ToMessage("{InDungeonAmount}  {DungeonTimerTxt}"))
            # Setting color of alter and dungeon text when timer is done and tokens have reached to max amount
            if (AlterTimerTxt = "Ready"):
                CanvasData.AlterTokenAmount.SetTextColor(color{R := 1.0, G := 0.3, B := 0.0})
            else:
                CanvasData.AlterTokenAmount.SetTextColor(color{R := 1.0, G := 1.0, B := 1.0})
            if (DungeonTimerTxt = "Ready"):
                CanvasData.DungeonTokenAmount.SetTextColor(color{R := 1.0, G := 0.3, B := 0.0})
            else:
                CanvasData.DungeonTokenAmount.SetTextColor(color{R := 1.0, G := 1.0, B := 1.0})
            #CanvasData.SilverAmount.SetText(ToMessage("{InSilverAmount}"))
            #CanvasData.GoldAmount.SetText(ToMessage("{InGoldAmount}"))
            #CanvasData.DiamondAmount.SetText(ToMessage("{InDiamondAmount}"))
    
    # Returns resource ammount and timer (If has one) in text format. Handles normal floats, Big number and Timers within resource
    MakeTextFromAmount(InAmount : big_number, bIntAmount : ?int, bTimer : ?float):message=
        var TotalMessage : message = ToMessage("")
        var TimerText : string = ""
        # Handles timers with in the resource slot
        if (Timer := bTimer?):
            set TimerText = MakeTimerText(Timer)
        if (IntAmmount := bIntAmount?):
            return ToMessage("{IntAmmount}  {TimerText}")
        return ToMessage("{BigNumber_FormatDisplay(InAmount)}")
    
    # Converting a float timer to text timer
    MakeTimerText(Timer : float):string=
        var Minutes : int = 0
        var Seconds : int = 0
        var MinutesTxt : string = ""
        var SecondsTxt : string = ""
        if(IMinutes := Floor[Timer]):
            set Minutes = IMinutes
            FSeconds := Timer - (IMinutes * 1.0)
            if(ISeconds := Round[FSeconds * 100.0]):
                set Seconds = ISeconds
        if (Minutes < 10):
            set MinutesTxt = "0{Minutes}"
        else:
            set MinutesTxt = "{Minutes}"
        if (Seconds < 10):
            set SecondsTxt = "0{Seconds}"
        else:
            set SecondsTxt = "{Seconds}"
        
        return "{MinutesTxt}:{SecondsTxt}"
            

    # Make a singular resource slot
    MakeResourceSlot(InResourceImage : texture, InAmount : big_number, ?bIntAmount : ?int = false, ?bTimer : ?float = false):ResourceSlotData=
        # Variables for other parts of UI
        # Use these to copy and paste the settings into overlays
        BackgroundColorBlock := color_block:
            DefaultColor := NamedColors.White
            DefaultOpacity := 0.2

        BorderSize := 1.0
        BorderColorBlock := color_block:
            DefaultColor := NamedColors.White
            DefaultDesiredSize := vector2{X := BorderSize, Y := BorderSize}
    
        CoinWidget := texture_block:
            DefaultImage := VerseExposed.T_Stone
        
        AmountText := text_block:
            DefaultText := MakeTextFromAmount(InAmount, bIntAmount, bTimer)
            DefaultJustification := text_justification.Left
            DefaultShadowColor := NamedColors.Black
            DefaultShadowOffset := option{vector2{X:=2.0,Y:=2.0}}
        AmountText.SetTextColor(NamedColors.White)
        AmountText.SetShadowOpacity(0.8)
        
        IconOverlay := overlay:
            Slots := array:
                overlay_slot:
                    VerticalAlignment := vertical_alignment.Center
                    HorizontalAlignment := horizontal_alignment.Center
                    Padding := margin{Top := -8.0, Bottom := -8.0, Left := -6.0, Right := -6.0}
                    Widget := color_block:
                        DefaultColor := NamedColors.Black
                        DefaultOpacity := 0.4
                        DefaultDesiredSize := vector2{X := 52.0, Y := 52.0}
                overlay_slot:
                    VerticalAlignment := vertical_alignment.Center
                    HorizontalAlignment := horizontal_alignment.Center
                    Widget := texture_block:
                        DefaultImage := InResourceImage
        
        ResourceOverlay := overlay:
            Slots := array:
                overlay_slot:
                    VerticalAlignment := vertical_alignment.Center
                    HorizontalAlignment := horizontal_alignment.Left
                    Widget := color_block:
                        DefaultColor := NamedColors.Black
                        DefaultOpacity := 0.4
                        DefaultDesiredSize := vector2{X:=225.0,Y:=60.0}
                overlay_slot:
                    VerticalAlignment := vertical_alignment.Center
                    HorizontalAlignment := horizontal_alignment.Left
                    Padding := margin{Top := 10.0, Bottom := 10.0, Left := 10.0, Right := 10.0}
                    Widget := stack_box:
                        Orientation := orientation.Horizontal
                        Slots := 
                            array:
                                stack_box_slot:
                                    Widget := IconOverlay
                                    HorizontalAlignment := horizontal_alignment.Left
                                    VerticalAlignment := vertical_alignment.Center
                                stack_box_slot:
                                    Widget := AmountText
                                    Padding := margin{Left := 15.0}
                                    VerticalAlignment := vertical_alignment.Center
                                    HorizontalAlignment := horizontal_alignment.Left
        
        return ResourceSlotData:
            Widget := ResourceOverlay
            SlotText := AmountText

    # Make the Resource canvas and all widgets
    MakeResourceCanvas(InResourceDisplayData : ResourceDisplayData):ResourceDisplayCanvasData=
        # Variables for other parts of UI
        # Use these to copy and paste the settings into overlays
        BackgroundColorBlock := color_block:
            DefaultColor := NamedColors.White
            DefaultOpacity := 0.2

        BorderSize := 1.0
        BorderColorBlock := color_block:
            DefaultColor := NamedColors.White
            DefaultDesiredSize := vector2{X := BorderSize, Y := BorderSize}
    
        CoinWidget := texture_block:
            DefaultImage := VerseExposed.T_Stone
        
        # Make resource slots for all 5 resources
        #CoinSlot := MakeResourceSlot(VerseExposed.T_MoneyIcon, InResourceDisplayData.CoinAmount)
        StoneSlot := MakeResourceSlot(VerseExposed.T_Stone, InResourceDisplayData.StoneAmount)
        PrestigeSlot := MakeResourceSlot(VerseExposed.T_RainbowCrystal, big_number{})
        CardsSlot := MakeResourceSlot(VerseExposed.T_GlitterDust, big_number{})
        DungeonTokenSlot := MakeResourceSlot(VerseExposed.T_Skull_01, big_number{}, ?bIntAmount := option{InResourceDisplayData.DungeonToken}, ?bTimer := option{0.0})
        AlterTokenSlot := MakeResourceSlot(VerseExposed.T_AltarIcon, big_number{}, ?bIntAmount := option{InResourceDisplayData.AlterToken}, ?bTimer := option{0.0})
        #GoldSlot := MakeResourceSlot(VerseExposed.T_GoldOre, InResourceDisplayData.GoldAmount)
        #DiamondSlot := MakeResourceSlot(VerseExposed.T_Diamond, InResourceDisplayData.DiamondAmount)

        # Add the resource slots into a stack box to be displayed
        ResourcesBox :=
            array:
                stack_box_slot:
                    Widget := StoneSlot.Widget
                    Padding := margin{Top := 1.0, Bottom := 1.0}
                    VerticalAlignment := vertical_alignment.Center
                    HorizontalAlignment := horizontal_alignment.Left
                stack_box_slot:
                    Widget := PrestigeSlot.Widget
                    Padding := margin{Top := 1.0, Bottom := 1.0}
                    VerticalAlignment := vertical_alignment.Center
                    HorizontalAlignment := horizontal_alignment.Left
                stack_box_slot:
                    Widget := CardsSlot.Widget
                    Padding := margin{Top := 1.0, Bottom := 1.0}
                    VerticalAlignment := vertical_alignment.Center
                    HorizontalAlignment := horizontal_alignment.Left
                stack_box_slot:
                    Widget := AlterTokenSlot.Widget
                    Padding := margin{Top := 1.0, Bottom := 1.0}
                    VerticalAlignment := vertical_alignment.Center
                    HorizontalAlignment := horizontal_alignment.Left
                stack_box_slot:
                    Widget := DungeonTokenSlot.Widget
                    Padding := margin{Top := 1.0, Bottom := 1.0}
                    VerticalAlignment := vertical_alignment.Center
                    HorizontalAlignment := horizontal_alignment.Left
                <#stack_box_slot:
                    Widget := GoldSlot.Widget
                    Padding := margin{Top := 1.0, Bottom := 1.0}
                    VerticalAlignment := vertical_alignment.Center
                    HorizontalAlignment := horizontal_alignment.Left
                stack_box_slot:
                    Widget := DiamondSlot.Widget
                    Padding := margin{Top := 1.0, Bottom := 1.0}
                    VerticalAlignment := vertical_alignment.Center
                    HorizontalAlignment := horizontal_alignment.Left#>
        
        ResourcesOverlay := overlay:
            Slots := array:
                overlay_slot:
                    VerticalAlignment := vertical_alignment.Fill
                    HorizontalAlignment := horizontal_alignment.Fill
                    Widget := overlay:
                        Slots := array:
                            overlay_slot:
                                VerticalAlignment := vertical_alignment.Fill
                                HorizontalAlignment := horizontal_alignment.Fill
                                Widget := stack_box:
                                    Orientation := orientation.Vertical
                                    Slots := ResourcesBox
        OutputCanvas := canvas:
                    Slots := array:
                        canvas_slot:
                            Anchors := anchors{Minimum := vector2{X:=0.01, Y:=0.58}, Maximum := vector2{X:=0.16, Y:=0.78}}
                            Alignment := CenteredAlignment
                            Widget := ResourcesOverlay
                            SizeToContent := true

        return ResourceDisplayCanvasData:
            Canvas := OutputCanvas
            StoneAmount := StoneSlot.SlotText
            PrestigeAmount := PrestigeSlot.SlotText
            CardsAmount := CardsSlot.SlotText
            DungeonTokenAmount := DungeonTokenSlot.SlotText
            DungeonTokenWidget := DungeonTokenSlot.Widget
            AlterTokenAmount := AlterTokenSlot.SlotText
            AlterTokenWidget := AlterTokenSlot.Widget
            #AlterTimer := "00:00"
            #DungeonTimer := "00:00"
            #SilverAmount := SilverSlot.SlotText
            #GoldAmount := GoldSlot.SlotText
            #DiamondAmount := DiamondSlot.SlotText

