using { /Fortnite.com/Devices }
using { /Verse.org/Simulation }
using { /UnrealEngine.com/Temporary/Diagnostics }
using { /UnrealEngine.com/Temporary/UI }
using { /Fortnite.com/UI }
using { /UnrealEngine.com/Temporary/SpatialMath}
using { /Verse.org/Colors }
using { /Verse.org/Assets }

GetSwordImage(InSwordRarity : SwordRarity):texture=
    case(InSwordRarity):
        SwordRarity.Common =>
            return VerseExposed.T_Sword_Common
        SwordRarity.Uncommon =>
            return VerseExposed.T_Sword_Uncommon
        SwordRarity.Rare =>
            return VerseExposed.T_Sword_Rare
        SwordRarity.Epic =>
            return VerseExposed.T_Sword_Epic
        SwordRarity.Legendary =>
            return VerseExposed.T_Sword_Legendary
        SwordRarity.Null =>
            return VerseExposed.T_Sword_Unknown

GetSwordRarityString(InSwordRarity : SwordRarity):string=
    case(InSwordRarity):
        SwordRarity.Common =>
            return "Common"
        SwordRarity.Uncommon =>
            return "Uncommon"
        SwordRarity.Rare =>
            return "Rare"
        SwordRarity.Epic =>
            return "Epic"
        SwordRarity.Legendary =>
            return "Legendary"
        SwordRarity.Null =>
            return "Null"

SwordMenuCanvasData := class:
    Canvas : canvas

    var CurrentResourceType : PropResourceType
    var CurrentSwordRarity : SwordRarity

    WeaponGradeTitle : text_block
    DamageTierText : text_block

    UnlockButtonOverlay : overlay
    StoneCostText : text_block
    ResourceCostText : text_block
    ResourceCostImage : texture_block
    UnlockedText : text_block

    PlayerStoneText : text_block
    PlayerOrbText : text_block

    SwordCommon : SwordButtonData
    SwordUncommon : SwordButtonData
    SwordRare : SwordButtonData
    #SwordEpic : SwordButtonData
    #SwordLegendary : SwordButtonData

    SwordStatDatas : []SwordStatData

    SwordDisplayTexture : texture_block

    PrestigeButtonOverlay : overlay
    PrestigeStoneCostText : text_block
    PrestigeResourceCostText : text_block
    PrestigeResourceCostImage : texture_block
    PrestigeCantUseText : text_block

    CurrentResourceDisplay : texture_block
    UnlockButton : text_button_base
    UpgradeButton : text_button_base

    ExitButton  : text_button_base

SwordInventoryData := class:
    Widget : widget
    Text : text_block

SwordButtonData := class:
    Widget : widget
    SwordImage : texture_block
    Button : text_button_base

SwordStatData := class:
    Widget : widget
    Value : text_block

UI_SwordShopMenu := class():
    ToMessage<localizes>(String : string) : message = "{String}"
    CenteredAlignment : vector2 = vector2{X := 0.5, Y := 0.5}

    var BuiltCanvasMap : [player]SwordMenuCanvasData = map{}

    InitUI(InPlayer : player, UnlockTrigger : trigger_device, UpgradeTrigger : trigger_device,
    Sword1Trigger : trigger_device, Sword2Trigger : trigger_device, Sword3Trigger : trigger_device, Sword4Trigger : trigger_device, Sword5Trigger : trigger_device):void=

        SwordMenuCanvas := MakeSwordMenuCanvas()
        # Set the players build canvas data
        if(set BuiltCanvasMap[InPlayer] = SwordMenuCanvas){}
        SwordMenuCanvas.ExitButton.OnClick().Subscribe(HandlePlayerExit)
        for(StatData : SwordMenuCanvas.SwordStatDatas):
            StatData.Widget.SetVisibility(widget_visibility.Hidden)
        
        SwordMenuCanvas.UnlockButton.OnClick().Subscribe(TriggerHandler{Trigger := UnlockTrigger, PlayerRef := array{InPlayer}}.HandlerFunction)
        SwordMenuCanvas.UpgradeButton.OnClick().Subscribe(TriggerHandler{Trigger := UpgradeTrigger, PlayerRef := array{InPlayer}}.HandlerFunction)
        SwordMenuCanvas.SwordCommon.Button.OnClick().Subscribe(TriggerHandler{Trigger := Sword1Trigger, PlayerRef := array{InPlayer}}.HandlerFunction)
        SwordMenuCanvas.SwordUncommon.Button.OnClick().Subscribe(TriggerHandler{Trigger := Sword2Trigger, PlayerRef := array{InPlayer}}.HandlerFunction)
        SwordMenuCanvas.SwordRare.Button.OnClick().Subscribe(TriggerHandler{Trigger := Sword3Trigger, PlayerRef := array{InPlayer}}.HandlerFunction)
        #SwordMenuCanvas.SwordEpic.Button.OnClick().Subscribe(TriggerHandler{Trigger := Sword4Trigger, PlayerRef := array{InPlayer}}.HandlerFunction)
        #SwordMenuCanvas.SwordLegendary.Button.OnClick().Subscribe(TriggerHandler{Trigger := Sword5Trigger, PlayerRef := array{InPlayer}}.HandlerFunction)

    # Add the UI to the player
    AddUI(InPlayer : player):void=
        if(PlayerUI := GetPlayerUI[InPlayer], CanvasData := BuiltCanvasMap[InPlayer]):
            CanvasData.Canvas.SetVisibility(widget_visibility.Visible)
            PlayerUI.AddWidget(CanvasData.Canvas, player_ui_slot{InputMode := ui_input_mode.All} )

    # When player presses the exit button
    HandlePlayerExit(WidgetMessage: widget_message):void=
        spawn{HideUI(WidgetMessage.Player)}

    HideUI(Player: player)<suspends>:void=
        Sleep(0.0) # So we're sure that the canvas has been added to BuiltCanvas map
        if(PlayerUI := GetPlayerUI[Player], CanvasData := BuiltCanvasMap[Player]):
            CanvasData.Canvas.SetVisibility(widget_visibility.Hidden)
            PlayerUI.RemoveWidget(CanvasData.Canvas)

    UpdateUI(InPlayer : player, InPlayerData : Player_Stats_Data, InPlayerStatManager : player_stats_manager, OwnedPlayerSwords : []Sword_Data, LastSeenSwordRarity : SwordRarity, InSelectedResource : PropResourceType):void=
        if(CanvasData := BuiltCanvasMap[InPlayer]):
            Print("Update UI")
            CanvasData.PlayerStoneText.SetText(ToMessage(BigNumber_FormatDisplay(InPlayerData.PlayerStoneOreAmount)))
            CanvasData.PlayerOrbText.SetText(ToMessage(BigNumber_FormatDisplay(InPlayerData.OrbAmount)))

            CanvasData.WeaponGradeTitle.SetText(ToMessage("{GetSwordRarityString(LastSeenSwordRarity)} Sword"))
            CanvasData.DamageTierText.SetText(ToMessage("Level: 1"))

            TiersAndCostData := GetSwordUnlockData(InSelectedResource)

            CanvasData.StoneCostText.SetText(ToMessage(BigNumber_FormatDisplay(TiersAndCostData.StoneCost)))
            CanvasData.ResourceCostText.SetText(ToMessage(BigNumber_FormatDisplay(TiersAndCostData.ResourceCost)))
            CanvasData.ResourceCostImage.SetImage(TiersAndCostData.ResourceImage)
            if(LastSeenSwordRarity <> SwordRarity.Null):
                CanvasData.SwordDisplayTexture.SetImage(GetSwordImage(LastSeenSwordRarity))
            CanvasData.CurrentResourceDisplay.SetImage(TiersAndCostData.ResourceImage)

            set CanvasData.CurrentSwordRarity = LastSeenSwordRarity
            set CanvasData.CurrentResourceType = InSelectedResource

            PrestigeCostData := GetSwordUpgradeData(LastSeenSwordRarity)

            UpdateCanAfford(InPlayer, InPlayerStatManager, OwnedPlayerSwords, LastSeenSwordRarity)

            if(PlayerSword := OwnedPlayerSwords[0]):
                case(PlayerSword.Rarity):
                    SwordRarity.Common =>
                        CanvasData.UnlockButtonOverlay.SetVisibility(widget_visibility.Visible)
                    SwordRarity.Uncommon =>
                        if(LastSeenSwordRarity = SwordRarity.Common):
                            CanvasData.UnlockButtonOverlay.SetVisibility(widget_visibility.Hidden)
                        else:
                            CanvasData.UnlockButtonOverlay.SetVisibility(widget_visibility.Visible)
                    SwordRarity.Rare =>
                        if(LastSeenSwordRarity = SwordRarity.Common or LastSeenSwordRarity = SwordRarity.Uncommon):
                            CanvasData.UnlockButtonOverlay.SetVisibility(widget_visibility.Hidden)
                        else:
                            CanvasData.UnlockButtonOverlay.SetVisibility(widget_visibility.Visible)
                    SwordRarity.Epic =>
                        if(LastSeenSwordRarity = SwordRarity.Common or LastSeenSwordRarity = SwordRarity.Uncommon or LastSeenSwordRarity = SwordRarity.Rare):
                            CanvasData.UnlockButtonOverlay.SetVisibility(widget_visibility.Hidden)
                        else:
                            CanvasData.UnlockButtonOverlay.SetVisibility(widget_visibility.Visible)
                    SwordRarity.Legendary =>
                        if(LastSeenSwordRarity = SwordRarity.Common or LastSeenSwordRarity = SwordRarity.Uncommon or LastSeenSwordRarity = SwordRarity.Rare or LastSeenSwordRarity = SwordRarity.Epic):
                            CanvasData.UnlockButtonOverlay.SetVisibility(widget_visibility.Hidden)
                    SwordRarity.Null =>

            for(CurrentSword : OwnedPlayerSwords):
                CanvasData.PrestigeStoneCostText.SetText(ToMessage(BigNumber_FormatDisplay(BigNumber_MultiplyWithAFloat(PrestigeCostData.StoneCost, CurrentSword.Level * 1.0))))
                CanvasData.PrestigeResourceCostText.SetText(ToMessage(BigNumber_FormatDisplay(BigNumber_MultiplyWithAFloat(PrestigeCostData.ResourceCost, CurrentSword.Level * 1.0))))
                CanvasData.PrestigeResourceCostImage.SetImage(PrestigeCostData.ResourceImage)
                if(CurrentSword.Rarity = LastSeenSwordRarity):
                    UpdateSwordStat(InPlayer, CurrentSword)
                    CanvasData.UnlockButtonOverlay.SetVisibility(widget_visibility.Hidden)
                    CanvasData.UnlockedText.SetVisibility(widget_visibility.Visible)
                    CanvasData.DamageTierText.SetText(ToMessage("Level: {CurrentSword.Level}"))
                    return

            
            CanvasData.UnlockedText.SetVisibility(widget_visibility.Collapsed)
            UpdateSwordStat(InPlayer, GetSwordPresetData(LastSeenSwordRarity))

    UpdateCanAfford(InPlayer : player, InPlayerStatManager : player_stats_manager, OwnedPlayerSwords : []Sword_Data, LastSeenSwordRarity : SwordRarity):void=
        if(CanvasData := BuiltCanvasMap[InPlayer]):
            CanvasData.PrestigeCantUseText.SetVisibility(widget_visibility.Collapsed)
            CanvasData.PrestigeButtonOverlay.SetVisibility(widget_visibility.Visible)

            PrestigeData := GetSwordUpgradeData(CanvasData.CurrentSwordRarity)
            ResourceAmount := InPlayerStatManager.GetPlayerOreAmount(InPlayer, CanvasData.CurrentResourceType)

            StoneAmount := InPlayerStatManager.GetPlayerStoneOreAmount(InPlayer)
            PrestigeResourceAmount := InPlayerStatManager.GetPlayerOreAmount(InPlayer, PrestigeData.ResourceType)

            if(PlayerSword := OwnedPlayerSwords[0]):
                PrestigeCost := BigNumber_MultiplyWithAFloat(PrestigeData.StoneCost, PlayerSword.Level * 1.0)
                PrestigeResourceCost := BigNumber_MultiplyWithAFloat(PrestigeData.ResourceCost, PlayerSword.Level * 1.0)
                PrestigeCanAffordStone := BigNumber_GreaterThanOrEquals(StoneAmount, PrestigeCost)
                PrestigeCanAffordResource := BigNumber_GreaterThanOrEquals(PrestigeResourceAmount, PrestigeResourceCost)
                #Check if we have enough ore for the next upgrade
                if(PlayerSword.Level = 10):
                    CanvasData.PrestigeCantUseText.SetVisibility(widget_visibility.Visible)
                    CanvasData.PrestigeCantUseText.SetText(ToMessage("WEAPON MAXED"))
                    CanvasData.PrestigeCantUseText.SetTextColor(NamedColors.Gold)
                    CanvasData.PrestigeButtonOverlay.SetVisibility(widget_visibility.Hidden)

                if(PlayerSword.Rarity <> LastSeenSwordRarity):
                    CanvasData.PrestigeCantUseText.SetVisibility(widget_visibility.Hidden)
                    #CanvasData.PrestigeCantUseText.SetText(ToMessage("NEED WEAPON"))
                    CanvasData.PrestigeButtonOverlay.SetVisibility(widget_visibility.Hidden)

                if(PrestigeCanAffordStone = true):
                    CanvasData.PrestigeStoneCostText.SetTextColor(NamedColors.White)
                else:
                    CanvasData.PrestigeStoneCostText.SetTextColor(NamedColors.Red)

                if(PrestigeCanAffordResource = true):
                    CanvasData.PrestigeResourceCostText.SetTextColor(NamedColors.White)
                else:
                    CanvasData.PrestigeResourceCostText.SetTextColor(NamedColors.Red)
            else:
                CanvasData.PrestigeCantUseText.SetVisibility(widget_visibility.Hidden)
                #CanvasData.PrestigeCantUseText.SetText(ToMessage("NEED WEAPON"))
                CanvasData.PrestigeButtonOverlay.SetVisibility(widget_visibility.Hidden)

            TiersAndCostData := GetSwordUnlockData(CanvasData.CurrentResourceType)

            Cost := TiersAndCostData.StoneCost
            ResourceCost := TiersAndCostData.ResourceCost
            CanAffordStone := BigNumber_GreaterThanOrEquals(StoneAmount, Cost)
            CanAffordResource := BigNumber_GreaterThanOrEquals(PrestigeResourceAmount, ResourceCost)
            #Check if we have enough ore for the next upgrade

            if(CanAffordStone = true):
                CanvasData.StoneCostText.SetTextColor(NamedColors.White)
            else:
                CanvasData.StoneCostText.SetTextColor(NamedColors.Red)

            if(CanAffordResource = true):
                CanvasData.ResourceCostText.SetTextColor(NamedColors.White)
            else:
                CanvasData.ResourceCostText.SetTextColor(NamedColors.Red)

    UpdateSwordStat(InPlayer : player, InSword : Sword_Data):void=
        if(CanvasData := BuiltCanvasMap[InPlayer], StatData := CanvasData.SwordStatDatas[0]):
            var StatString : string = ""

            for(Index -> SwordStat : InSword.BigNumber_StatValues):
                if(Index > 0):
                    set StatString += "+{FormatFloatTo2DP(BigNumber_ConvertBigNumberIntoFloat(SwordStat))}%\n"
                else:
                    set StatString += "{BigNumber_FormatDisplay(SwordStat)}/Hit\n"

            StatData.Value.SetText(ToMessage("{StatString}"))
            StatData.Widget.SetVisibility(widget_visibility.Visible)

    MakeStatDisplay(InTitle : string, InValue : string):SwordStatData=
        Title := text_block:
            DefaultText := ToMessage("{InTitle}:")
            DefaultJustification := text_justification.Left
            DefaultShadowColor := NamedColors.Black
            DefaultShadowOffset := option{vector2{X:=2.0,Y:=2.0}}
        Title.SetTextColor(NamedColors.White)
        Title.SetShadowOpacity(0.8)
        Value := text_block:
            DefaultText := ToMessage("{InValue}")
            DefaultJustification := text_justification.Right
            DefaultShadowColor := NamedColors.Black
            DefaultShadowOffset := option{vector2{X:=2.0,Y:=2.0}}
        Value.SetTextColor(NamedColors.LightGreen)
        Value.SetShadowOpacity(0.8)

        StackBox := stack_box:
            Orientation := orientation.Horizontal
            Slots := 
                array:
                    stack_box_slot:
                        Widget := Title
                        HorizontalAlignment := horizontal_alignment.Left
                        VerticalAlignment := vertical_alignment.Top
                    stack_box_slot:
                        Widget := Value
                        Padding := margin{Left := 50.0}
                        HorizontalAlignment := horizontal_alignment.Right
                        VerticalAlignment := vertical_alignment.Top

        return SwordStatData:
            Widget := StackBox
            Value := Value


    MakeInventoryDisplay(InResourceImage : texture):SwordInventoryData=
        Text := text_block:
            DefaultText := ToMessage("")
            DefaultJustification := text_justification.Center
            DefaultShadowColor := NamedColors.Black
            DefaultShadowOffset := option{vector2{X:=2.0,Y:=2.0}}
        Text.SetTextColor(NamedColors.White)
        Text.SetShadowOpacity(0.8)

        IconOverlay := overlay:
            Slots := array:
                overlay_slot:
                    VerticalAlignment := vertical_alignment.Center
                    HorizontalAlignment := horizontal_alignment.Center
                    Padding := margin{Top := -8.0, Bottom := -8.0, Left := -6.0, Right := -6.0}
                    Widget := color_block:
                        DefaultColor := NamedColors.White
                        DefaultOpacity := 0.2
                        DefaultDesiredSize := vector2{X := 52.0, Y := 52.0}
                overlay_slot:
                    VerticalAlignment := vertical_alignment.Center
                    HorizontalAlignment := horizontal_alignment.Center
                    Widget := texture_block:
                        DefaultImage := InResourceImage
        
        Overlay := overlay:
            Slots := array:
                overlay_slot:
                    VerticalAlignment := vertical_alignment.Center
                    HorizontalAlignment := horizontal_alignment.Left
                    Widget := color_block:
                        DefaultColor := NamedColors.White
                        DefaultOpacity := 0.2
                        DefaultDesiredSize := vector2{X:=225.0,Y:=60.0}
                overlay_slot:
                    VerticalAlignment := vertical_alignment.Center
                    HorizontalAlignment := horizontal_alignment.Left
                    Padding := margin{Top := 10.0, Bottom := 10.0, Left := 10.0, Right := 10.0}
                    Widget := stack_box:
                        Orientation := orientation.Horizontal
                        Slots := 
                            array:
                                stack_box_slot:
                                    Widget := IconOverlay
                                    HorizontalAlignment := horizontal_alignment.Left
                                    VerticalAlignment := vertical_alignment.Center
                                stack_box_slot:
                                    Widget := Text
                                    Padding := margin{Left := 15.0}
                                    VerticalAlignment := vertical_alignment.Center
                                    HorizontalAlignment := horizontal_alignment.Left

        return SwordInventoryData:
            Widget := Overlay
            Text := Text

    MakeSwordButtonDisplay(InTitle : string, InSwordImage : texture):SwordButtonData=
        SwordButton := button_quiet:
            DefaultText := ToMessage("")

        SwordImage := texture_block:
            DefaultImage := InSwordImage
            DefaultDesiredSize := vector2{X:= 128.0, Y:=128.0}
        
        SwordTitleText := text_block:
            DefaultText := ToMessage("{InTitle}")
            DefaultJustification := text_justification.Center
            DefaultShadowColor := NamedColors.Black
            DefaultShadowOffset := option{vector2{X:=2.0,Y:=2.0}}
        SwordTitleText.SetTextColor(NamedColors.White)
        SwordTitleText.SetShadowOpacity(0.8)

        TextStackBox := stack_box:
            Orientation := orientation.Vertical
            Slots := 
                array:
                    stack_box_slot:
                        Widget := SwordTitleText
                        HorizontalAlignment := horizontal_alignment.Left
                        VerticalAlignment := vertical_alignment.Center

        SwordOverlay := overlay:
            Slots := array:
                overlay_slot:
                    VerticalAlignment := vertical_alignment.Center
                    HorizontalAlignment := horizontal_alignment.Center
                    Widget := color_block:
                        DefaultDesiredSize := vector2{X:=650.0, Y:=100.0}
                        DefaultOpacity := 0.0
                overlay_slot:
                    VerticalAlignment := vertical_alignment.Center
                    HorizontalAlignment := horizontal_alignment.Right
                    Padding:= margin{Right := 20.0}
                    Widget := SwordImage
                overlay_slot:
                    VerticalAlignment := vertical_alignment.Center
                    HorizontalAlignment := horizontal_alignment.Left
                    Padding:= margin{Left := 20.0}
                    Widget := TextStackBox
                overlay_slot:
                    VerticalAlignment := vertical_alignment.Fill
                    HorizontalAlignment := horizontal_alignment.Fill
                    Widget := SwordButton

        return SwordButtonData:
            Widget := SwordOverlay
            SwordImage := SwordImage
            Button := SwordButton

    MakeSwordMenuCanvas():SwordMenuCanvasData=
        Print("Make Sword")
        ExitButton := button_loud:
            DefaultText := ToMessage("Exit")
        SmeltButton := button_quiet:
            DefaultText := ToMessage("")
        PrestigeButton := button_quiet:
            DefaultText := ToMessage("")

        StoneInventory := MakeInventoryDisplay(VerseExposed.T_Stone)
        OrbInventory := MakeInventoryDisplay(VerseExposed.T_Orb)

        SwordCommon := MakeSwordButtonDisplay("Common Sword", VerseExposed.T_Sword_Common)
        SwordUncommon := MakeSwordButtonDisplay("Uncommon Sword", VerseExposed.T_Sword_Uncommon)
        SwordRare := MakeSwordButtonDisplay("Rare Sword", VerseExposed.T_Sword_Rare)
        #SwordEpic := MakeSwordButtonDisplay("Epic Sword", VerseExposed.T_Sword_Epic)
        #SwordLegendary := MakeSwordButtonDisplay("Legendary Sword", VerseExposed.T_Sword_Legendary)
        
        SiphonStat := MakeStatDisplay("Base Damage\nCritical Rate\nCritical Damage\nBonus Strength", "5.00\n5.00\n5.00\n5.00")
        #StrengthStat := MakeStatDisplay("Critical Rate", "5.00")
        #CriticalDamageStat := MakeStatDisplay("Critical Damage", "5.00")

        SwordTitle := text_block:
            DefaultText := ToMessage("Swords")
            DefaultJustification := text_justification.Center
            DefaultShadowColor := NamedColors.Black
            DefaultShadowOffset := option{vector2{X:=2.0,Y:=2.0}}
        SwordTitle.SetTextColor(NamedColors.White)
        SwordTitle.SetShadowOpacity(0.8)

        WeaponGradeTitle := text_block:
            DefaultText := ToMessage("Weapon Grade: Common Sword")
            DefaultJustification := text_justification.Center
            DefaultShadowColor := NamedColors.Black
            DefaultShadowOffset := option{vector2{X:=2.0,Y:=2.0}}
        WeaponGradeTitle.SetTextColor(NamedColors.White)
        WeaponGradeTitle.SetShadowOpacity(0.8)
        WeaponDamageText := text_block:
            DefaultText := ToMessage("Damage: Tier 1")
            DefaultJustification := text_justification.Center
            DefaultShadowColor := NamedColors.Black
            DefaultShadowOffset := option{vector2{X:=2.0,Y:=2.0}}
        WeaponDamageText.SetTextColor(NamedColors.White)
        WeaponDamageText.SetShadowOpacity(0.8)
        
        InventoryTitle := text_block:
            DefaultText := ToMessage("Inventory")
            DefaultJustification := text_justification.Center
            DefaultShadowColor := NamedColors.Black
            DefaultShadowOffset := option{vector2{X:=2.0,Y:=2.0}}
        InventoryTitle.SetTextColor(NamedColors.White)
        InventoryTitle.SetShadowOpacity(0.8)

        StatsTitle := text_block:
            DefaultText := ToMessage("Stats")
            DefaultJustification := text_justification.Center
            DefaultShadowColor := NamedColors.Black
            DefaultShadowOffset := option{vector2{X:=2.0,Y:=2.0}}
        StatsTitle.SetTextColor(NamedColors.White)
        StatsTitle.SetShadowOpacity(0.8)

        UnlockText := text_block:
            DefaultText := ToMessage("UNLOCK")
            DefaultJustification := text_justification.Center
            DefaultShadowColor := NamedColors.Black
            DefaultShadowOffset := option{vector2{X:=2.0,Y:=2.0}}
        UnlockText.SetTextColor(NamedColors.White)
        UnlockText.SetShadowOpacity(0.8)

        UnlockedText := text_block:
            DefaultText := ToMessage("UNLOCKED")
            DefaultJustification := text_justification.Center
            DefaultShadowColor := NamedColors.Black
            DefaultShadowOffset := option{vector2{X:=2.0,Y:=2.0}}
        UnlockedText.SetTextColor(NamedColors.LightGreen)
        UnlockedText.SetShadowOpacity(0.8)

        NeedWeaponText := text_block:
            DefaultText := ToMessage("NEED WEAPON")
            DefaultJustification := text_justification.Center
            DefaultShadowColor := NamedColors.Black
            DefaultShadowOffset := option{vector2{X:=2.0,Y:=2.0}}
        NeedWeaponText.SetTextColor(NamedColors.DarkRed)
        NeedWeaponText.SetShadowOpacity(0.8)

        StoneCost := text_block:
            DefaultText := ToMessage("100.0M")
            DefaultJustification := text_justification.Center
            DefaultShadowColor := NamedColors.Black
            DefaultShadowOffset := option{vector2{X:=2.0,Y:=2.0}}
        StoneCost.SetTextColor(NamedColors.White)
        StoneCost.SetShadowOpacity(0.8)
        StoneCostStack := stack_box:
            Orientation := orientation.Horizontal
            Slots := 
                array:
                    stack_box_slot:
                        Widget := texture_block:
                            DefaultImage := VerseExposed.T_Stone
                        Padding := margin{Right := 10.0}
                        HorizontalAlignment := horizontal_alignment.Left
                        VerticalAlignment := vertical_alignment.Center
                    stack_box_slot:
                        Widget := StoneCost
                        HorizontalAlignment := horizontal_alignment.Center
                        VerticalAlignment := vertical_alignment.Center
        
        ResourceCostImage := texture_block:
            DefaultImage := VerseExposed.T_CopperOreStroke

        OreCost := text_block:
            DefaultText := ToMessage("1")
            DefaultJustification := text_justification.Center
            DefaultShadowColor := NamedColors.Black
            DefaultShadowOffset := option{vector2{X:=2.0,Y:=2.0}}
        OreCost.SetTextColor(NamedColors.White)
        OreCost.SetShadowOpacity(0.8)
        OreCostStack := stack_box:
            Orientation := orientation.Horizontal
            Slots := 
                array:
                    stack_box_slot:
                        Widget := ResourceCostImage
                        Padding := margin{Right := 10.0}
                        HorizontalAlignment := horizontal_alignment.Left
                        VerticalAlignment := vertical_alignment.Center
                    stack_box_slot:
                        Widget := OreCost
                        HorizontalAlignment := horizontal_alignment.Center
                        VerticalAlignment := vertical_alignment.Center

        CurrentResourceImage := texture_block:
            DefaultImage := VerseExposed.T_CopperOreStroke

        UnlockedOverlay := overlay:
            Slots := array:
                overlay_slot:
                    VerticalAlignment := vertical_alignment.Center
                    HorizontalAlignment := horizontal_alignment.Center
                    Widget := UnlockedText

        SmeltButtonOverlay := overlay:
            Slots := array:
                overlay_slot:
                    VerticalAlignment := vertical_alignment.Center
                    HorizontalAlignment := horizontal_alignment.Center
                    Widget := color_block:
                        DefaultDesiredSize := vector2{X:=720.0, Y:=100.0}
                        DefaultOpacity := 0.0
                overlay_slot:
                    VerticalAlignment := vertical_alignment.Top
                    HorizontalAlignment := horizontal_alignment.Center
                    Padding := margin{Top := 2.5}
                    Widget := UnlockText
                overlay_slot:
                    VerticalAlignment := vertical_alignment.Center
                    HorizontalAlignment := horizontal_alignment.Center
                    Widget := StoneCostStack
                overlay_slot:
                    VerticalAlignment := vertical_alignment.Bottom
                    HorizontalAlignment := horizontal_alignment.Center
                    Padding := margin{Bottom := 2.5}
                    Widget := OreCostStack
                overlay_slot:
                    Widget := SmeltButton
                    HorizontalAlignment := horizontal_alignment.Fill
                    VerticalAlignment := vertical_alignment.Fill

        UnlocksOverlay := overlay:
            Slots := array:
                overlay_slot:
                    VerticalAlignment := vertical_alignment.Center
                    HorizontalAlignment := horizontal_alignment.Center
                    Widget := SmeltButtonOverlay
                overlay_slot:
                    VerticalAlignment := vertical_alignment.Center
                    HorizontalAlignment := horizontal_alignment.Center
                    Widget := UnlockedOverlay

        UpgradeText := text_block:
            DefaultText := ToMessage("UPGRADE")
            DefaultJustification := text_justification.Center
            DefaultShadowColor := NamedColors.Black
            DefaultShadowOffset := option{vector2{X:=2.0,Y:=2.0}}
        UpgradeText.SetTextColor(NamedColors.White)
        UpgradeText.SetShadowOpacity(0.8)

        PrestigeStoneCost := text_block:
            DefaultText := ToMessage("100.0M")
            DefaultJustification := text_justification.Center
            DefaultShadowColor := NamedColors.Black
            DefaultShadowOffset := option{vector2{X:=2.0,Y:=2.0}}
        PrestigeStoneCost.SetTextColor(NamedColors.White)
        PrestigeStoneCost.SetShadowOpacity(0.8)
        PrestigeStoneCostStack := stack_box:
            Orientation := orientation.Horizontal
            Slots := 
                array:
                    stack_box_slot:
                        Widget := texture_block:
                            DefaultImage := VerseExposed.T_Stone
                        Padding := margin{Right := 10.0}
                        HorizontalAlignment := horizontal_alignment.Left
                        VerticalAlignment := vertical_alignment.Center
                    stack_box_slot:
                        Widget := PrestigeStoneCost
                        HorizontalAlignment := horizontal_alignment.Center
                        VerticalAlignment := vertical_alignment.Center
        
        PrestigeResourceCostImage := texture_block:
            DefaultImage := VerseExposed.T_CopperOreStroke

        PrestigeOreCost := text_block:
            DefaultText := ToMessage("1")
            DefaultJustification := text_justification.Center
            DefaultShadowColor := NamedColors.Black
            DefaultShadowOffset := option{vector2{X:=2.0,Y:=2.0}}
        PrestigeOreCost.SetTextColor(NamedColors.White)
        PrestigeOreCost.SetShadowOpacity(0.8)
        PrestigeOreCostStack := stack_box:
            Orientation := orientation.Horizontal
            Slots := 
                array:
                    stack_box_slot:
                        Widget := PrestigeResourceCostImage
                        Padding := margin{Right := 10.0}
                        HorizontalAlignment := horizontal_alignment.Left
                        VerticalAlignment := vertical_alignment.Center
                    stack_box_slot:
                        Widget := PrestigeOreCost
                        HorizontalAlignment := horizontal_alignment.Center

        PrestigeButtonOverlay := overlay:
            Slots := array:
                overlay_slot:
                    VerticalAlignment := vertical_alignment.Center
                    HorizontalAlignment := horizontal_alignment.Center
                    Widget := color_block:
                        DefaultDesiredSize := vector2{X:=720.0, Y:=100.0}
                        DefaultOpacity := 0.0
                overlay_slot:
                    VerticalAlignment := vertical_alignment.Top
                    HorizontalAlignment := horizontal_alignment.Center
                    Padding := margin{Top := 2.5}
                    Widget := UpgradeText
                overlay_slot:
                    VerticalAlignment := vertical_alignment.Center
                    HorizontalAlignment := horizontal_alignment.Center
                    Widget := PrestigeStoneCostStack
                overlay_slot:
                    VerticalAlignment := vertical_alignment.Bottom
                    HorizontalAlignment := horizontal_alignment.Center
                    Padding := margin{Bottom := 2.5}
                    Widget := PrestigeOreCostStack
                overlay_slot:
                    Widget := PrestigeButton
                    HorizontalAlignment := horizontal_alignment.Fill
                    VerticalAlignment := vertical_alignment.Fill

        PrestigeCantUseOverlay := overlay:
            Slots := array:
                overlay_slot:
                    VerticalAlignment := vertical_alignment.Center
                    HorizontalAlignment := horizontal_alignment.Center
                    Widget := NeedWeaponText

        PrestigesOverlay := overlay:
            Slots := array:
                overlay_slot:
                    VerticalAlignment := vertical_alignment.Center
                    HorizontalAlignment := horizontal_alignment.Center
                    Widget := PrestigeCantUseOverlay
                overlay_slot:
                    VerticalAlignment := vertical_alignment.Center
                    HorizontalAlignment := horizontal_alignment.Center
                    Widget := PrestigeButtonOverlay

        StatTextStackBox := stack_box:
            Orientation := orientation.Vertical
            Slots := 
                array:
                    stack_box_slot:
                        Widget := WeaponDamageText
                        Padding := margin{Bottom := 2.5}
                        HorizontalAlignment := horizontal_alignment.Center
                        VerticalAlignment := vertical_alignment.Top
                    stack_box_slot:
                        Widget := SiphonStat.Widget
                        Padding := margin{Bottom := 2.5}
                        HorizontalAlignment := horizontal_alignment.Center
                        VerticalAlignment := vertical_alignment.Top
                    <#stack_box_slot:
                        Widget := StrengthStat.Widget
                        Padding := margin{Bottom := 2.5}
                        HorizontalAlignment := horizontal_alignment.Center
                        VerticalAlignment := vertical_alignment.Top
                    stack_box_slot:
                        Widget := CriticalDamageStat.Widget
                        Padding := margin{Bottom := 2.5}
                        HorizontalAlignment := horizontal_alignment.Center
                        VerticalAlignment := vertical_alignment.Top#>

        SwordDisplayTexture := texture_block:
            DefaultImage := VerseExposed.T_UnknownSword
            DefaultDesiredSize := vector2{X:= 200.0, Y:=200.0}

        NoImageStatTextStackBox := stack_box:
            Orientation := orientation.Vertical
            Slots := 
                array:
                    stack_box_slot:
                        Widget := StatsTitle
                        Padding := margin{Bottom := 10.0}
                        HorizontalAlignment := horizontal_alignment.Center
                        VerticalAlignment := vertical_alignment.Top
                    stack_box_slot:
                        Widget := StatTextStackBox
                        HorizontalAlignment := horizontal_alignment.Center
                        VerticalAlignment := vertical_alignment.Top

        FullStatTextStackBox := stack_box:
            Orientation := orientation.Horizontal
            Slots := 
                array:
                    stack_box_slot:
                        Widget := NoImageStatTextStackBox
                        HorizontalAlignment := horizontal_alignment.Right
                        VerticalAlignment := vertical_alignment.Top

        WeaponStackBox := stack_box:
            Orientation := orientation.Vertical
            Slots := 
                array:
                    stack_box_slot:
                        Widget := WeaponGradeTitle
                        #Padding := margin{Bottom := 2.5}
                        HorizontalAlignment := horizontal_alignment.Center
                        VerticalAlignment := vertical_alignment.Center

        WeaponTextStackBox := stack_box:
            Orientation := orientation.Horizontal
            Slots := 
                array:
                    stack_box_slot:
                        Widget := WeaponStackBox
                        HorizontalAlignment := horizontal_alignment.Left
                        VerticalAlignment := vertical_alignment.Center
                        Padding := margin{Right := 50.0}
                    stack_box_slot:
                        Widget := SwordDisplayTexture
                        HorizontalAlignment := horizontal_alignment.Right
                        Padding := margin{Left := 50.0}
                        VerticalAlignment := vertical_alignment.Center

        RightFullStackBox := stack_box:
            Orientation := orientation.Vertical
            Slots := 
                array:
                    stack_box_slot:
                        Widget := WeaponTextStackBox
                        HorizontalAlignment := horizontal_alignment.Center
                        VerticalAlignment := vertical_alignment.Center
                    stack_box_slot:
                        Widget := UnlocksOverlay
                        Padding := margin{Top := 2.0}
                        HorizontalAlignment := horizontal_alignment.Center
                        VerticalAlignment := vertical_alignment.Center
                    stack_box_slot:
                        Widget := FullStatTextStackBox
                        Padding := margin{Top := 2.0}
                        HorizontalAlignment := horizontal_alignment.Center
                        VerticalAlignment := vertical_alignment.Center
                    stack_box_slot:
                        Widget := PrestigesOverlay
                        Padding := margin{Top := 2.0}
                        HorizontalAlignment := horizontal_alignment.Center
                        VerticalAlignment := vertical_alignment.Center

        SwordStackBox := stack_box:
            Orientation := orientation.Vertical
            Slots := 
                array:
                    stack_box_slot:
                        Widget := SwordTitle
                        Padding := margin{Bottom := 5.0}
                        HorizontalAlignment := horizontal_alignment.Center
                        VerticalAlignment := vertical_alignment.Center
                    stack_box_slot:
                        Widget := SwordCommon.Widget
                        Padding := margin{Bottom := 5.0}
                        HorizontalAlignment := horizontal_alignment.Center
                        VerticalAlignment := vertical_alignment.Center
                    stack_box_slot:
                        Widget := SwordUncommon.Widget
                        Padding := margin{Bottom := 5.0}
                        HorizontalAlignment := horizontal_alignment.Fill
                        VerticalAlignment := vertical_alignment.Fill
                    stack_box_slot:
                        Widget := SwordRare.Widget
                        Padding := margin{Bottom := 5.0}
                        HorizontalAlignment := horizontal_alignment.Fill
                        VerticalAlignment := vertical_alignment.Fill
                    <#stack_box_slot:
                        Widget := SwordEpic.Widget
                        Padding := margin{Bottom := 5.0}
                        HorizontalAlignment := horizontal_alignment.Fill
                        VerticalAlignment := vertical_alignment.Fill#>
                    <#stack_box_slot:
                        Widget := SwordLegendary.Widget
                        Padding := margin{Bottom := 5.0}
                        HorizontalAlignment := horizontal_alignment.Fill
                        VerticalAlignment := vertical_alignment.Fill#>

        InventoryStackBox := stack_box:
            Orientation := orientation.Vertical
            Slots := 
                array:
                    stack_box_slot:
                        Widget := InventoryTitle
                        Padding := margin{Bottom := 5.0}
                        HorizontalAlignment := horizontal_alignment.Center
                        VerticalAlignment := vertical_alignment.Center
                    stack_box_slot:
                        Widget := StoneInventory.Widget
                        Padding := margin{Bottom := 5.0}
                        HorizontalAlignment := horizontal_alignment.Center
                        VerticalAlignment := vertical_alignment.Center
                    stack_box_slot:
                        Widget := OrbInventory.Widget
                        Padding := margin{Bottom := 5.0}
                        VerticalAlignment := vertical_alignment.Center
                        HorizontalAlignment := horizontal_alignment.Center
        
        BackgroundOverlay := overlay:
            Slots := array:
                overlay_slot:
                    VerticalAlignment := vertical_alignment.Fill
                    HorizontalAlignment := horizontal_alignment.Fill
                    Widget := texture_block:
                        DefaultImage := VerseExposed.T_UIBackground

        OutCanvas := canvas:
            Slots := array:
                canvas_slot:
                    Anchors := anchors{Minimum := vector2{X:=0.0, Y:=0.0}, Maximum := vector2{X:=1.0, Y:=1.0}}
                    Alignment := CenteredAlignment
                    Widget := BackgroundOverlay
                    ZOrder := 0
                canvas_slot:
                    Anchors := anchors{Minimum := vector2{X:=0.2, Y:=0.2}, Maximum := vector2{X:=0.48, Y:=0.85}}
                    Alignment := CenteredAlignment
                    Widget := SwordStackBox
                    ZOrder := 1
                canvas_slot:
                    Anchors := anchors{Minimum := vector2{X:=0.1, Y:=0.2}, Maximum := vector2{X:=0.1, Y:=0.85}}
                    Alignment := CenteredAlignment
                    Widget := InventoryStackBox
                    ZOrder := 1
                canvas_slot:
                    Anchors := anchors{Minimum := vector2{X:=0.52, Y:=0.2}, Maximum := vector2{X:=0.9, Y:=0.85}}
                    Alignment := CenteredAlignment
                    Widget := RightFullStackBox
                    ZOrder := 1
                canvas_slot:
                    Anchors := anchors{Minimum := vector2{X:=0.4, Y:=0.95}, Maximum := vector2{X:=0.6, Y:=0.95}}
                    Alignment := CenteredAlignment
                    Widget := ExitButton
                    ZOrder := 1

        return SwordMenuCanvasData:
            Canvas := OutCanvas
            CurrentResourceType := PropResourceType.Copper
            CurrentSwordRarity := SwordRarity.Common
            WeaponGradeTitle := WeaponGradeTitle
            DamageTierText := WeaponDamageText
            UnlockButtonOverlay := SmeltButtonOverlay
            UnlockedText := UnlockedText
            StoneCostText := StoneCost
            ResourceCostText := OreCost
            ResourceCostImage := ResourceCostImage
            PlayerStoneText := StoneInventory.Text
            PlayerOrbText := OrbInventory.Text
            SwordCommon := SwordCommon
            SwordUncommon := SwordUncommon
            SwordRare := SwordRare
            #SwordEpic := SwordEpic
            #SwordLegendary := SwordLegendary
            SwordStatDatas := array{SiphonStat<#, StrengthStat, CriticalDamageStat#>}
            SwordDisplayTexture := SwordDisplayTexture
            CurrentResourceDisplay := CurrentResourceImage
            PrestigeButtonOverlay := PrestigeButtonOverlay
            PrestigeStoneCostText := PrestigeStoneCost
            PrestigeResourceCostText := PrestigeOreCost
            PrestigeResourceCostImage := PrestigeResourceCostImage
            PrestigeCantUseText := NeedWeaponText
            UnlockButton := SmeltButton
            UpgradeButton := PrestigeButton
            ExitButton  := ExitButton