using { /Fortnite.com/Devices }
using { /Verse.org/Simulation }
using { /UnrealEngine.com/Temporary/Diagnostics }
using { /UnrealEngine.com/Temporary/UI }
using { /Fortnite.com/UI }
using { /UnrealEngine.com/Temporary/SpatialMath}
using { /Verse.org/Colors }
using { /Verse.org/Assets }

MakeTeleporterItem<constructor>(InImage : texture, InName : string, InTrigger : trigger_device) := TeleporterItemData:
    Image := InImage
    Name := InName
    Trigger := InTrigger

MakeTeleporterButtons<constructor>(InName : string, InTrigger : trigger_device, InAgent : agent) := TeleporterButtonItemData:
    Name := InName
    TriggerRef := InTrigger
    AgentRef := InAgent

#BUTTONS ON LEFT START
TeleporterButtonItemData := class:
    Name : string
    TriggerRef : trigger_device
    AgentRef : agent

TeleporterButtonSlotData := class:
    Widget : widget
    Button : text_button_base
    ButtonColor : texture_block
    var StoredTrigger : ?trigger_device = false
    OnTeleportButtonEvent(WidgetMessage : widget_message):void=
        if(bStoredTrigger := StoredTrigger?):
            bStoredTrigger.Trigger(WidgetMessage.Player)

TeleporterButtonCanvasData := class:
    #TeleporterSlots : []TeleporterButtonSlotData
    Canvas : canvas

#BUTTONS ON LEFT END

TeleporterItemData := class:
    Image : texture
    Name : string
    Trigger : trigger_device

TeleporterPageData := class:
    Slots : []TeleporterSlotData
    WidgetTop : widget
    WidgetBottom : widget

TeleporterSlotData := class:
    Widget : widget
    Button : text_button_base
    ButtonColor : texture_block

    var StoredTrigger : ?trigger_device = false
    #Add Logic check here
    var Unlocked : logic = false

    EnableUnlocked():void=
        set Unlocked = true

    OnButtonEvent(WidgetMessage : widget_message):void=
        if(Unlocked = true):
            if(bStoredTrigger := StoredTrigger?):
                bStoredTrigger.Trigger(WidgetMessage.Player)

TeleporterCanvasData := class:
    ExitButton : text_button_base
    TownTeleporterSlots : []TeleporterSlotData
    MushroomTeleporterSlots : []TeleporterSlotData
    LavaTeleporterSlots : []TeleporterSlotData
    UnderworldTeleporterSlots : []TeleporterSlotData
    MagicTeleporterSlots : []TeleporterSlotData


    TownOverlayTop : widget
    MushroomOverlayTop : widget
    LavaOverlayTop : widget
    UnderworldOverlayTop : widget
    MagicOverlayTop : widget
    TownOverlayBottom : widget
    MushroomOverlayBottom : widget
    LavaOverlayBottom : widget
    UnderworldOverlayBottom : widget
    MagicOverlayBottom : widget

    Canvas : canvas

UI_TeleporterDisplay := class(creative_device):
    ToMessage<localizes>(String : string) : message = "{String}"
    var BuiltCanvasMap : [player]TeleporterCanvasData = map{}
    var ButtonCanvas : [player]TeleporterButtonCanvasData = map{}

    var TownButtonTrigger : trigger_device = trigger_device{}
    var MushroomButtonTrigger : trigger_device = trigger_device{}
    var LavaButtonTrigger : trigger_device = trigger_device{}
    var UnderworldButtonTrigger : trigger_device = trigger_device{}
    var MagicButtonTrigger : trigger_device = trigger_device{}


    #Creating a map of Canvas Buttons
    var ButtonBuiltCanvasMap : [player]TeleporterButtonCanvasData = map{}
    CenteredAlignment : vector2 = vector2{X := 0.5, Y := 0.5}

    # Initiliase the UI so it can added to the player much quicker
    InitUI(InPlayer : player, TownTeleporterItems : []TeleporterItemData, MushroomTeleporterItems : []TeleporterItemData, LavaTeleporterItems : []TeleporterItemData, UnderworldTeleporterItems : []TeleporterItemData, MagicTeleporterItems : []TeleporterItemData):void=
        TeleporterCanvas := MakeTeleporterCanvas(TownTeleporterItems, MushroomTeleporterItems, LavaTeleporterItems, UnderworldTeleporterItems, MagicTeleporterItems)
        # Set the players build canvas data
        if(set BuiltCanvasMap[InPlayer] = TeleporterCanvas){}
        TeleporterCanvas.ExitButton.OnClick().Subscribe(HandlePlayerExit)

    # Add the UI to the player
    AddUI(InPlayer : player):void=
        if(PlayerUI := GetPlayerUI[InPlayer], CanvasData := BuiltCanvasMap[InPlayer]):
            CanvasData.Canvas.SetVisibility(widget_visibility.Visible)
            PlayerUI.AddWidget(CanvasData.Canvas, player_ui_slot{InputMode := ui_input_mode.All} )

    # When player presses the exit button
    HandlePlayerExit(WidgetMessage: widget_message):void=
        spawn{HideUI(WidgetMessage.Player)}

    #ChangePages(WidgetMessage : widget_message):void=
        #spawn{HideUI(WidgetMessage.Player)}
        
    HideUI(Player: player)<suspends>:void=
        #Sleep(0.0) # So we're sure that the canvas has been added to BuiltCanvas map
        if(PlayerUI := GetPlayerUI[Player], CanvasData := BuiltCanvasMap[Player]):
            CanvasData.Canvas.SetVisibility(widget_visibility.Hidden)
            PlayerUI.RemoveWidget(CanvasData.Canvas)

    #Here is where we update the UI
    #TODO: Add the new teleporter stuff to it as well
    UpdateUI(InPlayer : player, OpenedBarriers : []logic, Town : logic, Mushroom : logic, Lava : logic, Underworld : logic, Magic : logic):void=
        if(CanvasData := BuiltCanvasMap[InPlayer]):
            # Make all widget collapsed to update the correct one
            CanvasData.TownOverlayTop.SetVisibility(widget_visibility.Collapsed)
            CanvasData.TownOverlayBottom.SetVisibility(widget_visibility.Collapsed)
            CanvasData.MushroomOverlayTop.SetVisibility(widget_visibility.Collapsed)
            CanvasData.MushroomOverlayBottom.SetVisibility(widget_visibility.Collapsed)
            CanvasData.LavaOverlayTop.SetVisibility(widget_visibility.Collapsed)
            CanvasData.LavaOverlayBottom.SetVisibility(widget_visibility.Collapsed)
            CanvasData.UnderworldOverlayTop.SetVisibility(widget_visibility.Collapsed)
            CanvasData.UnderworldOverlayBottom.SetVisibility(widget_visibility.Collapsed)
            CanvasData.MagicOverlayTop.SetVisibility(widget_visibility.Collapsed)
            CanvasData.MagicOverlayBottom.SetVisibility(widget_visibility.Collapsed)
            for(Index -> OpenedBarrier : OpenedBarriers):
                if(Town = true):
                    CanvasData.TownOverlayTop.SetVisibility(widget_visibility.Visible)
                    CanvasData.TownOverlayBottom.SetVisibility(widget_visibility.Visible)
                    if(Slot := CanvasData.TownTeleporterSlots[Index]):
                        if(OpenedBarrier = true):
                            Slot.Button.SetText(ToMessage("TELEPORT"))
                            Slot.Button.SetEnabled(true)
                            Slot.ButtonColor.SetTint(NamedColors.Yellow)
                            Slot.EnableUnlocked()
                        else:
                            Slot.Button.SetText(ToMessage("🔒"))
                            Slot.Button.SetEnabled(false)
                            Slot.ButtonColor.SetTint(NamedColors.Grey)
                else if(Mushroom = true):
                    CanvasData.MushroomOverlayTop.SetVisibility(widget_visibility.Visible)
                    CanvasData.MushroomOverlayBottom.SetVisibility(widget_visibility.Visible)
                    if(Slot := CanvasData.MushroomTeleporterSlots[Index]):
                        if(OpenedBarrier = true):
                            Slot.Button.SetText(ToMessage("TELEPORT"))
                            Slot.Button.SetEnabled(true)
                            Slot.ButtonColor.SetTint(NamedColors.Yellow)
                            Slot.EnableUnlocked()
                        else:
                            Slot.Button.SetText(ToMessage("🔒"))
                            Slot.Button.SetEnabled(false)
                            Slot.ButtonColor.SetTint(NamedColors.Grey)
                else if(Lava = true):
                    CanvasData.LavaOverlayTop.SetVisibility(widget_visibility.Visible)
                    CanvasData.LavaOverlayBottom.SetVisibility(widget_visibility.Visible)
                    if(Slot := CanvasData.LavaTeleporterSlots[Index]):
                        if(OpenedBarrier = true):
                            Slot.Button.SetText(ToMessage("TELEPORT"))
                            Slot.Button.SetEnabled(true)
                            Slot.ButtonColor.SetTint(NamedColors.Yellow)
                            Slot.EnableUnlocked()
                        else:
                            Slot.Button.SetText(ToMessage("🔒"))
                            Slot.Button.SetEnabled(false)
                            Slot.ButtonColor.SetTint(NamedColors.Grey)
                else if(Underworld = true):
                    CanvasData.UnderworldOverlayTop.SetVisibility(widget_visibility.Visible)
                    CanvasData.UnderworldOverlayBottom.SetVisibility(widget_visibility.Visible)
                    if(Slot := CanvasData.UnderworldTeleporterSlots[Index]):
                        if(OpenedBarrier = true):
                            Slot.Button.SetText(ToMessage("TELEPORT"))
                            Slot.Button.SetEnabled(true)
                            Slot.ButtonColor.SetTint(NamedColors.Yellow)
                            Slot.EnableUnlocked()
                        else:
                            Slot.Button.SetText(ToMessage("🔒"))
                            Slot.Button.SetEnabled(false)
                            Slot.ButtonColor.SetTint(NamedColors.Grey)
                else if(Magic = true):
                    CanvasData.MagicOverlayTop.SetVisibility(widget_visibility.Visible)
                    CanvasData.MagicOverlayBottom.SetVisibility(widget_visibility.Visible)
                    if(Slot := CanvasData.MagicTeleporterSlots[Index]):
                        if(OpenedBarrier = true):
                            Slot.Button.SetText(ToMessage("TELEPORT"))
                            Slot.Button.SetEnabled(true)
                            Slot.ButtonColor.SetTint(NamedColors.Yellow)
                            Slot.EnableUnlocked()
                        else:
                            Slot.Button.SetText(ToMessage("🔒"))
                            Slot.Button.SetEnabled(false)
                            Slot.ButtonColor.SetTint(NamedColors.Grey)

    TownButtonEvent(WidgetMessage : widget_message):void=
        Print("Town Button Pressed")
        #ChangePages(WidgetMessage)
        TownButtonTrigger.Trigger(WidgetMessage.Player)

    MushroomButtonEvent(WidgetMessage : widget_message):void=
        Print("Mushroom Button Pressed")
        #ChangePages(WidgetMessage)
        MushroomButtonTrigger.Trigger(WidgetMessage.Player)

    LavaButtonEvent(WidgetMessage : widget_message):void=
        Print("Lava Button Pressed")
        #ChangePages(WidgetMessage)
        LavaButtonTrigger.Trigger(WidgetMessage.Player)

    UnderworldButtonEvent(WidgetMessage : widget_message):void=
        Print("Underworld Button Pressed")
        #ChangePages(WidgetMessage)
        UnderworldButtonTrigger.Trigger(WidgetMessage.Player)

    MagicButtonEvent(WidgetMessage : widget_message):void=
        Print("Underworld Button Pressed")
        #ChangePages(WidgetMessage)
        MagicButtonTrigger.Trigger(WidgetMessage.Player)

    MakeTeleporterCanvas(InTownTeleporterItems : []TeleporterItemData, InMushroomTeleporterItems : []TeleporterItemData, InLavaTeleporterItems : []TeleporterItemData, InUnderworldTeleporterItems : []TeleporterItemData, InMagicTeleporterItems : []TeleporterItemData):TeleporterCanvasData=
        ExitButton := button_quiet:
            DefaultText := ToMessage("Exit")

        Print("Making Button Canvas")
        #var TeleporterButtonSlotsArray : []TeleporterButtonSlotData = array{}
        TownButton := button_quiet:
            DefaultText := ToMessage("Town")
        MushroomButton := button_quiet:
            DefaultText := ToMessage("Mushroom Caves")
        LavaButton := button_quiet:
            DefaultText := ToMessage("Lava Caves")
        UnderworldButton := button_quiet:
            DefaultText := ToMessage("Underworld Caves")
        MagicButton := button_quiet:
            DefaultText := ToMessage("Ancient Library")

        TownButton.OnClick().Subscribe(TownButtonEvent)
        MushroomButton.OnClick().Subscribe(MushroomButtonEvent)
        LavaButton.OnClick().Subscribe(LavaButtonEvent)
        UnderworldButton.OnClick().Subscribe(UnderworldButtonEvent)
        MagicButton.OnClick().Subscribe(MagicButtonEvent)

        TownPage := MakeTeleporterPage(InTownTeleporterItems)
        MushroomPage := MakeTeleporterPage(InMushroomTeleporterItems)
        MushroomPage.WidgetTop.SetVisibility(widget_visibility.Collapsed)
        MushroomPage.WidgetBottom.SetVisibility(widget_visibility.Collapsed)
        LavaPage := MakeTeleporterPage(InLavaTeleporterItems)
        LavaPage.WidgetTop.SetVisibility(widget_visibility.Collapsed)
        LavaPage.WidgetBottom.SetVisibility(widget_visibility.Collapsed)
        UnderworldPage := MakeTeleporterPage(InUnderworldTeleporterItems)
        UnderworldPage.WidgetTop.SetVisibility(widget_visibility.Collapsed)
        UnderworldPage.WidgetBottom.SetVisibility(widget_visibility.Collapsed)
        MagicPage := MakeTeleporterPage(InMagicTeleporterItems)
        MagicPage.WidgetTop.SetVisibility(widget_visibility.Collapsed)
        MagicPage.WidgetBottom.SetVisibility(widget_visibility.Collapsed)

        BackgroundOverlay := overlay:
            Slots := array:
                overlay_slot:
                    VerticalAlignment := vertical_alignment.Fill
                    HorizontalAlignment := horizontal_alignment.Fill
                    Widget := texture_block:
                        DefaultImage := VerseExposed.T_TeleporterBackground1
        
        #Create the canvas for the whole widget
        Canvas := canvas:
            Slots := array:
                #--Add Buttons--#
                #The buttons themselves
                canvas_slot:
                    Anchors := anchors{Minimum := vector2{X:=0.20, Y:=0.30}, Maximum := vector2{X:=0.20, Y:=0.30}}
                    Offsets := margin{Top :=0.0, Left:= -300.0, Right:=0.0, Bottom:=0.0}
                    Alignment := vector2{X := 0.0, Y := 0.0}
                    SizeToContent := true
                    Widget := TownButton
                canvas_slot:
                    Anchors := anchors{Minimum := vector2{X:=0.20, Y:=0.30}, Maximum := vector2{X:=0.20, Y:=0.30}}
                    Offsets := margin{Top :=100.0, Left:= -300.0, Right:=0.0, Bottom:=0.0}
                    Alignment := vector2{X := 0.0, Y := 0.0}
                    SizeToContent := true
                    Widget := MushroomButton
                canvas_slot:
                    Anchors := anchors{Minimum := vector2{X:=0.20, Y:=0.30}, Maximum := vector2{X:=0.20, Y:=0.30}}
                    Offsets := margin{Top :=200.0, Left:= -300.0, Right:=0.0, Bottom:=0.0}
                    Alignment := vector2{X := 0.0, Y := 0.0}
                    SizeToContent := true
                    Widget := LavaButton
                canvas_slot:
                    Anchors := anchors{Minimum := vector2{X:=0.20, Y:=0.30}, Maximum := vector2{X:=0.20, Y:=0.30}}
                    Offsets := margin{Top :=300.0, Left:= -300.0, Right:=0.0, Bottom:=0.0}
                    Alignment := vector2{X := 0.0, Y := 0.0}
                    SizeToContent := true
                    Widget := UnderworldButton
                canvas_slot:
                    Anchors := anchors{Minimum := vector2{X:=0.20, Y:=0.30}, Maximum := vector2{X:=0.20, Y:=0.30}}
                    Offsets := margin{Top :=400.0, Left:= -300.0, Right:=0.0, Bottom:=0.0}
                    Alignment := vector2{X := 0.0, Y := 0.0}
                    SizeToContent := true
                    Widget := MagicButton
                #--BUTTTONSSS--#

                #--NOT BUTTONS--#
                canvas_slot:
                    Anchors := anchors{Minimum := vector2{X:=0.0, Y:=0.0}, Maximum := vector2{X:=1.0, Y:=1.0}}
                    Offsets := margin{Top :=0.0, Left:= 0.0, Right:=0.0, Bottom:=0.0}
                    Alignment := CenteredAlignment
                    Widget := BackgroundOverlay
                # Each Page has 2 widgets as a quick fix, should not be made by this :(
                canvas_slot:
                    #Changing Anchor Position Affects location on the Canvas
                    #0.35 difference
                    #X should be the same
                    #Y will change the height of the image
                    Anchors := anchors{Minimum := vector2{X:=0.25, Y:=0.05}, Maximum := vector2{X:=1.00, Y:=0.40}}
                    #{Minimum := vector2{X:=0.10, Y:=0.05}, Maximum := vector2{X:=1.00, Y:=0.40}}
                    Alignment := CenteredAlignment
                    Widget := TownPage.WidgetTop
                    #SizeToContent := true
                #Added
                #X:= 0.95 [Previous Max]
                canvas_slot:
                    Anchors := anchors{Minimum := vector2{X:=0.25, Y:=0.50}, Maximum := vector2{X:=1.00, Y:=0.85}}
                    Alignment := CenteredAlignment
                    Widget := TownPage.WidgetBottom
                    #SizeToContent := true

                canvas_slot:
                    #Changing Anchor Position Affects location on the Canvas
                    #0.35 difference
                    #X should be the same
                    #Y will change the height of the image
                    Anchors := anchors{Minimum := vector2{X:=0.25, Y:=0.05}, Maximum := vector2{X:=1.00, Y:=0.40}}
                    #{Minimum := vector2{X:=0.10, Y:=0.05}, Maximum := vector2{X:=1.00, Y:=0.40}}
                    Alignment := CenteredAlignment
                    Widget := MushroomPage.WidgetTop
                    #SizeToContent := true
                #Added
                #X:= 0.95 [Previous Max]
                canvas_slot:
                    Anchors := anchors{Minimum := vector2{X:=0.25, Y:=0.50}, Maximum := vector2{X:=1.00, Y:=0.85}}
                    Alignment := CenteredAlignment
                    Widget := MushroomPage.WidgetBottom
                    #SizeToContent := true

                canvas_slot:
                    #Changing Anchor Position Affects location on the Canvas
                    #0.35 difference
                    #X should be the same
                    #Y will change the height of the image
                    Anchors := anchors{Minimum := vector2{X:=0.25, Y:=0.05}, Maximum := vector2{X:=1.00, Y:=0.40}}
                    #{Minimum := vector2{X:=0.10, Y:=0.05}, Maximum := vector2{X:=1.00, Y:=0.40}}
                    Alignment := CenteredAlignment
                    Widget := LavaPage.WidgetTop
                    #SizeToContent := true
                #Added
                #X:= 0.95 [Previous Max]
                canvas_slot:
                    Anchors := anchors{Minimum := vector2{X:=0.25, Y:=0.50}, Maximum := vector2{X:=1.00, Y:=0.85}}
                    Alignment := CenteredAlignment
                    Widget := LavaPage.WidgetBottom
                    #SizeToContent := true

                canvas_slot:
                    #Changing Anchor Position Affects location on the Canvas
                    #0.35 difference
                    #X should be the same
                    #Y will change the height of the image
                    Anchors := anchors{Minimum := vector2{X:=0.25, Y:=0.05}, Maximum := vector2{X:=1.00, Y:=0.40}}
                    #{Minimum := vector2{X:=0.10, Y:=0.05}, Maximum := vector2{X:=1.00, Y:=0.40}}
                    Alignment := CenteredAlignment
                    Widget := UnderworldPage.WidgetTop
                    #SizeToContent := true
                #Added
                #X:= 0.95 [Previous Max]
                canvas_slot:
                    Anchors := anchors{Minimum := vector2{X:=0.25, Y:=0.50}, Maximum := vector2{X:=1.00, Y:=0.85}}
                    Alignment := CenteredAlignment
                    Widget := UnderworldPage.WidgetBottom
                    #SizeToContent := true

                canvas_slot:
                    #Changing Anchor Position Affects location on the Canvas
                    #0.35 difference
                    #X should be the same
                    #Y will change the height of the image
                    Anchors := anchors{Minimum := vector2{X:=0.25, Y:=0.05}, Maximum := vector2{X:=1.00, Y:=0.40}}
                    #{Minimum := vector2{X:=0.10, Y:=0.05}, Maximum := vector2{X:=1.00, Y:=0.40}}
                    Alignment := CenteredAlignment
                    Widget := MagicPage.WidgetTop
                    #SizeToContent := true
                #Added
                #X:= 0.95 [Previous Max]
                canvas_slot:
                    Anchors := anchors{Minimum := vector2{X:=0.25, Y:=0.50}, Maximum := vector2{X:=1.00, Y:=0.85}}
                    Alignment := CenteredAlignment
                    Widget := MagicPage.WidgetBottom
                    #SizeToContent := true

                canvas_slot:
                    # Fix for controllers not being able to find buttons that are not perpendicular to the selected button possition
                    #Anchors := anchors{Minimum := vector2{X:=0.4, Y:=0.95}, Maximum := vector2{X:=0.6, Y:=0.95}}
                    Anchors := anchors{Minimum := vector2{X:=0.05, Y:=0.95}, Maximum := vector2{X:=0.2, Y:=0.95}}
                    Alignment := CenteredAlignment
                    Widget := ExitButton
                    
        return TeleporterCanvasData:
            ExitButton := ExitButton
            TownTeleporterSlots := TownPage.Slots
            MushroomTeleporterSlots := MushroomPage.Slots
            LavaTeleporterSlots := LavaPage.Slots
            UnderworldTeleporterSlots := UnderworldPage.Slots
            MagicTeleporterSlots := MagicPage.Slots
            TownOverlayTop := TownPage.WidgetTop
            MushroomOverlayTop := MushroomPage.WidgetTop
            LavaOverlayTop := LavaPage.WidgetTop
            UnderworldOverlayTop := UnderworldPage.WidgetTop
            MagicOverlayTop := MagicPage.WidgetTop
            TownOverlayBottom := TownPage.WidgetBottom
            MushroomOverlayBottom := MushroomPage.WidgetBottom
            LavaOverlayBottom := LavaPage.WidgetBottom
            UnderworldOverlayBottom := UnderworldPage.WidgetBottom
            MagicOverlayBottom := MagicPage.WidgetBottom
            Canvas := Canvas


    # Make a singular Teleporter slot
    MakeTeleporterSlot(InImage : texture, InName : string):TeleporterSlotData=
        #This margin has been changed. The bottom is the amount of space we have for the image.
        IconMargin := margin{Bottom := 0.0, Left := 10.0, Right := 10.0, Top := 30.0} #Bottom was 175.0
        PurchaseButton := button_quiet:
            DefaultText := ToMessage("🔒")

        PurchaseButtonColor := texture_block:
            DefaultImage := VerseExposed.T_TeleporterButtonBackground
            DefaultDesiredSize := vector2{X:=512.0,Y:=54.0}
            DefaultTint := NamedColors.Yellow
        
        #This is the biome name
        CostBlock := text_block:
            DefaultText := ToMessage(InName)
            DefaultJustification := text_justification.Center
            DefaultShadowColor := NamedColors.Black
            DefaultShadowOffset := option{vector2{X:=2.0,Y:=2.0}}
        CostBlock.SetTextColor(NamedColors.White)
        CostBlock.SetShadowOpacity(0.8)
    
        BackgroundColorBlock := color_block:
            DefaultColor := NamedColors.White
            DefaultOpacity := 0.4

        BorderSize := 1.0
        TopBorderColorBlock := color_block{DefaultColor := NamedColors.DarkBlue, DefaultDesiredSize := vector2{X := BorderSize, Y := BorderSize}}
        RightBorderColorBlock := color_block{DefaultColor := NamedColors.DarkBlue, DefaultDesiredSize := vector2{X := BorderSize, Y := BorderSize}}
        BottomBorderColorBlock := color_block{DefaultColor := NamedColors.DarkBlue, DefaultDesiredSize := vector2{X := BorderSize, Y := BorderSize}}
        LeftBorderColorBlock := color_block{DefaultColor := NamedColors.DarkBlue, DefaultDesiredSize := vector2{X := BorderSize, Y := BorderSize}}

        ButtonOverlay := overlay:
            Slots := array:
                overlay_slot:
                    VerticalAlignment := vertical_alignment.Fill
                    HorizontalAlignment := horizontal_alignment.Fill
                    Widget := color_block:
                        DefaultColor := NamedColors.White
                        DefaultOpacity := 0.0

                overlay_slot:
                    VerticalAlignment := vertical_alignment.Center
                    HorizontalAlignment := horizontal_alignment.Center
                    Widget := PurchaseButtonColor

                overlay_slot:
                    VerticalAlignment := vertical_alignment.Center
                    HorizontalAlignment := horizontal_alignment.Fill
                    Padding := margin{Top := 3.0, Bottom := 3.0, Left := 10.0, Right := 10.0}
                    Widget := PurchaseButton

        StackBoxSlots :=
            array:
                stack_box_slot:
                    Widget := CostBlock
                stack_box_slot:
                    Padding := margin{Top := 10.0}
                    VerticalAlignment := vertical_alignment.Fill
                    HorizontalAlignment := horizontal_alignment.Fill
                    Widget := ButtonOverlay

        Overlay := overlay:
            Slots := array:
                #The Transparent White Background
                overlay_slot:
                    VerticalAlignment := vertical_alignment.Fill
                    HorizontalAlignment := horizontal_alignment.Fill
                    #Allows us to control how far down the white background will go.
                    #We want this number and the outline number to be the same
                    Padding := margin{Bottom := -60.0}  #-50.0
                    Widget := BackgroundColorBlock

                #Image to display for the Biome [Moved up here so the text can overlay above the button]
                overlay_slot:
                    VerticalAlignment := vertical_alignment.Fill
                    HorizontalAlignment := horizontal_alignment.Fill
                    Padding := IconMargin
                    Widget := texture_block:
                        DefaultImage := InImage
                        DefaultDesiredSize := vector2{X:=512.0,Y:=512.0}

                #This is the background "Plate" for each teleporter image [I think]
                overlay_slot:
                    VerticalAlignment := vertical_alignment.Bottom
                    HorizontalAlignment := horizontal_alignment.Fill
                    Padding := margin{Top := 0.0}  #Padding := margin{Top := 50.0}
                    #This is for the stack box holding the cave name [Cost Block] and Button
                    Widget := overlay:
                        Slots := array:
                            overlay_slot:
                                #The Top being lower will allow the button + text to come down [Content in StackBoxSlots]
                                Padding := margin{Top := 120.0, Bottom := 0.0}  #Was 20.0 #top was 50.0
                                VerticalAlignment := vertical_alignment.Center
                                HorizontalAlignment := horizontal_alignment.Fill
                                Widget := stack_box:
                                    Orientation := orientation.Vertical
                                    Slots := StackBoxSlots

                #Outlines for each box
                overlay_slot:
                    VerticalAlignment := vertical_alignment.Top
                    HorizontalAlignment := horizontal_alignment.Fill
                    Widget := TopBorderColorBlock
                overlay_slot:
                    VerticalAlignment := vertical_alignment.Bottom
                    HorizontalAlignment := horizontal_alignment.Fill
                    #This is the bounds of the outline
                    Padding := margin{Bottom := -60.0}
                    Widget := BottomBorderColorBlock
                overlay_slot:
                    VerticalAlignment := vertical_alignment.Fill
                    HorizontalAlignment := horizontal_alignment.Right
                    Padding := margin{Bottom := -60.0}
                    Widget := RightBorderColorBlock
                overlay_slot:
                    VerticalAlignment := vertical_alignment.Fill
                    HorizontalAlignment := horizontal_alignment.Left
                    Padding := margin{Bottom := -60.0}
                    Widget := LeftBorderColorBlock

        return TeleporterSlotData:
            Widget := Overlay
            Button := PurchaseButton
            ButtonColor := PurchaseButtonColor

    MakeTeleporterPage(InTeleporterItems : []TeleporterItemData):TeleporterPageData=
        var TeleporterSlots : []TeleporterSlotData = array{}
        var StackBoxSlotsTop : []stack_box_slot = array{}
        var StackBoxSlotsMiddle : []stack_box_slot = array{}

        for(Index -> CurrentItem : InTeleporterItems):
            CurrentTeleporterSlot := MakeTeleporterSlot(CurrentItem.Image, CurrentItem.Name)
            # Set the trigger for the TeleporterSlot and bind the button to the button event
            set CurrentTeleporterSlot.StoredTrigger = option{CurrentItem.Trigger}
            CurrentTeleporterSlot.Button.OnClick().Subscribe(CurrentTeleporterSlot.OnButtonEvent)
            set TeleporterSlots += array{CurrentTeleporterSlot}
            # Create the Stack Box Slot
            StackBox := stack_box_slot:
                VerticalAlignment := vertical_alignment.Center
                HorizontalAlignment := horizontal_alignment.Center
                Padding := margin{Left := 20.0, Right := 20.0}
                Widget := CurrentTeleporterSlot.Widget
            # Change the stack box that will be used based on the Index
            if(Index < 2):
                set StackBoxSlotsTop += array{StackBox}
            #Added so we can add from 3 to 6
            if((Index > 1) and (Index < 4)):
                set StackBoxSlotsMiddle += array{StackBox}

        # Create the overlays for the stack boxes
        StacksTop := overlay:
            Slots := array:
                overlay_slot:
                    VerticalAlignment := vertical_alignment.Center
                    HorizontalAlignment := horizontal_alignment.Center
                    Widget := stack_box:
                        Orientation := orientation.Horizontal
                        Slots := StackBoxSlotsTop
        #Added
        StacksMiddle := overlay:
            Slots := array:
                overlay_slot:
                    VerticalAlignment := vertical_alignment.Center
                    HorizontalAlignment := horizontal_alignment.Center
                    Widget := stack_box:
                        Orientation := orientation.Horizontal
                        Slots := StackBoxSlotsMiddle

        return TeleporterPageData:
            Slots := TeleporterSlots
            WidgetTop := StacksTop
            WidgetBottom := StacksMiddle