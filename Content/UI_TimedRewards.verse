using { /Fortnite.com/Devices }
using { /Verse.org/Simulation }
using { /UnrealEngine.com/Temporary/UI }
using { /UnrealEngine.com/Temporary/Diagnostics }
using { /UnrealEngine.com/Temporary/SpatialMath }
using { /Fortnite.com/UI }

UI_TimedRewards := class(creative_device):
    var MainCanvas : canvas = canvas{}
    var Button1 : button_quiet = button_quiet{}
    var Button2 : button_quiet = button_quiet{}
    var Button3 : button_quiet = button_quiet{}
    var Button4 : button_quiet = button_quiet{}
    var Button5 : button_quiet = button_quiet{}
    var Button6 : button_quiet = button_quiet{}

    OnBegin<override>()<suspends>:void=
        Print("Bring up Timed Rewards UI")
        #CreateUIPopUp()

    CreateUIPopUp() : void=
        Print("Timed Rewards UI Added")
        for(Player : GetPlayspace().GetPlayers()):
            set MainCanvas = CreateCavas()
            if(PlayerUI := GetPlayerUI[Player]):
                PlayerUI.AddWidget(MainCanvas, player_ui_slot{InputMode := ui_input_mode.All})

    CreateCavas() : canvas=
        Canvas : canvas = canvas:
            Slots := array:
                MakeCanvasSlot(Button1, vector2{X:=200.0, Y:=400.0}, ?Size:=vector2{X:=200.0,Y:=200.0})
                MakeCanvasSlot(Button2, vector2{X:=600.0, Y:=400.0}, ?Size:=vector2{X:=200.0,Y:=200.0})
                MakeCanvasSlot(Button3, vector2{X:=1000.0, Y:=400.0}, ?Size:=vector2{X:=200.0,Y:=200.0})
                MakeCanvasSlot(Button4, vector2{X:=200.0, Y:=800.0}, ?Size:=vector2{X:=200.0,Y:=200.0})
                MakeCanvasSlot(Button5, vector2{X:=600.0, Y:=800.0}, ?Size:=vector2{X:=200.0,Y:=200.0})
                MakeCanvasSlot(Button6, vector2{X:=1000.0, Y:=800.0}, ?Size:=vector2{X:=200.0,Y:=200.0})

        Button1.OnClick().Subscribe(OnButtonPress)
        Button2.OnClick().Subscribe(OnButtonPress)
        Button3.OnClick().Subscribe(OnButtonPress)
        Button4.OnClick().Subscribe(OnButtonPress)
        Button5.OnClick().Subscribe(OnButtonPress)
        Button6.OnClick().Subscribe(OnButtonPress)
        return Canvas

    OnButtonPress(WidgetMessage : widget_message) : void=
            if(PlayerUI := GetPlayerUI[WidgetMessage.Player]):
                MainCanvas.SetVisibility(widget_visibility.Hidden)
                PlayerUI.RemoveWidget(MainCanvas)
                Print("OnButtonPressed")