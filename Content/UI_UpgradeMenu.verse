# fill this file with your verse-script

using { /Fortnite.com/Devices }
using { /Verse.org/Simulation }
using { /UnrealEngine.com/Temporary/Diagnostics }
using { /UnrealEngine.com/Temporary/UI }
using { /Fortnite.com/UI }
using { /UnrealEngine.com/Temporary/SpatialMath}
using { /Verse.org/Colors }
using { /Verse.org/Assets }

PageButtonHandler := class() {
    Device : UI_UpgradeDisplay
    Next : logic

    HandlerFunction(WidgetMessage : widget_message) : void = {
        Device.SwitchPage(WidgetMessage.Player, Next, false)
    }
}

MakeUpgradeSlotConstructData<constructor>(InName : string, InOwnedAmount : string, InHitPowerAmount : string, InPurchaseCost : string, InUpgradeCost : string,
InPurchaseTrigger : trigger_device, InUpgradeTrigger : trigger_device) := UpgradeSlotConstructData:
    Name := InName
    OwnedAmount := InOwnedAmount
    HitPowerAmount := InHitPowerAmount
    PurchaseCost := InPurchaseCost
    UpgradeCost := InUpgradeCost

    PurchaseTrigger := InPurchaseTrigger
    UpgradeTrigger := InUpgradeTrigger

UpgradeCanvasData := class:
    Canvas : canvas
    var CurrentPageIndex : int
    UpgradeSlots : []UpgradeSlotData
    UpgradePages : []UpgradePageData
    PlayerStoneAmount : text_block
    PlayerStrengthAmount : text_block
    PurchaseAmount : text_block
    PageNumberText : text_block

    X1Trigger : trigger_device
    X10Trigger : trigger_device
    X100Trigger : trigger_device

    X1Button : text_button_base
    X10Button : text_button_base
    X100Button : text_button_base
    ExitButton : text_button_base
    NextPageButton : text_button_base
    PrevPageButton : text_button_base

    OnX1TriggerButtonEvent(WidgetMessage : widget_message):void=
        X1Trigger.Trigger(WidgetMessage.Player)
    
    OnX10TriggerButtonEvent(WidgetMessage : widget_message):void=
        X10Trigger.Trigger(WidgetMessage.Player)
        
    OnX100TriggerButtonEvent(WidgetMessage : widget_message):void=
        X100Trigger.Trigger(WidgetMessage.Player)

UpgradePageData := class:
    UpgradeWidgets : []UpgradeSlotData
    OverlaySlot : overlay_slot
    Index : int

UpgradeSlotConstructData := class:
    Name : string
    OwnedAmount : string
    HitPowerAmount : string
    PurchaseCost : string
    UpgradeCost : string

    PurchaseTrigger : trigger_device
    UpgradeTrigger : trigger_device

UpgradeSlotData := class:
    Widget : widget

    Name : string
    OwnedText : text_block
    HitPowerText : text_block
    NextHitPowerText : text_block
    PurchaseText : text_block
    UpgradeText : text_block
    UpgradeTextWidget : widget
    CantUpgradeText : text_block
    UpgradeButtonColor : texture_block
    PurchaseButtonColor : texture_block

    var ShouldShow : logic

    PurchaseButton : text_button_base
    UpgradeButton : text_button_base

    PurchaseTrigger : trigger_device
    UpgradeTrigger : trigger_device

    OnPurchaseButtonEvent(WidgetMessage : widget_message):void=
        #Print("Purchase Button Press")
        PurchaseTrigger.Trigger(WidgetMessage.Player)
    
    OnUpgradeButtonEvent(WidgetMessage : widget_message):void=
        #Print("Upgrade Button Press")
        UpgradeTrigger.Trigger(WidgetMessage.Player)

UI_UpgradeDisplay := class():
    ToMessage<localizes>(String : string) : message = "{String}"
    CenteredAlignment : vector2 = vector2{X := 0.5, Y := 0.5}

    var BuiltCanvasMap : [player]UpgradeCanvasData = map{}

    # Initiliase the UI so it can added to the player much quicker
    InitUI(InPlayer : player, InUpgradeSlotConstructDatas : []UpgradeSlotConstructData, InPlayerStoneAmount : string, InX1Trigger : trigger_device, InX10Trigger : trigger_device, InX100Trigger : trigger_device):void=
        UpgradeDisplayCanvas := MakeUpgradeCanvas(InUpgradeSlotConstructDatas, InPlayerStoneAmount, InX1Trigger, InX10Trigger, InX100Trigger)
        # Set the players build canvas data
        if(set BuiltCanvasMap[InPlayer] = UpgradeDisplayCanvas, PlayerCanvas := BuiltCanvasMap[InPlayer]):
            PlayerCanvas.ExitButton.OnClick().Subscribe(HandlePlayerExit)
            PlayerCanvas.X1Button.OnClick().Subscribe(PlayerCanvas.OnX1TriggerButtonEvent)
            PlayerCanvas.X10Button.OnClick().Subscribe(PlayerCanvas.OnX10TriggerButtonEvent)
            PlayerCanvas.X100Button.OnClick().Subscribe(PlayerCanvas.OnX100TriggerButtonEvent)

            PlayerCanvas.NextPageButton.OnClick().Subscribe(PageButtonHandler{Device:=Self, Next:=true}.HandlerFunction)
            PlayerCanvas.PrevPageButton.OnClick().Subscribe(PageButtonHandler{Device:=Self, Next:=false}.HandlerFunction)

            for(Index -> Slot : PlayerCanvas.UpgradeSlots):
                Slot.PurchaseButton.OnClick().Subscribe(Slot.OnPurchaseButtonEvent)
                Slot.UpgradeButton.OnClick().Subscribe(Slot.OnUpgradeButtonEvent)

    # Add the UI to the player
    AddUI(InPlayer : player):void=
        if(PlayerUI := GetPlayerUI[InPlayer], CanvasData := BuiltCanvasMap[InPlayer]):
            CanvasData.Canvas.SetVisibility(widget_visibility.Visible)
            PlayerUI.AddWidget(CanvasData.Canvas, player_ui_slot{InputMode := ui_input_mode.All})
            
    # When player presses the exit button
    HandlePlayerExit(WidgetMessage: widget_message):void=
        spawn{HideUI(WidgetMessage.Player)}

    HideUI(InPlayer: player)<suspends>:void=
        Sleep(0.0) # So we're sure that the canvas has been added to BuiltCanvas map
        if(PlayerUI := GetPlayerUI[InPlayer], CanvasData := BuiltCanvasMap[InPlayer]):
            CanvasData.Canvas.SetVisibility(widget_visibility.Hidden)
            PlayerUI.RemoveWidget(CanvasData.Canvas)
    
    # Upates a given player's UI
    UpdateUI(InPlayer : player, InUpgradeName : string, InOwnedAmount : string, InHitPowerAmount : string, InNextHitPowerAmount : string, InPurchaseCost : string, InUpgradeCost : string, InPlayerStoneAmount : string, 
    InPlayerStrengthAmount : string, InHarvestRates : []big_number, InPlayerData : Player_Stats_Data, InTierSystems : [][]float, InPurchaseAmount : string, CanAfford : logic):void=
        if(PlayerUI := GetPlayerUI[InPlayer], CanvasData := BuiltCanvasMap[InPlayer]):
            CanvasData.PlayerStoneAmount.SetText(ToMessage("{InPlayerStoneAmount}"))
            CanvasData.PlayerStrengthAmount.SetText(ToMessage("{InPlayerStrengthAmount}"))
            CanvasData.PurchaseAmount.SetText(ToMessage("Buy x{InPurchaseAmount}"))
            ShouldUpgradesShow(InPlayer, InPlayerData, InTierSystems, InHarvestRates)
            for(Slot : CanvasData.UpgradeSlots):
                if(Slot.Name = InUpgradeName):
                    Slot.OwnedText.SetText(ToMessage("{InOwnedAmount}"))
                    Slot.HitPowerText.SetText(ToMessage("{InHitPowerAmount}"))
                    Slot.NextHitPowerText.SetText(ToMessage("(+{InNextHitPowerAmount})"))
                    Slot.PurchaseText.SetText(ToMessage("{InPurchaseCost}"))
                    Slot.UpgradeText.SetText(ToMessage("{InUpgradeCost}"))
                    if(CanAfford = true):
                        Slot.PurchaseButtonColor.SetTint(NamedColors.Yellow)
                        Slot.PurchaseText.SetTextColor(NamedColors.White)
                        #Slot.PurchaseButton.SetEnabled(true)
                    else:
                        Slot.PurchaseButtonColor.SetTint(NamedColors.Grey)
                        Slot.PurchaseText.SetTextColor(NamedColors.Red)
                        #Slot.PurchaseButton.SetEnabled(false)
            
            var PageCount : int = 0

            for(Page : CanvasData.UpgradePages):
                if(FirstUpgrade := Page.UpgradeWidgets[0], FirstUpgrade.ShouldShow = true):
                    set PageCount += 1
            CanvasData.PageNumberText.SetText(ToMessage("{CanvasData.CurrentPageIndex + 1}/{PageCount}"))
            if(CanvasData.CurrentPageIndex + 1 > PageCount):
                SwitchPage(InPlayer, false, true)

                for(Index -> Page : CanvasData.UpgradePages):
                    if(Index <> 0):
                        Page.OverlaySlot.Widget.SetVisibility(widget_visibility.Collapsed)
                        for(UpgradeSlot : Page.UpgradeWidgets):
                            UpgradeSlot.Widget.SetVisibility(widget_visibility.Collapsed)

    SwitchPage(InPlayer : player, Next : logic, ToFirst : logic):void=
        Print("Switching Page")
        if(CanvasData := BuiltCanvasMap[InPlayer]):
            # Hide Current Page
            var PageCount : int = 0

            for(Page : CanvasData.UpgradePages):
                if(FirstUpgrade := Page.UpgradeWidgets[0], FirstUpgrade.ShouldShow = true):
                    set PageCount += 1

            if(CurrentPage := CanvasData.UpgradePages[CanvasData.CurrentPageIndex]):
                if(PageCount > 1):
                    CurrentPage.OverlaySlot.Widget.SetVisibility(widget_visibility.Collapsed)
                    for(UpgradeSlot : CurrentPage.UpgradeWidgets):
                        UpgradeSlot.Widget.SetVisibility(widget_visibility.Collapsed)
                else if(ToFirst = false):
                    return
            
            if(ToFirst = false):
                var NextPageIndex : int = CanvasData.CurrentPageIndex
                # Show Next or Previous Page
                if(Next = true):
                    if(NextPageIndex < PageCount - 1):
                        set NextPageIndex += 1
                    else:
                        set NextPageIndex = 0
                else:
                    if(NextPageIndex = 0):
                        set NextPageIndex = PageCount - 1
                    else:
                        set NextPageIndex -= 1

                set CanvasData.CurrentPageIndex = NextPageIndex

                Print("NEW PAGE :{CanvasData.CurrentPageIndex}")
            # Show new page
            if(ToFirst = true):
                set CanvasData.CurrentPageIndex = 0

            if(CurrentPage := CanvasData.UpgradePages[CanvasData.CurrentPageIndex]):
                CurrentPage.OverlaySlot.Widget.SetVisibility(widget_visibility.Visible)
                for(UpgradeSlot : CurrentPage.UpgradeWidgets):
                    if(UpgradeSlot.ShouldShow = true):
                        UpgradeSlot.Widget.SetVisibility(widget_visibility.Visible)
                CanvasData.PageNumberText.SetText(ToMessage("{CanvasData.CurrentPageIndex + 1}/{PageCount}"))


    # Upates a given player's UI
    UpdateStonePowerUI(InPlayer : player, InPlayerStoneAmount : string, InPlayerStrengthAmount : string):void=
        if(PlayerUI := GetPlayerUI[InPlayer], CanvasData := BuiltCanvasMap[InPlayer]):
            CanvasData.PlayerStoneAmount.SetText(ToMessage("{InPlayerStoneAmount}"))
            CanvasData.PlayerStrengthAmount.SetText(ToMessage("{InPlayerStrengthAmount}"))

    # Check if the upgrades should be shown to that player
    ShouldUpgradesShow(InPlayer : player, InPlayerData : Player_Stats_Data, InTierSystems : [][]float, InHarvestRates : []big_number):void=
        if(PlayerUI := GetPlayerUI[InPlayer], CanvasData := BuiltCanvasMap[InPlayer]):
            for(Index -> Slot : CanvasData.UpgradeSlots):
                # Check if the upgrade slot can be shown
                var UpgradeLevel  : float = 0.0
                var UpgradeTierAvailable : int = 0
                var TierSystems : []float = array{}
                var TierCost : big_number = big_number{}

                var CurrentLevel : float = 0.0
                var LevelRequired : float = 0.0

                if(PreviousUpgradeLevel := InPlayerData.HRUpgradeLevels[Index - 1]):
                    set UpgradeLevel = PreviousUpgradeLevel
                else:
                    set UpgradeLevel = 1.0

                if(PlayerUpgradeLevel := InPlayerData.HRUpgradeLevels[Index],  PlayerUpgradeTierAvailable := InPlayerData.HRUpgradeTierMultiplierUpgradesAvaliable[Index], 
                PlayerTierMultiplier := InPlayerData.HRUpgradeTierMultipliers[Index]):
                    set CurrentLevel = PlayerUpgradeLevel
                    set UpgradeTierAvailable = PlayerUpgradeTierAvailable
                    if(NewHarvestRate := InHarvestRates[Index]):
                        set TierCost = BigNumber_MultiplyWithAFloat(NewHarvestRate, PlayerTierMultiplier * 500)

                if(NewTierSystems := InTierSystems[Index]):
                    set TierSystems = NewTierSystems

                # If the slot can be shown check if the tier multiplier can be used
                if(UpgradeLevel = 0.0):
                    Slot.Widget.SetVisibility(widget_visibility.Hidden)
                    set Slot.ShouldShow = false
                else:
                    set Slot.ShouldShow = true
                    StoneOreAmount := InPlayerData.PlayerStoneOreAmount
                    CanAfford := BigNumber_GreaterThanOrEquals(StoneOreAmount, TierCost)
                    if(UpgradeTierAvailable >= 1 and CanAfford = true):
                        Slot.UpgradeButtonColor.SetTint(NamedColors.Yellow)
                        Slot.UpgradeButton.SetEnabled(true)
                        Slot.CantUpgradeText.SetVisibility(widget_visibility.Hidden)
                        Slot.UpgradeTextWidget.SetVisibility(widget_visibility.Visible)
                        Slot.UpgradeText.SetTextColor(NamedColors.White)
                    else if(UpgradeTierAvailable = 0):
                        Slot.UpgradeButtonColor.SetTint(NamedColors.Grey)
                        #Slot.UpgradeButton.SetEnabled(false)
                        Slot.CantUpgradeText.SetVisibility(widget_visibility.Visible)
                        Slot.UpgradeTextWidget.SetVisibility(widget_visibility.Hidden)

                        for(Tier : TierSystems):
                            if(Tier > CurrentLevel and LevelRequired = 0.0):
                                set LevelRequired = Tier
                                Slot.CantUpgradeText.SetText(ToMessage("Required : {BigNumber_FormatDisplay(BigNumber_ConvertFloatIntoBigNumber(CurrentLevel))}/{BigNumber_FormatDisplay(BigNumber_ConvertFloatIntoBigNumber(LevelRequired))}"))
                    else if(UpgradeTierAvailable = 1 and CanAfford = false):
                        Slot.UpgradeButtonColor.SetTint(NamedColors.Grey)
                        #Slot.UpgradeButton.SetEnabled(true)
                        Slot.CantUpgradeText.SetVisibility(widget_visibility.Hidden)
                        Slot.UpgradeTextWidget.SetVisibility(widget_visibility.Visible)
                        Slot.UpgradeText.SetTextColor(NamedColors.Red)

                    Slot.Widget.SetVisibility(widget_visibility.Visible)

    # Make a single slot for equipment given the variables
    MakeUpgradeSlot( InUpgradeSlotConstructData : UpgradeSlotConstructData):UpgradeSlotData=
        # Variables for other parts of UI
        # Use these to copy and paste the settings into overlays
        BackgroundColorBlock := color_block:
            DefaultColor := NamedColors.White
            DefaultOpacity := 0.2

        BorderSize := 1.0
        BorderColorBlock := color_block:
            DefaultColor := NamedColors.White
            DefaultDesiredSize := vector2{X := BorderSize, Y := BorderSize}

        StoneWidget := texture_block:
            DefaultImage := VerseExposed.T_Stone

        UpgradeButtonColor := texture_block:
            DefaultImage := VerseExposed.T_EquipmentButtonBackground
            DefaultDesiredSize := vector2{X:=400.0,Y:=54.0}
            DefaultTint := NamedColors.Yellow

        PurchaseButtonColor := texture_block:
            DefaultImage := VerseExposed.T_EquipmentButtonBackground
            DefaultDesiredSize := vector2{X:=400.0,Y:=54.0}
            DefaultTint := NamedColors.Yellow

        # Not updating text blocks
        NameText := text_block:
            DefaultText := ToMessage("{InUpgradeSlotConstructData.Name}")
            DefaultJustification := text_justification.Left
            DefaultShadowColor := NamedColors.Black
            DefaultShadowOffset := option{vector2{X:=2.0,Y:=2.0}}
        NameText.SetTextColor(NamedColors.White)
        NameText.SetShadowOpacity(0.8)
        BuyText1 := text_block:
            DefaultText := ToMessage("BUY")
            DefaultJustification := text_justification.Left
            DefaultShadowColor := NamedColors.Black
            DefaultShadowOffset := option{vector2{X:=2.0,Y:=2.0}}
        BuyText1.SetTextColor(NamedColors.White)
        BuyText1.SetShadowOpacity(0.8)
        BuyText2 := text_block:
            DefaultText := ToMessage("BUY")
            DefaultJustification := text_justification.Left
            DefaultShadowColor := NamedColors.Black
            DefaultShadowOffset := option{vector2{X:=2.0,Y:=2.0}}
        BuyText2.SetTextColor(NamedColors.White)
        BuyText2.SetShadowOpacity(0.8)
        
        # Updating text blocks
        OwnedText := text_block:
            DefaultText := ToMessage("{InUpgradeSlotConstructData.OwnedAmount}")
            DefaultJustification := text_justification.Left
            DefaultShadowColor := NamedColors.Black
            DefaultShadowOffset := option{vector2{X:=2.0,Y:=2.0}}
        OwnedText.SetTextColor(NamedColors.White)
        OwnedText.SetShadowOpacity(0.8)
        HitPowerText := text_block:
            DefaultText := ToMessage("{InUpgradeSlotConstructData.HitPowerAmount}")
            DefaultJustification := text_justification.Left
            DefaultShadowColor := NamedColors.Black
            DefaultShadowOffset := option{vector2{X:=2.0,Y:=2.0}}
        HitPowerText.SetTextColor(NamedColors.White)
        HitPowerText.SetShadowOpacity(0.8)
        NextHitPowerText := text_block:
            DefaultText := ToMessage("(+{InUpgradeSlotConstructData.HitPowerAmount})")
            DefaultJustification := text_justification.Left
            DefaultShadowColor := NamedColors.Black
            DefaultShadowOffset := option{vector2{X:=2.0,Y:=2.0}}
        NextHitPowerText.SetTextColor(NamedColors.LightGreen)
        NextHitPowerText.SetShadowOpacity(0.8)
        PurchaseText := text_block:
            DefaultText := ToMessage("{InUpgradeSlotConstructData.PurchaseCost}")
            DefaultJustification := text_justification.Left
            DefaultShadowColor := NamedColors.Black
            DefaultShadowOffset := option{vector2{X:=2.0,Y:=2.0}}
        PurchaseText.SetTextColor(NamedColors.White)
        PurchaseText.SetShadowOpacity(0.8)
        UpgradeText := text_block:
            DefaultText := ToMessage("{InUpgradeSlotConstructData.UpgradeCost}")
            DefaultJustification := text_justification.Left
            DefaultShadowColor := NamedColors.Black
            DefaultShadowOffset := option{vector2{X:=2.0,Y:=2.0}}
        UpgradeText.SetTextColor(NamedColors.White)
        UpgradeText.SetShadowOpacity(0.8)
        CantUpgradeText := text_block:
            DefaultText := ToMessage("CAN'T UPGRADE")
            DefaultJustification := text_justification.Left
            DefaultShadowColor := NamedColors.Black
            DefaultShadowOffset := option{vector2{X:=2.0,Y:=2.0}}
        CantUpgradeText.SetTextColor(NamedColors.DarkRed)
        CantUpgradeText.SetShadowOpacity(0.8)
        CantUpgradeText.SetVisibility(widget_visibility.Hidden)

        # Updating buttons
        PurchaseButton := button_quiet:
            DefaultText := ToMessage("")
        UpgradeButton := button_quiet:
            DefaultText := ToMessage("")

        NameOverlay := overlay:
            Slots := array:
                overlay_slot:
                    VerticalAlignment := vertical_alignment.Center
                    HorizontalAlignment := horizontal_alignment.Left
                    Padding := margin{Top := 10.0, Bottom := 10.0, Left := 30.0, Right := 30.0}
                    Widget := NameText

        NameStackBox := stack_box:
            Orientation := orientation.Vertical
            Slots :=
                array:
                    stack_box_slot:
                        Widget := color_block:
                            DefaultColor := NamedColors.DarkBlue
                            DefaultDesiredSize := vector2{X := 400.0, Y := BorderSize}
                        HorizontalAlignment := horizontal_alignment.Center
                        VerticalAlignment := vertical_alignment.Top
                    stack_box_slot:
                        Widget := NameOverlay
                        Padding := margin{Top := 10.0}
                        HorizontalAlignment := horizontal_alignment.Fill
                        VerticalAlignment := vertical_alignment.Fill
            
        OwnedOverlay := overlay:
            Slots := array:
                overlay_slot:
                    VerticalAlignment := vertical_alignment.Center
                    HorizontalAlignment := horizontal_alignment.Center
                    Padding := margin{Top := 10.0, Bottom := 10.0, Left := 30.0, Right := 30.0}
                    Widget := OwnedText
        
        OwnedStackBox := stack_box:
            Orientation := orientation.Vertical
            Slots :=
                array:
                    stack_box_slot:
                        Widget := color_block:
                            DefaultColor := NamedColors.DarkBlue
                            DefaultDesiredSize := vector2{X := 175.0, Y := BorderSize}
                        HorizontalAlignment := horizontal_alignment.Center
                        VerticalAlignment := vertical_alignment.Top
                    stack_box_slot:
                        Widget := OwnedOverlay
                        Padding := margin{Top := 10.0}
                        HorizontalAlignment := horizontal_alignment.Fill
                        VerticalAlignment := vertical_alignment.Fill

        HitPowerStackText := stack_box:
            Orientation := orientation.Horizontal
            Slots :=
                array:
                    stack_box_slot:
                        Widget := HitPowerText
                        Padding := margin{Right := 15.0}
                    stack_box_slot:
                        Widget := NextHitPowerText

        HitPowerOverlay := overlay:
            Slots := array:
                overlay_slot:
                    VerticalAlignment := vertical_alignment.Center
                    HorizontalAlignment := horizontal_alignment.Left
                    Padding := margin{Top := 10.0, Bottom := 10.0, Left := 30.0, Right := 30.0}
                    Widget := HitPowerStackText

        HitPowerStackBox := stack_box:
            Orientation := orientation.Vertical
            Slots :=
                array:
                    stack_box_slot:
                        Widget := color_block:
                            DefaultColor := NamedColors.DarkBlue
                            DefaultDesiredSize := vector2{X := 375.0, Y := BorderSize}
                        HorizontalAlignment := horizontal_alignment.Center
                        VerticalAlignment := vertical_alignment.Top
                    stack_box_slot:
                        Widget := HitPowerOverlay
                        Padding := margin{Top := 10.0}
                        HorizontalAlignment := horizontal_alignment.Fill
                        VerticalAlignment := vertical_alignment.Fill

        PurchaseTextStack := stack_box:
            Orientation := orientation.Horizontal
            Slots :=
                array:
                    stack_box_slot:
                        VerticalAlignment := vertical_alignment.Center
                        HorizontalAlignment := horizontal_alignment.Center
                        Padding := margin{Left := 10.0, Right := 10.0}
                        Widget := BuyText1
                    stack_box_slot:
                        VerticalAlignment := vertical_alignment.Center
                        HorizontalAlignment := horizontal_alignment.Center
                        Padding := margin{Left := 10.0, Right := 5.0}
                        Widget := texture_block:
                            DefaultImage := VerseExposed.T_Stone
                    stack_box_slot:
                        VerticalAlignment := vertical_alignment.Center
                        HorizontalAlignment := horizontal_alignment.Center
                        Padding := margin{Left := 5.0, Right := 10.0} 
                        Widget := PurchaseText
                    
        PurchaseOverlay := overlay:
            Slots := array:
                overlay_slot:
                    VerticalAlignment := vertical_alignment.Center
                    HorizontalAlignment := horizontal_alignment.Center
                    Padding := margin{Top := 3.0, Bottom := 3.0, Left := 3.0, Right := 3.0}
                    Widget := PurchaseButtonColor
                overlay_slot:
                    VerticalAlignment := vertical_alignment.Center
                    HorizontalAlignment := horizontal_alignment.Center
                    Padding := margin{Left := 10.0, Right := 10.0}
                    Widget := PurchaseTextStack
                overlay_slot:
                    VerticalAlignment := vertical_alignment.Center
                    HorizontalAlignment := horizontal_alignment.Fill
                    Padding := margin{Top := 3.0, Bottom := 3.0, Left := 3.0, Right := 3.0}
                    Widget := PurchaseButton
                

        PurchaseStackBox := stack_box:
            Orientation := orientation.Vertical
            Slots :=
                array:
                    stack_box_slot:
                        Widget := color_block:
                            DefaultColor := NamedColors.DarkBlue
                            DefaultDesiredSize := vector2{X := 410.0, Y := BorderSize}
                        HorizontalAlignment := horizontal_alignment.Center
                        VerticalAlignment := vertical_alignment.Top
                    stack_box_slot:
                        Widget := PurchaseOverlay
                        Padding := margin{Top := 10.0}
                        HorizontalAlignment := horizontal_alignment.Fill
                        VerticalAlignment := vertical_alignment.Fill


        UpgradeTextStack := stack_box:
            Orientation := orientation.Horizontal
            Slots :=
                array:
                    stack_box_slot:
                        VerticalAlignment := vertical_alignment.Center
                        HorizontalAlignment := horizontal_alignment.Center
                        Padding := margin{Left := 10.0, Right := 10.0}
                        Widget := BuyText2
                    stack_box_slot:
                        VerticalAlignment := vertical_alignment.Center
                        HorizontalAlignment := horizontal_alignment.Center
                        Padding := margin{Left := 10.0, Right := 5.0}
                        Widget := texture_block:
                            DefaultImage := VerseExposed.T_Stone
                    stack_box_slot:
                        VerticalAlignment := vertical_alignment.Center
                        HorizontalAlignment := horizontal_alignment.Center
                        Padding := margin{Left := 5.0, Right := 10.0}
                        Widget := UpgradeText
        
        UpgradeOverlay := overlay:
            Slots := array:
                overlay_slot:
                    VerticalAlignment := vertical_alignment.Center
                    HorizontalAlignment := horizontal_alignment.Center
                    Padding := margin{Top := 3.0, Bottom := 3.0, Left := 3.0, Right := 3.0}
                    Widget := UpgradeButtonColor
                overlay_slot:
                    VerticalAlignment := vertical_alignment.Center
                    HorizontalAlignment := horizontal_alignment.Center
                    Padding := margin{Left := 10.0, Right := 10.0}
                    Widget := CantUpgradeText
                overlay_slot:
                    VerticalAlignment := vertical_alignment.Center
                    HorizontalAlignment := horizontal_alignment.Center
                    Padding := margin{Left := 10.0, Right := 10.0}
                    Widget := UpgradeTextStack
                overlay_slot:
                    VerticalAlignment := vertical_alignment.Center
                    HorizontalAlignment := horizontal_alignment.Fill
                    Padding := margin{Top := 3.0, Bottom := 3.0, Left := 3.0, Right := 3.0}
                    Widget := UpgradeButton

        UpgradeStackBox := stack_box:
            Orientation := orientation.Vertical
            Slots :=
                array:
                    stack_box_slot:
                        Widget := color_block:
                            DefaultColor := NamedColors.DarkBlue
                            DefaultDesiredSize := vector2{X := 410.0, Y := BorderSize}
                        HorizontalAlignment := horizontal_alignment.Center
                        VerticalAlignment := vertical_alignment.Top
                    stack_box_slot:
                        Widget := UpgradeOverlay
                        Padding := margin{Top := 10.0}
                        HorizontalAlignment := horizontal_alignment.Fill
                        VerticalAlignment := vertical_alignment.Fill

        SlotStackBox := stack_box:
            Orientation := orientation.Horizontal
            Slots :=
                array:
                    stack_box_slot:
                        Widget := color_block:
                            DefaultColor := NamedColors.DarkBlue
                            DefaultDesiredSize := vector2{X := BorderSize, Y := 90.0}
                        HorizontalAlignment := horizontal_alignment.Left
                        VerticalAlignment := vertical_alignment.Center
                    stack_box_slot:
                        Widget := NameStackBox
                        HorizontalAlignment := horizontal_alignment.Left
                        VerticalAlignment := vertical_alignment.Top
                    stack_box_slot:
                        Widget := color_block:
                            DefaultColor := NamedColors.DarkBlue
                            DefaultDesiredSize := vector2{X := BorderSize, Y := BorderSize}
                        HorizontalAlignment := horizontal_alignment.Center
                        VerticalAlignment := vertical_alignment.Fill
                    stack_box_slot:
                        Widget := OwnedStackBox
                        HorizontalAlignment := horizontal_alignment.Left
                        VerticalAlignment := vertical_alignment.Top
                    stack_box_slot:
                        Widget := color_block:
                            DefaultColor := NamedColors.DarkBlue
                            DefaultDesiredSize := vector2{X := BorderSize, Y := BorderSize}
                        HorizontalAlignment := horizontal_alignment.Center
                        VerticalAlignment := vertical_alignment.Fill
                    stack_box_slot:
                        Widget := HitPowerStackBox
                        HorizontalAlignment := horizontal_alignment.Left
                        VerticalAlignment := vertical_alignment.Top
                    stack_box_slot:
                        Widget := color_block:
                            DefaultColor := NamedColors.DarkBlue
                            DefaultDesiredSize := vector2{X := BorderSize, Y := BorderSize}
                        HorizontalAlignment := horizontal_alignment.Center
                        VerticalAlignment := vertical_alignment.Fill
                    stack_box_slot:
                        Widget := PurchaseStackBox
                        HorizontalAlignment := horizontal_alignment.Left
                        VerticalAlignment := vertical_alignment.Top
                    stack_box_slot:
                        Widget := color_block:
                            DefaultColor := NamedColors.DarkBlue
                            DefaultDesiredSize := vector2{X := BorderSize, Y := BorderSize}
                        HorizontalAlignment := horizontal_alignment.Center
                        VerticalAlignment := vertical_alignment.Fill
                    stack_box_slot:
                        Widget := UpgradeStackBox
                        HorizontalAlignment := horizontal_alignment.Left
                        VerticalAlignment := vertical_alignment.Top

        OverallOverlay := overlay:
            Slots := array:
                overlay_slot:
                    VerticalAlignment := vertical_alignment.Fill
                    HorizontalAlignment := horizontal_alignment.Fill
                    Widget := color_block:
                        DefaultColor := NamedColors.White
                        DefaultOpacity := 0.4
                overlay_slot:
                    VerticalAlignment := vertical_alignment.Center
                    HorizontalAlignment := horizontal_alignment.Center
                    Widget := SlotStackBox
                overlay_slot:
                    Widget := color_block:
                        DefaultColor := NamedColors.DarkBlue
                        DefaultDesiredSize := vector2{X := BorderSize, Y := BorderSize}
                    HorizontalAlignment := horizontal_alignment.Fill
                    VerticalAlignment := vertical_alignment.Bottom
                overlay_slot:
                    Widget := color_block:
                        DefaultColor := NamedColors.DarkBlue
                        DefaultDesiredSize := vector2{X := BorderSize, Y := BorderSize}
                    HorizontalAlignment := horizontal_alignment.Right
                    VerticalAlignment := vertical_alignment.Fill

        return UpgradeSlotData:
            Widget := OverallOverlay
            Name := InUpgradeSlotConstructData.Name
            OwnedText := OwnedText
            HitPowerText := HitPowerText
            NextHitPowerText := NextHitPowerText
            PurchaseText := PurchaseText
            UpgradeText := UpgradeText
            UpgradeTextWidget := UpgradeTextStack
            CantUpgradeText := CantUpgradeText
            UpgradeButtonColor := UpgradeButtonColor
            PurchaseButtonColor := PurchaseButtonColor
            ShouldShow := false

            PurchaseButton := PurchaseButton
            UpgradeButton := UpgradeButton

            PurchaseTrigger := InUpgradeSlotConstructData.PurchaseTrigger
            UpgradeTrigger := InUpgradeSlotConstructData.UpgradeTrigger

    #Make a page for upgrades
    MakeUpgradePage(PageIndex : int, InUpgradeSlots : []UpgradeSlotData):UpgradePageData=
        #Print("Making Upgrade Page")
        var EquipmentStackBoxes : []stack_box_slot = array{}
        var UpgradeWidgets : []widget = array{}

        # Take the upgrade slots and make an overlay out of them
        for(Index -> Slot : InUpgradeSlots):
            CurrentStackBox := stack_box_slot:
                Widget := Slot.Widget
                HorizontalAlignment := horizontal_alignment.Center
                VerticalAlignment := vertical_alignment.Top
                Padding := margin{Top := 5.0, Bottom := 5.0}

            set EquipmentStackBoxes += array{CurrentStackBox}
            set UpgradeWidgets += array{Slot.Widget}

        EquipmentStackBox := stack_box:
            Orientation := orientation.Vertical
            Slots := EquipmentStackBoxes

        OutOverlaySlot := overlay_slot:
            VerticalAlignment := vertical_alignment.Fill
            HorizontalAlignment := horizontal_alignment.Fill
            Widget := EquipmentStackBox

        
        # Return the overlay
        return UpgradePageData:
            UpgradeWidgets := InUpgradeSlots
            OverlaySlot := OutOverlaySlot
            Index := PageIndex


    # Make the entire canvas to display to players
    MakeUpgradeCanvas(InUpgradeSlotConstructDatas : []UpgradeSlotConstructData, InPlayerStoneAmount : string, InX1Trigger : trigger_device, InX10Trigger : trigger_device, InX100Trigger : trigger_device):UpgradeCanvasData=
        var EquipmentStackBoxes : []stack_box_slot = array{}
        var EquipmentSlots : []UpgradeSlotData = array{}
        var EquipmentPages : []UpgradePageData = array{}
        var PageOverlaySlots : []overlay_slot = array{}
        
        var CurrentPageSlots : []UpgradeSlotData = array{}
        var UpgradeCount : int = 0
        var PageIndex : int = 0

        for(Index -> ConstructData : InUpgradeSlotConstructDatas):
            set UpgradeCount += 1

            #Print("INDEX : {Index} | UPGRADE COUNT : {UpgradeCount} | TOTAL DATA : {InUpgradeSlotConstructDatas.Length}")
            EquipmentSlot := MakeUpgradeSlot(ConstructData)

            set EquipmentSlots += array{EquipmentSlot}
            set CurrentPageSlots += array{EquipmentSlot}
            if(UpgradeCount = 5):
                NewPage := MakeUpgradePage(PageIndex, CurrentPageSlots)
                if(PageIndex <> 0):
                    NewPage.OverlaySlot.Widget.SetVisibility(widget_visibility.Collapsed)
                    for(UpgradeSlot : NewPage.UpgradeWidgets):
                        UpgradeSlot.Widget.SetVisibility(widget_visibility.Collapsed)

                set EquipmentPages += array{NewPage}
                set PageOverlaySlots += array{NewPage.OverlaySlot}
                set CurrentPageSlots = array{}
                set UpgradeCount = 0
                set PageIndex += 1
            
        ExitButton := button_loud:
            DefaultText := ToMessage("EXIT")

        NextPageButton := button_quiet:
            DefaultText := ToMessage(">")
        PrevPageButton := button_quiet:
            DefaultText := ToMessage("<")

        X1Button := button_quiet:
            DefaultText := ToMessage("x1")
        X10Button := button_quiet:
            DefaultText := ToMessage("x10")
        X100Button := button_quiet:
            DefaultText := ToMessage("x100")

        BorderSize := 1.0

        TitleText := text_block:
            DefaultText := ToMessage("Equipment")
            DefaultJustification := text_justification.Center
            DefaultShadowColor := NamedColors.Black
            DefaultShadowOffset := option{vector2{X:=2.0,Y:=2.0}}
        TitleText.SetTextColor(NamedColors.White)
        TitleText.SetShadowOpacity(0.8)
        TitleTexture := texture_block:
            DefaultImage := VerseExposed.T_EquipmentTitleText
            DefaultDesiredSize := vector2{X:=512.0,Y:=128.0}

        NameText := text_block:
            DefaultText := ToMessage("Name")
            DefaultJustification := text_justification.Center
            DefaultShadowColor := NamedColors.Black
            DefaultShadowOffset := option{vector2{X:=2.0,Y:=2.0}}
        NameText.SetTextColor(NamedColors.White)
        NameText.SetShadowOpacity(0.8)
        OwnedText := text_block:
            DefaultText := ToMessage("Owned")
            DefaultJustification := text_justification.Center
            DefaultShadowColor := NamedColors.Black
            DefaultShadowOffset := option{vector2{X:=2.0,Y:=2.0}}
        OwnedText.SetTextColor(NamedColors.White)
        OwnedText.SetShadowOpacity(0.8)
        HitPowerText := text_block:
            DefaultText := ToMessage("Hit Power")
            DefaultJustification := text_justification.Center
            DefaultShadowColor := NamedColors.Black
            DefaultShadowOffset := option{vector2{X:=2.0,Y:=2.0}}
        HitPowerText.SetTextColor(NamedColors.White)
        HitPowerText.SetShadowOpacity(0.8)
        PurchaseText := text_block:
            DefaultText := ToMessage("Purchase")
            DefaultJustification := text_justification.Center
            DefaultShadowColor := NamedColors.Black
            DefaultShadowOffset := option{vector2{X:=2.0,Y:=2.0}}
        PurchaseText.SetTextColor(NamedColors.White)
        PurchaseText.SetShadowOpacity(0.8)
        UpgradeText := text_block:
            DefaultText := ToMessage("Power Multiplier")
            DefaultJustification := text_justification.Center
            DefaultShadowColor := NamedColors.Black
            DefaultShadowOffset := option{vector2{X:=2.0,Y:=2.0}}
        UpgradeText.SetTextColor(NamedColors.White)
        UpgradeText.SetShadowOpacity(0.8)
        PlayerStoneText := text_block:
            DefaultText := ToMessage("{InPlayerStoneAmount}")
            DefaultJustification := text_justification.Center
            DefaultShadowColor := NamedColors.Black
            DefaultShadowOffset := option{vector2{X:=2.0,Y:=2.0}}
        PlayerStoneText.SetTextColor(NamedColors.White)
        PlayerStoneText.SetShadowOpacity(0.8)
        PlayerStrengthText := text_block:
            DefaultText := ToMessage("{InPlayerStoneAmount}")
            DefaultJustification := text_justification.Center
            DefaultShadowColor := NamedColors.Black
            DefaultShadowOffset := option{vector2{X:=2.0,Y:=2.0}}
        PlayerStrengthText.SetTextColor(NamedColors.White)
        PlayerStrengthText.SetShadowOpacity(0.8)
        BuyAmountText := text_block:
            DefaultText := ToMessage("Buy x1")
            DefaultJustification := text_justification.Center
            DefaultShadowColor := NamedColors.Black
            DefaultShadowOffset := option{vector2{X:=2.0,Y:=2.0}}
        BuyAmountText.SetTextColor(NamedColors.White)
        BuyAmountText.SetShadowOpacity(0.8)
        PageNumberText := text_block:
            DefaultText := ToMessage("1/4")
            DefaultJustification := text_justification.Center
            DefaultShadowColor := NamedColors.Black
            DefaultShadowOffset := option{vector2{X:=2.0,Y:=2.0}}
        PageNumberText.SetTextColor(NamedColors.White)
        PageNumberText.SetShadowOpacity(0.8)

        NameOverlay := overlay:
            Slots := array:
                overlay_slot:
                    VerticalAlignment := vertical_alignment.Fill
                    HorizontalAlignment := horizontal_alignment.Fill
                    Widget := color_block:
                        DefaultColor := NamedColors.White
                        DefaultOpacity := 0.4
                overlay_slot:
                    VerticalAlignment := vertical_alignment.Center
                    HorizontalAlignment := horizontal_alignment.Left
                    Padding := margin{Top := 10.0, Bottom := 10.0, Left := 30.0, Right := 30.0}
                    Widget := NameText

        NameStackBox := stack_box:
            Orientation := orientation.Vertical
            Slots :=
                array:
                    stack_box_slot:
                        Widget := color_block:
                            DefaultColor := NamedColors.DarkBlue
                            DefaultDesiredSize := vector2{X := 400.0, Y := BorderSize}
                    stack_box_slot:
                        Widget := NameOverlay
                        HorizontalAlignment := horizontal_alignment.Fill
                        VerticalAlignment := vertical_alignment.Fill
            
        OwnedOverlay := overlay:
            Slots := array:
                overlay_slot:
                    VerticalAlignment := vertical_alignment.Fill
                    HorizontalAlignment := horizontal_alignment.Fill
                    Widget := color_block:
                        DefaultColor := NamedColors.White
                        DefaultOpacity := 0.4
                overlay_slot:
                    VerticalAlignment := vertical_alignment.Center
                    HorizontalAlignment := horizontal_alignment.Center
                    Padding := margin{Top := 10.0, Bottom := 10.0, Left := 30.0, Right := 30.0}
                    Widget := OwnedText
        
        OwnedStackBox := stack_box:
            Orientation := orientation.Vertical
            Slots :=
                array:
                    stack_box_slot:
                        Widget := color_block:
                            DefaultColor := NamedColors.DarkBlue
                            DefaultDesiredSize := vector2{X := 175.0, Y := BorderSize}
                    stack_box_slot:
                        Widget := OwnedOverlay
                        HorizontalAlignment := horizontal_alignment.Fill
                        VerticalAlignment := vertical_alignment.Fill

        HitPowerOverlay := overlay:
            Slots := array:
                overlay_slot:
                    VerticalAlignment := vertical_alignment.Fill
                    HorizontalAlignment := horizontal_alignment.Fill
                    Widget := color_block:
                        DefaultColor := NamedColors.White
                        DefaultOpacity := 0.4
                overlay_slot:
                    VerticalAlignment := vertical_alignment.Center
                    HorizontalAlignment := horizontal_alignment.Left
                    Padding := margin{Top := 10.0, Bottom := 10.0, Left := 30.0, Right := 30.0}
                    Widget := HitPowerText

        HitPowerStackBox := stack_box:
            Orientation := orientation.Vertical
            Slots :=
                array:
                    stack_box_slot:
                        Widget := color_block:
                            DefaultColor := NamedColors.DarkBlue
                            DefaultDesiredSize := vector2{X := 375.0, Y := BorderSize}
                    stack_box_slot:
                        Widget := HitPowerOverlay
                        HorizontalAlignment := horizontal_alignment.Fill
                        VerticalAlignment := vertical_alignment.Fill
                    
        PurchaseOverlay := overlay:
            Slots := array:
                overlay_slot:
                    VerticalAlignment := vertical_alignment.Fill
                    HorizontalAlignment := horizontal_alignment.Fill
                    Widget := color_block:
                        DefaultColor := NamedColors.White
                        DefaultOpacity := 0.4
                overlay_slot:
                    VerticalAlignment := vertical_alignment.Center
                    HorizontalAlignment := horizontal_alignment.Fill
                    Padding := margin{Top := 10.0, Bottom := 10.0, Left := 30.0, Right := 30.0}
                    Widget := PurchaseText

        PurchaseStackBox := stack_box:
            Orientation := orientation.Vertical
            Slots :=
                array:
                    stack_box_slot:
                        Widget := color_block:
                            DefaultColor := NamedColors.DarkBlue
                            DefaultDesiredSize := vector2{X := 410.0, Y := BorderSize}
                    stack_box_slot:
                        Widget := PurchaseOverlay
                        HorizontalAlignment := horizontal_alignment.Fill
                        VerticalAlignment := vertical_alignment.Fill
                    
        UpgradeOverlay := overlay:
            Slots := array:
                overlay_slot:
                    VerticalAlignment := vertical_alignment.Fill
                    HorizontalAlignment := horizontal_alignment.Fill
                    Widget := color_block:
                        DefaultColor := NamedColors.White
                        DefaultOpacity := 0.4
                overlay_slot:
                    VerticalAlignment := vertical_alignment.Center
                    HorizontalAlignment := horizontal_alignment.Fill
                    Padding := margin{Top := 10.0, Bottom := 10.0, Left := 30.0, Right := 30.0}
                    Widget := UpgradeText

        UpgradeStackBox := stack_box:
            Orientation := orientation.Vertical
            Slots :=
                array:
                    stack_box_slot:
                        Widget := color_block:
                            DefaultColor := NamedColors.DarkBlue
                            DefaultDesiredSize := vector2{X := 410.0, Y := BorderSize}
                    stack_box_slot:
                        Widget := UpgradeOverlay
                        HorizontalAlignment := horizontal_alignment.Fill
                        VerticalAlignment := vertical_alignment.Fill

        LegendStackBox := stack_box:
            Orientation := orientation.Horizontal
            Slots :=
                array:
                    stack_box_slot:
                        Widget := NameStackBox
                        HorizontalAlignment := horizontal_alignment.Left
                        VerticalAlignment := vertical_alignment.Center
                    stack_box_slot:
                        Widget := color_block:
                            DefaultColor := NamedColors.DarkBlue
                            DefaultDesiredSize := vector2{X := BorderSize, Y := BorderSize}
                        HorizontalAlignment := horizontal_alignment.Center
                        VerticalAlignment := vertical_alignment.Fill
                    stack_box_slot:
                        Widget := OwnedStackBox
                        HorizontalAlignment := horizontal_alignment.Left
                        VerticalAlignment := vertical_alignment.Center
                    stack_box_slot:
                        Widget := color_block:
                            DefaultColor := NamedColors.DarkBlue
                            DefaultDesiredSize := vector2{X := BorderSize, Y := BorderSize}
                        HorizontalAlignment := horizontal_alignment.Center
                        VerticalAlignment := vertical_alignment.Fill
                    stack_box_slot:
                        Widget := HitPowerStackBox
                        HorizontalAlignment := horizontal_alignment.Left
                        VerticalAlignment := vertical_alignment.Center
                    stack_box_slot:
                        Widget := color_block:
                            DefaultColor := NamedColors.DarkBlue
                            DefaultDesiredSize := vector2{X := BorderSize, Y := BorderSize}
                        HorizontalAlignment := horizontal_alignment.Center
                        VerticalAlignment := vertical_alignment.Fill
                    stack_box_slot:
                        Widget := PurchaseStackBox
                        HorizontalAlignment := horizontal_alignment.Left
                        VerticalAlignment := vertical_alignment.Center
                    stack_box_slot:
                        Widget := color_block:
                            DefaultColor := NamedColors.DarkBlue
                            DefaultDesiredSize := vector2{X := BorderSize, Y := BorderSize}
                        HorizontalAlignment := horizontal_alignment.Center
                        VerticalAlignment := vertical_alignment.Fill
                    stack_box_slot:
                        Widget := UpgradeStackBox
                        HorizontalAlignment := horizontal_alignment.Left
                        VerticalAlignment := vertical_alignment.Center

        LegendOverlay := overlay:
            Slots := array:
                overlay_slot:
                    VerticalAlignment := vertical_alignment.Fill
                    HorizontalAlignment := horizontal_alignment.Fill
                    Widget := LegendStackBox
                overlay_slot:
                    Widget := color_block:
                        DefaultColor := NamedColors.DarkBlue
                        DefaultDesiredSize := vector2{X := BorderSize, Y := BorderSize}
                    HorizontalAlignment := horizontal_alignment.Fill
                    VerticalAlignment := vertical_alignment.Bottom
                overlay_slot:
                    Widget := color_block:
                        DefaultColor := NamedColors.DarkBlue
                        DefaultDesiredSize := vector2{X := BorderSize, Y := BorderSize}
                    HorizontalAlignment := horizontal_alignment.Right
                    VerticalAlignment := vertical_alignment.Fill
                overlay_slot:
                    Widget := color_block:
                        DefaultColor := NamedColors.DarkBlue
                        DefaultDesiredSize := vector2{X := BorderSize, Y := BorderSize}
                    HorizontalAlignment := horizontal_alignment.Left
                    VerticalAlignment := vertical_alignment.Fill

        StoneIconOverlay := overlay:
            Slots := array:
                overlay_slot:
                    VerticalAlignment := vertical_alignment.Center
                    HorizontalAlignment := horizontal_alignment.Center
                    Padding := margin{Top := -8.0, Bottom := -8.0, Left := -6.0, Right := -6.0}
                    Widget := color_block:
                        DefaultColor := NamedColors.White
                        DefaultOpacity := 0.4
                        DefaultDesiredSize := vector2{X := 52.0, Y := 52.0}
                overlay_slot:
                    VerticalAlignment := vertical_alignment.Center
                    HorizontalAlignment := horizontal_alignment.Center
                    Widget := texture_block:
                        DefaultImage := VerseExposed.T_Stone

        PlayerStoneOverlay := overlay:
            Slots := array:
                overlay_slot:
                    VerticalAlignment := vertical_alignment.Center
                    HorizontalAlignment := horizontal_alignment.Left
                    Widget := color_block:
                        DefaultColor := NamedColors.White
                        DefaultOpacity := 0.4
                        DefaultDesiredSize := vector2{X:=225.0,Y:=60.0}
                overlay_slot:
                    VerticalAlignment := vertical_alignment.Center
                    HorizontalAlignment := horizontal_alignment.Left
                    Padding := margin{Top := 10.0, Bottom := 10.0, Left := 10.0, Right := 10.0}
                    Widget := stack_box:
                        Orientation := orientation.Horizontal
                        Slots := 
                            array:
                                stack_box_slot:
                                    Widget := StoneIconOverlay
                                    HorizontalAlignment := horizontal_alignment.Left
                                    VerticalAlignment := vertical_alignment.Center
                                stack_box_slot:
                                    Widget := PlayerStoneText
                                    Padding := margin{Left := 15.0}
                                    VerticalAlignment := vertical_alignment.Center
                                    HorizontalAlignment := horizontal_alignment.Left
                overlay_slot:
                    Widget := color_block:
                        DefaultColor := NamedColors.DarkBlue
                        DefaultDesiredSize := vector2{X := BorderSize, Y := BorderSize}
                    HorizontalAlignment := horizontal_alignment.Fill
                    VerticalAlignment := vertical_alignment.Top
                overlay_slot:
                    Widget := color_block:
                        DefaultColor := NamedColors.DarkBlue
                        DefaultDesiredSize := vector2{X := BorderSize, Y := BorderSize}
                    HorizontalAlignment := horizontal_alignment.Fill
                    VerticalAlignment := vertical_alignment.Bottom
                overlay_slot:
                    Widget := color_block:
                        DefaultColor := NamedColors.DarkBlue
                        DefaultDesiredSize := vector2{X := BorderSize, Y := BorderSize}
                    HorizontalAlignment := horizontal_alignment.Right
                    VerticalAlignment := vertical_alignment.Fill
                overlay_slot:
                    Widget := color_block:
                        DefaultColor := NamedColors.DarkBlue
                        DefaultDesiredSize := vector2{X := BorderSize, Y := BorderSize}
                    HorizontalAlignment := horizontal_alignment.Left
                    VerticalAlignment := vertical_alignment.Fill

        StrengthIconOverlay := overlay:
            Slots := array:
                overlay_slot:
                    VerticalAlignment := vertical_alignment.Center
                    HorizontalAlignment := horizontal_alignment.Center
                    Padding := margin{Top := -8.0, Bottom := -8.0, Left := -6.0, Right := -6.0}
                    Widget := color_block:
                        DefaultColor := NamedColors.White
                        DefaultOpacity := 0.4
                        DefaultDesiredSize := vector2{X := 52.0, Y := 52.0}
                overlay_slot:
                    VerticalAlignment := vertical_alignment.Center
                    HorizontalAlignment := horizontal_alignment.Center
                    Widget := texture_block:
                        DefaultImage := VerseExposed.T_Achievements_Strength

        PlayerStrengthOverlay := overlay:
            Slots := array:
                overlay_slot:
                    VerticalAlignment := vertical_alignment.Center
                    HorizontalAlignment := horizontal_alignment.Left
                    Widget := color_block:
                        DefaultColor := NamedColors.White
                        DefaultOpacity := 0.4
                        DefaultDesiredSize := vector2{X:=225.0,Y:=60.0}
                overlay_slot:
                    VerticalAlignment := vertical_alignment.Center
                    HorizontalAlignment := horizontal_alignment.Left
                    Padding := margin{Top := 10.0, Bottom := 10.0, Left := 10.0, Right := 10.0}
                    Widget := stack_box:
                        Orientation := orientation.Horizontal
                        Slots := 
                            array:
                                stack_box_slot:
                                    Widget := StrengthIconOverlay
                                    HorizontalAlignment := horizontal_alignment.Left
                                    VerticalAlignment := vertical_alignment.Center
                                stack_box_slot:
                                    Widget := PlayerStrengthText
                                    Padding := margin{Left := 15.0}
                                    VerticalAlignment := vertical_alignment.Center
                                    HorizontalAlignment := horizontal_alignment.Left
                overlay_slot:
                    Widget := color_block:
                        DefaultColor := NamedColors.DarkBlue
                        DefaultDesiredSize := vector2{X := BorderSize, Y := BorderSize}
                    HorizontalAlignment := horizontal_alignment.Fill
                    VerticalAlignment := vertical_alignment.Top
                overlay_slot:
                    Widget := color_block:
                        DefaultColor := NamedColors.DarkBlue
                        DefaultDesiredSize := vector2{X := BorderSize, Y := BorderSize}
                    HorizontalAlignment := horizontal_alignment.Fill
                    VerticalAlignment := vertical_alignment.Bottom
                overlay_slot:
                    Widget := color_block:
                        DefaultColor := NamedColors.DarkBlue
                        DefaultDesiredSize := vector2{X := BorderSize, Y := BorderSize}
                    HorizontalAlignment := horizontal_alignment.Right
                    VerticalAlignment := vertical_alignment.Fill
                overlay_slot:
                    Widget := color_block:
                        DefaultColor := NamedColors.DarkBlue
                        DefaultDesiredSize := vector2{X := BorderSize, Y := BorderSize}
                    HorizontalAlignment := horizontal_alignment.Left
                    VerticalAlignment := vertical_alignment.Fill

        PurchaseAmountStack := stack_box:
            Orientation := orientation.Horizontal
            Slots :=
                array:
                    stack_box_slot:
                        HorizontalAlignment := horizontal_alignment.Center
                        VerticalAlignment := vertical_alignment.Center
                        Padding := margin{ Left := 10.0, Right  := 10.0}
                        Widget := BuyAmountText
                    stack_box_slot:
                        Widget := color_block:
                            DefaultColor := NamedColors.DarkBlue
                            DefaultDesiredSize := vector2{X := BorderSize, Y := BorderSize}
                        HorizontalAlignment := horizontal_alignment.Right
                        VerticalAlignment := vertical_alignment.Fill
                    stack_box_slot:
                        HorizontalAlignment := horizontal_alignment.Center
                        VerticalAlignment := vertical_alignment.Center
                        Padding := margin{ Left := 10.0, Right  := 10.0}
                        Widget := X1Button
                    stack_box_slot:
                        HorizontalAlignment := horizontal_alignment.Center
                        VerticalAlignment := vertical_alignment.Center
                        Padding := margin{ Left := 10.0, Right  := 10.0}
                        Widget := X10Button
                    stack_box_slot:
                        HorizontalAlignment := horizontal_alignment.Center
                        VerticalAlignment := vertical_alignment.Center
                        Padding := margin{ Left := 10.0, Right  := 10.0}
                        Widget := X100Button

        PurchaseAmountOverlay := overlay:
            Slots := array:
                overlay_slot:
                    VerticalAlignment := vertical_alignment.Center
                    HorizontalAlignment := horizontal_alignment.Fill
                    Widget := color_block:
                        DefaultColor := NamedColors.White
                        DefaultOpacity := 0.4
                        DefaultDesiredSize := vector2{X:=225.0,Y:=60.0}
                overlay_slot:
                    VerticalAlignment := vertical_alignment.Fill
                    HorizontalAlignment := horizontal_alignment.Fill
                    Widget := PurchaseAmountStack
                overlay_slot:
                    Widget := color_block:
                        DefaultColor := NamedColors.DarkBlue
                        DefaultDesiredSize := vector2{X := BorderSize, Y := BorderSize}
                    HorizontalAlignment := horizontal_alignment.Fill
                    VerticalAlignment := vertical_alignment.Top
                overlay_slot:
                    Widget := color_block:
                        DefaultColor := NamedColors.DarkBlue
                        DefaultDesiredSize := vector2{X := 525.0, Y := BorderSize}
                    HorizontalAlignment := horizontal_alignment.Center
                    VerticalAlignment := vertical_alignment.Bottom
                overlay_slot:
                    Widget := color_block:
                        DefaultColor := NamedColors.DarkBlue
                        DefaultDesiredSize := vector2{X := BorderSize, Y := BorderSize}
                    HorizontalAlignment := horizontal_alignment.Right
                    VerticalAlignment := vertical_alignment.Fill
                overlay_slot:
                    Widget := color_block:
                        DefaultColor := NamedColors.DarkBlue
                        DefaultDesiredSize := vector2{X := BorderSize, Y := BorderSize}
                    HorizontalAlignment := horizontal_alignment.Left
                    VerticalAlignment := vertical_alignment.Fill

        UnderHeaderStackBox := stack_box:
            Orientation := orientation.Horizontal
            Slots := array:
                stack_box_slot:
                    VerticalAlignment := vertical_alignment.Center
                    HorizontalAlignment := horizontal_alignment.Center
                    Padding := margin{Left := 10.0, Right := 10.0}
                    Widget := PlayerStoneOverlay
                stack_box_slot:
                    VerticalAlignment := vertical_alignment.Center
                    HorizontalAlignment := horizontal_alignment.Center
                    Padding := margin{Left := 10.0, Right := 10.0}
                    Widget := PlayerStrengthOverlay
                stack_box_slot:
                    VerticalAlignment := vertical_alignment.Center
                    HorizontalAlignment := horizontal_alignment.Center
                    Padding := margin{Left := 10.0, Right := 10.0}
                    Widget := PurchaseAmountOverlay
        
        UnderHeaderOverlay := overlay:
                Slots := array:
                    overlay_slot:
                        VerticalAlignment := vertical_alignment.Fill
                        HorizontalAlignment := horizontal_alignment.Fill
                        Widget := UnderHeaderStackBox

        EquipmentOverlay := overlay:
            Slots := PageOverlaySlots

        StackBox := stack_box:
            Orientation := orientation.Vertical
            Slots :=
                array:
                    stack_box_slot:
                        Widget := TitleTexture
                        HorizontalAlignment := horizontal_alignment.Center
                        VerticalAlignment := vertical_alignment.Top
                    stack_box_slot:
                        Widget := UnderHeaderOverlay
                        HorizontalAlignment := horizontal_alignment.Center
                        VerticalAlignment := vertical_alignment.Top
                        Padding := margin{Top := 10.0, Bottom := 10.0}
                    stack_box_slot:
                        Widget := LegendOverlay
                        HorizontalAlignment := horizontal_alignment.Center
                        VerticalAlignment := vertical_alignment.Top
                        Padding := margin{Top := 10.0, Bottom := 10.0}
                    stack_box_slot:
                        Widget := EquipmentOverlay
                        HorizontalAlignment := horizontal_alignment.Center
                        VerticalAlignment := vertical_alignment.Top
                        Padding := margin{Top := 5.0, Bottom := 5.0}

        BackgroundOverlay := overlay:
            Slots := array:
                overlay_slot:
                    VerticalAlignment := vertical_alignment.Fill
                    HorizontalAlignment := horizontal_alignment.Fill
                    Widget := texture_block:
                        DefaultImage := VerseExposed.T_UIBackground

        OutputCanvas := canvas:
            Slots := array:
                canvas_slot:
                    Anchors := anchors{Minimum := vector2{X:=0.0, Y:=0.0}, Maximum := vector2{X:=1.0, Y:=1.0}}
                    Alignment := CenteredAlignment
                    Widget := BackgroundOverlay
                    ZOrder := 0
                canvas_slot:
                    Anchors := anchors{Minimum := vector2{X:=0.06255, Y:=0.0475}, Maximum := vector2{X:=0.2, Y:=0.175}}
                    Alignment := CenteredAlignment
                    Widget := PlayerStoneOverlay
                    SizeToContent := true
                    ZOrder := 1
                canvas_slot:
                    Anchors := anchors{Minimum := vector2{X:=0.21, Y:=0.0475}, Maximum := vector2{X:=0.37255, Y:=0.175}}
                    Alignment := CenteredAlignment
                    Widget := PlayerStrengthOverlay
                    SizeToContent := true
                    ZOrder := 1
                canvas_slot:
                    Anchors := anchors{Minimum := vector2{X:=0.65, Y:=0.05}, Maximum := vector2{X:=0.9, Y:=0.1}}
                    Alignment := CenteredAlignment
                    Widget := PurchaseAmountOverlay
                    SizeToContent := true
                    ZOrder := 1
                canvas_slot:
                    Anchors := anchors{Minimum := vector2{X:=0.02, Y:=0.05}, Maximum := vector2{X:=0.98, Y:=0.95}}
                    Alignment := CenteredAlignment
                    Widget := StackBox
                    SizeToContent := true
                    ZOrder := 1
                canvas_slot:
                    Anchors := anchors{Minimum := vector2{X:=0.52, Y:=0.85}, Maximum := vector2{X:=0.9, Y:=0.85}}
                    Alignment := CenteredAlignment
                    Widget := NextPageButton
                    SizeToContent := true
                    ZOrder := 1
                canvas_slot:
                    Anchors := anchors{Minimum := vector2{X:=0.5, Y:=0.85}, Maximum := vector2{X:=0.5, Y:=0.85}}
                    Alignment := CenteredAlignment
                    Widget := PageNumberText
                    SizeToContent := true
                    ZOrder := 1
                canvas_slot:
                    Anchors := anchors{Minimum := vector2{X:=0.1, Y:=0.85}, Maximum := vector2{X:=0.48, Y:=0.85}}
                    Alignment := CenteredAlignment
                    Widget := PrevPageButton
                    SizeToContent := true
                    ZOrder := 1
                canvas_slot:
                    Anchors := anchors{Minimum := vector2{X:=0.5, Y:=0.95}, Maximum := vector2{X:=0.5, Y:=0.95}}
                    Alignment := CenteredAlignment
                    Widget := ExitButton
                    SizeToContent := true
                    ZOrder := 1
        
        return UpgradeCanvasData:
            Canvas := OutputCanvas
            CurrentPageIndex := 0
            UpgradeSlots := EquipmentSlots
            UpgradePages := EquipmentPages
            PlayerStoneAmount := PlayerStoneText
            PlayerStrengthAmount := PlayerStrengthText
            PurchaseAmount := BuyAmountText
            PageNumberText := PageNumberText
            X1Trigger := InX1Trigger
            X10Trigger := InX10Trigger
            X100Trigger := InX100Trigger
            X1Button := X1Button
            X10Button := X10Button
            X100Button := X100Button
            ExitButton := ExitButton
            NextPageButton := NextPageButton
            PrevPageButton := PrevPageButton
