# fill this file with your verse-script
using { /Fortnite.com/Devices }
using { /Verse.org/Simulation }
using { /UnrealEngine.com/Temporary/Diagnostics }
using { /UnrealEngine.com/Temporary/UI }
using { /Fortnite.com/UI }
using { /UnrealEngine.com/Temporary/SpatialMath}
using { /Verse.org/Colors }
using { /Verse.org/Assets }

MakeVendorData<constructor>(InResourceImage : texture, InResourceName : string, InTitleTexture : texture, InCharacterTexture : texture, InPlayerResourceAmount : string, InPlayerCoinAmount : string, InCurrentPrice : string, InTotalSellAmount : string, InUpgradeCost : string,
InMultiplierCost : string, InSellTrigger : trigger_device, InUpgradeTrigger : trigger_device, InMultiplierTrigger : trigger_device) := VendorData:
    ResourceImage := InResourceImage
    ResourceName := InResourceName
    TitleTexture := InTitleTexture
    CharacterTexture := InCharacterTexture
    PlayerResourceAmount := InPlayerResourceAmount
    PlayerCoinAmount := InPlayerCoinAmount
    CurrentPrice := InCurrentPrice
    TotalSellAmount := InTotalSellAmount
    UpgradeCost := InUpgradeCost
    MultiplierCost := InMultiplierCost
    SellTrigger := InSellTrigger
    UpgradeTrigger := InUpgradeTrigger
    MultiplierTrigger := InMultiplierTrigger

# Store data related to the canva and buttons
VendorCanvasData := class:
    Canvas : canvas

    # Widgets
    PlayerResourceText : text_block
    PlayerCoinText : text_block
    TotalSellAmountText : text_block
    CurrentPriceText : text_block
    UpgradeText : text_block
    PurchaseAmount : text_block

    MultiplierText : text_block
    MultiplierTextWidget : widget
    MultiplierButtonColor : texture_block
    UpgradeButtonColor : texture_block
    CantUpgradeText : text_block

    # Buttons
    ExitButton : text_button_base
    X1Button : text_button_base
    X10Button : text_button_base
    X100Button : text_button_base
    UpgradeButton : text_button_base
    SellButton : text_button_base
    MultiplierButton : text_button_base

    X1Trigger : trigger_device
    X10Trigger : trigger_device
    X100Trigger : trigger_device
    
    OnX1TriggerButtonEvent(WidgetMessage : widget_message):void=
        X1Trigger.Trigger(WidgetMessage.Player)
    
    OnX10TriggerButtonEvent(WidgetMessage : widget_message):void=
        X10Trigger.Trigger(WidgetMessage.Player)
        
    OnX100TriggerButtonEvent(WidgetMessage : widget_message):void=
        X100Trigger.Trigger(WidgetMessage.Player)

# Store data directly related to the variables displayed
VendorData := class:
    ResourceImage : texture
    ResourceName : string
    TitleTexture : texture
    CharacterTexture : texture
    PlayerResourceAmount : string
    PlayerCoinAmount : string
    CurrentPrice : string
    TotalSellAmount : string
    UpgradeCost : string
    MultiplierCost : string

    SellTrigger : trigger_device = trigger_device{}
    UpgradeTrigger : trigger_device = trigger_device{}
    MultiplierTrigger : trigger_device = trigger_device{}

    OnSellButtonEvent(WidgetMessage : widget_message):void=
        Print("Sell Button Press")
        SellTrigger.Trigger(WidgetMessage.Player)
    
    OnUpgradeButtonEvent(WidgetMessage : widget_message):void=
        Print("Upgrade Button Press")
        UpgradeTrigger.Trigger(WidgetMessage.Player)

    OnMultiplierButtonEvent(WidgetMessage : widget_message):void=
        Print("Multiplier Button Press")
        MultiplierTrigger.Trigger(WidgetMessage.Player)

# Mining Vendor for Mining Tycoon
VendorTemplate_UI := class():
    ToMessage<localizes>(String : string) : message = "{String}"
    CenteredAlignment : vector2 = vector2{X := 0.5, Y := 0.5}

    var BuiltCanvasMap : [player]VendorCanvasData = map{}

    # Initiliase the UI so it can added to the player much quicker
    InitUI(InPlayer : player, InMiningVendorData : VendorData,  InX1Trigger : trigger_device, InX10Trigger : trigger_device, InX100Trigger : trigger_device):void=
        VendorCanvas := MakeVendorCanvas(InMiningVendorData, InX1Trigger, InX10Trigger, InX100Trigger)
        # Set the players build canvas data
        if(set BuiltCanvasMap[InPlayer] = VendorCanvas, PlayerCanvas := BuiltCanvasMap[InPlayer]):
            #Print("Initiliased UI")
            PlayerCanvas.ExitButton.OnClick().Subscribe(HandlePlayerExit)
            # Set up triggers and their events
            PlayerCanvas.X1Button.OnClick().Subscribe(PlayerCanvas.OnX1TriggerButtonEvent)
            PlayerCanvas.X10Button.OnClick().Subscribe(PlayerCanvas.OnX10TriggerButtonEvent)
            PlayerCanvas.X100Button.OnClick().Subscribe(PlayerCanvas.OnX100TriggerButtonEvent)
            PlayerCanvas.UpgradeButton.OnClick().Subscribe(InMiningVendorData.OnUpgradeButtonEvent)
            PlayerCanvas.SellButton.OnClick().Subscribe(InMiningVendorData.OnSellButtonEvent)
            PlayerCanvas.MultiplierButton.OnClick().Subscribe(InMiningVendorData.OnMultiplierButtonEvent)

    # Add the UI to the player
    AddUI(InPlayer : player):void=
        if(PlayerUI := GetPlayerUI[InPlayer], CanvasData := BuiltCanvasMap[InPlayer]):
            CanvasData.Canvas.SetVisibility(widget_visibility.Visible)
            PlayerUI.AddWidget(CanvasData.Canvas, player_ui_slot{InputMode := ui_input_mode.All} )

    # When player presses the exit button
    HandlePlayerExit(WidgetMessage: widget_message):void=
        spawn{HideUI(WidgetMessage.Player)}

    HideUI(InPlayer: player)<suspends>:void=
        Sleep(0.0) # So we're sure that the canvas has been added to BuiltCanvas map
        if(PlayerUI := GetPlayerUI[InPlayer], CanvasData := BuiltCanvasMap[InPlayer]):
            CanvasData.Canvas.SetVisibility(widget_visibility.Hidden)
            PlayerUI.RemoveWidget(CanvasData.Canvas)
    
    # Upates a given player's UI
    UpdateUI(InPlayer : player, InPlayerResource : string, InPlayerResourceAmount : string, InPlayerCoinAmount : string, InCurrentPrice : string, InTotalSellAmount : string, InUpgradeCost : string,
    InMultiplierCost : string, InPlayerData : Player_Stats_Data, InTierSystems : []float, InPurchaseAmount : string, CanAfford : logic):void=
        if(PlayerUI := GetPlayerUI[InPlayer], CanvasData := BuiltCanvasMap[InPlayer]):
            Print("Update UI")
            CanvasData.PlayerResourceText.SetText(ToMessage("{InPlayerResourceAmount}"))
            CanvasData.PlayerCoinText.SetText(ToMessage("{InPlayerCoinAmount}"))
            CanvasData.TotalSellAmountText.SetText(ToMessage("{InTotalSellAmount}"))
            CanvasData.CurrentPriceText.SetText(ToMessage("{InCurrentPrice}"))
            CanvasData.UpgradeText.SetText(ToMessage("{InUpgradeCost}"))
            CanvasData.PurchaseAmount.SetText(ToMessage("Buy x{InPurchaseAmount}"))
            ShouldMultiplierShow(InPlayer, InPlayerResource, InPlayerData, InTierSystems)
            if(CanAfford = true):
                CanvasData.UpgradeButtonColor.SetTint(NamedColors.Yellow)
                CanvasData.UpgradeText.SetTextColor(NamedColors.White)
                #CanvasData.UpgradeButton.SetEnabled(true)
            else:
                CanvasData.UpgradeButtonColor.SetTint(NamedColors.Grey)
                CanvasData.UpgradeText.SetTextColor(NamedColors.Red)
                #CanvasData.UpgradeButton.SetEnabled(false)


    ShouldMultiplierShow(InPlayer : player, InPlayerResourceAmount : string, InPlayerData : Player_Stats_Data, InTierSystems : []float):void=
        if(PlayerUI := GetPlayerUI[InPlayer], CanvasData := BuiltCanvasMap[InPlayer]):
            # Check if the upgrade slot can be shown
            var UpgradeTierAvailable : int = 0

            var CurrentLevel : float = 0.0
            var LevelRequired : float = 0.0

            if(InPlayerResourceAmount = "Stone"):
                set CurrentLevel = InPlayerData.StoneVendorUpgradeLevel
                set UpgradeTierAvailable = InPlayerData.StoneVendorTierMultiplierUpgradesAvaliable

            if(UpgradeTierAvailable >= 1):
                CanvasData.MultiplierButtonColor.SetTint(NamedColors.Yellow)
                CanvasData.CantUpgradeText.SetVisibility(widget_visibility.Hidden)
                CanvasData.MultiplierTextWidget.SetVisibility(widget_visibility.Visible)
                #CanvasData.MultiplierButton.SetEnabled(true)
            else:
                CanvasData.MultiplierButtonColor.SetTint(NamedColors.Grey)
                CanvasData.CantUpgradeText.SetVisibility(widget_visibility.Visible)
                CanvasData.MultiplierTextWidget.SetVisibility(widget_visibility.Hidden)
                #CanvasData.MultiplierButton.SetEnabled(false)

                for(Tier : InTierSystems):
                    if(Tier > CurrentLevel and LevelRequired = 0.0):
                        set LevelRequired = Tier
                        CanvasData.CantUpgradeText.SetText(ToMessage("Required : {BigNumber_FormatDisplay(BigNumber_ConvertFloatIntoBigNumber(CurrentLevel))}/{BigNumber_FormatDisplay(BigNumber_ConvertFloatIntoBigNumber(LevelRequired))}"))


    # Make the vendor canvas ui
    MakeVendorCanvas(InMiningVendorData : VendorData,  InX1Trigger : trigger_device, InX10Trigger : trigger_device, InX100Trigger : trigger_device):VendorCanvasData=
        # Create all the buttons for the UI
        #Print("Make Vendor Canvas")
        ExitButton := button_loud:
            DefaultText := ToMessage("Exit")
        UpgradeButton := button_quiet:
            DefaultText := ToMessage("")
        SellButton := button_loud:
            DefaultText := ToMessage("Sell")
        MultiplierButton := button_quiet:
            DefaultText := ToMessage("")

        X1Button := button_quiet:
            DefaultText := ToMessage("x1")
        X10Button := button_quiet:
            DefaultText := ToMessage("x10")
        X100Button := button_quiet:
            DefaultText := ToMessage("x100")

        TitleTexture := texture_block:
            DefaultImage := InMiningVendorData.TitleTexture
            DefaultDesiredSize := vector2{X:=512.0,Y:=128.0}

        VendorTitleTexture := texture_block:
            DefaultImage := VerseExposed.T_VendorPricesTitle
            DefaultDesiredSize := vector2{X:=512.0,Y:=128.0}

        MultiplierButtonColor := texture_block:
            DefaultImage := VerseExposed.T_VendorButtonBackground
            DefaultDesiredSize := vector2{X:=494.0,Y:=54.0}
            DefaultTint := NamedColors.Yellow

        UpgradeButtonColor := texture_block:
            DefaultImage := VerseExposed.T_VendorButtonBackground
            DefaultDesiredSize := vector2{X:=494.0,Y:=54.0}
            DefaultTint := NamedColors.Yellow

        # Create the text blocks that do not update
        TitleText := text_block:
            DefaultText := ToMessage("{InMiningVendorData.ResourceName} Vendor")
            DefaultJustification := text_justification.Center
            DefaultShadowColor := NamedColors.Black
            DefaultShadowOffset := option{vector2{X:=2.0,Y:=2.0}}
        TitleText.SetTextColor(NamedColors.White)
        TitleText.SetShadowOpacity(0.8)
        DescriptionText := text_block:
            DefaultJustification := text_justification.Center
            DefaultShadowColor := NamedColors.Black
            DefaultText := ToMessage("Exchange your {InMiningVendorData.ResourceName} for coins!\nUpgrade your exchange price for better prices!")
            DefaultShadowOffset := option{vector2{X:=2.0,Y:=2.0}}
        DescriptionText.SetTextColor(NamedColors.White)
        DescriptionText.SetShadowOpacity(0.8)
        VendorPricesText := text_block:
            DefaultText := ToMessage("Vendor Prices")
            DefaultJustification := text_justification.Center
            DefaultShadowColor := NamedColors.Black
            DefaultShadowOffset := option{vector2{X:=2.0,Y:=2.0}}
        VendorPricesText.SetTextColor(NamedColors.White)
        VendorPricesText.SetShadowOpacity(0.8)
        CurrentPriceText := text_block:
            DefaultText := ToMessage("Current Price")
            DefaultJustification := text_justification.Center
            DefaultShadowColor := NamedColors.Black
            DefaultShadowOffset := option{vector2{X:=2.0,Y:=2.0}}
        CurrentPriceText.SetTextColor(NamedColors.White)
        CurrentPriceText.SetShadowOpacity(0.8)
        UpgradePriceText := text_block:
            DefaultText := ToMessage("Upgrade")
            DefaultJustification := text_justification.Center
            DefaultShadowColor := NamedColors.Black
            DefaultShadowOffset := option{vector2{X:=2.0,Y:=2.0}}
        UpgradePriceText.SetTextColor(NamedColors.White)
        UpgradePriceText.SetShadowOpacity(0.8)
        MultiplierText := text_block:
            DefaultText := ToMessage("2x Price")
            DefaultJustification := text_justification.Center
            DefaultShadowColor := NamedColors.Black
            DefaultShadowOffset := option{vector2{X:=2.0,Y:=2.0}}
        MultiplierText.SetTextColor(NamedColors.White)
        MultiplierText.SetShadowOpacity(0.8)
        BuyText1 := text_block:
            DefaultText := ToMessage("BUY")
            DefaultJustification := text_justification.Left
            DefaultShadowColor := NamedColors.Black
            DefaultShadowOffset := option{vector2{X:=2.0,Y:=2.0}}
        BuyText1.SetTextColor(NamedColors.White)
        BuyText1.SetShadowOpacity(0.8)
        BuyText2 := text_block:
            DefaultText := ToMessage("BUY")
            DefaultJustification := text_justification.Left
            DefaultShadowColor := NamedColors.Black
            DefaultShadowOffset := option{vector2{X:=2.0,Y:=2.0}}
        BuyText2.SetTextColor(NamedColors.White)
        BuyText2.SetShadowOpacity(0.8)
        PlayerCoinText := text_block:
            DefaultText := ToMessage("{InMiningVendorData.PlayerCoinAmount}")
            DefaultJustification := text_justification.Center
            DefaultShadowColor := NamedColors.Black
            DefaultShadowOffset := option{vector2{X:=2.0,Y:=2.0}}
        PlayerCoinText.SetTextColor(NamedColors.White)
        PlayerCoinText.SetShadowOpacity(0.8)

        # Create the text blocks that do update
        PlayerResourceText := text_block:
            DefaultText := ToMessage("{InMiningVendorData.PlayerResourceAmount}")
            DefaultJustification := text_justification.Center
            DefaultShadowColor := NamedColors.Black
            DefaultShadowOffset := option{vector2{X:=2.0,Y:=2.0}}
        PlayerResourceText.SetTextColor(NamedColors.White)
        PlayerResourceText.SetShadowOpacity(0.8)
        TotalSellAmountText := text_block:
            DefaultText := ToMessage("{InMiningVendorData.TotalSellAmount}")
            DefaultJustification := text_justification.Center
            DefaultShadowColor := NamedColors.Black
            DefaultShadowOffset := option{vector2{X:=2.0,Y:=2.0}}
        TotalSellAmountText.SetTextColor(NamedColors.White)
        TotalSellAmountText.SetShadowOpacity(0.8)
        CurrentPriceNumberText := text_block:
            DefaultText := ToMessage(InMiningVendorData.CurrentPrice)
            DefaultJustification := text_justification.Center
            DefaultShadowColor := NamedColors.Black
            DefaultShadowOffset := option{vector2{X:=2.0,Y:=2.0}}
        CurrentPriceNumberText.SetTextColor(NamedColors.White)
        CurrentPriceNumberText.SetShadowOpacity(0.8)
        MultiplierNumberText := text_block:
            DefaultText := ToMessage(InMiningVendorData.MultiplierCost)
            DefaultJustification := text_justification.Left
            DefaultShadowColor := NamedColors.Black
            DefaultShadowOffset := option{vector2{X:=2.0,Y:=2.0}}
        MultiplierNumberText.SetTextColor(NamedColors.White)
        MultiplierNumberText.SetShadowOpacity(0.8)
        UpgradeText := text_block:
            DefaultText := ToMessage("{InMiningVendorData.UpgradeCost}")
            DefaultJustification := text_justification.Left
            DefaultShadowColor := NamedColors.Black
            DefaultShadowOffset := option{vector2{X:=2.0,Y:=2.0}}
        UpgradeText.SetTextColor(NamedColors.White)
        UpgradeText.SetShadowOpacity(0.8)
        CantUpgradeText := text_block:
            DefaultText := ToMessage("CAN'T UPGRADE")
            DefaultJustification := text_justification.Left
            DefaultShadowColor := NamedColors.Black
            DefaultShadowOffset := option{vector2{X:=2.0,Y:=2.0}}
        CantUpgradeText.SetTextColor(NamedColors.DarkRed)
        CantUpgradeText.SetShadowOpacity(0.8)
        CantUpgradeText.SetVisibility(widget_visibility.Hidden)
        BuyAmountText := text_block:
            DefaultText := ToMessage("Buy x1")
            DefaultJustification := text_justification.Center
            DefaultShadowColor := NamedColors.Black
            DefaultShadowOffset := option{vector2{X:=2.0,Y:=2.0}}
        BuyAmountText.SetTextColor(NamedColors.White)
        BuyAmountText.SetShadowOpacity(0.8)

        # Variables for other parts of UI
        # Use these to copy and paste the settings into overlays
        BackgroundColorBlock := color_block:
            DefaultColor := NamedColors.White
            DefaultOpacity := 0.2

        BorderSize := 2.0
        BorderColorBlock := color_block:
            DefaultColor := NamedColors.White
            DefaultDesiredSize := vector2{X := BorderSize, Y := BorderSize}
    
        CoinWidget := texture_block:
            DefaultImage := VerseExposed.T_Stone

        # Start creating the overlay widgets that will input into a stack box then into a canvas
        # These are ordered from top to bottom
        # TitleOverlay displays the title at the top of the UI
        TitleOverlay := overlay:
            Slots := array:
                overlay_slot:
                    VerticalAlignment := vertical_alignment.Fill
                    HorizontalAlignment := horizontal_alignment.Fill
                    Widget := TitleTexture
        
        # DescriptionOverlay shows the description of this vendor
        DescriptionOverlay := overlay:
            Slots := array:
                overlay_slot:
                    VerticalAlignment := vertical_alignment.Fill
                    HorizontalAlignment := horizontal_alignment.Fill
                    Widget := color_block:
                        DefaultColor := NamedColors.White
                        DefaultOpacity := 0.2
                overlay_slot:
                    VerticalAlignment := vertical_alignment.Fill
                    HorizontalAlignment := horizontal_alignment.Fill
                    Padding := margin{Top := 10.0, Bottom := 10.0, Left := 10.0, Right := 10.0}
                    Widget := DescriptionText
                overlay_slot:
                    VerticalAlignment := vertical_alignment.Top
                    HorizontalAlignment := horizontal_alignment.Fill
                    Widget := color_block:
                        DefaultColor := NamedColors.DarkBlue
                        DefaultDesiredSize := vector2{X := BorderSize, Y := BorderSize}
                overlay_slot:
                    VerticalAlignment := vertical_alignment.Bottom
                    HorizontalAlignment := horizontal_alignment.Fill
                    Widget := color_block:
                        DefaultColor := NamedColors.DarkBlue
                        DefaultDesiredSize := vector2{X := BorderSize, Y := BorderSize}
                overlay_slot:
                    VerticalAlignment := vertical_alignment.Fill
                    HorizontalAlignment := horizontal_alignment.Right
                    Widget := color_block:
                        DefaultColor := NamedColors.DarkBlue
                        DefaultDesiredSize := vector2{X := BorderSize, Y := BorderSize}
                overlay_slot:
                    VerticalAlignment := vertical_alignment.Fill
                    HorizontalAlignment := horizontal_alignment.Left
                    Widget := color_block:
                        DefaultColor := NamedColors.DarkBlue
                        DefaultDesiredSize := vector2{X := BorderSize, Y := BorderSize}

        # PlayerResourceOverlay shows the player their current resources
        PlayerResourceOverlay := overlay:
            Slots := array:
                overlay_slot:
                    VerticalAlignment := vertical_alignment.Fill
                    HorizontalAlignment := horizontal_alignment.Fill
                    Widget := color_block:
                        DefaultColor := NamedColors.White
                        DefaultOpacity := 0.2
                overlay_slot:
                    VerticalAlignment := vertical_alignment.Fill
                    HorizontalAlignment := horizontal_alignment.Fill
                    Padding := margin{Top := 30.0, Bottom := 30.0, Left := 30.0, Right := 30.0}
                    Widget := stack_box:
                        Orientation := orientation.Horizontal
                        Slots := 
                            array:
                                stack_box_slot:
                                    Widget := texture_block:
                                        DefaultImage := InMiningVendorData.ResourceImage
                                stack_box_slot:
                                    Widget := PlayerResourceText
                overlay_slot:
                    VerticalAlignment := vertical_alignment.Top
                    HorizontalAlignment := horizontal_alignment.Fill
                    Widget := color_block:
                        DefaultColor := NamedColors.DarkBlue
                        DefaultDesiredSize := vector2{X := BorderSize, Y := BorderSize}
                overlay_slot:
                    VerticalAlignment := vertical_alignment.Bottom
                    HorizontalAlignment := horizontal_alignment.Fill
                    Widget := color_block:
                        DefaultColor := NamedColors.DarkBlue
                        DefaultDesiredSize := vector2{X := BorderSize, Y := BorderSize}
                overlay_slot:
                    VerticalAlignment := vertical_alignment.Fill
                    HorizontalAlignment := horizontal_alignment.Right
                    Widget := color_block:
                        DefaultColor := NamedColors.DarkBlue
                        DefaultDesiredSize := vector2{X := BorderSize, Y := BorderSize}
                overlay_slot:
                    VerticalAlignment := vertical_alignment.Fill
                    HorizontalAlignment := horizontal_alignment.Left
                    Widget := color_block:
                        DefaultColor := NamedColors.DarkBlue
                        DefaultDesiredSize := vector2{X := BorderSize, Y := BorderSize}
        
        # SellOverlay shows the player the sale price of all of their resource
        SellOverlay := overlay:
            Slots := array:
                overlay_slot:
                    VerticalAlignment := vertical_alignment.Fill
                    HorizontalAlignment := horizontal_alignment.Fill
                    Widget := color_block:
                        DefaultColor := NamedColors.White
                        DefaultOpacity := 0.2
                overlay_slot:
                    VerticalAlignment := vertical_alignment.Fill
                    HorizontalAlignment := horizontal_alignment.Fill
                    Padding := margin{Top := 12.0, Bottom := 10.0, Left := 30.0, Right := 30.0}
                    Widget := stack_box:
                        Orientation := orientation.Horizontal
                        Slots := 
                            array:
                                stack_box_slot:
                                    Widget := CoinWidget
                                stack_box_slot:
                                    Widget := TotalSellAmountText
                                stack_box_slot:
                                    Padding := margin{Left := 10.0}
                                    Widget := SellButton
                overlay_slot:
                    VerticalAlignment := vertical_alignment.Top
                    HorizontalAlignment := horizontal_alignment.Fill
                    Widget := color_block:
                        DefaultColor := NamedColors.DarkBlue
                        DefaultDesiredSize := vector2{X := BorderSize, Y := BorderSize}
                overlay_slot:
                    VerticalAlignment := vertical_alignment.Bottom
                    HorizontalAlignment := horizontal_alignment.Fill
                    Widget := color_block:
                        DefaultColor := NamedColors.DarkBlue
                        DefaultDesiredSize := vector2{X := BorderSize, Y := BorderSize}
                overlay_slot:
                    VerticalAlignment := vertical_alignment.Fill
                    HorizontalAlignment := horizontal_alignment.Right
                    Widget := color_block:
                        DefaultColor := NamedColors.DarkBlue
                        DefaultDesiredSize := vector2{X := BorderSize, Y := BorderSize}
                overlay_slot:
                    VerticalAlignment := vertical_alignment.Fill
                    HorizontalAlignment := horizontal_alignment.Left
                    Widget := color_block:
                        DefaultColor := NamedColors.DarkBlue
                        DefaultDesiredSize := vector2{X := BorderSize, Y := BorderSize}
        
        ResourceSellBox :=
            array:
                stack_box_slot:
                    Widget := PlayerResourceOverlay
                    Padding := margin{Left := 20.0, Right := 20.0}
                stack_box_slot:
                    Widget := SellOverlay
                    Padding := margin{Left := 20.0, Right := 20.0}

        ResourceSellOverlay := overlay:
            Slots := array:
                overlay_slot:
                    VerticalAlignment := vertical_alignment.Fill
                    HorizontalAlignment := horizontal_alignment.Fill
                    Widget := overlay:
                        Slots := array:
                            overlay_slot:
                                VerticalAlignment := vertical_alignment.Fill
                                HorizontalAlignment := horizontal_alignment.Fill
                                Widget := stack_box:
                                    Orientation := orientation.Horizontal
                                    Slots := ResourceSellBox

        # VendorPricesOverlay is the sub title for the table
        VendorPricesOverlay := overlay:
            Slots := array:
                overlay_slot:
                    VerticalAlignment := vertical_alignment.Fill
                    HorizontalAlignment := horizontal_alignment.Fill
                    Widget := VendorTitleTexture

        # CurrentPriceTextOverlay and CurrentPriceNumberOverlay show the player what that resource is currently selling for
        CurrentPriceTextOverlay := overlay:
            Slots := array:
                overlay_slot:
                    VerticalAlignment := vertical_alignment.Fill
                    HorizontalAlignment := horizontal_alignment.Fill
                    Padding := margin{Top := 10.0, Bottom := 10.0, Left := 30.0, Right := 30.0}
                    Widget := stack_box:
                        Orientation := orientation.Horizontal
                        Slots := 
                            array:
                                #stack_box_slot:
                                    #Widget := texture_block:
                                        #DefaultImage := VerseExposed.T_MoneyIcon
                                stack_box_slot:
                                    Widget := CurrentPriceText

        CurrentPriceNumberOverlay := overlay:
            Slots := array:
                overlay_slot:
                    VerticalAlignment := vertical_alignment.Fill
                    HorizontalAlignment := horizontal_alignment.Fill
                    Padding := margin{Top := 10.0, Bottom := 10.0, Left := 30.0, Right := 30.0}
                    Widget := stack_box:
                        Orientation := orientation.Horizontal
                        Slots := 
                            array:
                                stack_box_slot:
                                    Widget := texture_block:
                                        DefaultImage := VerseExposed.T_Stone
                                stack_box_slot:
                                    Widget := CurrentPriceNumberText
        
        # UpgradeTextOverlay and UpgradeButtonOverlay do something :)
        UpgradeTextOverlay := overlay:
            Slots := array:
                overlay_slot:
                    VerticalAlignment := vertical_alignment.Fill
                    HorizontalAlignment := horizontal_alignment.Fill
                    Padding := margin{Top := 10.0, Bottom := 10.0, Left := 30.0, Right := 30.0}
                    Widget := UpgradePriceText

        UpgradeButtonTextStack := stack_box:
            Orientation := orientation.Horizontal
            Slots :=
                array:
                    stack_box_slot:
                        VerticalAlignment := vertical_alignment.Center
                        HorizontalAlignment := horizontal_alignment.Center
                        Padding := margin{Left := 10.0, Right := 10.0}
                        Widget := BuyText1
                    stack_box_slot:
                        VerticalAlignment := vertical_alignment.Center
                        HorizontalAlignment := horizontal_alignment.Center
                        Padding := margin{Left := 10.0, Right := 5.0}
                        Widget := texture_block:
                            DefaultImage := VerseExposed.T_Stone
                    stack_box_slot:
                        VerticalAlignment := vertical_alignment.Center
                        HorizontalAlignment := horizontal_alignment.Center
                        Padding := margin{Left := 5.0, Right := 10.0}
                        Widget := UpgradeText

        UpgradeButtonOverlay := overlay:
            Slots := array:
                overlay_slot:
                    VerticalAlignment := vertical_alignment.Center
                    HorizontalAlignment := horizontal_alignment.Center
                    Widget := UpgradeButtonColor
                overlay_slot:
                    VerticalAlignment := vertical_alignment.Center
                    HorizontalAlignment := horizontal_alignment.Center
                    Padding := margin{Top := 10.0, Bottom := 10.0, Left := 15.0, Right := 15.0}
                    Widget := UpgradeButtonTextStack
                overlay_slot:
                    VerticalAlignment := vertical_alignment.Center
                    HorizontalAlignment := horizontal_alignment.Fill
                    Padding := margin{Left := 3.0, Right := 3.0}
                    Widget := UpgradeButton
        
        # MultiplierTextOverlay and MultiplierButtonOverlay do something too
        MultiplierTextOverlay := overlay:
            Slots := array:
                overlay_slot:
                    VerticalAlignment := vertical_alignment.Fill
                    HorizontalAlignment := horizontal_alignment.Fill
                    Padding := margin{Top := 10.0, Bottom := 10.0, Left := 30.0, Right := 30.0}
                    Widget := MultiplierText

        MultiplierButtonTextStack := stack_box:
            Orientation := orientation.Horizontal
            Slots :=
                array:
                    stack_box_slot:
                        VerticalAlignment := vertical_alignment.Center
                        HorizontalAlignment := horizontal_alignment.Center
                        Padding := margin{Left := 10.0, Right := 10.0}
                        Widget := BuyText2
                    stack_box_slot:
                        VerticalAlignment := vertical_alignment.Center
                        HorizontalAlignment := horizontal_alignment.Center
                        Padding := margin{Left := 10.0, Right := 5.0}
                        Widget := texture_block:
                            DefaultImage := VerseExposed.T_Stone
                    stack_box_slot:
                        VerticalAlignment := vertical_alignment.Center
                        HorizontalAlignment := horizontal_alignment.Center
                        Padding := margin{Left := 5.0, Right := 10.0}
                        Widget := MultiplierNumberText
        
        MultiplierButtonOverlay := overlay:
            Slots := array:
                overlay_slot:
                    VerticalAlignment := vertical_alignment.Center
                    HorizontalAlignment := horizontal_alignment.Center
                    Widget := MultiplierButtonColor
                overlay_slot:
                    VerticalAlignment := vertical_alignment.Center
                    HorizontalAlignment := horizontal_alignment.Center
                    Padding := margin{Left := 30.0, Right := 30.0}
                    Widget := CantUpgradeText
                overlay_slot:
                    VerticalAlignment := vertical_alignment.Center
                    HorizontalAlignment := horizontal_alignment.Center
                    Padding := margin{Top := 10.0, Bottom := 10.0, Left := 15.0, Right := 15.0}
                    Widget := MultiplierButtonTextStack
                overlay_slot:
                    VerticalAlignment := vertical_alignment.Center
                    HorizontalAlignment := horizontal_alignment.Fill
                    Padding := margin{Left := 3.0, Right := 3.0}
                    Widget := MultiplierButton
        
        # PricesTableLeftBox is the left section of the table
        PricesTableLeftBox :=
            array:
                stack_box_slot:
                    Widget := CurrentPriceTextOverlay
                stack_box_slot:
                    Widget := color_block:
                        DefaultColor := NamedColors.DarkBlue
                        DefaultDesiredSize := vector2{X := 500.0, Y := BorderSize}
                stack_box_slot:
                    Widget := UpgradeTextOverlay
                stack_box_slot:
                    Widget := color_block:
                            DefaultColor := NamedColors.DarkBlue
                            DefaultDesiredSize := vector2{X := 500.0, Y := BorderSize}
                stack_box_slot:
                    Widget := MultiplierTextOverlay

        # PricesTableRightBox is the right section of the table
        PricesTableRightBox :=
            array:
                stack_box_slot:
                    Widget := CurrentPriceNumberOverlay
                stack_box_slot:
                    Widget := color_block:
                            DefaultColor := NamedColors.DarkBlue
                            DefaultDesiredSize := vector2{X := 500.0, Y := BorderSize}
                stack_box_slot:
                    Widget := UpgradeButtonOverlay
                    VerticalAlignment := vertical_alignment.Center
                    HorizontalAlignment := horizontal_alignment.Fill
                stack_box_slot:
                    Widget := color_block:
                            DefaultColor := NamedColors.DarkBlue
                            DefaultDesiredSize := vector2{X := 500.0, Y := BorderSize}
                stack_box_slot:
                    Widget := MultiplierButtonOverlay
                    VerticalAlignment := vertical_alignment.Center
                    HorizontalAlignment := horizontal_alignment.Fill
        
        # PricesTableBox is the collection of the other table boxes with a divider
        PricesTableBox :=
            array:
                stack_box_slot:
                        Widget := stack_box:
                            Orientation := orientation.Vertical
                            Slots := PricesTableLeftBox
                stack_box_slot:
                    Widget := color_block:
                        DefaultColor := NamedColors.DarkBlue
                        DefaultDesiredSize := vector2{X := BorderSize, Y := 200.0}
                stack_box_slot:
                        Widget := stack_box:
                            Orientation := orientation.Vertical
                            Slots := PricesTableRightBox

        # PricesTableOverlay is the parent of the table that is actualy displayed
        PricesTableOverlay := overlay:
            Slots := array:
                overlay_slot:
                    VerticalAlignment := vertical_alignment.Fill
                    HorizontalAlignment := horizontal_alignment.Fill
                    Widget := color_block:
                        DefaultColor := NamedColors.White
                        DefaultOpacity := 0.2
                overlay_slot:
                    VerticalAlignment := vertical_alignment.Fill
                    HorizontalAlignment := horizontal_alignment.Fill
                    Widget := overlay:
                        Slots := array:
                            overlay_slot:
                                VerticalAlignment := vertical_alignment.Fill
                                HorizontalAlignment := horizontal_alignment.Fill
                                Widget := stack_box:
                                    Orientation := orientation.Horizontal
                                    Slots := PricesTableBox
                overlay_slot:
                    VerticalAlignment := vertical_alignment.Top
                    HorizontalAlignment := horizontal_alignment.Fill
                    Widget := color_block:
                        DefaultColor := NamedColors.DarkBlue
                        DefaultDesiredSize := vector2{X := BorderSize, Y := BorderSize}
                overlay_slot:
                    VerticalAlignment := vertical_alignment.Bottom
                    HorizontalAlignment := horizontal_alignment.Fill
                    Widget := color_block:
                        DefaultColor := NamedColors.DarkBlue
                        DefaultDesiredSize := vector2{X := BorderSize, Y := BorderSize}
                overlay_slot:
                    VerticalAlignment := vertical_alignment.Fill
                    HorizontalAlignment := horizontal_alignment.Right
                    Widget := color_block:
                        DefaultColor := NamedColors.DarkBlue
                        DefaultDesiredSize := vector2{X := BorderSize, Y := BorderSize}
                overlay_slot:
                    VerticalAlignment := vertical_alignment.Fill
                    HorizontalAlignment := horizontal_alignment.Left
                    Widget := color_block:
                        DefaultColor := NamedColors.DarkBlue
                        DefaultDesiredSize := vector2{X := BorderSize, Y := BorderSize}

        PurchaseAmountStack := stack_box:
            Orientation := orientation.Horizontal
            Slots :=
                array:
                    stack_box_slot:
                        HorizontalAlignment := horizontal_alignment.Center
                        VerticalAlignment := vertical_alignment.Center
                        Padding := margin{ Left := 10.0, Right  := 10.0}
                        Widget := BuyAmountText
                    stack_box_slot:
                        Widget := color_block:
                            DefaultColor := NamedColors.DarkBlue
                            DefaultDesiredSize := vector2{X := BorderSize, Y := BorderSize}
                        HorizontalAlignment := horizontal_alignment.Right
                        VerticalAlignment := vertical_alignment.Fill
                    stack_box_slot:
                        HorizontalAlignment := horizontal_alignment.Center
                        VerticalAlignment := vertical_alignment.Center
                        Padding := margin{ Left := 10.0, Right  := 10.0}
                        Widget := X1Button
                    stack_box_slot:
                        HorizontalAlignment := horizontal_alignment.Center
                        VerticalAlignment := vertical_alignment.Center
                        Padding := margin{ Left := 10.0, Right  := 10.0}
                        Widget := X10Button
                    stack_box_slot:
                        HorizontalAlignment := horizontal_alignment.Center
                        VerticalAlignment := vertical_alignment.Center
                        Padding := margin{ Left := 10.0, Right  := 10.0}
                        Widget := X100Button

        PurchaseAmountOverlay := overlay:
            Slots := array:
                overlay_slot:
                    VerticalAlignment := vertical_alignment.Center
                    HorizontalAlignment := horizontal_alignment.Fill
                    Widget := color_block:
                        DefaultColor := NamedColors.White
                        DefaultOpacity := 0.4
                        DefaultDesiredSize := vector2{X:=225.0,Y:=60.0}
                overlay_slot:
                    VerticalAlignment := vertical_alignment.Fill
                    HorizontalAlignment := horizontal_alignment.Fill
                    Widget := PurchaseAmountStack
                overlay_slot:
                    Widget := color_block:
                        DefaultColor := NamedColors.DarkBlue
                        DefaultDesiredSize := vector2{X := BorderSize, Y := BorderSize}
                    HorizontalAlignment := horizontal_alignment.Fill
                    VerticalAlignment := vertical_alignment.Top
                overlay_slot:
                    Widget := color_block:
                        DefaultColor := NamedColors.DarkBlue
                        DefaultDesiredSize := vector2{X := 525.0, Y := BorderSize}
                    HorizontalAlignment := horizontal_alignment.Center
                    VerticalAlignment := vertical_alignment.Bottom
                overlay_slot:
                    Widget := color_block:
                        DefaultColor := NamedColors.DarkBlue
                        DefaultDesiredSize := vector2{X := BorderSize, Y := BorderSize}
                    HorizontalAlignment := horizontal_alignment.Right
                    VerticalAlignment := vertical_alignment.Fill
                overlay_slot:
                    Widget := color_block:
                        DefaultColor := NamedColors.DarkBlue
                        DefaultDesiredSize := vector2{X := BorderSize, Y := BorderSize}
                    HorizontalAlignment := horizontal_alignment.Left
                    VerticalAlignment := vertical_alignment.Fill

        ExitButtonOverlay := overlay:
            Slots := array:
                overlay_slot:
                    VerticalAlignment := vertical_alignment.Center
                    HorizontalAlignment := horizontal_alignment.Fill
                    Widget := ExitButton

        # OverallStackBox is used to stack all of the widgets together
        OverallStackBox :=
            array:
                stack_box_slot:
                    Widget := TitleOverlay
                stack_box_slot:
                    Widget := DescriptionOverlay
                    Padding := margin{Bottom := 20.0}
                stack_box_slot:
                    Widget := ResourceSellOverlay
                    Padding := margin{Top := 20.0, Bottom := 20.0}
                stack_box_slot:
                    Widget := VendorPricesOverlay
                stack_box_slot:
                    Widget := PurchaseAmountOverlay
                stack_box_slot:
                    Widget := PricesTableOverlay
                    Padding := margin{Top := 20.0, Bottom := 20.0}
                stack_box_slot:
                    Widget := ExitButtonOverlay
                    Padding := margin{Top := 20.0, Bottom := 20.0}

        # OverallOverlay is the actual overlay slot for the widgets
        OverallOverlay := overlay:
            Slots := array:
                overlay_slot:
                    VerticalAlignment := vertical_alignment.Fill
                    HorizontalAlignment := horizontal_alignment.Fill
                    Widget := overlay:
                        Slots := array:
                            overlay_slot:
                                VerticalAlignment := vertical_alignment.Fill
                                HorizontalAlignment := horizontal_alignment.Fill
                                Widget := stack_box:
                                    Orientation := orientation.Vertical
                                    Slots := OverallStackBox

        # Player Coin Count
        IconOverlay := overlay:
            Slots := array:
                overlay_slot:
                    VerticalAlignment := vertical_alignment.Center
                    HorizontalAlignment := horizontal_alignment.Center
                    Padding := margin{Top := -8.0, Bottom := -8.0, Left := -6.0, Right := -6.0}
                    Widget := color_block:
                        DefaultColor := NamedColors.Black
                        DefaultOpacity := 0.4
                        DefaultDesiredSize := vector2{X := 52.0, Y := 52.0}
                overlay_slot:
                    VerticalAlignment := vertical_alignment.Center
                    HorizontalAlignment := horizontal_alignment.Center
                    Widget := texture_block:
                        DefaultImage := VerseExposed.T_Stone
        
        PlayerCoinOverlay := overlay:
            Slots := array:
                overlay_slot:
                    VerticalAlignment := vertical_alignment.Center
                    HorizontalAlignment := horizontal_alignment.Left
                    Widget := color_block:
                        DefaultColor := NamedColors.Black
                        DefaultOpacity := 0.4
                        DefaultDesiredSize := vector2{X:=225.0,Y:=60.0}
                overlay_slot:
                    VerticalAlignment := vertical_alignment.Center
                    HorizontalAlignment := horizontal_alignment.Left
                    Padding := margin{Top := 10.0, Bottom := 10.0, Left := 10.0, Right := 10.0}
                    Widget := stack_box:
                        Orientation := orientation.Horizontal
                        Slots := 
                            array:
                                stack_box_slot:
                                    Widget := IconOverlay
                                    HorizontalAlignment := horizontal_alignment.Left
                                    VerticalAlignment := vertical_alignment.Center
                                stack_box_slot:
                                    Widget := PlayerCoinText
                                    Padding := margin{Left := 15.0}
                                    VerticalAlignment := vertical_alignment.Center
                                    HorizontalAlignment := horizontal_alignment.Left

        BackgroundOverlay := overlay:
            Slots := array:
                overlay_slot:
                    VerticalAlignment := vertical_alignment.Fill
                    HorizontalAlignment := horizontal_alignment.Fill
                    Widget := texture_block:
                        DefaultImage := VerseExposed.T_UIBackground
                overlay_slot:
                    VerticalAlignment := vertical_alignment.Bottom
                    HorizontalAlignment := horizontal_alignment.Left
                    Padding := margin{Left := -100.0}
                    Widget := texture_block:
                        DefaultImage := InMiningVendorData.CharacterTexture
                        DefaultDesiredSize := vector2{X:=1024.0,Y:=1024.0}
        
        # OutputCanvas is the canvas that is output with in the MiningCanvasData
        OutputCanvas := canvas:
            Slots := array:
                canvas_slot:
                    Anchors := anchors{Minimum := vector2{X:=0.0, Y:=0.0}, Maximum := vector2{X:=1.0, Y:=1.0}}
                    Alignment := CenteredAlignment
                    Widget := BackgroundOverlay
                    ZOrder := 0
                canvas_slot:
                    Anchors := anchors{Minimum := vector2{X:=0.025, Y:=0.05}, Maximum := vector2{X:=0.15, Y:=0.2}}
                    Alignment := CenteredAlignment
                    Widget := PlayerCoinOverlay
                    SizeToContent := true
                    ZOrder := 1
                canvas_slot:
                    Anchors := anchors{Minimum := vector2{X:=0.35, Y:=0.05}, Maximum := vector2{X:=0.95, Y:=0.95}}
                    Alignment := CenteredAlignment
                    Widget := OverallOverlay
                    SizeToContent := true
                    ZOrder := 1

        return VendorCanvasData:
            Canvas := OutputCanvas
            PlayerResourceText := PlayerResourceText
            PlayerCoinText := PlayerCoinText
            TotalSellAmountText := TotalSellAmountText
            CurrentPriceText := CurrentPriceNumberText
            UpgradeText := UpgradeText
            MultiplierText := MultiplierNumberText
            PurchaseAmount := BuyAmountText
            ExitButton := ExitButton
            X1Button := X1Button
            X10Button := X10Button
            X100Button := X100Button
            X1Trigger := InX1Trigger
            X10Trigger := InX10Trigger
            X100Trigger := InX100Trigger
            UpgradeButton := UpgradeButton
            SellButton := SellButton
            MultiplierButton := MultiplierButton
            MultiplierButtonColor := MultiplierButtonColor
            UpgradeButtonColor := UpgradeButtonColor
            MultiplierTextWidget := MultiplierButtonTextStack
            CantUpgradeText := CantUpgradeText