using { /Fortnite.com/UI }
using { /Fortnite.com/Devices }
using { /Fortnite.com/Characters }
using { /Verse.org/Simulation }
using { /Verse.org/Random }
using { /Verse.org/Colors }
using { /Verse.org/Assets }
using { /UnrealEngine.com/Temporary/Diagnostics }
using { /UnrealEngine.com/Temporary/UI }
using { /UnrealEngine.com/Temporary/SpatialMath }

MakeXPBarData<constructor>(InXPAmount : float, InMaxXPAmount : float, InLevel : int):= XPBarData:
    XPAmount := InXPAmount
    MaxXPAmount := InMaxXPAmount
    Level := InLevel

XPBarCanvasData := class:
    Canvas : canvas
    LevelText : text_block
    XPText : text_block
    XPAmount : float
    XPColorBlock : color_block

XPBarData := class:
    XPAmount : float
    MaxXPAmount : float
    Level : int

UI_XPDisplay := class():
    ToMessage<localizes>(String : string) : message = "{String}"
    CenteredAlignment : vector2 = vector2{X := 0.5, Y := 0.5}

    var BuiltCanvasMap : [player]XPBarCanvasData = map{}

    XPBarMaxSize : float = 600.0

    InitUI(InPlayer : player, InXPBarData : XPBarData):void=
        XPBarCanvas := MakeXPBar(InXPBarData)
        # Set the players build canvas data
        if(set BuiltCanvasMap[InPlayer] = XPBarCanvas) {}

    # Add the UI to the player
    AddUI(InPlayer : player):void=
        if(PlayerUI := GetPlayerUI[InPlayer], CanvasData := BuiltCanvasMap[InPlayer]):
            CanvasData.Canvas.SetVisibility(widget_visibility.Visible)
            PlayerUI.AddWidget(CanvasData.Canvas)

    # Add the UI to the player
    HideUI(InPlayer : player):void=
        if(PlayerUI := GetPlayerUI[InPlayer], CanvasData := BuiltCanvasMap[InPlayer]):
            CanvasData.Canvas.SetVisibility(widget_visibility.Hidden)
            PlayerUI.RemoveWidget(CanvasData.Canvas)

    # Updates the UI for the given player given the other variables
    UpdateUI(InPlayer : player, InXPAmount : float, InMaxXPAmount : float, InLevel : int):void=
        if(PlayerUI := GetPlayerUI[InPlayer], CanvasData := BuiltCanvasMap[InPlayer]):
            var XPInt : string = BigNumber_FormatDisplay(BigNumber_ConvertFloatIntoBigNumber(InXPAmount))
            var MaxXPInt : string = BigNumber_FormatDisplay(BigNumber_ConvertFloatIntoBigNumber(InMaxXPAmount))

            CanvasData.XPText.SetText(ToMessage("{XPInt}/{MaxXPInt}"))
            CanvasData.LevelText.SetText(ToMessage("Level: {InLevel}"))
            NewColorBlockSize := vector2{X:=(InXPAmount / InMaxXPAmount) * XPBarMaxSize, Y:=42.0}
            CanvasData.XPColorBlock.SetDesiredSize(NewColorBlockSize)

    # Make the xp bar
    MakeXPBar(InXPBarData : XPBarData):XPBarCanvasData=
        var XPInt : string = BigNumber_FormatDisplay(BigNumber_ConvertFloatIntoBigNumber(InXPBarData.XPAmount))
        var MaxXPInt : string = BigNumber_FormatDisplay(BigNumber_ConvertFloatIntoBigNumber(InXPBarData.MaxXPAmount))

        # Variables used to copy paste into overlays
        BorderSize := 1.0
        BorderColorBlock := color_block:
            DefaultColor := NamedColors.White
            DefaultDesiredSize := vector2{X := BorderSize, Y := BorderSize}
        
        # Color blocks
        BackgroundColorBlock := color_block:
            DefaultColor := NamedColors.Black
            DefaultOpacity := 0.6
            DefaultDesiredSize := vector2{X:=600.0,Y:=42.0}
        XPColorBlock := color_block:
            DefaultColor := NamedColors.Cyan
            DefaultOpacity := 1.0
            DefaultDesiredSize := vector2{X:=0.0,Y:=42.0}
        BarShadingBottomColorBlock := color_block:
            DefaultColor := NamedColors.Black
            DefaultOpacity := 0.4
            DefaultDesiredSize := vector2{X:= 600.0,Y:=21.0}
        BarShadingTopColorBlock := color_block:
            DefaultColor := NamedColors.White
            DefaultOpacity := 0.2
            DefaultDesiredSize := vector2{X:= 600.0,Y:=21.0}

        # Updating text blocks
        XPText := text_block:
            DefaultText := ToMessage("{XPInt}/{MaxXPInt}")
            DefaultJustification := text_justification.Left
            DefaultShadowColor := NamedColors.Black
            DefaultShadowOffset := option{vector2{X:=2.0,Y:=2.0}}
        XPText.SetTextColor(NamedColors.White)
        XPText.SetShadowOpacity(0.8)
        LevelText := text_block:
            DefaultText := ToMessage("Level: {InXPBarData.Level}")
            DefaultJustification := text_justification.Left
            DefaultShadowColor := NamedColors.Black
            DefaultShadowOffset := option{vector2{X:=2.0,Y:=2.0}}
        LevelText.SetTextColor(NamedColors.White)
        LevelText.SetShadowOpacity(0.8)

        BarOverlay := overlay:
            Slots := array:
                overlay_slot:
                    VerticalAlignment := vertical_alignment.Center
                    HorizontalAlignment := horizontal_alignment.Center
                    Widget := BackgroundColorBlock
                overlay_slot:
                    VerticalAlignment := vertical_alignment.Center
                    HorizontalAlignment := horizontal_alignment.Left
                    Widget := XPColorBlock
                overlay_slot:
                    VerticalAlignment := vertical_alignment.Bottom
                    HorizontalAlignment := horizontal_alignment.Left
                    Widget := BarShadingBottomColorBlock
                overlay_slot:
                    VerticalAlignment := vertical_alignment.Top
                    HorizontalAlignment := horizontal_alignment.Left
                    Widget := BarShadingTopColorBlock
                overlay_slot:
                    VerticalAlignment := vertical_alignment.Bottom
                    HorizontalAlignment := horizontal_alignment.Left
                    Padding := margin{Left := 20.0}
                    Widget := LevelText
                overlay_slot:
                    VerticalAlignment := vertical_alignment.Bottom
                    HorizontalAlignment := horizontal_alignment.Right
                    Padding := margin{Right := 20.0}
                    Widget := XPText
                overlay_slot:
                    VerticalAlignment := vertical_alignment.Top
                    HorizontalAlignment := horizontal_alignment.Fill
                    Widget := color_block:
                        DefaultColor := NamedColors.Black
                        DefaultDesiredSize := vector2{X := BorderSize, Y := BorderSize}
                overlay_slot:
                    VerticalAlignment := vertical_alignment.Bottom
                    HorizontalAlignment := horizontal_alignment.Fill
                    Widget := color_block:
                        DefaultColor := NamedColors.Black
                        DefaultDesiredSize := vector2{X := BorderSize, Y := BorderSize}
                overlay_slot:
                    VerticalAlignment := vertical_alignment.Fill
                    HorizontalAlignment := horizontal_alignment.Right
                    Widget := color_block:
                        DefaultColor := NamedColors.Black
                        DefaultDesiredSize := vector2{X := BorderSize, Y := BorderSize}
                overlay_slot:
                    VerticalAlignment := vertical_alignment.Fill
                    HorizontalAlignment := horizontal_alignment.Left
                    Widget := color_block:
                        DefaultColor := NamedColors.Black
                        DefaultDesiredSize := vector2{X := BorderSize, Y := BorderSize}

        XPBarStackBox := stack_box:
            Orientation := orientation.Vertical
            Slots := array:
                    stack_box_slot:
                        Widget := BarOverlay
                        HorizontalAlignment := horizontal_alignment.Center
                        VerticalAlignment := vertical_alignment.Center

        OutputCanvas := canvas:
            Slots := array:
                <#
                #TODO: Add Image
                canvas_slot:
                    Anchors := anchors{Minimum := vector2{X:=0.0, Y:=0.0}, Maximum := vector2{X:=0.5, Y:=0.5}}
                    Alignment := CenteredAlignment
                    Offsets := margin{Top := 0.0, Left:= 0.0, Right:= 0.0, Bottom:=0.0}
                    Widget := texture_block:
                        DefaultImage := VerseExposed.T_XPBar
                        DefaultDesiredSize := vector2{X:=512.0,Y:=512.0}
                #>
                canvas_slot:
                    Anchors := anchors{Minimum := vector2{X:=0.5, Y:=0.90}, Maximum := vector2{X:=0.5, Y:=0.90}}
                    Alignment := CenteredAlignment
                    Widget := XPBarStackBox
                    SizeToContent := true

        return XPBarCanvasData:
            Canvas := OutputCanvas
            LevelText := LevelText
            XPText := XPText
            XPAmount := InXPBarData.XPAmount
            XPColorBlock := XPColorBlock