using { /Verse.org/Simulation }
using { /UnrealEngine.com/Temporary/SpatialMath }
using { /Verse.org/Random }
using { /Fortnite.com/Devices }
using { /Fortnite.com/Game }
using { /Fortnite.com/Characters }

patrol_rect := struct<concrete>:
    MaxX : float = 1.0
    MinX : float = 0.0
    MaxY : float = 1.0
    MinY : float = 0.0

# Returns a random point within the bounds of the patrol rect
# Good for making AI patrol a certain area and not go out of bounds
GetRandomPointWithinBounds(PatrolRect : patrol_rect):vector3=
    RandomX := GetRandomFloat(PatrolRect.MinX, PatrolRect.MaxX)
    RandomY := GetRandomFloat(PatrolRect.MinY, PatrolRect.MaxY)

    vector3{X:=RandomX, Y:=RandomY, Z:= 0.0}

# Returns the closest positional to the base transform
# Can be used on players or props or anything with a transform
(Base : transform).GetClosestPositional(Targets : []positional)<transacts>:?positional=
    var ClosestIDX : int = -1
    var ClosestDistance : float = 5000.0

    for (IDX -> Target : Targets):
        TDistance := Distance(Target.GetTransform().Translation, Base.Translation)
        if (TDistance < ClosestDistance):
            set ClosestDistance = TDistance
            set ClosestIDX = IDX

    if (Target := Targets[ClosestIDX]):
        option{Target}
    else:
        false

# Casts players into positionals for easier math work
(Players : []player).ToPositionals()<transacts>:[]positional=
    var Positionals : []positional = array{}

    for (Player : Players, Agent := agent[Player], Fort := Agent.GetFortCharacter[], Pos := positional[Fort]):
        set Positionals += array{Pos}

    Positionals

# What rotation a transform needs to make to look at a target
(Location : transform).GetLookAtRotation(Target : vector3):rotation=
    BaseTran := Location.Translation
    TargetTran := Target

    if (LookDirection := (TargetTran - BaseTran).MakeUnitVector[]):
        Yaw := RadiansToDegrees(ArcTan(LookDirection.Y, LookDirection.X)) - 90.0
        Pitch := 0.0 
        Roll := 0.0

        MakeRotationFromYawPitchRollDegrees(Yaw, Pitch, Roll)
    else:
        IdentityRotation()

#Generates random points within a circle given the radius and center point
GenerateRandomPointsInCircle(TotalPoints : int, CenterPoint : vector3, Radius: int):[]vector3=
        var NewPoints: []vector3 = array{}
        var IDX: int = 0
        loop:
            if (IDX >= TotalPoints):
                break
                
            RandomFloat := GetRandomFloat(0.0, 1.0)
            Theta := RandomFloat * (PiFloat * 2)

            var RandomRadius : int = GetRandomInt(-Radius, Radius)
            XCoord := RandomRadius * Cos(Theta)
            set RandomRadius = GetRandomInt(-Radius, Radius)
            YCoord := RandomRadius * Sin(Theta)

            NewPoint := vector3{X:= XCoord, Y:= YCoord, Z:= 180.0} + CenterPoint
            set NewPoints += array{NewPoint}
            set IDX += 1
        
        NewPoints
