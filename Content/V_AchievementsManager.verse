
using { /Fortnite.com/Devices }
using { /Verse.org/Simulation }
using { /UnrealEngine.com/Temporary/Diagnostics }

using { /Verse.org/Assets }

AchievementType := enum:
    TotalStone
    HighestStrength
    DungeonsCompleted
    ArtifactsSmelted
    RocksDestroyed
    Null

achievement_data := class<concrete>():
    Image : texture = VerseExposed.T_UIBackground
    Name : string = "Name"
    Description : string = "Description"
    Type : AchievementType = AchievementType.Null
    Levels : []big_number = array{}
    
    GetCurrentLevelAmount(InLevel : int):big_number=
        if(LevelAmount := Levels[InLevel]):
            return LevelAmount
        return big_number{}

AchievementCheckHandler := class() {
    Device : V_AchievementsManager
    Type : AchievementType

    HandlerFunction(InMaybeAgent : ?agent) : void = {
        if(Player := player[InMaybeAgent?]):
            spawn:
                Device.CheckCompletedAcheievements(Player, Type)
    }
}

# A Verse-authored creative device that can be placed in a level
V_AchievementsManager := class(creative_device):
    ToMessage<localizes>(String : string) : message = "{String}"

    TotalStoneCollectedAchievement : achievement_data = achievement_data:
        Name := "Stone Collected"
        Image := VerseExposed.T_Achievements_Stone
        Levels := array{big_number{Value := 1.00, Exponent := 2}, big_number{Value := 1.00, Exponent := 3}, big_number{Value := 1.00, Exponent := 6}, big_number{Value := 1.00, Exponent := 9}, big_number{Value := 1.00, Exponent := 12}, big_number{Value := 1.00, Exponent := 15}, big_number{Value := 1.00, Exponent := 18}, big_number{Value := 1.00, Exponent := 21}, big_number{Value := 1.00, Exponent := 24}, big_number{Value := 1.00, Exponent := 27}, big_number{Value := 1.00, Exponent := 30}, big_number{Value := 1.00, Exponent := 33}, big_number{Value := 1.00, Exponent := 36}, big_number{Value := 1.00, Exponent := 39}, big_number{Value := 1.00, Exponent := 42}, big_number{Value := 1.00, Exponent := 45}, big_number{Value := 1.00, Exponent := 48}, big_number{Value := 1.00, Exponent := 51}, big_number{Value := 1.00, Exponent := 54}, big_number{Value := 1.00, Exponent := 57}}
        Description := "Total amount of stone collected over your career."
        Type := AchievementType.TotalStone

    TotalStrengthAmountAchievement : achievement_data = achievement_data:
        Name := "Strength Achieved"
        Image := VerseExposed.T_Achievements_Strength
        Levels := array{big_number{Value := 1.00, Exponent := 2}, big_number{Value := 1.00, Exponent := 3}, big_number{Value := 1.00, Exponent := 6}, big_number{Value := 1.00, Exponent := 9}, big_number{Value := 1.00, Exponent := 12}, big_number{Value := 1.00, Exponent := 15}, big_number{Value := 1.00, Exponent := 18}, big_number{Value := 1.00, Exponent := 21}, big_number{Value := 1.00, Exponent := 24}, big_number{Value := 1.00, Exponent := 27}, big_number{Value := 1.00, Exponent := 30}, big_number{Value := 1.00, Exponent := 33}, big_number{Value := 1.00, Exponent := 36}, big_number{Value := 1.00, Exponent := 39}, big_number{Value := 1.00, Exponent := 42}, big_number{Value := 1.00, Exponent := 45}, big_number{Value := 1.00, Exponent := 48}, big_number{Value := 1.00, Exponent := 51}, big_number{Value := 1.00, Exponent := 54}, big_number{Value := 1.00, Exponent := 57}}
        Description := "Heighest strength total you have achieved over your career."
        Type := AchievementType.HighestStrength

    RocksDestroyedAchievement : achievement_data = achievement_data:
        Name := "Rocks Destroyed"
        Image := VerseExposed.T_Achievements_Rocks
        Levels := array{big_number{Value := 1.00, Exponent := 1}, big_number{Value := 1.00, Exponent := 2}, big_number{Value := 5.00, Exponent := 2}, big_number{Value := 1.00, Exponent := 3}, big_number{Value := 5.00, Exponent := 3}, big_number{Value := 1.00, Exponent := 4}, big_number{Value := 5.0, Exponent := 4}, big_number{Value := 1.0, Exponent := 6}}
        Description := "Total number of rocks you have destroyed."
        Type := AchievementType.RocksDestroyed

    ArtifactsSmeltedAchievement : achievement_data = achievement_data:
        Name := "Artifacts Smelted"
        Image := VerseExposed.T_Achievements_Artifacts
        Levels := array{big_number{Value := 1.00, Exponent := 1}, big_number{Value := 1.00, Exponent := 2}, big_number{Value := 5.00, Exponent := 2}, big_number{Value := 1.00, Exponent := 3}, big_number{Value := 5.00, Exponent := 3}, big_number{Value := 1.00, Exponent := 4}, big_number{Value := 5.0, Exponent := 4}, big_number{Value := 1.0, Exponent := 6}}
        Description := "Total amount of Artifacts you have smelted."
        Type := AchievementType.ArtifactsSmelted

    DungeonsCompletedAchievement : achievement_data = achievement_data:
        Name := "Dungeons Completed"
        Image := VerseExposed.T_Achievements_Dungeons
        Levels := array{big_number{Value := 1.00, Exponent := 0}, big_number{Value := 1.00, Exponent := 2}, big_number{Value := 5.00, Exponent := 2}, big_number{Value := 1.00, Exponent := 3}, big_number{Value := 5.00, Exponent := 3}, big_number{Value := 1.00, Exponent := 4}, big_number{Value := 5.0, Exponent := 4}, big_number{Value := 1.0, Exponent := 6}}
        Description := "Total number of dungeons completed over your career."
        Type := AchievementType.DungeonsCompleted

    var AllAchievements : []achievement_data = array{}

    var PlayerStatsManager : player_stats_manager = player_stats_manager{}

    @editable
    StoneCollectedTriggers : []trigger_device = array{}
    @editable
    RockDestroyedTrigger : trigger_device = trigger_device{}
    @editable
    HighestStrengthTriggers : []trigger_device = array{}
    @editable
    ArtifactsSmeltedTrigger : trigger_device = trigger_device{}
    @editable
    DungeonsCompletedTriggers : []trigger_device = array{}

    @editable
    HUDMessage : hud_message_device = hud_message_device{}

    @editable
    AchievementAccolade : accolades_device = accolades_device{}

    # Runs when the device is started in a running game
    OnBegin<override>()<suspends>:void=
        set AllAchievements = array{TotalStoneCollectedAchievement, TotalStrengthAmountAchievement, RocksDestroyedAchievement, ArtifactsSmeltedAchievement, DungeonsCompletedAchievement}

        RockDestroyedTrigger.TriggeredEvent.Subscribe(AchievementCheckHandler{Device := Self, Type := AchievementType.RocksDestroyed}.HandlerFunction)
        ArtifactsSmeltedTrigger.TriggeredEvent.Subscribe(AchievementCheckHandler{Device := Self, Type := AchievementType.ArtifactsSmelted}.HandlerFunction)

        for(HighestStrengthTrigger : HighestStrengthTriggers):
            HighestStrengthTrigger.TriggeredEvent.Subscribe(AchievementCheckHandler{Device := Self, Type := AchievementType.HighestStrength}.HandlerFunction)

        for(DungeonsCompletedTrigger : DungeonsCompletedTriggers):
            DungeonsCompletedTrigger.TriggeredEvent.Subscribe(AchievementCheckHandler{Device := Self, Type := AchievementType.DungeonsCompleted}.HandlerFunction)

        for(StoneCollectedTrigger : StoneCollectedTriggers):
            StoneCollectedTrigger.TriggeredEvent.Subscribe(AchievementCheckHandler{Device := Self, Type := AchievementType.TotalStone}.HandlerFunction)

    CheckCompletedAcheievements(InPlayer : player, CheckType : AchievementType)<suspends>:void=
        if(PlayerStatsData := PlayerStatsManager.GetPlayerStats[InPlayer]):
            var AchievementCount : float = 0.0
            var Achievement : achievement_data = achievement_data{}
            var AchievementBonus : float = 0.0
            var CompareStat : big_number = big_number{}

            case(CheckType):
                AchievementType.TotalStone =>
                    Print("Checking Total Stone")
                    set CompareStat = PlayerStatsData.TotalStoneOreAmount
                    set Achievement = TotalStoneCollectedAchievement
                    set AchievementBonus = PlayerStatsData.AchievementTotalStoneBonus
                AchievementType.HighestStrength =>
                    Print("Checking Total Strength")
                    set CompareStat = PlayerStatsData.HighestStrength
                    set Achievement = TotalStrengthAmountAchievement
                    set AchievementBonus = PlayerStatsData.AchievementHighestStrengthBonus
                AchievementType.RocksDestroyed =>
                    Print("Checking Rocks Destroyed")
                    set CompareStat = PlayerStatsData.TotalRocksDestroyed
                    set Achievement = RocksDestroyedAchievement
                    set AchievementBonus = PlayerStatsData.AchievementTotalRocksDestroyedBonus
                AchievementType.ArtifactsSmelted =>
                    Print("Checking Artifacts Smelted")
                    set CompareStat = PlayerStatsData.TotalArtifactsSmelted
                    set Achievement = ArtifactsSmeltedAchievement
                    set AchievementBonus = PlayerStatsData.AchievementTotalArtifactsSmeltedBonus
                AchievementType.DungeonsCompleted =>
                    Print("Checking Dungeons Completed")
                    set CompareStat = PlayerStatsData.TotalDungeonsCompleted
                    set Achievement = DungeonsCompletedAchievement
                    set AchievementBonus = PlayerStatsData.AchievementTotalDungeonsCompletedBonus
                _ =>
                    Print("Checking Nothing")

            var LastLevel : big_number = big_number{}

            # Compare the stat to the acheivement to see if it's done
            for(Index := 0..Achievement.Levels.Length - 1):
                if(Level := Achievement.Levels[Index]):
                    GreaterThan := BigNumber_GreaterThanOrEquals(CompareStat, Level)
                    if(GreaterThan = true):
                        set AchievementCount += 1.0
                        Print("Achievement: {Achievement.Name} | Index : {Index}")
                        set LastLevel = Level
            
            if(AchievementCount > AchievementBonus):
                HUDMessage.SetText(ToMessage("{BigNumber_FormatDisplay(LastLevel)} {Achievement.Name}"))
                HUDMessage.Show(InPlayer)
                PlayerStatsManager.UpdateAchievementBonusHarvestRate(InPlayer, AchievementCount, Achievement.Type)
                #Give Accolade on achievement unlock
                AchievementAccolade.Award(InPlayer)
