using { /Fortnite.com/Devices }
using { /Verse.org/Simulation }
using { /UnrealEngine.com/Temporary/Diagnostics }

Player_Afk_Data := class():
    var Seconds : int = 0
    var Minutes : int = 0
    var Hours : int = 0
    var ResourceAmount : big_number = big_number{}

V_AFKZoneManager := class(creative_device):
    @editable
    PlayerManager : ?V_PlayerManagerDevice = false

    @editable
    AreaVolume : volume_device = volume_device{}

    @editable
    AreaSwitch : switch_device = switch_device{}

    @editable
    TimePerResourceGive : float = 1.0

    @editable
    StrengthDividedNumber : float = 0.1

    #@editable
    #ExitHUDMessage : hud_message_device = hud_message_device{}

    @editable
    ExitHUDBackground : hud_message_device = hud_message_device{}

    @editable
    EntryHUDBackground : hud_message_device = hud_message_device{}

    #@editable
    #EntryHUDMessage : hud_message_device = hud_message_device{}

    var PlayerStatsManager: player_stats_manager = player_stats_manager{}
    var PlayerAfkData :  [agent]Player_Afk_Data = map{}

    ToMessage<localizes>(String : string) : message = "{String}"

    OnBegin<override>()<suspends>:void=
        AreaSwitch.TurnedOnEvent.Subscribe(PlayerEnteredAFKZone)

        if(bPlayerManager := PlayerManager?):
            set PlayerStatsManager = bPlayerManager.PlayerStatManager

    #Called when a player enters the AFK zone
    PlayerEnteredAFKZone(Agent : agent) : void=
        Print("Entered AFK Zone")
        if(set PlayerAfkData[Agent] = Player_Afk_Data{}):
        ShowEntryHUD(Agent)
        spawn:
            LoopingPlayerRewards(Agent)
            
    #Called when a player exits the AFK zone
    PlayerExitedAFKZone(Agent : agent) : void=
        Print("Left AFK Zone")
        ShowExitHUD(Agent)
        #ExitHUDBackground.Show(Agent)
        if(set PlayerAfkData[Agent] = Player_Afk_Data{}):

    #Called when player is in AFK Zone and provides resources based on strength
    LoopingPlayerRewards(Agent : agent)<suspends> : void=
        Sleep(0.1)

        loop:
            Sleep(TimePerResourceGive)

            if(AreaSwitch.GetCurrentState[Agent]):
                Print("In AFK Zone")
                if(Player := player[Agent]):
                    if(PlayerData := PlayerStatsDataMap[Player]):
                        PlayerUpgradeLevel := PlayerData.PrestigeAFKAreaUpgrade
                        PrestigeData := PlayerStatsManager.GetPlayerUpgrade(Agent, PlayerUpgradeType.Prestige_AFK)
                        AFKPercentage := PrestigeData.GetCurrentLevelAmount(PlayerUpgradeLevel)

                        var ResourceAmount : big_number = BigNumber_MultiplyWithAFloat(PlayerData.HarvestRatePerHit, (AFKPercentage/100.0))

                        if(FoundCard := SearchForCard(10, Player, PlayerStatsManager)?):
                            set ResourceAmount = GetBonusBigNumber(ResourceAmount, GetCardBonus(10, FoundCard.Level))

                        PlayerStatsManager.UpdatePlayerOreAmount(Player, PropResourceType.Stone, ResourceAmount)
                        # Get resource of player that was gained so far
                        if(AFKData :=  PlayerAfkData[Agent], ITimePerResourceGive := Ceil[TimePerResourceGive]):
                            var ISeconds : int = AFKData.Seconds + ITimePerResourceGive
                            var IMinutes : int = AFKData.Minutes
                            var IHours : int = AFKData.Hours
                            if (ISeconds > 59):
                                set IMinutes += 1
                                set ISeconds = 0
                            if (IMinutes > 59):
                                set IHours += 1
                                set IMinutes = 0
                            Print("{ISeconds}")
                            TotalAmountGained := BigNumber_Add(AFKData.ResourceAmount, ResourceAmount)
                            # Add gained resource and time to the data map
                            if (set  PlayerAfkData[Agent] = Player_Afk_Data{ResourceAmount := TotalAmountGained, Seconds := ISeconds, Minutes := IMinutes, Hours := IHours}){}
            else:
                PlayerExitedAFKZone(Agent)
                break

    ShowExitHUD(Agent : agent):void=
        if(AfkData := PlayerAfkData[Agent]):
            Seconds := MakeTimerText(AfkData.Seconds)
            Minutes := MakeTimerText(AfkData.Minutes)
            Hours := MakeTimerText(AfkData.Hours)
            ResourceAmount := BigNumber_FormatDisplay(AfkData.ResourceAmount)
            Message := ToMessage("{Hours}:{Minutes}:{Seconds}\n\n+{ResourceAmount}")
            ExitHUDBackground.SetText(Message)
        ExitHUDBackground.Show(Agent)

    ShowEntryHUD(Agent : agent):void=
        if(AfkData := PlayerAfkData[Agent]):
            if(Player := player[Agent]):
                if(PlayerData := PlayerStatsDataMap[Player]):
                    PlayerUpgradeLevel := PlayerData.PrestigeAFKAreaUpgrade
                    PrestigeData := PlayerStatsManager.GetPlayerUpgrade(Agent, PlayerUpgradeType.Prestige_AFK)
                    AFKPercentage := PrestigeData.GetCurrentLevelAmount(PlayerUpgradeLevel)
                    AFKPercentageStr := FormatFloatToNDP(AFKPercentage, 1)
                    Message := ToMessage("+{AFKPercentageStr}%") 
                    EntryHUDBackground.SetText(Message)
        EntryHUDBackground.Show(Agent)

    # Converting a float timer to text timer
    MakeTimerText(Timer : int):string=
        var TimeTxt : string = "00"
        if (Timer < 10):
            set TimeTxt = "0{Timer}"
        else:
            set TimeTxt = "{Timer}"
        return "{TimeTxt}"