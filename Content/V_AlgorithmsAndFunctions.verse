
using { /Fortnite.com/Devices }
using { /Verse.org/Simulation }
using { /UnrealEngine.com/Temporary/Diagnostics }

#Added
using { /Verse.org/Random }

#Previously had different checking methods, with enums we can easily identify these upgrade types in all verse scripts and it keeps all code consistent
VendorUpgradeResourceTypes := enum:
    Stone
    Null    #This is used when we want an empty value in some functions. It should always be last
    
HarvestRemoteUpgradeTypes := enum:
    Upgrade1
    Upgrade2
    Upgrade3
    Upgrade4
    Upgrade5
    Null    #This is used when we want an empty value in some functions. It should always be last

PropResourceType := enum:
    Stone
    Gold
    Diamond
    Copper
    Silver
    Malachite
    Spectrolite
    Obsidian
    PrestigeResource
    Coin    #This is not used as a resource but will be used for custom conditional button
    RandomOres
    Orb
    Dust
    Null    #This is used when we want an empty value in some functions. It should always be last

PlayerUpgradeType := enum:
    CriticalChance
    CriticalDamage
    UnlockMegaCritical
    MegaCritical
    CriticalCharge
    FinalHit
    OreQuantity
    Friends
    Damage
    Movement
    HealthRegen
    MaxHealth
    ExplosiveHit
    ExplosiveHitRange
    ExplosiveHitDamage
    ExplosiveHitCharges
    Altar
    AltarStrength
    AltarTime
    GlobalVendor
    Prestige_AFK
    Prestige_Ascension
    Prestige_DungeonMastery
    Player_PrestigeCore
    Prestige_ArtifactMastery
    Prestige_ArtifactSlots
    Prestige_AutoMiningPlot
    Prestige_AutoMiningCount
    Prestige_AutoMiningHitRate
    DungeonKingTokenAmount
    DungeonKingTokenGenerationSpeed
    Null

#This is the Title To Show, we only want each Vendor here. Do not put the upgrade types here.
PlayerUpgradeTitleToShow := enum:
    Critical
    FinalHit
    Friends
    Movement
    Health
    ExplosiveHit
    Altar
    Prestige_ArtifactMastery
    AutoMining
    DungeonKing
    AFKUpgrades
    DungeonMastery
    MegaCriticalHit
    MiningAscension
    OreUpgrades
    Null

GetBonusBigNumber(InNumber : big_number, Bonus : float):big_number=
    return BigNumber_Add(InNumber, BigNumber_MultiplyWithAFloat(InNumber, Bonus / 100.0))

Algorithms := class<concrete>:
    PlayerStatData : player_stats_manager = player_stats_manager{}

    SetHRUpgrade1Cost<localizes>(FloatValue:float) : message = "Cost = {FloatValue}"
    SetStoneVendorUpgradeCost<localizes>(FloatValue:float) : message = "Cost = {FloatValue}"

    #Function that gives you the cost of the next upgrade.
    CalculateCostOfNextUpgrade(BaseCost : big_number, CostIncreaseRate : big_number, CurrentUpgradeLevel : float):big_number=
        #Formula for cost of next upgrade
        CostIncreaseRatePowerOfCurrentUpgradeLevel := BigNumber_PowWithAFloat(CostIncreaseRate, CurrentUpgradeLevel)
        var CostOfNextUpgrade : big_number = BigNumber_Multiply(CostIncreaseRatePowerOfCurrentUpgradeLevel, BaseCost)

        return CostOfNextUpgrade

    #Tier System Multiplier Calculation. When we want to have an exponent multiplier to our base systems.
    TierSystemMultiplierCalculation(InAgent : agent, CurrentUpgradeLevel : float, HarvestRemote : logic, Vendor : logic, UpgradeIndex : int, UpgradeVendorEnum : VendorUpgradeResourceTypes):logic=
        var IncreaseTierMultiplier : logic = false
        var UpgradeAvaliable : logic = false
        if(CurrentPlayerStats := PlayerStatData.GetPlayerStats[InAgent]): 
            if(OldData := PlayerStatsDataMap[InAgent]):
                if(HarvestRemote = true):
                    Print("Harvest Remote Purchase")
                    if(UpgradesAvaliable := OldData.HRUpgradeTierMultiplierUpgradesAvaliable[UpgradeIndex]):
                        if(UpgradesAvaliable >= 1):
                            set UpgradeAvaliable = true
                            
                    if(UpgradeAvaliable = true):
                        set IncreaseTierMultiplier = true  
                        #PlayerStatData.ModifyTierMultiplerAmount(InAgent, Upgrade, false, true)   
                        #Reduce the amount of upgrades
                    else:
                        Print("No Tier Upgrades Avaliable")
                        set IncreaseTierMultiplier = false

                if(Vendor = true):
                    Print("Vendor Multiplier Purchase")
                    if(UpgradeVendorEnum = VendorUpgradeResourceTypes.Stone):
                        if(OldData.StoneVendorTierMultiplierUpgradesAvaliable >=1):
                            set UpgradeAvaliable = true
                            
                    if(UpgradeAvaliable = true):
                        set IncreaseTierMultiplier = true  
                    else:
                        Print("No Tier Upgrades Avaliable")
                        set IncreaseTierMultiplier = false

        return IncreaseTierMultiplier

    PurchasePlayerHarvestRateUpgrade(InAgent : agent, BaseCost : big_number, UpgradeIndex : int, InTierSystems : [][]float, HarvestBaseRate : big_number, HarvestRateIncrease : big_number, InPurchaseAmount : int)<suspends>:void=
        if(CurrentPlayerStats := PlayerStatData.GetPlayerStats[InAgent]): 
            if(OldData := PlayerStatsDataMap[InAgent]):
                if(PlayerUpgradeLevel := CurrentPlayerStats.HRUpgradeLevels[UpgradeIndex], PlayerTierMultiplier := CurrentPlayerStats.HRUpgradeTierMultipliers[UpgradeIndex]):
                    #Upon purchase if the HRUpgrade2TierMultiplier is at 0.0 let's increase it by 1
                    if(PlayerTierMultiplier = 0.0):
                        PlayerStatData.IncreaseRemoteUpgradeTierMultiplierOnPurchase(InAgent, UpgradeIndex)

                    #Check if the player has enough ore for the upgrade
                    StoneAmount := PlayerStatData.GetPlayerStoneOreAmount(InAgent)

                    AffordableUpgrades := CalculateAffordableUpgrades(StoneAmount, BaseCost, CurrentPlayerStats.HRCostIncreaseRate, PlayerUpgradeLevel, InPurchaseAmount)

                    #Check if we have enough ore for the next upgrade
                    if(AffordableUpgrades(0) > 0):
                        PlayerStatData.SpendPlayerOre(InAgent, AffordableUpgrades(1), false, PropResourceType.Stone)
                        #Upgrades your harvest rate level
                        PlayerStatData.UpdateHRUpgradesLevel(InAgent, UpgradeIndex, AffordableUpgrades(0))
                        #Update Harvest Rate - Everytime we purchase an upgrade we increase the players harvest rate per hit.
                        #We have an overall number that gets upgraded per remote purchase.
                        PlayerStatData.UpdateHarvestRatePerHit(InAgent, Self, UpgradeIndex, HarvestBaseRate, HarvestRateIncrease)

        if(CurrentPlayerStats := PlayerStatData.GetPlayerStats[InAgent]): 
            if(NewPlayerUpgradeLevel := CurrentPlayerStats.HRUpgradeLevels[UpgradeIndex], CurrentTierSystem := InTierSystems[UpgradeIndex]):
                #Loop through all tiers. The TierSystems is assigned earlier depending on what upgrade we want to check
                var TierToUse : int = 0
                for(Index -> TierSystem : CurrentTierSystem):
                    #Check if the upgrade we have purchased is the same as the tier system vlue required to unlock
                    AscendedUpgradeLevel := PlayerStatData.GetPlayerUpgradeLevel(InAgent, PlayerUpgradeType.Prestige_Ascension)
                        if(AscendedUpgradeLevel >= UpgradeIndex + 1):
                            # Ascended
                            if(TierMultiplier := CurrentPlayerStats.HRUpgradeTierMultipliers[UpgradeIndex]):
                                Print("{NewPlayerUpgradeLevel} >= {TierSystem} and {TierMultiplier} = {Pow(3.0, Index * 1.0)}")
                            if(Index = 0):
                                if(NewPlayerUpgradeLevel >= TierSystem and CurrentPlayerStats.HRUpgradeTierMultipliers[UpgradeIndex] = 1.0):
                                    #Increase the amount of upgrades we can apply each time the player reaches a Tier Multiplier unlock 
                                    PlayerStatData.ModifyTierMultiplerAmount(InAgent, UpgradeIndex, true, false)
                                    Print("increase Tier Modifier Amount")
                            else:
                                if(NewPlayerUpgradeLevel >= TierSystem and CurrentPlayerStats.HRUpgradeTierMultipliers[UpgradeIndex] = Pow(3.0, Index * 1.0)):
                                    #Increase the amount of upgrades we can apply each time the player reaches a Tier Multiplier unlock 
                                    PlayerStatData.ModifyTierMultiplerAmount(InAgent, UpgradeIndex, true, false)
                                    Print("increase Tier Modifier Amount")
                        else:
                            if(TierMultiplier := CurrentPlayerStats.HRUpgradeTierMultipliers[UpgradeIndex]):
                                Print("{NewPlayerUpgradeLevel} >= {TierSystem} and {TierMultiplier} = {Pow(2.0, Index * 1.0)}")
                                if(Index = 0):
                                    if(NewPlayerUpgradeLevel >= TierSystem and CurrentPlayerStats.HRUpgradeTierMultipliers[UpgradeIndex] = 1.0):
                                        #Increase the amount of upgrades we can apply each time the player reaches a Tier Multiplier unlock 
                                        PlayerStatData.ModifyTierMultiplerAmount(InAgent, UpgradeIndex, true, false)
                                        Print("increase Tier Modifier Amount")
                                else:
                                    if(NewPlayerUpgradeLevel >= TierSystem and CurrentPlayerStats.HRUpgradeTierMultipliers[UpgradeIndex] = Pow(2.0, Index * 1.0)):
                                        #Increase the amount of upgrades we can apply each time the player reaches a Tier Multiplier unlock 
                                        PlayerStatData.ModifyTierMultiplerAmount(InAgent, UpgradeIndex, true, false)
                                        Print("increase Tier Modifier Amount")
                    


    CalculateAffordableUpgrades(InCurrency : big_number, BaseCost : big_number, HarvestRateIncrease : big_number, InPlayerUpgradeLevel : float, InPurchaseAmount : int):tuple(int, big_number)=
        var AffordableUpgrades : int = 1
        var TotalCost : big_number = big_number{}
        loop:
            Cost := CalculateCostOfNextUpgrade(BaseCost, HarvestRateIncrease, InPlayerUpgradeLevel + AffordableUpgrades * 1.0)
            set TotalCost = BigNumber_Add(TotalCost, Cost)
            Affordable := BigNumber_GreaterThanOrEquals(InCurrency, TotalCost)
            if:
                not Affordable?
            then:
                break

            set AffordableUpgrades += 1

            if:
                AffordableUpgrades >= InPurchaseAmount
            then:
                return (InPurchaseAmount, TotalCost)


        return (AffordableUpgrades - 1, TotalCost)
        
    #Seperate Function to purchase Remote Multiplier Upgrades
    PurchaseTierMultiplierRemoteUpgrades(InAgent : agent, TierSystems : []float, UpgradeIndex : int, RateIncrease : big_number, BaseRate : big_number):void=
        if(CurrentPlayerStats := PlayerStatData.GetPlayerStats[InAgent]): 
            if(OldData := PlayerStatsDataMap[InAgent]):
                StoneAmount := PlayerStatData.GetPlayerStoneOreAmount(InAgent)
                #Applys Tier Multiplier if you hit certain upgrade levels
                #This block of code is for Harvest Rate Upgrades. Currently we have 4 types.
                if(UpgradeLevel := CurrentPlayerStats.HRUpgradeLevels[UpgradeIndex], TierMultiplier := CurrentPlayerStats.HRUpgradeTierMultipliers[UpgradeIndex]):
                    TierMultiplierCost := BigNumber_MultiplyWithAFloat(RateIncrease, TierMultiplier * 500)
                    BNCost := TierMultiplierCost
                    CanAfford := BigNumber_GreaterThanOrEquals(StoneAmount, BNCost)
                    #Check if we have enough ore for the next upgrade
                    if(CanAfford = true):
                        #Update Harvest Rate Upgrade 1. Currently known as "Strength" Upgrade.
                        ApplyTierMultiplier := TierSystemMultiplierCalculation(InAgent, UpgradeLevel, true, false, UpgradeIndex, VendorUpgradeResourceTypes.Null)
                        if(ApplyTierMultiplier = true):
                            #Spend the players coins
                            PlayerStatData.SpendPlayerOre(InAgent, BNCost, false, PropResourceType.Stone)
                            PlayerStatData.ModifyTierMultiplerAmount(InAgent, UpgradeIndex, false, true)

                            AscendedUpgradeLevel := PlayerStatData.GetPlayerUpgradeLevel(InAgent, PlayerUpgradeType.Prestige_Ascension)
                            if(AscendedUpgradeLevel >= UpgradeIndex + 1):
                                PlayerStatData.UpdateHarvestRateTierSystem(InAgent, UpgradeIndex, true)
                            else:
                                PlayerStatData.UpdateHarvestRateTierSystem(InAgent, UpgradeIndex, false)

                            PlayerStatData.UpdateHarvestRatePerHit(InAgent, Self, UpgradeIndex, BaseRate, RateIncrease)
    
    #Converted to work with all ORE
    <#TradeOreWithVendor(InAgent : agent, BaseCost : float, TierSystems : []float, VendorBaseRate : float, VendorRateIncrease : float, VendorCostIncreaseRate : float, Resource : VendorUpgradeResourceTypes, SellOre : logic, UpgradingVendor : logic, UpgradingVendorTierMultiplier : logic, InPurchaseAmount : int):void=
        if(CurrentPlayerStats := PlayerStatData.GetPlayerStats[InAgent]):
            #Need to add a check if the player has enough ore for the trade [Check harvest for that]
            var CostOfNextUpgrade : big_number = big_number{}

            #Trade Stone for Coin
            if(Resource = VendorUpgradeResourceTypes.Stone):
                OreAmount := PlayerStatData.GetPlayerStoneOreAmount(InAgent)
                set CostOfNextUpgrade = CalculateCostOfNextUpgrade(BaseCost, VendorCostIncreaseRate,  CurrentPlayerStats.StoneVendorUpgradeLevel) 
                if(SellOre = true):
                    SellOreToVendor(InAgent, VendorRateIncrease, VendorBaseRate, CostOfNextUpgrade, PropResourceType.Stone)
                if(UpgradingVendor = true):
                    for(X := 1..InPurchaseAmount):
                        if(UpdatedPlayerStats := PlayerStatData.GetPlayerStats[InAgent]):
                            set CostOfNextUpgrade = CalculateCostOfNextUpgrade(BaseCost, VendorCostIncreaseRate,  UpdatedPlayerStats.StoneVendorUpgradeLevel) 
                            UpgradeVendor(InAgent, VendorRateIncrease, CostOfNextUpgrade, VendorBaseRate, Resource)
                            #Increase the amount of tier multiplier upgrades we can purchase as we press the purchase button
                            for(TierSystem : TierSystems):
                                #Check if the upgrade we have purchased is the same as the tier system vlue required to unlock
                                if(UpdatedPlayerStats.StoneVendorUpgradeLevel = TierSystem - 1.0):
                                    PlayerStatData.ModifyVendorMultiplierAmount(InAgent,1,true,false)
                if(UpgradingVendorTierMultiplier = true):
                    UpgradeVendorTierMultiplier(InAgent, TierSystems, PropResourceType.Stone, VendorRateIncrease, VendorBaseRate)
            if(Resource = VendorUpgradeResourceTypes.Silver):
                Print("Trade Silver")
                OreAmount := PlayerStatData.GetPlayerSilverOreAmount(InAgent)
                set CostOfNextUpgrade = CalculateCostOfNextUpgrade(BaseCost,VendorCostIncreaseRate,  CurrentPlayerStats.SilverVendorUpgradeLevel) 
                if(SellOre = true):
                    SellOreToVendor(InAgent, VendorRateIncrease, VendorBaseRate, CostOfNextUpgrade, PropResourceType.Silver)
                if(UpgradingVendor = true):
                    for(X := 1..InPurchaseAmount):
                        if(UpdatedPlayerStats := PlayerStatData.GetPlayerStats[InAgent]):
                            set CostOfNextUpgrade = CalculateCostOfNextUpgrade(BaseCost,VendorCostIncreaseRate,  UpdatedPlayerStats.SilverVendorUpgradeLevel) 
                            UpgradeVendor(InAgent, VendorRateIncrease, CostOfNextUpgrade, VendorBaseRate, Resource)
                            #Increase the amount of tier multiplier upgrades we can purchase as we press the purchase button
                            for(TierSystem : TierSystems):
                                #Check if the upgrade we have purchased is the same as the tier system vlue required to unlock
                                if(UpdatedPlayerStats.SilverVendorUpgradeLevel = TierSystem - 1.0):
                                    PlayerStatData.ModifyVendorMultiplierAmount(InAgent,2,true,false)  
                if(UpgradingVendorTierMultiplier = true):
                    UpgradeVendorTierMultiplier(InAgent, TierSystems, PropResourceType.Silver, VendorRateIncrease, VendorBaseRate)

            if(Resource = VendorUpgradeResourceTypes.Gold):
                Print("Trade Gold")
                OreAmount := PlayerStatData.GetPlayerGoldOreAmount(InAgent)
                set CostOfNextUpgrade = CalculateCostOfNextUpgrade(BaseCost,VendorCostIncreaseRate,  CurrentPlayerStats.GoldVendorUpgradeLevel) 
                if(SellOre = true):
                    SellOreToVendor(InAgent, VendorRateIncrease, VendorBaseRate, CostOfNextUpgrade, PropResourceType.Gold)
                if(UpgradingVendor = true):
                    for(X := 1..InPurchaseAmount):
                        if(UpdatedPlayerStats := PlayerStatData.GetPlayerStats[InAgent]):
                            set CostOfNextUpgrade = CalculateCostOfNextUpgrade(BaseCost,VendorCostIncreaseRate,  UpdatedPlayerStats.GoldVendorUpgradeLevel) 
                            UpgradeVendor(InAgent, VendorRateIncrease, CostOfNextUpgrade, VendorBaseRate, Resource)
                            #Increase the amount of tier multiplier upgrades we can purchase as we press the purchase button
                            for(TierSystem : TierSystems):
                                #Check if the upgrade we have purchased is the same as the tier system vlue required to unlock
                                if(UpdatedPlayerStats.GoldVendorUpgradeLevel = TierSystem - 1.0):
                                    PlayerStatData.ModifyVendorMultiplierAmount(InAgent,3,true,false)  
                if(UpgradingVendorTierMultiplier = true):
                    UpgradeVendorTierMultiplier(InAgent, TierSystems, PropResourceType.Gold, VendorRateIncrease, VendorBaseRate)

            if(Resource = VendorUpgradeResourceTypes.Diamond):
                Print("Trade Diamond")
                OreAmount := PlayerStatData.GetPlayerDiamondOreAmount(InAgent)
                set CostOfNextUpgrade = CalculateCostOfNextUpgrade(BaseCost,VendorCostIncreaseRate,  CurrentPlayerStats.DiamondVendorUpgradeLevel) 
                if(SellOre = true):
                    SellOreToVendor(InAgent, VendorRateIncrease, VendorBaseRate, CostOfNextUpgrade, PropResourceType.Diamond)
                if(UpgradingVendor = true):
                    for(X := 1..InPurchaseAmount):
                        if(UpdatedPlayerStats := PlayerStatData.GetPlayerStats[InAgent]):
                            set CostOfNextUpgrade = CalculateCostOfNextUpgrade(BaseCost,VendorCostIncreaseRate,  UpdatedPlayerStats.DiamondVendorUpgradeLevel) 
                            UpgradeVendor(InAgent, VendorRateIncrease, CostOfNextUpgrade, VendorBaseRate, Resource)
                            #Increase the amount of tier multiplier upgrades we can purchase as we press the purchase button
                            for(TierSystem : TierSystems):
                                #Check if the upgrade we have purchased is the same as the tier system vlue required to unlock
                                if(UpdatedPlayerStats.DiamondVendorUpgradeLevel = TierSystem - 1.0):
                                    PlayerStatData.ModifyVendorMultiplierAmount(InAgent,4,true,false)  
                if(UpgradingVendorTierMultiplier = true):
                    UpgradeVendorTierMultiplier(InAgent, TierSystems, PropResourceType.Diamond, VendorRateIncrease, VendorBaseRate)#>

    #Converted to support all ORE types
    <#SellOreToVendor(InAgent : agent, VendorRateIncrease : float, VendorBaseRate : float, CostOfNextUpgrade : big_number, InTradeResourceType : PropResourceType):void=
            var CalculatedCoinGainOnTrade : big_number = big_number{}
            var OreAmount : big_number = big_number{}
            if(Player := player[InAgent]):
                if(OldData := PlayerStatsDataMap[Player]):
                    if(InTradeResourceType = PropResourceType.Stone):
                        set CalculatedCoinGainOnTrade = CalculateHarvestRateOnHit(VendorRateIncrease, OldData.StoneVendorUpgradeLevel , VendorBaseRate, OldData.StoneVendorTierMultiplier)        
                        set OreAmount = PlayerStatData.GetPlayerStoneOreAmount(InAgent)
                    if(InTradeResourceType = PropResourceType.Silver):
                        Print("Selling Silver Ore")
                        set CalculatedCoinGainOnTrade = CalculateHarvestRateOnHit(VendorRateIncrease, OldData.SilverVendorUpgradeLevel , VendorBaseRate, OldData.SilverVendorTierMultiplier)        
                        set OreAmount = PlayerStatData.GetPlayerSilverOreAmount(InAgent)
                    if(InTradeResourceType = PropResourceType.Gold):
                        set CalculatedCoinGainOnTrade = CalculateHarvestRateOnHit(VendorRateIncrease, OldData.GoldVendorUpgradeLevel , VendorBaseRate, OldData.GoldVendorTierMultiplier)        
                        set OreAmount = PlayerStatData.GetPlayerGoldOreAmount(InAgent)
                    if(InTradeResourceType = PropResourceType.Diamond):
                        set CalculatedCoinGainOnTrade = CalculateHarvestRateOnHit(VendorRateIncrease, OldData.DiamondVendorUpgradeLevel , VendorBaseRate, OldData.DiamondVendorTierMultiplier)        
                        set OreAmount = PlayerStatData.GetPlayerDiamondOreAmount(InAgent)
                    
                    BNNegative := BigNumber_ConvertFloatIntoBigNumber(-0.1)
                    CanAfford := BigNumber_GreaterThanOrEquals(OreAmount, BNNegative)
                    if(CanAfford = true):
                        #We have to update gold before selling. If we spend it first the player won't have the proper ore amount.
                        TotalOreTradeAmount := BigNumber_Multiply(OreAmount, CalculatedCoinGainOnTrade)
                        PlayerStatData.UpdatePlayerCoinAmount(InAgent, InTradeResourceType)
                        #After updating the coin amount we then spend it
                        PlayerStatData.SpendPlayerOre(InAgent, TotalOreTradeAmount, true, InTradeResourceType)#>
    
    #Converted to support all ORE types
    <#UpgradeVendor(InAgent : agent, VendorRateIncrease : float, CostOfNextUpgrade : big_number, VendorBaseRate : float, Resource : VendorUpgradeResourceTypes):void=
        CoinAmount := PlayerStatData.GetPlayerCoinAmount(InAgent)
        #Check if we have enough coins for the next upgrade.
        CanAfford := BigNumber_GreaterThanOrEquals(CoinAmount, CostOfNextUpgrade)
        if(CanAfford = true):
            Print("Upgrade Vendor")
            #Update Upgrade Level
            PlayerStatData.UpdateVendorUpgradesLevel(InAgent,Resource)
            #Spend the gold, if you do this later you'll go into negatives
            PlayerStatData.SpendPlayerCoins(InAgent, CostOfNextUpgrade)
            #Modify Gold Rate Per Trade
            PlayerStatData.UpdateGoldRatePerTrade(InAgent, Self, VendorRateIncrease, VendorBaseRate, Resource)#>
    
    #Converted to support all ORE types
    <#UpgradeVendorTierMultiplier(InAgent : agent, TierSystems : []float, InTradeResourceType : PropResourceType, VendorRateIncrease : float, VendorBaseRate : float):void=
        Print("Update Tier Multiplier")
        var ApplyTierMultiplier : logic = false
        var Upgrade : int = 1
        var VendorResource : VendorUpgradeResourceTypes = VendorUpgradeResourceTypes.Stone

        CoinAmount := PlayerStatData.GetPlayerCoinAmount(InAgent)

        #TODO: Calculate cost of Tier Multiplier and add it here
        if(CurrentPlayerStats := PlayerStatData.GetPlayerStats[InAgent]): 
            var TierMultiplierCostOfNextUpgrade : float = 0.0
            if(OldData := PlayerStatsDataMap[InAgent]):
                    #Make sure the player can afford the upgrade
                    if(InTradeResourceType = PropResourceType.Stone):
                        set ApplyTierMultiplier = TierSystemMultiplierCalculation(InAgent, CurrentPlayerStats.StoneVendorUpgradeLevel, false, true, 0, VendorUpgradeResourceTypes.Stone)
                        set Upgrade = 1
                        set VendorResource = VendorUpgradeResourceTypes.Stone
                        set TierMultiplierCostOfNextUpgrade = VendorRateIncrease * OldData.StoneVendorTierMultiplier
                        Print("Tier Upgrade Cost = {TierMultiplierCostOfNextUpgrade}")
                    if(InTradeResourceType = PropResourceType.Silver):
                        set ApplyTierMultiplier = TierSystemMultiplierCalculation(InAgent, CurrentPlayerStats.SilverVendorUpgradeLevel, false, true, 1, VendorUpgradeResourceTypes.Stone)
                        set Upgrade = 2
                        set VendorResource = VendorUpgradeResourceTypes.Silver
                        set TierMultiplierCostOfNextUpgrade = VendorRateIncrease * OldData.SilverVendorTierMultiplier
                        Print("Tier Upgrade Cost = {TierMultiplierCostOfNextUpgrade}")
                    if(InTradeResourceType = PropResourceType.Gold):
                        set ApplyTierMultiplier = TierSystemMultiplierCalculation(InAgent, CurrentPlayerStats.GoldVendorUpgradeLevel, false, true, 2, VendorUpgradeResourceTypes.Stone)
                        set Upgrade = 3
                        set VendorResource = VendorUpgradeResourceTypes.Gold
                        set TierMultiplierCostOfNextUpgrade = VendorRateIncrease * OldData.GoldVendorTierMultiplier
                        Print("Tier Upgrade Cost = {TierMultiplierCostOfNextUpgrade}")
                    if(InTradeResourceType = PropResourceType.Diamond):
                        set ApplyTierMultiplier = TierSystemMultiplierCalculation(InAgent, CurrentPlayerStats.DiamondVendorUpgradeLevel, false, true, 3, VendorUpgradeResourceTypes.Stone)
                        set Upgrade = 4
                        set VendorResource = VendorUpgradeResourceTypes.Diamond
                        set TierMultiplierCostOfNextUpgrade = VendorRateIncrease * OldData.DiamondVendorTierMultiplier
                        Print("Tier Upgrade Cost = {TierMultiplierCostOfNextUpgrade}")

                    BNCost := BigNumber_ConvertFloatIntoBigNumber(TierMultiplierCostOfNextUpgrade)
                    #Check if we have enough coins for the next upgrade.
                    CanAfford := BigNumber_GreaterThanOrEquals(CoinAmount, BNCost)
                    if(ApplyTierMultiplier = true and CanAfford = true):
                        #Upgrade the tier multiplier
                        PlayerStatData.UpdateVendorTierMultiplier(InAgent, InTradeResourceType)
                        #Added to update the gold rate and try fix it
                        PlayerStatData.UpdateGoldRatePerTrade(InAgent, Self, VendorRateIncrease, VendorBaseRate, VendorResource)
                        #Reduce the amount of upgrades we can spend
                        PlayerStatData.ModifyVendorMultiplierAmount(InAgent,Upgrade,false,true)
                        PlayerStatData.SpendPlayerCoins(InAgent, BNCost)
                        Print("Apply multiplier to the vendor")#>

    #Calculation for extra power a purchase will give you in the Remote Upgrade
    
    #Function is here but calling it  in PlayerStatsData for some reason caused a crash after adding <transacts>
    CalculateHarvestRateOnHit(RateIncrease : big_number, UpgradeLevel : float, BaseRate : big_number, TierMultiplier : float):big_number=
        #HarvestRateOnHit := ((RateIncrease * UpgradeLevel) + BaseRate) * TierMultiplier
        HarvestRateOnHit := BigNumber_MultiplyWithAFloat(BigNumber_MultiplyWithAFloat(RateIncrease, UpgradeLevel), TierMultiplier)
        
        #Print("Calculated Harvest Rate On Hit: {HarvestRateOnHit}")
        return HarvestRateOnHit
    
    CalculateHarvestRateOnHitForUI(RateIncrease : big_number, UpgradeLevel : float, BaseRate : big_number, TierMultiplier : float):big_number=
        #HarvestRateOnHit := ((RateIncrease * UpgradeLevel) + BaseRate) * TierMultiplier
        #Attempt without the + Base Rate
        var HarvestRateOnHit : big_number = big_number{}
        if(TierMultiplier = 0.0):
            set  HarvestRateOnHit = BigNumber_MultiplyWithAFloat(BigNumber_MultiplyWithAFloat(RateIncrease, UpgradeLevel), 1.0)
        else:
            set  HarvestRateOnHit = BigNumber_MultiplyWithAFloat(BigNumber_MultiplyWithAFloat(RateIncrease, UpgradeLevel), TierMultiplier)
        
        #Print("Calculated Harvest Rate On Hit: {HarvestRateOnHit}")
        return HarvestRateOnHit

V_AlgorithmsAndFunctions := class(creative_device):
    var AlgorithmAndFunctionClassReference : Algorithms = Algorithms{}
      # Runs when the device is started in a running game
      OnBegin<override>()<suspends>:void={}
