
using { /Fortnite.com/Devices }
using { /Verse.org/Simulation }
using { /UnrealEngine.com/Temporary/Diagnostics }
using { /Verse.org/Random }
using { /Verse.org/Assets }
using { /Verse.org/Simulation/Tags }

DebugPrintArtifactStats(InArtifactData : Artifact_Data):void=
    case(InArtifactData.Rarity):
        ArtifactRarity.MushroomCommon =>
            Print("ARTIFACT | RARITY : MUSH COMMON | EVOLUTION : {InArtifactData.Prestige} | LEVEL : {InArtifactData.Level} | PRESTIGE : {InArtifactData.Prestige}")
        ArtifactRarity.MushroomUncommon =>
            Print("ARTIFACT | RARITY : MUSH UNCOMMON | EVOLUTION : {InArtifactData.Prestige} | LEVEL : {InArtifactData.Level} | PRESTIGE : {InArtifactData.Prestige}")
        ArtifactRarity.MushroomRare =>
            Print("ARTIFACT | RARITY : MUSH RARE | EVOLUTION : {InArtifactData.Prestige} | LEVEL : {InArtifactData.Level} | PRESTIGE : {InArtifactData.Prestige}")
        ArtifactRarity.MushroomEpic =>
            Print("ARTIFACT | RARITY : MUSH EPIC | EVOLUTION : {InArtifactData.Prestige} | LEVEL : {InArtifactData.Level} | PRESTIGE : {InArtifactData.Prestige}")
        ArtifactRarity.MushroomLegendary =>
            Print("ARTIFACT | RARITY : MUSH LEGENDARY | EVOLUTION : {InArtifactData.Prestige} | LEVEL : {InArtifactData.Level} | PRESTIGE : {InArtifactData.Prestige}")
        ArtifactRarity.LavaCommon =>
            Print("ARTIFACT | RARITY : LAVA COMMON | EVOLUTION : {InArtifactData.Prestige} | LEVEL : {InArtifactData.Level} | PRESTIGE : {InArtifactData.Prestige}")
        ArtifactRarity.LavaUncommon =>
            Print("ARTIFACT | RARITY : LAVA UNCOMMON | EVOLUTION : {InArtifactData.Prestige} | LEVEL : {InArtifactData.Level} | PRESTIGE : {InArtifactData.Prestige}")
        ArtifactRarity.LavaRare =>
            Print("ARTIFACT | RARITY : LAVA RARE | EVOLUTION : {InArtifactData.Prestige} | LEVEL : {InArtifactData.Level} | PRESTIGE : {InArtifactData.Prestige}")
        ArtifactRarity.LavaEpic =>
            Print("ARTIFACT | RARITY : LAVA EPIC | EVOLUTION : {InArtifactData.Prestige} | LEVEL : {InArtifactData.Level} | PRESTIGE : {InArtifactData.Prestige}")
        ArtifactRarity.LavaLegendary =>
            Print("ARTIFACT | RARITY : LAVA LEGENDARY | EVOLUTION : {InArtifactData.Prestige} | LEVEL : {InArtifactData.Level} | PRESTIGE : {InArtifactData.Prestige}")
        ArtifactRarity.UnderworldCommon =>
            Print("ARTIFACT | RARITY : UNDER COMMON | EVOLUTION : {InArtifactData.Prestige} | LEVEL : {InArtifactData.Level} | PRESTIGE : {InArtifactData.Prestige}")
        ArtifactRarity.UnderworldUncommon =>
            Print("ARTIFACT | RARITY : UNDER UNCOMMON | EVOLUTION : {InArtifactData.Prestige} | LEVEL : {InArtifactData.Level} | PRESTIGE : {InArtifactData.Prestige}")
        ArtifactRarity.UnderworldRare =>
            Print("ARTIFACT | RARITY : UNDER RARE | EVOLUTION : {InArtifactData.Prestige} | LEVEL : {InArtifactData.Level} | PRESTIGE : {InArtifactData.Prestige}")
        ArtifactRarity.UnderworldEpic =>
            Print("ARTIFACT | RARITY : UNDER EPIC | EVOLUTION : {InArtifactData.Prestige} | LEVEL : {InArtifactData.Level} | PRESTIGE : {InArtifactData.Prestige}")
        ArtifactRarity.UnderworldLegendary =>
            Print("ARTIFACT | RARITY : UNDER LEGENDARY | EVOLUTION : {InArtifactData.Prestige} | LEVEL : {InArtifactData.Level} | PRESTIGE : {InArtifactData.Prestige}")
        ArtifactRarity.Null =>
            Print("ARTIFACT | RARITY : NULL | EVOLUTION : {InArtifactData.Prestige} | LEVEL : {InArtifactData.Level} | PRESTIGE : {InArtifactData.Prestige}")

GetResourceFromRarity(InArtifactRarity : ArtifactRarity):PropResourceType=
    case(InArtifactRarity):
        ArtifactRarity.MushroomCommon =>
            return PropResourceType.Copper
        ArtifactRarity.MushroomUncommon =>
            return PropResourceType.Silver
        ArtifactRarity.MushroomRare =>
            return PropResourceType.Malachite
        ArtifactRarity.MushroomEpic =>
            return PropResourceType.Spectrolite
        ArtifactRarity.MushroomLegendary =>
            return PropResourceType.Obsidian
        ArtifactRarity.LavaCommon =>
            return PropResourceType.Copper
        ArtifactRarity.LavaUncommon =>
            return PropResourceType.Silver
        ArtifactRarity.LavaRare =>
            return PropResourceType.Malachite
        ArtifactRarity.LavaEpic =>
            return PropResourceType.Spectrolite
        ArtifactRarity.LavaLegendary =>
            return PropResourceType.Obsidian
        ArtifactRarity.UnderworldCommon =>
            return PropResourceType.Copper
        ArtifactRarity.UnderworldUncommon =>
            return PropResourceType.Silver
        ArtifactRarity.UnderworldRare =>
            return PropResourceType.Malachite
        ArtifactRarity.UnderworldEpic =>
            return PropResourceType.Spectrolite
        ArtifactRarity.UnderworldLegendary =>
            return PropResourceType.Obsidian
        _ =>
            return PropResourceType.Copper

GetUIArtifactSmeltData(InResourceType : PropResourceType, InBiome : Biome):UIArtifactSmeltData=
    case(InResourceType):
        PropResourceType.Copper =>
            case(InBiome):
                Biome.Mushroom =>
                    return UIArtifactSmeltData:
                        StoneCost := big_number{Value := 2.0, Exponent := 7}
                        ResourceCost := big_number{Value := 5.0, Exponent := 0}
                        ResourceImage := VerseExposed.T_CopperOreStroke
                        RarityString := "Common - Rare"
                Biome.Lava =>
                    return UIArtifactSmeltData:
                        StoneCost := big_number{Value := 2.5, Exponent := 16}
                        ResourceCost := big_number{Value := 5.0, Exponent := 1}
                        ResourceImage := VerseExposed.T_CopperOreStroke
                        RarityString := "Common - Rare"
                Biome.Underworld =>
                    return UIArtifactSmeltData:
                        StoneCost := big_number{Value := 2.0, Exponent := 24}
                        ResourceCost := big_number{Value := 1.5, Exponent := 2}
                        ResourceImage := VerseExposed.T_CopperOreStroke
                        RarityString := "Common - Rare"
        PropResourceType.Silver =>
            case(InBiome):
                Biome.Mushroom =>
                    return UIArtifactSmeltData:
                        StoneCost := big_number{Value := 4.0, Exponent := 7}
                        ResourceCost := big_number{Value := 5.0, Exponent := 0}
                        ResourceImage := VerseExposed.T_SilverOre
                        RarityString := "Common - Epic"
                Biome.Lava =>
                    return UIArtifactSmeltData:
                        StoneCost := big_number{Value := 5.0, Exponent := 16}
                        ResourceCost := big_number{Value := 5.0, Exponent := 1}
                        ResourceImage := VerseExposed.T_SilverOre
                        RarityString := "Common - Epic"
                Biome.Underworld =>
                    return UIArtifactSmeltData:
                        StoneCost := big_number{Value := 4.5, Exponent := 24}
                        ResourceCost := big_number{Value := 1.5, Exponent := 2}
                        ResourceImage := VerseExposed.T_SilverOre
                        RarityString := "Common - Epic"
        PropResourceType.Malachite =>
            case(InBiome):
                Biome.Mushroom =>
                    return UIArtifactSmeltData:
                        StoneCost := big_number{Value := 8.0, Exponent := 7}
                        ResourceCost := big_number{Value := 5.0, Exponent := 0}
                        ResourceImage := VerseExposed.T_MalachiteOreStroke
                        RarityString := "Uncommon - Epic"
                Biome.Lava =>
                    return UIArtifactSmeltData:
                        StoneCost := big_number{Value := 1.0, Exponent := 17}
                        ResourceCost := big_number{Value := 5.0, Exponent := 1}
                        ResourceImage := VerseExposed.T_MalachiteOreStroke
                        RarityString := "Uncommon - Epic"
                Biome.Underworld =>
                    return UIArtifactSmeltData:
                        StoneCost := big_number{Value := 8.0, Exponent := 24}
                        ResourceCost := big_number{Value := 1.5, Exponent := 2}
                        ResourceImage := VerseExposed.T_MalachiteOreStroke
                        RarityString := "Uncommon - Epic"
        PropResourceType.Spectrolite =>
            case(InBiome):
                Biome.Mushroom =>
                    return UIArtifactSmeltData:
                        StoneCost := big_number{Value := 1.5, Exponent := 8}
                        ResourceCost := big_number{Value := 5.0, Exponent := 0}
                        ResourceImage := VerseExposed.T_SpectroliteOreStroke
                        RarityString := "Rare - Legendary"
                Biome.Lava =>
                    return UIArtifactSmeltData:
                        StoneCost := big_number{Value := 2.0, Exponent := 17}
                        ResourceCost := big_number{Value := 5.0, Exponent := 1}
                        ResourceImage := VerseExposed.T_SpectroliteOreStroke
                        RarityString := "Rare - Legendary"
                Biome.Underworld =>
                    return UIArtifactSmeltData:
                        StoneCost := big_number{Value := 2.0, Exponent := 25}
                        ResourceCost := big_number{Value := 1.5, Exponent := 2}
                        ResourceImage := VerseExposed.T_SpectroliteOreStroke
                        RarityString := "Rare - Legendary"
        PropResourceType.Obsidian =>
            case(InBiome):
                Biome.Mushroom =>
                    return UIArtifactSmeltData:
                        StoneCost := big_number{Value := 3.0, Exponent := 8}
                        ResourceCost := big_number{Value := 5.0, Exponent := 0}
                        ResourceImage := VerseExposed.T_ObsidianOreStroke
                        RarityString := "Epic - Legendary"
                Biome.Lava =>
                    return UIArtifactSmeltData:
                        StoneCost := big_number{Value := 5.0, Exponent := 17}
                        ResourceCost := big_number{Value := 5.0, Exponent := 1}
                        ResourceImage := VerseExposed.T_ObsidianOreStroke
                        RarityString := "Epic - Legendary"
                Biome.Underworld =>
                    return UIArtifactSmeltData:
                        StoneCost := big_number{Value := 5.0, Exponent := 25}
                        ResourceCost := big_number{Value := 1.5, Exponent := 2}
                        ResourceImage := VerseExposed.T_ObsidianOreStroke
                        RarityString := "Epic - Legendary"
        _ =>
            Print("### ARTIFACT ORE TYPE NOT SETUP IN GetArtifactSmeltData ###")
            return UIArtifactSmeltData:
                StoneCost := big_number{Value := 0.0, Exponent := 0}
                ResourceCost := big_number{Value := 0.0, Exponent := 0}
                ResourceImage := VerseExposed.T_Stone
                RarityString := "Error"

GetArtifactPrestigeData(InResourceType : PropResourceType, InBiome : Biome):ArtifactPrestigeData=
    case(InResourceType):
        PropResourceType.Copper =>
            case(InBiome):
                Biome.Mushroom =>
                    return ArtifactPrestigeData:
                        StoneCost := big_number{Value := 2.0, Exponent := 7}
                        ResourceCost := big_number{Value := 5.0, Exponent := 0}
                        ResourceType := PropResourceType.Copper
                        ResourceImage := VerseExposed.T_CopperOreStroke
                Biome.Lava =>
                    return ArtifactPrestigeData:
                        StoneCost := big_number{Value := 2.5, Exponent := 16}
                        ResourceCost := big_number{Value := 5.0, Exponent := 1}
                        ResourceType := PropResourceType.Copper
                        ResourceImage := VerseExposed.T_CopperOreStroke
                Biome.Underworld =>
                    return ArtifactPrestigeData:
                        StoneCost := big_number{Value := 2.0, Exponent := 24}
                        ResourceCost := big_number{Value := 1.5, Exponent := 2}
                        ResourceType := PropResourceType.Copper
                        ResourceImage := VerseExposed.T_CopperOreStroke
        PropResourceType.Silver =>
            case(InBiome):
                Biome.Mushroom =>
                    return ArtifactPrestigeData:
                        StoneCost := big_number{Value := 4.0, Exponent := 7}
                        ResourceCost := big_number{Value := 5.0, Exponent := 0}
                        ResourceType := PropResourceType.Silver
                        ResourceImage := VerseExposed.T_SilverOre
                Biome.Lava =>
                    return ArtifactPrestigeData:
                        StoneCost := big_number{Value := 5.0, Exponent := 16}
                        ResourceCost := big_number{Value := 5.0, Exponent := 1}
                        ResourceType := PropResourceType.Silver
                        ResourceImage := VerseExposed.T_SilverOre
                Biome.Underworld =>
                    return ArtifactPrestigeData:
                        StoneCost := big_number{Value := 4.5, Exponent := 24}
                        ResourceCost := big_number{Value := 1.5, Exponent := 2}
                        ResourceType := PropResourceType.Silver
                        ResourceImage := VerseExposed.T_SilverOre
        PropResourceType.Malachite =>
            case(InBiome):
                Biome.Mushroom =>
                    return ArtifactPrestigeData:
                        StoneCost := big_number{Value := 8.0, Exponent := 7}
                        ResourceCost := big_number{Value := 1.0, Exponent := 0}
                        ResourceType := PropResourceType.Malachite
                        ResourceImage := VerseExposed.T_MalachiteOreStroke
                Biome.Lava =>
                    return ArtifactPrestigeData:
                        StoneCost := big_number{Value := 1.0, Exponent := 17}
                        ResourceCost := big_number{Value := 5.0, Exponent := 1}
                        ResourceType := PropResourceType.Malachite
                        ResourceImage := VerseExposed.T_MalachiteOreStroke
                Biome.Underworld =>
                    return ArtifactPrestigeData:
                        StoneCost := big_number{Value := 8.0, Exponent := 24}
                        ResourceCost := big_number{Value := 1.5, Exponent := 2}
                        ResourceType := PropResourceType.Malachite
                        ResourceImage := VerseExposed.T_MalachiteOreStroke
        PropResourceType.Spectrolite =>
            case(InBiome):
                Biome.Mushroom =>
                    return ArtifactPrestigeData:
                        StoneCost := big_number{Value := 1.5, Exponent := 8}
                        ResourceCost := big_number{Value := 1.0, Exponent := 0}
                        ResourceType := PropResourceType.Spectrolite
                        ResourceImage := VerseExposed.T_SpectroliteOreStroke
                Biome.Lava =>
                    return ArtifactPrestigeData:
                        StoneCost := big_number{Value := 2.0, Exponent := 17}
                        ResourceCost := big_number{Value := 5.0, Exponent := 1}
                        ResourceType := PropResourceType.Spectrolite
                        ResourceImage := VerseExposed.T_SpectroliteOreStroke
                Biome.Underworld =>
                    return ArtifactPrestigeData:
                        StoneCost := big_number{Value := 2.0, Exponent := 25}
                        ResourceCost := big_number{Value := 1.5, Exponent := 2}
                        ResourceType := PropResourceType.Spectrolite
                        ResourceImage := VerseExposed.T_SpectroliteOreStroke
        PropResourceType.Obsidian =>
            case(InBiome):
                Biome.Mushroom =>
                    return ArtifactPrestigeData:
                        StoneCost := big_number{Value := 3.0, Exponent := 8}
                        ResourceCost := big_number{Value := 1.0, Exponent := 0}
                        ResourceType := PropResourceType.Obsidian
                        ResourceImage := VerseExposed.T_ObsidianOreStroke
                Biome.Lava =>
                    return ArtifactPrestigeData:
                        StoneCost := big_number{Value := 5.0, Exponent := 17}
                        ResourceCost := big_number{Value := 5.0, Exponent := 1}
                        ResourceType := PropResourceType.Obsidian
                        ResourceImage := VerseExposed.T_ObsidianOreStroke
                Biome.Underworld =>
                    return ArtifactPrestigeData:
                        StoneCost := big_number{Value := 5.0, Exponent := 25}
                        ResourceCost := big_number{Value := 1.5, Exponent := 2}
                        ResourceType := PropResourceType.Obsidian
                        ResourceImage := VerseExposed.T_ObsidianOreStroke
        _ =>
            Print("### ARTIFACT ORE TYPE NOT SETUP IN GetArtifactSmeltData ###")
            return ArtifactPrestigeData:
                StoneCost := big_number{Value := 0.0, Exponent := 0}
                ResourceCost := big_number{Value := 0.0, Exponent := 0}
                ResourceType := PropResourceType.Copper
                ResourceImage := VerseExposed.T_Stone

GetArtifactStat(InArtifact : Artifact_Data, Stat :ArtifactStats):float=
    case(Stat):
        ArtifactStats.ArtifactBonus =>
            if(ReturnStat := InArtifact.StatValues[0]):
                return ReturnStat
        ArtifactStats.CriticalChance =>
            if(ReturnStat := InArtifact.StatValues[1]):
                return ReturnStat
        ArtifactStats.CriticalDamage =>
            if(ReturnStat := InArtifact.StatValues[2]):
                return ReturnStat
        ArtifactStats.FinalStrike =>
            if(ReturnStat := InArtifact.StatValues[3]):
                return ReturnStat
    return 0.0

GetArtifactFromPlayer(InPlayerStats : Player_Stats_Data, InArtifactRarity : ArtifactRarity):Artifact_Data=
    for(Artifact : InPlayerStats.OwnedPlayerArtifacts):
        if(InArtifactRarity = Artifact.Rarity):
            return Artifact

    return Artifact_Data{}

UIArtifactSmeltData := struct:
    StoneCost : big_number
    ResourceCost : big_number
    ResourceImage : texture
    RarityString : string

ArtifactPrestigeData := struct:
    StoneCost : big_number
    ResourceCost : big_number
    ResourceImage : texture
    ResourceType : PropResourceType

ArtifactRarity := enum<persistable>:
    MushroomCommon
    MushroomUncommon
    MushroomRare
    MushroomEpic
    MushroomLegendary
    LavaCommon
    LavaUncommon
    LavaRare
    LavaEpic
    LavaLegendary
    UnderworldCommon
    UnderworldUncommon
    UnderworldRare
    UnderworldEpic
    UnderworldLegendary
    Null

ArtifactStats := enum<persistable>:
    ArtifactBonus
    CriticalChance
    CriticalDamage
    FinalStrike

Biome := enum<persistable>:
    Mushroom
    Lava
    Underworld

MakeArtifact<constructor>(OldData : Artifact_Data)<transacts> := Artifact_Data:
    Prestige := OldData.Prestige
    Level := OldData.Level
    Rarity := OldData.Rarity
    RarityInt := OldData.RarityInt
    StatEnums := OldData.StatEnums
    StatValues := OldData.StatValues
    EquippedSlot := OldData.EquippedSlot

Artifact_Data := class<final><persistable>:
    Prestige : int = 1
    Level : int = 0
    Rarity : ArtifactRarity = ArtifactRarity.Null
    RarityInt : int = 0
    IncreaseRate : float = 1.1
    StatEnums : []ArtifactStats = array{}
    StatValues : []float = array{}
    EquippedSlot : int = -1

UpdateEquippedSlotArtifact<constructor>(OldData : Artifact_Data, InEpuippedSlot : int)<transacts> := Artifact_Data:
    Prestige := OldData.Prestige
    Level := OldData.Level
    Rarity := OldData.Rarity
    RarityInt := OldData.RarityInt
    StatEnums := OldData.StatEnums
    StatValues := OldData.StatValues
    EquippedSlot := InEpuippedSlot

ArtifactOre_Data := struct<concrete>:
    ResourceType : PropResourceType = PropResourceType.Null
    RarityToDrop : []ArtifactRarity = array{}
    DropChances : []int = array{}

ArtifactManagerTag := class(tag){}
# A Verse-authored creative device that can be placed in a level
V_ArtifactManager := class(creative_device):
    @editable
    NPCCharacters : []character_device = array{}
    @editable
    PropFollower : ?V_PropFollower = false
    @editable
    UIManager : ?V_UIManagerDevice = false
    @editable
    SuccessAudio : audio_player_device = audio_player_device{}
    @editable
    ArtifactSmelted : trigger_device = trigger_device{}

    @editable
    ArtifactRarity1 : ArtifactRarity = ArtifactRarity.Null
    @editable
    ArtifactRarity2 : ArtifactRarity = ArtifactRarity.Null
    @editable
    ArtifactRarity3 : ArtifactRarity = ArtifactRarity.Null
    @editable
    ArtifactRarity4 : ArtifactRarity = ArtifactRarity.Null
    @editable
    ArtifactRarity5 : ArtifactRarity = ArtifactRarity.Null
    @editable
    Trigger_Equip : trigger_device = trigger_device{}

    @editable
    CurrentBiome : Biome = Biome.Mushroom
    @editable
    UnlockSwitch : switch_device = switch_device{}
    @editable
    NoAccessMessage : hud_message_device = hud_message_device{}
    @editable
    PlayerManager : ?V_PlayerManagerDevice = false

    var PlayerStatManager : player_stats_manager = player_stats_manager{}

    UIArtifactSmelterMenu : UI_ArtifactSmelterMenu = UI_ArtifactSmelterMenu{}

    ## PRESETS FOR ARTIFACTS ##
    ArtifactMushroomCommon : Artifact_Data = Artifact_Data:
        Prestige := 1
        Level := 1
        Rarity := ArtifactRarity.MushroomCommon
        RarityInt := 0
        StatEnums := array{ArtifactStats.ArtifactBonus, ArtifactStats.CriticalChance, ArtifactStats.CriticalDamage, ArtifactStats.FinalStrike}
        StatValues := array{2.0, 1.0, 2.0, 1.0}

    ArtifactMushroomUncommon : Artifact_Data = Artifact_Data:
        Prestige := 1
        Level := 1
        Rarity := ArtifactRarity.MushroomUncommon
        RarityInt := 1
        StatEnums := array{ArtifactStats.ArtifactBonus, ArtifactStats.CriticalChance, ArtifactStats.CriticalDamage, ArtifactStats.FinalStrike}
        StatValues := array{4.0, 2.0, 4.0, 2.0}

    ArtifactMushroomRare : Artifact_Data = Artifact_Data:
        Prestige := 1
        Level := 1
        Rarity := ArtifactRarity.MushroomRare
        RarityInt := 2
        StatEnums := array{ArtifactStats.ArtifactBonus, ArtifactStats.CriticalChance, ArtifactStats.CriticalDamage, ArtifactStats.FinalStrike}
        StatValues := array{6.0, 3.0, 6.0, 3.0}

    ArtifactMushroomEpic : Artifact_Data = Artifact_Data:
        Prestige := 1
        Level := 1
        Rarity := ArtifactRarity.MushroomEpic
        RarityInt := 3
        StatEnums := array{ArtifactStats.ArtifactBonus, ArtifactStats.CriticalChance, ArtifactStats.CriticalDamage, ArtifactStats.FinalStrike}
        StatValues := array{8.0, 4.0, 8.0, 4.0}

    ArtifactMushroomLegendary : Artifact_Data = Artifact_Data:
        Prestige := 1
        Level := 1
        Rarity := ArtifactRarity.MushroomLegendary
        RarityInt := 4
        StatEnums := array{ArtifactStats.ArtifactBonus, ArtifactStats.CriticalChance, ArtifactStats.CriticalDamage, ArtifactStats.FinalStrike}
        StatValues := array{10.0, 5.0, 10.0, 5.0}

    ArtifactLavaCommon : Artifact_Data = Artifact_Data:
        Prestige := 1
        Level := 1
        Rarity := ArtifactRarity.LavaCommon
        RarityInt := 5
        StatEnums := array{ArtifactStats.ArtifactBonus, ArtifactStats.CriticalChance, ArtifactStats.CriticalDamage, ArtifactStats.FinalStrike}
        StatValues := array{15.0, 7.0, 15.0, 7.0}

    ArtifactLavaUncommon : Artifact_Data = Artifact_Data:
        Prestige := 1
        Level := 1
        Rarity := ArtifactRarity.LavaUncommon
        RarityInt := 6
        StatEnums := array{ArtifactStats.ArtifactBonus, ArtifactStats.CriticalChance, ArtifactStats.CriticalDamage, ArtifactStats.FinalStrike}
        StatValues := array{20.0, 10.0, 20.0, 10.0}

    ArtifactLavaRare : Artifact_Data = Artifact_Data:
        Prestige := 1
        Level := 1
        Rarity := ArtifactRarity.LavaRare
        RarityInt := 7
        StatEnums := array{ArtifactStats.ArtifactBonus, ArtifactStats.CriticalChance, ArtifactStats.CriticalDamage, ArtifactStats.FinalStrike}
        StatValues := array{25.0, 12.0, 25.0, 12.0}

    ArtifactLavaEpic : Artifact_Data = Artifact_Data:
        Prestige := 1
        Level := 1
        Rarity := ArtifactRarity.LavaEpic
        RarityInt := 8
        StatEnums := array{ArtifactStats.ArtifactBonus, ArtifactStats.CriticalChance, ArtifactStats.CriticalDamage, ArtifactStats.FinalStrike}
        StatValues := array{30.0, 15.0, 30.0, 15.0}

    ArtifactLavaLegendary : Artifact_Data = Artifact_Data:
        Prestige := 1
        Level := 1
        Rarity := ArtifactRarity.LavaLegendary
        RarityInt := 9
        StatEnums := array{ArtifactStats.ArtifactBonus, ArtifactStats.CriticalChance, ArtifactStats.CriticalDamage, ArtifactStats.FinalStrike}
        StatValues := array{35.0, 17.0, 35.0, 17.0}

    ArtifactUnderworldCommon : Artifact_Data = Artifact_Data:
        Prestige := 1
        Level := 1
        Rarity := ArtifactRarity.UnderworldCommon
        RarityInt := 10
        StatEnums := array{ArtifactStats.ArtifactBonus, ArtifactStats.CriticalChance, ArtifactStats.CriticalDamage, ArtifactStats.FinalStrike}
        StatValues := array{40.0, 20.0, 40.0, 20.0}

    ArtifactUnderworldUncommon : Artifact_Data = Artifact_Data:
        Prestige := 1
        Level := 1
        Rarity := ArtifactRarity.UnderworldUncommon
        RarityInt := 11
        StatEnums := array{ArtifactStats.ArtifactBonus, ArtifactStats.CriticalChance, ArtifactStats.CriticalDamage, ArtifactStats.FinalStrike}
        StatValues := array{45.0, 22.0, 45.0, 22.0}

    ArtifactUnderworldRare : Artifact_Data = Artifact_Data:
        Prestige := 1
        Level := 1
        Rarity := ArtifactRarity.UnderworldRare
        RarityInt := 12
        StatEnums := array{ArtifactStats.ArtifactBonus, ArtifactStats.CriticalChance, ArtifactStats.CriticalDamage, ArtifactStats.FinalStrike}
        StatValues := array{50.0, 25.0, 50.0, 25.0}

    ArtifactUnderworldEpic : Artifact_Data = Artifact_Data:
        Prestige := 1
        Level := 1
        Rarity := ArtifactRarity.UnderworldEpic
        RarityInt := 13
        StatEnums := array{ArtifactStats.ArtifactBonus, ArtifactStats.CriticalChance, ArtifactStats.CriticalDamage, ArtifactStats.FinalStrike}
        StatValues := array{55.0, 27.0, 55.0, 27.0}

    ArtifactUnderworldLegendary : Artifact_Data = Artifact_Data:
        Prestige := 1
        Level := 1
        Rarity := ArtifactRarity.UnderworldLegendary
        RarityInt := 14
        StatEnums := array{ArtifactStats.ArtifactBonus, ArtifactStats.CriticalChance, ArtifactStats.CriticalDamage, ArtifactStats.FinalStrike}
        StatValues := array{60.0, 30.0, 60.0, 30.0}

    ## PRESETS FOR ORES ##
    ArtifactMushroomCopperOre : ArtifactOre_Data = ArtifactOre_Data:
        ResourceType := PropResourceType.Copper
        RarityToDrop := array{ArtifactRarity.MushroomCommon, ArtifactRarity.MushroomUncommon, ArtifactRarity.MushroomRare}
        DropChances := array{80, 15, 5}

    ArtifactMushroomSilverOre : ArtifactOre_Data = ArtifactOre_Data:
        ResourceType := PropResourceType.Silver
        RarityToDrop := array{ArtifactRarity.MushroomCommon, ArtifactRarity.MushroomUncommon, ArtifactRarity.MushroomRare, ArtifactRarity.MushroomEpic}
        DropChances := array{30, 45, 20, 5}
    
    ArtifactMushroomMalachiteOre : ArtifactOre_Data = ArtifactOre_Data:
        ResourceType := PropResourceType.Malachite
        RarityToDrop := array{ArtifactRarity.MushroomUncommon, ArtifactRarity.MushroomRare, ArtifactRarity.MushroomEpic}
        DropChances := array{40, 50, 10}

    ArtifactMushroomSpectroliteOre : ArtifactOre_Data = ArtifactOre_Data:
        ResourceType := PropResourceType.Spectrolite
        RarityToDrop := array{ArtifactRarity.MushroomRare, ArtifactRarity.MushroomEpic, ArtifactRarity.MushroomLegendary}
        DropChances := array{35, 50, 15}

    ArtifactMushroomObsidianOre : ArtifactOre_Data = ArtifactOre_Data:
        ResourceType := PropResourceType.Obsidian
        RarityToDrop := array{ArtifactRarity.MushroomEpic, ArtifactRarity.MushroomLegendary}
        DropChances := array{40, 60}

    ArtifactLavaCopperOre : ArtifactOre_Data = ArtifactOre_Data:
        ResourceType := PropResourceType.Copper
        RarityToDrop := array{ArtifactRarity.LavaCommon, ArtifactRarity.LavaUncommon, ArtifactRarity.LavaRare}
        DropChances := array{80, 15, 5}

    ArtifactLavaSilverOre : ArtifactOre_Data = ArtifactOre_Data:
        ResourceType := PropResourceType.Silver
        RarityToDrop := array{ArtifactRarity.LavaCommon, ArtifactRarity.LavaUncommon, ArtifactRarity.LavaRare, ArtifactRarity.LavaEpic}
        DropChances := array{30, 45, 20, 5}
    
    ArtifactLavaMalachiteOre : ArtifactOre_Data = ArtifactOre_Data:
        ResourceType := PropResourceType.Malachite
        RarityToDrop := array{ArtifactRarity.LavaUncommon, ArtifactRarity.LavaRare, ArtifactRarity.LavaEpic}
        DropChances := array{40, 50, 10}

    ArtifactLavaSpectroliteOre : ArtifactOre_Data = ArtifactOre_Data:
        ResourceType := PropResourceType.Spectrolite
        RarityToDrop := array{ArtifactRarity.LavaRare, ArtifactRarity.LavaEpic, ArtifactRarity.LavaLegendary}
        DropChances := array{35, 50, 15}

    ArtifactLavaObsidianOre : ArtifactOre_Data = ArtifactOre_Data:
        ResourceType := PropResourceType.Obsidian
        RarityToDrop := array{ArtifactRarity.LavaEpic, ArtifactRarity.LavaLegendary}
        DropChances := array{40, 60}

    ArtifactUnderworldCopperOre : ArtifactOre_Data = ArtifactOre_Data:
        ResourceType := PropResourceType.Copper
        RarityToDrop := array{ArtifactRarity.UnderworldCommon, ArtifactRarity.UnderworldUncommon, ArtifactRarity.UnderworldRare}
        DropChances := array{80, 15, 5}

    ArtifactUnderworldSilverOre : ArtifactOre_Data = ArtifactOre_Data:
        ResourceType := PropResourceType.Silver
        RarityToDrop := array{ArtifactRarity.UnderworldCommon, ArtifactRarity.UnderworldUncommon, ArtifactRarity.UnderworldRare, ArtifactRarity.UnderworldEpic}
        DropChances := array{30, 45, 20, 5}
    
    ArtifactUnderworldMalachiteOre : ArtifactOre_Data = ArtifactOre_Data:
        ResourceType := PropResourceType.Malachite
        RarityToDrop := array{ArtifactRarity.UnderworldUncommon, ArtifactRarity.UnderworldRare, ArtifactRarity.UnderworldEpic}
        DropChances := array{35, 50, 15}

    ArtifactUnderworldSpectroliteOre : ArtifactOre_Data = ArtifactOre_Data:
        ResourceType := PropResourceType.Spectrolite
        RarityToDrop := array{ArtifactRarity.UnderworldRare, ArtifactRarity.UnderworldEpic, ArtifactRarity.UnderworldLegendary}
        DropChances := array{35, 50, 15}

    ArtifactUnderworldObsidianOre : ArtifactOre_Data = ArtifactOre_Data:
        ResourceType := PropResourceType.Obsidian
        RarityToDrop := array{ArtifactRarity.UnderworldEpic, ArtifactRarity.UnderworldLegendary}
        DropChances := array{40, 60}

    @editable
    PrestigeBonus : float = 2.0

    OnBegin<override>()<suspends>:void=
        for(NPCCharacter : NPCCharacters):
            NPCCharacter.InteractedWithEvent.Subscribe(OnNPCInteract)

    InitiatePlayerUI(InPlayer : player):void=
        UIArtifactSmelterMenu.InitUI(InPlayer, Self, ArtifactRarity1, ArtifactRarity2, ArtifactRarity3, ArtifactRarity4, ArtifactRarity5)

    SmeltOre(InAgent : agent, InResourceType : PropResourceType):void=
        if(Player := player[InAgent]):
            if(CanvasData := UIArtifactSmelterMenu.BuiltCanvasMap[Player]):
                StoneAmount := PlayerStatManager.GetPlayerStoneOreAmount(InAgent)
                ResourceAmount := PlayerStatManager.GetPlayerOreAmount(InAgent, CanvasData.CurrentResourceType)
                SmeltData := GetUIArtifactSmeltData(CanvasData.CurrentResourceType, CurrentBiome)
                BNCost := SmeltData.StoneCost
                ResourceCost := SmeltData.ResourceCost
                CanAffordStone := BigNumber_GreaterThanOrEquals(StoneAmount, BNCost)
                CanAffordResource := BigNumber_GreaterThanOrEquals(ResourceAmount, ResourceCost)
                #Check if we have enough ore for the next upgrade

                if(CanAffordStone = true, CanAffordResource = true):
                    if(bPlayerManager := PlayerManager?):
                        bPlayerManager.QuestFCHQ.QuestCompleted.Signal((Player, "Forge Master"))
                    PlayerStatManager.SpendPlayerOre(InAgent, BNCost, false, PropResourceType.Stone)
                    PlayerStatManager.SpendPlayerOre(InAgent, ResourceCost, false, CanvasData.CurrentResourceType)
                    # Get the smelted artifact
                    RandomArtifact := GetSmeltedArtifact(InResourceType)

                    if(RandomArtifact.Rarity = ArtifactRarity.MushroomLegendary or RandomArtifact.Rarity = ArtifactRarity.LavaLegendary or RandomArtifact.Rarity = ArtifactRarity.UnderworldLegendary):
                        if(bPlayerManager := PlayerManager?):
                            bPlayerManager.QuestFCHQ.QuestCompleted.Signal((Player, "Legendary Craftsman"))
                    #DebugPrintArtifactStats(RandomArtifact)
                    # Add the artifact to the player's owned artifacts
                    # If the artifact is already owned by the player level up the already owned one
                    PlayerStatManager.GiveArtifact(InAgent, RandomArtifact, false)
                    SuccessAudio.Play(InAgent)
                    ArtifactSmelted.Trigger(InAgent)
                    PlayerStatManager.UpdateTotalArtifactsSmelted(InAgent, big_number{Value := 1.0, Exponent := 0})
                    if(PlayerStats := PlayerStatManager.GetPlayerStats[Player]):
                        UIArtifactSmelterMenu.UpdateUI(Player, PlayerStats, PlayerStatManager, PlayerStats.OwnedPlayerArtifacts, RandomArtifact.Rarity, InResourceType, CurrentBiome)

            Print("Give Player Artifact")

    # Returns an artifact at random given a resource type
    GetSmeltedArtifact(InResourceType : PropResourceType):Artifact_Data=
        var ArtifactOre : ArtifactOre_Data = ArtifactOre_Data{}
        var OutArtifact : Artifact_Data = Artifact_Data{}

        case(InResourceType):
            PropResourceType.Copper =>
                case(CurrentBiome):
                    Biome.Mushroom =>
                        set ArtifactOre = ArtifactMushroomCopperOre
                    Biome.Lava =>
                        set ArtifactOre = ArtifactLavaCopperOre
                    Biome.Underworld =>
                        set ArtifactOre = ArtifactUnderworldCopperOre
            PropResourceType.Silver =>
                case(CurrentBiome):
                    Biome.Mushroom =>
                        set ArtifactOre = ArtifactMushroomSilverOre
                    Biome.Lava =>
                        set ArtifactOre = ArtifactLavaSilverOre
                    Biome.Underworld =>
                        set ArtifactOre = ArtifactUnderworldSilverOre
            PropResourceType.Malachite =>
                case(CurrentBiome):
                    Biome.Mushroom =>
                        set ArtifactOre = ArtifactMushroomMalachiteOre
                    Biome.Lava =>
                        set ArtifactOre = ArtifactLavaMalachiteOre
                    Biome.Underworld =>
                        set ArtifactOre = ArtifactUnderworldMalachiteOre
            PropResourceType.Spectrolite =>
                case(CurrentBiome):
                    Biome.Mushroom =>
                        set ArtifactOre = ArtifactMushroomSpectroliteOre
                    Biome.Lava =>
                        set ArtifactOre = ArtifactLavaSpectroliteOre
                    Biome.Underworld =>
                        set ArtifactOre = ArtifactUnderworldSpectroliteOre
            PropResourceType.Obsidian =>
                case(CurrentBiome):
                    Biome.Mushroom =>
                        set ArtifactOre = ArtifactMushroomObsidianOre
                    Biome.Lava =>
                        set ArtifactOre = ArtifactLavaObsidianOre
                    Biome.Underworld =>
                        set ArtifactOre = ArtifactUnderworldObsidianOre
            _ =>
                Print("### ARTIFACT ORE TYPE NOT SETUP IN GetSmeltedArtifact ###")

        var ArtifactRarityChanceArray : []ArtifactRarity = array{}
        for(Index -> Amount : ArtifactOre.DropChances):
            if(CurrentRarity := ArtifactOre.RarityToDrop[Index]):
                for(AmountIndex := 0..Amount):
                    set ArtifactRarityChanceArray += array{CurrentRarity}

        RandomSortedArtifactRarityChanceArray := Shuffle(ArtifactRarityChanceArray)
        if(FirstRarity := RandomSortedArtifactRarityChanceArray[0]):
            set OutArtifact = GetArtifactOfRarity(FirstRarity)
        return OutArtifact

    # Returns the Artifact of that rarity
    GetArtifactOfRarity(InArtifactRarity : ArtifactRarity):Artifact_Data=
        case(InArtifactRarity):
            ArtifactRarity.MushroomCommon =>
                return ArtifactMushroomCommon
            ArtifactRarity.MushroomUncommon =>
                return ArtifactMushroomUncommon
            ArtifactRarity.MushroomRare =>
                return ArtifactMushroomRare
            ArtifactRarity.MushroomEpic =>
                return ArtifactMushroomEpic
            ArtifactRarity.MushroomLegendary =>
                return ArtifactMushroomLegendary
            ArtifactRarity.LavaCommon =>
                return ArtifactLavaCommon
            ArtifactRarity.LavaUncommon =>
                return ArtifactLavaUncommon
            ArtifactRarity.LavaRare =>
                return ArtifactLavaRare
            ArtifactRarity.LavaEpic =>
                return ArtifactLavaEpic
            ArtifactRarity.LavaLegendary =>
                return ArtifactLavaLegendary
            ArtifactRarity.UnderworldCommon =>
                return ArtifactUnderworldCommon
            ArtifactRarity.UnderworldUncommon =>
                return ArtifactUnderworldUncommon
            ArtifactRarity.UnderworldRare =>
                return ArtifactUnderworldRare
            ArtifactRarity.UnderworldEpic =>
                return ArtifactUnderworldEpic
            ArtifactRarity.UnderworldLegendary =>
                return ArtifactUnderworldLegendary
            ArtifactRarity.Null =>
                return Artifact_Data{}

    EquipArtifact(InPlayer : player, InSlot : int):void=
        Print("EQUIP ARTIFACT")
        if(PlayerStats := PlayerStatManager.GetPlayerStats[InPlayer]):
            if(bUIManager := UIManager?):
                if(CanvasData := bUIManager.UIPlayerMenuClass.BuiltCanvasMap[InPlayer]):
                    if(CanvasData.ArtifactEquip.SelectedArtifact = ArtifactRarity.Null):
                        Print("Failed Null")
                        return
                    for(Artifact : PlayerStats.EquippedPlayerArtifacts):
                        DebugPrintArtifactStats(Artifact)
                        if(Artifact.Rarity = CanvasData.ArtifactEquip.SelectedArtifact):
                            Print("Already Equipped")
                            return

                    for(Artifact : PlayerStats.OwnedPlayerArtifacts):
                        if(Artifact.Rarity = CanvasData.ArtifactEquip.SelectedArtifact):
                            UpdateArtifact := UpdateEquippedSlotArtifact(Artifact, -1)
                            PlayerStatManager.EquipArtifact(InPlayer, UpdateArtifact, InSlot)
                            PlayerStatManager.GiveArtifact(InPlayer, UpdateArtifact, true)
                            Trigger_Equip.Trigger(InPlayer)
                            if(bPropFollower := PropFollower?):
                                bPropFollower.SpawnPropWithRarity(InPlayer, UpdateArtifact.RarityInt, InSlot)
                                
                            if(UpdatedStats := PlayerStatManager.GetPlayerStats[InPlayer]):
                                bUIManager.UIPlayerMenuClass.UpdateAllSlots(InPlayer, UpdatedStats)
                        else:
                            Print("Not Artifact We Want {GetArtifactRarityString(Artifact.Rarity)}")


    ChangePropResourceSelected(InPlayer :player, InCurrentResource : PropResourceType, Forward : logic):void=
        Print("Changing PropResourceType")
        var ResourceToSet : PropResourceType = PropResourceType.Copper

        case(InCurrentResource):
            PropResourceType.Copper =>
                if(Forward = true):
                    set ResourceToSet = PropResourceType.Silver
                else:
                    set ResourceToSet = PropResourceType.Obsidian
            PropResourceType.Silver =>
                if(Forward = true):
                    set ResourceToSet = PropResourceType.Malachite
                else:
                    set ResourceToSet = PropResourceType.Copper
            PropResourceType.Malachite =>
                if(Forward = true):
                    set ResourceToSet = PropResourceType.Spectrolite
                else:
                    set ResourceToSet = PropResourceType.Silver
            PropResourceType.Spectrolite =>
                if(Forward = true):
                    set ResourceToSet = PropResourceType.Obsidian
                else:
                    set ResourceToSet = PropResourceType.Malachite
            PropResourceType.Obsidian =>
                if(Forward = true):
                    set ResourceToSet = PropResourceType.Copper
                else:
                    set ResourceToSet = PropResourceType.Spectrolite
            _=>
                Print("PROP RESOURCE SELECTED NOT A VALID TYPE")

        if(CanvasData := UIArtifactSmelterMenu.BuiltCanvasMap[InPlayer]):
            set CanvasData.CurrentResourceType = ResourceToSet
            if(PlayerStats := PlayerStatManager.GetPlayerStats[InPlayer]):
                UIArtifactSmelterMenu.UpdateUINoMainIcon(InPlayer, PlayerStatManager, PlayerStats, PlayerStats.OwnedPlayerArtifacts, CanvasData.CurrentResourceType, CurrentBiome)

    PrestigeArtifact(InPlayer : player, InArtifact : Artifact_Data):void=
        Print("TRIED TO PRESTIGE")
        if(InArtifact.Prestige = 3 or InArtifact.Level <> 30):
            return # SAFTEY

        if(CanvasData := UIArtifactSmelterMenu.BuiltCanvasMap[InPlayer], PlayerStats := PlayerStatManager.GetPlayerStats[InPlayer]):
            PrestigeData := GetArtifactPrestigeData(GetResourceFromRarity(InArtifact.Rarity), CurrentBiome)

            StoneAmount := PlayerStatManager.GetPlayerStoneOreAmount(InPlayer)
            ResourceAmount := PlayerStatManager.GetPlayerOreAmount(InPlayer, PrestigeData.ResourceType)
            BNCost := BigNumber_MultiplyWithAFloat(PrestigeData.StoneCost, (InArtifact.Prestige + 1) * 100.0)
            ResourceCost := BigNumber_MultiplyWithAFloat(PrestigeData.ResourceCost, (InArtifact.Prestige + 1) * 5.0)
            CanAffordStone := BigNumber_GreaterThanOrEquals(StoneAmount, BNCost)
            CanAffordResource := BigNumber_GreaterThanOrEquals(ResourceAmount, ResourceCost)
            #Check if we have enough ore for the next upgrade

            if(CanAffordStone = false or CanAffordResource = false):
                Print("Can't Afford Prestige")
                return
            
            PlayerStatManager.SpendPlayerOre(InPlayer, BNCost, false, PropResourceType.Stone)
            PlayerStatManager.SpendPlayerOre(InPlayer, ResourceCost, false, PrestigeData.ResourceType)

            var NewIncreaseRate : float = 1.1

            if(InArtifact.Prestige = 1):
                set NewIncreaseRate = 1.1
            else if(InArtifact.Prestige = 2):
                set NewIncreaseRate = 1.2
            else if(InArtifact.Prestige = 3):
                set NewIncreaseRate = 1.3
            
            var NewStatValues : []float = InArtifact.StatValues
            var BaseArtifactStats : []float = GetArtifactOfRarity(InArtifact.Rarity).StatValues

            for(Index -> Value : BaseArtifactStats):
                if(set NewStatValues[Index] += Value * PrestigeBonus):

            # Set Artifact Level to 1
            NewArtifact := Artifact_Data:
                MakeArtifact<constructor>(InArtifact)
                Prestige := InArtifact.Prestige + 1
                IncreaseRate := NewIncreaseRate
                Level := 1
                StatValues := NewStatValues

            PlayerStatManager.GiveArtifact(InPlayer, NewArtifact, false)

        if(CanvasData := UIArtifactSmelterMenu.BuiltCanvasMap[InPlayer], PlayerStats := PlayerStatManager.GetPlayerStats[InPlayer]):
            UIArtifactSmelterMenu.UpdateUI(InPlayer, PlayerStats, PlayerStatManager, PlayerStats.OwnedPlayerArtifacts, CanvasData.CurrentArtifactRarity, CanvasData.CurrentResourceType, CurrentBiome)

    ChangeSelectedArtifact(InPlayer : player, InSlot : int, InNext : logic):void=
        Print("CHANGE")
        # Get next owned
        SortPlayerOwnedArtifacts(InPlayer)
        if(PlayerStats := PlayerStatManager.GetPlayerStats[InPlayer]):
            var OutArtifact : ArtifactRarity = ArtifactRarity.Null
            if(bUIManager := UIManager?):
                <#if(CanvasData := bUIManager.UIPlayerMenuClass.BuiltCanvasMap[InPlayer], Display := CanvasData.ArtifactEquip.ArtifactDisplays[InSlot - 1]):
                    if(Display.CurrentArtifactRarity <> ArtifactRarity.Null):
                        ArrayLength := PlayerStats.OwnedPlayerArtifacts.Length
                        for(Index -> Artifact : PlayerStats.OwnedPlayerArtifacts):
                            var NewIndex : int = 0
                            if(Artifact.Rarity = Display.CurrentArtifactRarity):
                                if(InNext = true):
                                    if(Index <> ArrayLength - 1):
                                        set NewIndex = Index + 1
                                else:
                                    if(Index <> 0):
                                        set NewIndex = Index - 1
                                    else:
                                        set NewIndex = ArrayLength - 1
                                
                                Print("NewIndex : {NewIndex}")
                                if(NewArtifact := PlayerStats.OwnedPlayerArtifacts[NewIndex]):
                                    set OutArtifact = NewArtifact.Rarity
                    else:
                        if(InNext = true):
                            if(NewArtifact := PlayerStats.OwnedPlayerArtifacts[0]):
                                set OutArtifact = NewArtifact.Rarity
                        else:
                            ArrayLength := PlayerStats.OwnedPlayerArtifacts.Length
                            if(NewArtifact := PlayerStats.OwnedPlayerArtifacts[ArrayLength - 1]):
                                set OutArtifact = NewArtifact.Rarity
                        
                # Update Slot
                bUIManager.UIPlayerMenuClass.UpdateSlot(InPlayer, PlayerStats, InSlot, OutArtifact)#>

    SortPlayerOwnedArtifacts(InPlayer : player):void=
        Print("Sort")
        # Create New Array to be returned
        if(PlayerStats := PlayerStatManager.GetPlayerStats[InPlayer]):
            var NewArtifactArray : []Artifact_Data = PlayerStats.OwnedPlayerArtifacts
            ArrayLength := NewArtifactArray.Length

            # Use Bubblesort to sort NewArtifactArray by round wins
            loop:
                for(i := 0..ArrayLength - 1):
                    for(j := 0..ArrayLength - i - 1):
                        if(FirstArtifact := NewArtifactArray[j], SecondArtifact := NewArtifactArray[j + 1]):
                            FirstArtifactInt := FirstArtifact.RarityInt
                            SecondArtifactInt := SecondArtifact.RarityInt
                            if(FirstArtifactInt > SecondArtifactInt):
                                if(set NewArtifactArray[j] = SecondArtifact, set NewArtifactArray[j + 1] = FirstArtifact):
                break

            PlayerStatManager.SetArtifacts(InPlayer, NewArtifactArray)

    OnNPCInteract(InAgent : agent):void=
        if(UnlockSwitch.GetCurrentState[InAgent]):
            if(Player := player[InAgent], PlayerStats := PlayerStatManager.GetPlayerStats[Player]):
                if(CanvasData := UIArtifactSmelterMenu.BuiltCanvasMap[Player]):
                    UIArtifactSmelterMenu.UpdateUI(Player, PlayerStats, PlayerStatManager, PlayerStats.OwnedPlayerArtifacts, ArtifactRarity.Null, CanvasData.CurrentResourceType, CurrentBiome)
                UIArtifactSmelterMenu.AddUI(Player)
        else:
            NoAccessMessage.Show(InAgent)

    PrestigeButtonPressed(InPlayer : player):void=
        if(CanvasData := UIArtifactSmelterMenu.BuiltCanvasMap[InPlayer], PlayerStats := PlayerStatManager.GetPlayerStats[InPlayer]):
            PrestigeArtifact(InPlayer, GetArtifactFromPlayer(PlayerStats, CanvasData.CurrentArtifactRarity))
    
    SmeltButtonPressed(InPlayer : player):void=
        if(CanvasData := UIArtifactSmelterMenu.BuiltCanvasMap[InPlayer]):
            SmeltOre(InPlayer, CanvasData.CurrentResourceType)

    ChangeDisplayedResource(InPlayer : player, InNext : logic):void=
        if(CanvasData := UIArtifactSmelterMenu.BuiltCanvasMap[InPlayer]):
            ChangePropResourceSelected(InPlayer, CanvasData.CurrentResourceType, InNext)

    ChangedDisplayedArtifact(InPlayer : player, InRarity : ArtifactRarity):void=
        if(PlayerStats := PlayerStatManager.GetPlayerStats[InPlayer]):
            for(Artifact : PlayerStats.OwnedPlayerArtifacts):
                if(Artifact.Rarity = ArtifactRarity.MushroomCommon):
                    if(CanvasData := UIArtifactSmelterMenu.BuiltCanvasMap[InPlayer]):
                        UIArtifactSmelterMenu.UpdateUI(InPlayer, PlayerStats, PlayerStatManager, PlayerStats.OwnedPlayerArtifacts, InRarity, CanvasData.CurrentResourceType, CurrentBiome)
                    return