using { /Fortnite.com/Devices }
using { /Verse.org/Simulation }
using { /UnrealEngine.com/Temporary/SpatialMath }
using { /UnrealEngine.com/Temporary/Diagnostics }

V_AutoMinerPlotDevice := class(creative_device):
    @editable
    PlotNumber : int = 1
    @editable
    DetailsBillboard : billboard_device = billboard_device{}
    @editable
    UsernameBillboard : billboard_device = billboard_device{}
    @editable
    CollectVolume : volume_device = volume_device{}
    @editable
    CollectChestSwitch : switch_device = switch_device{}
    @editable
    PlotMoveObject : creative_prop = creative_prop{}

    var PlayerManager : V_PlayerManagerDevice = V_PlayerManagerDevice{}
    var AutoMinerManager : V_AutoMinerZoneManager = V_AutoMinerZoneManager{}

    var CollectUI : hud_message_device = hud_message_device{}

    var ResourceList : []ResourceClass = array{}

    var HasPlotBeenClaimed : logic = false
    var HasAnimated : logic = false

    var CurrentMinerHitRate : float = 0.0
    var CurrentMinerCount : int = 0
    var StartingPrice : int = 0

    var PlayerPrestige : int = 0

    #Custom Collect Amount per Plot Device stored as a big number
    var PlayerCollectAmount : big_number = big_number{}

    ToMessage<localizes>(String : string) : message = "{String}"

    #Called From Auto Miner Manager
    StartUp(InResourceList : []ResourceClass) : void=
        set ResourceList = InResourceList

        #Subscribes to Collect Button
        CollectVolume.AgentEntersEvent.Subscribe(CollectResources)

    #Called when player claims land plot
    ClaimLand(Agent : agent): void=
        if(set AutoMinerManager.PlayerPlotMap[Agent] = PlotNumber):
        set HasPlotBeenClaimed = true

        UpdateBillboard(Agent)
        if(Stats := PlayerManager.PlayerStatManager.GetPlayerStats[Agent]):
            set PlayerPrestige = Stats.Prestige
        #Print("AutoMinerPlotDevice - Claimed Plot: {PlotNumber}")

    #Provides Player with Resources
    CollectResources(Agent : agent) : void=
        if(Player := player[Agent]):
            for(Resource : ResourceList):
                PlayerManager.PlayerStatManager.UpdatePlayerOreAmount(Agent, Resource.ResourceType, PlayerCollectAmount)
                PlayerManager.UIManagerDevice.UIItemStreamDisplayClass.AddToStream(Player, "Stone", Resource.ResourceImage, PlayerCollectAmount, false, false)

                CollectUI.SetText(ToMessage("+{BigNumber_FormatDisplay(PlayerCollectAmount)}"))
                CollectUI.Show(Agent)

                #Animated Chest Opening for Visual Effect
                for(InPlayer : GetPlayspace().GetPlayers()):
                    CollectChestSwitch.TurnOn(Player)

                #Reset Collect Amount Per Resource
                set PlayerCollectAmount = big_number{}
                UpdateBillboard(Agent)

    #Starts AutoMiner per Resource
    MinerAdded(Agent : agent, Resource : ResourceClass, MaxMinerCount : int, MinersToAdd : int, IsOverride : logic) : void=
        if(MaxMinerCount > CurrentMinerCount + MinersToAdd):
            if(IsOverride = false):
                set CurrentMinerCount += MinersToAdd
            else:
                set CurrentMinerCount = MinersToAdd
            
        AutoMinerStartingCheck(Agent, CurrentMinerCount, Resource)

    #Checks if Auto Miner can be upgraded
    AutoMinerStartingCheck(Agent : agent, ResourceMinerCount : int, Resource : ResourceClass) : void=
        if(ResourceMinerCount > 0):
            #Print("AutoMinerPlotDevice - Mining Started Loop")
            spawn:
                LoopingAutoMiner(Agent, Resource)

    #Updates text on the Billboard Display
    UpdateBillboard(Agent : agent) : void=
        if(Resource := ResourceList[0]):
            PlayerName := GetAgentName(Agent)
            UsernameBillboard.SetText(PlayerName)

            DetailsBillboard.SetTextSize(15)
            DetailsBillboard.SetText(ToMessage("Miner Count: {CurrentMinerCount}\n Collect: {BigNumber_FormatDisplay(PlayerCollectAmount)}"))
            #Mining Damage: {FormatFloatTo2DP(CurrentMinerHitRate)}\n
    GetAgentName<localizes>(Agent:agent)<computes>: message = "{Agent}"

    #Proivdes Resources to Player after short period of time
    LoopingAutoMiner(Agent : agent, Resource : ResourceClass)<suspends> : void=
        loop:
            Sleep(1.0)

            if(Player := player[Agent]):
                if(not Player.IsActive[]):
                    break

                if(PlayerData := PlayerManager.PlayerStatManager.GetPlayerStats[Player]):
                    # Update 1.5 - changed to work with the Total Gather Rate not Base Strength

                    var ArtifactStrengthBonus : float = 0.0
                    AchievementMultipler := PlayerData.AchievementTotalStoneBonus + PlayerData.AchievementTotalDungeonsCompletedBonus + PlayerData.AchievementHighestStrengthBonus + PlayerData.AchievementTotalRocksDestroyedBonus + PlayerData.AchievementTotalArtifactsSmeltedBonus
                    for(EquippedArtifact : PlayerData.EquippedPlayerArtifacts):
                        set ArtifactStrengthBonus += GetArtifactStat(EquippedArtifact, ArtifactStats.ArtifactBonus)
                        
                    HarvestRateArtifactBuff := BigNumber_Add(PlayerData.HarvestRatePerHit, BigNumber_MultiplyWithAFloat(PlayerData.HarvestRatePerHit, (ArtifactStrengthBonus / 100.0)))
                    HarvestRatePrestige := BigNumber_Add(HarvestRateArtifactBuff, BigNumber_MultiplyWithAFloat(HarvestRateArtifactBuff, PlayerData.PrestigeBonusHarvestRate / 100.0))
                    CaveUnlockPercentage := BigNumber_Add(HarvestRatePrestige ,BigNumber_MultiplyWithAFloat(HarvestRateArtifactBuff,(PlayerData.CaveUnlockPercent/100.0)))
                    AchievementUnlockPercentage := BigNumber_Add(CaveUnlockPercentage ,BigNumber_MultiplyWithAFloat(HarvestRateArtifactBuff,(AchievementMultipler/100.0)))
                    HitAmount := AchievementUnlockPercentage

                    DividedHitRate := CurrentMinerHitRate / 100.0
                    FirstHalfEquation := DividedHitRate * CurrentMinerCount
                    SecondHalfEquation := HitAmount
                    CollectAmount := BigNumber_MultiplyWithAFloat(SecondHalfEquation, FirstHalfEquation)

                    #Rather then storing in the AutoMinerZoneManager, we store them in here
                    if(PlayerData.Prestige = PlayerPrestige):
                        set PlayerCollectAmount = BigNumber_Add(PlayerCollectAmount, CollectAmount)
                    else:
                        set PlayerCollectAmount = big_number{}
                        set PlayerPrestige = PlayerData.Prestige
                    #set Resource.CollectAmount = BigNumber_Add(Resource.CollectAmount, CollectAmount)
                    UpdateBillboard(Agent)
                else:
                    break

    #Resets all Plot Variables
    ResetPlot() : void=
        #Print("Reseting Plot: {PlotNumber}")
        for(Resource : ResourceList):
            set Resource.CollectAmount = big_number{}

        set CurrentMinerHitRate = 0.0
        set CurrentMinerCount = 0

        set HasPlotBeenClaimed = false
        set HasAnimated = false

        UsernameBillboard.SetText(ToMessage("For Sale!"))
        DetailsBillboard.SetText(ToMessage("For Sale!"))
        DetailsBillboard.SetTextSize(24)

        spawn:
            MovePlot(0.5, 212929.90)

    #Animates Auto Miner plot either up or down
    MovePlot(MoveTime : float, ZLocation : float)<suspends> : void=
        #Print("AutoMinerPlotDevice - Moving Plot")
        PlotObjectPosition := PlotMoveObject.GetTransform().Translation
        var NewPosition : vector3 = vector3{}
        #Changed to use a set ZLocation for going above ground and under ground
        set NewPosition = vector3{X:=PlotObjectPosition.X, Y:=PlotObjectPosition.Y, Z:= ZLocation}

        PlotMoveObject.MoveTo(NewPosition, PlotMoveObject.GetTransform().Rotation, MoveTime)