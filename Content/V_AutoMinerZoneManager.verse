using { /Fortnite.com/Devices }
using { /Verse.org/Simulation }
using { /Verse.org/Assets }
using { /Verse.org/Simulation/Tags }
using { /UnrealEngine.com/Temporary/SpatialMath }
using { /UnrealEngine.com/Temporary/Diagnostics }

using { /Verse.org/Random}

ResourceClass := class<concrete><unique>():
    @editable
    ResourceType : PropResourceType = PropResourceType.Stone

    var CollectAmount : big_number = big_number{}
    var ResourceImage : texture = VerseExposed.T_Stone
    var ResourceName : []char = "Stone"

V_AutoMinerZoneManager := class(creative_device):
    ToMessage<localizes>(String : string) : message = "{String}"

    @editable
    PlayerManager : ?V_PlayerManagerDevice = false
    #Not in use
    @editable
    ClaimPlotTrigger : trigger_device = trigger_device{}
    @editable
    AutoMinerCountUpgradeTrigger : trigger_device = trigger_device{}
    @editable
    DamagePerHitUpgradeTrigger : trigger_device = trigger_device{}
    #End not in use

    @editable
    ClaimAudio : audio_player_device = audio_player_device{}
    @editable
    CollectUI : hud_message_device = hud_message_device{}

    @editable
    MaxMinerCount : int = 10

    @editable
    LandPlots : []V_AutoMinerPlotDevice = array{}
    @editable
    ResourceList : []ResourceClass = array{}

    # Manages and updates stats for each player.
    var PlayerPlotMap : [agent]int = map{}
    var PlayerStatsManager: player_stats_manager = player_stats_manager{}

    OnBegin<override>()<suspends>:void=
        Print("Setup Auto Mining Zone")
        GetPlayspace().PlayerRemovedEvent().Subscribe(PlayerLeft)

        if(bPlayerManager := PlayerManager?):
            set PlayerStatsManager = bPlayerManager.PlayerStatManager

        for(LandPlot : LandPlots):
            #SetsUpAll of the Plots
            LandPlot.StartUp(ResourceList)
            set LandPlot.AutoMinerManager = Self
            set LandPlot.CollectUI = CollectUI

            #Resets Plot
            LandPlot.ResetPlot()

            if(bPlayerManager := PlayerManager?):
                set LandPlot.PlayerManager = bPlayerManager

    #Moves Land Plot Up
    AnimatePlotCheck(bAgent : agent)<suspends> : void=
        if(Player := player[bAgent]):
            Print("Animate Auto Miner Up")
            for(Plot : LandPlots):
                if(bPlayerManager := PlayerManager?):
                    if(Plot.PlotNumber = bPlayerManager.PlayerMap[Player].ClassNumber and Plot.HasAnimated = false and Plot.HasPlotBeenClaimed = true):
                        Print("AutoMinerZoneManager - Animate Plot: {Plot.PlotNumber}")
                        Sleep(1.0)
                        set Plot.HasAnimated = true
                        spawn:
                            Plot.MovePlot(5.0, 213442.80)
                        return

    #Resets Plot on player Left
    PlayerLeft(Player : player) : void=
        #Print("AutoMinerZoneManager - Player Left")
        if(Agent := agent[Player]):
            for(Plot : LandPlots):
                if(PlayerPlotMap[Agent] = Plot.PlotNumber):
                    #Add the Class back into the array
                    # Moved to PlayerManagerDevice PlayerLeft
                    <#if(bPlayerManager := PlayerManager?):
                        if(set bPlayerManager.PlayerClassesToAssign = bPlayerManager.PlayerClassesToAssign + array{bPlayerManager.PlayerMap[Player].ClassNumber}):
                            Print("Class added back to array")#>
                    #Print("AutoMinerZoneManager - Player Plot Found")
                    Plot.ResetPlot()

    #Called when plot upgrade has been bought
    UpgradePlayerPlot(InPlayer : player, UpgradeType : PlayerUpgradeType): void=
        #Print("AutoMinerZoneManager - Upgrade Called for Plot: {Plot.PlotNumber}")
        #Checks if Claim Land Plot button has been pressed
        var Plot : V_AutoMinerPlotDevice = V_AutoMinerPlotDevice{} 
        for(CurrentPlot : LandPlots):
            if(bPlayerManager := PlayerManager?):
                if(CurrentPlot.PlotNumber = bPlayerManager.PlayerMap[InPlayer].ClassNumber):
                    set Plot = CurrentPlot
        
        if(UpgradeType = PlayerUpgradeType.Prestige_AutoMiningPlot):
            spawn:
                ClaimPlot(InPlayer, Plot)

        case(UpgradeType):
            PlayerUpgradeType.Prestige_AutoMiningCount=>
                UpgradeMinerCount(InPlayer, Plot, true, false)
            PlayerUpgradeType.Prestige_AutoMiningHitRate=>
                spawn:
                    UpgradeHitRate(InPlayer, Plot)
            _=>
                #Print("AutoMinerZoneManager - No Upgrade Found")

    #Called the Claim Plot Function in each Plot
    #Trying to make this get called in it's own async thread
    ClaimPlot(Agent : agent, Plot : V_AutoMinerPlotDevice)<suspends>: void=
        Print("Claim Plot")
        if(Player := player[Agent]):
            if(bPlayerManager := PlayerManager?):
                if(bPlayerManager.PlayerMap[Player].ClassNumber = Plot.PlotNumber):
                    #Print("AutoMinerZoneManager - Claim Plot Called for: {Plot.PlotNumber}")
                    Plot.ClaimLand(Agent)
                    ClaimAudio.Play(Agent)
                    spawn:
                        UpgradeHitRate(Player, Plot)
                    UpgradeMinerCount(Player, Plot, false, true)
                    Plot.UpdateBillboard(Agent)

                    #Moved into the Claim Plot
                    spawn:
                        AnimatePlotCheck(Agent)

    #Upgrades Miner Count in each Plot
    UpgradeMinerCount(Agent : agent, Plot : V_AutoMinerPlotDevice, SimpleUpgrade : logic, IsOverride : logic) : void=
        #Print("AutoMinerZoneManager - Upgrade AutoMiner for Plot: {Plot.PlotNumber}")
        if(Resource := ResourceList[0]):
            if(SimpleUpgrade = false):
                if(Player := player[Agent], PlayerData := PlayerStatsDataMap[Player]):
                    PlayerMinerCountUpgradeLevel := PlayerData.PrestigeAutoMiningCount
                    MinerCountData := PlayerStatsManager.GetPlayerUpgrade(Agent, PlayerUpgradeType.Prestige_AutoMiningCount)
                    MinerCountValue := MinerCountData.GetCurrentLevelAmount(PlayerMinerCountUpgradeLevel)
                    if(IntConvert := Round[MinerCountValue]):
                        if(IsOverride = false):
                            Plot.MinerAdded(Agent, Resource, MaxMinerCount, IntConvert, false)
                        else: 
                            Plot.MinerAdded(Agent, Resource, MaxMinerCount, IntConvert, true)
            else:
                Plot.MinerAdded(Agent, Resource, MaxMinerCount, 1, false)

    #Upgrades Miner Damage in each Plot
    UpgradeHitRate(Agent : agent, Plot : V_AutoMinerPlotDevice)<suspends> : void= 
        #Set Hit Rate
        Sleep(0.25)
        #Print("AutoMinerZoneManager - Upgrade Damage Per Hit for Plot: {Plot.PlotNumber}")
        if(Player := player[Agent], PlayerData := PlayerStatsDataMap[Player]):
            PlayerHitRateUpgradeLevel := PlayerData.PrestigeAutoMiningHitRate
            Print("Hit Rate level: {PlayerHitRateUpgradeLevel}")
            HitRateData := PlayerStatsManager.GetPlayerUpgrade(Agent, PlayerUpgradeType.Prestige_AutoMiningHitRate)
            HitRateValue := HitRateData.GetCurrentLevelAmount(PlayerHitRateUpgradeLevel)
            set Plot.CurrentMinerHitRate = HitRateValue
            Print("Hit Rate: {HitRateValue}")