
using { /Fortnite.com/Devices }
using { /Verse.org/Simulation }
using { /UnrealEngine.com/Temporary/Diagnostics }

# See https://dev.epicgames.com/documentation/en-us/uefn/create-your-own-device-in-verse for how to create a verse device.

#################################
# We can represent very big numbers as a struct containing both a value and an exponent
#################################
big_number := struct <concrete><persistable>:
    @editable
    Value : float = 0.0
    @editable
    Exponent : int = 0

#################################
# Below are various functions to operate on big numbers
#################################

# Adds two big_numbers together
BigNumber_Add(NumberA : big_number, NumberB : big_number):big_number=
    CommonExponent : int = Max(NumberA.Exponent, NumberB.Exponent)

    NewValue : float =  (NumberA.Value * Pow(10.0, (NumberA.Exponent - CommonExponent) * 1.0)) + 
                        (NumberB.Value * Pow(10.0, (NumberB.Exponent - CommonExponent) * 1.0))

    BigNumber_Normalize(big_number{Value := NewValue, Exponent := CommonExponent})

# Subtracts a big_number from another
BigNumber_Subtract(NumberA : big_number, NumberB : big_number):big_number=
    CommonExponent : int = Max(NumberA.Exponent, NumberB.Exponent)

    NewValue : float =  (NumberA.Value * Pow(10.0, (NumberA.Exponent - CommonExponent) * 1.0)) - 
                        (NumberB.Value * Pow(10.0, (NumberB.Exponent - CommonExponent) * 1.0))

    BigNumber_Normalize(big_number{Value := NewValue, Exponent := CommonExponent})

# Multiply two big_numbers together
BigNumber_Multiply(NumberA : big_number, NumberB : big_number):big_number=
    NewValue := NumberA.Value * NumberB.Value
    NewExponent := NumberA.Exponent + NumberB.Exponent

    BigNumber_Normalize(big_number{Value := NewValue, Exponent := NewExponent})

# Multiply a big_number with a float
BigNumber_MultiplyWithAFloat(NumberA : big_number, NumberB : float):big_number=
    NewValue := NumberA.Value * NumberB
    NewExponent := NumberA.Exponent

    BigNumber_Normalize(big_number{Value := NewValue, Exponent := NewExponent})

# Multiply two big_numbers together
BigNumber_Divide(NumberA : big_number, NumberB : big_number):big_number=
    Multiplier := big_number{Value := 1.0 / NumberB.Value, Exponent := -NumberB.Exponent}

    OutBigNumber := BigNumber_Multiply(NumberA, Multiplier)

    return OutBigNumber

# Multiply two big_numbers together
BigNumber_PowWithAFloat(NumberA : big_number, InValue : float):big_number=
    NewExponent := NumberA.Exponent * InValue
    var NewValue : float = NumberA.Value

    if (RoundedExponent := Round[NewExponent])
    {
        set NewValue = Pow(NumberA.Value, InValue);

        #Print("Value : {NewValue}, Exponent : {RoundedExponent}")
        return BigNumber_Normalize(big_number{Value := NewValue, Exponent := RoundedExponent});
    }

    return big_number{}

# Tells you if one big_number is greater than another
BigNumber_GreaterThanOrEquals(NumberA : big_number, NumberB : big_number)<transacts>:logic=
    if:
        NumberA.Exponent > NumberB.Exponent or (NumberA.Exponent = NumberB.Exponent and NumberA.Value >= NumberB.Value)
    then:
        return true

    return false

BigNumber_ConvertFloatIntoBigNumber(NumberA : float):big_number=
    ConvertedBigNumber := BigNumber_Add(big_number{Value:=NumberA}, big_number{Value:=0.0})
    return ConvertedBigNumber

BigNumber_ConvertBigNumberIntoFloat(NumberA : big_number):float=
    var ReturnFloat : float = NumberA.Value
    if(NumberA.Exponent >= 1):
        for(Exponent := 1..NumberA.Exponent):
            set ReturnFloat = ReturnFloat * 10.0
    if(NumberA.Exponent < 0):
        for(Exponent := 1..-NumberA.Exponent):
            set ReturnFloat = ReturnFloat * 0.1
    return ReturnFloat

# Tells you how much it would cost to buy X items
# Start is how many items you currently own, and count is how many you would like to purchase (used for price scaling)
# CommonRatio refers to the scaling in price for subsequent purchases.  If set to 1.0, no price scaling occurs.  
BigNumber_CalculateCostRange(BaseCost : big_number, Start : int, Count : int, CommonRatio : float):big_number=
    term1 := Pow(CommonRatio, (Start + Count)*1.0);
    term2 := Pow(CommonRatio, Start * 1.0);

    TotalCost_Value := BaseCost.Value * (term1 - term2) / (CommonRatio - 1.0);
    TotalCost_Exponent := BaseCost.Exponent;

    BigNumber_Normalize(big_number{Value:=TotalCost_Value, Exponent:=TotalCost_Exponent});


# Tells you how many items you could afford to purchase
# CommonRatio refers to the scaling in price for subsequent purchases.  If set to 1.0, no price scaling occurs.  
BigNumber_CalculateAffordableItems(CurrentMoney : big_number, BaseCost : big_number, AmountOwned : int, CommonRatio : float):int=
    var AffordableItems : int = 1
    loop:
        Affordable := BigNumber_GreaterThanOrEquals(CurrentMoney, BigNumber_CalculateCostRange(BaseCost, AmountOwned, AffordableItems, CommonRatio))
        if:
            not Affordable?
        then:
            break
        
        set AffordableItems += 1

    return AffordableItems - 1


# Formats a big_number into a string
BigNumber_FormatDisplay(ValueIN : big_number):string=
    if (ValueIN.Exponent < 0):
        return "0"

    CorrectedNumber := BigNumber_NormalizeToNearestNumberSet(ValueIN)

    if:
        CorrectedNumber.Exponent < 3
        ValueAsInt := Floor[CorrectedNumber.Value]
    then:
        return "{ValueAsInt}"
    
    var ValueAsString : string = "{CorrectedNumber.Value}"

    var TargetLength : int = 4
    if(CorrectedNumber.Value >= 100.0):
        set TargetLength = 3

    if:
        TempValueAsString := ValueAsString.Remove[TargetLength, ValueAsString.Length]
        set ValueAsString = TempValueAsString
    if:
        SuffixIndex : int = Quotient[CorrectedNumber.Exponent, 3]
        set ValueAsString += " {NumberSuffixes[SuffixIndex]}"

    ValueAsString

NumberSuffixes : []string = array{"-","K","M","B","T","Qa","Qt","Sx","Sp","Oc","No","Dc","UDc","DDc","TDc","QdDc","QnDc","SxDc","SpDc","OcDc","NoDc","Vg","UVg","DVg","TVg","QaVg","QnVg","SxVg","SpVg","OcVg","NoVg","Tg","UTg","DTg","TTg","QaTg","QtTg","SxTg","SpTg","OcTg","NoTg","Qd","UQd","DQd","TQd","QaQd","QtQd","SxQd","SpQd","OcQd","NoQd","Qi","UQi","DQi","TQi","QaQi","QtQi","SxQi","SpQi","OcQi","NoQi","Se","USe","DSe","TSe","QaSe","QtSe","SxSe","SpSe","OcSe","NoSe","St","USt","DSt","TSt","QaSt","QtSt","SxSt","SpSt","OSt","NSt","Og","UOg","DOg","TOg","QaOg","QtOg","SxOg","SpOg","OOg","NOg","Nn","UNn","DNn","TNn","QaNn","QtNn","SxNn","SpNn","ONn","NNn"}
# Takes a big_number, and converts it to ensure a single-digit whole-number component for the Value field
# Ex)  14.2 * 10^2  will become 1.42 * 10^3
BigNumber_Normalize(Number : big_number):big_number=
    var NewExponent : int = Number.Exponent
    var NewValue : float = Number.Value

    if (NewValue <= 0.0):
        return NewResult := big_number{}

    loop:
        if (NewValue < 10.0):
            break
        set NewValue *= 0.1
        set NewExponent += 1
    
    loop:
        if (NewValue >= 1.0):
            break
        set NewValue *= 10.0
        set NewExponent -= 1

    Result := big_number:
        Value := NewValue
        Exponent := NewExponent


# Takes a big_number, and converts it to ensure a 1,2 or 3 digit whole-number component for the Value field.  This is used for string formatting.
#   Ex)  1.42K or 14.2K or 142K are all valid
#   1420k is not valid, and should instead be 1.42M
BigNumber_NormalizeToNearestNumberSet(Number : big_number):big_number=
    var NewExponent : int = Number.Exponent
    var NewValue : float = Number.Value

    loop:
        if (Mod[NewExponent, 3] = 0):
            break

        set NewValue *= 10.0
        set NewExponent -= 1

    Result := big_number:
        Value := NewValue
        Exponent := NewExponent