
using { /Fortnite.com/Devices }
using { /Verse.org/Simulation }
using { /UnrealEngine.com/Temporary/Diagnostics }

#Added
using { /Fortnite.com/Game }
using { /UnrealEngine.com/Temporary/SpatialMath }
using { /Fortnite.com/Characters }
using { /Verse.org/Random }
using { /UnrealEngine.com/Temporary/UI }
using { /Fortnite.com/UI }
#using { /Verse.org/Colors }
using { /Verse.org/Assets }

#Added Animation 
using { /Fortnite.com/Devices/CreativeAnimation}
using { /Fortnite.com/Devices/CreativeAnimation/InterpolationTypes}

#Enums
BossVFXToSpawnEnum := enum:
    eMeleeHit,  
    eAOEAttack1,
    Null

#The Boss Device
V_BossDevice := class(creative_device):
    #PROP HEALTH BAR CODE
    @editable
    AmountMultiplier : float = 1.0
    @editable 
    var HarvestPropHealth<public> : big_number = big_number{}
    @editable 
    var HarvestPropMaxHealth : big_number = big_number{}
    #Visual Effects when hitting Ore + Last Hit
    @editable
    HitVFX : ?creative_prop_asset = false
    #NEW For Playing SFX on Hit
    var HitAudio : audio_player_device = audio_player_device{}
    var LastHitAudio : audio_player_device = audio_player_device{}
    #PROP HEALTH BAR CODE END

    @editable
    PlayerManager : V_PlayerManagerDevice = V_PlayerManagerDevice{}
    #Boss Creative Prop and Manipulator for Hit Detection
    @editable 
    BossManipulator : prop_manipulator_device = prop_manipulator_device{}
    @editable
    BossBlueprint : creative_prop = creative_prop{}
    @editable
    var BossName : string = ""
    @editable
    GroundTargetBlueprint : creative_prop = creative_prop{}
    @editable
    DamageDeviceExplosive : explosive_device = explosive_device{}
    #Idle Animation
    @editable 
    IdleAnim : cinematic_sequence_device = cinematic_sequence_device{}
    #Dash Melee Animation
    @editable 
    RunAnim : cinematic_sequence_device = cinematic_sequence_device{}
    @editable 
    MeleeAttack1Anim : cinematic_sequence_device = cinematic_sequence_device{}
    #Jump and Smash Attack
    @editable 
    SkyAttack : cinematic_sequence_device = cinematic_sequence_device{}
    #Death
    @editable
    DeathAnimation : cinematic_sequence_device = cinematic_sequence_device{}
    @editable
    HitFlashAnimation : cinematic_sequence_device = cinematic_sequence_device{}
    #Attack 3

    #Boss VFX To Spawn
    @editable
    MeleeHitVFX : ?creative_prop_asset = false
    @editable
    AOEAttackVFX : ?creative_prop_asset = false

    @editable
    StartBossBattleTrigger : trigger_device = trigger_device{}
    @editable
    BossArenaCenterPoint : vector3 = vector3{X:= 0.0, Y:= 0.0, Z:= 0.0}

    #This will reference the objective complete trigger
    @editable
    BossDefeatedObjectiveCompleteTrigger : trigger_device = trigger_device{}

    #Allows us to return a null VFX as default so nothing spawns
    NullVFX : ?creative_prop_asset = false
    #Audio Effects To Play we set in the device
    @editable
    AudioToPlayForVFX : []audio_player_device = array{}
    @editable
    BossDifficultyLevel : int = 1

    @editable
    BossMutatorVolume : mutator_zone_device = mutator_zone_device{}

    var BossAlive : logic = true

    var PlayersInVolume : logic = false

    #@editable
    #BossVolume : mutator_zone_device = mutator_zone_device{}
    #Handle Dungeon Logic Variables
    #Teleport into arena
    #Teleport to chest
    #Teleport out

    RunSpeed : float = 50.0
    var MeleeDashAttackMovementSpeed : float = 45.0
    TimeBeforeFirstAttack : float = 2.0
    DashAttackLength : float = 2.0
    SkyAttackLength : float = 3.49

    AttackPatternComplete:event() = event(){}

    #Allows us to set the int in the array
    var AudioToPlay : int = 0

    var OriginalBossPosition : vector3 = vector3{}

    BossHPBar : UI_BossHealthBar = UI_BossHealthBar{}

    EnemyDeadEvent:event() = event(){}
    # Runs when the device is started in a running game
    OnBegin<override>()<suspends>:void=
        BossManipulator.DamagedEvent.Subscribe(HitBoss)
        #Boss Encounter occurs with the trigger so we can tie it into the dungeons
        StartBossBattleTrigger.TriggeredEvent.Subscribe(SetupEncounter)
        BossMutatorVolume.AgentEntersEvent.Subscribe(OnAgentEnterBossVolume)
        BossMutatorVolume.AgentExitsEvent.Subscribe(OnAgentExitBossVolume)
        #Set the Original Position
        set OriginalBossPosition = BossBlueprint.GetTransform().Translation
        spawn:
            BossFightLoop()
        
    OnAgentEnterBossVolume(Agent : agent):void=
        #Adds the health bar by default, not on hit
        if(Player := player[Agent]):
            PlayerManager.UIManagerDevice.UINodeHealthBar.AddUI(Player)

    OnAgentExitBossVolume(Agent : agent):void=
        #Adds the health bar by default, not on hit
        if(Player := player[Agent]):
            spawn:
                PlayerManager.UIManagerDevice.UINodeHealthBar.HideUI(Player)


    #This will be called when players enter the Dungeon
    SetupEncounter(InAgent : ?agent):void=    
        #TODO: Setup boss health bar [Currently only appears on hit]

        set BossAlive = true
        #spawn:
            #BossFightLoop()
        spawn:
            LoopHealthUpdate()
        
    SetUpPlayer(InAgent : agent):void=
        block:
    
    #We add the health bar loop here
    LoopHealthUpdate()<suspends>:void=
        Print("Start Loop Health")
        race:
            loop:
                #loop through and update health bar.
                IsPropAlive := PropAlive()
                if(IsPropAlive = true):
                    Players := GetPlayspace().GetPlayers()
                    for(InPlayer : Players):
                        if(BossMutatorVolume.IsInVolume[InPlayer]):
                            UpdateHealthBar(InPlayer, BossName)
                else:
                    #Stop all functionality
                    AllPlayers := GetPlayspace().GetPlayers()
                    for(Player : AllPlayers):
                        if(BossMutatorVolume.IsInVolume[Player]):
                            spawn:
                                PlayerManager.UIManagerDevice.UINodeHealthBar.HideUI(Player)

                    set BossAlive = false
                    EnemyDeadEvent.Signal()
                    #Stop all boss animation
                    StopAllAnimations()
                    #Queue up the boss reset
                    OnBossDefeated()
                    BossDefeatedObjectiveCompleteTrigger.Trigger()
                    #Break the loop
                    break
                Sleep(0.01)
            loop:
                if (PlayersInVolume = true):
                    Sleep(0.1)
                else:
                    break


    HideBossUI(InAgent : agent):void=
        if(Player := player[InAgent]):
            spawn:
                PlayerManager.UIManagerDevice.UINodeHealthBar.HideUI(Player)

    #For when boss is defeated
    OnBossDefeated():void=
        #TODO: Only update for the players in the "Dungeon" this is for debug currently
        Print(" Boss fight ended is true")
        #Commented out because I already check this in the health loop
        #EnemyDeadEvent.Signal()
        #Stop all boss animation
        StopAllAnimations()
        #Play boss death animation
        BossDeath()
        #Queue up the boss reset
        spawn:
            ResetBoss(1.25)
        #Break the loop
    
    StopAllAnimations():void=
        RunAnim.Stop()
        MeleeAttack1Anim.Stop()
        SkyAttack.Stop()

    StartIdle()<suspends>:void=
        IdleAnim.Play()
    
    #Teleport the boss back and reset everything. 
    ResetBoss(Time : float)<suspends>:void=
        Sleep(Time) 
        #Teleport out after death animation
        if(BossBlueprint.TeleportTo[vector3{X:=0.0,Y:=0.0, Z:= -100000.0}, BossBlueprint.GetTransform().Rotation]):
        #Sleep(1.0)
        #Player will now be in the chest room, teleport back
        if(BossBlueprint.TeleportTo[OriginalBossPosition, BossBlueprint.GetTransform().Rotation]):
        Print("Boss Fight ended is false, reset the boss")
        #Reset Health
        ResetHarvestNodeHealth()

    ResetHarvestNodeHealth<public>():void=
        set HarvestPropHealth = HarvestPropMaxHealth

    BossFightLoop()<suspends>:void=
        race:
            loop:
                Sleep(0.1)
                if (PlayersInVolume = true, BossAlive = true):
                    Print("Difficulty 1 loop starting again")
                    AttackPattern := GetRandomInt(1,3)
                    if(AttackPattern = 1):
                        #spawn:
                        BossPattern1()
                        #AttackPatternComplete.Await()
                    if(AttackPattern = 2):
                        #spawn:
                        BossPattern2()
                        #AttackPatternComplete.Await()
                    if(AttackPattern = 3):
                        #spawn:
                        BossPattern3()
                        #AttackPatternComplete.Await()
                    Print("Attack Pattern 1 Complete") 
                    Sleep(0.0)
            loop:
                var TempPlayersInVolume : logic = false
                AllPlayers := GetPlayspace().GetPlayers()
                for(Player:AllPlayers):
                    if(BossMutatorVolume.IsInVolume[Player]):
                        set TempPlayersInVolume = true
                set PlayersInVolume = TempPlayersInVolume       
                <#if (PlayersInVolume = false):
                    set BossFightEnded = true#>
                Sleep(0.0)

            
        <#
        case(BossDifficultyLevel):
            1 =>
                loop:
                    Print("Difficulty 1 loop starting again")
                    #If the boss fight is over stop the loop
                    if(BossFightEnded = true):
                        break
                    AttackPattern := GetRandomInt(1,3)
                    if(AttackPattern = 1):
                        spawn:
                            BossPattern1()
                        AttackPatternComplete.Await()
                    if(AttackPattern = 2):
                        spawn:
                            BossPattern2()
                        AttackPatternComplete.Await()
                    if(AttackPattern = 3):
                        spawn:
                            BossPattern3()
                        AttackPatternComplete.Await()
                    Print("Attack Pattern 1 Complete")
                    
            2 =>
                Print("Difficulty 2 loop")
                loop:
                    #If the boss fight is over stop the loop
                    if(BossFightEnded = true):
                        break
                    AttackPattern := GetRandomInt(1,3)
                    if(AttackPattern = 1):
                        BossPattern1()
                        AttackPatternComplete.Await()
                    if(AttackPattern = 2):
                        BossPattern2()
                        AttackPatternComplete.Await()
                    if(AttackPattern = 3):
                        BossPattern3()
                        AttackPatternComplete.Await()
            _=> Print("Default Value")
        #>

    #Level 1 Difficulty Patterns
    BossPattern1()<suspends>:void=
        race:
            block:
                Print("Signal Attack")
                Sleep(1.8)
                MoveToRandomArenaPosition()
                Sleep(1.8)
                MeleeDashAttack()
                Sleep(2.0)
                Print("Signal Attack Complete")
                AttackPatternComplete.Signal()
            loop:
                if(PlayersInVolume = false or BossAlive = false):
                    Print("NO PLAYER IN VOLUME")
                    break
                Sleep(0.0)
               
        
    BossPattern2()<suspends>:void=
        race:
            block:
                Sleep(1.8)
                MeleeDashAttack()
                Sleep(1.8)
                MoveToRandomArenaPosition()
                Sleep(1.8)
                JumpSlamAttack()
                Sleep(SkyAttackLength)
                Print("Signal Attack Complete")
                AttackPatternComplete.Signal()
            loop:
                if(PlayersInVolume = false or BossAlive = false):
                    break
                Sleep(0.0)

    BossPattern3()<suspends>:void=
        race:
            block:
                Sleep(1.8)
                MoveToRandomArenaPosition()
                Sleep(2.0)
                JumpSlamAttack()
                Sleep(SkyAttackLength)
                Print("Signal Attack Complete")
                AttackPatternComplete.Signal()
            loop:
                if(PlayersInVolume = false or BossAlive = false):
                    break
                Sleep(0.0)
    
    #Level 2 Difficulty Patterns
    #This is a simple teleportation function that allows the target on the floor to move.
    #This is important so players get feedback of where the boss is attacking and the general radius of the attack
    MoveAttackTargetBlueprint(MovePosition : vector3, AttackTime : float, Scale : vector3)<suspends>:void=
        race:
            block:
                #Teleport to the attack position
                GroundTargetPosition := vector3{X:= MovePosition.X, Y:= MovePosition.Y, Z:= MovePosition.Z + 25.0}
                if(GroundTargetBlueprint.TeleportTo[GroundTargetPosition, GroundTargetBlueprint.GetTransform().Rotation]):
                #Animate Here
                ScaleAttackOut(Scale)
                #Teleport the explosive device into place
                if(DamageDeviceExplosive.TeleportTo[MovePosition, DamageDeviceExplosive.GetTransform().Rotation]):
                Sleep(AttackTime)
                #Teleport it away from the play space when complete
                #GroundTargetBlueprint
                #Reset Ground Target position, rotation and scale after the attack
                TeleportTransform : transform = transform{Translation:= vector3{X:=0.0, Y:=0.0, Z:= -100000.0}, Rotation:= GroundTargetBlueprint.GetTransform().Rotation, Scale:= vector3{X:= 0.10, Y:= 0.10, Z:= 0.10}}
                if(GroundTargetBlueprint.TeleportTo[TeleportTransform]):

            loop:
                if(PlayersInVolume = false or BossAlive = false):
                    break
                Sleep(0.0)
                

    DamagePlayers(TimeBeforeExplosion : float)<suspends>:void=
        race:
            block:
                Sleep(TimeBeforeExplosion)
                Players := GetPlayspace().GetPlayers()
                if(Player := Players[0]):
                    DamageDeviceExplosive.Explode(Player)
                Sleep(0.001)
                DamageDeviceExplosive.Reset()
            loop:
                if(PlayersInVolume = false or BossAlive = false):
                    break
                Sleep(0.0)


    #Boss Melee AOE Attack 1
    MeleeDashAttack()<suspends>:void=
        race:
            block:
                if(PlayersInVolume = true or BossAlive = true):
                    Print("Begin Dash Attack")
                    StartLocation := BossBlueprint.GetTransform().Translation
                    Players := GetPlayspace().GetPlayers().ToPositionals()
                    #TODO: I can add a distance check here
                    if (Target := BossBlueprint.GetTransform().GetClosestPositional(Players)?):
                        #var TargetLocation : vector3 = Target.GetTransform().Translation
                        #Modified to ignore the Z height
                        var TargetLocation : vector3 = vector3{X:= Target.GetTransform().Translation.X, Y:= Target.GetTransform().Translation.Y, Z:= BossBlueprint.GetTransform().Translation.Z} 
                        # Rotate toward player before moving
                        var NewRotation : rotation = BossBlueprint.GetTransform().GetLookAtRotation(TargetLocation)
                        RotateToFacePlayerWithAnimationController(TargetLocation)
                        Sleep(1.0)

                        # Calculate how long it will take to move to the player using Miles Per Hour
                        MoveDistance := DistanceXY(StartLocation, TargetLocation)
                        DistanceCoveredPerSecond := 45.0 * MeleeDashAttackMovementSpeed
                        var MoveTime: float = MoveDistance / DistanceCoveredPerSecond

                        #Set to the length of the run animation in the dash
                        if (MoveTime <= 0.0):
                            set MoveTime = 1.30

                        #Move Attack Target For Visual Representation
                        AttackFinalScale : vector3 = vector3{X:= 600.0, Y:= 600.0, Z:= 600.0}
                        spawn:
                            MoveAttackTargetBlueprint(TargetLocation, 2.0, AttackFinalScale)
                        #Play the dash animation
                        RunAnim.Play()
                        MoveIntoPositionWithAnimationController(TargetLocation, MoveTime)
                        Sleep(MoveTime)
                        RunAnim.Stop()
                        #  Get target's latest position
                        #set TargetLocation = vector3{X:= Target.GetTransform().Translation.X, Y:= Target.GetTransform().Translation.Y, Z:= BossBlueprint.GetTransform().Translation.Z} 
                        # One more quick rotation to face the player
                        #set NewRotation = BossBlueprint.GetTransform().GetLookAtRotation(TargetLocation)
                        #Rotate To face
                        #RotateToFacePlayerWithAnimationController(TargetLocation)
                        #Sleep(1.0)
                        Print("Attack Player")
                        MeleeAttack1Anim.Play()
                        spawn:
                            BossSpawnVFX(0.6, BossVFXToSpawnEnum.eMeleeHit, 2.0, 1.0, TargetLocation)
                        spawn:
                            DamagePlayers(0.6)
            loop:
                if(PlayersInVolume = false or BossAlive = false):
                    break
                Sleep(0.0)


    MoveToRandomArenaPosition()<suspends>:void=
        race:
            block:
                if(PlayersInVolume = true or BossAlive = true):
                    #TODO: Convert to players in the dungeon only
                    StartLocation := BossBlueprint.GetTransform().Translation

                    #Generating a random movement location
                    RandomXPosition := GetRandomFloat(-800.0, 800.0) 
                    RandomYPosition := GetRandomFloat(-800.0, 800.0) 
                    #Influences movement around the boss arena center point. So we always start from a consistent location
                    #The Z will always be the boss blueprint height!
                    MovementLocation : vector3 = vector3{X:= BossArenaCenterPoint.X + RandomXPosition, Y:= BossArenaCenterPoint.Y + RandomYPosition, Z:= StartLocation.Z}

                    # Rotate toward player before moving
                    var NewRotation : rotation = BossBlueprint.GetTransform().GetLookAtRotation(MovementLocation)
                    RotateToFacePlayerWithAnimationController(MovementLocation)
                    Sleep(1.0)
                    # Calculate how long it will take to move to the player using Miles Per Hour
                    MoveDistance := DistanceXY(StartLocation, MovementLocation)
                    DistanceCoveredPerSecond := 45.0 * MeleeDashAttackMovementSpeed
                    var MoveTime: float = MoveDistance / DistanceCoveredPerSecond

                    RunAnim.Play()
                    MoveIntoPositionWithAnimationController(MovementLocation , MoveTime)
                    Sleep(MoveTime)
                    RunAnim.Stop()
                else:
                    AttackPatternComplete.Signal()
            loop:
                if(PlayersInVolume = false or BossAlive = false):
                    break
                Sleep(0.0)

    #Attempting an Animation Controller Movement Method. If this works then all movement should be smooth.
    MoveIntoPositionWithAnimationController(MovePosition : vector3, Speed : float):void=
        MovementDirection := MovePosition - BossBlueprint.GetTransform().Translation
        #CurrentRotation := BossBlueprint.GetTransform().Rotation
        #NewRotation : rotation = BossBlueprint.GetTransform().GetLookAtRotation(Target.GetTransform().translation)

        MovementKeyFrame1 : keyframe_delta = keyframe_delta
        {
            #We make it go backwards, spins and scale down
            DeltaLocation :=  MovementDirection * Speed * 1.2
            DeltaRotation := MakeRotationFromYawPitchRollDegrees(0.0 , 0.0 , 0.0 )
            Time := Speed
            Interpolation := EaseInOut
        }

        if(PropController := BossBlueprint.GetAnimationController[]):
            KnockbackKeyFrames:[]keyframe_delta = array{MovementKeyFrame1}  #Can extend the array of keyframes here
            PropController.SetAnimation(KnockbackKeyFrames, ?Mode:=animation_mode.OneShot)
            PropController.Play()


    RotateToFacePlayerWithAnimationController(TargetPosition : vector3):void=
        #MovementDirection := MovePosition - BossBlueprint.GetTransform().Translation
        CurrentRotation := BossBlueprint.GetTransform().Rotation
        NewRotation : rotation = BossBlueprint.GetTransform().GetLookAtRotation(TargetPosition)

        MovementKeyFrame1 : keyframe_delta = keyframe_delta
        {
            #We make it go backwards, spins and scale down
            DeltaLocation :=  vector3{X:=0.0, Y:=0.0, Z:=0.0}
            DeltaRotation :=  MakeComponentWiseDeltaRotation(NewRotation, CurrentRotation)
            Time := 0.5
            Interpolation := EaseOut
        }

        if(PropController := BossBlueprint.GetAnimationController[]):
            KnockbackKeyFrames:[]keyframe_delta = array{MovementKeyFrame1}  #Can extend the array of keyframes here
            PropController.SetAnimation(KnockbackKeyFrames, ?Mode:=animation_mode.OneShot)
            PropController.Play()

    ScaleAttackOut(Scale : vector3):void=
        MovementKeyFrame1 : keyframe_delta = keyframe_delta
        {
            #We make it go backwards, spins and scale down
            DeltaLocation :=  vector3{X:=0.0, Y:=0.0, Z:=0.0}
            DeltaRotation := MakeRotationFromYawPitchRollDegrees(0.0 , 0.0 , 0.0 )
            DeltaScale := vector3{X:= Scale.X, Y:= Scale.Y, Z:= Scale.Z}
            Time := 0.5
            Interpolation := InterpolationTypes.EaseInOut
        }

        if(PropController := GroundTargetBlueprint.GetAnimationController[]):
            KnockbackKeyFrames:[]keyframe_delta = array{MovementKeyFrame1}  #Can extend the array of keyframes here
            PropController.SetAnimation(KnockbackKeyFrames, ?Mode:=animation_mode.OneShot)
            PropController.Play()

    #Boss AOE Spell Attack 1
    JumpSlamAttack()<suspends>:void=
        race:
            block:
                if(PlayersInVolume = true or BossAlive = true):
                    Print("Begin Jump Slam Attack")
                    StartLocation := BossBlueprint.GetTransform().Translation
                    # Find the nearest player 
                    Players := GetPlayspace().GetPlayers().ToPositionals()
                    #TODO: I can add a distance check here
                    if (Target := BossBlueprint.GetTransform().GetClosestPositional(Players)?):
                        var TargetLocation : vector3 = vector3{X:= Target.GetTransform().Translation.X, Y:= Target.GetTransform().Translation.Y, Z:= BossBlueprint.GetTransform().Translation.Z} 
                        # Rotate toward player before moving
                        var NewRotation : rotation = BossBlueprint.GetTransform().GetLookAtRotation(TargetLocation)
                        RotateToFacePlayerWithAnimationController(TargetLocation)
                        Sleep(0.5)

                        # Calculate how long it will take to move to the player using Miles Per Hour
                        MoveDistance := DistanceXY(StartLocation, TargetLocation)
                        DistanceCoveredPerSecond := 45.0 * MeleeDashAttackMovementSpeed
                        var MoveTime: float = MoveDistance / DistanceCoveredPerSecond

                        #Set to the length of the run animation in the dash
                        if (MoveTime <= 0.0):
                            set MoveTime = 1.30

                        #Play the dash animation
                        RunAnim.Play()
                        MoveIntoPositionWithAnimationController(TargetLocation, MoveTime)
                        Sleep(MoveTime)
                        RunAnim.Stop()

                        #  Get target's latest position
                        set TargetLocation = vector3{X:= Target.GetTransform().Translation.X, Y:= Target.GetTransform().Translation.Y, Z:= BossBlueprint.GetTransform().Translation.Z} 

                        Print("Attack Player")
                        #Spawn Attack on enemy position this time
                        AttackTargetPosition : vector3 = vector3{X:= BossBlueprint.GetTransform().Translation.X, Y:= BossBlueprint.GetTransform().Translation.Y, Z:= BossBlueprint.GetTransform().Translation.Z}
                        AttackFinalScale : vector3 = vector3{X:= 1200.0, Y:= 1200.0, Z:= 1200.0}
                        spawn:
                            MoveAttackTargetBlueprint(AttackTargetPosition, 3.3, AttackFinalScale)
                        SkyAttack.Play()
                        spawn:
                            BossSpawnVFX(0.2, BossVFXToSpawnEnum.eAOEAttack1, 3.0, 90.0, Target.GetTransform().Translation)
                        spawn:
                            DamagePlayers(0.2)
                        spawn:
                            DamagePlayers(0.6) 
            loop:
                if(PlayersInVolume = false or BossAlive = false):
                    break
                Sleep(0.0)
        
    #Return VFX that can be spawned.
    BossVFXToSpawn(VFXSpawnEnum : BossVFXToSpawnEnum): ?creative_prop_asset=
        var VFXToReturn : ?creative_prop_asset = false
        case(VFXSpawnEnum):
            BossVFXToSpawnEnum.eMeleeHit => 
                set VFXToReturn = MeleeHitVFX
                set AudioToPlay = 0
            BossVFXToSpawnEnum.eAOEAttack1 => 
                set VFXToReturn = AOEAttackVFX
                set AudioToPlay = 1
            _=> set VFXToReturn = NullVFX

        return VFXToReturn

    #Function we can call from any script when wanting to display then dispose VFX
    BossSpawnVFX(TimeToWait : float, SpawnVFXEnum : BossVFXToSpawnEnum, TimeToDispose : float, ZAxisAddition : float, PlayerTargetPosition : vector3)<suspends>:void=
        var SpawnLocation : vector3 = vector3{X:= 0.0, Y:= 0.0, Z:= 0.0}
        Sleep(TimeToWait)

        #For anything else we can just use the props location and spawn the VFX there
        if(SpawnVFXEnum = BossVFXToSpawnEnum.eMeleeHit):
            set SpawnLocation = vector3{X:= PlayerTargetPosition.X, Y:= PlayerTargetPosition.Y, Z:= BossBlueprint.GetTransform().Translation.Z + ZAxisAddition}
        if(SpawnVFXEnum = BossVFXToSpawnEnum.eAOEAttack1):
            set SpawnLocation = vector3{X:= BossBlueprint.GetTransform().Translation.X, Y:= BossBlueprint.GetTransform().Translation.Y, Z:= BossBlueprint.GetTransform().Translation.Z + ZAxisAddition}
        #The effect we'll use based on the enum
        SpawnEffect := BossVFXToSpawn(SpawnVFXEnum)

        #Optional variable cast
        if(bVFX:= SpawnEffect?):
            #Spawn the prop at the SpawnLocation with a default rotation and Scale
            SpawnedProp := SpawnProp(bVFX, transform{Translation := SpawnLocation, Rotation := IdentityRotation(), Scale:= vector3{X:=1.0,Y:=1.0,Z:=1.0}})
            #Play Sound Effects for the InAgent based on the enum
            if(SFXToPlay := AudioToPlayForVFX[AudioToPlay]):
                SFXToPlay.Play()
            #Get the Creative_Prop
            if(CreativeProp := SpawnedProp(0)?):
                #After Spawning we Dispose, this way the entire system can handle spawning and disposing VFX rather then having unique functions across all Verse Scripts
                DisposeVFX(CreativeProp, TimeToDispose)

    #Dispose the VFX after a set time
    DisposeVFX(SpawnedProp : creative_prop, TimeToDispose : float)<suspends>:void=
        Sleep(TimeToDispose)
        SpawnedProp.Dispose()
    
    #Boss Death Function
    BossDeath():void=
        DeathAnimation.Play()

    PropAlive<public>():logic=
        return BigNumber_GreaterThanOrEquals(HarvestPropHealth, big_number{Value := 0.1, Exponent := 0})


    #----PROP HEALTH BAR FUNCTIONALITY----#

    #Get and use the reference
    GetVFXSpawnerReference(): V_CustomVFXSpawner=
        var VFXSpawnerReferenceDevice : V_CustomVFXSpawner = V_CustomVFXSpawner{}
        TaggedList := FindCreativeObjectsWithTag(VFXReferenceHolderTag{})
        for(TaggedActor : TaggedList):
            if(VFXSpawnerReference := V_CustomVFXSpawner[TaggedActor]):
                set VFXSpawnerReferenceDevice = VFXSpawnerReference
        return VFXSpawnerReferenceDevice

    #When harvesting a resource we show a custom HUD and update that HUD for the player
    HitBoss(InAgent : agent): void=
    #Only perform functionality if the prop is alive
    #if(IsPropAlive = true):
        #When we hit ore trigger this trigger
        HitFlashAnimation.Play()
        Print("Damage Boss")
        if(Player := player[InAgent]):
            #Play Hit VFX
            if(AudioToPlayOnHit := AudioToPlayForVFX[0]):
                AudioToPlayOnHit.Play(Player)
            FriendsAmount := PlayerManager.FriendsManager.FindFriends(Player)
            if(CurrentPlayerStats := PlayerManager.PlayerStatManager.GetPlayerStats[InAgent], PlayerSword := CurrentPlayerStats.OwnedPlayerSword[0]):
                var TrueCrit : logic = false
                var BaseDamage : big_number = big_number{}
                var CritRate : big_number = big_number{}
                var CritDamage : big_number = big_number{}


                if(SwordBaseDamage := PlayerSword.BigNumber_StatValues[0], SwordCritRate := PlayerSword.BigNumber_StatValues[1], SwordCriticalDamage := PlayerSword.BigNumber_StatValues[2]):
                    set BaseDamage = SwordBaseDamage
                    set CritRate = SwordCritRate
                    set CritDamage = SwordCriticalDamage

                RandomChance := GetRandomInt(1, 100)
                CritRateFloat := BigNumber_ConvertBigNumberIntoFloat(CritRate)
                if(RandomChance * 1.0 <= CritRateFloat):
                    spawn:
                        GetVFXSpawnerReference().SpawnVFX(BossBlueprint, 0.01, VFXToSpawnEnum.eChargedCriticalHitVFX, 1.2, InAgent, false)
                    set TrueCrit = true

                FriendsPlayerLevel := PlayerManager.PlayerStatManager.GetPlayerUpgradeLevel(Player, PlayerUpgradeType.Friends)
                FriendsPlayerBonus := PlayerManager.PlayerStatManager.UpgradesData.GetFriendsAmount(FriendsPlayerLevel)
                BNFriendPlayerBonusPer := BigNumber_ConvertFloatIntoBigNumber(FriendsPlayerBonus / 100.0)
                TotalFriendsBonus := BigNumber_MultiplyWithAFloat(BNFriendPlayerBonusPer, FriendsAmount * 1.0)
                CurrentSwordPower := BigNumber_Add(TotalFriendsBonus, BaseDamage)
                var OutAmount : big_number = CurrentSwordPower
                
                if(TrueCrit = true):
                    set OutAmount = BigNumber_Add(CurrentSwordPower, BigNumber_MultiplyWithAFloat(CurrentSwordPower, BigNumber_ConvertBigNumberIntoFloat(CritDamage) / 100.0))
                
                #CARD - Update 1.9 - Check for cardID 0 to do bonus damage to boss
                if(FoundCard := SearchForCard(1, Player, PlayerManager.PlayerStatManager)?):
                    set OutAmount = GetBonusBigNumber(OutAmount, GetCardBonus(0, FoundCard.Level))

                #Deal Damage
                set HarvestPropHealth = BigNumber_Subtract(HarvestPropHealth, OutAmount)
                MainPropAlive := PropAlive()
                if(MainPropAlive = false):
                    Print("Main Dead Resource")
                    spawn:
                        BossDefeated(InAgent)

                GainResource(Player, OutAmount, MainPropAlive, TrueCrit)
                HitAudio.Play(InAgent)

    GainResource(InPlayer : player, InMiningPower : big_number, IsPropAlive : logic, IsCrit : logic):void=
        var MiningPower : big_number = InMiningPower
        var ArtifactFinalHitBonus : float = 0.0
        var LastHit : logic = false


        if(IsCrit = true):
            Print("Critical Hit")
            #TODO: Critical Hit VFX
            spawn:
                GetVFXSpawnerReference().SpawnVFX(BossBlueprint, 0.01, VFXToSpawnEnum.eCriticalHitVFX, 1.0, InPlayer, true)
        
        if(PlayerData := PlayerStatsDataMap[InPlayer]):
            for(EquippedArtifact : PlayerData.EquippedPlayerArtifacts):
                set ArtifactFinalHitBonus += GetArtifactStat(EquippedArtifact, ArtifactStats.FinalStrike)

        <# I think this is final hit related, I don't need it.
        MainPropAlive := PropAlive()
        if(MainPropAlive = false):
            # Get Player Final Hit Bonus
            FinalHitLevel := PlayerManager.PlayerStatManager.GetPlayerUpgradeLevel(InPlayer, PlayerUpgradeType.FinalHit)
            FinalHitBonus := PlayerManager.PlayerStatManager.UpgradesData.GetFinalHitAmount(FinalHitLevel)
            FinalHitTotal := BigNumber_MultiplyWithAFloat(BossTotalHP, (FinalHitBonus / 100.0))
            BNFinalHitTotal := BigNumber_Add(FinalHitTotal, BigNumber_MultiplyWithAFloat(FinalHitTotal, (ArtifactFinalHitBonus / 100.0)))
            set MiningPower = BigNumber_Add(InMiningPower, BNFinalHitTotal)
            set LastHit = true
            PlayerManager.PlayerStatManager.UpdateTotalRocksDestroyed(InPlayer, big_number{Value := 1.0, Exponent := 0})
        #>
        #BNExtraResource := BigNumber_ConvertFloatIntoBigNumber(InExtraResource)
        #OutAmount := BigNumber_Multiply(MiningPower), BigNumber_ConvertFloatIntoBigNumber(1.0)
        #PlayerManager.PlayerStatManager.UpdatePlayerOreAmount(InPlayer, ResourceTypeToMine, OutAmount)
        #AfterAmount :=  PlayerManager.PlayerStatManager.GetPlayerOreAmount(InPlayer, ResourceTypeToMine)

        #Allows us to show damage in the stream
        PlayerManager.UIManagerDevice.UIItemStreamDisplayClass.AddToStream(InPlayer, "Damage", VerseExposed.T_Sword, MiningPower, IsCrit, LastHit)
        #UpdateHealthBar(InPlayer, BossName)

    DropOreResource(InPlayer : player)<suspends>:void=
        # START CHANCE TO GET ANY ORE #
        RandomChance := GetRandomInt(1, 100)
        ArtifactMasteryLevel := PlayerManager.PlayerStatManager.GetPlayerUpgradeLevel(InPlayer, PlayerUpgradeType.Prestige_ArtifactMastery)
        AdditionalChance := PlayerManager.PlayerStatManager.UpgradesData.GetPrestigeArtifactMasteryAmount(ArtifactMasteryLevel)
        TotalDropChance := AdditionalChance
        if(RandomChance * 1.0 <= TotalDropChance):
            # START CHANCE TO GET SPECIFIC ORE #
            TypeRandomChance := GetRandomFloat(0.0, 100.0)
            var ResourceGot : PropResourceType = PropResourceType.Null 
            if(GetOreDropChance(PropResourceType.Obsidian) >= TypeRandomChance):
                set ResourceGot = PropResourceType.Obsidian
                Print("Obsidian Got")
            else if(GetOreDropChance(PropResourceType.Spectrolite) >= TypeRandomChance):
                set ResourceGot = PropResourceType.Spectrolite
                Print("Spectrolite Got")
            else if(GetOreDropChance(PropResourceType.Malachite) >= TypeRandomChance):
                set ResourceGot = PropResourceType.Malachite
                Print("Malachite Got")
            else if(GetOreDropChance(PropResourceType.Silver) >= TypeRandomChance):
                set ResourceGot = PropResourceType.Silver
                Print("Silver Got")
            else if(GetOreDropChance(PropResourceType.Copper) >= TypeRandomChance):
                set ResourceGot = PropResourceType.Copper
                Print("Copper Got")

            if(ResourceGot <> PropResourceType.Null):
                QuantityLevel := PlayerManager.PlayerStatManager.GetPlayerUpgradeLevel(InPlayer, PlayerUpgradeType.OreQuantity)
                Upgrade := PlayerManager.PlayerStatManager.GetPlayerUpgrade(InPlayer, PlayerUpgradeType.OreQuantity)
                OutAmount := BigNumber_ConvertFloatIntoBigNumber(Upgrade.GetCurrentLevelAmount(QuantityLevel))

                PlayerManager.PlayerStatManager.UpdatePlayerOreAmount(InPlayer, ResourceGot, OutAmount)
                Sleep(0.25)
                PlayerManager.UIManagerDevice.UIItemStreamDisplayClass.AddToStream(InPlayer, "Stone", GetOreTexture(ResourceGot), OutAmount, false, false)

    #Hide the resource, after a delay reset the health and show it again
    BossDefeated(InAgent : agent)<suspends>:void=
        Print("Boss Dead")

    # Update a given player's prop health bar
    UpdateHealthBar(InPlayer : player, InResourceName : string):void=
        #Print("Update Health bar")
        BNPropHealth := BigNumber_FormatDisplay(HarvestPropHealth)
        BNPropMaxHealth := BigNumber_FormatDisplay(HarvestPropMaxHealth)
        HealthText := "{BNPropHealth} / {BNPropMaxHealth}"
        HealthPercentage := BigNumber_ConvertBigNumberIntoFloat(BigNumber_MultiplyWithAFloat(BigNumber_Divide(HarvestPropHealth, HarvestPropMaxHealth), 100.0))
        PropPosition := BossBlueprint.GetTransform().Translation
        HealthBarData := MakeNodeHealthBarData(InResourceName, HealthText, HealthPercentage, PropPosition, true)

        PlayerManager.UIManagerDevice.UINodeHealthBar.UpdateBossUI(InPlayer, HealthBarData)
        #PlayerManager.UIManagerDevice.UINodeHealthBar.AddUI(InPlayer)