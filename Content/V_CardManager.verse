
using { /Fortnite.com/Devices }
using { /Verse.org/Simulation }
using { /UnrealEngine.com/Temporary/Diagnostics }

using { /Verse.org/Assets }
using { /Verse.org/Random }

# See https://dev.epicgames.com/documentation/en-us/uefn/create-your-own-device-in-verse for how to create a verse device.

SavedCardData := class<final><persistable>:
    ID : int
    Rarity : int
    Level : int

CardData := class:
    ID : int
    Rarity : int
    Level : int
    Image : texture

MakeSavedCardData<constructor>(OldCard : SavedCardData)<transacts> := SavedCardData:
    ID := OldCard.ID
    Rarity := OldCard.Rarity
    Level := OldCard.Level

UnlockCardHandler := class(){
    CardManager : V_CardManager
    Index : int

    HandlerFunction(InAgent : agent) : void = {
        if(Player := player[InAgent]):
            spawn:
                CardManager.GetRandomCard(Player, Index)
    }
}

# EVERY CARD ADDED TO THE GAME NEEDS TO BE PUT IN THE LOOK UP FUNCTIONS
LookupCardDataPreset(InID : int):CardData=
    var OutTexture : texture = VerseExposed.T_EmptyIcon
    var OutRarity : int = 0
    case(InID):
        0 =>
            set OutRarity = 0
            set OutTexture = VerseExposed.T_Card001UI
        1 =>
            set OutRarity = 0
            set OutTexture = VerseExposed.T_Card002UI
        2 =>
            set OutRarity = 0
            set OutTexture = VerseExposed.T_Card003UI
        3 =>
            set OutRarity = 1
            set OutTexture = VerseExposed.T_Card004UI
        4 =>
            set OutRarity = 1
            set OutTexture = VerseExposed.T_Card005UI
        5 =>
            set OutRarity = 1
            set OutTexture = VerseExposed.T_Card006UI
        6 =>
            set OutRarity = 2
            set OutTexture = VerseExposed.T_Card007UI
        7 =>
            set OutRarity = 2
            set OutTexture = VerseExposed.T_Card008UI
        8 =>
            set OutRarity = 3
            set OutTexture = VerseExposed.T_Card009UI
        9 =>
            set OutRarity = 3
            set OutTexture = VerseExposed.T_Card010UI
        10 =>
            set OutRarity = 0
            set OutTexture = VerseExposed.T_Card011UI
        11 =>
            set OutRarity = 0
            set OutTexture = VerseExposed.T_Card012UI
        12 =>
            set OutRarity = 1
            set OutTexture = VerseExposed.T_Card013UI
        13 =>
            set OutRarity = 1
            set OutTexture = VerseExposed.T_Card014UI
        14 =>
            set OutRarity = 3
            set OutTexture = VerseExposed.T_Card015UI
        15 =>
            set OutRarity = 3
            set OutTexture = VerseExposed.T_Card016UI
        16 =>
            set OutRarity = 2
            set OutTexture = VerseExposed.T_Card017UI
        17 =>
            set OutRarity = 1
            set OutTexture = VerseExposed.T_Card018UI
        18 =>
            set OutRarity = 1
            set OutTexture = VerseExposed.T_Card019UI
        19 =>
            set OutRarity = 2
            set OutTexture = VerseExposed.T_Card020UI
        20 =>
            set OutRarity = 3
            set OutTexture = VerseExposed.T_Card021UI
        21 =>
            set OutRarity = 3
            set OutTexture = VerseExposed.T_Card022UI
        _ =>
            set OutRarity = 0
            set OutTexture = VerseExposed.T_EmptyIcon

    return CardData:
        ID := InID
        Rarity := OutRarity
        Level :=  0
        Image := OutTexture

LookupCardDataSaved(InSavedCardData : SavedCardData):CardData=
    var OutTexture : texture = VerseExposed.T_EmptyIcon
    case(InSavedCardData.ID):
        0 =>
            set OutTexture = VerseExposed.T_Card001UI
        1 =>
            set OutTexture = VerseExposed.T_Card002UI
        2 =>
            set OutTexture = VerseExposed.T_Card003UI
        3 =>
            set OutTexture = VerseExposed.T_Card004UI
        4 =>
            set OutTexture = VerseExposed.T_Card005UI
        5 =>
            set OutTexture = VerseExposed.T_Card006UI
        6 =>
            set OutTexture = VerseExposed.T_Card007UI
        7 =>
            set OutTexture = VerseExposed.T_Card008UI
        8 =>
            set OutTexture = VerseExposed.T_Card009UI
        9 =>
            set OutTexture = VerseExposed.T_Card010UI
        10 =>
            set OutTexture = VerseExposed.T_Card011UI
        11 =>
            set OutTexture = VerseExposed.T_Card012UI
        12 =>
            set OutTexture = VerseExposed.T_Card013UI
        13 =>
            set OutTexture = VerseExposed.T_Card014UI
        14 =>
            set OutTexture = VerseExposed.T_Card015UI
        15 =>
            set OutTexture = VerseExposed.T_Card016UI
        16 =>
            set OutTexture = VerseExposed.T_Card017UI
        17 =>
            set OutTexture = VerseExposed.T_Card018UI
        18 =>
            set OutTexture = VerseExposed.T_Card019UI
        19 =>
            set OutTexture = VerseExposed.T_Card020UI
        20 =>
            set OutTexture = VerseExposed.T_Card021UI
        21 =>
            set OutTexture = VerseExposed.T_Card022UI
        _ =>
            set OutTexture = VerseExposed.T_EmptyIcon

    return CardData:
        ID := InSavedCardData.ID
        Rarity := InSavedCardData.Rarity
        Level :=  InSavedCardData.Level
        Image := OutTexture

# EVERY CARD ADDED TO THE GAME NEEDS TO BE PUT IN THE CARD BONUS FUNCTION
GetCardBonus(InID : int, InLevel : int):float=
    var Bonus : float = 0.0
    case(InID):
        0 =>
            set Bonus = 3.0
        1 =>
            set Bonus = 2.0
        2 =>
            set Bonus = 2.0
        3 =>
            set Bonus = 2.0
        4 =>
            set Bonus = 2.0
        5 =>
            set Bonus = 5.0
        6 =>
            set Bonus = 3.0
        7 =>
            set Bonus = 3.0
        8 =>
            set Bonus = 3.0
        9 =>
            set Bonus = 3.0
        10 =>
            set Bonus = 3.0
        11 =>
            set Bonus = 3.0
        12 =>
            set Bonus = 2.0
        13 =>
            set Bonus = 2.0
        14 =>
            set Bonus = 2.0
        15 =>
            set Bonus = 2.0
        16 =>
            set Bonus = 3.0
        17 =>
            set Bonus = 2.0
        18 =>
            set Bonus = 2.0
        19 =>
            set Bonus = 2.0
        20 =>
            set Bonus = 2.0
        21 =>
            set Bonus = 10.0
        _ =>
            set Bonus = 5.0

    return Bonus * InLevel

SearchForCard(InID : int, InPlayer : player, InPlayerStatsManager : player_stats_manager)<transacts>:?SavedCardData=
    if(InID = -1):
        return false
        
    if(PlayerStats := InPlayerStatsManager.GetPlayerStats[InPlayer]):
        for(CurrentCard : PlayerStats.OwnedCards):
            if(CurrentCard.ID = InID):
                return option{CurrentCard}

    return false

# A Verse-authored creative device that can be placed in a level
V_CardManager := class(creative_device):
    @editable
    OpenUIButton : button_device = button_device{}
    @editable
    OpenButtons : []button_device = array{}
    @editable
    Cinematics : []cinematic_sequence_device = array{}
    @editable
    CardProps : []creative_prop = array{}
    @editable
    CardAudios : []audio_player_device = array{}

    @editable
    CardSwitches : []switch_device = array{}

    @editable
    CheckVolume : volume_device = volume_device{}

    @editable
    Accolades : accolades_device = accolades_device{}

    @editable
    PlayerManager : ?V_PlayerManagerDevice = false

    # Manages and updates stats for each player.
    var PlayerStatManager: player_stats_manager = player_stats_manager{}

    CardCommonChance : float = 50.0
    CardRareChance : float = 32.5
    CardEpicChance : float = 15.0
    CardLegendaryChance : float = 2.5

    # ADD THE IDS OF THE CARDS INTO THIS ARRAY
    CommonCardIDs : []int = array{0,1,2,10,11}
    RareCardIDs : []int = array{3,4,5,12,13,17,18}
    EpicCardIDs : []int = array{6,7,16,19}
    LegendaryCardIDs : []int = array{8,9,14,15,20,21}
    UICardDisplay : UI_CardDisplay = UI_CardDisplay{}
    
    # Runs when the device is started in a running game
    OnBegin<override>()<suspends>:void=
        OpenUIButton.InteractedWithEvent.Subscribe(OpenCardMenu)
        CheckVolume.AgentEntersEvent.Subscribe(CheckVolumeEnter)
        for(Index -> OpenButton : OpenButtons):
            OpenButton.InteractedWithEvent.Subscribe(UnlockCardHandler{CardManager := Self, Index := Index}.HandlerFunction)

    CheckVolumeEnter(InAgent : agent):void=
        if(Player := player[InAgent], PlayerStats := PlayerStatManager.GetPlayerStats[Player]):
            for(CurrentCard : PlayerStats.OwnedCards):
                if(Switch := CardSwitches[CurrentCard.ID]):
                    Switch.TurnOn(InAgent)

    InitiatePlayerUI(InPlayer : player):void=
        UICardDisplay.InitUI(InPlayer, PlayerStatManager)

    OpenCardMenu(InAgent : agent):void=
        if(Player := player[InAgent], PlayerStats := PlayerStatManager.GetPlayerStats[Player]):
            UICardDisplay.UpdateUI(Player, PlayerStatManager)
            UICardDisplay.AddUI(Player)
            

    GetRandomCard(InPlayer : player, InIndex : int)<suspends>:void=
        if(PlayerStats := PlayerStatManager.GetPlayerStats[InPlayer]):
            CanAfford := BigNumber_GreaterThanOrEquals(PlayerStats.PlayerGoldCoinAmount, big_number{Value := 5.0, Exponent := 0})
            if(CanAfford = false):
                return

        if(Button := OpenButtons[InIndex], Cinematic := Cinematics[InIndex], Prop := CardProps[InIndex], CardAudio := CardAudios[InIndex]):
            Button.Disable()
            Cinematic.Play()
            CardAudio.Play()

            if(bPlayerManager := PlayerManager?):
                bPlayerManager.QuestFCHQ.QuestCompleted.Signal((InPlayer, "Card Collector"))

            PlayerStatManager.SpendGoldCoin(InPlayer, big_number{Value := 5.0, Exponent := 0})
            var RandomRarityInt : int = 0

            TypeRandomChance := GetRandomFloat(0.0, 100.0)

            if(CardLegendaryChance >= TypeRandomChance):
                set RandomRarityInt = 3
            else if(CardEpicChance >= TypeRandomChance):
                set RandomRarityInt = 2
            else if(CardRareChance >= TypeRandomChance):
                set RandomRarityInt = 1
            else if(CardCommonChance >= TypeRandomChance):
                set RandomRarityInt = 0

            var RandomCardID : []int = array{}

            case(RandomRarityInt):
                0 =>
                    set RandomCardID = CommonCardIDs
                1 =>
                    set RandomCardID = RareCardIDs
                2 =>
                    set RandomCardID = EpicCardIDs
                3 =>
                    set RandomCardID = LegendaryCardIDs
                _=>
                    set RandomCardID = CommonCardIDs

            ShuffledCardID := Shuffle(RandomCardID)

            if(CardID := ShuffledCardID[0]):
                SetPropMaterial(Prop, CardID)
                Sleep(5.0)

                PlayerStatManager.GiveCard(InPlayer, CardID, RandomRarityInt)
                if(CardSwitch := CardSwitches[CardID]):
                    CardSwitch.TurnOn(InPlayer)

                Print("RandomRarityInt : {RandomRarityInt} | CardID : {CardID}")
                Button.Enable()
                Accolades.Award(InPlayer)

    SetPropMaterial(InProp : creative_prop, InID : int):void=
        case(InID):
            0 =>
                InProp.SetMaterial(VerseExposed.MI_Card001, ?Index := 3)
            1 =>
                InProp.SetMaterial(VerseExposed.MI_Card002, ?Index := 3)
            2 =>
                InProp.SetMaterial(VerseExposed.MI_Card003, ?Index := 3)
            3 =>
                InProp.SetMaterial(VerseExposed.MI_Card004, ?Index := 3)
            4 =>
                InProp.SetMaterial(VerseExposed.MI_Card005, ?Index := 3)
            5 =>
                InProp.SetMaterial(VerseExposed.MI_Card006, ?Index := 3)
            6 =>
                InProp.SetMaterial(VerseExposed.MI_Card007, ?Index := 3)
            7 =>
                InProp.SetMaterial(VerseExposed.MI_Card008, ?Index := 3)
            8 =>
                InProp.SetMaterial(VerseExposed.MI_Card009, ?Index := 3)
            9 =>
                InProp.SetMaterial(VerseExposed.MI_Card010, ?Index := 3)
            10 =>
                InProp.SetMaterial(VerseExposed.MI_Card011, ?Index := 3)
            11 =>
                InProp.SetMaterial(VerseExposed.MI_Card012, ?Index := 3)
            12 =>
                InProp.SetMaterial(VerseExposed.MI_Card013, ?Index := 3)
            13 =>
                InProp.SetMaterial(VerseExposed.MI_Card014, ?Index := 3)
            14 =>
                InProp.SetMaterial(VerseExposed.MI_Card015, ?Index := 3)
            15 =>
                InProp.SetMaterial(VerseExposed.MI_Card016, ?Index := 3)
            16 =>
                InProp.SetMaterial(VerseExposed.MI_Card017, ?Index := 3)
            17 =>
                InProp.SetMaterial(VerseExposed.MI_Card018, ?Index := 3)
            18 =>
                InProp.SetMaterial(VerseExposed.MI_Card019, ?Index := 3)
            19 =>
                InProp.SetMaterial(VerseExposed.MI_Card020, ?Index := 3)
            20 =>
                InProp.SetMaterial(VerseExposed.MI_Card021, ?Index := 3)
            21 =>
                InProp.SetMaterial(VerseExposed.MI_Card022, ?Index := 3)
            _ =>
                InProp.SetMaterial(VerseExposed.MI_Card001, ?Index := 3)