
using { /Fortnite.com/Devices }
using { /Verse.org/Simulation }
using { /UnrealEngine.com/Temporary/Diagnostics }

# See https://dev.epicgames.com/documentation/en-us/uefn/create-your-own-device-in-verse for how to create a verse device.

# A Verse-authored creative device that can be placed in a level
V_ChangeHatIndex := class(creative_device):

    @editable
    HatID : int = -1
    @editable
    Buttons : []button_device = array{}
    @editable
    Switch : ?switch_device = false

    PlayerStatsManager : player_stats_manager = player_stats_manager{}

    # Runs when the device is started in a running game
    OnBegin<override>()<suspends>:void=
        for(Button : Buttons):
            Button.InteractedWithEvent.Subscribe(OnButtonInteract)

    OnButtonInteract(InAgent : agent):void=
        if(bSwitch := Switch?):
            if(bSwitch.GetCurrentState[InAgent]):
                if(Player := player[InAgent]):
                    Print("Change HatID to {HatID}")
                    PlayerStatsManager.ChangeHatID(Player, HatID)
            else:
                Print("Don't Own Hat")
        else:
            if(Player := player[InAgent]):
                PlayerStatsManager.ChangeHatID(Player, HatID)
