
using { /Fortnite.com/Devices }
using { /Verse.org/Simulation }
using { /UnrealEngine.com/Temporary/Diagnostics }

#Added
using { /Verse.org/Simulation/Tags }
using { /Fortnite.com/Characters}
# See https://dev.epicgames.com/documentation/en-us/uefn/create-your-own-device-in-verse for how to create a verse device.

ChestButtons := class(tag){}

# A Verse-authored creative device that can be placed in a level
V_ChestEasterEgg := class(creative_device):
    @editable
    InitVolume : volume_device = volume_device{}

    var DeviceInitiated : logic = false
    
    # Runs when the device is started in a running game
    OnBegin<override>()<suspends>:void=
        InitVolume.AgentEntersEvent.Subscribe(OnAgentEnters)
        
    # UPDATE 1.1 FIX Moved from beginplay to on volume enter
    OnAgentEnters(Agent : agent):void=
        if (DeviceInitiated = false):
            set DeviceInitiated = true
            TaggedList := FindCreativeObjectsWithTag(ChestButtons{})
            for(TaggedActor : TaggedList):
                if(Prop := button_device[TaggedActor]):
                    Prop.InteractedWithEvent.Subscribe(EliminatePlayer)

    #This eliminates players if they choose the wrong button
    EliminatePlayer(InAgent : agent):void=
        if(FortChar := InAgent.GetFortCharacter[]):
            FortChar.Damage(1000.0)