
using { /Fortnite.com/Devices }
using { /Verse.org/Simulation }
using { /UnrealEngine.com/Temporary/Diagnostics }

# See https://dev.epicgames.com/documentation/en-us/uefn/create-your-own-device-in-verse for how to create a verse device.

# A Verse-authored creative device that can be placed in a level
V_CreatureManager := class(creative_device):
    @editable
    CreatureSpawner : creature_spawner_device = creature_spawner_device{}

    @editable
    DungeonEntranceManager : V_DungeonEntranceManager = V_DungeonEntranceManager{}

    @editable
    DungeonEntranceTrigger : trigger_device = trigger_device{}

    @editable
    MaxEliminations : int = 20

    @editable
    UnlockButton : conditional_button_device = conditional_button_device{}

    var CurentElim : int = 0

    var ObjectiveFinished : logic = false

    @editable
    BarrierDevice : barrier_device = barrier_device{}

    # Runs when the device is started in a running game
    OnBegin<override>()<suspends>:void=
        # TODO: Replace this with your code
        DungeonEntranceTrigger.TriggeredEvent.Subscribe(OnGameStart)
        #CreatureSpawner.EliminatedEvent.Subscribe(OnCreatureEliminated)
        UnlockButton.ActivatedEvent.Subscribe(OnButtonActivated)

    OnGameStart(Agent : ?agent):void=
        set ObjectiveFinished = false
        set CurentElim = 0
        BarrierDevice.Enable()
        CreatureSpawner.Enable()

    OnCreatureEliminated(Agent : agent):void=
        if (ObjectiveFinished = true):
            return
        set CurentElim += 1
        if (CurentElim >= MaxEliminations):
            set ObjectiveFinished = true
            BarrierDevice.Disable()
            CreatureSpawner.Disable()
            CreatureSpawner.EliminateCreatures()
            DungeonEntranceManager.ObjectiveCompleteTrigger.Trigger()

    OnButtonActivated(Agent:agent):void=
        if (ObjectiveFinished = true):
            return
        set ObjectiveFinished = true
        BarrierDevice.Disable()
        CreatureSpawner.Disable()
        CreatureSpawner.EliminateCreatures()
        DungeonEntranceManager.ObjectiveCompleteTrigger.Trigger()
        Time := DungeonEntranceManager.Timer.GetActiveDuration()
        DungeonEntranceManager.Timer.SetActiveDuration(Time + 120.0)
