
using { /Fortnite.com/Devices }
using { /Verse.org/Simulation }
using { /UnrealEngine.com/Temporary/Diagnostics }

# See https://dev.epicgames.com/documentation/en-us/uefn/create-your-own-device-in-verse for how to create a verse device.

# A Verse-authored creative device that can be placed in a level
V_CustomConditionalButton := class(creative_device):

    @editable
    ConditionalButton : ?conditional_button_device = false
    @editable
    Trigger : ?trigger_device = false
    @editable
    Cost : float = 1.0
    @editable
    ResourceType : PropResourceType = PropResourceType.Null

    # Verse Devices
    @editable
    PlayerManager : ?V_PlayerManagerDevice = false

    # Runs when the device is started in a running game
    OnBegin<override>()<suspends>:void=
        # Subscribe to button interact
        if(bButton := ConditionalButton?):
            bButton.ActivatedEvent.Subscribe(OnButtonInteract)

    # Subscribed to button interact
    OnButtonInteract(InAgent : agent):void=
        # Check if player has enough resources
        if(bPlayerManager := PlayerManager?):
            if(Player := player[InAgent], CurrentPlayerStats := bPlayerManager.PlayerStatManager.GetPlayerStats[Player]):
                #Spend Resources
                if(ResourceType = PropResourceType.Coin):
                    StoneAmount := bPlayerManager.PlayerStatManager.GetPlayerStoneOreAmount(InAgent)
                    BNCost := BigNumber_ConvertFloatIntoBigNumber(Cost)
                    CanAfford := BigNumber_GreaterThanOrEquals(StoneAmount, BNCost)
                    if(CanAfford = true):
                        bPlayerManager.PlayerStatManager.SpendPlayerOre(InAgent, BNCost, false, PropResourceType.Stone)
                        if(bTrigger := Trigger?):
                            bTrigger.Trigger(Player)
                else:
                    ResourceAmount := bPlayerManager.PlayerStatManager.GetPlayerOreAmount(Player, ResourceType)
                    BNCost := BigNumber_ConvertFloatIntoBigNumber(Cost)
                    CanAfford := BigNumber_GreaterThanOrEquals(ResourceAmount, BNCost)
                    if(CanAfford = true):
                        bPlayerManager.PlayerStatManager.SpendPlayerOre(Player, BNCost, false, ResourceType)
                        if(bTrigger := Trigger?):
                            bTrigger.Trigger(Player)
                    
        Print("Button Pressed")