using { /Fortnite.com/Devices }
using { /Verse.org/Simulation }
# See https://dev.epicgames.com/documentation/en-us/uefn/create-your-own-device-in-verse for how to create a verse device.


### Added
using { /UnrealEngine.com/Temporary/Diagnostics }
using { /Fortnite.com/Characters }
using { /UnrealEngine.com/Temporary/SpatialMath }
using { /Fortnite.com/AI }
using { /Verse.org/Random }
using { /Fortnite.com/Devices/CreativeAnimation }

#Setup the enum, currently we are only doing idle, walk and attack.
#Keep in mind we can do a ranged attack custom. Enemies do come with a default one
#Attacking is currently melee as the creatures use melee attacks
#We could invest more time in the future and make this way more custom. The advantage of this method is having the path finding avaliable to us.
#Think of the mesh we use as basically cosmetic. The value is the creature logic that Epic created.
MovementState := enum{Idle, Walking, Attacking}

enemy_info_class := class<concrete>:
    #These are the blueprints we will use. Each blueprint must be seperate.
    #Skeletal Mesh + Animation on that skeletal mesh. Idle, Walk and Attack. 
    #The code handles how it swaps between each blueprint. 
    @editable 
    IdleAsset : creative_prop_asset = DefaultCreativePropAsset
    @editable 
    WalkAsset : creative_prop_asset = DefaultCreativePropAsset
    @editable 
    AttackAsset : creative_prop_asset = DefaultCreativePropAsset

    #Variables to determine custom enemy behaviour
    @editable 
    Health : float = 1.0
    @editable 
    Damage : float = 0.0
    @editable 
    AttackWindupTime : float = 0.5
    @editable 
    AttackCooldownTime : float = 0.5

# A Verse-authored creative device that can be placed in a level
V_CustomCreatures := class(creative_device):
    #The spawner or spawners that you want to use
    @editable 
    Spawners : []creature_spawner_device = array{}
    #Give flexibility if we want this spawner to be enabled or disabled via a trigger. Say a player defeats X enemies or performs something in the area
    @editable 
    EnableTrigger : trigger_device = trigger_device{}
    @editable 
    DisableTrigger : trigger_device = trigger_device{}
    #Should this start enabled?
    @editable 
    var Enabled : logic = true
    #This array takes the enemy_info_class we've setup above and allows us to customise. We could have a few different types of enemies within the array.
    @editable 
    var EnemyArray : []enemy_info_class = array{}

    @editable
    var DestroyedVFX : creative_prop_asset = DefaultCreativePropAsset

    var PlayerRef : []player = array{}
    
    # Runs when the device is started in a running game
    OnBegin<override>()<suspends>:void=
        EnableTrigger.TriggeredEvent.Subscribe(Enable)      
        DisableTrigger.TriggeredEvent.Subscribe(Disable)
        for (Spawner : Spawners):
            Spawner.SpawnedEvent.Subscribe(CreatureSpawned)  
            Spawner.EliminatedEvent.Subscribe(SpawnHitEffects)

    # - - - - - - - - - - SPECIAL EFFECTS METHOD - - - - - - - - - - - # 
    SpawnHitEffects<public>(DeviceAIResult : device_ai_interaction_result):void=
        Print("Creature Destroyed")
        <#
        CreatureAgent := DeviceAIResult.Target
        if(Agent := CreatureAgent?, FortChar := Agent.GetFortCharacter[]):
            Print("Spawn VFX")
            #DeviceAIResult.Target
            #Spawn above the Prop Location
            SpawnPosition := vector3{X:= FortChar.GetTransform().Translation.X, Y:= FortChar.GetTransform().Translation.Y, Z:= FortChar.GetTransform().Translation.Z + 100.0}
            SpawnScale := vector3{X:= 1.0, Y:= 1.0, Z:= 1.0}
            SpawnedProp := SpawnProp(DestroyedVFX, transform{Translation := SpawnPosition, Rotation := IdentityRotation(), Scale:= SpawnScale})
            if(CreativeProp := SpawnedProp(0)?):
                spawn:
                    DestroyEffects(CreativeProp, 1.5)
        #>
    DestroyEffects(SpawnedProp : creative_prop, TimeToDispose : float)<suspends>:void=
        Sleep(TimeToDispose)
        SpawnedProp.Dispose()

    #Tracks the enemies health/damage. If we do defeat it 
    DamageTracker(FortChar : fort_character,MaxHealth : float)<suspends>: void=
        var  Health : float = MaxHealth
        #FortChar.Damage(1000.0)
        loop:
            DamageStruct := FortChar.DamagedEvent().Await()
            Print("Damaged by {DamageStruct.Amount}")
            set Health -= DamageStruct.Amount
            if (Health <= 0.0):
                Players := GetPlayspace().GetPlayers()
                EnemyPos := FortChar.GetTransform().Translation
                var DistanceToPlayer : float = 99999999.0
                for(Player : Players):
                    if(PlayerCharacter := Player.GetFortCharacter[]):
                        PlayerPos := PlayerCharacter.GetTransform().Translation
                        NewDistanceToPlayer := Distance(EnemyPos, PlayerPos)
                        if(NewDistanceToPlayer < DistanceToPlayer):
                            set DistanceToPlayer = NewDistanceToPlayer
                            set PlayerRef = array{Player}
                Print("Dead")
                #FortChar.Damage(999999.0)
                #if (FortChar.TeleportTo[vector3{Z:=-1000.0},IdentityRotation()]){}

    #When the creature is spawned we add it to our array. We can set the health and hide it! 
    #They are Fort Characters, this opens up so many possibilities
    CreatureSpawned(Agent : agent): void=
        if (Enabled = true):
            if (FortChar := Agent.GetFortCharacter[],Enemy := EnemyArray[GetRandomInt(0,EnemyArray.Length-1)]):
                FortChar.SetMaxHealth(Enemy.Health)
                FortChar.Hide()
                spawn{EnemyLoop(FortChar,Enemy)}

    #This loops through all of the enemies and allows the following logic
    #Within a certain distance will set the enemy to attack
    #Within a certain distance and not moving/attacking will set the enemy to idle
    #Within a certain distance and not moving will set enemy to attack
    EnemyLoop(FortChar : fort_character,EnemyInfo : enemy_info_class)<suspends>: void =
        #spawn{DamageTracker(FortChar,EnemyInfo.Health)}
        var LastTransform : transform = FortChar.GetTransform()
        var LastTranslation : vector3 = vector3{X:= LastTransform.Translation.X, Y:= LastTransform.Translation.Y, Z:= LastTransform.Translation.Z}
        MaybeProp := SpawnProp(EnemyInfo.IdleAsset,LastTranslation,LastTransform.Rotation)(0)
        
        var CurrentProp : creative_prop = creative_prop{}

        #We need to have the states set here to access them below
        var CurrentState : MovementState = MovementState.Idle
        var LastState : MovementState = MovementState.Idle
        var WalkSpeed : float = 10.0
        if(TempMaybeProp := MaybeProp?):
            set CurrentProp = TempMaybeProp
            loop:
                if (not FortChar.IsActive[]):
                    CurrentProp.Dispose()
                    break
                if (Distance(FortChar.GetTransform().Translation, LastTransform.Translation) = 0.0):
                    PlaySpace := GetPlayspace()
                    var PlayerNearby : logic = false
                    for (Player : PlaySpace.GetPlayers()):
                        if (LoopFortChar := Player.GetFortCharacter[]):
                            if (Distance(FortChar.GetTransform().Translation, LoopFortChar.GetTransform().Translation) < 120.0):
                                if (LoopFortChar.IsActive[]):
                                    set PlayerNearby = true

                    if (PlayerNearby = true):
                        set CurrentState = MovementState.Attacking

                        CurrentProp.Dispose()
                        MaybeAttackProp := SpawnProp(EnemyInfo.AttackAsset,FortChar.GetTransform().Translation,FortChar.GetTransform().Rotation)(0)
                        if(TempMaybeAttackProp := MaybeAttackProp?):
                            set CurrentProp = TempMaybeAttackProp

                            Sleep(EnemyInfo.AttackWindupTime)
                            for (Player : PlaySpace.GetPlayers()):
                                if (LoopFortChar := Player.GetFortCharacter[]):
                                    if (Distance(FortChar.GetTransform().Translation, LoopFortChar.GetTransform().Translation) < 120.0):
                                        LoopFortChar.Damage(EnemyInfo.Damage)
                    else:
                        if (not CurrentState = MovementState.Idle):

                            set CurrentState = MovementState.Idle

                            CurrentProp.Dispose()
                            MaybeIdleProp := SpawnProp(EnemyInfo.IdleAsset,FortChar.GetTransform().Translation,FortChar.GetTransform().Rotation)(0)
                            if(TempMaybeIdleProp := MaybeIdleProp?):
                                set CurrentProp = TempMaybeIdleProp
                else:
                    if (not CurrentState = MovementState.Walking):
                        set CurrentState = MovementState.Walking
                        CurrentProp.Dispose()
                        MaybeWalkProp := SpawnProp(EnemyInfo.WalkAsset,FortChar.GetTransform().Translation,FortChar.GetTransform().Rotation)(0)
                        if(TempMaybeWalkProp := MaybeWalkProp?):
                            set CurrentProp = TempMaybeWalkProp

                set LastTransform = FortChar.GetTransform()
                CurrentProp.MoveTo(FortChar.GetTransform().Translation,FortChar.GetTransform().Rotation,0.1)

    #If we want to enable or disable them in the future
    Enable(MaybeAgent : ?agent): void =
        set Enabled = true

    Disable(MaybeAgent : ?agent): void =
        set Enabled = false