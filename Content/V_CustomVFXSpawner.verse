using { /Fortnite.com/Devices }
using { /Verse.org/Simulation }
using { /UnrealEngine.com/Temporary/Diagnostics }

#Added
using { /Verse.org/Simulation/Tags }
using { /UnrealEngine.com/Temporary/SpatialMath}
using { /Fortnite.com/Characters }
using { /Verse.org/Random }

#Tags
VFXReferenceHolderTag := class(tag){}

#Enums
VFXToSpawnEnum := enum:
    eExplosiveHitTrailVFX,  
    eExplosiveHitImpactVFX,
    eLastHitVFX,
    eCriticalHitVFX,
    eChargedCriticalHitVFX,
    ePrestigeLevelUpVFX,
    Null

#Custom VFX Spawner Allows us to easily spawn VFX in any Verse device
#In future projects we modify this script for re-use as effects, names etc.. are project specific
V_CustomVFXSpawner := class(creative_device):
    #Optional VFX that we want to spawn
    @editable
    ExplosiveHitTrailVFX : ?creative_prop_asset = false
    @editable
    ExplosiveHitImpactVFX : ?creative_prop_asset = false
    @editable
    LastHitVFX : ?creative_prop_asset = false
    @editable
    CriticalHit : ?creative_prop_asset = false
    @editable
    ChargedCriticalHitVFX : ?creative_prop_asset = false
    @editable
    PrestigeVFX : ?creative_prop_asset = false

    #Audio Effects To Play we set in the device
    @editable
    AudioToPlayForVFX : []audio_player_device = array{}

    @editable
    PrestigeVFXTrigger : trigger_device = trigger_device{}

    #Allows us to return a null VFX as default so nothing spawns
    NullVFX : ?creative_prop_asset = false

    #Allows us to set the int in the array
    var AudioToPlay : int = 0
    
    # Runs when the device is started in a running game
    OnBegin<override>()<suspends>:void=
        PrestigeVFXTrigger.TriggeredEvent.Subscribe(TriggerPrestigeVFX)

    #Trigger for Prestige VFX
    TriggerPrestigeVFX(InAgent : ?agent):void=
        if(bInAgent := InAgent?):
            spawn:
                SpawnVFXOnPlayer(0.2,  VFXToSpawnEnum.ePrestigeLevelUpVFX, 2.0, bInAgent)

    #Return VFX that can be spawned.
    VFXToSpawn(VFXSpawnEnum : VFXToSpawnEnum): ?creative_prop_asset=
        var VFXToReturn : ?creative_prop_asset = false
        case(VFXSpawnEnum):
            VFXToSpawnEnum.eExplosiveHitTrailVFX => 
                set VFXToReturn = ExplosiveHitTrailVFX
                set AudioToPlay = 2
            VFXToSpawnEnum.eExplosiveHitImpactVFX => 
                set VFXToReturn = ExplosiveHitImpactVFX
                set AudioToPlay = 3
            VFXToSpawnEnum.eLastHitVFX => 
                set VFXToReturn = LastHitVFX
            VFXToSpawnEnum.eCriticalHitVFX => 
                set VFXToReturn = CriticalHit
                set AudioToPlay = 0
            VFXToSpawnEnum.eChargedCriticalHitVFX => 
                set VFXToReturn = ChargedCriticalHitVFX
                set AudioToPlay = 1
            VFXToSpawnEnum.ePrestigeLevelUpVFX => 
                set VFXToReturn = PrestigeVFX
                set AudioToPlay = 4
            _=> set VFXToReturn = NullVFX

        return VFXToReturn

    SpawnVFXOnPlayer(TimeToWait: float, SpawnVFXEnum : VFXToSpawnEnum, TimeToDispose : float, InAgent : agent)<suspends>:void=
        Sleep(TimeToWait)
        if(FortChar := InAgent.GetFortCharacter[]):
            #The effect we'll use based on the enum
            SpawnEffect := VFXToSpawn(SpawnVFXEnum)
            #Need to lower for player VFX
            SpawnLocation : vector3 = vector3{X:=FortChar.GetTransform().Translation.X, Y:=FortChar.GetTransform().Translation.Y, Z:= FortChar.GetTransform().Translation.Z -15.0}

            #Optional variable cast
            if(bVFX:= SpawnEffect?):
                #Spawn the prop at the SpawnLocation with a default rotation and Scale
                SpawnedProp := SpawnProp(bVFX, transform{Translation := SpawnLocation, Rotation := IdentityRotation(), Scale:= vector3{X:=1.0,Y:=1.0,Z:=1.0}})
                #Play Sound Effects for the InAgent based on the enum
                if(SFXToPlay := AudioToPlayForVFX[AudioToPlay]):
                    SFXToPlay.Play(InAgent)
                #Get the Creative_Prop
                if(CreativeProp := SpawnedProp(0)?):
                    #After Spawning we Dispose, this way the entire system can handle spawning and disposing VFX rather then having unique functions across all Verse Scripts
                    DisposeVFX(CreativeProp, TimeToDispose)
    
    #Function we can call from any script when wanting to display then dispose VFX
    SpawnVFX(PropToSpawnAt : creative_prop, TimeToWait : float, SpawnVFXEnum : VFXToSpawnEnum, TimeToDispose : float, InAgent : agent, SpawnInfrontOfPlayer : logic)<suspends>:void=
        var SpawnLocation : vector3 = vector3{X:= 0.0, Y:= 0.0, Z:= 0.0}
        Sleep(TimeToWait)
        if(FortChar := InAgent.GetFortCharacter[]):
            #Location of the Prop
            PropLocation := PropToSpawnAt.GetTransform().Translation
            #The Players View Rotation, where the camera is looking
            ViewRotation := FortChar.GetViewRotation()

            #We get the distance between the Prop and the Fort Character.
            #From there we divide by 2.0 to get a middle point that looks quite good if you're close to the ore or far away
            #We then multiply your local forward view by that impact point
            DistanceBetweenPoints := Distance(PropToSpawnAt.GetTransform().Translation, FortChar.GetTransform().Translation)
            ImpactPoint := DistanceBetweenPoints / 2.0
            ViewForward := ViewRotation.GetLocalForward() * ImpactPoint
            #Allows each hit to be slightly random with the VFX in terms of positioning. Players will see hit VFX a lot so it's important to make it varied.
            RandomVarianceOnHit := GetRandomFloat (-5.0, 5.0)
            #Setup Spawn Location and the Effect to spawn #DistanceBetweenPoints
            if(SpawnInfrontOfPlayer = true):
                #We get the player location for X and Y then add the ViewForward. We use the player Z as we don't want the effect dependent on the camera view
                set SpawnLocation = vector3{X:= FortChar.GetTransform().Translation.X + RandomVarianceOnHit + ViewForward.X, Y:= FortChar.GetTransform().Translation.Y + RandomVarianceOnHit + ViewForward.Y, Z:= FortChar.GetTransform().Translation.Z + RandomVarianceOnHit}
            else:
                #For anything else we can just use the props location and spawn the VFX there
                set SpawnLocation = vector3{X:= PropToSpawnAt.GetTransform().Translation.X, Y:= PropToSpawnAt.GetTransform().Translation.Y, Z:= PropToSpawnAt.GetTransform().Translation.Z + 20.0}

        #The effect we'll use based on the enum
        SpawnEffect := VFXToSpawn(SpawnVFXEnum)

        #Optional variable cast
        if(bVFX:= SpawnEffect?):
            #Spawn the prop at the SpawnLocation with a default rotation and Scale
            SpawnedProp := SpawnProp(bVFX, transform{Translation := SpawnLocation, Rotation := IdentityRotation(), Scale:= vector3{X:=1.0,Y:=1.0,Z:=1.0}})
            #Play Sound Effects for the InAgent based on the enum
            if(SFXToPlay := AudioToPlayForVFX[AudioToPlay]):
                SFXToPlay.Play(InAgent)
            #Get the Creative_Prop
            if(CreativeProp := SpawnedProp(0)?):
                #After Spawning we Dispose, this way the entire system can handle spawning and disposing VFX rather then having unique functions across all Verse Scripts
                DisposeVFX(CreativeProp, TimeToDispose)

    #Dispose the VFX after a set time
    DisposeVFX(SpawnedProp : creative_prop, TimeToDispose : float)<suspends>:void=
        Sleep(TimeToDispose)
        SpawnedProp.Dispose()