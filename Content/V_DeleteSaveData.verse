
using { /Fortnite.com/Devices }
using { /Verse.org/Simulation }
using { /UnrealEngine.com/Temporary/Diagnostics }

# See https://dev.epicgames.com/documentation/en-us/uefn/create-your-own-device-in-verse for how to create a verse device.

# A Verse-authored creative device that can be placed in a level
V_DeleteSaveData := class(creative_device):
    @editable
    PlayerManager : V_PlayerManagerDevice = V_PlayerManagerDevice{}
    @editable
    ClearButton : button_device = button_device{}
    @editable
    ClearGranter : item_granter_device = item_granter_device{}

    # Runs when the device is started in a running game
    OnBegin<override>()<suspends>:void=
        ClearButton.InteractedWithEvent.Subscribe(OnClearSave)

    OnClearSave(Agent : agent):void=
        block: