
using { /Fortnite.com/Devices }
using { /Verse.org/Simulation }
using { /UnrealEngine.com/Temporary/Diagnostics }

using { /Fortnite.com/Characters }
using { /Fortnite.com/Game }
using { /Fortnite.com/FortPlayerUtilities }
using { /UnrealEngine.com/Temporary/SpatialMath }
# See https://dev.epicgames.com/documentation/en-us/uefn/create-your-own-device-in-verse for how to create a verse device.

# A Verse-authored creative device that can be placed in a level
V_DungeonEntranceManager := class(creative_device):
    @editable
    Timer : timer_device = timer_device{}

    @editable
    Buttons : []conditional_button_device = array{}

    @editable
    DungeonVolume : mutator_zone_device = mutator_zone_device{}

    @editable
    EntranceTeleporter : teleporter_device = teleporter_device{}

    @editable
    ChestTeleporter : teleporter_device = teleporter_device{}

    @editable
    GameStartTrigger : trigger_device = trigger_device{}

    @editable
    GameEndTrigger : trigger_device = trigger_device{}

    @editable
    ObjectiveCompleteTrigger : trigger_device = trigger_device{}

    @editable
    Chest : switch_device = switch_device{}

    @editable
    RewardGranter : trigger_device = trigger_device{}

    @editable
    DungeonAudio : audio_player_device = audio_player_device{}

    @editable
    DungeonCompleteHUD : hud_message_device = hud_message_device{}

    @editable
    PlayerManager : ?V_PlayerManagerDevice = false

    @editable
    BossDeviceRef : V_BossDevice = V_BossDevice{}

    @editable
    VisualSwitch : switch_device = switch_device{}
    
    @editable
    DungeonStartTeleporter : teleporter_device = teleporter_device{}

    @editable
    LeaveDungeonButton : button_device = button_device{}

    #This is purely visual for players who are in the dungeon
    @editable
    HUDDungeonTimer : timer_device = timer_device{}

    var JoinedPlayers : []player = array{}

    var bDungeonStarted : logic = false
    var EliminatedPlayers : int = 0
    var Subscriptions : []cancelable = array{}
    var CountdownTime : int = 8
    ToMessage<localizes>(String : string) : message = "{String}"

    # Runs when the device is started in a running game
    OnBegin<override>()<suspends>:void=
        Timer.SuccessEvent.Subscribe(OnTimerEnded)
        ObjectiveCompleteTrigger.TriggeredEvent.Subscribe(OnObjectiveComplete)
        Chest.TurnedOnEvent.Subscribe(ChestOpened)

        #Button to leave the dungeon
        LeaveDungeonButton.InteractedWithEvent.Subscribe(LeaveDungeon)
        for(Button:Buttons):
            Button.ActivatedEvent.Subscribe(OnButtonActivated)

    OnButtonActivated(Agent : agent):void=
        if (bDungeonStarted = true):
            if (Player := player[Agent]):
                set JoinedPlayers += array{Player}
            return
        set bDungeonStarted = true
        if (Player := player[Agent]):
            set JoinedPlayers += array{Player}
        if (FortChar := Agent.GetFortCharacter[]):
            #Add PlayerEliminated to our cancelable subscriptions array
            var EliminationSubscription : cancelable = FortChar.EliminatedEvent().Subscribe(PlayerEliminated)
            if(NewArray := Subscriptions.Insert[0, array{EliminationSubscription}]):
                set Subscriptions = NewArray
        spawn:
            StartTimer(Agent )

        
    #Add to the Eliminated Players 
    PlayerEliminated(EliminatedResults : elimination_result):void=
        set EliminatedPlayers = EliminatedPlayers + 1
    
    #Handle Player States In Dungeon
    HandlePlayers()<suspends>:void=
        loop:
            Sleep(1.0)
            var PlayersOutOfVolume : int = 0
            #Print("Checking Players")
            #Set Downed Players and Eliminated as temp variables
            #var DownedPlayers : int = 0
            for(Player : JoinedPlayers):
                #Account for eliminated Players or Downed Players
                #Print("Looping Players")
                if(DungeonVolume.IsInVolume[Player]):
                    #Player is in the volume
                else:
                    #Player is out of the volume
                    #Increase the amount
                    set PlayersOutOfVolume = PlayersOutOfVolume +1
                    #Stops dungeon audio
                    DungeonAudio.Stop(Player)
                    #Removes the timer for that player
                    HUDDungeonTimer.Complete(Player)
                    #Timer.Complete(Player)

            #Once players are all dead handle logic here
            if(PlayersOutOfVolume >= JoinedPlayers.Length):
                Print("End Dungeon, all players are defeated")
                #Makes the array empty after players are teleported back
                #Reset the Boss

                #Complete the timer for the joined players
                for(Player : JoinedPlayers):
                    Timer.Complete(Player)

                #Enables the dungeon
                EnableDungeon()
                GameEndTrigger.Trigger()
                #TODO: BOSS IS NOT RESETTING!
                BossDeviceRef.ResetBoss(0.0)
                set JoinedPlayers = array{}

                break

    StartTimer(InAgent : agent)<suspends>:void=
        ObjectiveCompleteTrigger.Reset()

        TurnOffChest()
        Sleep(8.0)
        VisualSwitch.SetState(true)
        #Teleport Players in
        for(Button:Buttons):
            Button.Disable()
        #Only apply the timer to the players within the dungeon
        for(DungeonPlayer : JoinedPlayers):
            #Teleport players into the dungeon
            DungeonStartTeleporter.Teleport(DungeonPlayer)
            #Starts the time
            Timer.Start(DungeonPlayer)
            HUDDungeonTimer.Start(DungeonPlayer)
            #Starts the game via the trigger
            GameStartTrigger.Trigger(DungeonPlayer)
        Sleep(1.0)
        #Loops players in dungeon to see if all players are downed
        spawn:
            HandlePlayers()             


    #Leave Dungeon
    LeaveDungeon(InAgent : agent):void=
        #Remove the player from the joined players array
        for(Index -> Player : JoinedPlayers):
            if(Player = InAgent):
                if(NewArray := JoinedPlayers.RemoveElement[Index]):
                    set JoinedPlayers = NewArray
        #Stop the timer for that specific player
        Timer.Complete(InAgent)
        BossDeviceRef.HideBossUI(InAgent)
        GameEndTrigger.Trigger()
        #Teleport them out
        EntranceTeleporter.Teleport(InAgent)
        #Remove them from the array

    #Enable Dungeon Again
    EnableDungeon():void=
        set bDungeonStarted = false
        for(Button:Buttons):
            Button.Enable()

    #Timer has ended, players get teleported back
    OnTimerEnded(bAgent : ?agent):void=
        set bDungeonStarted = false
        VisualSwitch.SetState(false)
        #set EliminatedPlayers = 0
        GameEndTrigger.Trigger()
        for(Index -> Player: JoinedPlayers):
            if(NewArray := JoinedPlayers.RemoveElement[Index]):
                set JoinedPlayers = NewArray
            BossDeviceRef.HideBossUI(Player)
            DungeonAudio.Stop(Player)
            if(DungeonVolume.IsInVolume[Player]):
                EntranceTeleporter.Teleport(Player)
        
        #Reset the boss when the timer ends        
        #BossDeviceRef.InstantResetBossPlayerDeath()
        #Attempting Reset Boss
        spawn:
            BossDeviceRef.ResetBoss(1.0)
        set JoinedPlayers = array{}
        for(Button:Buttons):
            Button.Enable()
        
    TeleportPlayersToChest(DeviceAIInteractionResult : device_ai_interaction_result):void=
        spawn:
            TeleportPlayersWithDelay()
    
    TeleportPlayersWithDelay()<suspends>:void=
        for(DungeonPlayer : JoinedPlayers):
            Timer.Pause(DungeonPlayer)
        #We want to pause the timer
        #Then we wait 2 seconds and teleport
        Sleep(2.0)
        for(DungeonPlayer : JoinedPlayers):
            #Should Remove Visual HUD from player after dungeon
            HUDDungeonTimer.Complete(DungeonPlayer)
            Print("Teleport Player")
            ChestTeleporter.Teleport(DungeonPlayer)
        set JoinedPlayers = array{}

    OnObjectiveComplete(Agent:?agent):void=
        #This won't work, we should used joined players
        if(bPlayerManager := PlayerManager?, OutAgent := Agent?):
            BossDeviceRef.HideBossUI(OutAgent)
        spawn:
            TeleportPlayersWithDelay()
        VisualSwitch.SetState(false)
        for(Player:JoinedPlayers):
            DungeonCompleteHUD.Show(Player)
            if(bPlayerManager := PlayerManager?):
                bPlayerManager.PlayerStatManager.UpdateTotalDungeonsCompleted(Player, big_number{Value := 1.0, Exponent := 0})
        #Stopped players being teleported back automatically, they have to interact with the chest now
        #spawn:
            #TeleportPlayersBackFromChest(20.0)

    <#
    TeleportPlayersBackFromChest(DelayTime : float)<suspends>:void=
        Sleep(DelayTime)
        for(Player-> Value:ChestPlayers):
            #Only teleport the players if they are in the volume
            if(DungeonVolume.IsInVolume[Player]):
                #Should Remove Visual HUD from player after dungeon
                HUDDungeonTimer.Complete(Player)
                Timer.Complete(Player)
                EntranceTeleporter.Teleport(Player)
        set ChestPlayers = map{}
    #>

    ChestOpened(Agent : agent):void=
        if(Player := player[Agent]):
            RewardGranter.Trigger(Player)
            #UPDATE 1.1 Chest Bug Fix
            spawn:
                ChestOpenedTeleportBack(Player)
            # FCHQ Quest
            if(bPlayerManager := PlayerManager?):
                bPlayerManager.QuestFCHQ.QuestCompleted.Signal((Player, "Dungeon Warrior"))

    #UPDATE 1.1 Chest Bug Fix
    #After opening only teleport that single player back
    ChestOpenedTeleportBack(InPlayer : player)<suspends>:void=
        Sleep(0.25)
        #HUDDungeonTimer.Complete(InPlayer)
        #Timer.Complete(InPlayer)
        EntranceTeleporter.Teleport(InPlayer)

    TurnOffChest():void=
        Players := GetPlayspace().GetPlayers()
        for(Player:Players):
            Chest.TurnOff(Player)
