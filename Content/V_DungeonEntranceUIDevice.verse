
using { /Fortnite.com/Devices }
using { /Verse.org/Simulation }
using { /UnrealEngine.com/Temporary/Diagnostics }

using { /Verse.org/Assets }
using { /Verse.org/Simulation/Tags }
# See https://dev.epicgames.com/documentation/en-us/uefn/create-your-own-device-in-verse for how to create a verse device.

DungeonEntranceUITag := class(tag){}

# A Verse-authored creative device that can be placed in a level
V_DungeonEntranceUIDevice := class(creative_device):
        
    #Optional reference to Algorithm class
    @editable
    var AlgorithmAndFunctionsClassReference: ?V_AlgorithmsAndFunctions = false
    #NPC Reference
    @editable
    NPCReference : character_device = character_device{}
    @editable
    NameToUse : string = "Name"
    @editable
    DungeonDescription : string = "Name"
    
    @editable
    DungeonResourceAmountRequired : int = 1
    @editable
    IntTextureToUse : int = 0
    @editable
    PurchaseTrigger : trigger_device = trigger_device{}

    DungeonVendorDisplay : UI_DungeonEntranceDisplay = UI_DungeonEntranceDisplay{}

    # Manages and updates stats for each player.
    var PlayerStatManager: player_stats_manager = player_stats_manager{}

    # Runs when the device is started in a running game
    OnBegin<override>()<suspends>:void=
        NPCReference.InteractedWithEvent.Subscribe(OnCharacterInteract)
        PurchaseTrigger.TriggeredEvent.Subscribe(PurchaseUnlock)

    InitiateVendorUI(InPlayer : player):void=
        var TextureToUse : texture = VerseExposed.T_UIBackground
        case (IntTextureToUse):
            1 =>
                set TextureToUse = VerseExposed.T_BackplateDungeon1
            2 =>
                set TextureToUse = VerseExposed.T_UIBackground
            3 =>
                set TextureToUse = VerseExposed.T_UIBackground
            4 =>
                set TextureToUse = VerseExposed.T_UIBackground
            5 =>
                set TextureToUse = VerseExposed.T_UIBackground
            _ => # default
                set TextureToUse = VerseExposed.T_UIBackground

        DungeonEntranceItem := MakeDungeonItem(VerseExposed.T_BackplateDungeon1, NameToUse, PurchaseTrigger, DungeonDescription, DungeonResourceAmountRequired)
        DungeonVendorDisplay.InitUI(InPlayer, DungeonEntranceItem)

    #Interact with the dungeon entrance vendor
    #UI will appear
    OnCharacterInteract(InAgent : agent):void=
        if(Player := player[InAgent]):
            DungeonVendorDisplay.UpdateUI(Player, false)
            DungeonVendorDisplay.AddUI(Player)
            
    #This will call
    PurchaseUnlock(InMaybeAgent : ?agent):void=
        if(Agent := InMaybeAgent?):
            if(OldData := PlayerStatsDataMap[Agent]):
                if(Player := player[Agent]):
                    spawn:
                        DungeonVendorDisplay.HideUI(Player)