
using { /Fortnite.com/Devices }
using { /Verse.org/Simulation }
using { /UnrealEngine.com/Temporary/Diagnostics }

using { /Verse.org/Assets }
using { /Verse.org/Simulation/Tags }
using { /Fortnite.com/Game}
using { /Verse.org/Random }
# See https://dev.epicgames.com/documentation/en-us/uefn/create-your-own-device-in-verse for how to create a verse device.

ForcefieldVendorUITag := class(tag){}
CaveUnlockedAccolade := class(tag){}

# A Verse-authored creative device that can be placed in a level
V_ForceFieldVendor := class(creative_device):

    #NPC Reference
    @editable
    Button : button_device = button_device{}
    #Amount of resources the player needs
    @editable
    ForcefieldResourceRequirement : big_number = big_number{}
    @editable
    NameToUse : string = "Name"
    @editable
    IntTextureToUse : int = 0
    #Barrier we want to make ignore the player
    @editable
    Forcefield : barrier_device = barrier_device{}
    @editable
    Switch : switch_device = switch_device{}
    @editable
    PurchaseTrigger : trigger_device = trigger_device{}
    @editable
    GatherPercentageToAdd : float = 0.0

    @editable
    PreviousForcefieldSwitch : ?switch_device = false
    #I can display this if the PreviousForcefieldSwitch fails
    @editable
    PurchasePreviousForcefieldMessage : ?hud_message_device = false
    @editable
    LevelDevice : ?V_LevelUpAndCardUpgradeSelection = false

    UIBarrierVendorDisplay : UI_BarrierVendorDisplay = UI_BarrierVendorDisplay{}

    # Manages and updates stats for each player.
    var PlayerStatManager: player_stats_manager = player_stats_manager{}

    # Runs when the device is started in a running game
    OnBegin<override>()<suspends>:void=
        Button.InteractedWithEvent.Subscribe(OnCharacterInteract)
        PurchaseTrigger.TriggeredEvent.Subscribe(PurchaseUnlock)
        spawn{SafteyFeatureForBarriers()}
        
    InitiateVendorUI(InPlayer : player):void=
        var TextureToUse : texture = VerseExposed.T_UIBackground
        var CostToUse : string = BigNumber_FormatDisplay(ForcefieldResourceRequirement)
        case (IntTextureToUse):
            1 =>
                set TextureToUse = VerseExposed.T_MushroomCave1
            2 =>
                set TextureToUse = VerseExposed.T_MushroomCave2
            3 =>
                set TextureToUse = VerseExposed.T_MushroomCave3
            4 =>
                set TextureToUse = VerseExposed.T_MushroomCave4
            5 =>
                set TextureToUse = VerseExposed.T_LavaCave1
            6 =>
                set TextureToUse = VerseExposed.T_LavaCave2
            7 =>
                set TextureToUse = VerseExposed.T_LavaCave3
            8 =>
                set TextureToUse = VerseExposed.T_LavaCave4
            9 =>
                set TextureToUse = VerseExposed.T_UnderworldCave1
            10 =>
                set TextureToUse = VerseExposed.T_UnderworldCave2
            11 =>
                set TextureToUse = VerseExposed.T_UnderworldCave3
            12 =>
                set TextureToUse = VerseExposed.T_UnderworldCave4

            _ => # default
                set TextureToUse = VerseExposed.T_UIBackground

        BarrierVendorItem := MakeBarrierVendorItem(TextureToUse, NameToUse, CostToUse, PurchaseTrigger, GatherPercentageToAdd)
        UIBarrierVendorDisplay.InitUI(InPlayer, BarrierVendorItem)

    OnCharacterInteract(InAgent : agent):void=
        #We are setting this true or false and then calling the core logic at the end as there are two edge cases we want to take care of
        var ValidForcefieldPurchase : logic = false
        if(Player := player[InAgent]):
            #We check if the previous Force Field has been purchased
            if(bPreviousForcefieldSwitch := PreviousForcefieldSwitch?):
                #The first ForceFieldVendor in the level does not have a previous Switch Statement so this allows us to ignore it.
                #If the previous ForceField Vendor has been purchased then we check the current state
                if(bPreviousForcefieldSwitch.GetCurrentState[InAgent]):
                    #If it has been purchased the interaction will occur
                    set ValidForcefieldPurchase = true
                else:
                    #If it hasn't been purchased show this message
                    set ValidForcefieldPurchase = false
                    if(bPreviousForcefieldMessage := PurchasePreviousForcefieldMessage?):
                        bPreviousForcefieldMessage.Show(Player)
            else:
                #This allows caves without the ForceField Switch to just show the UI
                set ValidForcefieldPurchase = true

            if(ValidForcefieldPurchase = true):
                ResourceAmount := PlayerStatManager.GetPlayerStoneOreAmount(InAgent)
                    BNCost := ForcefieldResourceRequirement
                    CanAfford := BigNumber_GreaterThanOrEquals(ResourceAmount, BNCost)

                    if(Switch.GetCurrentState[InAgent]):
                        UIBarrierVendorDisplay.UpdateUI(Player, true, CanAfford)
                    else:
                        UIBarrierVendorDisplay.UpdateUI(Player, false, CanAfford)
                    UIBarrierVendorDisplay.AddUI(Player)

    
    PurchaseUnlock(InMaybeAgent : ?agent):void=
        if(Agent := InMaybeAgent?):
            if(OldData := PlayerStatsDataMap[Agent]):
                if(Player := player[Agent]):
                    spawn:
                        UIBarrierVendorDisplay.HideUI(Player)
                if(CurrentPlayerStats := PlayerStatManager.GetPlayerStats[Agent]):  
                    StoneAmount := PlayerStatManager.GetPlayerStoneOreAmount(Agent)
                    BNCost := ForcefieldResourceRequirement
                    CanAfford := BigNumber_GreaterThanOrEquals(StoneAmount, BNCost)
                    if(CanAfford = true):
                        if(bLevelDevice := LevelDevice?, Player := player[Agent]):
                            spawn:
                                PlayerStatManager.AddXPToPlayer(Player, big_number{Value := 1.0, Exponent := 2}, bLevelDevice)
                        Forcefield.AddToIgnoreList(Agent)
                        Switch.SetState(Agent, true)
                        PlayerStatManager.SpendPlayerOre(Agent, BNCost, false, PropResourceType.Stone)
                        PlayerStatManager.SetCaveStrengthPercentage(Agent, CurrentPlayerStats.CaveUnlockPercent + GatherPercentageToAdd)
                        #TODO: Add Accolade for cave

    #Reset values on player prestige
    #TODO: UI reset is not being called here, will ask Charlie
    ResetVendorOnPrestige(InAgent : agent):void=
        Print("Vendor Being Reset For Prestige")
        Forcefield.RemoveFromIgnoreList(InAgent)
        Switch.SetState(InAgent, false)


    SafteyFeatureForBarriers()<suspends>:void=
        RandFloat := GetRandomFloat(1.0, 2.0)
        Sleep(RandFloat)
        loop:
            Sleep(RandFloat)
            AllPlayers := GetPlayspace().GetPlayers()
            for(Player:AllPlayers):
                Switch.CheckState(Player)

