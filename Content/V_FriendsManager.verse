
using { /Fortnite.com/Devices }
using { /Verse.org/Simulation }
using { /UnrealEngine.com/Temporary/Diagnostics }

#ADDED
using { /Fortnite.com/Characters }
using { /UnrealEngine.com/Temporary/SpatialMath }

# See https://dev.epicgames.com/documentation/en-us/uefn/create-your-own-device-in-verse for how to create a verse device.

# A Verse-authored creative device that can be placed in a level
V_FriendsManager := class(creative_device):

    var PlayersToCheck : []player = array{}
    @editable
    Range : float = 8000.0
    @editable
    VFXSpawner : visual_effect_powerup_device = visual_effect_powerup_device{}
    @editable
    RemoveVFX : visual_effect_powerup_device = visual_effect_powerup_device{}

    #1 Second should be more then enough for this check, we don't want to update too frequently
    LoopUpdateTickRate : float = 1.0

    OnBegin<override>()<suspends>:void=
        LoopCheckPlayers()

    LoopCheckPlayers()<suspends>:void=
        loop:
            for(Player : GetPlayspace().GetPlayers()):
                Friends := FindFriends(Player)
                if(Friends > 0):
                    if(VFXSpawner.HasEffect[Player]):
                        #Print("You have friends and the effect")
                    else:
                        #Print("You have friends, apply the effect")
                        VFXSpawner.Pickup(Player)
                else:
                    #Print("You have no friends")
                    RemoveVFX.Pickup(Player)
            Sleep(LoopUpdateTickRate)

    #This will activate the VFX
    CheckToActivateVFX():void=
        VFXSpawner.Pickup()

    AddPlayerToCheck(InPlayer : player):void=
        Print("Add Player to be checked in Friends Manager")
        set PlayersToCheck += array{InPlayer}

    RemovePlayerToCheck(InPlayer : player):void=
        Print("Remove Player to be checked in Friends Manager")
        for(Index -> Player : PlayersToCheck):
            if(Player = InPlayer):
                if(NewArray := PlayersToCheck.RemoveElement[Index]):
                    set PlayersToCheck = NewArray

    FindFriends(InPlayer : player):int=
        var FriendAmount : int = 0
        if(InFortCharacter := InPlayer.GetFortCharacter[]):
            InPosition := InFortCharacter.GetTransform().Translation

            for(CheckPlayer : PlayersToCheck):
                if(InPlayer <> CheckPlayer):
                    if(CheckFortCharacter := CheckPlayer.GetFortCharacter[]):
                        CheckPlayerPosition := CheckFortCharacter.GetTransform().Translation

                        DistanceBetweenPlayers := Distance(InPosition, CheckPlayerPosition)
                        if(DistanceBetweenPlayers <= Range):
                            #Print("FOUND PLAYER")
                            set FriendAmount += 1

        return FriendAmount
