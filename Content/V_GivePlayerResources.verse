using { /Fortnite.com/Devices }
using { /Verse.org/Simulation }
using { /Verse.org/Assets }
using { /Verse.org/Random }
using { /UnrealEngine.com/Temporary/Diagnostics }

ResourceGiveType := enum:
    ChanceForEverything
    Individual

ResourceDetailsClass := class<concrete><unique>():
    @editable
    ResourceType : PropResourceType = PropResourceType.Stone
    @editable
    RewardMultiplier : float = 1.0
    @editable
    RewardAmount : float = 1.0
    @editable
    GiveSpecificAmountOfStone : logic = false
    
    #Exponential values for giving specific amount of stone
    @editable
    StoneToGive : big_number = big_number{}

    var ResourceImage : texture = VerseExposed.T_Stone
    var ResourceName : []char = "Resource"

V_GivePlayerResources := class(creative_device):
    @editable
    PlayerManager : ?V_PlayerManagerDevice = false

    @editable
    TriggerDevice : trigger_device = trigger_device{}

    @editable
    StrengthMultiplier : float = 1.0

    @editable
    OverallRewardMultiplier : float = 1.0
    
    @editable
    ResourceGiveStyle : ResourceGiveType = ResourceGiveType.Individual

    @editable
    GiveResources : []ResourceDetailsClass = array{}

    @editable
    UseDungeonMastery : logic = false

    OnBegin<override>()<suspends>:void=
        #Sets Name and Image of resource
        for(Resource : GiveResources):
            case(Resource.ResourceType):
                PropResourceType.Stone=>
                    set Resource.ResourceName = "Stone"
                    set Resource.ResourceImage = VerseExposed.T_Stone
                PropResourceType.Copper=>
                    set Resource.ResourceName = "Copper"
                    set Resource.ResourceImage = VerseExposed.T_CopperOreStroke
                PropResourceType.Silver=>
                    set Resource.ResourceName = "Silver"
                    set Resource.ResourceImage = VerseExposed.T_SilverOre
                PropResourceType.Malachite=>
                    set Resource.ResourceName = "Malachite"
                    set Resource.ResourceImage = VerseExposed.T_MalachiteOreStroke
                PropResourceType.Spectrolite=>
                    set Resource.ResourceName = "Spectrolite"
                    set Resource.ResourceImage = VerseExposed.T_SpectroliteOreStroke
                PropResourceType.Obsidian=>
                    set Resource.ResourceName = "Obsidian"
                    set Resource.ResourceImage = VerseExposed.T_ObsidianOreStroke
                PropResourceType.Orb=>
                    set Resource.ResourceName = "Orb"
                    set Resource.ResourceImage = VerseExposed.T_Orb
                _=>
                    Print("GivePlayerResources - Could not find Resource")

        #Subscribe to Trigger
        TriggerDevice.TriggeredEvent.Subscribe(GiveDefinedResourcesToPlayer)

    #Gives player all of the resources needed
    GiveDefinedResourcesToPlayer(InAgent : ?agent):void=
        Print("GivePlayerResources - Provides Resources")
        spawn:
            DelayedIndividualResourceToPlayer(InAgent)
        <#
        if(ResourceGiveStyle = ResourceGiveType.ChanceForEverything):
            spawn:
                DelayedAllResources(InAgent)
                
        if(ResourceGiveStyle = ResourceGiveType.Individual):
            spawn:
                DelayedIndividualResourceToPlayer(InAgent)
        #>
    #Delayed Individual Give Resources
    DelayedIndividualResourceToPlayer(InAgent : ?agent)<suspends>:void=
        if(bPlayerManager := PlayerManager?):
            if(bInAgent := InAgent?, Player := player[bInAgent], PlayerData := PlayerStatsDataMap[Player]):
                for(GiveResource : GiveResources):
                    var OutAmount : big_number = big_number{}

                    PlayerStrength := BigNumber_MultiplyWithAFloat(PlayerData.HarvestRatePerHit, StrengthMultiplier)
                    set OutAmount = BigNumber_MultiplyWithAFloat(PlayerStrength, GiveResource.RewardMultiplier)

                    if(UseDungeonMastery = true):
                        MasteryLevel := bPlayerManager.PlayerStatManager.GetPlayerUpgradeLevel(Player, PlayerUpgradeType.Prestige_DungeonMastery)
                        AdditionalAmount := bPlayerManager.PlayerStatManager.UpgradesData.GetDungeonMasteryAmount(MasteryLevel)
                        TotalExtra := BigNumber_MultiplyWithAFloat(OutAmount, AdditionalAmount / 100.0)

                        set OutAmount = BigNumber_Add(BigNumber_MultiplyWithAFloat(PlayerStrength, GiveResource.RewardMultiplier), TotalExtra)
                        if(FoundCard := SearchForCard(5, Player,  bPlayerManager.PlayerStatManager)?):
                            set OutAmount = GetBonusBigNumber(OutAmount, GetCardBonus(5, FoundCard.Level))
                    
                    #This check here is to give a specific amount of orbs, will not take into account dungeon mastery etc..
                    if(GiveResource.ResourceType = PropResourceType.Orb):
                        set OutAmount = BigNumber_ConvertFloatIntoBigNumber(GiveResource.RewardAmount)
                        if(FoundCard := SearchForCard(5, Player,  bPlayerManager.PlayerStatManager)?):
                            set OutAmount = GetBonusBigNumber(OutAmount, GetCardBonus(5, FoundCard.Level))
                        bPlayerManager.PlayerStatManager.UpdatePlayerOreAmount(bInAgent, GiveResource.ResourceType, OutAmount)
                        #bPlayerManager.UIManagerDevice.UIItemStreamDisplayClass.AddToStream(Player, "Stone", GiveResource.ResourceImage, OutAmount, false, false)
                    
                    #Give a specific amount of stone
                    if(GiveResource.GiveSpecificAmountOfStone = true):
                        bPlayerManager.PlayerStatManager.UpdatePlayerOreAmount(bInAgent, GiveResource.ResourceType, GiveResource.StoneToGive)
                        bPlayerManager.UIManagerDevice.UIItemStreamDisplayClass.AddToStream(Player, "Stone", GiveResource.ResourceImage, GiveResource.StoneToGive, false, false)

                    #If the resource type = random ores we give a specific amount of random ores based on "Reward Amount"
                    #This check goes here to override the OutAmount in this use case
                    if(GiveResource.ResourceType = PropResourceType.RandomOres):
                        #We get the total amount of resources we want to give and divide it by the amount of times we want to go through
                        var AmountOfResourcesToGivePerLoop : float = GiveResource.RewardAmount / 10.0 #Was 5.0
                        #Get a random ore chance 
                        for(X := 1..5):
                            var PropResourceTypeToGive : PropResourceType = PropResourceType.Copper
                            OreChance := GetRandomFloat(0.0, 100.0) 
                            #Currently each ore is giving a 20% chance, we can skew this which we probably should do
                            if((OreChance > 0.0) and (OreChance < 20.0)):
                                set PropResourceTypeToGive = PropResourceType.Copper
                                set GiveResource.ResourceImage = VerseExposed.T_CopperOreStroke                      
                            if((OreChance >20.0) and (OreChance < 40.0)):
                                set PropResourceTypeToGive = PropResourceType.Silver
                                set GiveResource.ResourceImage = VerseExposed.T_SilverOre   
                            if((OreChance >40.0) and (OreChance < 60.0)):
                                set PropResourceTypeToGive = PropResourceType.Malachite
                                set GiveResource.ResourceImage = VerseExposed.T_MalachiteOreStroke   
                            if((OreChance >60.0) and (OreChance < 80.0)):
                                set PropResourceTypeToGive = PropResourceType.Spectrolite
                                set GiveResource.ResourceImage = VerseExposed.T_SpectroliteOreStroke   
                            if((OreChance >80.0) and (OreChance < 100.0)): 
                                set PropResourceTypeToGive = PropResourceType.Obsidian
                                set GiveResource.ResourceImage = VerseExposed.T_ObsidianOreStroke   
                            #We then get out an amount of ore to give
                            #set OutAmount = BigNumber_ConvertFloatIntoBigNumber(GiveResource.RewardAmount)

                            if(UseDungeonMastery = true):
                                Print("Using Dungeon Mastery for Chest Reward")
                                #We get the amount
                                set OutAmount = BigNumber_ConvertFloatIntoBigNumber(AmountOfResourcesToGivePerLoop)
                                #Pass through to get the amount to give after dungeon mastery
                                MasteryLevel := bPlayerManager.PlayerStatManager.GetPlayerUpgradeLevel(Player, PlayerUpgradeType.Prestige_DungeonMastery)
                                AdditionalAmount := bPlayerManager.PlayerStatManager.UpgradesData.GetDungeonMasteryAmount(MasteryLevel)
                                TotalExtra := BigNumber_MultiplyWithAFloat(OutAmount, AdditionalAmount / 100.0)
                                #We get a new OutAmonut by taking the AmountOfResourcesToGivePerLoop and adding the TotalExtra from Dungeon Mastery
                                set OutAmount = BigNumber_Add(BigNumber_ConvertFloatIntoBigNumber(AmountOfResourcesToGivePerLoop), TotalExtra)
                                if(FoundCard := SearchForCard(5, Player,  bPlayerManager.PlayerStatManager)?):
                                    set OutAmount = GetBonusBigNumber(OutAmount, GetCardBonus(5, FoundCard.Level))
                            #We divide the reward by 5 and loop through so we can give a different amount of resource/ore each time
                            else:
                                Print("Not using dungeon mastery for Chest Reward")
                                set OutAmount = BigNumber_ConvertFloatIntoBigNumber(AmountOfResourcesToGivePerLoop)
                            #Changed so we give a PropResourceTypeToGive
                            Print("Giving Rewards to player for random ores")
                            bPlayerManager.PlayerStatManager.UpdatePlayerOreAmount(bInAgent, PropResourceTypeToGive, OutAmount)
                            bPlayerManager.UIManagerDevice.UIItemStreamDisplayClass.AddToStream(Player, "Stone", GiveResource.ResourceImage, OutAmount, false, false)
                            Sleep(0.25)
                    else:
                        #Not giving random ores, just give strength
                        Print("Giving Rewards to player for strength")
                        bPlayerManager.PlayerStatManager.UpdatePlayerOreAmount(bInAgent, GiveResource.ResourceType, OutAmount)
                        bPlayerManager.UIManagerDevice.UIItemStreamDisplayClass.AddToStream(Player, "Stone", GiveResource.ResourceImage, OutAmount, false, false)
                        Sleep(0.25)
                    Sleep(0.0)

    DelayedAllResources(InAgent : ?agent)<suspends>:void=
        # START CHANCE TO GET ANY ORE #
        RandomChance := GetRandomInt(1, 100)
        if(bPlayerManager := PlayerManager?):
            if(bInAgent := InAgent?, Player := player[bInAgent], PlayerData := PlayerStatsDataMap[Player]):
                Print("GivePlayerResources - Did checks")
                ArtifactMasteryLevel := bPlayerManager.PlayerStatManager.GetPlayerUpgradeLevel(Player, PlayerUpgradeType.Prestige_ArtifactMastery)
                TotalDropChance := bPlayerManager.PlayerStatManager.UpgradesData.GetPrestigeArtifactMasteryAmount(ArtifactMasteryLevel)

                var OutAmount : big_number = big_number{}
                var ResourceGot : PropResourceType = PropResourceType.Null 

                PlayerStrength := BigNumber_MultiplyWithAFloat(PlayerData.HarvestRatePerHit, StrengthMultiplier)
                set OutAmount = BigNumber_MultiplyWithAFloat(PlayerStrength, OverallRewardMultiplier)

                Print("GivePlayerResources - OutAmount: {OutAmount.Value}")
                
                #Give Guarenteed Stone
                bPlayerManager.PlayerStatManager.UpdatePlayerOreAmount(Player, PropResourceType.Stone, OutAmount)
                Sleep(0.25)
                bPlayerManager.UIManagerDevice.UIItemStreamDisplayClass.AddToStream(Player, "Stone", VerseExposed.T_Stone, OutAmount, false, false)

                if(RandomChance * 1.0 <= TotalDropChance):
                    # START CHANCE TO GET SPECIFIC ORE #
                    TypeRandomChance := GetRandomFloat(0.0, 100.0)

                    if(GetOreDropChance(PropResourceType.Obsidian) >= TypeRandomChance):
                        set ResourceGot = PropResourceType.Obsidian
                    else if(GetOreDropChance(PropResourceType.Spectrolite) >= TypeRandomChance):
                        set ResourceGot = PropResourceType.Spectrolite
                    else if(GetOreDropChance(PropResourceType.Malachite) >= TypeRandomChance):
                        set ResourceGot = PropResourceType.Malachite
                    else if(GetOreDropChance(PropResourceType.Silver) >= TypeRandomChance):
                        set ResourceGot = PropResourceType.Silver
                    else if(GetOreDropChance(PropResourceType.Copper) >= TypeRandomChance):
                        set ResourceGot = PropResourceType.Copper
                    
                    #Give Random Resource
                    if(ResourceGot <> PropResourceType.Null):
                        Print("GivePlayerResources - Got Random Resource")             
                        bPlayerManager.PlayerStatManager.UpdatePlayerOreAmount(Player, ResourceGot, OutAmount)
                        Sleep(0.25)
                        bPlayerManager.UIManagerDevice.UIItemStreamDisplayClass.AddToStream(Player, "Stone", GetOreTexture(ResourceGot), OutAmount, false, false)