
using { /Fortnite.com/Devices }
using { /Verse.org/Simulation }
using { /UnrealEngine.com/Temporary/Diagnostics }


using { /Verse.org/Random }
using { /UnrealEngine.com/Temporary/SpatialMath }

# See https://dev.epicgames.com/documentation/en-us/uefn/create-your-own-device-in-verse for how to create a verse device.

# A Verse-authored creative device that can be placed in a level
V_GnomeManager := class(creative_device):

    @editable
    GnomeSpots : []creative_prop = array{}
    @editable
    GnomeProp : creative_prop = creative_prop{}
    @editable
    GnomeAudio : audio_player_device = audio_player_device{}
    @editable
    GnomeButton : button_device = button_device{}
    @editable
    GnomeMessage : hud_message_device = hud_message_device{}

    @editable
    ActiveTime : float = 0.5
    @editable
    DownTime : float = 0.5

    @editable
    PlayerManager : ?V_PlayerManagerDevice = false
    @editable
    LevelDevice : ?V_LevelUpAndCardUpgradeSelection = false

    # Runs when the device is started in a running game
    OnBegin<override>()<suspends>:void=
        GnomeButton.InteractedWithEvent.Subscribe(OnButtonInteract)
        spawn:
            GnomeLoop()

    GnomeLoop()<suspends>:void=
        loop:
            Sleep(DownTime * 60.0)

            # Teleport Gnome to Random GnomeSpot
            RandomIndex := GetRandomInt(0, GnomeSpots.Length - 1)
            Print("Spot Index: {RandomIndex}")
            if(RandomSpot := GnomeSpots[RandomIndex]):
                FoundTransform := RandomSpot.GetTransform()

                if(GnomeProp.TeleportTo[FoundTransform]):
                    Print("Gnome Teleported To Spot")
                    GnomeAudio.Play()
                    GnomeMessage.Show()

            # Sleep ActiveTime
            Sleep(ActiveTime * 60.0)

            #Return Gnome to Origin Position
            if(GnomeProp.TeleportTo[vector3{X:=0.0,Y:=-25000.0,Z:=210000.0}, rotation{}]):
                Print("Gnome Teleported Home")
                GnomeAudio.Stop()

    ApplyGnomeBuff(InPlayer : player)<suspends>:void=
        if(bPlayerManager := PlayerManager?):
            set bPlayerManager.GetPlayerDataNotSaved(InPlayer).GnomeBuff = true

            bPlayerManager.QuestFCHQ.QuestCompleted.Signal((InPlayer, "Gnome Hunter"))

            bPlayerManager.UIManagerDevice.UIGnomeBuffDisplay.AddUI(InPlayer)
            var NewTimer : float = 3.0 * 60.0
            loop:
                bPlayerManager.UIManagerDevice.UIGnomeBuffDisplay.UpdateUI(InPlayer, NewTimer)
                # Sleep for 1 second if new time to wait is 0 
                Sleep(1.0)
                set NewTimer -= 1.0
                if(NewTimer < 0.0):
                    break
            
            # After Loop is broken reset buff
            set bPlayerManager.GetPlayerDataNotSaved(InPlayer).GnomeBuff = false
            spawn:
                bPlayerManager.UIManagerDevice.UIGnomeBuffDisplay.HideUI(InPlayer)

    OnButtonInteract(InAgent : agent):void=
        Print("Gnome Found")
        if(Player := player[InAgent], bPlayerManager := PlayerManager?, bLevelDevice := LevelDevice?):
            # Add some XP to player
            GnomeBuff := bPlayerManager.GetPlayerDataNotSaved(Player).GnomeBuff
            if(GnomeBuff = true):
                return
            spawn:
                bPlayerManager.PlayerStatManager.AddXPToPlayer(Player, big_number{Value := 5.0, Exponent := 2}, bLevelDevice)
            # Do buff
            spawn:
                ApplyGnomeBuff(Player)
