
using { /Fortnite.com/Devices }
using { /Verse.org/Simulation }
using { /UnrealEngine.com/Temporary/Diagnostics }

# See https://dev.epicgames.com/documentation/en-us/uefn/create-your-own-device-in-verse for how to create a verse device.

# A Verse-authored creative device that can be placed in a level
V_HalloweenHunt := class(creative_device):
    ToMessage<localizes>(String : string) : message = "{String}"

    @editable
    Character : character_device = character_device{}
    @editable
    Switches : []switch_device = array{}
    @editable
    PopUp : popup_dialog_device = popup_dialog_device{}

    @editable
    HatSwitch : switch_device = switch_device{}
    @editable
    HatChanger : V_ChangeHatIndex = V_ChangeHatIndex{}
    @editable
    HatFound : hud_message_device = hud_message_device{}

    # Runs when the device is started in a running game
    OnBegin<override>()<suspends>:void=
        Character.InteractedWithEvent.Subscribe(OnPopUpShown)
        PopUp.RespondingButtonEvent.Subscribe(OnResponseButton)

        for(Switch : Switches):
            Switch.TurnedOnEvent.Subscribe(OnSwitchOn)
        
    OnPopUpShown(InAgent : agent):void=
        #Get how many have been found and set in description
        var CurrentCount : int = 0
        for(Switch : Switches):
            if(Switch.GetCurrentState[InAgent]):
                set CurrentCount += 1

        PopUp.SetDescriptionText(ToMessage("Someone stole my Candy Buckets! They're all over the world. Bring back {Switches.Length} of them and I'll give you a reward.\n{CurrentCount}/{Switches.Length} Found!"))
        PopUp.Show(InAgent)

    OnResponseButton(InAgent : agent, InIndex : int):void=
        Print("Response Button Pressed")
        if(InIndex = 0):
            #Check if actually found all of them.
            var CurrentCount : int = 0
            for(Switch : Switches):
                if(Switch.GetCurrentState[InAgent]):
                    set CurrentCount += 1
                
            if(CurrentCount >= Switches.Length):
                #Give hat and turn on hat switch if found all
                Print("All found give hat")
                HatSwitch.TurnOn(InAgent)
                HatSwitch.CheckState(InAgent)
                HatChanger.OnButtonInteract(InAgent)


    OnSwitchOn(InAgent : agent):void=
        Print("Bucket Found")
        HatFound.Show(InAgent)