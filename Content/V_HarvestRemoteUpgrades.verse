
using { /Fortnite.com/Devices }
using { /Verse.org/Simulation }
using { /UnrealEngine.com/Temporary/Diagnostics }

#Added
using { /UnrealEngine.com/Temporary/SpatialMath}
using { /Verse.org/Random }

PurchaseTriggerHandler := class() {
    Device : V_HarvestRemoteUpgrades
    UpgradeIndex : int

    HandlerFunction(InMaybeAgent : ?agent) : void = {
        if(Agent := InMaybeAgent?):
            if(InPlayer := player[Agent]):
                #Purchase Upgrade
                if(Upgrade := Device.PlayerStatData.UpgradesData.HRUpgradesArray[UpgradeIndex]):
                    Device.PurchaseHarvestRateUpgrade(Agent, UpgradeIndex, Upgrade.BaseRate, Upgrade.RateIncrease, Device.PlayerStatData.GetPlayerUpgradePurchaseAmount(Agent), Upgrade.BaseCost)
                    Device.PurchaseUpdateUI(InPlayer)
    }
}

UpgradeTriggerHandler := class() {
    Device : V_HarvestRemoteUpgrades
    UpgradeIndex : int

    HandlerFunction(InMaybeAgent : ?agent) : void = {
        if(Agent := InMaybeAgent?):
            if(InPlayer := player[Agent]):
                if(Upgrade := Device.PlayerStatData.UpgradesData.HRUpgradesArray[UpgradeIndex]):
                    Device.AlgorithmAndFunctionsClassReference.AlgorithmAndFunctionClassReference.PurchaseTierMultiplierRemoteUpgrades(Agent, Upgrade.TierSystems, UpgradeIndex, Upgrade.RateIncrease, Upgrade.BaseRate)
                    Device.UpdateAllUpgradesUI(InPlayer)
    }
}

# A Verse-authored creative device that can be placed in a level
V_HarvestRemoteUpgrades := class(creative_device):
    @editable
    PurchaseTriggers : []trigger_device = array{}
    @editable
    UpgradeTriggers : []trigger_device = array{}

    @editable
    x1Trigger : trigger_device = trigger_device{}
    @editable
    x10Trigger : trigger_device = trigger_device{}
    @editable
    x100Trigger : trigger_device = trigger_device{}

    # Triggers for the right click menu
    @editable
    ShowEquipTrigger : trigger_device = trigger_device{}

    @editable
    UpgradeRemote : signal_remote_manager_device = signal_remote_manager_device{}
    #HUD Message Device to show the cost of the next upgrade
    
    SetStrengthCost<localizes>(FloatValue:float) : message = "Cost = {FloatValue}"

    # Manages and updates stats for each player.
    var PlayerStatData: player_stats_manager = player_stats_manager{}
    #Optional reference to Algorithm class
    @editable
    var AlgorithmAndFunctionsClassReference: V_AlgorithmsAndFunctions = V_AlgorithmsAndFunctions{}

    var CanTrigger100x : logic = true

    UIRemoteUpgradeMenuClass : UI_UpgradeDisplay = UI_UpgradeDisplay{}

    ToMessage<localizes>(String : string) : message = "{String}"

    # Runs when the device is started in a running game
    OnBegin<override>()<suspends>:void=
        #Trigger setup for Resource Remote Upgrades
        for(Index -> Trigger : PurchaseTriggers):
            Trigger.TriggeredEvent.Subscribe(PurchaseTriggerHandler{Device:=Self, UpgradeIndex:=Index}.HandlerFunction)

        for(Index -> Trigger : UpgradeTriggers):
            Trigger.TriggeredEvent.Subscribe(UpgradeTriggerHandler{Device:=Self, UpgradeIndex:=Index}.HandlerFunction)

        #Purchase Amounts
        x1Trigger.TriggeredEvent.Subscribe(OnX1Triggered)
        x10Trigger.TriggeredEvent.Subscribe(OnX10Triggered)
        x100Trigger.TriggeredEvent.Subscribe(OnX100Triggered)

        UpgradeRemote.PrimarySignalEvent.Subscribe(DisplayUpgradeUI)
    
    #Harvest Rate HUD:
    InitiateRemoteUI(InPlayer : player):void=
        if(OldData := PlayerStatsDataMap[InPlayer]):
            if(CurrentPlayerStats := PlayerStatData.GetPlayerStats[InPlayer]):
                var RemoteDataUpgrades : []UpgradeSlotConstructData = array{}

                for(Index -> HRUpgrade : PlayerStatData.UpgradesData.HRUpgradesArray):
                    if(UpgradeLevel := CurrentPlayerStats.HRUpgradeLevels[Index], TierMultiplier := CurrentPlayerStats.HRUpgradeTierMultipliers[Index]):

                        UpgradeCost := AlgorithmAndFunctionsClassReference.AlgorithmAndFunctionClassReference.CalculateCostOfNextUpgrade(HRUpgrade.BaseCost,CurrentPlayerStats.HRCostIncreaseRate, UpgradeLevel)
                        BNUpgradeLevel := BigNumber_FormatDisplay(BigNumber_ConvertFloatIntoBigNumber(UpgradeLevel))
                        BNUpgradeCost := BigNumber_FormatDisplay(UpgradeCost)
                        BNTierMultiplierUpgradeCost := BigNumber_FormatDisplay(BigNumber_MultiplyWithAFloat(HRUpgrade.RateIncrease, TierMultiplier))   
                        BNUpgradeTotalPower := BigNumber_FormatDisplay(GetTotalPowerPerUpgrade(InPlayer, Index, 0, false))
                        
                        if(UpgradeTrigger := UpgradeTriggers[Index], PurchaseTrigger := PurchaseTriggers[Index]):
                            RemoteDataUpgrade := MakeUpgradeSlotConstructData(HRUpgrade.Name, BNUpgradeLevel, BNUpgradeTotalPower, BNUpgradeCost, BNTierMultiplierUpgradeCost, PurchaseTrigger, UpgradeTrigger)
                            set RemoteDataUpgrades += array{RemoteDataUpgrade}
                        #Print("{Index}")

                BNPlayerStoneAmount := BigNumber_FormatDisplay(PlayerStatData.GetPlayerStoneOreAmount(InPlayer))
                UIRemoteUpgradeMenuClass.InitUI(InPlayer, RemoteDataUpgrades, BNPlayerStoneAmount, x1Trigger, x10Trigger, x100Trigger)

    GetTotalPowerPerUpgrade(InPlayer : player, UpgradeIndex : int, AdditionalLevels : int, ForUI : logic):big_number=
        var UpgradeTotalPower : big_number = big_number{}
        if(OldData := PlayerStatsDataMap[InPlayer]):
            if(CurrentPlayerStats := PlayerStatData.GetPlayerStats[InPlayer]):
                if(UpgradeLevel := CurrentPlayerStats.HRUpgradeLevels[UpgradeIndex], TierMultiplier := CurrentPlayerStats.HRUpgradeTierMultipliers[UpgradeIndex],
                RateIncrease := PlayerStatData.UpgradesData.HRUpgradesArray[UpgradeIndex].RateIncrease, BaseRate := PlayerStatData.UpgradesData.HRUpgradesArray[UpgradeIndex].BaseRate):
                    if(ForUI = true):
                        set UpgradeTotalPower = AlgorithmAndFunctionsClassReference.AlgorithmAndFunctionClassReference.CalculateHarvestRateOnHitForUI(RateIncrease, UpgradeLevel + AdditionalLevels * 1.0, BaseRate, TierMultiplier)
                    else:
                        set UpgradeTotalPower = AlgorithmAndFunctionsClassReference.AlgorithmAndFunctionClassReference.CalculateHarvestRateOnHit(RateIncrease, UpgradeLevel + AdditionalLevels * 1.0, BaseRate, TierMultiplier)
        return UpgradeTotalPower

    #Harvest Rate HUD:
    DisplayUpgradeUI(InAgent : agent):void=
        if(InPlayer := player[InAgent]):
            PurchaseAmount := PlayerStatData.GetPlayerUpgradePurchaseAmount(InPlayer)
            UpdateAllUpgradesUI(InPlayer)
            BNPlayerStoneAmount := BigNumber_FormatDisplay(PlayerStatData.GetPlayerStoneOreAmount(InPlayer))
            BNPlayerStrengthAmount := BigNumber_FormatDisplay(PlayerStatData.GetPlayerOreHarvestRatePerHit(InPlayer))
            UIRemoteUpgradeMenuClass.UpdateStonePowerUI(InPlayer, BNPlayerStoneAmount, BNPlayerStrengthAmount)
            UIRemoteUpgradeMenuClass.AddUI(InPlayer)
            Print("Show Upgrade UI")

    #Harvest Rate HUD:
    UpdateUIWithDelay(InPlayer: player, UpgradeIndex : int, InPurchaseAmount : int)<suspends>:void=
        Sleep(0.4)
        if(InPlayer.IsActive[]):
            if(OldData := PlayerStatsDataMap[InPlayer]):
                if(CurrentPlayerStats := PlayerStatData.GetPlayerStats[InPlayer]):
                    if(Upgrade := PlayerStatData.UpgradesData.HRUpgradesArray[UpgradeIndex], UpgradeLevel := CurrentPlayerStats.HRUpgradeLevels[UpgradeIndex], TierMultiplier := CurrentPlayerStats.HRUpgradeTierMultipliers[UpgradeIndex]):
                        FormattedCostAndCanAfford := GetFormattedCost(InPlayer, InPurchaseAmount, Upgrade.BaseCost, CurrentPlayerStats.HRCostIncreaseRate, UpgradeLevel)

                        OreHarvestRate := PlayerStatData.GetPlayerOreHarvestRatePerHit(InPlayer)
                        BNUpgrade1TotalPower := BigNumber_FormatDisplay(GetTotalPowerPerUpgrade(InPlayer, UpgradeIndex, 0, false))
                        var BNNextUpgradePower : string = "0"

                        if(BNUpgrade1TotalPower = "0"):
                            set BNNextUpgradePower = BigNumber_FormatDisplay(GetTotalPowerPerUpgrade(InPlayer, UpgradeIndex, 1, true))
                        else:
                            set BNNextUpgradePower = GetFormattedNextUpgrade(InPlayer, UpgradeIndex)

                        BNPlayerStoneAmount := BigNumber_FormatDisplay(PlayerStatData.GetPlayerStoneOreAmount(InPlayer))
                        BNPlayerStrengthAmount := BigNumber_FormatDisplay(PlayerStatData.GetPlayerOreHarvestRatePerHit(InPlayer))

                        BNUpgradeLevel := BigNumber_FormatDisplay(BigNumber_ConvertFloatIntoBigNumber(UpgradeLevel))
                        BNOreHarvestRate := BigNumber_FormatDisplay(OreHarvestRate)
                        BNTierMultiplierUpgrade1Cost := BigNumber_FormatDisplay(BigNumber_MultiplyWithAFloat(Upgrade.RateIncrease, TierMultiplier * 500))
                        BNPurchaseAmount := BigNumber_FormatDisplay(BigNumber_ConvertFloatIntoBigNumber(InPurchaseAmount * 1.0))

                        var RateIncreases : []big_number = array{}
                        var TierSystems : [][]float = array{}
                        for(Index -> CurrentUpgrade : PlayerStatData.UpgradesData.HRUpgradesArray):
                            set RateIncreases += array{CurrentUpgrade.RateIncrease}
                            set TierSystems += array{CurrentUpgrade.TierSystems}

                        UIRemoteUpgradeMenuClass.UpdateUI(InPlayer, Upgrade.Name, BNUpgradeLevel, BNUpgrade1TotalPower, BNNextUpgradePower, FormattedCostAndCanAfford(0), BNTierMultiplierUpgrade1Cost, BNPlayerStoneAmount, BNPlayerStrengthAmount, RateIncreases, OldData, TierSystems, BNPurchaseAmount, FormattedCostAndCanAfford(1))

    GetFormattedCost(InPlayer : player, InPurchaseAmount : int, InBaseCost : big_number, InHRCostIncrease : big_number, InUpgradeLevel : float):tuple(string, logic)=
        var RemoteUpgradeCost : big_number = AlgorithmAndFunctionsClassReference.AlgorithmAndFunctionClassReference.CalculateCostOfNextUpgrade(InBaseCost, InHRCostIncrease, InUpgradeLevel)
        var TruePurchaseAmount : int = 1
        var FormattedCost : string = "{BigNumber_FormatDisplay(RemoteUpgradeCost)} (x{TruePurchaseAmount})"
        var CanAfford : logic = false

        PlayerStoneAmount := PlayerStatData.GetPlayerStoneOreAmount(InPlayer)
        if(InPurchaseAmount > 1):
            for(X := 1..InPurchaseAmount):
                NextRemoteUpgradeCost := AlgorithmAndFunctionsClassReference.AlgorithmAndFunctionClassReference.CalculateCostOfNextUpgrade(InBaseCost, InHRCostIncrease, InUpgradeLevel + (X * 1.0))
                BNCost := BigNumber_Add(RemoteUpgradeCost, NextRemoteUpgradeCost)
                #Check if we have enough coins for the next upgrade.
                InternalCanAfford := BigNumber_GreaterThanOrEquals(PlayerStoneAmount, BNCost)
                if(InternalCanAfford = true):
                    set RemoteUpgradeCost = BigNumber_Add(RemoteUpgradeCost, NextRemoteUpgradeCost)
                    set TruePurchaseAmount = X + 1
                    if(InPurchaseAmount = 10 and X = 10):
                        set TruePurchaseAmount = 10
                    if(InPurchaseAmount = 100 and X = 100):
                        set TruePurchaseAmount = 100
                    set FormattedCost = "{BigNumber_FormatDisplay(RemoteUpgradeCost)} (x{TruePurchaseAmount})"
                    set CanAfford = true
                
        if (TruePurchaseAmount = 1):
            BNCost := RemoteUpgradeCost
            #Check if we have enough coins for the next upgrade.
            InternalCanAfford := BigNumber_GreaterThanOrEquals(PlayerStoneAmount, BNCost)
            if(InternalCanAfford = true):
                set RemoteUpgradeCost = AlgorithmAndFunctionsClassReference.AlgorithmAndFunctionClassReference.CalculateCostOfNextUpgrade(InBaseCost, InHRCostIncrease, InUpgradeLevel)
                set CanAfford = true
        
        return (FormattedCost, CanAfford)

    GetFormattedNextUpgrade(InPlayer : player, UpgradeIndex : int):string=
        Print("Get")
        if(OldData := PlayerStatsDataMap[InPlayer]):
            if(CurrentPlayerStats := PlayerStatData.GetPlayerStats[InPlayer]):
                if(Upgrade := PlayerStatData.UpgradesData.HRUpgradesArray[UpgradeIndex], TierMultiplier := CurrentPlayerStats.HRUpgradeTierMultipliers[UpgradeIndex]):
                    Amount := BigNumber_MultiplyWithAFloat(Upgrade.RateIncrease, TierMultiplier)

                    return BigNumber_FormatDisplay(Amount)

        return ""


    #Harvest Rate Upgrade:
    #We input the harvest rate upgrade the player wants from player interacting with a menu and calling triggers. We pass through the upgrade type 1,2,3,4 or 5
    PurchaseHarvestRateUpgrade(InAgent : agent, UpgradeIndex : int, HarvestBaseRate : big_number, HarvestRateIncrease : big_number, InPurchaseAmount : int, InBaseCost : big_number):void=
        if(CanTrigger100x = true):
            Print("Spamming Purchase")
            var TierSystems : [][]float = array{}
                for(Index -> CurrentUpgrade : PlayerStatData.UpgradesData.HRUpgradesArray):
                    set TierSystems += array{CurrentUpgrade.TierSystems}
            spawn:
                AlgorithmAndFunctionsClassReference.AlgorithmAndFunctionClassReference.PurchasePlayerHarvestRateUpgrade(InAgent, InBaseCost, UpgradeIndex, TierSystems, HarvestBaseRate, HarvestRateIncrease, InPurchaseAmount)
            spawn:
                AllowPurchase()

    #Allow players to purchase again
    AllowPurchase()<suspends>:void=
        Sleep(0.5)
        set CanTrigger100x = true
    
    OnX1Triggered(InMaybeAgent : ?agent):void=
        if(Agent := InMaybeAgent?):
            if(InPlayer := player[Agent]):
                PurchaseAmount := 1
                PlayerStatData.UpdatePlayerUpgradePurchaseAmount(Agent, PurchaseAmount)
                UpdateAllUpgradesUI(InPlayer)
                Print("X1")

    OnX10Triggered(InMaybeAgent : ?agent):void=
        if(Agent := InMaybeAgent?):
            if(InPlayer := player[Agent]):
                PurchaseAmount := 10
                PlayerStatData.UpdatePlayerUpgradePurchaseAmount(Agent, PurchaseAmount)
                UpdateAllUpgradesUI(InPlayer)
                Print("X10")

    OnX100Triggered(InMaybeAgent : ?agent):void=
        if(Agent := InMaybeAgent?):
            if(InPlayer := player[Agent]):
                PurchaseAmount := 100
                PlayerStatData.UpdatePlayerUpgradePurchaseAmount(Agent, PurchaseAmount)
                UpdateAllUpgradesUI(InPlayer)
                Print("X100")

    UpdateAllUpgradesUI(InPlayer : player):void=
        PurchaseAmount := PlayerStatData.GetPlayerUpgradePurchaseAmount(InPlayer)
        for(Index := 0..PlayerStatData.UpgradesData.HRUpgradesArray.Length - 1):
            spawn:
                UpdateUIWithDelay(InPlayer, Index, PurchaseAmount)

    PurchaseUpdateUI(InPlayer : player):void=
        if(CanTrigger100x = true):
            set CanTrigger100x = false
            PurchaseAmount := PlayerStatData.GetPlayerUpgradePurchaseAmount(InPlayer)
            for(Index := 0..PlayerStatData.UpgradesData.HRUpgradesArray.Length - 1):
                spawn:
                    UpdateUIWithDelay(InPlayer, Index, PurchaseAmount)