
using { /Fortnite.com/Devices }
using { /Verse.org/Simulation }
using { /UnrealEngine.com/Temporary/Diagnostics }

# See https://dev.epicgames.com/documentation/en-us/uefn/create-your-own-device-in-verse for how to create a verse device.

# A Verse-authored creative device that can be placed in a level
V_HatFCHQChecker := class(creative_device):

    @editable
    Switches : []switch_device = array{}

    @editable
    PlayerManager : V_PlayerManagerDevice = V_PlayerManagerDevice{}

    # Runs when the device is started in a running game
    OnBegin<override>()<suspends>:void=
        for(Switch : Switches):
            Switch.TurnedOnEvent.Subscribe(OnSwitchOn)

    OnSwitchOn(InAgent : agent):void=
        PlayerManager.QuestFCHQ.QuestCompleted.Signal((InAgent, "Collector"))