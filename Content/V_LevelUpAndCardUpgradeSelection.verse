using { /Fortnite.com/UI }
using { /Fortnite.com/Devices }
using { /Fortnite.com/Characters }
using { /Verse.org/Simulation }
using { /Verse.org/Random }
using { /Verse.org/Colors }
using { /Verse.org/Assets }
using { /UnrealEngine.com/Temporary/Diagnostics }
using { /UnrealEngine.com/Temporary/UI }
using { /UnrealEngine.com/Temporary/SpatialMath }

# A Verse-authored creative device that can be placed in a level
V_LevelUpAndCardUpgradeSelection := class(creative_device):
    ##UI Variables##
    UIText<localizes> : message = "Choose"

    var NewUI : canvas = canvas{}

    @editable
    OreHitTrigger : trigger_device = trigger_device{}
    ##XP Level Up Variables##
    @editable
    XPPerHit : float = 10.0
    @editable
    bLevelUpSoundDevice : ?audio_player_device = false
    @editable
    bLevelUpAccolades : ?accolades_device = false
    @editable
    bLevelUpMessage : ?hud_message_device = false

    UIDisplay : UI_XPDisplay = UI_XPDisplay{}
    PlayerStatsManager : player_stats_manager = player_stats_manager{}

    ###Level Up Variables Ended###

    OnBegin<override>()<suspends>:void=
        #Subscribes to player Events
        OreHitTrigger.TriggeredEvent.Subscribe(OnRockHit)

    # Updates the UI of the player on a loop
    UpdateXPBarUI(InPlayer : player)<suspends>:void=
        var OldXP : float = 0.0
        var NewXP : float = 0.0
        Print("Started Update Player UI")
        if(PlayerStatData := PlayerStatsManager.GetPlayerStats[InPlayer]):
            set OldXP = BigNumber_ConvertBigNumberIntoFloat(PlayerStatData.XPAmount)
            set NewXP = BigNumber_ConvertBigNumberIntoFloat(PlayerStatData.XPAmount)

            loop:
                if(UpdatedPlayerStatData := PlayerStatsManager.GetPlayerStats[InPlayer]):
                    set NewXP = BigNumber_ConvertBigNumberIntoFloat(UpdatedPlayerStatData.XPAmount)

                    if(not(NewXP = OldXP)):
                        if (NewXP > OldXP):
                            set OldXP = Lerp(OldXP, NewXP, 0.1)
                            if(NewXP - OldXP <= 0.1):
                                set OldXP = NewXP
                        else:
                            set OldXP = 0.0

                    UIDisplay.UpdateUI(InPlayer, OldXP, UpdatedPlayerStatData.Level * 200.0 * 1.25, UpdatedPlayerStatData.Level)
                    Sleep(0.02)

    #Tracks when player eliminates
    OnRockHit(InAgent : ?agent) : void=
        Print("Player Eliminated - RankManager")
        if(Player := player[InAgent?]):
            spawn:
                PlayerStatsManager.AddXPToPlayer(Player, big_number{Value := XPPerHit, Exponent := 0}, Self)

    #Creates XP Bar for Player
    InitPlayerXPBar(InPlayer : player):void=
        if(PlayerStatData := PlayerStatsManager.GetPlayerStats[InPlayer]):
            XPData := MakeXPBarData(BigNumber_ConvertBigNumberIntoFloat(PlayerStatData.XPAmount), PlayerStatData.Level * 200.0 * 1.25, PlayerStatData.Level)
            UIDisplay.InitUI(InPlayer, XPData)
            UIDisplay.AddUI(InPlayer)
            spawn:
                UpdateXPBarUI(InPlayer)