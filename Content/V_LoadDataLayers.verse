
using { /Fortnite.com/Devices }
using { /Verse.org/Simulation }
using { /UnrealEngine.com/Temporary/Diagnostics }

# See https://dev.epicgames.com/documentation/en-us/uefn/create-your-own-device-in-verse for how to create a verse device.

# A Verse-authored creative device that can be placed in a level
V_LoadDataLayers := class(creative_device):

    @editable
    CinematicDataLayers : []cinematic_sequence_device = array{}

    @editable
    TimeBetweenAssetLoading : float = 0.5

    OnBegin<override>()<suspends>:void=
        LoadLayers()

    LoadLayers()<suspends>:void=
        for(DL : CinematicDataLayers):
            Print("Loading Data Layer")
            DL.Play()
            Sleep(0.5)