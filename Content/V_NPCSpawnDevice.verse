using { /Fortnite.com/Devices }
using { /Verse.org/Simulation }
using { /UnrealEngine.com/Temporary/Diagnostics }

V_NPCSpawnDevice := class(creative_device):
    @editable
    AreaVolume : volume_device = volume_device{}
    @editable
    BossSpawner : npc_spawner_device = npc_spawner_device{}
    @editable
    GuardNPCSpawners : []npc_spawner_device = array{}
    @editable
    DungeonEndedTrigger : trigger_device = trigger_device{}
    var CurrentPlayerCount : int = 0

    # Runs when the device is started in a running game
    OnBegin<override>()<suspends>:void=
        AreaVolume.AgentEntersEvent.Subscribe(OnVolumeEnter)
        DungeonEndedTrigger.TriggeredEvent.Subscribe(DisableNPCSpawners)

    #Enables NPC Spawner when there are players nearby
    OnVolumeEnter(InAgent : agent):void=
        if(Player := player[InAgent]):
            set CurrentPlayerCount += 1
            Print("Number of Players in Volume: {CurrentPlayerCount}")
            #Only enable upon the first player running into the volume.
            if(CurrentPlayerCount <= 1):
                for(GuardSpawner : GuardNPCSpawners):
                    GuardSpawner.Enable()

    #We disable the spawners here, not on the volume
    DisableNPCSpawners(InAgent : ?agent):void=
        #Reset Current Player Count
        set CurrentPlayerCount = 0
        BossSpawner.Disable()
        for(GuardSpawner : GuardNPCSpawners):
            GuardSpawner.Disable()