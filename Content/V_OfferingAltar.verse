
using { /Fortnite.com/Devices }
using { /Verse.org/Simulation }
using { /UnrealEngine.com/Temporary/Diagnostics }

# See https://dev.epicgames.com/documentation/en-us/uefn/create-your-own-device-in-verse for how to create a verse device.

# A Verse-authored creative device that can be placed in a level
V_OfferingAltar := class(creative_device):

    @editable
    ConditionalButton : conditional_button_device = conditional_button_device{}
    @editable
    ItemGranter : item_granter_device = item_granter_device{}
    @editable
    PlayerManager : ?V_PlayerManagerDevice = false

    @editable
    VFXPickup : visual_effect_powerup_device = visual_effect_powerup_device{}
    @editable
    VFXClear : visual_effect_powerup_device = visual_effect_powerup_device{}
    @editable
    AltarAudio : audio_player_device = audio_player_device{}

    @editable
    UIManager : ?V_UIManagerDevice = false

    @editable
    AltarUpgradeTrigger : trigger_device = trigger_device{}
    
    # Runs when the device is started in a running game
    OnBegin<override>()<suspends>:void=
        ConditionalButton.ActivatedEvent.Subscribe(OnActivated)
        AltarUpgradeTrigger.TriggeredEvent.Subscribe(OnUpgraded)

    ActivateButton(InAgent : agent):void=
        ConditionalButton.Activate(InAgent)

    OnActivated(InAgent : agent):void=
        Print("Altar Button Activated")
        if(bPlayerManager := PlayerManager?):
            PlayerAltarLevel := bPlayerManager.PlayerStatManager.GetPlayerUpgradeLevel(InAgent, PlayerUpgradeType.Altar)
            if(PlayerAltarLevel = 0):
                ItemGranter.GrantItem(InAgent)
                return

            if(Player := player[InAgent]):
                PlayerNotSavedData := bPlayerManager.GetPlayerDataNotSaved(Player)
                # Make sure the Altar Buff isn't already activated
                if(PlayerNotSavedData.IsAltarBuffActive[]):
                    ItemGranter.GrantItem(InAgent)
                    return
            # If it is give the resource back to the player

            # Get AltarTimer for Player
            VFXPickup.Pickup(InAgent)
            AltarAudio.Play(InAgent)
            PlayerAltarTimeLevel := bPlayerManager.PlayerStatManager.GetPlayerUpgradeLevel(InAgent, PlayerUpgradeType.AltarTime)
            PlayerAltarTime := bPlayerManager.PlayerStatManager.UpgradesData.AltarTimeUpgrade.GetCurrentLevelAmount(PlayerAltarTimeLevel)
            # Call function that activates the altar buff for an amount of time
            spawn:
                AltarBuff(InAgent, PlayerAltarTime)

    AltarBuff(InAgent : agent, InTimer : float)<suspends>:void=
        Print("Buff Activated")
        var NewTimer : float = InTimer
        if(bPlayerManager := PlayerManager?, Player := player[InAgent]):
            PlayerNotSavedData := bPlayerManager.GetPlayerDataNotSaved(Player)
            # Activate the buff for Player
            set PlayerNotSavedData.AltarBuff = true
            # Add UI and update for Altar Buff Timer
            bPlayerManager.UIManagerDevice.UIAltarBuffDisplay.AddUI(Player)
            loop:
                bPlayerManager.UIManagerDevice.UIAltarBuffDisplay.UpdateUI(Player, NewTimer)
                # Sleep for 1 second if new time to wait is 0 
                Sleep(1.0)
                set NewTimer -= 1.0
                if(NewTimer < 0.0):
                    break
            
            # After Loop is broken reset buff
            set PlayerNotSavedData.AltarBuff = false
            VFXClear.Pickup(InAgent)
            spawn:
                bPlayerManager.UIManagerDevice.UIAltarBuffDisplay.HideUI(Player)
                

    OnUpgraded(InMaybeAgent : ?agent):void=
        if(bUIManager := UIManager?, InAgent := InMaybeAgent?, Player := player[InAgent]):
            spawn:
                bUIManager.HandleAlterTokenUIVisibility(Player)

