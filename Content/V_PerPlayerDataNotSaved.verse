using { /Fortnite.com/Devices }
using { /Verse.org/Simulation }
using { /UnrealEngine.com/Temporary/Diagnostics }
using { /Verse.org/Simulation/Tags }

FollowingProp := struct:
    Prop : creative_prop
    Slot : int

#Only store variables in here we are not saving. We have a player stats file which handles player data we are saving!
PerPlayerDataNotSaved := class<unique>():

    #var HealthBarOnScreen : logic = false
    var CurrentCriticalCharge : int = 0
    var NextHitCrit : logic = false

    var CurrentExplosiveCharge : int = 0
    var NextHitExplode : logic = false

    var AltarBuff : logic = false
    var GnomeBuff : logic = false

    var IsHealing : logic = false

    var HasClassBeenSet : logic = false
    var ClassNumber : int = 0

    var CountingTime : logic = false

    var FollowingProps : []FollowingProp = array{}

    var Pet : creative_prop = creative_prop{}

    var PlaytimeSessionMinutes : int = 0
    var PlaytimeSessionHours : int = 0

    #PlayerRaycast
    var TaggedActors : []creative_object_interface = array{}
    var ClosestPropHealthBar : V_PropHealthBar = V_PropHealthBar{}
    var WaitForFire : logic = false

    ResetFire():void=
        set WaitForFire = true
        #spawn:
            #SetFireTrue()

    SetFireTrue():void=
        #Sleep(0.5)
        set WaitForFire = false

    SetTagToSearch(InTag : creative_object_interface):void=
        #Make Empty
        #set TaggedActors = array{}
        #Add element
        set TaggedActors +=  array{InTag}

    EmptyArray():void=
        set TaggedActors = array{}

    SetClosestPropHealthBar(InPropHealthBar : V_PropHealthBar):void=
        set ClosestPropHealthBar = InPropHealthBar

    IncrementCriticalCharge(InIncrement : int):void=
        set CurrentCriticalCharge += InIncrement
        if(CurrentCriticalCharge >= 50):
            set NextHitCrit = true

    ResetCriticalCharge():void=
        set CurrentCriticalCharge = 0
        set NextHitCrit = false

    IncrementExplodeCharge(InIncrement : int):void=
        set CurrentExplosiveCharge += InIncrement
        if(CurrentExplosiveCharge >= 25):
            set NextHitExplode = true

    ResetExplodeCharge():void=
        set CurrentExplosiveCharge = 0
        set NextHitExplode = false

    IsAltarBuffActive()<transacts><decides>:void=
        AltarBuff = true

    SetClass(Num : int) : void=
        set ClassNumber = Num