
using { /Fortnite.com/Devices }
using { /Verse.org/Simulation }
using { /UnrealEngine.com/Temporary/Diagnostics }

using { /Fortnite.com/Characters }

# See https://dev.epicgames.com/documentation/en-us/uefn/create-your-own-device-in-verse for how to create a verse device.

# A Verse-authored creative device that can be placed in a level
V_PetSpawner := class(creative_device):

    @editable
    Button : button_device = button_device{}
    @editable
    PetToSpawn : creative_prop_asset = DefaultCreativePropAsset

    @editable
    Switch : switch_device = switch_device{}

    @editable
    PropFollower : V_PropFollower = V_PropFollower{}

    # Runs when the device is started in a running game
    OnBegin<override>()<suspends>:void=
        Button.InteractedWithEvent.Subscribe(SpawnPet)
        Switch.TurnedOnEvent.Subscribe(SpawnPet)

    SpawnPet(InAgent : agent):void=
        Print("Spawn Pet")
        if(Switch.GetCurrentState[InAgent]):
            PropFollower.SpawnPetFollower(InAgent, PetToSpawn)