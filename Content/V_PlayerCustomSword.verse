
using { /Fortnite.com/Devices }
using { /Verse.org/Simulation }
using { /UnrealEngine.com/Temporary/Diagnostics }

using { /Fortnite.com/Characters }
using { /Fortnite.com/Game }
using { /Verse.org/Assets }

GetSwordUnlockData(InResourceType : PropResourceType):SwordUnlockData=
    case(InResourceType):
        PropResourceType.Copper =>
            return SwordUnlockData:
                StoneCost := big_number{Value := 1.0, Exponent := 9}
                ResourceCost := big_number{Value := 0.0, Exponent := 0}
                ResourceImage := VerseExposed.T_Orb
        PropResourceType.Silver =>
            return SwordUnlockData:
                StoneCost := big_number{Value := 1.0, Exponent := 15}
                ResourceCost := big_number{Value := 2.0, Exponent := 1}
                ResourceImage := VerseExposed.T_Orb
        PropResourceType.Malachite =>
            return SwordUnlockData:
                StoneCost := big_number{Value := 1.0, Exponent := 23}
                ResourceCost := big_number{Value := 8.0, Exponent := 1}
                ResourceImage := VerseExposed.T_Orb
        PropResourceType.Spectrolite =>
            return SwordUnlockData:
                StoneCost := big_number{Value := 2.5, Exponent := 3}
                ResourceCost := big_number{Value := 1.0, Exponent := 0}
                ResourceImage := VerseExposed.T_Orb
        PropResourceType.Obsidian =>
            return SwordUnlockData:
                StoneCost := big_number{Value := 2.5, Exponent := 3}
                ResourceCost := big_number{Value := 1.0, Exponent := 0}
                ResourceImage := VerseExposed.T_Orb
        _ =>
            Print("### ARTIFACT ORE TYPE NOT SETUP IN GetArtifactSmeltData ###")
            return SwordUnlockData:
                StoneCost := big_number{Value := 2.5, Exponent := 3}
                ResourceCost := big_number{Value := 1.0, Exponent := 0}
                ResourceImage := VerseExposed.T_Stone

GetSwordUpgradeData(InSwordRarity : SwordRarity):SwordUpgradeData=
    case(InSwordRarity):
        SwordRarity.Common =>
            return SwordUpgradeData:
                StoneCost := big_number{Value := 5.0, Exponent := 8}
                ResourceCost := big_number{Value := 3.0, Exponent := 0}
                ResourceType := PropResourceType.Orb
                ResourceImage := VerseExposed.T_Orb
        SwordRarity.Uncommon =>
            return SwordUpgradeData:
                StoneCost := big_number{Value := 5.0, Exponent := 14}
                ResourceCost := big_number{Value := 2.0, Exponent := 1}
                ResourceType := PropResourceType.Orb
                ResourceImage := VerseExposed.T_Orb
        SwordRarity.Rare =>
            return SwordUpgradeData:
                StoneCost := big_number{Value := 5.0, Exponent := 22}
                ResourceCost := big_number{Value := 8.0, Exponent := 1}
                ResourceType := PropResourceType.Orb
                ResourceImage := VerseExposed.T_Orb
        SwordRarity.Epic =>
            return SwordUpgradeData:
                StoneCost := big_number{Value := 2.5, Exponent := 3}
                ResourceCost := big_number{Value := 1.0, Exponent := 0}
                ResourceType := PropResourceType.Orb
                ResourceImage := VerseExposed.T_Orb
        SwordRarity.Legendary =>
            return SwordUpgradeData:
                StoneCost := big_number{Value := 2.5, Exponent := 3}
                ResourceCost := big_number{Value := 1.0, Exponent := 0}
                ResourceType := PropResourceType.Orb
                ResourceImage := VerseExposed.T_Orb
        _ =>
            Print("### ARTIFACT ORE TYPE NOT SETUP IN GetArtifactSmeltData ###")
            return SwordUpgradeData:
                StoneCost := big_number{Value := 2.5, Exponent := 3}
                ResourceCost := big_number{Value := 1.0, Exponent := 0}
                ResourceType := PropResourceType.Orb
                ResourceImage := VerseExposed.T_Orb

GetSwordPresetData(InSwordRarity : SwordRarity):Sword_Data=
    case(InSwordRarity):
        SwordRarity.Common =>
            return Sword_Data:
                Level := 1
                Rarity := SwordRarity.Common
                StatIncreaseRate := 1.1
                SiphonIncreaseRate := 5.0
                StatEnums := array{SwordStats.Siphon, SwordStats.StrengthBonus, SwordStats.CriticalDamage}
                StatValues := array{10.0<#, 1.0, 1.0#>}
                BigNumber_StatValues := array{big_number{Value := 3.5, Exponent := 3}, big_number{Value := 1.0, Exponent := 1}, big_number{Value := 2.0, Exponent := 2}, big_number{Value := 2.0, Exponent := 1}}
        SwordRarity.Uncommon =>
            return Sword_Data:
                Level := 1
                Rarity := SwordRarity.Uncommon
                StatIncreaseRate := 1.20
                SiphonIncreaseRate := 5.0
                StatEnums := array{SwordStats.Siphon, SwordStats.StrengthBonus, SwordStats.CriticalDamage}
                StatValues := array{10.0<#, 2.0, 2.0#>}
                BigNumber_StatValues := array{big_number{Value := 3.5, Exponent := 9}, big_number{Value := 1.0, Exponent := 1}, big_number{Value := 2.0, Exponent := 2}, big_number{Value := 1.0, Exponent := 2}}
        SwordRarity.Rare =>
            return Sword_Data:
                Level := 1
                Rarity := SwordRarity.Rare
                StatIncreaseRate := 1.30
                SiphonIncreaseRate := 5.0
                StatEnums := array{SwordStats.Siphon, SwordStats.StrengthBonus, SwordStats.CriticalDamage}
                StatValues := array{10.0<#, 3.0, 3.0#>}
                BigNumber_StatValues := array{big_number{Value := 5.0, Exponent := 18}, big_number{Value := 1.0, Exponent := 1}, big_number{Value := 2.0, Exponent := 2}, big_number{Value := 2.5, Exponent := 2}}
        SwordRarity.Epic =>
            return Sword_Data:
                Level := 1
                Rarity := SwordRarity.Epic
                StatIncreaseRate := 1.05
                SiphonIncreaseRate := 5.0
                StatEnums := array{SwordStats.Siphon, SwordStats.StrengthBonus, SwordStats.CriticalDamage}
                StatValues := array{10.0<#, 4.0, 4.0#>}
                BigNumber_StatValues := array{big_number{Value := 2.5, Exponent := 3}, big_number{Value := 1.0, Exponent := 1}, big_number{Value := 2.0, Exponent := 2}, big_number{Value := 1.0, Exponent := 1}}
        SwordRarity.Legendary =>
            return Sword_Data:
                Level := 1
                Rarity := SwordRarity.Legendary
                StatIncreaseRate := 1.05
                SiphonIncreaseRate := 5.0
                StatEnums := array{SwordStats.Siphon, SwordStats.StrengthBonus, SwordStats.CriticalDamage}
                StatValues := array{10.0<#, 5.0, 5.0#>}
                BigNumber_StatValues := array{big_number{Value := 2.5, Exponent := 3}, big_number{Value := 1.0, Exponent := 1}, big_number{Value := 2.0, Exponent := 2}, big_number{Value := 1.0, Exponent := 1}}
        _ =>
            return Sword_Data:
                Level := 1
                Rarity := SwordRarity.Common
                StatIncreaseRate := 1.0
                StatEnums := array{SwordStats.Siphon, SwordStats.StrengthBonus, SwordStats.CriticalDamage}
                StatValues := array{10.0, 1.0, 1.0}
                BigNumber_StatValues := array{big_number{Value := 1.0, Exponent := 0}, big_number{Value := 1.0, Exponent := 0}, big_number{Value := 1.0, Exponent := 0}, big_number{Value := 1.0, Exponent := 1}}

        
GetSwordStat(InSword: Sword_Data, Stat : SwordStats):float=
    case(Stat):
        SwordStats.Siphon =>
            if(ReturnStat := InSword.StatValues[0]):
                return ReturnStat
        SwordStats.StrengthBonus =>
            if(ReturnStat := InSword.StatValues[1]):
                return ReturnStat
        SwordStats.CriticalDamage =>
            if(ReturnStat := InSword.StatValues[2]):
                return ReturnStat
        #SwordStats.CriticalRate =>
            #if(ReturnStat := InSword.StatValues[2]):
                #return ReturnStat
    return 0.0

SwordUnlockData := struct:
    StoneCost : big_number
    ResourceCost : big_number
    ResourceImage : texture

SwordUpgradeData := struct:
    StoneCost : big_number
    ResourceCost : big_number
    ResourceImage : texture
    ResourceType : PropResourceType

SwordRarity := enum<persistable>:
    Common
    Uncommon
    Rare
    Epic
    Legendary
    Null

SwordStats := enum<persistable>:
    Siphon # Base Damage
    StrengthBonus # Crtical Rate
    CriticalDamage # Critcal Damage
    #CriticalRate

MakeSword<constructor>(OldData : Sword_Data)<transacts> := Sword_Data:
    Level := OldData.Level
    Rarity := OldData.Rarity
    SiphonIncreaseRate := OldData.SiphonIncreaseRate
    StatIncreaseRate := OldData.StatIncreaseRate
    StatEnums := OldData.StatEnums
    StatValues := OldData.StatValues
    BigNumber_StatValues := OldData.BigNumber_StatValues

Sword_Data := class<final><persistable>:
    Level : int = 0
    Rarity : SwordRarity = SwordRarity.Null
    SiphonIncreaseRate : float = 5.0
    StatIncreaseRate : float = 1.0
    StatEnums : []SwordStats = array{} # STAT ENUMS IS NOT BEING USED, USING INDEX OF STAT VALUES INSTEAD
    StatValues : []float = array{} # STAT VALUES IS NOT BEING USED, USING BigNumber_StatValues
    BigNumber_StatValues : []big_number = array{}
    # INDEX = Different Stat
    # IMPORTANT: 0 = Base Damage, 1 = Critcal Rate, 2 = Critical Damage, 3 = Strength

# A Verse-authored creative device that can be placed in a level
V_PlayerCustomSword := class(creative_device):

    UISwordShopMenu : UI_SwordShopMenu = UI_SwordShopMenu{}

    @editable
    NPCCharacter : character_device = character_device{}
    @editable
    SuccessAudio : audio_player_device = audio_player_device{}
    #Detect If Sword is Held
    @editable 
    WeaponConditionalButton : conditional_button_device = conditional_button_device{}
    #VFX Power Up Devices Per level
    @editable
    VFXDeviceSwordLevel1 : visual_effect_powerup_device = visual_effect_powerup_device{}
    @editable
    VFXDeviceSwordLevel2 : visual_effect_powerup_device = visual_effect_powerup_device{}
    @editable
    VFXDeviceSwordLevel3 : visual_effect_powerup_device = visual_effect_powerup_device{}
    @editable
    VFXDeviceSwordLevel4 : visual_effect_powerup_device = visual_effect_powerup_device{}
    @editable
    VFXDeviceSwordLevel5 : visual_effect_powerup_device = visual_effect_powerup_device{}
    @editable
    VFXDevice_Clear : visual_effect_powerup_device = visual_effect_powerup_device{}

    @editable
    Trigger_Unlock : trigger_device = trigger_device{}
    @editable
    Trigger_Upgrade : trigger_device = trigger_device{}
    @editable
    Trigger_Sword1 : trigger_device = trigger_device{}
    @editable
    Trigger_Sword2 : trigger_device = trigger_device{}
    @editable
    Trigger_Sword3 : trigger_device = trigger_device{}
    @editable
    Trigger_Sword4 : trigger_device = trigger_device{}
    @editable
    Trigger_Sword5 : trigger_device = trigger_device{}

    @editable
    SwordItemGranter : item_granter_device = item_granter_device{}

    var PlayerStatManager : player_stats_manager = player_stats_manager{}

    InitiatePlayerUI(InPlayer : player):void=
        UISwordShopMenu.InitUI(InPlayer, Trigger_Unlock, Trigger_Upgrade, Trigger_Sword1, Trigger_Sword2, Trigger_Sword3, Trigger_Sword4, Trigger_Sword5)

    OnBegin<override>()<suspends>:void=
        NPCCharacter.InteractedWithEvent.Subscribe(OnNPCInteract)

        Trigger_Unlock.TriggeredEvent.Subscribe(OnUnlockTrigger)
        Trigger_Upgrade.TriggeredEvent.Subscribe(OnUpgradeTrigger)
        Trigger_Sword1.TriggeredEvent.Subscribe(OnSword1Trigger)
        Trigger_Sword2.TriggeredEvent.Subscribe(OnSword2Trigger)
        Trigger_Sword3.TriggeredEvent.Subscribe(OnSword3Trigger)
        Trigger_Sword4.TriggeredEvent.Subscribe(OnSword4Trigger)
        Trigger_Sword5.TriggeredEvent.Subscribe(OnSword5Trigger)

        LoopAndCheckForWeapon()

    GivePlayerSword(InAgent : agent, InSwordRarity : SwordRarity):void=
        if(Player := player[InAgent], CanvasData := UISwordShopMenu.BuiltCanvasMap[Player]):
            Print("Sword is being granted")
            SwordItemGranter.GrantItem(InAgent)
            #Save that the player has been granted a sword
            PlayerStatManager.GiveSword(InAgent, GetSwordPresetData(InSwordRarity))
            SuccessAudio.Play(InAgent)
            if(PlayerStats := PlayerStatManager.GetPlayerStats[Player]):
                    UISwordShopMenu.UpdateUI(Player, PlayerStats, PlayerStatManager, PlayerStats.OwnedPlayerSword, InSwordRarity, CanvasData.CurrentResourceType)

    UpgradePlayerSword(InAgent : agent, InSwordRarity : SwordRarity):void=
        PlayerStatManager.UpgradeSword(InAgent, InSwordRarity)

        if(Player := player[InAgent], PlayerStats := PlayerStatManager.GetPlayerStats[Player]):
            if(CanvasData := UISwordShopMenu.BuiltCanvasMap[Player]):
                UISwordShopMenu.UpdateUI(Player, PlayerStats, PlayerStatManager, PlayerStats.OwnedPlayerSword, InSwordRarity, CanvasData.CurrentResourceType)

    #Iteration 1 - This is the first iteration of looping and removing the weapon from a player
    #If the player is holding the item we equip it, else we trigger for a clear effect
    LoopAndCheckForWeapon()<suspends>:void=
        loop:
            AllPlayers : []player = GetPlayspace().GetPlayers()
            for(i -> Player : AllPlayers):
                if (WeaponConditionalButton.IsHoldingItem[Player]):
                    EquipCustomSword(Player)
                else:
                    #Print("Hide Weapon From Player")
                    VFXDevice_Clear.Pickup(Player)
            Sleep(0.1)

    #Iteration 1 - This is the first iteration of detecting if a player has the sword VFX. If not we will call "Pickup" so they get the effect
    EquipCustomSword(InPlayer : player):void=
        #Print("Checking For Weapon On Swap")
        #If the player does not currently have this effect apply it
        if(PlayerStats := PlayerStatManager.GetPlayerStats[InPlayer]):
            if(PlayerSword := PlayerStats.OwnedPlayerSword[0]):
                case(PlayerSword.Rarity):
                    SwordRarity.Common =>
                        #if(VFXDeviceSwordLevel1.HasEffect[InPlayer]):
                        #else:
                            VFXDeviceSwordLevel1.Pickup(InPlayer)
                    SwordRarity.Uncommon =>
                        #if(VFXDeviceSwordLevel2.HasEffect[InPlayer]):
                        #else:
                            VFXDeviceSwordLevel2.Pickup(InPlayer)
                    SwordRarity.Rare =>
                        #if(VFXDeviceSwordLevel3.HasEffect[InPlayer]):
                        #else:
                            VFXDeviceSwordLevel3.Pickup(InPlayer)
                    SwordRarity.Epic =>
                        #if(VFXDeviceSwordLevel4.HasEffect[InPlayer]):
                        #else:
                            VFXDeviceSwordLevel4.Pickup(InPlayer)
                    SwordRarity.Legendary =>
                        #if(VFXDeviceSwordLevel5.HasEffect[InPlayer]):
                        #else:
                            VFXDeviceSwordLevel5.Pickup(InPlayer)
                    _ =>


    OnNPCInteract(InAgent : agent):void=
        if(Player := player[InAgent], PlayerStats := PlayerStatManager.GetPlayerStats[Player]):
            if(CanvasData := UISwordShopMenu.BuiltCanvasMap[Player]):
                UISwordShopMenu.UpdateUI(Player, PlayerStats, PlayerStatManager, PlayerStats.OwnedPlayerSword, SwordRarity.Common, CanvasData.CurrentResourceType)
            UISwordShopMenu.AddUI(Player)


    OnUnlockTrigger(InMaybeAgent : ?agent):void=
        if(Agent := InMaybeAgent?):
            if(Player := player[Agent], PlayerStats := PlayerStatManager.GetPlayerStats[Player]):
                if(CanvasData := UISwordShopMenu.BuiltCanvasMap[Player]):
                    StoneAmount := PlayerStatManager.GetPlayerStoneOreAmount(Agent)
                    ResourceAmount := PlayerStatManager.GetPlayerOreAmount(Agent, PropResourceType.Orb)
                    SmeltData := GetSwordUnlockData(CanvasData.CurrentResourceType)
                    BNCost := SmeltData.StoneCost
                    ResourceCost := SmeltData.ResourceCost
                    CanAffordStone := BigNumber_GreaterThanOrEquals(StoneAmount, BNCost)
                    CanAffordResource := BigNumber_GreaterThanOrEquals(ResourceAmount, ResourceCost)
                    #Check if we have enough ore for the unlock
    
                    if(CanAffordStone = true, CanAffordResource = true):
                        PlayerStatManager.SpendPlayerOre(Agent, BNCost, false, PropResourceType.Stone)
                        PlayerStatManager.SpendPlayerOre(Agent, ResourceCost, false, PropResourceType.Orb)

                        GivePlayerSword(Agent, CanvasData.CurrentSwordRarity)

    OnUpgradeTrigger(InMaybeAgent : ?agent):void=
        Print("UPGRADE TRIGGER")
        if(Agent := InMaybeAgent?):
            if(Player := player[Agent], PlayerStats := PlayerStatManager.GetPlayerStats[Player], Sword := PlayerStats.OwnedPlayerSword[0]):
                if(CanvasData := UISwordShopMenu.BuiltCanvasMap[Player]):
                    StoneAmount := PlayerStatManager.GetPlayerStoneOreAmount(Agent)
                    ResourceAmount := PlayerStatManager.GetPlayerOreAmount(Agent, PropResourceType.Orb)
                    UpgradeData := GetSwordUpgradeData(CanvasData.CurrentSwordRarity)
                    
                    PrestigeCost := BigNumber_MultiplyWithAFloat(UpgradeData.StoneCost, Sword.Level * 1.0)
                    PrestigeResourceCost := BigNumber_MultiplyWithAFloat(UpgradeData.ResourceCost, Sword.Level * 1.0)
                    CanAffordStone := BigNumber_GreaterThanOrEquals(StoneAmount, PrestigeCost)
                    CanAffordResource := BigNumber_GreaterThanOrEquals(ResourceAmount, PrestigeResourceCost)
                    #Check if we have enough ore for the unlock
    
                    if(CanAffordStone = true, CanAffordResource = true):
                        PlayerStatManager.SpendPlayerOre(Agent, PrestigeCost, false, PropResourceType.Stone)
                        PlayerStatManager.SpendPlayerOre(Agent, PrestigeResourceCost, false, PropResourceType.Orb)

                        UpgradePlayerSword(Agent, CanvasData.CurrentSwordRarity)


    OnSword1Trigger(InMaybeAgent : ?agent):void=
        if(Agent := InMaybeAgent?):
            if(Player := player[Agent], PlayerStats := PlayerStatManager.GetPlayerStats[Player]):
                if(CanvasData := UISwordShopMenu.BuiltCanvasMap[Player]):
                    UISwordShopMenu.UpdateUI(Player, PlayerStats, PlayerStatManager, PlayerStats.OwnedPlayerSword, SwordRarity.Common, PropResourceType.Copper)
                return

    OnSword2Trigger(InMaybeAgent : ?agent):void=
        if(Agent := InMaybeAgent?):
            if(Player := player[Agent], PlayerStats := PlayerStatManager.GetPlayerStats[Player]):
                if(CanvasData := UISwordShopMenu.BuiltCanvasMap[Player]):
                    UISwordShopMenu.UpdateUI(Player, PlayerStats, PlayerStatManager, PlayerStats.OwnedPlayerSword, SwordRarity.Uncommon, PropResourceType.Silver)
                return

    OnSword3Trigger(InMaybeAgent : ?agent):void=
        if(Agent := InMaybeAgent?):
            if(Player := player[Agent], PlayerStats := PlayerStatManager.GetPlayerStats[Player]):
                if(CanvasData := UISwordShopMenu.BuiltCanvasMap[Player]):
                    UISwordShopMenu.UpdateUI(Player, PlayerStats, PlayerStatManager, PlayerStats.OwnedPlayerSword, SwordRarity.Rare, PropResourceType.Malachite)
                return

    OnSword4Trigger(InMaybeAgent : ?agent):void=
        if(Agent := InMaybeAgent?):
            if(Player := player[Agent], PlayerStats := PlayerStatManager.GetPlayerStats[Player]):
                if(CanvasData := UISwordShopMenu.BuiltCanvasMap[Player]):
                    UISwordShopMenu.UpdateUI(Player, PlayerStats, PlayerStatManager, PlayerStats.OwnedPlayerSword, SwordRarity.Epic, PropResourceType.Spectrolite)
                return
                
    OnSword5Trigger(InMaybeAgent : ?agent):void=
        if(Agent := InMaybeAgent?):
            if(Player := player[Agent], PlayerStats := PlayerStatManager.GetPlayerStats[Player]):
                if(CanvasData := UISwordShopMenu.BuiltCanvasMap[Player]):
                    UISwordShopMenu.UpdateUI(Player, PlayerStats, PlayerStatManager, PlayerStats.OwnedPlayerSword, SwordRarity.Legendary, PropResourceType.Obsidian)
                return