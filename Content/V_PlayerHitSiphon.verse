
using { /Fortnite.com/Devices }
using { /Verse.org/Simulation }
using { /UnrealEngine.com/Temporary/Diagnostics }
using { /Verse.org/Simulation/Tags }
using { /Fortnite.com/Characters }
using { /Fortnite.com/Game }

# See https://dev.epicgames.com/documentation/en-us/uefn/create-your-own-device-in-verse for how to create a verse device.
creature_manager := class(tag):
wildFire_manager := class(tag):
npc_manager := class(tag):

# A Verse-authored creative device that can be placed in a level
V_PlayerHitSiphon := class(creative_device):
    var CreatureSpawner : []creature_spawner_device = array{}
    var NPCSpawners : []npc_spawner_device = array{}
    
    @editable
    PlayerManager : ?V_PlayerManagerDevice = false

    # Runs when the device is started in a running game
    OnBegin<override>()<suspends>:void=
        block:
            # Get all objects with tag mini_game_manager
            TagObjects := FindCreativeObjectsWithTag(creature_manager{}) #returns creative_object_interface
            # Loop through TagObjects
            for(TagObject : TagObjects):
                # Make sure "TagObject" is actually a V_MiniGameManager type object by casting it to V_MiniGameManager
                if(TaggedButtonDevice := creature_spawner_device[TagObject]):
                    set CreatureSpawner += array{TaggedButtonDevice}
                    TaggedButtonDevice.SpawnedEvent.Subscribe(OnCreatureSpawned)
        block:
            # Get all objects with tag mini_game_manager
            TagObjects := FindCreativeObjectsWithTag(npc_manager{}) #returns creative_object_interface
            # Loop through TagObjects
            for(TagObject : TagObjects):
                # Make sure "TagObject" is actually a V_MiniGameManager type object by casting it to V_MiniGameManager
                if(TaggedButtonDevice := npc_spawner_device[TagObject]):
                    set NPCSpawners += array{TaggedButtonDevice}
                    TaggedButtonDevice.SpawnedEvent.Subscribe(OnNPCSpawned)

    
    OnCreatureSpawned(Agent : agent):void=
        if (FortChar := Agent.GetFortCharacter[]):
            FortChar.DamagedEvent().Subscribe(OnNPCDamaged)

    OnNPCSpawned(Agent : agent):void=
        if (FortChar := Agent.GetFortCharacter[]):
            FortChar.DamagedEvent().Subscribe(OnNPCDamaged)

    OnNPCDamaged(DamageResult : damage_result):void=
        #Print("NPC Damaged")
        if(Instigator := DamageResult.Instigator?, FortChar := fort_character[Instigator], Player := player[FortChar.GetAgent[]]):
            OnPlayerDamagedEnemy(Player)


    OnPlayerDamagedEnemy(Player : player):void=
        if(FortChar := Player.GetFortCharacter[]):
            # Get sword siphon amount
            if(bPlayerManager := PlayerManager?, PlayerStats := bPlayerManager.PlayerStatManager.GetPlayerStats[Player]):
                if(PlayerSword := PlayerStats.OwnedPlayerSword[0]):
                    FortChar.Heal(GetSwordStat(PlayerSword, SwordStats.Siphon))
