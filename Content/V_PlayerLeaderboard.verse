
using { /Fortnite.com/Devices }
using { /Verse.org/Simulation }
using { /UnrealEngine.com/Temporary/Diagnostics }

# See https://dev.epicgames.com/documentation/en-us/uefn/create-your-own-device-in-verse for how to create a verse device.

# A Verse-authored creative device that can be placed in a level
V_PlayerLeaderboard := class(creative_device):

    
    var PlayerStatsManager : player_stats_manager = player_stats_manager{}

    @editable
    PlayerReferences : []player_reference_device = array{}
    @editable
    Billboards : []billboard_device = array{}

    GetAgentName<localizes>(Agent:agent)<computes>: message = "{Agent}"
    ToMessage<localizes>(String : string) : message = "{String}"

    # Runs when the device is started in a running game
    OnBegin<override>()<suspends>:void=
        spawn:
            LeaderboardCheck()

    LeaderboardCheck()<suspends>:void=
        loop:
            SetupPlayerRanking()
            Sleep(30.0)

    #In progress, I can do the sort method 
    SetupPlayerRanking():void=
        var PlayerStoneArray : []big_number = array{}
        var PlayerArray : []player = array{}
        Print("Setup Player Ranking")

        Players := GetPlayspace().GetPlayers()
        for(InPlayer : Players):
            if(PlayerStatsData := PlayerStatsManager.GetPlayerStats[InPlayer]):
                #Array of big numbers
                set PlayerStoneArray += array{PlayerStatsData.TotalStoneOreAmount}
        #Once we have all of the arrays we can pass them in and sort them
        PlayersSorted := SortHighestStrength(PlayerStoneArray, Players)

        #Once sorted we can get the top three and display them
        for(Index -> Reference : PlayerReferences):
            Reference.Clear()
            if(Player := PlayersSorted[Index], Billboard := Billboards[Index]):
                if(PlayerStatsData := PlayerStatsManager.GetPlayerStats[Player]):
                    Reference.Register(Player)

                    Billboard.SetText(ToMessage("Total Stone Mined: {BigNumber_FormatDisplay(PlayerStatsData.TotalStoneOreAmount)}"))

    #Big Numbers Bubble Sort [This hasn't been tested yet]
    SortHighestStrength(UnsortedBigNumbers : []big_number, UnsortedPlayers : []player):[]player=
        Print("Sorting")
        # Create New Array to be returned
        var SortedBigNumbersArray : []big_number = UnsortedBigNumbers
        var SortedPlayersArray : []player = UnsortedPlayers
        ArrayLength := UnsortedBigNumbers.Length
        loop:
            for(i := 0..ArrayLength - 1):
                for(j := 0..ArrayLength - i - 1):
                    if(FirstBigNumber := SortedBigNumbersArray[j], SecondBigNumber := SortedBigNumbersArray[j + 1], FirstPlayer := SortedPlayersArray[j], SecondPlayer := SortedPlayersArray[j + 1]):
                        FirstBigNumberGreaterThanSecond := BigNumber_GreaterThanOrEquals(FirstBigNumber, SecondBigNumber)
                        if(FirstBigNumberGreaterThanSecond = false):
                        #if(FirstFloat > SecondFloat):
                            if:
                                set SortedBigNumbersArray[j] = SecondBigNumber
                                set SortedBigNumbersArray[j + 1] = FirstBigNumber
                                set SortedPlayersArray[j] = SecondPlayer
                                set SortedPlayersArray[j + 1] = FirstPlayer
            return SortedPlayersArray