using { /Fortnite.com/Devices }
using { /Verse.org/Simulation }
using { /UnrealEngine.com/Temporary/Diagnostics }

# Added
using { /UnrealEngine.com/Temporary/UI }
using { /Fortnite.com/Characters }
using { /UnrealEngine.com/Temporary/SpatialMath }
using { /Fortnite.com/Devices/CreativeAnimation }

V_PlayerManagerDevice := class(creative_device):
    @editable
    var AlgorithmAndFunctionsClassReference: V_AlgorithmsAndFunctions = V_AlgorithmsAndFunctions{}
    @editable
    FriendsManager : V_FriendsManager = V_FriendsManager{}

    # Classes need to help set up UI
    #@editable 
    #StoneVendorDevice : V_VendorDevice = V_VendorDevice{}
    #@editable 
    #SilverVendorDevice : V_VendorDevice = V_VendorDevice{}
    #@editable 
    #GoldVendorDevice : V_VendorDevice = V_VendorDevice{}
    #@editable 
    #DiamondVendorDevice : V_VendorDevice = V_VendorDevice{}
    @editable
    UIManagerDevice : V_UIManagerDevice = V_UIManagerDevice{}
    @editable
    PlayerUpgradeRemote : V_HarvestRemoteUpgrades = V_HarvestRemoteUpgrades{}
    @editable
    QuestSystem : V_QuestSystem = V_QuestSystem{}
    @editable
    TeleporterManager : V_TeleporterManager = V_TeleporterManager{}
    @editable
    PlayerCustomSword : V_PlayerCustomSword = V_PlayerCustomSword{}
    @editable
    AutoMinerZoneManager : V_AutoMinerZoneManager = V_AutoMinerZoneManager{}
    @editable
    PrestigeManager : V_PrestigeFunctionality = V_PrestigeFunctionality{}
    @editable
    AchievementManager : V_AchievementsManager = V_AchievementsManager{}
    @editable
    PlaytimeManager : V_PlaytimeManager = V_PlaytimeManager{}
    @editable
    XPTomePurchases : []V_XPTomePurchase = array{}

    @editable
    SpawnPads : []player_spawner_device = array{}
    @editable
    ClassSelectors : []class_and_team_selector_device = array{}

    @editable
    ResetStatsDevice : V_ResetAllStats = V_ResetAllStats{}

    @editable
    OGPlayerSwitch : switch_device = switch_device{}

    @editable
    LevelCardSelectionDevice : V_LevelUpAndCardUpgradeSelection = V_LevelUpAndCardUpgradeSelection{}

    @editable
    QuestFCHQ : quest_manager = quest_manager{}

    @editable
    HatStorage : V_HatVFXStorage = V_HatVFXStorage{}

    @editable
    ExtraPlayerSetupVolumes : []volume_device = array{}

    var PlayerMap : [player]PerPlayerDataNotSaved = map{}
    
    var ForcefieldVendors : []V_ForceFieldVendor = array{}
    var PlayerUpgradeVendors :[]V_PlayerUpgradeVendorDevice = array{}
    #Update 1.1 Critical - No longer required as the Dungeon UI Entrance Managers handle this
    var DungeonEntranceUI : []V_DungeonEntranceUIDevice = array{}
    var ArtifactManagers : []V_ArtifactManager = array{}

    # Manages and updates stats for each player.
    var PlayerStatManager: player_stats_manager = player_stats_manager{}

    #We have 4 classes total to assign
    var PlayerClassesToAssign : []int = array{1,2,3,4}

    # Runs when the device is started in a running game
    OnBegin<override>()<suspends>:void=
        Print("OnBegin")

        for(Quest : QuestFCHQ.Quests):
            spawn:
                QuestFCHQ.ManageCompletion(Quest)

        for(SpawnPad : SpawnPads):
            SpawnPad.SpawnedEvent.Subscribe(OnSpawn)

        for(ExtraPlayerSetupVolume : ExtraPlayerSetupVolumes):
            ExtraPlayerSetupVolume.AgentEntersEvent.Subscribe(OnSpawn)

        #Forcefield Vendors are set via tags now. Still using the same array system just filling via tag instead of device array @editable
        ForcefieldVendorTaggedList := FindCreativeObjectsWithTag(ForcefieldVendorTag{})
        for(ForcefieldVendorTaggedActor : ForcefieldVendorTaggedList):
            if(ForcefieldVendorActor := V_ForceFieldVendor[ForcefieldVendorTaggedActor]):
                set ForcefieldVendors += array{ForcefieldVendorActor}

        #Vendor UI Setup for Player Upgrade Vendors. Same as above just fills the array
        UpgradeVendorTaggedList := FindCreativeObjectsWithTag(UpgradeVendorUITag{})
        for(UpgradeVendorTaggedActor : UpgradeVendorTaggedList):
            if(UpgradeVendorActor := V_PlayerUpgradeVendorDevice[UpgradeVendorTaggedActor]):
                set PlayerUpgradeVendors += array{UpgradeVendorActor}
        #DO NOT ATTEMPT TO REMOVE THIS WE NEED IT!!!
        #Vendor UI Setup for Dungeon Entrace UI Devicces. Same as above just fills the array
        DungeonEntranceTaggedList := FindCreativeObjectsWithTag(DungeonEntranceUITag{})
        for(DungeonEntranceTaggedActor : DungeonEntranceTaggedList):
            if(DungeonEntranceActor := V_DungeonEntranceUIDevice[DungeonEntranceTaggedActor]):
                set DungeonEntranceUI += array{DungeonEntranceActor}

        ArtifactManagerTaggedList := FindCreativeObjectsWithTag(ArtifactManagerTag{})
        for(ArtifactManagerTaggedActor : ArtifactManagerTaggedList):
            if(ArtifactManagerActor := V_ArtifactManager[ArtifactManagerTaggedActor]):
                set ArtifactManagers += array{ArtifactManagerActor}

        SetUpPlayerStatManagers()
        PlayerStatManager.UpgradesData.SetUpHRUpgradesArray()
        GetPlayspace().PlayerAddedEvent().Subscribe(OnPlayerAdded)
        GetPlayspace().PlayerRemovedEvent().Subscribe(OnPlayerLeaves)

    SetUpPlayerStatManagers():void=
        set PlayerStatManager = AlgorithmAndFunctionsClassReference.AlgorithmAndFunctionClassReference.PlayerStatData
        set PlayerUpgradeRemote.PlayerStatData = PlayerStatManager
        set UIManagerDevice.PlayerStatsManager = PlayerStatManager
        set PlayerCustomSword.PlayerStatManager = PlayerStatManager
        set PrestigeManager.PlayerStatManager = PlayerStatManager
        set PlaytimeManager.PlayerStatManager = PlayerStatManager
        for(Vendor : PlayerUpgradeVendors):
            set Vendor.PlayerStatManager = PlayerStatManager
        for(Manager : ArtifactManagers):
            set Manager.PlayerStatManager = PlayerStatManager

    OnPlayerAdded<private>(Player : player):void=
        #Print("PlayerManager - Player Joined Game")
        if(Agent := agent[Player]):
    
    OnPlayerLeaves<private>(Player : player):void=
        # Dispose all used artifacts
        if(ArtifactManager := ArtifactManagers[0], DataNotSaved := PlayerMap[Player]):
            for(Prop : DataNotSaved.FollowingProps):
                if(PropFollower := ArtifactManager.PropFollower?):
                    PropFollower.DisposeProp(Player, Prop.Slot)

            set PlayerClassesToAssign += array{DataNotSaved.ClassNumber}
        Print("PlayerManager - Player Left Game")

    OnSpawn(InAgent : agent):void=
        if(Player := player[InAgent]):
            spawn:
                SetupPlayer(InAgent)

    # Removes all the widgets for the player that require input
    ResetPlayerUIOnRespawn(InPlayer : player):void=
        #Print("PlayerManager - Reset UI")
        UIManagerDevice.RemoveAllPlayerUI(InPlayer)
        LevelCardSelectionDevice.UIDisplay.HideUI(InPlayer)

        if(PlayerUI := GetPlayerUI[InPlayer]):
            if(CanvasData := PlayerUpgradeRemote.UIRemoteUpgradeMenuClass.BuiltCanvasMap[InPlayer]):
                PlayerUI.RemoveWidget(CanvasData.Canvas)
            if(CanvasData := TeleporterManager.UITeleporterDisplay.BuiltCanvasMap[InPlayer]):
                PlayerUI.RemoveWidget(CanvasData.Canvas)
            for(ForceFieldVendor : ForcefieldVendors):
                if(CanvasData := ForceFieldVendor.UIBarrierVendorDisplay.BuiltCanvasMap[InPlayer]):
                    PlayerUI.RemoveWidget(CanvasData.Canvas)

    #for(ForceFieldVendor : ForceFieldVendors):
    #    set ForceFieldVendor.PlayerStatManager = PlayerStatManager

    SetupPlayer(InAgent : agent)<suspends>:void=
        if(Player := player[InAgent], FortCharacter := InAgent.GetFortCharacter[]):
            #Print("PlayerManager - Setting up player")
            ResetPlayerUIOnRespawn(Player)
            Sleep(0.1)
            PlayerStatManager.InitializePlayer(Player)

            if(PlayerStats := PlayerStatManager.GetPlayerStats[Player]):
                if(PlayerStats.Version < 2):
                    ResetStatsDevice.Reset(Player)
                    Sleep(0.2)
                    ResetStatsDevice.SignalGranter.GrantItem(Player)
                    PlayerStatManager.SetPlayerDataVersion(Player, 2)
                    OGPlayerSwitch.TurnOn(InAgent)

            if(PlayerData := PlayerMap[Player]):
                # Start - Update 1.1 - Added
                if(PlayerData.CountingTime = false):
                    spawn:
                        PlaytimeManager.StartTracking(Player)
                # End - Update 1.1 - Added
            else if (PData := set PlayerMap[Player] = PerPlayerDataNotSaved{}):
                # Start - Update 1.1 - Added
                set PData.CountingTime = true
                spawn:
                    PlaytimeManager.StartTracking(Player)
                # End - Update 1.1 - Added
                for(ForceFieldVendor : ForcefieldVendors):
                    ForceFieldVendor.InitiateVendorUI(Player)
                for(Index -> Vendor : PlayerUpgradeVendors):
                    Vendor.InitiateVendorUI(Player)
                Print("VENDOR LENGTH: {PlayerUpgradeVendors.Length}")
                #NEVER EVER EVER EVER EVER EVER REMOVE THIS OR PREPARE FOR DEATH!!!!!!!!
                for(Entrance : DungeonEntranceUI):
                    Entrance.InitiateVendorUI(Player)

                for(Manager : ArtifactManagers):
                    Manager.InitiatePlayerUI(Player)

            if(PlayerStats := PlayerStatManager.GetPlayerStats[Player]):
                if(PlayerStats.HatID <> -1):
                    if(HatVFX := HatStorage.HatVFXStorage[PlayerStats.HatID]):
                        HatVFX.Pickup(Player)

            UIManagerDevice.InitiatePlayerUI(Player)
            PlayerUpgradeRemote.InitiateRemoteUI(Player)
            PlayerCustomSword.InitiatePlayerUI(Player)
            PrestigeManager.InitiatePlayerUI(Player)
            LevelCardSelectionDevice.InitPlayerXPBar(Player)

            # Removing so that the player doesn't get added more than once
            FriendsManager.RemovePlayerToCheck(Player)


            # Update 1.2 - Attempt to fix Quest UI showing empty quest
            QuestSystem.PlayerSetUp(Player)

            #Print("PlayerManager - Setting up Friends Manager Part 2")
            FriendsManager.AddPlayerToCheck(Player)

            #Check to see if players have claimed a plot last playthrough
            #Set up Custom Classes
            Sleep(0.25)
            CheckPlayerClasses(Player)
            CheckAutoMinerPlotClaiming(Player)

            # Needs a sleep to make sure loading it works on game start
            # Applying upgrades on spawn
            Sleep(2.0)
            ActivatePlayerDeviceUpgrade(Player, PlayerUpgradeType.Movement)
            spawn:
                SetUpArtifacts(Player)
            # Setup Health Upgrades
            #Print("PlayerManager - Setting up Health Upgrades")
            ChangeMaxHealth(Player)
            spawn:
                HealthRegen(Player)

            if(PlayerStats := PlayerStatManager.GetPlayerStats[Player]):                
                if(PlayerStats.Version < 5):
                    # Reset XP to the corrent amount
                    PlayerStatManager.ResetXPBonus(Player)

                    for(XPTomePurchase : XPTomePurchases):
                        if(XPTomePurchase.Switch.GetCurrentState[InAgent]):
                            PlayerStatManager.AddXPBonus(Player, XPTomePurchase.AmountToAddToXP)
                            PlayerStatManager.SetPlayerDataVersion(Player, 5)

    #Checks all player and what available classes there are
    CheckPlayerClasses(InPlayer : player):void=
        var ClassIndex : int = 0
        
        if(PlayerMap[InPlayer].HasClassBeenSet = true): 
            Print("PlayerManager - Player already has class")
            return
        
        #So we take a player and assign entry 0.
        #We then remove the entry and assign the next
        Print("Attempt to assign class to player")
        if(ClassToAssign := PlayerClassesToAssign[0]):
            AssignCustomClass(InPlayer, ClassToAssign)
            Print("Assign Player To Class {ClassToAssign}")
            if (ElementRemovedArray := PlayerClassesToAssign.RemoveElement[0]):
                set PlayerClassesToAssign = ElementRemovedArray
                Print("Assigned Player, Class Slots Left = {PlayerClassesToAssign.Length}")

    #Put you back into the array

    #Checks if player has claimed a plot of land before
    CheckAutoMinerPlotClaiming(InPlayer : player):void=
        if(PlayerStats := PlayerStatManager.GetPlayerStats[InPlayer]):
            if(PlayerStats.PrestigeAutoMiningPlotClaimedUpgrade = 1):
                Print("Player owns auto mining plot upgrade")
                if(ClassNumber := PlayerMap[InPlayer].ClassNumber - 1):
                    Print("Check class number")
                    if(Plot := AutoMinerZoneManager.LandPlots[ClassNumber]):
                        Print("Claim Plot for Player")
                        #Claim plot
                        #I think this was the issue, I'm going to try spawn so it can run paralell
                        spawn:
                            AutoMinerZoneManager.ClaimPlot(InPlayer, Plot)
    
    #Assigns Class to Player
    AssignCustomClass(InPlayer : player, Class : int) : void=
        #Print("PlayerManager - Class Chosen: {Class}")
        if(set PlayerMap[InPlayer].ClassNumber = Class):
        if(set PlayerMap[InPlayer].HasClassBeenSet = true):
        if(ClassSelector := ClassSelectors[Class - 1]):
            ClassSelector.ChangeClass(InPlayer)
            Print("PlayerManager - Class has been set")

    GetPlayerDataNotSaved(InPlayer : player):PerPlayerDataNotSaved=
        if(Data := PlayerMap[InPlayer]):
            return Data
        return PerPlayerDataNotSaved{}

    CriticalCharge(InPlayer : player, IsCrit : logic, NextCrit : logic):void=
        PlayerMegaUpgradeLevel := PlayerStatManager.GetPlayerUpgradeLevel(InPlayer, PlayerUpgradeType.UnlockMegaCritical)
        if(PlayerMegaUpgradeLevel = 0):
            return
        
        PlayerChargeUpgradeLevel := PlayerStatManager.GetPlayerUpgradeLevel(InPlayer, PlayerUpgradeType.CriticalCharge)
        PlayerChargeAmount := PlayerStatManager.UpgradesData.GetCriticalChargeAmount(PlayerChargeUpgradeLevel)

        if(IntPlayerChargeAmount := Int[PlayerChargeAmount]):
            PlayerDataNotSaved := GetPlayerDataNotSaved(InPlayer)
            PlayerDataNotSaved.IncrementCriticalCharge(IntPlayerChargeAmount)

            if(IntPlayerChargeAmount > 0):
                if(NextCrit = false):
                    UIManagerDevice.UICriticalChargeDisplayClass.AddCharge(InPlayer, PlayerDataNotSaved.CurrentCriticalCharge)
                    UIManagerDevice.UICriticalChargeDisplayClass.AddUI(InPlayer)
                else:
                    UIManagerDevice.UICriticalChargeDisplayClass.AddCharge(InPlayer, 0)
                    UIManagerDevice.UICriticalChargeDisplayClass.AddUI(InPlayer)
                    PlayerDataNotSaved.ResetCriticalCharge()

    ExplosiveCharge(InPlayer : player, NextHitExplode : logic):void=
        ExplosiveChargeLevel :=  PlayerStatManager.GetPlayerUpgradeLevel(InPlayer, PlayerUpgradeType.ExplosiveHit)
        if(ExplosiveChargeLevel = 0):
            PlayerDataNotSaved := GetPlayerDataNotSaved(InPlayer)
            set PlayerDataNotSaved.NextHitExplode = false
            return

        PlayerChargeUpgradeLevel := PlayerStatManager.GetPlayerUpgradeLevel(InPlayer, PlayerUpgradeType.ExplosiveHitCharges)
        PlayerChargeAmount := PlayerStatManager.UpgradesData.GetExplosiveHitChargesAmount(PlayerChargeUpgradeLevel)

        if(IntPlayerChargeAmount := Int[PlayerChargeAmount]):
            PlayerDataNotSaved := GetPlayerDataNotSaved(InPlayer)
            PlayerDataNotSaved.IncrementExplodeCharge(IntPlayerChargeAmount)

            if(IntPlayerChargeAmount > 0):
                if(NextHitExplode = false):
                    UIManagerDevice.UIExplosiveHitDisplayClass.AddCharge(InPlayer, PlayerDataNotSaved.CurrentExplosiveCharge)
                    UIManagerDevice.UIExplosiveHitDisplayClass.AddUI(InPlayer)
                else:
                    UIManagerDevice.UIExplosiveHitDisplayClass.AddCharge(InPlayer, 0)
                    UIManagerDevice.UIExplosiveHitDisplayClass.AddUI(InPlayer)
                    PlayerDataNotSaved.ResetExplodeCharge()

    ActivatePlayerDeviceUpgrade(InAgent : agent, InUpgradeType : PlayerUpgradeType):void=
        #Print("PlayerManager - Activating Player Device Upgrade")
        var Upgrade : player_upgrade = player_upgrade{}
        var CurrentLevel : int = 0
        case(InUpgradeType):
            PlayerUpgradeType.Damage =>
                set Upgrade = PlayerStatManager.UpgradesData.DamageUpgrade
                set CurrentLevel = PlayerStatManager.GetPlayerUpgradeLevel(InAgent, PlayerUpgradeType.Damage)
            PlayerUpgradeType.Movement =>
                set Upgrade = PlayerStatManager.UpgradesData.MovementUpgrade
                set CurrentLevel = PlayerStatManager.GetPlayerUpgradeLevel(InAgent, PlayerUpgradeType.Movement)
            PlayerUpgradeType.MaxHealth =>
                ChangeMaxHealth(InAgent)
                return
            _ =>
                #Print("PlayerManager - Missing or No Player Upgrade Type")
                return

        # Activate correct device
        CurrentLevelDamageDevice := Upgrade.GetDamageLevelDevice(CurrentLevel)
        CurrentLevelMovementDevice := Upgrade.GetMovementLevelDevice(CurrentLevel)

        case(InUpgradeType):
            PlayerUpgradeType.Damage =>
                CurrentLevelDamageDevice.Pickup(InAgent)
            PlayerUpgradeType.Movement =>
                CurrentLevelMovementDevice.Activate(InAgent)
                #Print("PlayerManager - Movement Level = {CurrentLevel}")
                #Print("PlayerManager - Movement Upgrade Unlocked, Activate Movement Device")
            _ =>
                #Print("PlayerManager - Missing or No Player Upgrade Type")

    HealthRegen(InPlayer : player)<suspends>:void=
        if(IsHealing := PlayerMap[InPlayer].IsHealing, IsHealing = true):
            #Print("PlayerManager - Already Healing")
            return
        else if (set PlayerMap[InPlayer].IsHealing = true):
            #Print("PlayerManager - Start Health Regen")

        # Get Fortcharacter
        if(CurrentFortCharacter := InPlayer.GetFortCharacter[]):
            # Add health regen amount onto them every 1 sec
            loop:
                if(not InPlayer.IsActive[]):
                    #Print("PlayerManager - BREAK HEALTH REGEN")
                    break
                HealthRegenLevel := PlayerStatManager.GetPlayerUpgradeLevel(InPlayer, PlayerUpgradeType.HealthRegen)
                HealthRegenAmount := PlayerStatManager.UpgradesData.GetHealthRegenAmount(HealthRegenLevel)
                Sleep(1.0)
                CurrentFortCharacter.Heal(HealthRegenAmount)

    ChangeMaxHealth(InAgent : agent):void=
        if(CurrentFortCharacter := InAgent.GetFortCharacter[]):
            MaxHealthLevel := PlayerStatManager.GetPlayerUpgradeLevel(InAgent, PlayerUpgradeType.MaxHealth)
            MaxHealthAmount := PlayerStatManager.UpgradesData.GetMaxHealthAmount(MaxHealthLevel)
            CurrentFortCharacter.SetHealth(100.0 + MaxHealthAmount)
            CurrentFortCharacter.SetMaxHealth(100.0 + MaxHealthAmount)

    SetUpArtifacts(InPlayer : player)<suspends>:void=
        if(PlayerData := PlayerStatManager.GetPlayerStats[InPlayer], PropFollower := UIManagerDevice.ArtifactManager.PropFollower?):
            # Update 1.5 - Checking version of save data so that we can clear the artifacts due to bug with adding artifacts
            if(PlayerData.Version = 3):
                PlayerStatManager.ClearEquippedArtifacts(InPlayer)
                PlayerStatManager.IncrementPlayerDataVersion(InPlayer)
            if(UpdatedPlayerData := PlayerStatManager.GetPlayerStats[InPlayer]):
                for(Artifact : UpdatedPlayerData.EquippedPlayerArtifacts):
                    #DebugPrintArtifactStats(Artifact)
                    PropFollower.DisposeProp(InPlayer, Artifact.EquippedSlot)
                    Sleep(0.5)
                    PropFollower.SpawnPropWithRarity(InPlayer, Artifact.RarityInt, Artifact.EquippedSlot)
                    

quest_manager := class<concrete>:

    QuestsTT<localizes>:message="Quests that are managed by this device."
    @editable:
        ToolTip := QuestsTT
    Quests : []quest = array{}

    RemoverTT<localizes>:message="Item Remover device containing all quest items."
    @editable:
        ToolTip := RemoverTT
    QuestItemRemover : item_remover_device = item_remover_device{}

    QuestCompleted : event(tuple(agent, string)) = event(tuple(agent, string)){}

    ManageCompletion(Quest : quest)<suspends>:void=
        loop:
            QuestCompletionInfo := QuestCompleted.Await()
            Print("Quest Checking Completion {Quest.QuestName} | {QuestCompletionInfo(1)}")
            if (QuestCompletionInfo(1) = Quest.QuestName):
                spawn:
                    Quest.Complete(QuestCompletionInfo(0))

    (Quest : quest).Complete(Agent : agent)<suspends>:void=
        Quest.Granter.GrantItem(Agent)
        Sleep(0.1)
        QuestItemRemover.Remove(Agent)
        Print("Quest Completed {Quest.QuestName}")

quest := class<concrete><unique>:

    QuestTT<localizes>:message="Custom code name for the quest.\n\nNOTE: Make sure this is unique."
    @editable:
        ToolTip := QuestTT
    QuestName : string = "INSERT NAME HERE"

    GranterTT<localizes>:message="Item granter device containing the quest item.\n\nNOTE: Set it to grant to triggering player."
    @editable:
        ToolTip := GranterTT
    Granter : item_granter_device = item_granter_device{}