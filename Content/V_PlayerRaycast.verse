
using { /Fortnite.com/Devices }
using { /Verse.org/Simulation }
using { /UnrealEngine.com/Temporary/Diagnostics }

using { /Fortnite.com/Characters}
using { /UnrealEngine.com/Temporary/SpatialMath }
using { /UnrealEngine.com/Temporary/Diagnostics }
using { /Verse.org/Simulation/Tags }
using { /Verse.org }

# See https://dev.epicgames.com/documentation/en-us/uefn/create-your-own-device-in-verse for how to create a verse device.
MiningProp := class(tag){}

Handler1 := class() {
    Device : V_PlayerRaycast  
    Parameter1 : int 

    HandlerFunction(Agent : agent) : void = {
        Device.MainFunction(Agent, Parameter1)
    }
}

# A Verse-authored creative device that can be placed in a level
V_PlayerRaycast := class(creative_device):
    #How far the raycast will go
    @editable
    var RaycastLength : int = 12
    #Track when a player has pressed the Fire Button
    @editable
    InputDevice : input_trigger_device = input_trigger_device{}
    @editable
    PropManipulator : prop_manipulator_device = prop_manipulator_device{}
    #The advantage of adding them into an array vs tag is that the order is consistent. I can use that order with an index as the foundation of the code.
    @editable
    PropHealthBars : []V_PropHealthBar = array{}

    @editable
    PlayerManager : ?V_PlayerManagerDevice = false

    @editable
    CheckForSword : conditional_button_device = conditional_button_device{}

    #I need to move this into a map for players. This should be per player
    var ClosestPropHealthBar : V_PropHealthBar = V_PropHealthBar{}
  
    #Per Player Data
    var PlayerMap : [player]PerPlayerDataNotSaved = map{}
    #Dot Product Check DotProduct
    var DistanceCheck : float = 400.0

    # Runs when the device is started in a running game
    OnBegin<override>()<suspends>:void=
        #Listen for when a player is attempting to hit an object with the fire button
        InputDevice.PressedEvent.Subscribe(PassPlayerForRaycast)
        #PropManipulator.HarvestingEvent.Subscribe(PassPlayerForRaycast)
        #We simply extend this whenever adding new caves/biomes
        for(Index -> PropHealthBar : PropHealthBars):
            PropHealthBar.ActivationVolume.AgentEntersEvent.Subscribe(Handler1{Device:=Self, Parameter1:=Index}.HandlerFunction)

    #This function will set the tag we're using each time you enter a new cave.
    #I need to handle this per player. It's the closest prop/tag for a specific player not for this device.
    MainFunction(InAgent : agent, Index : int):void=

        #Print("Entering Cave = Index is {Index}")
        if(InPlayer := player[InAgent]):
            if(bPlayerManager := PlayerManager?):
                PlayerNotSavedData := bPlayerManager.GetPlayerDataNotSaved(InPlayer)
                #Make the array empty then fill it
                PlayerNotSavedData.EmptyArray()
                case(Index): 
                    0 =>
                        for(TaggedDevice : FindCreativeObjectsWithTag(tGodNode{}), CastedTag := creative_object_interface[TaggedDevice]):
                            PlayerNotSavedData.SetTagToSearch(CastedTag)
                        if(PropHealthReference := PropHealthBars[0]):
                            PlayerNotSavedData.SetClosestPropHealthBar(PropHealthReference)
                        Print("God Node")
                    1 =>
                        for(TaggedDevice : FindCreativeObjectsWithTag(tMushroomCaveOne1{}), CastedTag := creative_object_interface[TaggedDevice]):
                            PlayerNotSavedData.SetTagToSearch(CastedTag)
                        if(PropHealthReference := PropHealthBars[1]):
                            PlayerNotSavedData.SetClosestPropHealthBar(PropHealthReference)
                        Print("In Mushroom Cave 1")
                    2 =>
                        for(TaggedDevice : FindCreativeObjectsWithTag(tMushroomCave2{}), CastedTag := creative_object_interface[TaggedDevice]):
                            PlayerNotSavedData.SetTagToSearch(CastedTag)
                        if(PropHealthReference := PropHealthBars[2]):
                            PlayerNotSavedData.SetClosestPropHealthBar(PropHealthReference)
                        Print("In Mushroom Cave 2")
                    3 =>
                        for(TaggedDevice : FindCreativeObjectsWithTag(tMushroomCave3{}), CastedTag := creative_object_interface[TaggedDevice]):
                            PlayerNotSavedData.SetTagToSearch(CastedTag)
                        if(PropHealthReference := PropHealthBars[3]):
                            PlayerNotSavedData.SetClosestPropHealthBar(PropHealthReference)
                        Print("In Mushroom Cave 3")
                    4 =>
                        for(TaggedDevice : FindCreativeObjectsWithTag(tMushroomCave4{}), CastedTag := creative_object_interface[TaggedDevice]):
                            PlayerNotSavedData.SetTagToSearch(CastedTag)
                        if(PropHealthReference := PropHealthBars[4]):
                            PlayerNotSavedData.SetClosestPropHealthBar(PropHealthReference)
                        Print("In Mushroom Cave 4")
                    5 =>
                        for(TaggedDevice : FindCreativeObjectsWithTag(tLavaCave1{}), CastedTag := creative_object_interface[TaggedDevice]):
                            PlayerNotSavedData.SetTagToSearch(CastedTag)
                        if(PropHealthReference := PropHealthBars[5]):
                            PlayerNotSavedData.SetClosestPropHealthBar(PropHealthReference)
                        Print("In Lava Cave 1")
                    6 =>
                        for(TaggedDevice : FindCreativeObjectsWithTag(tLavaCave2{}), CastedTag := creative_object_interface[TaggedDevice]):
                            PlayerNotSavedData.SetTagToSearch(CastedTag)
                        if(PropHealthReference := PropHealthBars[6]):
                            PlayerNotSavedData.SetClosestPropHealthBar(PropHealthReference)

                        Print("In Lava Cave 2")
                    7 =>
                        for(TaggedDevice : FindCreativeObjectsWithTag(tLavaCave3{}), CastedTag := creative_object_interface[TaggedDevice]):
                            PlayerNotSavedData.SetTagToSearch(CastedTag)
                        if(PropHealthReference := PropHealthBars[7]):
                            PlayerNotSavedData.SetClosestPropHealthBar(PropHealthReference)
                        Print("In Lava Cave 3")
                    8 =>
                        for(TaggedDevice : FindCreativeObjectsWithTag(tLavaCave4{}), CastedTag := creative_object_interface[TaggedDevice]):
                            PlayerNotSavedData.SetTagToSearch(CastedTag)
                        if(PropHealthReference := PropHealthBars[8]):
                            PlayerNotSavedData.SetClosestPropHealthBar(PropHealthReference)

                        Print("In Lava Cave 4")
                    9 =>
                        for(TaggedDevice : FindCreativeObjectsWithTag(tUnderworldCave1{}), CastedTag := creative_object_interface[TaggedDevice]):
                            PlayerNotSavedData.SetTagToSearch(CastedTag)
                        if(PropHealthReference := PropHealthBars[9]):
                            PlayerNotSavedData.SetClosestPropHealthBar(PropHealthReference)
                        Print("In Underworld Cave 1")
                    10 =>
                        for(TaggedDevice : FindCreativeObjectsWithTag(tUnderworldCave2{}), CastedTag := creative_object_interface[TaggedDevice]):
                            PlayerNotSavedData.SetTagToSearch(CastedTag)
                        if(PropHealthReference := PropHealthBars[10]):
                            PlayerNotSavedData.SetClosestPropHealthBar(PropHealthReference)
                        Print("In Underworld Cave 2")
                    11 =>
                        for(TaggedDevice : FindCreativeObjectsWithTag(tUnderworldCave3{}), CastedTag := creative_object_interface[TaggedDevice]):
                            PlayerNotSavedData.SetTagToSearch(CastedTag)
                        if(PropHealthReference := PropHealthBars[11]):
                            PlayerNotSavedData.SetClosestPropHealthBar(PropHealthReference)
                        Print("In Underworld Cave 3")
                    12 =>
                        for(TaggedDevice : FindCreativeObjectsWithTag(tUnderworldCave4{}), CastedTag := creative_object_interface[TaggedDevice]):
                            PlayerNotSavedData.SetTagToSearch(CastedTag)
                        if(PropHealthReference := PropHealthBars[12]):
                            PlayerNotSavedData.SetClosestPropHealthBar(PropHealthReference)
                        Print("In Underworld Cave 4")
                    13 =>
                        for(TaggedDevice : FindCreativeObjectsWithTag(tMagicCave1{}), CastedTag := creative_object_interface[TaggedDevice]):
                            PlayerNotSavedData.SetTagToSearch(CastedTag)
                        if(PropHealthReference := PropHealthBars[13]):
                            PlayerNotSavedData.SetClosestPropHealthBar(PropHealthReference)
                        Print("In Magic Cave 1")
                    14 =>
                        for(TaggedDevice : FindCreativeObjectsWithTag(tMagicCave2{}), CastedTag := creative_object_interface[TaggedDevice]):
                            PlayerNotSavedData.SetTagToSearch(CastedTag)
                        if(PropHealthReference := PropHealthBars[14]):
                            PlayerNotSavedData.SetClosestPropHealthBar(PropHealthReference)

                        Print("In Magic Cave 2")
                    15 =>
                        for(TaggedDevice : FindCreativeObjectsWithTag(tMagicCave3{}), CastedTag := creative_object_interface[TaggedDevice]):
                            PlayerNotSavedData.SetTagToSearch(CastedTag)
                        if(PropHealthReference := PropHealthBars[15]):
                            PlayerNotSavedData.SetClosestPropHealthBar(PropHealthReference)
                        Print("In Magic Cave 3")
                    16 =>
                        for(TaggedDevice : FindCreativeObjectsWithTag(tMagicCave4{}), CastedTag := creative_object_interface[TaggedDevice]):
                            PlayerNotSavedData.SetTagToSearch(CastedTag)
                        if(PropHealthReference := PropHealthBars[16]):
                            PlayerNotSavedData.SetClosestPropHealthBar(PropHealthReference)
                        Print("In Magic Cave 4")
                    17 =>
                        for(TaggedDevice : FindCreativeObjectsWithTag(tMagicCave5{}), CastedTag := creative_object_interface[TaggedDevice]):
                            PlayerNotSavedData.SetTagToSearch(CastedTag)
                        if(PropHealthReference := PropHealthBars[17]):
                            PlayerNotSavedData.SetClosestPropHealthBar(PropHealthReference)
                        Print("In Magic Cave 5")
                    18 =>
                        for(TaggedDevice : FindCreativeObjectsWithTag(tMagicCave6{}), CastedTag := creative_object_interface[TaggedDevice]):
                            PlayerNotSavedData.SetTagToSearch(CastedTag)
                        if(PropHealthReference := PropHealthBars[18]):
                            PlayerNotSavedData.SetClosestPropHealthBar(PropHealthReference)
                        Print("In Magic Cave 6")
                    19 =>
                        for(TaggedDevice : FindCreativeObjectsWithTag(tMagicCave7{}), CastedTag := creative_object_interface[TaggedDevice]):
                            PlayerNotSavedData.SetTagToSearch(CastedTag)
                        if(PropHealthReference := PropHealthBars[19]):
                            PlayerNotSavedData.SetClosestPropHealthBar(PropHealthReference)
                        Print("In Magic Cave 7")
                    _=>
                        Print("Default")

                #After the closest prop health bar and tags are set, we setup all of the nodes in that area.
                #PlayerNotSavedData.ClosestPropHealthBar.SetupAllNodes(PlayerNotSavedData.TagToSearch)
                PlayerNotSavedData.ClosestPropHealthBar.SetupAllNodes(PlayerNotSavedData.TaggedActors)

    #I have a reset fire implementation which is per player, but I need to test in multiplayer
    PassPlayerForRaycast(InAgent : agent):void=
        #We are first checking if the player has the sword out, we only want players to pickaxe the nodes.
        if(InPlayer := player[InAgent]):
            if (CheckForSword.IsHoldingItem[InPlayer]):
                #Print("Player Holding Sword, don't run raycast")
            else:
                if(bPlayerManager := PlayerManager?):
                    PlayerNotSavedData := bPlayerManager.GetPlayerDataNotSaved(InPlayer)
                    #This stops you from spamming left click
                        if(PlayerNotSavedData.WaitForFire = false):
                            PlayerNotSavedData.ResetFire()
                            spawn:
                                FindTaggedProps(InPlayer)
    
    FindTaggedProps(InPlayer : player)<suspends>:void=
        Print("Raycast Thing")
        if(bPlayerManager := PlayerManager?):
            PlayerNotSavedData := bPlayerManager.GetPlayerDataNotSaved(InPlayer)
            #Initial delay for player left click animation
            Sleep(0.3)
            #If I used a mutator I could check which one I'm in. This would tell me which cave in the world I'm in.
            loop:
                var HitPropLocation : vector3 = vector3{X:=0.0,Y:=0.0,Z:=0.0}
                var PropsToFind : []creative_object_interface = PlayerNotSavedData.TaggedActors
                var RaycastIndex : int = -1
                if(InAgent := agent[InPlayer], FortCharacter := InPlayer.GetFortCharacter[]):
                    set RaycastIndex = RaycastForward(FortCharacter, PropsToFind, RaycastLength, PlayerNotSavedData)
                    if(HitProp := PropsToFind[RaycastIndex]):
                        set HitPropLocation = vector3{X:=HitProp.GetTransform().Translation.X, Y:= HitProp.GetTransform().Translation.Y, Z:= HitProp.GetTransform().Translation.Z}
                        #We can get the creative prop the player hit by casting to creative_prop
                        #We simply pass through the agent and the hit location.
                        #The hit location is used to compare in the array of resource props in the prop health bar
                        PlayerNotSavedData.ClosestPropHealthBar.ResourceHarvest(InAgent, HitPropLocation)
                #If the player isn't holding down left click the we break
                if(InputDevice.IsHeld[InPlayer]):
                else:
                    PlayerNotSavedData.SetFireTrue()
                    break
                Sleep(0.5757)

    RaycastForward(PlayerReference : fort_character, ObjectToSearchFor : []creative_object_interface, TraceLength : int, PData : PerPlayerDataNotSaved): int=
        var Location : vector3 = PlayerReference.GetTransform().Translation
        PlayerRotation := PlayerReference.GetTransform().Rotation
        #Starts at zero but increases the further you get away
        var AngleDistanceInfluence : float = 0.0
        #Closest Distance to bother checking
        var ClosestDistance: float = 1000.0 
        #Set the ClosestNode Location
        var ClosestNodeLocation : vector3 = vector3{X:= 0.0, Y:= 0.0, Z:=0.0}
        #Setup Check Look Angle to false
        var CheckLookAngle : logic = false
        var IndexForClosestObjectAtCorrectAngle : int = 0
        #Setup for the god node if we are at it
        if(PData.ClosestPropHealthBar.HarvestResourceCave = MiningNodeCave.eGodNode):
            set DistanceCheck = 530.0
        #For all of the tags we are going to check distance and then return the closest one
        for(Index2 := 0..ObjectToSearchFor.Length -1):
            #This does mean we distance check for every single tagged actor and then we do the collision detection afterwards on one of the tagged actors
            if(NewSearchVector := vector3{X:= ObjectToSearchFor[Index2].GetTransform().Translation.X, Y:= ObjectToSearchFor[Index2].GetTransform().Translation.Y, Z:= ObjectToSearchFor[Index2].GetTransform().Translation.Z }):
                #We need a different distance check when it's a god node
                #We check the distance first
                EDistance := Distance(Location, NewSearchVector)
                if (EDistance < ClosestDistance): 
                    set ClosestNodeLocation = NewSearchVector
                    set ClosestDistance = EDistance
                    set IndexForClosestObjectAtCorrectAngle = Index2

        #Since we get distance previously, I can just use the closest distance here.
        if(ClosestDistance < DistanceCheck):
            #I need to figure out why the distance isn't quite working
            Print("Distance = {ClosestDistance}")
            if(PData.ClosestPropHealthBar.HarvestResourceCave = MiningNodeCave.eGodNode):
                    set CheckLookAngle = IsLookingAtLocation(PlayerReference, ClosestNodeLocation, 0.01)
            else:
                if(ClosestDistance > 100.0 and ClosestDistance < 200.0):
                    set CheckLookAngle = IsLookingAtLocation(PlayerReference, ClosestNodeLocation, 0.65)
                if(ClosestDistance > 200.0 and ClosestDistance < 300.0):
                    Print("Distance is over 200")
                    set CheckLookAngle = IsLookingAtLocation(PlayerReference, ClosestNodeLocation, 0.75)
                if(ClosestDistance > 300.0 and ClosestDistance < 400.0):
                    Print("Distance is over 300")
                    set CheckLookAngle = IsLookingAtLocation(PlayerReference, ClosestNodeLocation, 0.9)
                if(ClosestDistance > 400.0 and ClosestDistance < 500.0):
                    Print("Distance is over 400")
                    set CheckLookAngle = IsLookingAtLocation(PlayerReference, ClosestNodeLocation, 0.95)
                if(ClosestDistance > 500.0 and ClosestDistance < 600.0):
                    Print("Distance is over 500")
                    set CheckLookAngle = IsLookingAtLocation(PlayerReference, ClosestNodeLocation, 0.95)
            #If the check look angle is true then we can get resources
            if(CheckLookAngle = true):
                return IndexForClosestObjectAtCorrectAngle
        return -1

    IsLookingAtLocation(Character:fort_character, LocationToCheck: vector3, Threshold: float):logic=
        # Get the character's view location
        ViewLocation := Character.GetViewLocation()
        # Get the character's view rotation
        ViewRotation := Character.GetViewRotation()
        # Create a direction vector from the view location to the target vecto
        DirectionVector := Normalize(LocationToCheck - ViewLocation)
        # Convert the view rotation to a direction vector
        ViewDirection := ViewRotation.GetLocalForward()
        DrawDebugCone(Character.GetTransform().Translation, ViewDirection * 1000.0)
        # Compare the direction vector and the view direction
        # Use a dot product to find the cosine of the angle between the two vectors.
        # If it's close to 1, then the vectors are nearly parallel, so the character is looking at the target.
        CosAngle := DotProduct(ViewDirection, DirectionVector)
        if(IsFacing := CosAngle > Threshold):
            return true
        return false

    Magnitude(v1:vector3):float=
        return Sqrt(Pow(v1.X, 2.0) + Pow(v1.Y, 2.0) + Pow(v1.Z, 2.0))

    Normalize(v1:vector3):vector3=
        Mag := Magnitude(v1)
        return vector3{X := v1.X/Mag, Y := v1.Y/Mag, Z:= v1.Z/Mag}

    RotationToDirection(r1:rotation):vector3=
        # Assuming the rotation is Euler angles (pitch, yaw, roll)
        Pitch := r1.GetAxis().X
        Yaw := r1.GetAxis().Y
        
        # Convert angles from degrees to radians if necessary
        PitchRad := DegreesToRadians(Pitch)
        YawRad := DegreesToRadians(Yaw)

        # Convert rotation to direction vector
        Direction := vector3{X:= Cos(Pitch) * Cos(Yaw), Y:= Sin(Pitch), Z:= Cos(Pitch) * Sin(Yaw)}
        return Direction

    DrawDebugCone(ConeOrigin: vector3, ViewDirection : vector3):void=
        # Draw a cone at the character's view location, in the direction they are facing
        Drawer := debug_draw{}
        Drawer.DrawCone(Origin := ConeOrigin, 
        Direction := ViewDirection, 
        ?NumSides:= 6,
        ?Height := float(50.0), 
        ?AngleWidthRadians := float(0.3), 
        ?AngleHeightRadians:= float(0.3),
        ?Color:= Colors.NamedColors.Red,
        ?Thickness:= float(2.0),
        ?DrawDurationPolicy := debug_draw_duration_policy.FiniteDuration, 
        ?Duration:= float(5.0) )