using { /Fortnite.com/Devices }
using { /Verse.org/Simulation }
using { /UnrealEngine.com/Temporary/Diagnostics }
using { /Verse.org/Random }

IsDivisiblePer5(X:int)<decides><transacts>:void = (Mod[X, 5] = 0)

#Track multiple stats per player
Player_Stats_Data := class<final><persistable>:
    Version : int = 6

    XPAmount : big_number = big_number{}
    XPAdd : float = 0.0
    Level : int = 1

    Prestige : int = 0

    HatID : int = -1

    HRBaseCost : big_number = big_number:
        Value := 1.0
        Exponent := 0
    HRCostIncreaseRate : big_number = big_number:
        Value := 1.12
        Exponent := 0
    HRRateIncrease : float = 1.0

    # Saving HarvestRatePerHit for each upgrade
    HarvestRatePerHit : big_number = big_number:
        Value := 1.0
        Exponent := 0

    PrestigeBonusHarvestRate : float = 0.0

    LifetimePrestigeResourceAmount : big_number = big_number:
        Value := 0.0
        Exponent := 0

    LastPrestigedAmount : big_number = big_number:
        Value := 0.0
        Exponent := 0
    
    TotalStoneSincePrestige : big_number = big_number:
        Value := 0.0
        Exponent := 0

    # Add an extra one to these for every upgrade we want
    HarvestRatePerHitUpgrades : []big_number = {big_number{}, big_number{}, big_number{}, big_number{}, big_number{}, 
    big_number{}, big_number{}, big_number{}, big_number{}, big_number{}, 
    big_number{}, big_number{}, big_number{}, big_number{}, big_number{}, 
    big_number{}, big_number{}, big_number{}, big_number{}, big_number{}}

    HRUpgradeTierMultipliers : []float = array{1.0, 0.0, 0.0, 0.0, 0.0, 
    0.0, 0.0, 0.0, 0.0, 0.0,
    0.0, 0.0, 0.0, 0.0, 0.0,
    0.0, 0.0, 0.0, 0.0, 0.0}

    HRUpgradeLevels : []float = array{1.0, 0.0, 0.0, 0.0, 0.0, 
    0.0, 0.0, 0.0, 0.0, 0.0,
    0.0, 0.0, 0.0, 0.0, 0.0,
    0.0, 0.0, 0.0, 0.0, 0.0}
    
    HRUpgradeTierMultiplierUpgradesAvaliable : []int = array{0, 0, 0, 0, 0,
    0, 0, 0, 0, 0,
    0, 0, 0, 0, 0,
    0, 0, 0, 0, 0}

    #Vendor Upgrade Tier Multiplier Avaliable Count
    StoneVendorTierMultiplierUpgradesAvaliable : int = 0
    SilverVendorTierMultiplierUpgradesAvaliable : int = 0
    GoldVendorTierMultiplierUpgradesAvaliable : int = 0
    DiamondVendorTierMultiplierUpgradesAvaliable : int = 0

    UpgradePurchaseAmount  : int = 1

    # Percentage bonus on cave unlocks
    CaveUnlockPercent : float = 0.0

    #Stone Variables
    PlayerStoneOreAmount : big_number = big_number:
        Value := 0.0
        Exponent := 0
    StoneOreRatePerTrade : big_number = big_number:
        Value := 1.0
        Exponent := 0
    StoneVendorUpgradeLevel : float = 0.0
    StoneVendorTierMultiplier : float = 1.0

    PlayerGoldCoinAmount : big_number = big_number:
        Value := 0.0
        Exponent := 0

    #Artifact Variables
    OwnedPlayerArtifacts : []Artifact_Data = array{}
    EquippedPlayerArtifacts : []Artifact_Data = array{}

    #Sword Variables
    OwnedPlayerSword : []Sword_Data = array{}
    
    PlayerCopperOreAmount : big_number = big_number:
        Value := 0.0
        Exponent := 0
    PlayerSilverOreAmount : big_number = big_number:
        Value := 0.0
        Exponent := 0
    PlayerMalachiteOreAmount : big_number = big_number:
        Value := 0.0
        Exponent := 0
    PlayerSpectroliteOreAmount : big_number = big_number:
        Value := 0.0
        Exponent := 0
    PlayerObsidianOreAmount : big_number = big_number:
        Value := 0.0
        Exponent := 0
    PlayerPrestigeResourceAmount : big_number = big_number:
        Value := 0.0
        Exponent := 0

    #Quest Variables
    CompletedQuestLines : []string = array{}
    CompletedQuests : []string = array{}
    CurrentQuest : string = ""
    CurrentQuestIncrement : big_number = big_number{}

    #Player Upgrades
    CriticalChanceUpgrade : int = 0
    CriticalDamageUpgrade : int = 0
    UnlockMegaCriticalUpgrade : int = 0
    MegaCriticalUpgrade : int = 0
    CriticalChargeUpgrade : int = 0
    FinalHitUpgrade : int = 0
    OreQuantityUpgrade : int = 0
    FriendsUpgrade : int = 0
    DamageUpgrade : int = 0
    MovementUpgrade : int = 0
    HealthRegenUpgrade : int = 0
    MaxHealthUpgrade : int = 0
    ExplosiveHitUpgrade : int = 0
    ExplosiveHitRangeUpgrade : int = 0
    ExplosiveHitDamageUpgrade :  int = 0
    ExplosiveHitChargesUpgrade : int = 0
    AltarUpgrade : int = 0
    AltarStrengthUpgrade : int = 0
    AltarTimeUpgrade : int = 0
    GlobalVendorUpgrade : int = 1

    #Player Token Var 
    AlterMaxAmount : int = 0
    DungeonMaxAmount   : int = 0
    AlterTimer : float = 15.0
    DungeonTimer : float = 15.0

    #Auto Mining Upgrades
    PrestigeAutoMiningPlotClaimedUpgrade : int = 0
    PrestigeAutoMiningCount : int = 0
    PrestigeAutoMiningHitRate : int = 0

    #DungeonKing
    DungeonKingTokenSlotAmount : int = 0
    DungeonKingTokenSpeedAmount : int = 0

    #Prestige Upgrades
    PrestigeAFKAreaUpgrade : int = 0
    PrestigeStrengthUpgrade : int = 0
    PrestigeDungeonMasteryUpgrade : int = 0
    PrestigeCore : int = 0
    PrestigeArtifactMastery : int = 0 
    PrestigeArtifactSlots : int = 0 

    # Achievement Variables
    AchievementTotalStoneBonus : float = 0.0
    AchievementTotalArtifactsSmeltedBonus : float = 0.0
    AchievementTotalRocksDestroyedBonus : float = 0.0
    AchievementHighestStrengthBonus : float = 0.0
    AchievementTotalDungeonsCompletedBonus : float = 0.0

    TotalRocksDestroyed : big_number = big_number{}
    TotalArtifactsSmelted : big_number = big_number{}
    HighestStrength : big_number = big_number{}
    TotalDungeonsCompleted : big_number = big_number{}
    TotalStoneOreAmount : big_number = big_number{}

    #Total Playtime Variables
    MinutesPlayed : float = 0.0
    HoursPlayed : float = 0.0

    #New Resources
    OrbAmount : big_number = big_number{}
    DustAmount : big_number = big_number{}

    #Cards
    OwnedCards : []SavedCardData = array{}

#Creates a map of players and saved stats data. Stores player data in the map.
var PlayerStatsDataMap : weak_map(player, Player_Stats_Data) = map{}

#Constructor used to make a new instance of player_stats_data to make updates
MakePlayerStatsUpdate<constructor>(OldData : Player_Stats_Data)<transacts> := Player_Stats_Data:
    Version := OldData.Version
    XPAmount := OldData.XPAmount
    XPAdd := OldData.XPAdd
    Level := OldData.Level
    Prestige := OldData.Prestige

    HatID := OldData.HatID

    HarvestRatePerHit := OldData.HarvestRatePerHit
    PrestigeBonusHarvestRate := OldData.PrestigeBonusHarvestRate
    LastPrestigedAmount := OldData.LastPrestigedAmount
    LifetimePrestigeResourceAmount := OldData.LifetimePrestigeResourceAmount
    TotalStoneSincePrestige := OldData.TotalStoneSincePrestige
    HarvestRatePerHitUpgrades := OldData.HarvestRatePerHitUpgrades
    HRBaseCost := OldData.HRBaseCost
    HRCostIncreaseRate := OldData.HRCostIncreaseRate
    HRRateIncrease := OldData.HRRateIncrease
    HRUpgradeTierMultipliers := OldData.HRUpgradeTierMultipliers
    HRUpgradeLevels := OldData.HRUpgradeLevels

    #Stone
    TotalStoneOreAmount := OldData.TotalStoneOreAmount
    PlayerStoneOreAmount := OldData.PlayerStoneOreAmount
    StoneOreRatePerTrade := OldData.StoneOreRatePerTrade
    StoneVendorUpgradeLevel := OldData.StoneVendorUpgradeLevel
    StoneVendorTierMultiplier := OldData.StoneVendorTierMultiplier

    PlayerGoldCoinAmount := OldData.PlayerGoldCoinAmount
    #Artifact Variables
    OwnedPlayerArtifacts := OldData.OwnedPlayerArtifacts
    EquippedPlayerArtifacts := OldData.EquippedPlayerArtifacts

    #Sword Variables
    OwnedPlayerSword := OldData.OwnedPlayerSword

    PlayerCopperOreAmount := OldData.PlayerCopperOreAmount
    PlayerSilverOreAmount := OldData.PlayerSilverOreAmount
    PlayerMalachiteOreAmount := OldData.PlayerMalachiteOreAmount
    PlayerSpectroliteOreAmount := OldData.PlayerSpectroliteOreAmount
    PlayerObsidianOreAmount := OldData.PlayerObsidianOreAmount
    PlayerPrestigeResourceAmount := OldData.PlayerPrestigeResourceAmount

    UpgradePurchaseAmount := OldData.UpgradePurchaseAmount

    #Tier Multiplier Upgrades Avaliable
    HRUpgradeTierMultiplierUpgradesAvaliable := OldData.HRUpgradeTierMultiplierUpgradesAvaliable

    StoneVendorTierMultiplierUpgradesAvaliable := OldData.StoneVendorTierMultiplierUpgradesAvaliable
    SilverVendorTierMultiplierUpgradesAvaliable := OldData.SilverVendorTierMultiplierUpgradesAvaliable
    GoldVendorTierMultiplierUpgradesAvaliable := OldData.GoldVendorTierMultiplierUpgradesAvaliable
    DiamondVendorTierMultiplierUpgradesAvaliable := OldData.DiamondVendorTierMultiplierUpgradesAvaliable

    # Cave unlock strength percentage 
    CaveUnlockPercent := OldData.CaveUnlockPercent

    #Quest Save Data
    CompletedQuestLines := OldData.CompletedQuestLines
    CompletedQuests := OldData.CompletedQuests
    CurrentQuest := OldData.CurrentQuest
    CurrentQuestIncrement := OldData.CurrentQuestIncrement

    #Player Upgrades
    CriticalChanceUpgrade := OldData.CriticalChanceUpgrade
    CriticalDamageUpgrade := OldData.CriticalDamageUpgrade
    UnlockMegaCriticalUpgrade := OldData.UnlockMegaCriticalUpgrade
    MegaCriticalUpgrade := OldData.MegaCriticalUpgrade
    CriticalChargeUpgrade := OldData.CriticalChargeUpgrade
    FinalHitUpgrade := OldData.FinalHitUpgrade
    OreQuantityUpgrade := OldData.OreQuantityUpgrade
    FriendsUpgrade := OldData.FriendsUpgrade
    DamageUpgrade := OldData.DamageUpgrade
    MovementUpgrade := OldData.MovementUpgrade
    HealthRegenUpgrade := OldData.HealthRegenUpgrade
    MaxHealthUpgrade := OldData.MaxHealthUpgrade
    ExplosiveHitUpgrade := OldData.ExplosiveHitUpgrade
    ExplosiveHitRangeUpgrade := OldData.ExplosiveHitRangeUpgrade
    ExplosiveHitDamageUpgrade := OldData.ExplosiveHitDamageUpgrade
    ExplosiveHitChargesUpgrade := OldData.ExplosiveHitChargesUpgrade
    AltarUpgrade := OldData.AltarUpgrade
    AltarStrengthUpgrade := OldData.AltarStrengthUpgrade
    AltarTimeUpgrade := OldData.AltarTimeUpgrade

    #Player Token Var
    AlterMaxAmount := OldData.AlterMaxAmount
    DungeonMaxAmount   := OldData.DungeonMaxAmount
    AlterTimer := OldData.AlterTimer
    DungeonTimer := OldData.DungeonTimer

    #Auto Mining Upgrades
    PrestigeAutoMiningPlotClaimedUpgrade := OldData.PrestigeAutoMiningPlotClaimedUpgrade
    PrestigeAutoMiningCount := OldData.PrestigeAutoMiningCount
    PrestigeAutoMiningHitRate := OldData.PrestigeAutoMiningHitRate

    #Prestige Upgrades
    PrestigeAFKAreaUpgrade := OldData.PrestigeAFKAreaUpgrade
    PrestigeStrengthUpgrade := OldData.PrestigeStrengthUpgrade
    PrestigeDungeonMasteryUpgrade := OldData.PrestigeDungeonMasteryUpgrade
    PrestigeCore := OldData.PrestigeCore
    PrestigeArtifactMastery := OldData.PrestigeArtifactMastery
    PrestigeArtifactSlots := OldData.PrestigeArtifactSlots

    #Dungeon King
    DungeonKingTokenSlotAmount := OldData.DungeonKingTokenSlotAmount
    DungeonKingTokenSpeedAmount := OldData.DungeonKingTokenSpeedAmount

    # Achievement
    AchievementTotalStoneBonus := OldData.AchievementTotalStoneBonus
    AchievementTotalArtifactsSmeltedBonus := OldData.AchievementTotalArtifactsSmeltedBonus
    AchievementTotalRocksDestroyedBonus := OldData.AchievementTotalRocksDestroyedBonus
    AchievementHighestStrengthBonus := OldData.AchievementHighestStrengthBonus
    AchievementTotalDungeonsCompletedBonus := OldData.AchievementTotalDungeonsCompletedBonus

    TotalRocksDestroyed := OldData.TotalRocksDestroyed
    TotalArtifactsSmelted := OldData.TotalArtifactsSmelted
    HighestStrength := OldData.HighestStrength
    TotalDungeonsCompleted := OldData.TotalDungeonsCompleted

    #Total Playtime
    MinutesPlayed := OldData.MinutesPlayed
    HoursPlayed := OldData.HoursPlayed

    #New Resources
    OrbAmount := OldData.OrbAmount
    DustAmount := OldData.DustAmount

    #Cards
    OwnedCards := OldData.OwnedCards

#Class that defines a custom player. Handles loading data and updating data. We manage players here.
player_stats_manager := class:
    UpgradesData : upgrades_data = upgrades_data{}
    
    #See if the player has persistent data and return the stats
    GetPlayerStats(InAgent : agent)<decides><transacts>:Player_Stats_Data=
        var PlayerStats: Player_Stats_Data = Player_Stats_Data{}
        if(Player := player[InAgent], Player.IsActive[]):
            if(PlayerStatData := PlayerStatsDataMap[Player]):
                set PlayerStats = PlayerStatData
        return PlayerStats

    InitializePlayer(Player : player):void=
        LoadSaveData(Player)

    
    InitializeAllPlayers(Players:[]player):void=
        for(Player : Players):
            InitializePlayer(Player)

    #Load Player Stats Map
    LoadSaveData(Player : player):void=
        if:
            not PlayerStatsDataMap[Player]
            set PlayerStatsDataMap[Player] = Player_Stats_Data{}


    # Update 1.5 - Added so that we can do stuff to player's data depending on version
    IncrementPlayerDataVersion(InAgent : agent):void=
        if(Player := player[InAgent], Player.IsActive[]):
            if(OldData := PlayerStatsDataMap[Player]):
                if:
                    set PlayerStatsDataMap[Player] = Player_Stats_Data:
                        MakePlayerStatsUpdate<constructor>(OldData)  
                        Version := OldData.Version + 1
        
    SetPlayerDataVersion(InAgent : agent, InVersion : int):void=
        if(Player := player[InAgent], Player.IsActive[]):
            if(OldData := PlayerStatsDataMap[Player]):
                if:
                    set PlayerStatsDataMap[Player] = Player_Stats_Data:
                        MakePlayerStatsUpdate<constructor>(OldData)  
                        Version := InVersion

    #Spend the tier multipler upgrade amount so we can't infinitely use them
    ModifyVendorMultiplierAmount(InAgent : agent, Upgrade : int, Increment : logic, Decrement : logic):void=
        if(Player := player[InAgent], Player.IsActive[]):
            if(OldData := PlayerStatsDataMap[Player]):
                if(Upgrade = 1):
                    if(Increment = true):
                        if:
                            set PlayerStatsDataMap[Player] = Player_Stats_Data:
                                MakePlayerStatsUpdate<constructor>(OldData)  
                                StoneVendorTierMultiplierUpgradesAvaliable := OldData.StoneVendorTierMultiplierUpgradesAvaliable + 1
                    if(Decrement = true):
                        if:
                            set PlayerStatsDataMap[Player] = Player_Stats_Data:
                                MakePlayerStatsUpdate<constructor>(OldData)  
                                StoneVendorTierMultiplierUpgradesAvaliable := OldData.StoneVendorTierMultiplierUpgradesAvaliable - 1
                                #Duplicate this
                    Print("StoneVendorUpgradesAvaliable {OldData.StoneVendorTierMultiplierUpgradesAvaliable}")
                if(Upgrade = 2):
                    if(Increment = true):
                        if:
                            set PlayerStatsDataMap[Player] = Player_Stats_Data:
                                MakePlayerStatsUpdate<constructor>(OldData)  
                                SilverVendorTierMultiplierUpgradesAvaliable := OldData.SilverVendorTierMultiplierUpgradesAvaliable + 1
                    if(Decrement = true):
                        if:
                            set PlayerStatsDataMap[Player] = Player_Stats_Data:
                                MakePlayerStatsUpdate<constructor>(OldData)  
                                SilverVendorTierMultiplierUpgradesAvaliable := OldData.SilverVendorTierMultiplierUpgradesAvaliable - 1
                if(Upgrade = 3):
                    if(Increment = true):
                        if:
                            set PlayerStatsDataMap[Player] = Player_Stats_Data:
                                MakePlayerStatsUpdate<constructor>(OldData)  
                                GoldVendorTierMultiplierUpgradesAvaliable := OldData.GoldVendorTierMultiplierUpgradesAvaliable + 1
                    if(Decrement = true):
                        if:
                            set PlayerStatsDataMap[Player] = Player_Stats_Data:
                                MakePlayerStatsUpdate<constructor>(OldData)  
                                GoldVendorTierMultiplierUpgradesAvaliable := OldData.GoldVendorTierMultiplierUpgradesAvaliable - 1
                                
                if(Upgrade = 4):
                    if(Increment = true):
                        if:
                            set PlayerStatsDataMap[Player] = Player_Stats_Data:
                                MakePlayerStatsUpdate<constructor>(OldData)  
                                DiamondVendorTierMultiplierUpgradesAvaliable := OldData.DiamondVendorTierMultiplierUpgradesAvaliable + 1
                    if(Decrement = true):
                        if:
                            set PlayerStatsDataMap[Player] = Player_Stats_Data:
                                MakePlayerStatsUpdate<constructor>(OldData)  
                                DiamondVendorTierMultiplierUpgradesAvaliable := OldData.DiamondVendorTierMultiplierUpgradesAvaliable - 1

    #Increase the tier multiplier amount so we can spend them
    ModifyTierMultiplerAmount(InAgent : agent, UpgradeIndex : int, Increment : logic, Decrement : logic):void=
        if(Player := player[InAgent], Player.IsActive[]):
            if(OldData := PlayerStatsDataMap[Player]):
                if(Increment = true):
                    if(NewTierMultiplier := OldData.HRUpgradeTierMultiplierUpgradesAvaliable[UpgradeIndex] + 1):
                        var TierMultiplierArray : []int = OldData.HRUpgradeTierMultiplierUpgradesAvaliable
                        if(set TierMultiplierArray[UpgradeIndex] = NewTierMultiplier):
                            if:
                                set PlayerStatsDataMap[Player] = Player_Stats_Data:
                                    MakePlayerStatsUpdate<constructor>(OldData)  
                                    HRUpgradeTierMultiplierUpgradesAvaliable := TierMultiplierArray
                if(Decrement = true):
                    if(NewTierMultiplier := OldData.HRUpgradeTierMultiplierUpgradesAvaliable[UpgradeIndex] - 1):
                        var TierMultiplierArray : []int = OldData.HRUpgradeTierMultiplierUpgradesAvaliable
                        if(set TierMultiplierArray[UpgradeIndex] = NewTierMultiplier):
                            if:
                                set PlayerStatsDataMap[Player] = Player_Stats_Data:
                                    MakePlayerStatsUpdate<constructor>(OldData)  
                                    HRUpgradeTierMultiplierUpgradesAvaliable := TierMultiplierArray

    #This function allows us to modify the players Harvest Rate Upgrades. Currently these are 1=Strength, 2=Sharper Pick, 3=Re-Enforced Pick, 4=Auto Harvesting Pick
    UpdateHRUpgradesLevel(InAgent : agent, UpgradeIndex : int, InPurchaseAmount : int):void=
        if(Player := player[InAgent], Player.IsActive[]):
            if(OldData := PlayerStatsDataMap[Player]):
                #Conditional checks for Harvest Rate Upgrades
                if(NewUpgradeLevel := OldData.HRUpgradeLevels[UpgradeIndex] + (InPurchaseAmount * 1.0)):
                    var UpgradeLevelsArray : []float = OldData.HRUpgradeLevels
                    if(set UpgradeLevelsArray[UpgradeIndex] = NewUpgradeLevel):
                        if:
                            set PlayerStatsDataMap[Player] = Player_Stats_Data:
                                MakePlayerStatsUpdate<constructor>(OldData)  
                                HRUpgradeLevels := UpgradeLevelsArray
        
    UpdateVendorUpgradesLevel(InAgent : agent, Resource : VendorUpgradeResourceTypes):void=
        if(Player := player[InAgent], Player.IsActive[]):
            if(OldData := PlayerStatsDataMap[Player]):
                if(Resource = VendorUpgradeResourceTypes.Stone):
                    Print("Upgrade Stone Vendor Level")
                    if:
                        set PlayerStatsDataMap[Player] = Player_Stats_Data:
                            MakePlayerStatsUpdate<constructor>(OldData)  
                            StoneVendorUpgradeLevel := OldData.StoneVendorUpgradeLevel + 1.0


    #Hit buy, how much extra power [Rate Increase x Tier Multiplier] [Purchase]

    #Combines all Harvest Rate Upgrades for when all upgrades are in the game
    UpdateHarvestRatePerHit(InAgent : agent, Algorithm : Algorithms, UpgradeIndex : int, HarvestBaseRate : big_number, HarvestRateIncrease : big_number):void=
        if(Player := player[InAgent], Player.IsActive[]):
            if(OldData := PlayerStatsDataMap[Player]):
                var NewHarvestRatePerHit : big_number = big_number{}
                
                var HarvestRatePerHitArray : []big_number = OldData.HarvestRatePerHitUpgrades

                if(UpgradeLevel := OldData.HRUpgradeLevels[UpgradeIndex], TierMultiplier := OldData.HRUpgradeTierMultipliers[UpgradeIndex]):
                    NewHarvestRatePerHitUpgrade := Algorithm.CalculateHarvestRateOnHit(HarvestRateIncrease, UpgradeLevel, HarvestBaseRate, TierMultiplier)
                    if(set HarvestRatePerHitArray[UpgradeIndex] = NewHarvestRatePerHitUpgrade):

                    for(HarvestRatePerHit : HarvestRatePerHitArray):
                        set NewHarvestRatePerHit = BigNumber_Add(NewHarvestRatePerHit, HarvestRatePerHit)
                        if:
                            set PlayerStatsDataMap[Player] = Player_Stats_Data:
                                MakePlayerStatsUpdate<constructor>(OldData)
                                HarvestRatePerHit := NewHarvestRatePerHit
                                HarvestRatePerHitUpgrades := HarvestRatePerHitArray

            if(NewData := PlayerStatsDataMap[Player]):
                GreaterThan := BigNumber_GreaterThanOrEquals(NewData.HarvestRatePerHit, NewData.HighestStrength)
                if(GreaterThan = true):
                    if:
                        set PlayerStatsDataMap[Player] = Player_Stats_Data:
                            MakePlayerStatsUpdate<constructor>(NewData)
                            HighestStrength := NewData.HarvestRatePerHit

    UpdatePrestigeBonusHarvestRate(InAgent : agent, InAmount : float):void=
        if(Player := player[InAgent], Player.IsActive[]):
            if(OldData := PlayerStatsDataMap[Player]):
                NewAmount := OldData.PrestigeBonusHarvestRate + InAmount
                
                if:
                    set PlayerStatsDataMap[Player] = Player_Stats_Data:
                        MakePlayerStatsUpdate<constructor>(OldData)
                        PrestigeBonusHarvestRate := NewAmount

    UpdatePrestige(InAgent : agent):void=
        if(Player := player[InAgent], Player.IsActive[]):
            if(OldData := PlayerStatsDataMap[Player]):
                NewAmount := OldData.Prestige + 1
                
                if:
                    set PlayerStatsDataMap[Player] = Player_Stats_Data:
                        MakePlayerStatsUpdate<constructor>(OldData)
                        Prestige := NewAmount

    UpdateAchievementBonusHarvestRate(InAgent : agent, InAmount : float, Type : AchievementType):void=
        if(Player := player[InAgent], Player.IsActive[]):
            if(OldData := PlayerStatsDataMap[Player]):
                case(Type):
                    AchievementType.TotalStone =>
                        NewAmount := OldData.AchievementTotalStoneBonus + InAmount
                    
                        if:
                            set PlayerStatsDataMap[Player] = Player_Stats_Data:
                                MakePlayerStatsUpdate<constructor>(OldData)
                                AchievementTotalStoneBonus := InAmount
                    AchievementType.HighestStrength =>
                        NewAmount := OldData.AchievementHighestStrengthBonus + InAmount
                        
                        if:
                            set PlayerStatsDataMap[Player] = Player_Stats_Data:
                                MakePlayerStatsUpdate<constructor>(OldData)
                                AchievementHighestStrengthBonus := InAmount

                    AchievementType.DungeonsCompleted =>
                        NewAmount := OldData.AchievementTotalDungeonsCompletedBonus + InAmount
                        
                        if:
                            set PlayerStatsDataMap[Player] = Player_Stats_Data:
                                MakePlayerStatsUpdate<constructor>(OldData)
                                AchievementTotalDungeonsCompletedBonus := InAmount

                    AchievementType.ArtifactsSmelted =>
                        NewAmount := OldData.AchievementTotalArtifactsSmeltedBonus + InAmount
                        
                        if:
                            set PlayerStatsDataMap[Player] = Player_Stats_Data:
                                MakePlayerStatsUpdate<constructor>(OldData)
                                AchievementTotalArtifactsSmeltedBonus := InAmount
                    AchievementType.RocksDestroyed =>
                        NewAmount := OldData.AchievementTotalRocksDestroyedBonus + InAmount
                        
                        if:
                            set PlayerStatsDataMap[Player] = Player_Stats_Data:
                                MakePlayerStatsUpdate<constructor>(OldData)
                                AchievementTotalRocksDestroyedBonus := InAmount
                    AchievementType.Null =>

    UpdateTotalRocksDestroyed(InAgent : agent, InAmount : big_number):void=
        if(Player := player[InAgent], Player.IsActive[]):
            if(OldData := PlayerStatsDataMap[Player]):
                NewAmount := BigNumber_Add(OldData.TotalRocksDestroyed, InAmount)
                
                if:
                    set PlayerStatsDataMap[Player] = Player_Stats_Data:
                        MakePlayerStatsUpdate<constructor>(OldData)
                        TotalRocksDestroyed := NewAmount

    UpdateTotalArtifactsSmelted(InAgent : agent, InAmount : big_number):void=
        if(Player := player[InAgent], Player.IsActive[]):
            if(OldData := PlayerStatsDataMap[Player]):
                NewAmount := BigNumber_Add(OldData.TotalArtifactsSmelted, InAmount)
                
                if:
                    set PlayerStatsDataMap[Player] = Player_Stats_Data:
                        MakePlayerStatsUpdate<constructor>(OldData)
                        TotalArtifactsSmelted := NewAmount

    UpdateTotalDungeonsCompleted(InAgent : agent, InAmount : big_number):void=
        if(Player := player[InAgent], Player.IsActive[]):
            if(OldData := PlayerStatsDataMap[Player]):
                NewAmount := BigNumber_Add(OldData.TotalDungeonsCompleted, InAmount)
                
                if:
                    set PlayerStatsDataMap[Player] = Player_Stats_Data:
                        MakePlayerStatsUpdate<constructor>(OldData)
                        TotalDungeonsCompleted := NewAmount

    UpdateLastPrestigeAmount(InAgent : agent, InAmount : big_number):void=
        if(Player := player[InAgent], Player.IsActive[]):
            if(OldData := PlayerStatsDataMap[Player]):
                if:
                    set PlayerStatsDataMap[Player] = Player_Stats_Data:
                        MakePlayerStatsUpdate<constructor>(OldData)
                        LastPrestigedAmount := InAmount
        
    SetCaveStrengthPercentage(InAgent : agent, Percentage : float):void=
        if(Player := player[InAgent], Player.IsActive[]):
            if(OldData := PlayerStatsDataMap[Player]):
                if:
                    set PlayerStatsDataMap[Player] = Player_Stats_Data:
                        MakePlayerStatsUpdate<constructor>(OldData)
                        CaveUnlockPercent := Percentage

    #Stone Vendor?
    <#UpdateGoldRatePerTrade(InAgent : agent, Algorithm : Algorithms, VendorIncreaseRate : float, VendorBaseRate : float, Resource : VendorUpgradeResourceTypes):void=
        if(Player := player[InAgent]):
            if(OldData := PlayerStatsDataMap[Player]):
                #Combines all Harvest Rate Upgrades [Will add additional upgrades when they exist]  
                if(Resource = VendorUpgradeResourceTypes.Stone):
                    TotalCoinRateUpgrades := OldData.StoneVendorUpgradeLevel 
                    CalculatedCoinGainOnTrade := Algorithm.CalculateHarvestRateOnHit(VendorIncreaseRate, TotalCoinRateUpgrades, VendorBaseRate, OldData.StoneVendorTierMultiplier)
                    if:
                        set PlayerStatsDataMap[Player] = Player_Stats_Data:
                            MakePlayerStatsUpdate<constructor>(OldData)
                            StoneOreRatePerTrade := CalculatedCoinGainOnTrade#>

    UpdatePlayerOreAmount(InAgent : agent, InResourceType : PropResourceType, InAmount : big_number):void=
        if(Player := player[InAgent], Player.IsActive[]):
            if(OldData := PlayerStatsDataMap[Player]):
                case(InResourceType):
                    PropResourceType.Stone =>
                        PlayerOreAmount := BigNumber_Add(OldData.PlayerStoneOreAmount, InAmount)
                        TotalStoneAmount := BigNumber_Add(OldData.TotalStoneOreAmount, InAmount)
                        TotalStoneSincePrestige := BigNumber_Add(OldData.TotalStoneSincePrestige, InAmount)
                        
                        if:
                            set PlayerStatsDataMap[Player] = Player_Stats_Data:
                                MakePlayerStatsUpdate<constructor>(OldData)
                                PlayerStoneOreAmount := PlayerOreAmount
                                TotalStoneOreAmount := TotalStoneAmount
                                TotalStoneSincePrestige := TotalStoneSincePrestige
                    PropResourceType.Copper =>
                        PlayerOreAmount := BigNumber_Add(OldData.PlayerCopperOreAmount, InAmount)
                        if:
                            set PlayerStatsDataMap[Player] = Player_Stats_Data:
                                MakePlayerStatsUpdate<constructor>(OldData)
                                PlayerCopperOreAmount := PlayerOreAmount
                    PropResourceType.Silver =>
                        PlayerOreAmount := BigNumber_Add(OldData.PlayerSilverOreAmount, InAmount)
                        if:
                            set PlayerStatsDataMap[Player] = Player_Stats_Data:
                                MakePlayerStatsUpdate<constructor>(OldData)
                                PlayerSilverOreAmount := PlayerOreAmount
                    PropResourceType.Malachite =>
                        PlayerOreAmount := BigNumber_Add(OldData.PlayerMalachiteOreAmount, InAmount)
                        if:
                            set PlayerStatsDataMap[Player] = Player_Stats_Data:
                                MakePlayerStatsUpdate<constructor>(OldData)
                                PlayerMalachiteOreAmount := PlayerOreAmount
                    PropResourceType.Spectrolite =>
                        PlayerOreAmount := BigNumber_Add(OldData.PlayerSpectroliteOreAmount, InAmount)
                        if:
                            set PlayerStatsDataMap[Player] = Player_Stats_Data:
                                MakePlayerStatsUpdate<constructor>(OldData)
                                PlayerSpectroliteOreAmount := PlayerOreAmount
                    PropResourceType.Obsidian =>
                        PlayerOreAmount := BigNumber_Add(OldData.PlayerObsidianOreAmount, InAmount)
                        if:
                            set PlayerStatsDataMap[Player] = Player_Stats_Data:
                                MakePlayerStatsUpdate<constructor>(OldData)
                                PlayerObsidianOreAmount := PlayerOreAmount
                    PropResourceType.PrestigeResource =>
                        PlayerOreAmount := BigNumber_Add(OldData.PlayerPrestigeResourceAmount, InAmount)
                        PrestigeLifetime := BigNumber_Add(OldData.LifetimePrestigeResourceAmount, InAmount) 
                        if:
                            set PlayerStatsDataMap[Player] = Player_Stats_Data:
                                MakePlayerStatsUpdate<constructor>(OldData)
                                PlayerPrestigeResourceAmount := PlayerOreAmount
                                LifetimePrestigeResourceAmount := PrestigeLifetime
                    PropResourceType.Orb =>
                        PlayerOrbAmount := BigNumber_Add(OldData.OrbAmount, InAmount)
                        if:
                            set PlayerStatsDataMap[Player] = Player_Stats_Data:
                                MakePlayerStatsUpdate<constructor>(OldData)
                                OrbAmount := PlayerOrbAmount
                    PropResourceType.Dust =>
                        PlayeDustAmount := BigNumber_Add(OldData.DustAmount, InAmount)
                        if:
                            set PlayerStatsDataMap[Player] = Player_Stats_Data:
                                MakePlayerStatsUpdate<constructor>(OldData)
                                DustAmount := PlayeDustAmount
                    _ =>
                        Print("###NULL OR NOT SET UP PROP RESOURCE TYPE###")
                    
                Print("Increase Ore")
    
    #This occurs on a fresh playthrough. We simply increase the level by one upon purchasing a Upgrade via the player remote for the 5 player harvesting upgrades.
    IncreaseRemoteUpgradeTierMultiplierOnPurchase(InAgent : agent, UpgradeIndex : int):void=
        if(Player := player[InAgent], Player.IsActive[]):
            if(OldData := PlayerStatsDataMap[Player]):
                if(NewTierMultiplier := OldData.HRUpgradeTierMultipliers[UpgradeIndex] + 1.0):
                    var TierMultiplierArray : []float = OldData.HRUpgradeTierMultipliers
                    if(set TierMultiplierArray[UpgradeIndex] = NewTierMultiplier):
                        if:
                            set PlayerStatsDataMap[Player] = Player_Stats_Data:
                                MakePlayerStatsUpdate<constructor>(OldData)
                                HRUpgradeTierMultipliers := TierMultiplierArray

    GetPlayerStoneOreAmount(InAgent : agent):big_number=
        if(Player := player[InAgent], Player.IsActive[]):
            if(OldData := PlayerStatsDataMap[Player]):
                return OldData.PlayerStoneOreAmount       
        return big_number{}

    UpdatePlayerUpgrade(InAgent : agent, UpgradeType : PlayerUpgradeType, UpgradeAmount : int):void=
        if(Player := player[InAgent], Player.IsActive[]):
            if(OldData := PlayerStatsDataMap[Player]):
                case(UpgradeType):
                    PlayerUpgradeType.CriticalChance =>
                        if:
                            set PlayerStatsDataMap[Player] = Player_Stats_Data:
                                MakePlayerStatsUpdate<constructor>(OldData)
                                CriticalChanceUpgrade := OldData.CriticalChanceUpgrade + UpgradeAmount
                    PlayerUpgradeType.CriticalDamage =>
                        if:
                            set PlayerStatsDataMap[Player] = Player_Stats_Data:
                                MakePlayerStatsUpdate<constructor>(OldData)
                                CriticalDamageUpgrade := OldData.CriticalDamageUpgrade + UpgradeAmount
                    PlayerUpgradeType.UnlockMegaCritical =>
                        if:
                            set PlayerStatsDataMap[Player] = Player_Stats_Data:
                                MakePlayerStatsUpdate<constructor>(OldData)
                                UnlockMegaCriticalUpgrade := OldData.UnlockMegaCriticalUpgrade + UpgradeAmount
                    PlayerUpgradeType.MegaCritical =>
                        if:
                            set PlayerStatsDataMap[Player] = Player_Stats_Data:
                                MakePlayerStatsUpdate<constructor>(OldData)
                                MegaCriticalUpgrade := OldData.MegaCriticalUpgrade + UpgradeAmount
                    PlayerUpgradeType.CriticalCharge =>
                        if:
                            set PlayerStatsDataMap[Player] = Player_Stats_Data:
                                MakePlayerStatsUpdate<constructor>(OldData)
                                CriticalChargeUpgrade := OldData.CriticalChargeUpgrade + UpgradeAmount
                    PlayerUpgradeType.FinalHit =>
                        if:
                            set PlayerStatsDataMap[Player] = Player_Stats_Data:
                                MakePlayerStatsUpdate<constructor>(OldData)
                                FinalHitUpgrade := OldData.FinalHitUpgrade + UpgradeAmount
                    PlayerUpgradeType.OreQuantity =>
                        if:
                            set PlayerStatsDataMap[Player] = Player_Stats_Data:
                                MakePlayerStatsUpdate<constructor>(OldData)
                                OreQuantityUpgrade := OldData.OreQuantityUpgrade + UpgradeAmount
                    PlayerUpgradeType.Friends =>
                        if:
                            set PlayerStatsDataMap[Player] = Player_Stats_Data:
                                MakePlayerStatsUpdate<constructor>(OldData)
                                FriendsUpgrade := OldData.FriendsUpgrade + UpgradeAmount
                    PlayerUpgradeType.Damage =>
                        if:
                            set PlayerStatsDataMap[Player] = Player_Stats_Data:
                                MakePlayerStatsUpdate<constructor>(OldData)
                                DamageUpgrade := OldData.DamageUpgrade + UpgradeAmount
                    PlayerUpgradeType.Movement =>
                        if:
                            set PlayerStatsDataMap[Player] = Player_Stats_Data:
                                MakePlayerStatsUpdate<constructor>(OldData)
                                MovementUpgrade := OldData.MovementUpgrade + UpgradeAmount
                    PlayerUpgradeType.HealthRegen =>
                        if:
                            set PlayerStatsDataMap[Player] = Player_Stats_Data:
                                MakePlayerStatsUpdate<constructor>(OldData)
                                HealthRegenUpgrade := OldData.HealthRegenUpgrade + UpgradeAmount
                    PlayerUpgradeType.MaxHealth =>
                        if:
                            set PlayerStatsDataMap[Player] = Player_Stats_Data:
                                MakePlayerStatsUpdate<constructor>(OldData)
                                MaxHealthUpgrade := OldData.MaxHealthUpgrade + UpgradeAmount
                    PlayerUpgradeType.ExplosiveHit =>
                        if:
                            set PlayerStatsDataMap[Player] = Player_Stats_Data:
                                MakePlayerStatsUpdate<constructor>(OldData)
                                ExplosiveHitUpgrade := OldData.ExplosiveHitUpgrade + UpgradeAmount
                    PlayerUpgradeType.ExplosiveHitRange =>
                        if:
                            set PlayerStatsDataMap[Player] = Player_Stats_Data:
                                MakePlayerStatsUpdate<constructor>(OldData)
                                ExplosiveHitRangeUpgrade := OldData.ExplosiveHitRangeUpgrade + UpgradeAmount
                    PlayerUpgradeType.ExplosiveHitDamage =>
                        if:
                            set PlayerStatsDataMap[Player] = Player_Stats_Data:
                                MakePlayerStatsUpdate<constructor>(OldData)
                                ExplosiveHitDamageUpgrade := OldData.ExplosiveHitDamageUpgrade + UpgradeAmount
                    PlayerUpgradeType.ExplosiveHitCharges =>
                        if:
                            set PlayerStatsDataMap[Player] = Player_Stats_Data:
                                MakePlayerStatsUpdate<constructor>(OldData)
                                ExplosiveHitChargesUpgrade := OldData.ExplosiveHitChargesUpgrade + UpgradeAmount
                    PlayerUpgradeType.Altar =>
                        if:
                            set PlayerStatsDataMap[Player] = Player_Stats_Data:
                                MakePlayerStatsUpdate<constructor>(OldData)
                                AltarUpgrade := OldData.AltarUpgrade + UpgradeAmount
                    PlayerUpgradeType.AltarStrength =>
                        if:
                            set PlayerStatsDataMap[Player] = Player_Stats_Data:
                                MakePlayerStatsUpdate<constructor>(OldData)
                                AltarStrengthUpgrade := OldData.AltarStrengthUpgrade + UpgradeAmount
                    PlayerUpgradeType.AltarTime =>
                        if:
                            set PlayerStatsDataMap[Player] = Player_Stats_Data:
                                MakePlayerStatsUpdate<constructor>(OldData)
                                AltarTimeUpgrade := OldData.AltarTimeUpgrade + UpgradeAmount
                    PlayerUpgradeType.GlobalVendor =>
                        if:
                            set PlayerStatsDataMap[Player] = Player_Stats_Data:
                                MakePlayerStatsUpdate<constructor>(OldData)
                                GlobalVendorUpgrade := OldData.GlobalVendorUpgrade + UpgradeAmount
                    PlayerUpgradeType.Prestige_AFK =>
                        if:
                            set PlayerStatsDataMap[Player] = Player_Stats_Data:
                                MakePlayerStatsUpdate<constructor>(OldData)
                                PrestigeAFKAreaUpgrade := OldData.PrestigeAFKAreaUpgrade + UpgradeAmount
                    PlayerUpgradeType.Prestige_Ascension =>
                        if:
                            set PlayerStatsDataMap[Player] = Player_Stats_Data:
                                MakePlayerStatsUpdate<constructor>(OldData)
                                PrestigeStrengthUpgrade := OldData.PrestigeStrengthUpgrade + UpgradeAmount
                    PlayerUpgradeType.Prestige_DungeonMastery =>
                        if:
                            set PlayerStatsDataMap[Player] = Player_Stats_Data:
                                MakePlayerStatsUpdate<constructor>(OldData)
                                PrestigeDungeonMasteryUpgrade := OldData.PrestigeDungeonMasteryUpgrade + UpgradeAmount

                    PlayerUpgradeType.Player_PrestigeCore =>
                        if:
                            set PlayerStatsDataMap[Player] = Player_Stats_Data:
                                MakePlayerStatsUpdate<constructor>(OldData)
                                PrestigeCore := OldData.PrestigeCore + UpgradeAmount

                    PlayerUpgradeType.Prestige_ArtifactMastery =>
                        if:
                            set PlayerStatsDataMap[Player] = Player_Stats_Data:
                                MakePlayerStatsUpdate<constructor>(OldData)
                                PrestigeArtifactMastery := OldData.PrestigeArtifactMastery + UpgradeAmount 
                    
                    PlayerUpgradeType.Prestige_ArtifactSlots =>
                        if:
                            set PlayerStatsDataMap[Player] = Player_Stats_Data:
                                MakePlayerStatsUpdate<constructor>(OldData)
                                PrestigeArtifactSlots := OldData.PrestigeArtifactSlots + UpgradeAmount 
                    #Automining
                    PlayerUpgradeType.Prestige_AutoMiningPlot =>
                        if:
                            set PlayerStatsDataMap[Player] = Player_Stats_Data:
                                MakePlayerStatsUpdate<constructor>(OldData)
                                PrestigeAutoMiningPlotClaimedUpgrade := OldData.PrestigeAutoMiningPlotClaimedUpgrade + UpgradeAmount

                    PlayerUpgradeType.Prestige_AutoMiningCount =>
                        if:
                            set PlayerStatsDataMap[Player] = Player_Stats_Data:
                                MakePlayerStatsUpdate<constructor>(OldData)
                                PrestigeAutoMiningCount := OldData.PrestigeAutoMiningCount + UpgradeAmount 

                    PlayerUpgradeType.Prestige_AutoMiningHitRate =>
                        if:
                            set PlayerStatsDataMap[Player] = Player_Stats_Data:
                                MakePlayerStatsUpdate<constructor>(OldData)
                                PrestigeAutoMiningHitRate := OldData.PrestigeAutoMiningHitRate + UpgradeAmount

                    #Dungeon King
                    PlayerUpgradeType.DungeonKingTokenAmount =>
                        if:
                            set PlayerStatsDataMap[Player] = Player_Stats_Data:
                                MakePlayerStatsUpdate<constructor>(OldData)
                                DungeonKingTokenSlotAmount := OldData.DungeonKingTokenSlotAmount + UpgradeAmount 

                    PlayerUpgradeType.DungeonKingTokenGenerationSpeed =>
                        if:
                            set PlayerStatsDataMap[Player] = Player_Stats_Data:
                                MakePlayerStatsUpdate<constructor>(OldData)
                                DungeonKingTokenSpeedAmount := OldData.DungeonKingTokenSpeedAmount + UpgradeAmount
                    _ =>
                        Print("###NULL OR NOT SET UP PLAYER UPGRADE###")
    
    GetPlayerUpgradeLevel(InAgent : agent, UpgradeType : PlayerUpgradeType):int=
        if(Player := player[InAgent], Player.IsActive[]):
            if(OldData := PlayerStatsDataMap[Player]):
                case(UpgradeType):
                    PlayerUpgradeType.CriticalChance =>
                        return OldData.CriticalChanceUpgrade
                    PlayerUpgradeType.CriticalDamage =>
                        return OldData.CriticalDamageUpgrade
                    PlayerUpgradeType.UnlockMegaCritical =>
                        return OldData.UnlockMegaCriticalUpgrade
                    PlayerUpgradeType.MegaCritical =>
                        return OldData.MegaCriticalUpgrade
                    PlayerUpgradeType.CriticalCharge =>
                        return OldData.CriticalChargeUpgrade
                    PlayerUpgradeType.FinalHit =>
                        return OldData.FinalHitUpgrade
                    PlayerUpgradeType.OreQuantity =>
                        return OldData.OreQuantityUpgrade
                    PlayerUpgradeType.Friends =>
                        return OldData.FriendsUpgrade
                    PlayerUpgradeType.Damage =>
                        return OldData.DamageUpgrade
                    PlayerUpgradeType.Movement =>
                        return OldData.MovementUpgrade
                    PlayerUpgradeType.HealthRegen =>
                        return OldData.HealthRegenUpgrade
                    PlayerUpgradeType.MaxHealth =>
                        return OldData.MaxHealthUpgrade
                    PlayerUpgradeType.ExplosiveHit =>
                        return OldData.ExplosiveHitUpgrade
                    PlayerUpgradeType.ExplosiveHitRange =>
                        return OldData.ExplosiveHitRangeUpgrade
                    PlayerUpgradeType.ExplosiveHitDamage =>
                        return OldData.ExplosiveHitDamageUpgrade
                    PlayerUpgradeType.ExplosiveHitCharges =>
                        return OldData.ExplosiveHitChargesUpgrade
                    PlayerUpgradeType.Altar =>
                        return OldData.AltarUpgrade
                    PlayerUpgradeType.AltarStrength =>
                        return OldData.AltarStrengthUpgrade
                    PlayerUpgradeType.AltarTime =>
                        return OldData.AltarTimeUpgrade
                    PlayerUpgradeType.GlobalVendor =>
                        return OldData.GlobalVendorUpgrade
                    #Prestige
                    PlayerUpgradeType.Prestige_AFK =>
                        return OldData.PrestigeAFKAreaUpgrade
                    PlayerUpgradeType.Prestige_Ascension =>
                        return OldData.PrestigeStrengthUpgrade
                    PlayerUpgradeType.Prestige_DungeonMastery =>
                        return OldData.PrestigeDungeonMasteryUpgrade
                    PlayerUpgradeType.Prestige_ArtifactSlots =>
                        return OldData.PrestigeArtifactSlots
                    PlayerUpgradeType.Prestige_ArtifactMastery =>
                        return OldData.PrestigeArtifactMastery
                    #Auto Mining
                    PlayerUpgradeType.Prestige_AutoMiningPlot =>
                        return OldData.PrestigeAutoMiningPlotClaimedUpgrade
                    PlayerUpgradeType.Prestige_AutoMiningCount =>
                        return OldData.PrestigeAutoMiningCount
                    PlayerUpgradeType.Prestige_AutoMiningHitRate =>
                        return OldData.PrestigeAutoMiningHitRate

                    #Dungeon King
                    PlayerUpgradeType.DungeonKingTokenAmount =>
                        return OldData.DungeonKingTokenSlotAmount
                    PlayerUpgradeType.DungeonKingTokenGenerationSpeed =>
                        return OldData.DungeonKingTokenSpeedAmount

                    _ =>
                        Print("###NULL OR NOT SET UP PLAYER UPGRADE###")
                        return 0
        Print("IF SEE THIS CHECK GetPlayerUpgradeLevel IN PLAYER STATS DATA")
        return 0

    GetPlayerUpgrade(InAgent : agent, UpgradeType : PlayerUpgradeType)<transacts>:player_upgrade=
        if(Player := player[InAgent], Player.IsActive[]):
            case(UpgradeType):
                PlayerUpgradeType.CriticalChance =>
                    return UpgradesData.CriticalChanceUpgrade
                PlayerUpgradeType.CriticalDamage =>
                    return UpgradesData.CriticalDamageUpgrade
                PlayerUpgradeType.UnlockMegaCritical =>
                    return UpgradesData.UnlockMegaCriticalUpgrade
                PlayerUpgradeType.MegaCritical =>
                    return UpgradesData.MegaCriticalUpgrade
                PlayerUpgradeType.CriticalCharge =>
                    return UpgradesData.CriticalChargeUpgrade
                PlayerUpgradeType.FinalHit =>
                    return UpgradesData.FinalHitUpgrade
                PlayerUpgradeType.OreQuantity =>
                    return UpgradesData.OreQuantityUpgrade
                PlayerUpgradeType.Friends =>
                    return UpgradesData.FriendsUpgrade
                PlayerUpgradeType.Damage =>
                    return UpgradesData.DamageUpgrade
                PlayerUpgradeType.Movement =>
                    return UpgradesData.MovementUpgrade
                PlayerUpgradeType.HealthRegen =>
                    return UpgradesData.HealthRegenUpgrade
                PlayerUpgradeType.MaxHealth =>
                    return UpgradesData.MaxHealthUpgrade
                PlayerUpgradeType.ExplosiveHit =>
                    return UpgradesData.ExplosiveHitUpgrade
                PlayerUpgradeType.ExplosiveHitRange =>
                    return UpgradesData.ExplosiveHitRangeUpgrade
                PlayerUpgradeType.ExplosiveHitDamage =>
                    return UpgradesData.ExplosiveHitDamageUpgrade
                PlayerUpgradeType.ExplosiveHitCharges =>
                    return UpgradesData.ExplosiveHitChargesUpgrade
                PlayerUpgradeType.Altar =>
                    return UpgradesData.AltarUpgrade
                PlayerUpgradeType.AltarStrength =>
                    return UpgradesData.AltarStrengthUpgrade
                PlayerUpgradeType.AltarTime =>
                    return UpgradesData.AltarTimeUpgrade
                PlayerUpgradeType.GlobalVendor =>
                    return UpgradesData.GlobalVendorUpgrade
                PlayerUpgradeType.Prestige_AFK =>
                    return UpgradesData.Prestige_AFKArea
                PlayerUpgradeType.Prestige_Ascension =>
                    return UpgradesData.Prestige_MiningAscension
                PlayerUpgradeType.Prestige_DungeonMastery =>
                    return UpgradesData.Prestige_DungeonMastery
                PlayerUpgradeType.Player_PrestigeCore =>
                    return UpgradesData.Prestige_Core
                PlayerUpgradeType.Prestige_ArtifactMastery =>
                    return UpgradesData.Prestige_ArtifactMastery
                PlayerUpgradeType.Prestige_ArtifactSlots =>
                    return UpgradesData.Prestige_ArtifactSlots

                #Automining
                PlayerUpgradeType.Prestige_AutoMiningPlot =>
                    return UpgradesData.Prestige_AutoMining
                PlayerUpgradeType.Prestige_AutoMiningCount =>
                    return UpgradesData.Prestige_AutoMiningCount
                PlayerUpgradeType.Prestige_AutoMiningHitRate =>
                    return UpgradesData.Prestige_AutoMiningHitRate
                #Dungeon King
                PlayerUpgradeType.DungeonKingTokenAmount =>
                    return UpgradesData.DungeonKingTokenAmount
                PlayerUpgradeType.DungeonKingTokenGenerationSpeed =>
                    return UpgradesData.DungeonKingTokenGenerationSpeed
                _ =>
                    Print("###NULL OR NOT SET UP PLAYER UPGRADE###")
                    return player_upgrade{}
        return player_upgrade{}

    #<<<Purely For Stats Debugging>>>
    GetPlayerOreUpgrade1Level(InAgent : agent):float=
        if(Player := player[InAgent], Player.IsActive[]):
            if(OldData := PlayerStatsDataMap[Player]):
                if(UpgradeLevel := OldData.HRUpgradeLevels[0]):
                    return UpgradeLevel
        return 0.0
    GetPlayerOreUpgrade2Level(InAgent : agent):float=
        if(Player := player[InAgent], Player.IsActive[]):
            if(OldData := PlayerStatsDataMap[Player]):
                if(UpgradeLevel := OldData.HRUpgradeLevels[1]):
                    return UpgradeLevel  
        return 0.0
    GetPlayerOreUpgrade3Level(InAgent : agent):float=
        if(Player := player[InAgent], Player.IsActive[]):
            if(OldData := PlayerStatsDataMap[Player]):
                if(UpgradeLevel := OldData.HRUpgradeLevels[2]):
                    return UpgradeLevel   
        return 0.0
    GetPlayerOreUpgrade4Level(InAgent : agent):float=
        if(Player := player[InAgent], Player.IsActive[]):
            if(OldData := PlayerStatsDataMap[Player]):
                if(UpgradeLevel := OldData.HRUpgradeLevels[3]):
                    return UpgradeLevel  
        return 0.0
    GetPlayerOreUpgrade5Level(InAgent : agent):float=
        if(Player := player[InAgent], Player.IsActive[]):
            if(OldData := PlayerStatsDataMap[Player]):
                if(UpgradeLevel := OldData.HRUpgradeLevels[4]):
                    return UpgradeLevel     
        return 0.0

    GetPlayerOreHarvestRatePerHit(InAgent : agent):big_number=
        if(Player := player[InAgent], Player.IsActive[]):
            if(OldData := PlayerStatsDataMap[Player]):
                return OldData.HarvestRatePerHit
        return big_number{}

    #GetPlayerCoinAmount(InAgent : agent):big_number=
        #if(Player := player[InAgent]):
            #if(OldData := PlayerStatsDataMap[Player]):
                #return OldData.PlayerCoinAmount       
        #return big_number{}

    GetPlayerUpgradePurchaseAmount(InAgent : agent):int=
        if(Player := player[InAgent], Player.IsActive[]):
            if(OldData := PlayerStatsDataMap[Player]):
                return OldData.UpgradePurchaseAmount       
        return 0
        
    GetStoneVendorUpgradeLevel(InAgent : agent):float=
        if(Player := player[InAgent], Player.IsActive[]):
            if(OldData := PlayerStatsDataMap[Player]):
                return OldData.StoneVendorUpgradeLevel       
        return 0.0

    GetStoneVendorCoinPerTrade(InAgent : agent):big_number=
        if(Player := player[InAgent], Player.IsActive[]):
            if(OldData := PlayerStatsDataMap[Player]):
                return OldData.StoneOreRatePerTrade       
        return big_number{}
    
    GetStoneVendorTierMultiplier(InAgent : agent):float=
        if(Player := player[InAgent], Player.IsActive[]):
            if(OldData := PlayerStatsDataMap[Player]):
                return OldData.StoneVendorTierMultiplier       
        return 0.0
        
    #<<<End of Debugging stats>>> 
    
    SpendPlayerOre(InAgent : agent, Cost : big_number, TradeAll : logic, InTradeResourceType : PropResourceType):void=
        if(Player := player[InAgent], Player.IsActive[]):
            if(OldData := PlayerStatsDataMap[Player]):
                    if(TradeAll = true):
                        #Print("Trade all Ore")
                        if(InTradeResourceType = PropResourceType.Stone):
                            if:
                                set PlayerStatsDataMap[Player] = Player_Stats_Data:
                                    MakePlayerStatsUpdate<constructor>(OldData)
                                    PlayerStoneOreAmount := big_number{}
                    if(TradeAll = false):
                        #Print("Don't trade all Ore")
                        if(InTradeResourceType = PropResourceType.Stone):
                            TradedAmount := BigNumber_Subtract(OldData.PlayerStoneOreAmount, Cost)
                            if:
                                set PlayerStatsDataMap[Player] = Player_Stats_Data:
                                    MakePlayerStatsUpdate<constructor>(OldData)
                                    PlayerStoneOreAmount := TradedAmount
                        if(InTradeResourceType = PropResourceType.Copper):
                            TradedAmount := BigNumber_Subtract(OldData.PlayerCopperOreAmount, Cost)
                            if:
                                set PlayerStatsDataMap[Player] = Player_Stats_Data:
                                    MakePlayerStatsUpdate<constructor>(OldData)
                                    PlayerCopperOreAmount := TradedAmount
                        if(InTradeResourceType = PropResourceType.Silver):
                            TradedAmount := BigNumber_Subtract(OldData.PlayerSilverOreAmount, Cost)
                            if:
                                set PlayerStatsDataMap[Player] = Player_Stats_Data:
                                    MakePlayerStatsUpdate<constructor>(OldData)
                                    PlayerSilverOreAmount := TradedAmount
                        if(InTradeResourceType = PropResourceType.Malachite):
                            TradedAmount := BigNumber_Subtract(OldData.PlayerMalachiteOreAmount, Cost)
                            if:
                                set PlayerStatsDataMap[Player] = Player_Stats_Data:
                                    MakePlayerStatsUpdate<constructor>(OldData)
                                    PlayerMalachiteOreAmount := TradedAmount
                        if(InTradeResourceType = PropResourceType.Spectrolite):
                            TradedAmount := BigNumber_Subtract(OldData.PlayerSpectroliteOreAmount, Cost)
                            if:
                                set PlayerStatsDataMap[Player] = Player_Stats_Data:
                                    MakePlayerStatsUpdate<constructor>(OldData)
                                    PlayerSpectroliteOreAmount := TradedAmount
                        if(InTradeResourceType = PropResourceType.Obsidian):
                            TradedAmount := BigNumber_Subtract(OldData.PlayerObsidianOreAmount, Cost)
                            if:
                                set PlayerStatsDataMap[Player] = Player_Stats_Data:
                                    MakePlayerStatsUpdate<constructor>(OldData)
                                    PlayerObsidianOreAmount := TradedAmount
                        if(InTradeResourceType = PropResourceType.PrestigeResource):
                            TradedAmount := BigNumber_Subtract(OldData.PlayerPrestigeResourceAmount, Cost)
                            if:
                                set PlayerStatsDataMap[Player] = Player_Stats_Data:
                                    MakePlayerStatsUpdate<constructor>(OldData)
                                    PlayerPrestigeResourceAmount := TradedAmount
                        if(InTradeResourceType = PropResourceType.Orb):
                            TradedAmount := BigNumber_Subtract(OldData.OrbAmount, Cost)
                            if:
                                set PlayerStatsDataMap[Player] = Player_Stats_Data:
                                    MakePlayerStatsUpdate<constructor>(OldData)
                                    OrbAmount := TradedAmount
                        if(InTradeResourceType = PropResourceType.Dust):
                            TradedAmount := BigNumber_Subtract(OldData.DustAmount, Cost)
                            if:
                                set PlayerStatsDataMap[Player] = Player_Stats_Data:
                                    MakePlayerStatsUpdate<constructor>(OldData)
                                    DustAmount := TradedAmount

    # Get a given players ore amount for input resource
    GetPlayerOreAmount(InAgent : agent, InResourceType : PropResourceType):big_number=
        if(Player := player[InAgent], Player.IsActive[]):
            if(OldData := PlayerStatsDataMap[Player]):
                case(InResourceType):
                    PropResourceType.Stone =>
                        return OldData.PlayerStoneOreAmount
                    PropResourceType.Copper =>
                        return OldData.PlayerCopperOreAmount
                    PropResourceType.Silver =>
                        return OldData.PlayerSilverOreAmount
                    PropResourceType.Malachite =>
                        return OldData.PlayerMalachiteOreAmount
                    PropResourceType.Spectrolite =>
                        return OldData.PlayerSpectroliteOreAmount
                    PropResourceType.Obsidian =>
                        return OldData.PlayerObsidianOreAmount
                    PropResourceType.PrestigeResource =>
                        return OldData.PlayerPrestigeResourceAmount
                    PropResourceType.Orb =>
                        return OldData.OrbAmount
                    PropResourceType.Dust =>
                        return OldData.DustAmount
                    _ =>
                        Print("###NULL OR NOT SET UP PROP RESOURCE TYPE###")
        return big_number{}

    #This next block of code are all tier related systems. 
    #Since the game has multiple tier systems we need multiple update functions
    UpdateHarvestRateTierSystem(InAgent : agent, UpgradeIndex : int, Ascended : logic):void=
        if(Player := player[InAgent], Player.IsActive[]):
            if(OldData := PlayerStatsDataMap[Player]):
                if(Ascended = false):
                    if(NewTierMultiplier := OldData.HRUpgradeTierMultipliers[UpgradeIndex] * 2.0):
                        var TierMultiplierArray : []float = OldData.HRUpgradeTierMultipliers
                        if(set TierMultiplierArray[UpgradeIndex] = NewTierMultiplier):
                            if:
                                set PlayerStatsDataMap[Player] = Player_Stats_Data:
                                    MakePlayerStatsUpdate<constructor>(OldData)
                                    HRUpgradeTierMultipliers := TierMultiplierArray
                else:
                    if(NewTierMultiplier := OldData.HRUpgradeTierMultipliers[UpgradeIndex] * 3.0):
                        var TierMultiplierArray : []float = OldData.HRUpgradeTierMultipliers
                        if(set TierMultiplierArray[UpgradeIndex] = NewTierMultiplier):
                            if:
                                set PlayerStatsDataMap[Player] = Player_Stats_Data:
                                    MakePlayerStatsUpdate<constructor>(OldData)
                                    HRUpgradeTierMultipliers := TierMultiplierArray

    UpdateVendorTierMultiplier(InAgent : agent, InTradeResourceType : PropResourceType):void=
        if(Player := player[InAgent], Player.IsActive[]):
            if(OldData := PlayerStatsDataMap[Player]):
                if(InTradeResourceType = PropResourceType.Stone):
                    if:
                        set PlayerStatsDataMap[Player] = Player_Stats_Data:
                            MakePlayerStatsUpdate<constructor>(OldData)      
                            StoneVendorTierMultiplier := OldData.StoneVendorTierMultiplier * 2.0

    #This is called from Algorithms and functions.
    #The players gold amount becomes equal to the gold earned which is a float parameter we pass into this function.
    #Increase our Coin amount by adding coin rate per trade 
    <#UpdatePlayerCoinAmount(InAgent : agent, InTradeResourceType : PropResourceType):void=
        if(Player := player[InAgent]):
            if(OldData := PlayerStatsDataMap[Player]):
                #TotalTradeInAmount := BigNumber_Multiply(OldData.StoneOreRatePerTrade, OldData.PlayerStoneOreAmount)
                #OldData.StoneOreRatePerTrade * OldData.PlayerStoneOreAmount
                if(InTradeResourceType = PropResourceType.Stone):
                    TotalTradeInAmount := BigNumber_Multiply(OldData.StoneOreRatePerTrade, OldData.PlayerStoneOreAmount)
                    NewCoinAmount := BigNumber_Add(TotalTradeInAmount, OldData.PlayerCoinAmount )
                    if:
                        set PlayerStatsDataMap[Player] = Player_Stats_Data : 
                            MakePlayerStatsUpdate<constructor>(OldData)
                            PlayerCoinAmount := NewCoinAmount
                if(InTradeResourceType = PropResourceType.Silver):
                    TotalTradeInAmount := BigNumber_Multiply(OldData.SilverOreRatePerTrade, OldData.PlayerSilverOreAmount)
                    NewCoinAmount := BigNumber_Add(TotalTradeInAmount, OldData.PlayerCoinAmount )
                    if:
                        set PlayerStatsDataMap[Player] = Player_Stats_Data : 
                            MakePlayerStatsUpdate<constructor>(OldData) 
                            PlayerCoinAmount := NewCoinAmount
                if(InTradeResourceType = PropResourceType.Gold):
                    TotalTradeInAmount := BigNumber_Multiply(OldData.GoldOreRatePerTrade, OldData.PlayerGoldOreAmount)
                    NewCoinAmount := BigNumber_Add(TotalTradeInAmount, OldData.PlayerCoinAmount )
                    if:
                        set PlayerStatsDataMap[Player] = Player_Stats_Data : 
                            MakePlayerStatsUpdate<constructor>(OldData) 
                            PlayerCoinAmount := NewCoinAmount
                if(InTradeResourceType = PropResourceType.Diamond):
                    TotalTradeInAmount := BigNumber_Multiply(OldData.DiamondOreRatePerTrade, OldData.PlayerDiamondOreAmount)
                    NewCoinAmount := BigNumber_Add(TotalTradeInAmount, OldData.PlayerCoinAmount )
                    if:
                        set PlayerStatsDataMap[Player] = Player_Stats_Data : 
                            MakePlayerStatsUpdate<constructor>(OldData) 
                            PlayerCoinAmount := NewCoinAmount#>

    #This is purely for granting coins and testing big numbers currently
    <#GivePlayersCoins(InAgent : agent, CoinsToGive : big_number):void=
        if(Player := player[InAgent]):
            if(OldData := PlayerStatsDataMap[Player]):
                NewCoinAmount := BigNumber_Add(OldData.PlayerCoinAmount, CoinsToGive)
                if:
                    set PlayerStatsDataMap[Player] = Player_Stats_Data : 
                        MakePlayerStatsUpdate<constructor>(OldData) 
                        PlayerCoinAmount:= NewCoinAmount#>

    GivePlayerStone(InAgent : agent, StoneToGive : big_number):void=
        if(Player := player[InAgent], Player.IsActive[]):
            if(OldData := PlayerStatsDataMap[Player]):
                PlayerOreAmount := BigNumber_Add(OldData.PlayerStoneOreAmount, StoneToGive)
                TotalStoneAmount := BigNumber_Add(OldData.TotalStoneOreAmount, StoneToGive)
                if:
                    set PlayerStatsDataMap[Player] = Player_Stats_Data:
                        MakePlayerStatsUpdate<constructor>(OldData)
                        PlayerStoneOreAmount := PlayerOreAmount
                        TotalStoneOreAmount := TotalStoneAmount

    # This is called to add the completed quest to the player's data
    AddCompletedQuest(InAgent : agent, InCompletedQuest : string):void=
        if(Player := player[InAgent], Player.IsActive[]):
            if(OldData := PlayerStatsDataMap[Player]):
                if:
                    set PlayerStatsDataMap[Player] = Player_Stats_Data :
                        MakePlayerStatsUpdate<constructor>(OldData) 
                        CompletedQuests := OldData.CompletedQuests + array{InCompletedQuest}

    # This is called to add the completed quest line to the player's data
    AddCompletedQuestLine(InAgent : agent, InCompletedQuestLine : string):void=
        if(Player := player[InAgent], Player.IsActive[]):
            if(OldData := PlayerStatsDataMap[Player]):
                if:
                    set PlayerStatsDataMap[Player] = Player_Stats_Data :
                        MakePlayerStatsUpdate<constructor>(OldData) 
                        CompletedQuestLines := OldData.CompletedQuestLines + array{InCompletedQuestLine}

    # This is called to change the current quest for the player
    ChangeCurrentQuest(InAgent : agent, InCurrentQuest : string):void=
        if(Player := player[InAgent], Player.IsActive[]):
            if(OldData := PlayerStatsDataMap[Player]):
                if:
                    set PlayerStatsDataMap[Player] = Player_Stats_Data :
                        MakePlayerStatsUpdate<constructor>(OldData) 
                        CurrentQuest := InCurrentQuest

    # This is called to change the current quest increment for the player
    ChangeCurrentQuestIncrement(InAgent : agent, InCurrentQuestIncrement : big_number):void=
        if(Player := player[InAgent], Player.IsActive[]):
            if(OldData := PlayerStatsDataMap[Player]):
                if:
                    set PlayerStatsDataMap[Player] = Player_Stats_Data :
                        MakePlayerStatsUpdate<constructor>(OldData)
                        CurrentQuestIncrement := InCurrentQuestIncrement

    UpdatePlayerUpgradePurchaseAmount(InAgent : agent, InPurchaseAmount : int):void=
        if(Player := player[InAgent], Player.IsActive[]):
            if(OldData := PlayerStatsDataMap[Player]):
                if:
                    set PlayerStatsDataMap[Player] = Player_Stats_Data :
                        MakePlayerStatsUpdate<constructor>(OldData)
                        UpgradePurchaseAmount := InPurchaseAmount

    GetPlayerHitWithCritical(InAgent : agent, ConfirmCrit : logic, AltarBuff : logic, GnomeBuff : logic):tuple(big_number, logic)=
        if(Player := player[InAgent], Player.IsActive[]):
            var PlayerAltarBuffAmount : big_number = big_number{}

            var ArtifactStrengthBonus : float = 0.0
            var ArtifactCriticalChanceBonus : float = 0.0
            var ArtifactCriticalDamageBonus : float = 0.0

            var SwordStrengthBonus : float = 0.0

            var OutAmount : big_number = big_number{}

            if(PlayerData := PlayerStatsDataMap[Player]):
                # Set artifact buffs
                AchievementMultipler := PlayerData.AchievementTotalStoneBonus + PlayerData.AchievementTotalDungeonsCompletedBonus + PlayerData.AchievementHighestStrengthBonus + PlayerData.AchievementTotalRocksDestroyedBonus + PlayerData.AchievementTotalArtifactsSmeltedBonus

                for(EquippedArtifact : PlayerData.EquippedPlayerArtifacts):
                    Print("EQUIPPED ARTIFACT FOUND")
                    set ArtifactStrengthBonus += GetArtifactStat(EquippedArtifact, ArtifactStats.ArtifactBonus)
                    #CARD - Update 1.9 - Check for cardID 0 to do bonus damage to boss
                    var CardIDToCheck : int = -1

                    if(EquippedArtifact.Rarity = ArtifactRarity.MushroomCommon or EquippedArtifact.Rarity = ArtifactRarity.LavaCommon or EquippedArtifact.Rarity = ArtifactRarity.UnderworldCommon):
                        set CardIDToCheck = 2
                    else if(EquippedArtifact.Rarity = ArtifactRarity.MushroomUncommon or EquippedArtifact.Rarity = ArtifactRarity.LavaUncommon or EquippedArtifact.Rarity = ArtifactRarity.UnderworldUncommon):
                        set CardIDToCheck = 12
                    else if(EquippedArtifact.Rarity = ArtifactRarity.MushroomRare or EquippedArtifact.Rarity = ArtifactRarity.LavaRare or EquippedArtifact.Rarity = ArtifactRarity.UnderworldRare):
                        set CardIDToCheck = 13
                    else if(EquippedArtifact.Rarity = ArtifactRarity.MushroomEpic or EquippedArtifact.Rarity = ArtifactRarity.LavaEpic or EquippedArtifact.Rarity = ArtifactRarity.UnderworldEpic):
                        set CardIDToCheck = 14
                    else if(EquippedArtifact.Rarity = ArtifactRarity.MushroomLegendary or EquippedArtifact.Rarity = ArtifactRarity.LavaLegendary or EquippedArtifact.Rarity = ArtifactRarity.UnderworldLegendary):
                        set CardIDToCheck = 15

                    if(FoundCard := SearchForCard(CardIDToCheck, Player, Self)?):
                        set ArtifactStrengthBonus += GetArtifactStat(EquippedArtifact, ArtifactStats.ArtifactBonus) * (GetCardBonus(CardIDToCheck, FoundCard.Level) / 100.0)
                    
                    Print("Artifact Bonus Percent : {ArtifactStrengthBonus}")
                    set ArtifactCriticalChanceBonus += GetArtifactStat(EquippedArtifact, ArtifactStats.CriticalChance)
                    set ArtifactCriticalDamageBonus += GetArtifactStat(EquippedArtifact, ArtifactStats.CriticalDamage)

                if(Sword := PlayerData.OwnedPlayerSword[0], Bonus := Sword.BigNumber_StatValues[3]):
                    set SwordStrengthBonus = BigNumber_ConvertBigNumberIntoFloat(Bonus)

                var HarvestRateCard : big_number = PlayerData.HarvestRatePerHit

                if(FoundCard := SearchForCard(8, Player, Self)?):
                    BonusAmount := GetCardBonus(8, FoundCard.Level)
                    set HarvestRateCard = BigNumber_Add(PlayerData.HarvestRatePerHit, BigNumber_MultiplyWithAFloat(PlayerData.HarvestRatePerHit, (BonusAmount / 100.0)))
                
                HarvestRateArtifactBuff := BigNumber_Add(HarvestRateCard, BigNumber_MultiplyWithAFloat(HarvestRateCard, (ArtifactStrengthBonus / 100.0)))
                HarvestRateSword := BigNumber_Add(HarvestRateArtifactBuff, BigNumber_MultiplyWithAFloat(HarvestRateArtifactBuff, (SwordStrengthBonus / 100.0)))
                CaveUnlockPercentage := BigNumber_Add(HarvestRateSword ,BigNumber_MultiplyWithAFloat(HarvestRateSword, (PlayerData.CaveUnlockPercent/100.0)))
                AchievementUnlockPercentage := BigNumber_Add(CaveUnlockPercentage ,BigNumber_MultiplyWithAFloat(CaveUnlockPercentage, (AchievementMultipler/100.0)))

                var HarvestRateHatBoost : big_number = AchievementUnlockPercentage
                if(PlayerData.HatID <> -1):
                    set HarvestRateHatBoost = BigNumber_Add(CaveUnlockPercentage ,BigNumber_MultiplyWithAFloat(AchievementUnlockPercentage, (5.0/100.0)))

                HarvestRatePrestige := BigNumber_Add(HarvestRateHatBoost, BigNumber_MultiplyWithAFloat(HarvestRateHatBoost, PlayerData.PrestigeBonusHarvestRate / 100.0))

                set OutAmount = HarvestRatePrestige

                # Get Player Critical Chance
                if(AltarBuff = true):
                    var PlayerAltarBuffStrength : float = UpgradesData.GetAltarStrengthAmount(GetPlayerUpgradeLevel(InAgent, PlayerUpgradeType.AltarStrength))
                    if(FoundCard := SearchForCard(16, Player, Self)?):
                        set PlayerAltarBuffStrength += GetCardBonus(16, FoundCard.Level)
                    BNPlayerAltarBuffStrength := BigNumber_ConvertFloatIntoBigNumber(PlayerAltarBuffStrength / 100.0)
                    set PlayerAltarBuffAmount = BigNumber_Multiply(OutAmount, BNPlayerAltarBuffStrength)
                    set OutAmount =  BigNumber_Add(OutAmount, PlayerAltarBuffAmount)
                if(GnomeBuff = true):
                    BNPlayerGnomeBuffStrength := BigNumber_ConvertFloatIntoBigNumber(1000.0 / 100.0)
                    PlayerGnomeBuffAmount := BigNumber_Multiply(OutAmount, BNPlayerGnomeBuffStrength)
                    set OutAmount =  BigNumber_Add(OutAmount, PlayerGnomeBuffAmount)

                var PlayerCriticalChance : float = UpgradesData.GetCriticalChanceAmount(GetPlayerUpgradeLevel(InAgent, PlayerUpgradeType.CriticalChance))

                if(FoundCard := SearchForCard(6, Player, Self)?):
                    BonusAmount := GetCardBonus(6, FoundCard.Level)
                    set PlayerCriticalChance += PlayerCriticalChance * BonusAmount / 100.0

                PlayerArtifactCriticalChance := PlayerCriticalChance * (ArtifactCriticalChanceBonus / 100.0)
                TotalCalculatedCriticalChance := PlayerCriticalChance + PlayerArtifactCriticalChance
                if(IntPlayerCriticalChance := Int[TotalCalculatedCriticalChance]):
                    Print("CRIT CHANCE : {IntPlayerCriticalChance}")
                    ## USING MEGA CRIT
                    if(ConfirmCrit = true):
                        PlayerCriticalDamage :=  UpgradesData.GetMegaCriticalAmount(GetPlayerUpgradeLevel(InAgent, PlayerUpgradeType.MegaCritical)) / 100.0
                        BNPlayerCriticalDamage := BigNumber_ConvertFloatIntoBigNumber(PlayerCriticalDamage + (PlayerCriticalDamage * (ArtifactCriticalDamageBonus / 100.0)))
                        HarvestRateAltarBuffed := BigNumber_Add(PlayerData.HarvestRatePerHit, PlayerAltarBuffAmount)
                        TotalHitAmount := BigNumber_Add(OutAmount, BigNumber_Multiply(OutAmount, BNPlayerCriticalDamage))
                        HitAmount := TotalHitAmount
                        return (HitAmount, true)
                    ## USING NORMAL CRIT
                    RandomChance := GetRandomInt(1, 100)
                    if(RandomChance <= IntPlayerCriticalChance):
                        # Get Player Damage and Multiply by the player's critical damage
                        var PlayerCriticalDamage : float =  UpgradesData.GetCriticalDamageAmount(GetPlayerUpgradeLevel(InAgent, PlayerUpgradeType.CriticalDamage)) / 100.0

                        if(FoundCard := SearchForCard(7, Player, Self)?):
                            BonusAmount := GetCardBonus(7, FoundCard.Level)
                            set PlayerCriticalDamage += PlayerCriticalDamage * BonusAmount / 100.0

                        BNPlayerCriticalDamage := BigNumber_ConvertFloatIntoBigNumber(PlayerCriticalDamage + (PlayerCriticalDamage * (ArtifactCriticalDamageBonus / 100.0)))
                        
                        TotalHitAmount := BigNumber_Add(OutAmount, BigNumber_Multiply(OutAmount, BNPlayerCriticalDamage))
                        HitAmount := TotalHitAmount
                        return (HitAmount, true)
                    ## USING NO CRIT
                    else:
                        return (OutAmount, false)

        # SHOULD NEVER REACH THIS POINT
        Print("IF YOU SEE THIS IN THE LOG, CHECK CRITICAL HIT CODE")
        return (big_number{}, false)

    UpdateAlterMaxAmount(InAgent : agent):void=
        if(Player := player[InAgent], Player.IsActive[]):
            if(OldData := PlayerStatsDataMap[Player]):
                Amount := OldData.AlterMaxAmount
                if:
                    set PlayerStatsDataMap[Player] = Player_Stats_Data:
                        MakePlayerStatsUpdate<constructor>(OldData)
                        AlterMaxAmount := Amount + 1

    UpdateDungeonMaxAmount(InAgent : agent):void=
        if(Player := player[InAgent], Player.IsActive[]):
            if(OldData := PlayerStatsDataMap[Player]):
                Amount := OldData.DungeonMaxAmount
                if:
                    set PlayerStatsDataMap[Player] = Player_Stats_Data:
                        MakePlayerStatsUpdate<constructor>(OldData)
                        DungeonMaxAmount := Amount + 1

    UpdateAlterTimer(InAgent : agent, Time : float):void=
        if(Player := player[InAgent], Player.IsActive[]):
            if(OldData := PlayerStatsDataMap[Player]):
                if:
                    set PlayerStatsDataMap[Player] = Player_Stats_Data:
                        MakePlayerStatsUpdate<constructor>(OldData)
                        AlterTimer := Time
    
    UpdateDungeonTimer(InAgent : agent, Time : float):void=
        if(Player := player[InAgent], Player.IsActive[]):
            if(OldData := PlayerStatsDataMap[Player]):
                if:
                    set PlayerStatsDataMap[Player] = Player_Stats_Data:
                        MakePlayerStatsUpdate<constructor>(OldData)
                        DungeonTimer := Time

    SetArtifacts(InAgent : agent, InArtifacts : []Artifact_Data):void=
        if(Player := player[InAgent], Player.IsActive[]):
            if(OldData := PlayerStatsDataMap[Player]):
                if:
                    set PlayerStatsDataMap[Player] = Player_Stats_Data:
                        MakePlayerStatsUpdate<constructor>(OldData)
                        OwnedPlayerArtifacts := InArtifacts

    GiveArtifact(InAgent : agent, InArtifact : Artifact_Data, InEquip : logic):void=
        if(Player := player[InAgent], Player.IsActive[]):
            if(OldData := PlayerStatsDataMap[Player]):
                var NewArtifacts : []Artifact_Data = OldData.OwnedPlayerArtifacts
                var NewEquippedArtifacts : []Artifact_Data = OldData.EquippedPlayerArtifacts

                if(InEquip = true):
                    for(Index -> CurrentArtifact : NewArtifacts):
                        if(CurrentArtifact.Rarity = InArtifact.Rarity):
                            if(set NewArtifacts[Index] = InArtifact):
                                if:
                                    set PlayerStatsDataMap[Player] = Player_Stats_Data:
                                        MakePlayerStatsUpdate<constructor>(OldData)
                                        OwnedPlayerArtifacts := NewArtifacts

                                return

                for(Index -> CurrentArtifact : OldData.OwnedPlayerArtifacts):
                    if(CurrentArtifact.Rarity = InArtifact.Rarity):
                        var NewStatValues : []float = CurrentArtifact.StatValues
                        if(CurrentArtifact.Prestige < InArtifact.Prestige):
                            if(set NewArtifacts[Index] = InArtifact):
                                Print("Update Owned Artifact Prestige")
                                if:
                                    set PlayerStatsDataMap[Player] = Player_Stats_Data:
                                        MakePlayerStatsUpdate<constructor>(OldData)
                                        OwnedPlayerArtifacts := NewArtifacts
                                then:
                                    for(EquippedIndex -> EquippedArtifact : NewEquippedArtifacts):
                                        if(EquippedArtifact.Rarity = InArtifact.Rarity):
                                            if(NewData := PlayerStatsDataMap[Player]):
                                                NewArtifactEquipped := Artifact_Data:
                                                    MakeArtifact<constructor>(InArtifact)
                                                    EquippedSlot := EquippedArtifact.EquippedSlot
                                                Print("Update Equipped Artifact Prestige")
                                                Print("Index : {EquippedIndex} | Rarity : {GetArtifactRarityString(EquippedArtifact.Rarity)}")
                                                if(set NewEquippedArtifacts[EquippedIndex] = NewArtifactEquipped):
                                                    if:
                                                        set PlayerStatsDataMap[Player] = Player_Stats_Data:
                                                            MakePlayerStatsUpdate<constructor>(NewData)
                                                            EquippedPlayerArtifacts := NewEquippedArtifacts
    
                                #DebugPrintArtifactStats(InArtifact)
                                return

                        if(CurrentArtifact.Level = 30):
                            return

                        Print("PLAYER ALREADY OWNS THAT RARITY LEVEL IT UP")
                        for(StatIndex -> CurrentValue : InArtifact.StatValues):
                            if(set NewStatValues[StatIndex] += (CurrentValue * CurrentArtifact.IncreaseRate * 1.1) - CurrentValue):
                            
                        TempArtifact := Artifact_Data:
                            MakeArtifact<constructor>(CurrentArtifact)
                            Level := CurrentArtifact.Level + 1
                            StatValues := NewStatValues

                        if(set NewArtifacts[Index] = TempArtifact):
                            Print("Update Owned Artifact Stats")
                            if:
                                set PlayerStatsDataMap[Player] = Player_Stats_Data:
                                    MakePlayerStatsUpdate<constructor>(OldData)
                                    OwnedPlayerArtifacts := NewArtifacts
                            then:
                                for(EquippedIndex -> EquippedArtifact : OldData.EquippedPlayerArtifacts):
                                    # Update 1.5 - Added rarity check as a saftey
                                    if(EquippedArtifact.Rarity = TempArtifact.Rarity):
                                        if(NewData := PlayerStatsDataMap[Player]):
                                            NewArtifactEquipped := Artifact_Data:
                                                MakeArtifact<constructor>(TempArtifact)
                                                EquippedSlot := EquippedArtifact.EquippedSlot
                                            Print("Update Equipped Artifact Stats")
                                            Print("Index : {EquippedIndex} | Rarity : {GetArtifactRarityString(EquippedArtifact.Rarity)}")
                                            if(set NewEquippedArtifacts[EquippedIndex] = NewArtifactEquipped):
                                                if:
                                                    set PlayerStatsDataMap[Player] = Player_Stats_Data:
                                                        MakePlayerStatsUpdate<constructor>(NewData)
                                                        EquippedPlayerArtifacts := NewEquippedArtifacts
                                            

                            #DebugPrintArtifactStats(TempArtifact)
                            return
                
                set NewArtifacts += array{InArtifact}
                Print("ARTIFACT NOT FOUND SET OWNED")
                if:
                    set PlayerStatsDataMap[Player] = Player_Stats_Data:
                        MakePlayerStatsUpdate<constructor>(OldData)
                        OwnedPlayerArtifacts := NewArtifacts

    EquipArtifact(InAgent : agent, InArtifact : Artifact_Data, InSlot : int):void=
        if(Player := player[InAgent], Player.IsActive[]):
            if(OldData := PlayerStatsDataMap[Player]):
                
                var NewArtifacts : []Artifact_Data = OldData.EquippedPlayerArtifacts
                NewArtifact := Artifact_Data:
                    MakeArtifact<constructor>(InArtifact)
                    EquippedSlot := InSlot

                Print("Slot {InSlot} : Rarity {GetArtifactRarityString(NewArtifact.Rarity)} : EquippedSlot {NewArtifact.EquippedSlot}")
                for(Index -> CurrentArtifact : OldData.EquippedPlayerArtifacts):
                    #DebugPrintArtifactStats(CurrentArtifact)
                    if(CurrentArtifact.Rarity = InArtifact.Rarity):
                        Print("Already Equipped")
                        return
                    if(CurrentArtifact.EquippedSlot = InSlot):
                        # Need to remove current slot artifact
                        if(set NewArtifacts[Index] = NewArtifact):
                            if:
                                set PlayerStatsDataMap[Player] = Player_Stats_Data:
                                    MakePlayerStatsUpdate<constructor>(OldData)
                                    EquippedPlayerArtifacts := NewArtifacts
                            then:
                                return

                set NewArtifacts += array{NewArtifact}

                if:
                    set PlayerStatsDataMap[Player] = Player_Stats_Data:
                        MakePlayerStatsUpdate<constructor>(OldData)
                        EquippedPlayerArtifacts := NewArtifacts

    # Update 1.5 - Added to make sure the players current equipped artifact get cleared if their version is 0
    ClearEquippedArtifacts(InAgent : agent):void=
        if(Player := player[InAgent], Player.IsActive[]):
            var NewArtifacts : []Artifact_Data = array{}
            if(OldData := PlayerStatsDataMap[Player]):
                if:
                    set PlayerStatsDataMap[Player] = Player_Stats_Data:
                        MakePlayerStatsUpdate<constructor>(OldData)
                        EquippedPlayerArtifacts := array{}
                        
                for(Artifact : OldData.OwnedPlayerArtifacts):
                    TempArifact := Artifact_Data:
                        MakeArtifact<constructor>(Artifact)
                        EquippedSlot := -1

                    set NewArtifacts += array{TempArifact}

                if(NewData := PlayerStatsDataMap[Player]):
                    if:
                        set PlayerStatsDataMap[Player] = Player_Stats_Data:
                            MakePlayerStatsUpdate<constructor>(NewData)
                            OwnedPlayerArtifacts := NewArtifacts

    GiveSword(InAgent : agent, InSword : Sword_Data):void=
        if(Player := player[InAgent], Player.IsActive[]):
            if(OldData := PlayerStatsDataMap[Player]):
                if:
                    set PlayerStatsDataMap[Player] = Player_Stats_Data:
                        MakePlayerStatsUpdate<constructor>(OldData)
                        OwnedPlayerSword := array{InSword}

    UpgradeSword(InAgent : agent, InSwordRarity : SwordRarity):void=
        if(Player := player[InAgent], Player.IsActive[]):
            if(OldData := PlayerStatsDataMap[Player]):
                var NewSwords : []Sword_Data = OldData.OwnedPlayerSword

                for(Index -> CurrentSword : NewSwords):
                    if(InSwordRarity = CurrentSword.Rarity):
                        if(CurrentSword.Level = 10):
                            return
                            
                        var NewStatValues : []big_number = CurrentSword.BigNumber_StatValues

                        PresetSword := GetSwordPresetData(InSwordRarity)

                        if(NewStatValue1 := NewStatValues[0], PresetSwordStat := PresetSword.BigNumber_StatValues[0]):
                            Stat :=  BigNumber_Add(NewStatValue1, BigNumber_Subtract(BigNumber_MultiplyWithAFloat(PresetSwordStat, CurrentSword.StatIncreaseRate), PresetSwordStat))
                            if(set NewStatValues[0] = Stat){}
                        if(NewStatValue2 := NewStatValues[1], PresetSwordStat := PresetSword.BigNumber_StatValues[1]):
                            Stat :=  BigNumber_Add(NewStatValue2, BigNumber_Subtract(BigNumber_MultiplyWithAFloat(PresetSwordStat, CurrentSword.StatIncreaseRate), PresetSwordStat))
                            if(set NewStatValues[1] = Stat){}
                        if(NewStatValue3 := NewStatValues[2], PresetSwordStat := PresetSword.BigNumber_StatValues[2]):
                            Stat :=  BigNumber_Add(NewStatValue3, BigNumber_Subtract(BigNumber_MultiplyWithAFloat(PresetSwordStat, CurrentSword.StatIncreaseRate), PresetSwordStat))
                            if(set NewStatValues[2] = Stat){}
                        if(NewStatValue4 := NewStatValues[3], PresetSwordStat := PresetSword.BigNumber_StatValues[3]):
                            Stat :=  BigNumber_Add(NewStatValue4, BigNumber_Subtract(BigNumber_MultiplyWithAFloat(PresetSwordStat, CurrentSword.StatIncreaseRate), PresetSwordStat))
                            if(set NewStatValues[3] = Stat){}

                        NewSword := Sword_Data:
                            MakeSword<constructor>(CurrentSword)
                            Level := CurrentSword.Level + 1
                            BigNumber_StatValues := NewStatValues

                        if:
                            set PlayerStatsDataMap[Player] = Player_Stats_Data:
                                MakePlayerStatsUpdate<constructor>(OldData)
                                OwnedPlayerSword := array{NewSword}
                        then:
                            Print("New Sword Set")
                            return

    GiveCard(InAgent : agent, InCardID : int, InRarity : int):void=
        if(Player := player[InAgent], Player.IsActive[]):
            if(OldData := PlayerStatsDataMap[Player]):
                var NewCardsArray : []SavedCardData = OldData.OwnedCards
                # Search Cards to see if the players already has this Card
                for(Index -> CurrentCard : OldData.OwnedCards):
                    if(CurrentCard.ID = InCardID):
                        # ID found add to Quantity
                        if(CurrentCard.Level = 10):
                            return
                            
                        if:
                            set NewCardsArray[Index] = SavedCardData:
                                MakeSavedCardData<constructor>(CurrentCard)
                                Level := CurrentCard.Level + 1
                        then:
                            if:
                                set PlayerStatsDataMap[Player] = Player_Stats_Data:
                                    MakePlayerStatsUpdate<constructor>(OldData)
                                    OwnedCards := NewCardsArray
                            then:
                                return

                # If not just add it to the OwnedCards Array
                Card := SavedCardData:
                    ID := InCardID
                    Rarity := InRarity
                    Level := 1
                set NewCardsArray += array{Card}
                
                if:
                    set PlayerStatsDataMap[Player] = Player_Stats_Data:
                        MakePlayerStatsUpdate<constructor>(OldData)
                        OwnedCards := NewCardsArray
                then:
                    return

    SetPlaytimeHours(InAgent : agent, NewHours : float):void=
        if(Player := player[InAgent], Player.IsActive[]):
            if(OldData := PlayerStatsDataMap[Player]):
                if:
                    set PlayerStatsDataMap[Player] = Player_Stats_Data:
                        MakePlayerStatsUpdate<constructor>(OldData)
                        HoursPlayed := NewHours

    SetPlaytimeMinutes(InAgent : agent, NewMinutes : float):void=
        if(Player := player[InAgent], Player.IsActive[]):
            if(OldData := PlayerStatsDataMap[Player]):
                if:
                    set PlayerStatsDataMap[Player] = Player_Stats_Data:
                        MakePlayerStatsUpdate<constructor>(OldData)
                        MinutesPlayed := NewMinutes

    # Get the XP Required for that player to level up
    GetXPRequired(InPlayer : player):big_number=
        var XPRequired : big_number = big_number{}
        if(InPlayer.IsActive[]):
            if(PlayerStatsData := PlayerStatsDataMap[InPlayer]):
                XPRequiredFloat := PlayerStatsData.Level * 200.0 * 1.25
                set XPRequired = BigNumber_ConvertFloatIntoBigNumber(XPRequiredFloat)
        return XPRequired

    # Get the XP Required for that player to level up
    GetXPRequiredSpecificLevel(InPlayer : player, InLevel : int):big_number=
        var XPRequired : big_number = big_number{}
        XPRequiredFloat := InLevel * 200.0 * 1.25
        set XPRequired = BigNumber_ConvertFloatIntoBigNumber(XPRequiredFloat)
        return XPRequired

    # Adds XP to the player
    AddXPToPlayer(InPlayer : player, InXPToAdd : big_number, Device : V_LevelUpAndCardUpgradeSelection)<suspends>:void=
        if(InPlayer.IsActive[]):
            if(OldData := PlayerStatsDataMap[InPlayer]):
                # Check if the player can level up
                MultipliedXP := BigNumber_Add(InXPToAdd, BigNumber_ConvertFloatIntoBigNumber(OldData.XPAdd))
                var TotalNewXP : big_number = BigNumber_Add(OldData.XPAmount, MultipliedXP)
                if(FoundCard := SearchForCard(21, InPlayer, Self)?):
                    set TotalNewXP = BigNumber_Add(TotalNewXP, GetBonusBigNumber(InXPToAdd, GetCardBonus(21, FoundCard.Level)))
                var LevelCount : int = 0
                Print("XP Gained : {BigNumber_FormatDisplay(TotalNewXP)}")
                loop:
                    CanLevel := BigNumber_GreaterThanOrEquals(TotalNewXP, GetXPRequiredSpecificLevel(InPlayer, OldData.Level + LevelCount))
                    if(CanLevel = true):
                        set TotalNewXP = BigNumber_Subtract(TotalNewXP, GetXPRequiredSpecificLevel(InPlayer, OldData.Level + LevelCount))
                        set LevelCount += 1
                    else:
                        break

                Print("Level Gained : {LevelCount}")
                if:
                    set PlayerStatsDataMap[InPlayer] = Player_Stats_Data:
                        MakePlayerStatsUpdate<constructor>(OldData)
                        Level := OldData.Level + LevelCount
                        XPAmount := TotalNewXP
                then:
                    if(LevelCount > 0):
                        if(LevelAudioDevice := Device.bLevelUpSoundDevice?):
                            LevelAudioDevice.Play(InPlayer)
                        if(LevelAccolades := Device.bLevelUpAccolades?):
                            LevelAccolades.Award(InPlayer)
                        if(LevelMessage := Device.bLevelUpMessage?):
                            LevelMessage.Show(InPlayer)

                        AddGoldCoin(InPlayer, BigNumber_ConvertFloatIntoBigNumber(LevelCount * 1.0))

    AddGoldCoin(InPlayer : player, InAmount : big_number):void=
        if(InPlayer.IsActive[]):
            if(OldData := PlayerStatsDataMap[InPlayer]):
                GoldCoinAmount := BigNumber_Add(OldData.PlayerGoldCoinAmount, InAmount)
                if:
                    set PlayerStatsDataMap[InPlayer] = Player_Stats_Data:
                        MakePlayerStatsUpdate<constructor>(OldData)
                        PlayerGoldCoinAmount := GoldCoinAmount

    SpendGoldCoin(InPlayer : player, InAmount : big_number):void=
        if(InPlayer.IsActive[]):
            if(OldData := PlayerStatsDataMap[InPlayer]):
                GoldCoinAmount := BigNumber_Subtract(OldData.PlayerGoldCoinAmount, InAmount)
                if:
                    set PlayerStatsDataMap[InPlayer] = Player_Stats_Data:
                        MakePlayerStatsUpdate<constructor>(OldData)
                        PlayerGoldCoinAmount := GoldCoinAmount

    AddXPBonus(InPlayer : player, InAmount : float):void=
        if(InPlayer.IsActive[]):
            if(OldData := PlayerStatsDataMap[InPlayer]):

                if:
                    set PlayerStatsDataMap[InPlayer] = Player_Stats_Data:
                        MakePlayerStatsUpdate<constructor>(OldData)
                        XPAdd := OldData.XPAdd + InAmount

    ResetXPBonus(InPlayer : player):void=
        if(InPlayer.IsActive[]):
            if(OldData := PlayerStatsDataMap[InPlayer]):
                if:
                    set PlayerStatsDataMap[InPlayer] = Player_Stats_Data:
                        MakePlayerStatsUpdate<constructor>(OldData)
                        XPAdd := 0.0

    ChangeHatID(InPlayer : player, InID : int):void=
        if(InPlayer.IsActive[]):
            if(OldData := PlayerStatsDataMap[InPlayer]):
                if:
                    set PlayerStatsDataMap[InPlayer] = Player_Stats_Data:
                        MakePlayerStatsUpdate<constructor>(OldData)
                        HatID := InID

    #Reset all player data [Not currently in use, will use for player prestige]
    ResetStats(InAgent : agent):void=
        if(Player := player[InAgent], Player.IsActive[]):
            if:
                OldData := PlayerStatsDataMap[Player]
                set PlayerStatsDataMap[Player] = Player_Stats_Data:
                    MakePlayerStatsUpdate<constructor>(OldData)
                    #Prestige level will save
                    #Prestige := 0
                    HarvestRatePerHit := big_number:
                        Value := 1.0
                        Exponent := 0
                    HRBaseCost := big_number:
                        Value := 1.0
                        Exponent := 0
                    HRCostIncreaseRate := big_number:
                        Value := 1.12
                        Exponent := 0
                    HRRateIncrease := 1.0
                    # Add an extra one to these for every upgrade we want
                    HarvestRatePerHitUpgrades := {big_number{}, big_number{}, big_number{}, big_number{}, big_number{}, 
                    big_number{}, big_number{}, big_number{}, big_number{}, big_number{}, 
                    big_number{}, big_number{}, big_number{}, big_number{}, big_number{}, 
                    big_number{}, big_number{}, big_number{}, big_number{}, big_number{}}
                    HRUpgradeTierMultipliers := array{1.0, 0.0, 0.0, 0.0, 0.0, 
                    0.0, 0.0, 0.0, 0.0, 0.0,
                    0.0, 0.0, 0.0, 0.0, 0.0,
                    0.0, 0.0, 0.0, 0.0, 0.0}
                    HRUpgradeLevels := array{1.0, 0.0, 0.0, 0.0, 0.0, 
                    0.0, 0.0, 0.0, 0.0, 0.0,
                    0.0, 0.0, 0.0, 0.0, 0.0,
                    0.0, 0.0, 0.0, 0.0, 0.0}
                    HRUpgradeTierMultiplierUpgradesAvaliable := array{0, 0, 0, 0, 0,
                    0, 0, 0, 0, 0,
                    0, 0, 0, 0, 0,
                    0, 0, 0, 0, 0}

                    #Vendor Upgrade Tier Multiplier Avaliable Count
                    StoneVendorTierMultiplierUpgradesAvaliable := 0
                    SilverVendorTierMultiplierUpgradesAvaliable := 0
                    GoldVendorTierMultiplierUpgradesAvaliable := 0
                    DiamondVendorTierMultiplierUpgradesAvaliable := 0

                    CaveUnlockPercent := 0.0
                    #Stone Variables
                    PlayerStoneOreAmount := big_number:
                        Value := 0.0
                        Exponent := 0
                    StoneOreRatePerTrade := big_number:
                        Value := 1.0
                        Exponent := 1
                    StoneVendorUpgradeLevel := 0.0
                    StoneVendorTierMultiplier := 0.0
                    TotalStoneSincePrestige := big_number:
                        Value := 0.0
                        Exponent := 0

                    #Player Upgrades
                    CriticalChanceUpgrade := 0
                    CriticalDamageUpgrade := 0
                    UnlockMegaCriticalUpgrade := 0
                    MegaCriticalUpgrade := 0
                    CriticalChargeUpgrade := 0
                    FinalHitUpgrade := 0
                    ExplosiveHitUpgrade := 0
                    ExplosiveHitRangeUpgrade := 0
                    ExplosiveHitDamageUpgrade := 0
                    ExplosiveHitChargesUpgrade := 0

    # RESET ALL STATS FOR DEBUG REASONS
    ResetStatsDEBUG(InAgent : agent):void=
        if(Player := player[InAgent], Player.IsActive[]):
            if:
                OldData := PlayerStatsDataMap[Player]
                set PlayerStatsDataMap[Player] = Player_Stats_Data:
                    MakePlayerStatsUpdate<constructor>(OldData)
                    Prestige := 0
                    XPAmount := big_number{}
                    XPAdd := 0.0
                    Level := 1
                    HarvestRatePerHit := big_number:
                        Value := 1.0
                        Exponent := 0
                    HRBaseCost := big_number:
                        Value := 1.0
                        Exponent := 0
                    HRCostIncreaseRate := big_number:
                        Value := 1.12
                        Exponent := 0
                    HRRateIncrease := 1.0
                    PrestigeBonusHarvestRate := 0.0
                    LifetimePrestigeResourceAmount := big_number{}
                    TotalStoneSincePrestige := big_number{}
                    LastPrestigedAmount := big_number:
                        Value := 1.5
                        Exponent := 18
                    # Add an extra one to these for every upgrade we want
                    HarvestRatePerHitUpgrades := {big_number{}, big_number{}, big_number{}, big_number{}, big_number{}, 
                    big_number{}, big_number{}, big_number{}, big_number{}, big_number{}, 
                    big_number{}, big_number{}, big_number{}, big_number{}, big_number{}, 
                    big_number{}, big_number{}, big_number{}, big_number{}, big_number{}}
                    HRUpgradeTierMultipliers := array{1.0, 0.0, 0.0, 0.0, 0.0, 
                    0.0, 0.0, 0.0, 0.0, 0.0,
                    0.0, 0.0, 0.0, 0.0, 0.0,
                    0.0, 0.0, 0.0, 0.0, 0.0}
                    HRUpgradeLevels := array{1.0, 0.0, 0.0, 0.0, 0.0, 
                    0.0, 0.0, 0.0, 0.0, 0.0,
                    0.0, 0.0, 0.0, 0.0, 0.0,
                    0.0, 0.0, 0.0, 0.0, 0.0}
                    HRUpgradeTierMultiplierUpgradesAvaliable := array{0, 0, 0, 0, 0,
                    0, 0, 0, 0, 0,
                    0, 0, 0, 0, 0,
                    0, 0, 0, 0, 0}

                    #Vendor Upgrade Tier Multiplier Avaliable Count
                    StoneVendorTierMultiplierUpgradesAvaliable := 0
                    SilverVendorTierMultiplierUpgradesAvaliable := 0
                    GoldVendorTierMultiplierUpgradesAvaliable := 0
                    DiamondVendorTierMultiplierUpgradesAvaliable := 0
                    UpgradePurchaseAmount  := 1
                    # Percentage bonus on cave unlocks
                    CaveUnlockPercent := 0.0
                    #Stone Variables
                    TotalStoneOreAmount := big_number:
                        Value := 0.0
                        Exponent := 0
                    PlayerStoneOreAmount := big_number:
                        Value := 0.0
                        Exponent := 0
                    StoneOreRatePerTrade := big_number:
                        Value := 1.0
                        Exponent := 1
                    StoneVendorUpgradeLevel := 0.0
                    StoneVendorTierMultiplier := 0.0

                    PlayerGoldCoinAmount := big_number:
                        Value := 0.0
                        Exponent := 0

                    OwnedPlayerArtifacts := array{}
                    EquippedPlayerArtifacts := array{}

                    #Sword Variables
                    OwnedPlayerSword := array{}

                    PlayerCopperOreAmount := big_number:
                        Value := 0.0
                        Exponent := 0
                    PlayerSilverOreAmount := big_number:
                        Value := 0.0
                        Exponent := 0
                    PlayerMalachiteOreAmount := big_number:
                        Value := 0.0
                        Exponent := 0
                    PlayerSpectroliteOreAmount := big_number:
                        Value := 0.0
                        Exponent := 0
                    PlayerObsidianOreAmount := big_number:
                        Value := 0.0
                        Exponent := 0
                    PlayerPrestigeResourceAmount := big_number:
                        Value := 0.0
                        Exponent := 0

                    #Quest Variables
                    CompletedQuestLines := array{}
                    CompletedQuests := array{}
                    CurrentQuest := ""
                    CurrentQuestIncrement := big_number{}

                    #Player Upgrades
                    CriticalChanceUpgrade := 0
                    CriticalDamageUpgrade := 0
                    UnlockMegaCriticalUpgrade := 0
                    MegaCriticalUpgrade := 0
                    CriticalChargeUpgrade := 0
                    FinalHitUpgrade := 0
                    OreQuantityUpgrade := 0
                    FriendsUpgrade := 0
                    DamageUpgrade := 0
                    MovementUpgrade := 0
                    HealthRegenUpgrade := 0
                    MaxHealthUpgrade := 0
                    ExplosiveHitUpgrade := 0
                    ExplosiveHitRangeUpgrade := 0
                    ExplosiveHitDamageUpgrade := 0
                    ExplosiveHitChargesUpgrade := 0
                    AltarUpgrade := 0
                    AltarStrengthUpgrade := 0
                    AltarTimeUpgrade := 0
                    GlobalVendorUpgrade := 1

                    #Player Token Var 
                    AlterMaxAmount := 0
                    DungeonMaxAmount := 0
                    AlterTimer := 1.0
                    DungeonTimer := 1.0

                    #Auto Mining Upgrades
                    PrestigeAutoMiningPlotClaimedUpgrade := 0
                    PrestigeAutoMiningCount := 0
                    PrestigeAutoMiningHitRate := 0

                    #DungeonKing
                    DungeonKingTokenSlotAmount := 0
                    DungeonKingTokenSpeedAmount := 0

                    #Prestige Upgrades
                    PrestigeAFKAreaUpgrade := 0
                    PrestigeStrengthUpgrade := 0
                    PrestigeDungeonMasteryUpgrade := 0
                    PrestigeCore := 0
                    PrestigeArtifactMastery := 0 
                    PrestigeArtifactSlots := 0 

                    AchievementTotalStoneBonus := 0.0
                    AchievementTotalArtifactsSmeltedBonus := 0.0
                    AchievementTotalRocksDestroyedBonus := 0.0
                    AchievementHighestStrengthBonus := 0.0
                    AchievementTotalDungeonsCompletedBonus := 0.0
                
                    TotalRocksDestroyed := big_number{}
                    TotalArtifactsSmelted := big_number{}
                    HighestStrength := big_number{}
                    TotalDungeonsCompleted := big_number{}

                    MinutesPlayed := 0.0
                    HoursPlayed := 0.0

                    OrbAmount := big_number{}

                    OwnedCards := array{}
