
using { /Fortnite.com/Devices }
using { /Verse.org/Simulation }
using { /UnrealEngine.com/Temporary/Diagnostics }
using { /Verse.org/Assets }

harvest_rate_upgrade := class<concrete>():
    Name : string = "Pickaxe Power"
    BaseCost : big_number = big_number{Value := 1.0, Exponent := 1}
    RateIncrease : big_number = big_number{Value := 1.0, Exponent := 0}
    BaseRate : big_number = big_number{Value := 1.0, Exponent := 0}
    TierSystems : []float = array{10.0,25.0,50.0,100.0,150.0,200.0,250.0,300.0,350.0,400.0,450.0,500.0,550.0,600.0,650.0,700.0,750.0,800.0}

player_upgrade := class<concrete>():
    Image : texture = VerseExposed.T_UIBackground
    UpgradeName : string = "Name"
    UpgradeDescription : string = "This is an upgrade"
    Levels : []float = array{}
    Costs : []big_number = array{}
    DisplayPercentage : logic = false
    
    GetCurrentLevelAmount(InLevel : int):float=
        if(LevelAmount := Levels[InLevel]):
            return LevelAmount
        return 0.0

    var OptDamageLevels : []damage_amplifier_powerup_device = array{}
    var OptMovementLevels : []movement_modulator_device = array{}
    
    GetDamageLevelDevice(InLevel : int):damage_amplifier_powerup_device=
        if(LevelDevice := OptDamageLevels[InLevel]):
            return LevelDevice
        return damage_amplifier_powerup_device{}

    GetMovementLevelDevice(InLevel : int):movement_modulator_device=
        if(LevelDevice := OptMovementLevels[InLevel]):
            return LevelDevice
        return movement_modulator_device{}

upgrades_data := class:
    #### HARVEST RATE UPGRADES ####
    HRUpgrade1 : harvest_rate_upgrade = harvest_rate_upgrade:
        Name := "Pickaxe Power I"
        BaseCost := big_number{Value := 2.0, Exponent := 1}
        RateIncrease := big_number{Value := 1.0, Exponent := 0}
        BaseRate := big_number{Value := 0.0, Exponent := 0}
        TierSystems := array{10.0,25.0,50.0,100.0,150.0,200.0,250.0,300.0,350.0,400.0,450.0,500.0,550.0,600.0,650.0,700.0,750.0,800.0,850.0,900.0,950.0,1000.0,1050.0,1100.0}

    HRUpgrade2 : harvest_rate_upgrade = harvest_rate_upgrade:
        Name := "Pickaxe Power II"
        BaseCost := big_number{Value := 2.0, Exponent := 2}
        RateIncrease := big_number{Value := 1.0, Exponent := 1}
        BaseRate := big_number{Value := 1.0, Exponent := 1}
        TierSystems := array{10.0,25.0,50.0,100.0,150.0,200.0,250.0,300.0,350.0,400.0,450.0,500.0,550.0,600.0,650.0,700.0,750.0,800.0,850.0,900.0,950.0,1000.0,1050.0,1100.0}

    HRUpgrade3 : harvest_rate_upgrade = harvest_rate_upgrade:
        Name := "Pickaxe Power III"
        BaseCost := big_number{Value := 1.4, Exponent := 3}
        RateIncrease := big_number{Value := 9.0, Exponent := 1}
        BaseRate := big_number{Value := 9.0, Exponent := 1}
        TierSystems := array{10.0,25.0,50.0,100.0,150.0,200.0,250.0,300.0,350.0,400.0,450.0,500.0,550.0,600.0,650.0,700.0,750.0,800.0,850.0,900.0,950.0,1000.0,1050.0,1100.0}
    
    HRUpgrade4 : harvest_rate_upgrade = harvest_rate_upgrade:
        Name := "Pickaxe Power IV"
        BaseCost := big_number{Value := 1.5, Exponent := 4}
        RateIncrease := big_number{Value := 6.0, Exponent := 2}
        BaseRate := big_number{Value := 6.0, Exponent := 2}
        TierSystems := array{10.0,25.0,50.0,100.0,150.0,200.0,250.0,300.0,350.0,400.0,450.0,500.0,550.0,600.0,650.0,700.0,750.0,800.0,850.0,900.0,950.0,1000.0,1050.0,1100.0}
    
    HRUpgrade5 : harvest_rate_upgrade = harvest_rate_upgrade:
        Name := "Pickaxe Power V"
        BaseCost := big_number{Value := 1.5, Exponent := 5}
        RateIncrease := big_number{Value := 3.6, Exponent := 3}
        BaseRate := big_number{Value := 3.6, Exponent := 3}
        TierSystems := array{10.0,25.0,50.0,100.0,150.0,200.0,250.0,300.0,350.0,400.0,450.0,500.0,550.0,600.0,650.0,700.0,750.0,800.0,850.0,900.0,950.0,1000.0,1050.0,1100.0}

    HRUpgrade6 : harvest_rate_upgrade = harvest_rate_upgrade:
        Name := "Pickaxe Power VI"
        BaseCost := big_number{Value := 2.0, Exponent := 6}
        RateIncrease := big_number{Value := 2.2, Exponent := 4}
        BaseRate := big_number{Value := 2.2, Exponent := 4}
        TierSystems := array{10.0,25.0,50.0,100.0,150.0,200.0,250.0,300.0,350.0,400.0,450.0,500.0,550.0,600.0,650.0,700.0,750.0,800.0,850.0,900.0,950.0,1000.0,1050.0,1100.0}
    
    HRUpgrade7 : harvest_rate_upgrade = harvest_rate_upgrade:
        Name := "Pickaxe Power VII"
        BaseCost := big_number{Value := 3.0, Exponent := 7}
        RateIncrease := big_number{Value := 1.1, Exponent := 5}
        BaseRate := big_number{Value := 1.1, Exponent := 5}
        TierSystems := array{10.0,25.0,50.0,100.0,150.0,200.0,250.0,300.0,350.0,400.0,450.0,500.0,550.0,600.0,650.0,700.0,750.0,800.0,850.0,900.0,950.0,1000.0,1050.0,1100.0}
    
    HRUpgrade8 : harvest_rate_upgrade = harvest_rate_upgrade:
        Name := "Pickaxe Power VIII"
        BaseCost := big_number{Value := 4.0, Exponent := 8}
        RateIncrease := big_number{Value := 6.6, Exponent := 5}
        BaseRate := big_number{Value := 6.6, Exponent := 5}
        TierSystems := array{10.0,25.0,50.0,100.0,150.0,200.0,250.0,300.0,350.0,400.0,450.0,500.0,550.0,600.0,650.0,700.0,750.0,800.0,850.0,900.0,950.0,1000.0,1050.0,1100.0}
    
    HRUpgrade9: harvest_rate_upgrade = harvest_rate_upgrade:
        Name := "Pickaxe Power IX"
        BaseCost := big_number{Value := 5.5, Exponent := 9}
        RateIncrease := big_number{Value := 3.3, Exponent := 6}
        BaseRate := big_number{Value := 3.3, Exponent := 6}
        TierSystems := array{10.0,25.0,50.0,100.0,150.0,200.0,250.0,300.0,350.0,400.0,450.0,500.0,550.0,600.0,650.0,700.0,750.0,800.0,850.0,900.0,950.0,1000.0,1050.0,1100.0}

    HRUpgrade10: harvest_rate_upgrade = harvest_rate_upgrade:
        Name := "Pickaxe Power X"
        BaseCost := big_number{Value := 1.0, Exponent := 11}
        RateIncrease := big_number{Value := 2.2, Exponent := 7}
        BaseRate := big_number{Value := 2.2, Exponent := 7}
        TierSystems := array{10.0,25.0,50.0,100.0,150.0,200.0,250.0,300.0,350.0,400.0,450.0,500.0,550.0,600.0,650.0,700.0,750.0,800.0,850.0,900.0,950.0,1000.0,1050.0,1100.0}

    HRUpgrade11: harvest_rate_upgrade = harvest_rate_upgrade:
        Name := "Pickaxe Power XI"
        BaseCost := big_number{Value := 1.5, Exponent := 12}
        RateIncrease := big_number{Value := 1.6, Exponent := 8}
        BaseRate := big_number{Value := 1.6, Exponent := 8}
        TierSystems := array{10.0,25.0,50.0,100.0,150.0,200.0,250.0,300.0,350.0,400.0,450.0,500.0,550.0,600.0,650.0,700.0,750.0,800.0,850.0,900.0,950.0,1000.0,1050.0,1100.0}

    HRUpgrade12: harvest_rate_upgrade = harvest_rate_upgrade:
        Name := "Pickaxe Power XII"
        BaseCost :=big_number{Value := 2.0, Exponent := 13}
        RateIncrease := big_number{Value := 1.1, Exponent := 9}
        BaseRate := big_number{Value := 1.1, Exponent := 9}
        TierSystems := array{10.0,25.0,50.0,100.0,150.0,200.0,250.0,300.0,350.0,400.0,450.0,500.0,550.0,600.0,650.0,700.0,750.0,800.0,850.0,900.0,950.0,1000.0,1050.0,1100.0}

    HRUpgrade13: harvest_rate_upgrade = harvest_rate_upgrade:
        Name := "Pickaxe Power XIII"
        BaseCost := big_number{Value := 3.5, Exponent := 14}
        RateIncrease := big_number{Value := 7.7, Exponent := 9}
        BaseRate := big_number{Value := 7.7, Exponent := 9}
        TierSystems := array{10.0,25.0,50.0,100.0,150.0,200.0,250.0,300.0,350.0,400.0,450.0,500.0,550.0,600.0,650.0,700.0,750.0,800.0,850.0,900.0,950.0,1000.0,1050.0,1100.0}

    HRUpgrade14: harvest_rate_upgrade = harvest_rate_upgrade:
        Name := "Pickaxe Power XIV"
        BaseCost := big_number{Value := 5.0, Exponent := 15}
        RateIncrease := big_number{Value := 5.5, Exponent := 10}
        BaseRate := big_number{Value := 5.5, Exponent := 10}
        TierSystems := array{10.0,25.0,50.0,100.0,150.0,200.0,250.0,300.0,350.0,400.0,450.0,500.0,550.0,600.0,650.0,700.0,750.0,800.0,850.0,900.0,950.0,1000.0,1050.0,1100.0}

    HRUpgrade15: harvest_rate_upgrade = harvest_rate_upgrade:
        Name := "Pickaxe Power XV"
        BaseCost := big_number{Value := 7.5, Exponent := 16}
        RateIncrease := big_number{Value := 3.8, Exponent := 11}
        BaseRate := big_number{Value := 3.8, Exponent := 11}
        TierSystems := array{10.0,25.0,50.0,100.0,150.0,200.0,250.0,300.0,350.0,400.0,450.0,500.0,550.0,600.0,650.0,700.0,750.0,800.0,850.0,900.0,950.0,1000.0,1050.0,1100.0}

    HRUpgrade16: harvest_rate_upgrade = harvest_rate_upgrade:
        Name := "Pickaxe Power XVI"
        BaseCost := big_number{Value := 1.125, Exponent := 18}
        RateIncrease := big_number{Value := 2.65, Exponent := 12}
        BaseRate := big_number{Value := 2.45, Exponent := 12}
        TierSystems := array{10.0,25.0,50.0,100.0,150.0,200.0,250.0,300.0,350.0,400.0,450.0,500.0,550.0,600.0,650.0,700.0,750.0,800.0,850.0,900.0,950.0,1000.0,1050.0,1100.0}

    HRUpgrade17: harvest_rate_upgrade = harvest_rate_upgrade:
        Name := "Pickaxe Power XVII"
        BaseCost := big_number{Value := 1.68, Exponent := 19}
        RateIncrease := big_number{Value := 1.815, Exponent := 13}
        BaseRate := big_number{Value := 1.815, Exponent := 13}
        TierSystems := array{10.0,25.0,50.0,100.0,150.0,200.0,250.0,300.0,350.0,400.0,450.0,500.0,550.0,600.0,650.0,700.0,750.0,800.0,850.0,900.0,950.0,1000.0,1050.0,1100.0}

    HRUpgrade18: harvest_rate_upgrade = harvest_rate_upgrade:
        Name := "Pickaxe Power XVIII"
        BaseCost := big_number{Value := 2.53, Exponent := 20}
        RateIncrease := big_number{Value := 1.30, Exponent := 14}
        BaseRate := big_number{Value := 1.30, Exponent := 14}
        TierSystems := array{10.0,25.0,50.0,100.0,150.0,200.0,250.0,300.0,350.0,400.0,450.0,500.0,550.0,600.0,650.0,700.0,750.0,800.0,850.0,900.0,950.0,1000.0,1050.0,1100.0}

    HRUpgrade19: harvest_rate_upgrade = harvest_rate_upgrade:
        Name := "Pickaxe Power XIX"
        BaseCost := big_number{Value := 3.8, Exponent := 21}
        RateIncrease := big_number{Value := 8.50, Exponent := 14}
        BaseRate := big_number{Value := 8.50, Exponent := 14}
        TierSystems := array{10.0,25.0,50.0,100.0,150.0,200.0,250.0,300.0,350.0,400.0,450.0,500.0,550.0,600.0,650.0,700.0,750.0,800.0,850.0,900.0,950.0,1000.0,1050.0,1100.0}

    HRUpgrade20: harvest_rate_upgrade = harvest_rate_upgrade:
        Name := "Pickaxe Power XX"
        BaseCost := big_number{Value := 5.7, Exponent := 22}
        RateIncrease := big_number{Value := 6.38, Exponent := 15}
        BaseRate := big_number{Value := 6.38, Exponent := 15}
        TierSystems := array{10.0,25.0,50.0,100.0,150.0,200.0,250.0,300.0,350.0,400.0,450.0,500.0,550.0,600.0,650.0,700.0,750.0,800.0,850.0,900.0,950.0,1000.0,1050.0,1100.0}
    var HRUpgradesArray : []harvest_rate_upgrade = array{}

    #### MINING ABILITY UPGRADES ####
    CriticalChanceUpgrade : player_upgrade = player_upgrade:
        UpgradeName := "Critical Chance"
        UpgradeDescription := "Increase your\ncritical odds!"
        Levels := array{0.0, 3.0, 6.0, 9.0, 12.0, 15.0, 18.0, 21.0, 24.0, 27.0, 30.0}
        Costs := array{big_number{}, big_number{Value := 5.00, Exponent := 4}, big_number{Value := 5.00, Exponent := 6}, big_number{Value := 5.00, Exponent := 8}, big_number{Value := 5.00, Exponent := 10}, big_number{Value := 5.00, Exponent := 12}, big_number{Value := 5.00, Exponent := 14}, big_number{Value := 5.00, Exponent := 16}, big_number{Value := 5.00, Exponent := 18}, big_number{Value := 5.00, Exponent := 20}, big_number{Value := 5.00, Exponent := 22}}        
        Image := VerseExposed.T_CriticalChance
        DisplayPercentage := true

    CriticalDamageUpgrade : player_upgrade = player_upgrade:
        UpgradeName := "Critical Damage"
        UpgradeDescription := "Increase your critical\nhit damage"
        Levels := array{125.0, 150.0, 175.0, 200.0, 225.0, 250.0, 275.0, 300.0, 325.0, 350.0, 375.0}
        Costs := array{big_number{}, big_number{Value := 2.00, Exponent := 5}, big_number{Value := 2.00, Exponent := 7}, big_number{Value := 2.00, Exponent := 9}, big_number{Value := 2.00, Exponent := 11}, big_number{Value := 2.00, Exponent := 13}, big_number{Value := 2.00, Exponent := 15}, big_number{Value := 2.00, Exponent := 17}, big_number{Value := 2.00, Exponent := 19}, big_number{Value := 2.00, Exponent := 21}, big_number{Value := 2.00, Exponent := 23}}
        Image := VerseExposed.T_CriticalPower
        DisplayPercentage := true
    
    MegaCriticalUpgrade : player_upgrade = player_upgrade:
        UpgradeName := "Mega Critical Damage"
        UpgradeDescription := "Deal more damage\nwith every mega crit!"
        Levels := array{800.0, 900.0, 1000.0, 1100.0, 1200.0, 1300.0, 1400.0, 1500.0}
        Costs := array{big_number{}, big_number{Value := 2.00, Exponent := 12}, big_number{Value := 2.00, Exponent := 15}, big_number{Value := 2.00, Exponent := 18}, big_number{Value := 2.00, Exponent := 21}, big_number{Value := 2.00, Exponent := 24}, big_number{Value := 2.00, Exponent := 27}, big_number{Value := 2.00, Exponent := 30}}
        Image := VerseExposed.T_MegaCrit
        DisplayPercentage := true

    UnlockMegaCriticalUpgrade : player_upgrade = player_upgrade:
        UpgradeName := "Mega Critical"
        UpgradeDescription := "Deal more damage\nwith every mega crit!"
        Levels := array{0.0, 1.0}
        Costs := array{big_number{}, big_number{Value := 5.00, Exponent := 11}}
        Image := VerseExposed.T_MegaCrit
        DisplayPercentage := true

    CriticalChargeUpgrade : player_upgrade = player_upgrade:
        UpgradeName := "Mega Critical Charge"
        UpgradeDescription := "More charge with\nevery hit!"
        Levels := array{1.0, 2.0, 3.0, 4.0, 5.0}
        Costs := array{big_number{}, big_number{Value := 5.00, Exponent := 11}, big_number{Value := 5.00, Exponent := 14}, big_number{Value := 5.00, Exponent := 17}, big_number{Value := 5.00, Exponent := 20}}
        Image := VerseExposed.T_MegaCrit_Charge

    FinalHitUpgrade : player_upgrade = player_upgrade:
        UpgradeName := "Final Hit"
        UpgradeDescription := "Destroyed rocks\ngive bonus stone!"
        Levels := array{0.0, 50.0, 60.0, 70.0, 80.0, 90.0, 100.0, 110.0, 120.0, 130.0, 140.0, 150.0}
        Costs := array{big_number{}, big_number{Value := 2.50, Exponent := 3}, big_number{Value := 2.50, Exponent := 5}, big_number{Value := 2.50, Exponent := 7}, big_number{Value := 2.50, Exponent := 9}, big_number{Value := 2.50, Exponent := 11}, big_number{Value := 2.50, Exponent := 13}, big_number{Value := 2.50, Exponent := 15}, big_number{Value := 2.50, Exponent := 17}, big_number{Value := 2.50, Exponent := 19}, big_number{Value := 2.50, Exponent := 21}, big_number{Value := 2.50, Exponent := 23}}
        Image := VerseExposed.T_FinalHit
        DisplayPercentage := true

    OreQuantityUpgrade : player_upgrade = player_upgrade:
        UpgradeName := "Ore Quantity"
        UpgradeDescription := "Get more ores when\ndestroying a rock"
        Levels := array{1.0, 2.0, 3.0, 4.0, 5.0, 6.0, 7.0, 8.0, 9.0, 10.0}
        Costs := array{big_number{}, big_number{Value := 5.00, Exponent := 1}, big_number{Value := 5.00, Exponent := 3}, big_number{Value := 5.00, Exponent := 5}, big_number{Value := 5.00, Exponent := 7}, big_number{Value := 7.00, Exponent := 7}, big_number{Value := 5.00, Exponent := 9}, big_number{Value := 5.00, Exponent := 11}, big_number{Value := 5.00, Exponent := 13}, big_number{Value := 5.00, Exponent := 15}}
        Image := VerseExposed.T_OreAmount

    FriendsUpgrade : player_upgrade = player_upgrade:
        UpgradeName := "Friends"
        UpgradeDescription := "Get more ores\nwhen around players"
        Levels := array{5.0, 10.0, 15.0, 20.0, 25.0, 30.0}
        Costs := array{big_number{}, big_number{Value := 1.00, Exponent := 5}, big_number{Value := 1.00, Exponent := 8}, big_number{Value := 1.00, Exponent := 11}, big_number{Value := 1.00, Exponent := 14}, big_number{Value := 1.00, Exponent := 17}}
        Image := VerseExposed.T_FriendsUpgrade
        DisplayPercentage := true

    DamageUpgrade : player_upgrade = player_upgrade:
        UpgradeName := "Damage"
        UpgradeDescription := "Deal more damage"
        Costs := array{big_number{}, big_number{Value := 6.0, Exponent := 12}, big_number{Value := 1.8, Exponent := 14}, big_number{Value := 3.6, Exponent := 14}, big_number{Value := 5.4, Exponent := 14}, big_number{Value := 7.2, Exponent := 14}, big_number{Value := 9.0, Exponent := 14}, big_number{Value := 1.08, Exponent := 15}, big_number{Value := 1.26, Exponent := 15}, big_number{Value := 1.44, Exponent := 15}, big_number{Value := 1.62, Exponent := 15}}
        Image := VerseExposed.T_SwordDamage
        DisplayPercentage := true

    MovementUpgrade : player_upgrade = player_upgrade:
        UpgradeName := "Movement"
        UpgradeDescription := "Move faster\n"
        Levels := array{100.0, 105.0, 110.0, 115.0, 120.0, 125.0, 130.0, 135.0, 140.0, 145.0, 150.0}
        Costs := array{big_number{}, big_number{Value := 1.00, Exponent := 8}, big_number{Value := 1.00, Exponent := 11}, big_number{Value := 1.00, Exponent := 14}, big_number{Value := 1.00, Exponent := 17}, big_number{Value := 1.00, Exponent := 20}, big_number{Value := 1.00, Exponent := 23}, big_number{Value := 1.00, Exponent := 26}, big_number{Value := 1.00, Exponent := 29}, big_number{Value := 1.00, Exponent := 32}, big_number{Value := 1.00, Exponent := 35}}
        Image := VerseExposed.T_Speed
        DisplayPercentage := true

    HealthRegenUpgrade : player_upgrade = player_upgrade:
        UpgradeName := "Health Regen"
        UpgradeDescription := "Regen health per second\n"
        Levels := array{1.0, 5.0, 10.0, 15.0, 20.0, 25.0, 30.0, 35.0, 40.0, 45.0, 50.0}
        Costs := array{big_number{}, big_number{Value := 4.00, Exponent := 8}, big_number{Value := 4.00, Exponent := 11}, big_number{Value := 4.00, Exponent := 14}, big_number{Value := 4.00, Exponent := 17}, big_number{Value := 4.00, Exponent := 20}, big_number{Value := 4.00, Exponent := 23}, big_number{Value := 4.00, Exponent := 26}, big_number{Value := 4.00, Exponent := 29}, big_number{Value := 4.00, Exponent := 32}, big_number{Value := 4.00, Exponent := 35}}
        Image := VerseExposed.T_HealthRegen
        
    MaxHealthUpgrade : player_upgrade = player_upgrade:
        UpgradeName := "Max Health"
        UpgradeDescription := "Get extra health\n"
        Levels := array{0.0, 100.0, 200.0, 300.0, 400.0, 500.0, 600.0, 700.0, 800.0, 900.0, 1000.0}
        Costs := array{big_number{}, big_number{Value := 2.00, Exponent := 8}, big_number{Value := 2.00, Exponent := 11}, big_number{Value := 2.00, Exponent := 14}, big_number{Value := 2.00, Exponent := 17}, big_number{Value := 2.00, Exponent := 20}, big_number{Value := 2.00, Exponent := 23}, big_number{Value := 2.00, Exponent := 26}, big_number{Value := 2.00, Exponent := 29}, big_number{Value := 2.00, Exponent := 32}, big_number{Value := 2.00, Exponent := 35}}
        Image := VerseExposed.T_Health

    ExplosiveHitUpgrade : player_upgrade = player_upgrade:
        UpgradeName := "Explosive Hit"
        UpgradeDescription := "Unlock Explosive Hit!\n"
        Levels := array{0.0, 1.0}
        Costs := array{big_number{}, big_number{Value := 1.00, Exponent := 21}}
        Image := VerseExposed.T_ExplosiveHitDamage

    ExplosiveHitRangeUpgrade : player_upgrade = player_upgrade:
        UpgradeName := "Explosive Range"
        UpgradeDescription := "Increase The Explosive Area\n"
        Levels := array{1000.0, 1200.0, 1300.0, 1500.0, 1600.0, 1750.0}
        Costs := array{big_number{}, big_number{Value := 8.00, Exponent := 23}, big_number{Value := 8.00, Exponent := 27}, big_number{Value := 8.00, Exponent := 31}, big_number{Value := 8.00, Exponent := 35}, big_number{Value := 8.00, Exponent := 39}}
        Image := VerseExposed.T_ExplosiveHitRange
    
    ExplosiveHitDamageUpgrade : player_upgrade = player_upgrade:
        UpgradeName := "Explosive Damage"
        UpgradeDescription := "Increase Explosive Damage!\n"
        Levels := array{30.0, 35.0, 40.0, 45.0, 50.0, 55.0, 60.0, 65.0, 70.0, 75.0}
        Costs := array{big_number{}, big_number{Value := 8.00, Exponent := 23}, big_number{Value := 8.00, Exponent := 27}, big_number{Value := 8.00, Exponent := 31}, big_number{Value := 8.00, Exponent := 35}, big_number{Value := 8.00, Exponent := 39}, big_number{Value := 8.00, Exponent := 43}, big_number{Value := 8.00, Exponent := 47}, big_number{Value := 8.00, Exponent := 51}, big_number{Value := 8.00, Exponent := 55}}
        Image := VerseExposed.T_ExplosiveHitDamage
        DisplayPercentage := true
    
    ExplosiveHitChargesUpgrade : player_upgrade = player_upgrade:
        UpgradeName := "Explosive Charge"
        UpgradeDescription := "Gain more explosive charge\nevery hit"
        Levels := array{1.0, 2.0, 4.0, 5.0}
        Costs := array{big_number{}, big_number{Value := 1.00, Exponent := 23}, big_number{Value := 1.00, Exponent := 27}, big_number{Value := 1.00, Exponent := 31}}  
        Image := VerseExposed.T_ExplosiveHitCharge

    AltarUpgrade : player_upgrade = player_upgrade:
        UpgradeName := "Unlock Altar"
        UpgradeDescription := "Unlock the altar\n "
        Levels := array{0.0, 1.0}
        Costs := array{big_number{}, big_number{Value := 5.0, Exponent := 7}}
        Image := VerseExposed.T_AltarIcon
    
    AltarStrengthUpgrade : player_upgrade = player_upgrade:
        UpgradeName := "Altar Strength"
        UpgradeDescription := "Increase the bonus power\nfrom altar blessings"
        Levels := array{100.0, 120.0, 140.0, 160.0, 180.0, 200.0}
        Costs := array{big_number{}, big_number{Value := 2.00, Exponent := 8}, big_number{Value := 2.00, Exponent := 11}, big_number{Value := 2.00, Exponent := 14}, big_number{Value := 2.00, Exponent := 17}, big_number{Value := 2.00, Exponent := 20}}
        Image := VerseExposed.T_AltarStrength
        DisplayPercentage := true
    
    AltarTimeUpgrade : player_upgrade = player_upgrade:
        UpgradeName := "Altar Time"
        UpgradeDescription := "Increase the duration\nof altar blessings"
        Levels := array{180.0, 190.0, 200.0, 210.0, 220.0, 230.0, 240.0, 250.0, 260.0, 270.0, 280.0, 290.0, 300.0, 310.0, 320.0, 330.0, 340.0, 350.0, 360.0, 370.0, 380.0}
        Costs := array{big_number{}, big_number{Value := 2.00, Exponent := 8}, big_number{Value := 2.00, Exponent := 9}, big_number{Value := 2.00, Exponent := 10}, big_number{Value := 2.00, Exponent := 11}, big_number{Value := 2.00, Exponent := 12}, big_number{Value := 2.00, Exponent := 13}, big_number{Value := 2.00, Exponent := 14}, big_number{Value := 2.00, Exponent := 15}, big_number{Value := 2.00, Exponent := 16}, big_number{Value := 2.00, Exponent := 17}, big_number{Value := 2.00, Exponent := 18}, big_number{Value := 2.00, Exponent := 19}, big_number{Value := 2.00, Exponent := 20}, big_number{Value := 2.00, Exponent := 21}, big_number{Value := 2.00, Exponent := 22}, big_number{Value := 2.00, Exponent := 23}, big_number{Value := 2.00, Exponent := 24}, big_number{Value := 2.00, Exponent := 25}, big_number{Value := 2.00, Exponent := 26}, big_number{Value := 2.00, Exponent := 27}}
        Image := VerseExposed.T_AltarTime
    
    GlobalVendorUpgrade : player_upgrade = player_upgrade:
        UpgradeName := "Global Vendor"
        UpgradeDescription := "This is the global vendor\nupgrade"
        Levels := array{0.0, 1.0}
        Costs := array{big_number{}, big_number{Value := 1.5, Exponent := 15}}

    Prestige_AFKArea : player_upgrade = player_upgrade:
        UpgradeName := "Lazy Looting"
        UpgradeDescription := "Increase your AFK\narea Earning Rate!"
        Levels := array{5.0, 10.0, 15.0, 20.0, 25.0, 30.0, 35.0, 40.0, 45.0, 50.0 }
        Costs := array{big_number{}, big_number{Value := 5.00, Exponent := 8}, big_number{Value := 5.00, Exponent := 11}, big_number{Value := 5.00, Exponent := 14}, big_number{Value := 5.00, Exponent := 17}, big_number{Value := 5.00, Exponent := 20}, big_number{Value := 5.00, Exponent := 23}, big_number{Value := 5.00, Exponent := 26}, big_number{Value := 5.00, Exponent := 29}, big_number{Value := 5.00, Exponent := 32}}
        Image := VerseExposed.T_AFKMining
        DisplayPercentage := true
    
    Prestige_MiningAscension : player_upgrade = player_upgrade:
        UpgradeName := "Mining Ascension"
        UpgradeDescription := "Number of strength\nupgrades ascended"
        Levels := array{0.0, 1.0, 2.0, 3.0, 4.0, 5.0, 6.0, 7.0, 8.0, 9.0, 10.0, 11.0, 12.0, 13.0, 14.0, 15.0, 16.0, 17.0, 18.0, 19.0, 20.0}
        Costs := array{big_number{}, big_number{Value := 2.0, Exponent := 1}, big_number{Value := 2.0, Exponent := 2}, big_number{Value := 1.4, Exponent :=3}, big_number{Value := 1.5, Exponent := 4}, big_number{Value := 1.5, Exponent := 5}, big_number{Value := 2.0, Exponent := 6}, big_number{Value := 3.0, Exponent := 7}, big_number{Value := 4.0, Exponent := 8}, big_number{Value := 5.5, Exponent := 9}, big_number{Value := 1.0, Exponent := 11}, big_number{Value := 1.5, Exponent := 12}, big_number{Value := 2.0, Exponent := 13}, big_number{Value := 3.5, Exponent := 14}, big_number{Value := 5.0, Exponent := 15}, big_number{Value := 7.5, Exponent := 16}, big_number{Value := 1.125, Exponent := 18}, big_number{Value := 1.68, Exponent := 19}, big_number{Value := 2.53, Exponent := 20}, big_number{Value := 3.8, Exponent := 21}, big_number{Value := 5.7, Exponent := 22}}
        Image := VerseExposed.T_Prestige

    Prestige_DungeonMastery : player_upgrade = player_upgrade:
        UpgradeName := "Dungeon Loot Bonus"
        UpgradeDescription := "Gives extra loot when\ncompeleting a Dungeon"
        Levels := array{0.0, 5.0, 10.0, 15.0, 20.0, 25.0, 30.0, 35.0, 40.0, 45.0 }
        Costs := array{big_number{}, big_number{Value := 2.00, Exponent := 2}, big_number{Value := 2.00, Exponent := 3}, big_number{Value := 2.00, Exponent := 4}, big_number{Value := 2.00, Exponent := 5}, big_number{Value := 2.00, Exponent := 6}, big_number{Value := 2.00, Exponent := 7}, big_number{Value := 2.00, Exponent := 8}, big_number{Value := 2.00, Exponent := 9}, big_number{Value := 2.00, Exponent := 10}}
        DisplayPercentage := true
        Image := VerseExposed.T_IslandLootBonus_Stone

    Prestige_Core: player_upgrade = player_upgrade:
        UpgradeName := "Prestige"
        UpgradeDescription := "Prestige to upgrade your character\nProgress will be reset"
        Levels := array{100.0, 105.0, 110.0, 115.0, 120.0, 130.0, 135.0, 140.0, 145.0, 150.0 }
        Costs := array{big_number{}, big_number{Value := 5.0, Exponent := 16}, big_number{Value := 4.0, Exponent := 19}, big_number{Value := 8.0, Exponent := 19}, big_number{Value := 1.2, Exponent := 20}, big_number{Value := 1.6, Exponent := 20}}

    Prestige_ArtifactMastery: player_upgrade = player_upgrade:
        UpgradeName := "Ore Chance"
        UpgradeDescription := "Improve chance to\nget ores"
        Levels := array{10.0, 15.0, 20.0, 25.0, 30.0, 35.0, 40.0}
        Costs := array{big_number{}, big_number{Value := 5.00, Exponent := 5}, big_number{Value := 5.00, Exponent := 8}, big_number{Value := 5.00, Exponent := 11}, big_number{Value := 5.00, Exponent := 14}, big_number{Value := 5.00, Exponent := 17}, big_number{Value := 5.00, Exponent := 20}}
        DisplayPercentage := true
        Image := VerseExposed.T_ArtifactChance

    Prestige_ArtifactSlots: player_upgrade = player_upgrade:
        UpgradeName := "Artifact Slots"
        UpgradeDescription := "Gain more slots\nfor Artifacts"
        Levels := array{1.0, 2.0, 3.0, 4.0}
        Costs := array{big_number{}, big_number{Value := 1.0, Exponent := 1}, big_number{Value := 1.0, Exponent := 3}, big_number{Value := 1.0, Exponent := 4}}
        Image := VerseExposed.T_ArtifactEquip

    #Automining
    #Auto Miner Plot Claim only requires 1 purchase and then it will max out.
    Prestige_AutoMining : player_upgrade = player_upgrade:
        UpgradeName := "Claim Plot"
        UpgradeDescription := "Claim Your Plot and\ngain access to auto mining"
        #The levels are a value we can access and use. When not using them we can iterate up and keep things clean.
        #These are not levels displayed to the player, they are used for backend development.
        Levels := array{0.0, 1.0}
        #Costs are something we'll always use and want to fill out.
        Costs := array{big_number{}, big_number{Value := 1.0, Exponent := 0}}
        Image := VerseExposed.T_AutoMinerClaimPlot

    #TODO: We need to plan the maximum amount of upgrades for auto mining speed
    Prestige_AutoMiningCount : player_upgrade = player_upgrade:
        UpgradeName := "Extra Auto Miner"
        UpgradeDescription := "Improve the speed\nof your auto miner"
        Levels := array{1.0, 2.0, 3.0, 4.0}
        Costs := array{big_number{}, big_number{Value := 2.5, Exponent := 1}, big_number{Value := 2.0, Exponent := 4}, big_number{Value := 5.0, Exponent := 6}}
        Image := VerseExposed.T_AutoMinerSpeed

    #TODO: We need to plan the maximum amount of upgrades for mining hit rate for auto miners
    Prestige_AutoMiningHitRate : player_upgrade = player_upgrade:
        UpgradeName := "Auto Mining Damage"
        UpgradeDescription := "Increase the power\nof your auto miner"
        Levels := array{2.0, 4.0, 6.0, 8.0, 10.0}
        Costs := array{big_number{}, big_number{Value := 1.0, Exponent := 2}, big_number{Value := 1.0, Exponent := 4}, big_number{Value := 1.0, Exponent := 6}, big_number{Value := 1.0, Exponent := 8}}
        Image := VerseExposed.T_AutoMinerDamage
    #Auto Mining End

    #Dungeon King
    DungeonKingTokenAmount : player_upgrade = player_upgrade:
        UpgradeName := "Token Amount"
        UpgradeDescription := "Increase your max\namount of tokens"
        Levels := array{0.0, 1.0, 2.0, 3.0, 4.0}
        Costs := array{big_number{}, big_number{Value := 1.00, Exponent := 16}, big_number{Value := 1.00, Exponent := 19}, big_number{Value := 1.00, Exponent := 22}, big_number{Value := 1.00, Exponent := 25}}
        Image := VerseExposed.T_TokenAmount
        
    DungeonKingTokenGenerationSpeed : player_upgrade = player_upgrade:
        UpgradeName := "Faster Token Generation"
        UpgradeDescription := "Increase the speed\nDungeon tokens generate"
        Levels := array{100.0, 110.0, 120.0, 130.0, 140.0, 150.0}
        Costs := array{big_number{}, big_number{Value := 1.50, Exponent := 16}, big_number{Value := 1.50, Exponent := 19}, big_number{Value := 1.50, Exponent := 22}, big_number{Value := 1.50, Exponent := 25}, big_number{Value := 1.50, Exponent := 28}}
        Image := VerseExposed.T_TokenSpeed
        DisplayPercentage := true

    GetCriticalChanceAmount(InLevel : int):float=
        return CriticalChanceUpgrade.GetCurrentLevelAmount(InLevel)

    GetCriticalDamageAmount(InLevel : int):float=
        return CriticalDamageUpgrade.GetCurrentLevelAmount(InLevel)

    GetMegaCriticalAmount(InLevel : int):float=
        return MegaCriticalUpgrade.GetCurrentLevelAmount(InLevel)

    GetUnlockMegaCriticalAmount(InLevel : int):float=
        return UnlockMegaCriticalUpgrade.GetCurrentLevelAmount(InLevel)

    GetCriticalChargeAmount(InLevel : int):float=
        return CriticalChargeUpgrade.GetCurrentLevelAmount(InLevel)

    GetFinalHitAmount(InLevel : int):float=
        return FinalHitUpgrade.GetCurrentLevelAmount(InLevel)
    
    GetOreQuantityAmount(InLevel : int):float=
        return OreQuantityUpgrade.GetCurrentLevelAmount(InLevel)
    
    GetFriendsAmount(InLevel : int):float=
        return FriendsUpgrade.GetCurrentLevelAmount(InLevel)

    GetDamageDevice(InLevel : int):damage_amplifier_powerup_device=
        return DamageUpgrade.GetDamageLevelDevice(InLevel)

    GetMovementDevice(InLevel : int):movement_modulator_device=
        return MovementUpgrade.GetMovementLevelDevice(InLevel)

    SetDamageDevices(DamagePowerups : []damage_amplifier_powerup_device):void=
        set DamageUpgrade.OptDamageLevels = DamagePowerups

    SetMovementDevices(MovementModulators : []movement_modulator_device):void=
        set MovementUpgrade.OptMovementLevels = MovementModulators

    GetHealthRegenAmount(InLevel : int):float=
        return HealthRegenUpgrade.GetCurrentLevelAmount(InLevel)

    GetMaxHealthAmount(InLevel : int):float=
        return MaxHealthUpgrade.GetCurrentLevelAmount(InLevel)

    GetExplosiveHitAmount(InLevel : int):float=
        return ExplosiveHitUpgrade.GetCurrentLevelAmount(InLevel)
    
    GetExplosiveHitRangeAmount(InLevel : int):float=
        return ExplosiveHitRangeUpgrade.GetCurrentLevelAmount(InLevel)

    GetExplosiveHitDamageAmount(InLevel : int):float=
        return ExplosiveHitDamageUpgrade.GetCurrentLevelAmount(InLevel)

    GetExplosiveHitChargesAmount(InLevel : int):float=
        return ExplosiveHitChargesUpgrade.GetCurrentLevelAmount(InLevel)

    GetAltarStrengthAmount(InLevel : int):float=
        return AltarStrengthUpgrade.GetCurrentLevelAmount(InLevel)

    GetAltarTimeAmount(InLevel : int):float=
        return AltarTimeUpgrade.GetCurrentLevelAmount(InLevel)

    GetGlobalVendorAmount(InLevel : int):float=
        return GlobalVendorUpgrade.GetCurrentLevelAmount(InLevel)

    GetPrestigeAFKAreaAmount(InLevel : int):float=
        return Prestige_AFKArea.GetCurrentLevelAmount(InLevel)

    GetPrestigeMiningAscensionAmount(InLevel : int):float=
        return Prestige_MiningAscension.GetCurrentLevelAmount(InLevel)

    GetDungeonMasteryAmount(InLevel : int):float=
        return Prestige_DungeonMastery.GetCurrentLevelAmount(InLevel)

    GetPrestigeCoreAmount(InLevel : int):float=
        return Prestige_Core.GetCurrentLevelAmount(InLevel)

    GetPrestigeArtifactMasteryAmount(InLevel : int):float=
        return Prestige_ArtifactMastery.GetCurrentLevelAmount(InLevel)
    
    GetPrestigeArtifactSlotsAmount(InLevel : int):float=
        return Prestige_ArtifactSlots.GetCurrentLevelAmount(InLevel)

    #Auto Mining
    GetPrestigeAutoMinerPlotAmount(InLevel : int):float=
        return Prestige_AutoMining.GetCurrentLevelAmount(InLevel)
        
    GetPrestigeAutoMinerSpeedAmount(InLevel : int):float=
        return Prestige_AutoMiningCount.GetCurrentLevelAmount(InLevel)
        
    GetPrestigeAutoMinerHitRateAmount(InLevel : int):float=
        return Prestige_AutoMiningHitRate.GetCurrentLevelAmount(InLevel)

    #Dungeon King
    GetDungeonKingTokenAmount(InLevel : int):float=
        return DungeonKingTokenAmount.GetCurrentLevelAmount(InLevel)
        
    GetDungeonKingTokenGenerationSpeed(InLevel : int):float=
        return DungeonKingTokenGenerationSpeed.GetCurrentLevelAmount(InLevel)

    SetUpHRUpgradesArray():void=
        set HRUpgradesArray = array{HRUpgrade1, HRUpgrade2, HRUpgrade3, HRUpgrade4, HRUpgrade5, HRUpgrade6, HRUpgrade7, HRUpgrade8, HRUpgrade9, HRUpgrade10, HRUpgrade11, HRUpgrade12, 
        HRUpgrade13, HRUpgrade14, HRUpgrade15, HRUpgrade16, HRUpgrade17, HRUpgrade18, HRUpgrade19, HRUpgrade20}
