
using { /Fortnite.com/Devices }
using { /Verse.org/Simulation }
using { /UnrealEngine.com/Temporary/Diagnostics }

#Added
using { /UnrealEngine.com/Temporary/SpatialMath}
using { /Verse.org/Assets }
using { /Verse.org/Simulation/Tags }

##HOW TO USE##
#When a vendor is setup you must add it to the V_PlayerManagerDevice in Editor look for PlayerUpgradeVendors an array
PlayerUpgradeTriggerHandler := class() {
    Device : V_PlayerUpgradeVendorDevice
    UpgradeType : PlayerUpgradeType
    PlayerRef : []player

    HandlerFunction(InMaybeAgent : ?agent) : void = {
        if(Player := player[InMaybeAgent?]):
            if(PlayerRef = array{Player}):
                Device.PurchasePlayerUpgrade(Player, UpgradeType)
                Print("HANDLER")
    }
}

UpgradeVendorUITag := class(tag){}
UpgradeAccoladeTag := class(tag){}

V_PlayerUpgradeVendorDevice := class(creative_device):
    @editable
    UpgradeTitleToShow : PlayerUpgradeTitleToShow = PlayerUpgradeTitleToShow.Null
    @editable
    NeedFirstUpgrade : logic = false
    @editable
    NPCReference : character_device = character_device{}
    @editable
    UpgradesToDisplay : []PlayerUpgradeType = array{}
    @editable
    UpgradePurchaseTriggers : []trigger_device = array{}
    @editable
    UseFortniteResource : logic = false
    @editable
    UsePrestigeResource : logic = false
    @editable
    OptConditionalButton : ?conditional_button_device = false

    @editable
    OptDamagePowerUps : []damage_amplifier_powerup_device = array{}
    @editable
    OptMovementModulators : []movement_modulator_device = array{}
    @editable
    OptAutoMinerManger : ?V_AutoMinerZoneManager = false

    @editable
    CanUpgradeHUDMessage : ?hud_message_device = false

    @editable
    PlayerManager : ?V_PlayerManagerDevice = false
    @editable
    LevelDevice : ?V_LevelUpAndCardUpgradeSelection = false

    #@editable
    UIPlayerUpgradeVendor : UI_PlayerUpgradeVendor = UI_PlayerUpgradeVendor{}

    # Manages and updates stats for each player.
    var PlayerStatManager: player_stats_manager = player_stats_manager{}
    
    # Runs when the device is started in a running game
    OnBegin<override>()<suspends>:void=
        NPCReference.InteractedWithEvent.Subscribe(InteractWithNPCAndDisplayVendorUI)
        spawn:
            CheckCanUpgrade()
    
    #We currently have to add upgrade vendors to the PlayerManagerDevice array
    InitiateVendorUI(InPlayer : player):void=
        for(Index -> UpgradeType : UpgradesToDisplay):
            if(UpgradeType = PlayerUpgradeType.Movement):
                Print("InitiateVendorUI")
            case(UpgradeType):
                PlayerUpgradeType.Damage =>
                    if(FirstDamageLevel := OptDamagePowerUps[0]):
                        PlayerStatManager.UpgradesData.SetDamageDevices(OptDamagePowerUps)
                PlayerUpgradeType.Movement =>
                    if(FirstMovementLevel := OptMovementModulators[0]):
                        PlayerStatManager.UpgradesData.SetMovementDevices(OptMovementModulators)
                _ =>    
        Items := MakeItemData(InPlayer, true)

        var TextureToShow : texture = VerseExposed.T_UpgradeVendorTemp
        case(UpgradeTitleToShow):
            PlayerUpgradeTitleToShow.Critical =>
                set TextureToShow = VerseExposed.T_UpgradeVendorTemp
            PlayerUpgradeTitleToShow.FinalHit =>
                set TextureToShow = VerseExposed.T_CriticalFinal
            PlayerUpgradeTitleToShow.Friends =>
                set TextureToShow = VerseExposed.T_Friends
            PlayerUpgradeTitleToShow.Movement =>
                set TextureToShow = VerseExposed.T_MovementUpgrade
            PlayerUpgradeTitleToShow.Health =>
                set TextureToShow = VerseExposed.T_HealthUpgrades
            PlayerUpgradeTitleToShow.ExplosiveHit =>
                set TextureToShow = VerseExposed.T_ExplosiveHit
            PlayerUpgradeTitleToShow.Altar =>
                set TextureToShow = VerseExposed.T_AltarUpgrades
            PlayerUpgradeTitleToShow.Prestige_ArtifactMastery =>
                set TextureToShow = VerseExposed.T_Artifact_Upgrades
            PlayerUpgradeTitleToShow.AutoMining =>
                set TextureToShow = VerseExposed.T_AutoMiner
            PlayerUpgradeTitleToShow.DungeonKing =>
                set TextureToShow = VerseExposed.T_DungeonKing
            PlayerUpgradeTitleToShow.AFKUpgrades =>
                set TextureToShow = VerseExposed.T_AFKUpgrades
            PlayerUpgradeTitleToShow.DungeonMastery =>
                set TextureToShow = VerseExposed.T_DungeonMastery
            PlayerUpgradeTitleToShow.MegaCriticalHit =>
                set TextureToShow = VerseExposed.T_MegaCriticalHit
            PlayerUpgradeTitleToShow.MiningAscension =>
                set TextureToShow = VerseExposed.T_MiningAscension
            PlayerUpgradeTitleToShow.OreUpgrades =>
                set TextureToShow = VerseExposed.T_OreUpgrades
            _ =>
                #Print("PlayerUpgradeTitleToShow Not Set")

        UIPlayerUpgradeVendor.InitUI(InPlayer, Items, TextureToShow)

    UpdateVendorUI(InPlayer : player):void=
        Items := MakeItemData(InPlayer, false)
        Costs := GetCostsArray(InPlayer)
        if(PlayerData := PlayerStatManager.GetPlayerStats[InPlayer]):
            UIPlayerUpgradeVendor.UpdateUI(InPlayer, Items, PlayerData, Costs, UseFortniteResource, UsePrestigeResource, NeedFirstUpgrade)

    MakeItemData(InPlayer : player, ShouldSubscribe : logic):[]PlayerUpgradeItemData=
        var Items : []PlayerUpgradeItemData = array{}
        for(Index -> UpgradeType : UpgradesToDisplay):
            if(CurrentTrigger := UpgradePurchaseTriggers[Index]):
                Upgrade := PlayerStatManager.GetPlayerUpgrade(InPlayer, UpgradeType)
                if(ShouldSubscribe = true):
                    if(UpgradeType = PlayerUpgradeType.Movement):
                        Print("Setup Trigger")
                    CurrentTrigger.TriggeredEvent.Subscribe(PlayerUpgradeTriggerHandler{Device := Self, UpgradeType := UpgradeType, PlayerRef := array{InPlayer}}.HandlerFunction)

                CurrentLevel := PlayerStatManager.GetPlayerUpgradeLevel(InPlayer, UpgradeType)
                CurrentBonus := Upgrade.GetCurrentLevelAmount(CurrentLevel)
                NextBonus := Upgrade.GetCurrentLevelAmount(CurrentLevel + 1)
                if(Cost := Upgrade.Costs[CurrentLevel + 1]):
                    ItemData := MakePlayerUpgradeItem(Upgrade.Image, CurrentLevel, "Lv. {CurrentLevel}", CurrentBonus, NextBonus, Upgrade.UpgradeName, Upgrade.UpgradeDescription, BigNumber_FormatDisplay(Cost), CurrentTrigger, Upgrade.DisplayPercentage)
                    set Items += array{ItemData}
                else:
                    ItemData := MakePlayerUpgradeItem(Upgrade.Image, CurrentLevel, "Lv. {CurrentLevel}", CurrentBonus, NextBonus, Upgrade.UpgradeName, Upgrade.UpgradeDescription, "MAXED", CurrentTrigger, Upgrade.DisplayPercentage)
                    set Items += array{ItemData}
        return Items

    GetCostsArray(InPlayer : player):[]big_number=
        var OutCosts : []big_number = array{}
        for(Index -> UpgradeType : UpgradesToDisplay):
            if(CurrentTrigger := UpgradePurchaseTriggers[Index]):
                Upgrade := PlayerStatManager.GetPlayerUpgrade(InPlayer, UpgradeType)
                
                CurrentLevel := PlayerStatManager.GetPlayerUpgradeLevel(InPlayer, UpgradeType)
                if(Cost := Upgrade.Costs[CurrentLevel + 1]):
                    set OutCosts += array{Cost}
                else:
                    set OutCosts += array{big_number{}}
        return OutCosts

    InteractWithNPCAndDisplayVendorUI(InAgent : agent):void=
        Print("Bring up Player Vendor UI")
            if(InPlayer := player[InAgent]):
                Print("Cast to Player")
                if(OldData := PlayerStatsDataMap[InPlayer]):
                    if(CurrentPlayerStats := PlayerStatManager.GetPlayerStats[InPlayer]):
                        Print("All checks have succeded")
                        #Display UI to the player
                        UpdateVendorUI(InPlayer)
                        UIPlayerUpgradeVendor.AddUI(InPlayer)

    PurchasePlayerUpgrade(InAgent : agent, InUpgradeType : PlayerUpgradeType):void=
        Print("Purchase Upgrade")
        Upgrade := PlayerStatManager.GetPlayerUpgrade(InAgent, InUpgradeType)
        CurrentLevel := PlayerStatManager.GetPlayerUpgradeLevel(InAgent, InUpgradeType)
        if(UseFortniteResource = false and UsePrestigeResource = false):
            PlayerStoneAmount := PlayerStatManager.GetPlayerStoneOreAmount(InAgent)
            if(Cost := Upgrade.Costs[CurrentLevel + 1]):
                BNCost := Cost
                CanAfford := BigNumber_GreaterThanOrEquals(PlayerStoneAmount, BNCost)
                if(CanAfford = true):
                    #Accolade for Unlocking Upgrade
                    TaggedList := FindCreativeObjectsWithTag(UpgradeAccoladeTag{})
                    if(Player := player[InAgent]):
                        for(TaggedActor : TaggedList):
                            if(Accolade := accolades_device[TaggedActor]):
                                Accolade.Award(Player)
                    #End Of Accolade Unlock
                    for(Index -> UpgradeType : UpgradesToDisplay):
                        if(UpgradeType = InUpgradeType):
                            if(NeedFirstUpgrade = true):
                                if(Index <> 0):
                                    # check if the first upgrade type is higher then level 1
                                    if(UpgradeType1 := UpgradesToDisplay[0]):
                                        UpgradeType1Level := PlayerStatManager.GetPlayerUpgradeLevel(InAgent, UpgradeType1)
                                        if(UpgradeType1Level > 0):
                                            PlayerStatManager.SpendPlayerOre(InAgent, BNCost, false, PropResourceType.Stone)
                                            PlayerStatManager.UpdatePlayerUpgrade(InAgent, InUpgradeType, 1)
                                            if(bPlayerManager := PlayerManager?):
                                                bPlayerManager.ActivatePlayerDeviceUpgrade(InAgent, InUpgradeType)
                                            if(bAutoMinerManger := OptAutoMinerManger?, Player := player[InAgent]):
                                                bAutoMinerManger.UpgradePlayerPlot(Player, InUpgradeType)
                                            if(Player := player[InAgent]):
                                                UpdateVendorUI(Player)
                                                if(bLevelDevice := LevelDevice?):
                                                    spawn:
                                                        PlayerStatManager.AddXPToPlayer(Player, big_number{Value := 1.0, Exponent := 2}, bLevelDevice)
                                            return
                                else:
                                    PlayerStatManager.SpendPlayerOre(InAgent, BNCost, false, PropResourceType.Stone)
                                    PlayerStatManager.UpdatePlayerUpgrade(InAgent, InUpgradeType, 1)
                                    if(bPlayerManager := PlayerManager?):
                                        bPlayerManager.ActivatePlayerDeviceUpgrade(InAgent, InUpgradeType)
                                    if(bAutoMinerManger := OptAutoMinerManger?, Player := player[InAgent]):
                                        bAutoMinerManger.UpgradePlayerPlot(Player, InUpgradeType)
                                    if(Player := player[InAgent]):
                                        UpdateVendorUI(Player)
                                        if(bLevelDevice := LevelDevice?):
                                            spawn:
                                                PlayerStatManager.AddXPToPlayer(Player, big_number{Value := 1.0, Exponent := 2}, bLevelDevice)
                                    return
                            else:
                                PlayerStatManager.SpendPlayerOre(InAgent, BNCost, false, PropResourceType.Stone)
                                PlayerStatManager.UpdatePlayerUpgrade(InAgent, InUpgradeType, 1)
                                if(bPlayerManager := PlayerManager?):
                                    bPlayerManager.ActivatePlayerDeviceUpgrade(InAgent, InUpgradeType)
                                if(bAutoMinerManger := OptAutoMinerManger?, Player := player[InAgent]):
                                    bAutoMinerManger.UpgradePlayerPlot(Player, InUpgradeType)
                                if(Player := player[InAgent]):
                                    UpdateVendorUI(Player)
                                    if(bLevelDevice := LevelDevice?):
                                        spawn:
                                            PlayerStatManager.AddXPToPlayer(Player, big_number{Value := 1.0, Exponent := 2}, bLevelDevice)
                                return
                else:
        else if(UsePrestigeResource = true):
            PlayerResourceAmount := PlayerStatManager.GetPlayerOreAmount(InAgent, PropResourceType.PrestigeResource)
            if(Cost := Upgrade.Costs[CurrentLevel + 1]):
                BNCost := Cost
                CanAfford := BigNumber_GreaterThanOrEquals(PlayerResourceAmount, BNCost)
                if(CanAfford = true):
                    for(Index -> UpgradeType : UpgradesToDisplay):
                        if(UpgradeType = InUpgradeType):
                            if(NeedFirstUpgrade = true):
                                if(Index <> 0):
                                    # check if the first upgrade type is higher then level 1
                                    if(UpgradeType1 := UpgradesToDisplay[0]):
                                        UpgradeType1Level := PlayerStatManager.GetPlayerUpgradeLevel(InAgent, UpgradeType1)
                                        if(UpgradeType1Level > 0):
                                            PlayerStatManager.SpendPlayerOre(InAgent, BNCost, false, PropResourceType.PrestigeResource)
                                            PlayerStatManager.UpdatePlayerUpgrade(InAgent, InUpgradeType, 1)
                                            if(bPlayerManager := PlayerManager?):
                                                bPlayerManager.ActivatePlayerDeviceUpgrade(InAgent, InUpgradeType)
                                            if(bAutoMinerManger := OptAutoMinerManger?, Player := player[InAgent]):
                                                bAutoMinerManger.UpgradePlayerPlot(Player, InUpgradeType)
                                            if(Player := player[InAgent]):
                                                UpdateVendorUI(Player)
                                                if(bLevelDevice := LevelDevice?):
                                                    spawn:
                                                        PlayerStatManager.AddXPToPlayer(Player, big_number{Value := 1.0, Exponent := 2}, bLevelDevice)
                                        return
                                else:
                                    PlayerStatManager.SpendPlayerOre(InAgent, BNCost, false, PropResourceType.PrestigeResource)
                                    PlayerStatManager.UpdatePlayerUpgrade(InAgent, InUpgradeType, 1)
                                    if(bPlayerManager := PlayerManager?):
                                        bPlayerManager.ActivatePlayerDeviceUpgrade(InAgent, InUpgradeType)
                                    if(bAutoMinerManger := OptAutoMinerManger?, Player := player[InAgent]):
                                        bAutoMinerManger.UpgradePlayerPlot(Player, InUpgradeType)
                                    if(Player := player[InAgent]):
                                        UpdateVendorUI(Player)
                                        if(bLevelDevice := LevelDevice?):
                                            spawn:
                                                PlayerStatManager.AddXPToPlayer(Player, big_number{Value := 1.0, Exponent := 2}, bLevelDevice)
                                    return
                            else:
                                PlayerStatManager.SpendPlayerOre(InAgent, BNCost, false, PropResourceType.PrestigeResource)
                                PlayerStatManager.UpdatePlayerUpgrade(InAgent, InUpgradeType, 1)
                                if(bPlayerManager := PlayerManager?):
                                    bPlayerManager.ActivatePlayerDeviceUpgrade(InAgent, InUpgradeType)
                                if(bAutoMinerManger := OptAutoMinerManger?, Player := player[InAgent]):
                                    bAutoMinerManger.UpgradePlayerPlot(Player, InUpgradeType)
                                if(Player := player[InAgent]):
                                    UpdateVendorUI(Player)
                                    if(bLevelDevice := LevelDevice?):
                                        spawn:
                                            PlayerStatManager.AddXPToPlayer(Player, big_number{Value := 1.0, Exponent := 2}, bLevelDevice)
                                return

    CheckCanUpgrade()<suspends>:void=
        Print("Checking if players can afford the Upgrades")
        loop:
            if(not CanUpgradeHUDMessage?):
                break
                
            AllPlayers := GetPlayspace().GetPlayers()
            for(Player : AllPlayers):
                for(UpgradeType : UpgradesToDisplay):
                    Upgrade := PlayerStatManager.GetPlayerUpgrade(Player, UpgradeType)
                    CurrentLevel := PlayerStatManager.GetPlayerUpgradeLevel(Player, UpgradeType)
                    var PlayerResourceAmount : big_number = big_number{}

                    if(UsePrestigeResource = true):
                        set PlayerResourceAmount = PlayerStatManager.GetPlayerOreAmount(Player, PropResourceType.PrestigeResource)
                    else:
                        set PlayerResourceAmount = PlayerStatManager.GetPlayerOreAmount(Player, PropResourceType.Stone)

                    if(Cost := Upgrade.Costs[CurrentLevel + 1]):
                        CanAfford := BigNumber_GreaterThanOrEquals(PlayerResourceAmount, Cost)
                        if(CanAfford = true, bCanUpgradeHUDMessage := CanUpgradeHUDMessage?):
                            bCanUpgradeHUDMessage.Show(Player)

            Sleep(5.0)
