using { /Fortnite.com/Devices }
using { /Verse.org/Simulation }
using { /UnrealEngine.com/Temporary/Diagnostics }

V_PlaytimeManager := class(creative_device):
    # Update 1.1 - Simplified the class

    @editable
    PlayerManager : ?V_PlayerManagerDevice = false

    var PlayerStatManager: player_stats_manager = player_stats_manager{}

    #Starts tracking the playtime per player
    StartTracking(InPlayer : player)<suspends> : void=
        loop:
            Print("Tracking")
            if(not InPlayer.IsActive[]):
                Print("Player Not Active")
                break
            Sleep(60.0)
            MinutePassed(InPlayer)

    #Called when the timer is complete
    MinutePassed(InPlayer : player) : void=
        #Print("PlaytimeManager - Timer Complete")
        if(bPlayerManager := PlayerManager?):
            NotSaved := bPlayerManager.GetPlayerDataNotSaved(InPlayer)
            if(PlayerStats := PlayerStatManager.GetPlayerStats[InPlayer]):

                set NotSaved.PlaytimeSessionMinutes += 1

                if(PlayerStats.MinutesPlayed >= 60.0):
                    UpdateHours(InPlayer)
                else:
                    UpdateMinutes(InPlayer)

                if(NotSaved.PlaytimeSessionMinutes >= 60):
                    set NotSaved.PlaytimeSessionHours += 1
                    if(NotSaved.PlaytimeSessionHours = 2):
                        #Quest Complete set 0
                        bPlayerManager.QuestFCHQ.QuestCompleted.Signal((InPlayer, "Touch Grass"))
                        set NotSaved.PlaytimeSessionHours = 0

                Print("Minutes Passed : {NotSaved.PlaytimeSessionMinutes}", ?Duration:=6.0)
                if(NotSaved.PlaytimeSessionMinutes = 60 or NotSaved.PlaytimeSessionMinutes = 30):
                    bPlayerManager.QuestFCHQ.QuestCompleted.Signal((InPlayer, "Casually Committed"))
                    if(NotSaved.PlaytimeSessionMinutes = 60):
                        set NotSaved.PlaytimeSessionMinutes = 0

    #Updates Player's Minutes in Total Playtimes
    UpdateMinutes(InPlayer : player) : void =
        #Print("PlaytimeManager - Updated Minutes")
        if(PlayerStats := PlayerStatManager.GetPlayerStats[InPlayer]):
            NewMinute := PlayerStats.MinutesPlayed + 1.0
            PlayerStatManager.SetPlaytimeMinutes(InPlayer, NewMinute)

    #Updates Player's Hours in Total Playtime
    UpdateHours(InPlayer : player) : void =
        #Print("PlaytimeManager - Updated Hours")
        if(PlayerStats := PlayerStatManager.GetPlayerStats[InPlayer]):
            NewHour := PlayerStats.HoursPlayed + 1.0
            PlayerStatManager.SetPlaytimeHours(InPlayer, NewHour)
            PlayerStatManager.SetPlaytimeMinutes(InPlayer, 0.0)