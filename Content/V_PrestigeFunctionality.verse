
using { /Fortnite.com/Devices }
using { /Verse.org/Simulation }
using { /UnrealEngine.com/Temporary/Diagnostics }

using { /Verse.org/Simulation/Tags }

# See https://dev.epicgames.com/documentation/en-us/uefn/create-your-own-device-in-verse for how to create a verse device.
PrestigeTag := class(tag):
ForcefieldVendorTag := class(PrestigeTag):

# A Verse-authored creative device that can be placed in a level
V_PrestigeFunctionality := class(creative_device):
    # TODO: PRESTIGE CHECKLIST
    # - UI

    #Class Goal
        #Get PlayerStatsData Reset [This will handle all player data, vendors etc..]

    #Trigger to tie devices together
    @editable
    PrestigeTrigger : trigger_device = trigger_device{}
    @editable
    NPCCharacter : character_device = character_device{}
    @editable
    UIManager : ?V_UIManagerDevice = false
    @editable
    PrestigeVFXTrigger : trigger_device = trigger_device{}
    @editable
    PlayerManager : ?V_PlayerManagerDevice = false

    PrestigeBonusMultiplier : float = 0.00001
    #Array of all the forcefield vendors in the game, we must make sure each one is tagged "ForcefieldVendorTag"
    var ForceFieldVendors : []V_ForceFieldVendor = array{}
    # Manages and updates stats for each player.
    var PlayerStatManager: player_stats_manager = player_stats_manager{}
    UIPrestigeMenu : UI_PrestigeMenu = UI_PrestigeMenu{}

    # Runs when the device is started in a running game
    OnBegin<override>()<suspends>:void=
        SetupPrestigeDevices()
        PrestigeTrigger.TriggeredEvent.Subscribe(PlayerPrestige)
        NPCCharacter.InteractedWithEvent.Subscribe(OnNPCInteract)

    InitiatePlayerUI(InPlayer : player):void=
        UIPrestigeMenu.InitUI(InPlayer, PrestigeTrigger)

    SetupPrestigeDevices():void=
        #Create lists on begin so this does not have to occur during gameplay and potentially cause a performance issue for players
        TaggedList1 := FindCreativeObjectsWithTag(ForcefieldVendorTag{})
                for(TaggedActor : TaggedList1):
                    if(ForceFieldVendorClass := V_ForceFieldVendor[TaggedActor]):
                        set ForceFieldVendors += array{ForceFieldVendorClass}

    GetPrestigeNeededAmount(InPlayer : player):big_number=
        var PrestigeNeeded : big_number = big_number{}
        if(PlayerData := PlayerStatManager.GetPlayerStats[InPlayer]):
            if(PlayerData.Prestige <> 0):
                set PrestigeNeeded = BigNumber_MultiplyWithAFloat(PlayerData.LastPrestigedAmount, 10.0)
            else:
                set PrestigeNeeded = big_number{Value := 1.5, Exponent := 16}

            if(PlayerData.Prestige > 4):
                set PrestigeNeeded = BigNumber_MultiplyWithAFloat(PlayerData.LastPrestigedAmount, 1.1)
                
        return PrestigeNeeded

    GetPrestigeBonusAmount(InPlayer : player):big_number=
        var PrestigeBonusAmount : big_number = big_number{}
        if(PlayerData := PlayerStatManager.GetPlayerStats[InPlayer]):
            PlayerTotalStoneAmount := PlayerData.TotalStoneOreAmount
            NeededAmount := GetPrestigeNeededAmount(InPlayer)
            
            CanPrestige := BigNumber_GreaterThanOrEquals(PlayerTotalStoneAmount, NeededAmount)
            if(CanPrestige = true):
                LogNumber := (PlayerData.TotalStoneSincePrestige.Exponent * 1.0) + (PlayerData.TotalStoneSincePrestige.Value / 10.0)
                DividedNumber := LogNumber / 3.0

                PowNumber := Pow(10.0, DividedNumber)
                set PrestigeBonusAmount = BigNumber_ConvertFloatIntoBigNumber(PowNumber / 5000.0)
            else:
                set PrestigeBonusAmount = big_number{Value := 0.0, Exponent := 0}
        return PrestigeBonusAmount

    PlayerPrestige(InAgent : ?agent):void=
        if(bInAgent := InAgent?):
            if(Player := player[bInAgent], PlayerStats := PlayerStatManager.GetPlayerStats[Player]):
                #Reset Player Stats
                PlayerTotalStoneAmount := PlayerStats.TotalStoneOreAmount
                LastPrestigeAmount := PlayerStats.LastPrestigedAmount
                NeededAmount := GetPrestigeNeededAmount(Player)
                
                CanPrestige := BigNumber_GreaterThanOrEquals(PlayerTotalStoneAmount, NeededAmount)
                if(CanPrestige = true):
                    PlayerStatManager.UpdatePrestigeBonusHarvestRate(bInAgent, BigNumber_ConvertBigNumberIntoFloat(GetPrestigeBonusAmount(Player)))
                    PlayerStatManager.UpdatePlayerOreAmount(bInAgent, PropResourceType.PrestigeResource, GetPrestigeBonusAmount(Player))
                    PlayerStatManager.UpdateLastPrestigeAmount(bInAgent, PlayerTotalStoneAmount)
                    PlayerStatManager.UpdatePrestige(bInAgent)
                    PlayerStatManager.ResetStats(bInAgent)
                    #Hides UI when you click Prestige
                    spawn:
                        UIPrestigeMenu.HideUI(Player)
                    if(bPlayerManager := PlayerManager?):
                        bPlayerManager.QuestFCHQ.QuestCompleted.Signal((Player, "Ascended"))
                    #VFX for Prestige. This also calls a teleport function.
                    PrestigeVFXTrigger.Trigger(bInAgent)
                    #Reset all forcefield vendors
                    for(ForceFieldVendor : ForceFieldVendors):
                        Print("Resetting Forcefield Vendors")
                        ForceFieldVendor.ResetVendorOnPrestige(bInAgent)

            if(Player := player[bInAgent], PlayerStats := PlayerStatManager.GetPlayerStats[Player]):
                PlayerTotalStoneAmount := PlayerStats.TotalStoneOreAmount
                LastPrestigeAmount := PlayerStats.LastPrestigedAmount
                NeededAmount := GetPrestigeNeededAmount(Player)

                CanPrestige := BigNumber_GreaterThanOrEquals(PlayerTotalStoneAmount, NeededAmount)
                UIPrestigeMenu.UpdateUI(Player, PlayerStats.TotalStoneOreAmount, GetPrestigeBonusAmount(Player), GetPrestigeBonusAmount(Player), NeededAmount, CanPrestige)
                if(bUIManager := UIManager?):
                    spawn:
                        bUIManager.UpdatePlayerUI(Player)

        Print("Prestige and Reset all Progress")

    OnNPCInteract(InAgent : agent):void=
        OpenPrestigeMenu(InAgent)

    OpenPrestigeMenu(InAgent : agent):void=
        if(Player := player[InAgent], PlayerStats := PlayerStatManager.GetPlayerStats[Player]):
            PlayerTotalStoneAmount := PlayerStats.TotalStoneOreAmount
            LastPrestigeAmount := PlayerStats.LastPrestigedAmount
            NeededAmount := GetPrestigeNeededAmount(Player)
            
            CanPrestige := BigNumber_GreaterThanOrEquals(PlayerTotalStoneAmount, NeededAmount)

            UIPrestigeMenu.UpdateUI(Player, PlayerStats.TotalStoneOreAmount, GetPrestigeBonusAmount(Player), GetPrestigeBonusAmount(Player), NeededAmount, CanPrestige)
            UIPrestigeMenu.AddUI(Player)