using { /Fortnite.com/Devices }
using { /Verse.org/Simulation }
using { /UnrealEngine.com/Temporary/Diagnostics }

V_PropDestructionManager := class(creative_device):

    @editable
    QuestTracker : tracker_device = tracker_device{}

    @editable
    PropManipulator : ?prop_manipulator_device = false

    @editable
    ExplosiveDevice : ?explosive_device = false

    @editable
    PropMaxHealth : float = 20.0

    @editable
    DamagePerHit : float = 10.0

    @editable
    RecoveryTime : float = 10.0

    var PropCurrentHealth : float = 0.0

    OnBegin<override>()<suspends>:void=
        #Subscribes to damaged event
        if(Prop := PropManipulator?):
            Prop.DamagedEvent.Subscribe(OnPropDamaged)

        if(Objective := ExplosiveDevice?):
            Objective.ExplodedEvent.Subscribe(OnObjectiveDestroyed)

        set PropCurrentHealth = PropMaxHealth

    #Called when the prop is damaged
    OnPropDamaged(Agent : agent) : void=
        #Subtracts the health of the prop
        set PropCurrentHealth = PropCurrentHealth - DamagePerHit

        #Checks if the prop has been destroyed
        if(PropCurrentHealth <= 0.0):
            QuestTracker.Increment(Agent)
            if(Prop := PropManipulator?):
                Prop.HideProps()

            spawn:
                PropRecoveryTimer()
    
    OnObjectiveDestroyed(Agent : agent) : void=
        QuestTracker.Increment(Agent)
        spawn:
            ObjectiveRecoveryTimer()

    #Timer for objective to reset
    ObjectiveRecoveryTimer()<suspends> : void=
        Sleep(RecoveryTime)
        ResetObjective()

    #Timer for prop to reset
    PropRecoveryTimer()<suspends> : void=
        Sleep(RecoveryTime)
        ResetProp()

    #Resets the props values
    ResetProp() : void=
        set PropCurrentHealth = PropMaxHealth

        if(Prop := PropManipulator?):
            Prop.ShowProps()

    #Resets the objective
    ResetObjective() : void=
        if(Objective := ExplosiveDevice?):
            Objective.Reset()