
using { /Fortnite.com/Devices }
using { /Verse.org/Simulation }
using { /UnrealEngine.com/Temporary/Diagnostics }
#Added
using { /UnrealEngine.com/Temporary/SpatialMath }
using { /Fortnite.com/Playspaces }
using { /Fortnite.com/Characters }
using { /Fortnite.com/Devices/CreativeAnimation }

# See https://dev.epicgames.com/documentation/en-us/uefn/create-your-own-device-in-verse for how to create a verse device.

# A Verse-authored creative device that can be placed in a level
V_PropFollower := class(creative_device):
    @editable
    PropFollowerBlueprints : []creative_prop_asset = array{}
    @editable
    DebugButton : button_device = button_device{}
    @editable
    PlayerManager : ?V_PlayerManagerDevice = false

    @editable
    UnequipButton : button_device = button_device{}

    #Prop Movement Values
    var PropSpeed : float = 10.0

    # Runs when the device is started in a running game
    OnBegin<override>()<suspends>:void=
        DebugButton.InteractedWithEvent.Subscribe(DebugTestFollowing)
        UnequipButton.InteractedWithEvent.Subscribe(UnequipButtonPressed)

    DebugTestFollowing(InAgent : agent):void=
        Print("Spawning Prop")
        if(SpawnProp1 := PropFollowerBlueprints[0], SpawnProp2 := PropFollowerBlueprints[1], SpawnProp3 := PropFollowerBlueprints[2],
        SpawnProp4 := PropFollowerBlueprints[3], SpawnProp5 := PropFollowerBlueprints[4] ):
            SpawnPropFollower(InAgent,SpawnProp1, 1)
            SpawnPropFollower(InAgent,SpawnProp2, 2)
            SpawnPropFollower(InAgent,SpawnProp3, 3)
            SpawnPropFollower(InAgent,SpawnProp4, 4)
            SpawnPropFollower(InAgent,SpawnProp5, 5)

    SpawnPropWithRarity(InAgent : agent, InArtifactRarityInt : int, InSlot : int):void=
        Print("Spawning Prop")
        if(CurrentSpawnProp := PropFollowerBlueprints[InArtifactRarityInt]):
            SpawnPropFollower(InAgent, CurrentSpawnProp, InSlot)

    #Spawn the prop follower
    SpawnPropFollower(InAgent : agent, PropToSpawn : creative_prop_asset, FollowSlot : int):void=
        if(FortChar := InAgent.GetFortCharacter[], bPlayerManager := PlayerManager?):
            PlayerPosition := FortChar.GetTransform().Translation
            PlayerBackward := FortChar.GetViewRotation().GetLocalForward() * -300.0
            PropSpawnPosition := PlayerPosition + PlayerBackward
            Result := SpawnProp(PropToSpawn, PlayerPosition, rotation{})
            Print("Prop Spawned")
            if(Player := player[InAgent], DataNotSaved := bPlayerManager.PlayerMap[Player]):
                if(SpawnedProp := Result(0)?):
                    for(FollowerProp : DataNotSaved.FollowingProps):
                        if(FollowerProp.Slot = FollowSlot):
                            DisposeProp(Player, FollowSlot)

                    set DataNotSaved.FollowingProps += array{FollowingProp{Prop := SpawnedProp, Slot := FollowSlot}}

                    case(FollowSlot):
                        1=>
                            spawn:
                                FollowPlayer(FortChar, SpawnedProp, 1)
                        2=>
                            spawn:
                                FollowPlayer(FortChar, SpawnedProp, 2)
                        3=>
                            spawn:
                                FollowPlayer(FortChar, SpawnedProp, 3)
                        4=>           
                            spawn:
                                FollowPlayer(FortChar, SpawnedProp, 4)
                        5=>
                            spawn:
                                FollowPlayer(FortChar, SpawnedProp, 5)
                        _=>

    FollowPlayer(FortChar : fort_character, CreativeProp : creative_prop, FollowSlot : int)<suspends>:void=
        Print("Following the player")
        loop:
            if(not CreativeProp.IsValid[]):
                break
            PlayerLocationRef : vector3 = FortChar.GetTransform().Translation              # Get the players location
            var FollowerPropLocation : vector3 = CreativeProp.GetTransform().Translation   # Get the prop
            var TargetOffset : vector3 = -FortChar.GetViewRotation().GetLocalRight()       # Get the players view rotation
            var XOffsetAmount : float = 200.0
            var YOffsetAmount : float = 100.0
            #Only move once the player is at a certain distance away
            if(Distance(FollowerPropLocation, PlayerLocationRef) > 200.0):
                #Slots from where the props can follow you
                case(FollowSlot):
                    1 =>
                        #Print("Left")
                        set TargetOffset = -FortChar.GetViewRotation().GetLocalRight()
                    2 =>
                        #Print("Left Middle")
                        set TargetOffset = -FortChar.GetViewRotation().GetLocalRight() 
                        set XOffsetAmount = 100.0
                    3 =>
                        #Print("Right")
                        set TargetOffset = FortChar.GetViewRotation().GetLocalRight()
                    4 =>
                        #Print("Right Middle")
                        set TargetOffset = FortChar.GetViewRotation().GetLocalRight()
                        set XOffsetAmount = 100.0
                    5 =>
                        #Print("Middle")
                        set TargetOffset = -FortChar.GetViewRotation().GetLocalForward()
                    _=>

                if:
                    AnimationController := CreativeProp.GetAnimationController[]    # Get reference to animation controller
                then:

                    NoJumpPlayerPos := vector3{X:= PlayerLocationRef.X + TargetOffset.X * XOffsetAmount * 0.75, Y:= PlayerLocationRef.Y + TargetOffset.Y * YOffsetAmount * 0.75, Z := PlayerLocationRef.Z}
                    TargetPosition := NoJumpPlayerPos - FollowerPropLocation                        # Get the target position for the pet to move to

                    Rotation := FortChar.GetViewRotation().ApplyYaw(-90.0)
                    YawPitchRoll := Rotation.GetYawPitchRollDegrees()
                    if(NewRotation := MakeRotationFromYawPitchRollDegrees(YawPitchRoll[0], 0.0, 0.0)):
                        TargetRotation := MakeShortestRotationBetween(CreativeProp.GetTransform().Rotation, NewRotation)        # Get the rotation for the pet to face the player

                        KeyFrame := keyframe_delta :                                                                            # Create a keyframe for the pet to move to the player
                            DeltaLocation := TargetPosition
                            DeltaRotation := TargetRotation
                            Time := 1.0
                        KeyFrames : []keyframe_delta = array { KeyFrame }                                                       # Create an array of keyframes
                        AnimationController.SetAnimation(KeyFrames, ?Mode:=animation_mode.OneShot)                              # Set the animation controller to move the pet to the player
                        AnimationController.Play()                                                                              # Play the animation controller    
            Sleep(1.05)

    SpawnPetFollower(InAgent : agent, PropToSpawn : creative_prop_asset):void=
        if(FortChar := InAgent.GetFortCharacter[], bPlayerManager := PlayerManager?):
            PlayerPosition := FortChar.GetTransform().Translation
            PlayerBackward := FortChar.GetViewRotation().GetLocalForward() * -300.0
            PropSpawnPosition := PlayerPosition + PlayerBackward
            Result := SpawnProp(PropToSpawn, PlayerPosition, rotation{})
            Print("Prop Spawned")
            if(Player := player[InAgent], DataNotSaved := bPlayerManager.PlayerMap[Player]):
                if(SpawnedProp := Result(0)?):
                    DisposePet(Player)
                    set DataNotSaved.Pet = SpawnedProp

                    spawn:
                        PetFollowPlayer(FortChar, SpawnedProp)

    PetFollowPlayer(FortChar : fort_character, CreativeProp : creative_prop)<suspends>:void=
        loop:
            if(not CreativeProp.IsValid[]):
                break
            PlayerLocationRef : vector3 = FortChar.GetTransform().Translation              # Get the players location
            var FollowerPropLocation : vector3 = CreativeProp.GetTransform().Translation   # Get the prop
            var TargetOffset : vector3 = -FortChar.GetViewRotation().GetLocalForward()     # Get the players view rotation
            var XOffsetAmount : float = 200.0
            var YOffsetAmount : float = 100.0
            var ZOffsetAmount : float = -80.0

            #Only move once the player is at a certain distance away
            if(Distance(FollowerPropLocation, PlayerLocationRef) > 250.0):
                #Slots from where the props can follow you
                if:
                    AnimationController := CreativeProp.GetAnimationController[]    # Get reference to animation controller
                then:

                    NoJumpPlayerPos := vector3{X:= PlayerLocationRef.X + TargetOffset.X * XOffsetAmount * 0.75, Y:= PlayerLocationRef.Y + TargetOffset.Y * YOffsetAmount * 0.75, Z := PlayerLocationRef.Z + ZOffsetAmount}
                    TargetPosition := NoJumpPlayerPos - FollowerPropLocation                        # Get the target position for the pet to move to

                    Rotation := FortChar.GetViewRotation().ApplyYaw(-90.0)
                    YawPitchRoll := Rotation.GetYawPitchRollDegrees()
                    if(NewRotation := MakeRotationFromYawPitchRollDegrees(YawPitchRoll[0], 0.0, 0.0)):
                        TargetRotation := MakeShortestRotationBetween(CreativeProp.GetTransform().Rotation, NewRotation)        # Get the rotation for the pet to face the player

                        KeyFrame := keyframe_delta :                                                                            # Create a keyframe for the pet to move to the player
                            DeltaLocation := TargetPosition
                            DeltaRotation := TargetRotation
                            Time := 1.0
                        KeyFrames : []keyframe_delta = array { KeyFrame }                                                       # Create an array of keyframes
                        AnimationController.SetAnimation(KeyFrames, ?Mode:=animation_mode.OneShot)                              # Set the animation controller to move the pet to the player
                        AnimationController.Play()                                                                              # Play the animation controller    
            Sleep(0.5)

    #We can call this in multiple places as a way to remove the prop from the world
    DisposePet(InPlayer : player):void=
        if(bPlayerManager := PlayerManager?, DataNotSaved := bPlayerManager.PlayerMap[InPlayer]):
                if(DataNotSaved.Pet.IsValid[]):
                    DataNotSaved.Pet.Dispose()
                    Print("Dispose The Pet")

    #We can call this in multiple places as a way to remove the prop from the world
    DisposeProp(InPlayer : player, InSlot : int):void=
        if(bPlayerManager := PlayerManager?, DataNotSaved := bPlayerManager.PlayerMap[InPlayer]):
            for(FollowerProp : DataNotSaved.FollowingProps):
                if(FollowerProp.Slot = InSlot, FollowerProp.Prop.IsValid[]):
                    FollowerProp.Prop.Dispose()
                    Print("Dispose The Prop")

    UnequipButtonPressed(InAgent : agent):void=
        if(Player := player[InAgent]):
            DisposePet(Player)