using { /Fortnite.com/Devices }
using { /Verse.org/Simulation }
using { /UnrealEngine.com/Temporary/Diagnostics }

#Added Tags
using { /UnrealEngine.com/Temporary/UI }
using { /Fortnite.com/UI }
using { /UnrealEngine.com/Temporary/SpatialMath}
using { /Verse.org/Colors }
using { /Verse.org/Assets }
using { /Verse.org/Simulation/Tags }
using { /Fortnite.com/Characters}
using { /Verse.org/Random }
#Added Animation 
using { /Fortnite.com/Devices/CreativeAnimation}
using { /Fortnite.com/Devices/CreativeAnimation/InterpolationTypes}


#Tag devices setup to expand for multiple zones
#Update 1.1 Critical Fix - We're using tags per cave so the begin check does not have to sort through hundreds of tagged actors
MiningPropBiomeAndCave := class(tag){}
tGodNode := class(tag){}
tEndGodNode := class(tag){}
tMushroomCaveOne1 :=class(tag){}
tMushroomCave2 :=class(tag){}
tMushroomCave3 :=class(tag){}
tMushroomCave4 :=class(tag){}

tLavaCave1 :=class(tag){}
tLavaCave2 :=class(tag){}
tLavaCave3 :=class(tag){}
tLavaCave4 :=class(tag){}

tUnderworldCave1 :=class(tag){}
tUnderworldCave2 :=class(tag){}
tUnderworldCave3 :=class(tag){}
tUnderworldCave4 :=class(tag){}

tMagicCave1 :=class(tag){}
tMagicCave2 :=class(tag){}
tMagicCave3 :=class(tag){}
tMagicCave4 :=class(tag){}
tMagicCave5 :=class(tag){}
tMagicCave6 :=class(tag){}
tMagicCave7 :=class(tag){}

OreHitAccolade := class(tag){}

PropHitAudioTags := class(tag){}
PropHitAudio := class(PropHitAudioTags){}
PropLastHitAudio := class(PropHitAudioTags){}
tPropDevice := class(tag){}


MiningNodeCave := enum:
    eGodNode,
    eMushroomCave1, 
    eMushroomCave2,
    eMushroomCave3,
    eMushroomCave4,
    eLavaCave1,
    eLavaCave2,
    eLavaCave3,
    eLavaCave4,
    eUnderworldCave1,
    eUnderworldCave2,
    eUnderworldCave3,
    eUnderworldCave4,
    eMagicCave1,
    eMagicCave2,
    eMagicCave3,
    eMagicCave4,
    eMagicCave5,
    eMagicCave6,
    eMagicCave7,
    Null

GetOreDropChance(InResourceType : PropResourceType)<transacts>:float=
    case(InResourceType):
        PropResourceType.Copper =>
            return 60.0
        PropResourceType.Silver =>
            return 25.0
        PropResourceType.Malachite =>
            return 12.5
        PropResourceType.Spectrolite =>
            return 2.0
        PropResourceType.Obsidian =>
            return 0.5
        _ =>
            return 0.0

GetOreTexture(InResourceType : PropResourceType):texture=
    case(InResourceType):
        PropResourceType.Copper =>
            return VerseExposed.T_CopperOreStroke
        PropResourceType.Silver =>
            return VerseExposed.T_SilverOre
        PropResourceType.Malachite =>
            return VerseExposed.T_MalachiteOreStroke
        PropResourceType.Spectrolite =>
            return VerseExposed.T_SpectroliteOreStroke
        PropResourceType.Obsidian =>
            return VerseExposed.T_ObsidianOreStroke
        _ =>
            return VerseExposed.T_CopperOreStroke


HarvestResourceNodeClass := class<concrete><unique>():
    var CreativePropToHit : creative_prop = creative_prop{}
    var PerNodeHarvestPropHealth : big_number = big_number{}
    var PerNodeHarvestPropMaxHealth : big_number = big_number{}
    var OriginalTransform : transform  = transform{}
    var HitVFX : creative_prop_asset = DefaultCreativePropAsset
    var LastHitVFX : creative_prop_asset = DefaultCreativePropAsset
    var PropBeingAnimated : logic = false

    var NodePropScale : vector3 = vector3{X:= 0.1, Y:= 0.1, Z:= 0.1}

# - - - - - - - - - - PROP TAG & HEALTH SETUP METHOD - - - - - - - - - - - # 
#We are now getting Props with Tags. 
#Simply place the props and tag them. Then put the prop manipulators in them and make sure they are assigned.
#We are also setting health per device now and not per Node.
    SetupNode(PropHealth : big_number, PropMaxHealth : big_number):void=           
        #NEW: Set the per Resource Node to the overall Prop Health defined in the CoreValues Class
        set PerNodeHarvestPropHealth = PropHealth
        set PerNodeHarvestPropMaxHealth = PropMaxHealth
        set OriginalTransform = CreativePropToHit.GetTransform()

        #NEW: Debug to test if the amount is correct
# - - - - - - - - - - PROP TAG & HEALTH SETUP METHOD END - - - - - - - - - - - # 

    TakeDamage<public>(DamageAmount : big_number):void=
        set PerNodeHarvestPropHealth = BigNumber_Subtract(PerNodeHarvestPropHealth, DamageAmount)
        #ScalePropDown()
    <#
    #TODO: Base on scale
    ScalePropDown<public>():void=
        PropScale := CreativePropToHit.GetTransform().Scale.X
        ScaleSmaller := vector3{X:= PropScale - 0.001, Y:= PropScale -0.001, Z:= PropScale - 0.001}
        Transform := transform{Translation := CreativePropToHit.GetTransform().Translation, Rotation := CreativePropToHit.GetTransform().Rotation , Scale := ScaleSmaller}
        if(CreativePropToHit.TeleportTo[Transform]):
    #>
    GetHarvestPropHealth<public>():big_number=
        return PerNodeHarvestPropHealth

    GetHarvestPropMaxHealth<public>():big_number=
        return PerNodeHarvestPropMaxHealth
        
    ResetHarvestNodeHealth<public>():void=
        set PerNodeHarvestPropHealth = PerNodeHarvestPropMaxHealth
    
    # - - - - - - - - - - SPECIAL EFFECTS METHOD - - - - - - - - - - - # 
    SpawnHitEffects<public>():void=
        #Spawn above the Prop Location
        SpawnPosition := vector3{X:= OriginalTransform.Translation.X, Y:= OriginalTransform.Translation.Y, Z:= OriginalTransform.Translation.Z + 40.0}
        SpawnScale := vector3{X:= 1.0, Y:= 1.0, Z:= 1.0}
        SpawnedProp := SpawnProp(HitVFX, transform{Translation := SpawnPosition, Rotation := IdentityRotation(), Scale:= SpawnScale})
        if(CreativeProp := SpawnedProp(0)?):
            spawn:
                DestroyEffects(CreativeProp, 1.5)

    SpawnLastHitEffects<public>():void=
        #Spawn above the Prop Location
        SpawnPosition := vector3{X:= OriginalTransform.Translation.X, Y:= OriginalTransform.Translation.Y, Z:= OriginalTransform.Translation.Z + 20.0}
        SpawnScale := vector3{X:= 1.0, Y:= 1.0, Z:= 1.0}
        SpawnedProp := SpawnProp(LastHitVFX, transform{Translation := SpawnPosition, Rotation := IdentityRotation(), Scale:= SpawnScale})
        if(CreativeProp := SpawnedProp(0)?):
            spawn:
                DestroyEffects(CreativeProp, 2.5)

    DestroyEffects(SpawnedProp : creative_prop, TimeToDispose : float)<suspends>:void=
        Sleep(TimeToDispose)
        SpawnedProp.Dispose()
    # - - - - - - - - - - SPECIAL EFFECTS END METHOD - - - - - - - - - - - # 

    # - - - - - - - - - - ANIMATION KNOCKBACK METHOD - - - - - - - - - - - # 
    KnockBackHarvestProp<public>(InAgent : agent, DistanceToHitGround : ?float):void=
        #Spawn Effects for the last hit
        SpawnLastHitEffects()
        #Update 1.1 - Setting it true here as I was setting it too early previously
        #set PropBeingAnimated = true
        if(FortChar := InAgent.GetFortCharacter[]):
            KnockbackDistance : float = 0.5
            KnockbackDirectionCalculation := CreativePropToHit.GetTransform().Translation - FortChar.GetTransform().Translation 
            KnockbackDirection := vector3{X:= KnockbackDirectionCalculation.X * 2.0, Y:= KnockbackDirectionCalculation.Y * 2.0, Z:= 60.0}

            if(bDistanceToHitTheGround := DistanceToHitGround?):
                MovementKeyFrame1 : keyframe_delta = keyframe_delta
                {
                    DeltaLocation := KnockbackDirection * 0.8
                    DeltaRotation := MakeRotationFromYawPitchRollDegrees(0.0, 0.0, 0.0)
                    DeltaScale := vector3 {X:=0.5, Y:= 0.5, Z:= 0.5}
                    Time := 0.55
                    Interpolation := EaseInOut
                }
                MovementKeyFrame2 : keyframe_delta = keyframe_delta
                {
                    #Moved onto the ground
                    DeltaLocation := vector3{Z := bDistanceToHitTheGround}
                    DeltaRotation := MakeRotationFromYawPitchRollDegrees(0.0, 0.0, 0.0)
                    Time := 0.3
                    Interpolation := EaseInOut
                }
                MovementKeyFrame3 : keyframe_delta = keyframe_delta
                {
                    #Scales away
                    DeltaLocation := vector3{X := 0.0 }
                    DeltaRotation := MakeRotationFromYawPitchRollDegrees(180.0, 0.0, 0.0)
                    DeltaScale := vector3{X:= 0.1, Y:= 0.1, Z:= 0.1}
                    Time := 0.6
                    Interpolation := EaseInOut
                }
                if(PropController := CreativePropToHit.GetAnimationController[]):
                    KnockbackKeyFrames:[]keyframe_delta = array{MovementKeyFrame1,MovementKeyFrame2,MovementKeyFrame3}  #Can extend the array of keyframes here
                    PropController.SetAnimation(KnockbackKeyFrames, ?Mode:=animation_mode.OneShot)
                    PropController.Play()
    # - - - - - - - - - - ANIMATION KNOCKBACK END METHOD - - - - - - - - - - - # 
            
    # - - - - - - - - - - ANIMATION RESET PROP METHOD - - - - - - - - - - - # 
    ResetHarvestProp<public>():void=
        MovementKeyFrame1 : keyframe_delta = keyframe_delta
        {
            #We make it go backwards, spins and scale down
            DeltaLocation :=  DeltaLocation := vector3{X := 0.0 }
            DeltaRotation := MakeRotationFromYawPitchRollDegrees(180.0, 0.0, 0.0)
            #Multiply by 20.0 to get the original scale
            DeltaScale := vector3{X:= 20.0 , Y:= 20.0 , Z:= 20.0 }
            Time := 1.0
            Interpolation := EaseInOut
        }

        if(PropController := CreativePropToHit.GetAnimationController[]):
            KnockbackKeyFrames:[]keyframe_delta = array{MovementKeyFrame1}  #Can extend the array of keyframes here
            PropController.SetAnimation(KnockbackKeyFrames, ?Mode:=animation_mode.OneShot)
            PropController.Play()
    # - - - - - - - - - - ANIMATION RESET PROP END METHOD - - - - - - - - - - - #
    
    #Changed to show and hide the prop itself, not using the Prop Manipulator Functionality
    HideProp<public>():void=
        CreativePropToHit.Hide()
    
    ShowProp<public>():void=            
        CreativePropToHit.Show()

    #Teleport the prop into position and set the base scale
    ResetProp<public>():void=
        #Scale was 0.1
        Transform := transform{Translation := OriginalTransform.Translation, Rotation := OriginalTransform.Rotation , Scale := NodePropScale}
        if(CreativePropToHit.TeleportTo[Transform]):

    GetHarvestProp<public>():creative_prop=
        return CreativePropToHit

    PropAlive<public>():logic=
        return BigNumber_GreaterThanOrEquals(PerNodeHarvestPropHealth, big_number{Value := 0.1, Exponent := 0})

# A Verse-authored creative device that can be placed in a level
V_PropHealthBar := class(creative_device):
    @editable
    BeginDelay : float = 1.0
    @editable
    PlayerManager : ?V_PlayerManagerDevice = false
    @editable
    RockHitTrigger : trigger_device = trigger_device{}
    @editable
    RockDestroyedTrigger : trigger_device = trigger_device{}
    @editable
    ResourceTypeToMine : PropResourceType = PropResourceType.Stone
    @editable 
    SetHarvestPropHealth<public> : big_number = big_number{}
    @editable 
    SetHarvestPropMaxHealth : big_number = big_number{}
    @editable
    TimeBeforePropReappears : float = 2.0
    @editable
    Invincible : logic = false 
    @editable
    DropsDust : logic = false 
    @editable 
    DustDropAmount : big_number = big_number{}
    #Visual Effects when hitting Ore + Last Hit
    @editable
    HitVFX : ?creative_prop_asset = false
    @editable
    LastHitVFX : ?creative_prop_asset = false
    @editable
    PropScale : vector3 = vector3{X:= 0.1, Y:= 0.1, Z:= 0.1}
    @editable
    DistanceToHitGroundAnimationController : ?float = false
    @editable
    HarvestResourceCave : MiningNodeCave = MiningNodeCave.Null

    @editable
    VFXSpawnerReference : V_CustomVFXSpawner = V_CustomVFXSpawner{}

    @editable
    ActivationVolume : volume_device = volume_device{}

    #Update 1.3 - Checking Cache Data, keep this @editable for now [Previously was not @editable]
    var HarvestResourcesProps : []HarvestResourceNodeClass = array{}

    #If you get duplicates make this @editable and make sure the array is empty!
    #@editable
    var Props : []creative_prop = array{}

    #NEW For Playing SFX on Hit
    @editable
    HitAudio : audio_player_device = audio_player_device{}
    @editable
    LastHitAudio : audio_player_device = audio_player_device{}
    @editable
    AccoladeForHittingNode : accolades_device = accolades_device{}

    DebugMultiplier : float = 1.0

    # Runs when the device is started in a running game
    OnBegin<override>()<suspends>:void={}

    #Removed a sleep
    SetupAllNodes(SearchTag : []creative_object_interface):void=
        Print("Agent Entered Cave {Props.Length}")
        #Only perform this function if Props is empty, this means it hasn't been initiated
        if(Props.Length <= 0):
            Print("Setting Up Cave")
            var DebugTagSearch : string = "God Node"
            #Define it as 0.0,0.0,0.0 from the start
            var PreviousPropPosition : vector3 = vector3{X:= 0.0, Y:= 0.0, Z:= 0.0}
            #TagObjects := SearchTag
            #TaggedDevice : FindCreativeObjectsWithTag(tMagicCave3{}), CastedTag := creative_object_interface[TaggedDevice]
            #for(TaggedDevice : FindCreativeObjectsWithTag(tMagicCave2{}), CastedTag := creative_object_interface[TaggedDevice]):
            #Is tag to search empty? I updated the method I use to add to it
            Print("Cast to get tag from SearchTag")
            #TaggedObjects := FindCreativeObjectsWithTag(TagToSearch)
            #I already get the tagged actors, I just need to go through them here! I've saved myself a step with the new tag system
            for (i-> TagProp : SearchTag):
                if(Prop := creative_prop[TagProp]):
                    Print("Found Prop")
                    CurrentTagPropPosition := TagProp.GetTransform().Translation
                    #We check if the X and Y is the same as the original position, this is to get rid of duplicates
                    if(CurrentTagPropPosition.X = PreviousPropPosition.X and CurrentTagPropPosition.Y = PreviousPropPosition.Y):
                        #Print("Same Position don't add")
                    else:
                        #Print("Unique Position, add it")
                        set Props += array{Prop}
                        set HarvestResourcesProps += array{HarvestResourceNodeClass{}}
                        #Set the vector3 position here
                        set PreviousPropPosition = TagProp.GetTransform().Translation
            Print("{DebugTagSearch} Node Amount = {Props.Length}")
            #Create all harvest resource props, just count the amount and write it into the device
            for(Index : Props):
                set HarvestResourcesProps += array{HarvestResourceNodeClass{}}
            SetupHarvestProps()
    
    #Setup all Harvest Props
    SetupHarvestProps():void=
        for(i-> HarvestResourceProp : HarvestResourcesProps):
            if(Prop := Props[i]):
                set HarvestResourceProp.CreativePropToHit = Prop
                set HarvestResourceProp.NodePropScale = PropScale
                HarvestResourceProp.SetupNode(SetHarvestPropHealth, SetHarvestPropMaxHealth)
                if(bHitVFX := HitVFX?, bLastHitVFX := LastHitVFX?):
                    set HarvestResourceProp.HitVFX = bHitVFX
                    set HarvestResourceProp.LastHitVFX = bLastHitVFX

    #When harvesting a resource we show a custom HUD and update that HUD for the player
    ResourceHarvest(InAgent : agent, HitLocation : vector3): void=
        #Harvest Resource Class reference that gets set once we find the closest node to the player. Once we have this we can use the variables stored in the class for required functionality.
        var HarvestResource : HarvestResourceNodeClass = HarvestResourceNodeClass{}

        var YValue : float = 0.0
        var XValue : float = 0.0
        if (Player := player[InAgent], bPlayerManager := PlayerManager?, FortChar := InAgent.GetFortCharacter[]):
            #Check all of the setup HarvestResourceProps
            for(i -> HarvestResourceProp : HarvestResourcesProps):
                #We then compare the X,Y and Z to check if the creativeproptohit is the same as the prop we hit.
                if(HarvestResourceProp.CreativePropToHit.GetTransform().Translation.X = HitLocation.X and HarvestResourceProp.CreativePropToHit.GetTransform().Translation.Y = HitLocation.Y and HarvestResourceProp.CreativePropToHit.GetTransform().Translation.Z = HitLocation.Z):
                    #If it is the same prop we set HarvestResource to the HarvestResourceProp
                    if(PropInSamePosition := HarvestResourcesProps[i]):
                        set HarvestResource = PropInSamePosition
            #Only perform functionality if the prop is alive
            #I believe this is stopping the double hit, but it was also the only thing stopping the map from breaking previously.
            IsPropAlive := HarvestResource.PropAlive()
            if(IsPropAlive = false):
                return
            #When we hit ore trigger this trigger
            FriendsAmount := bPlayerManager.FriendsManager.FindFriends(Player)
            var ResourceHitCount : int = 0
            var ResourcesHit : []HarvestResourceNodeClass = array{}
            var PropsAlive : []logic = array{}
            
            if(CurrentPlayerStats := bPlayerManager.PlayerStatManager.GetPlayerStats[InAgent]):
                var TrueCrit : logic = false
                var ExplosiveBonus : big_number = big_number{}
                var TotalExplosiveBonus : big_number = big_number{}
                NextHitExplode := bPlayerManager.GetPlayerDataNotSaved(Player).NextHitExplode
                NextHitCrit := bPlayerManager.GetPlayerDataNotSaved(Player).NextHitCrit
                HitAmountTuple := bPlayerManager.PlayerStatManager.GetPlayerHitWithCritical(InAgent, NextHitCrit, bPlayerManager.GetPlayerDataNotSaved(Player).AltarBuff, bPlayerManager.GetPlayerDataNotSaved(Player).GnomeBuff)

                if(NextHitCrit = true):
                    Print("Charged Critical Hit")
                    spawn:
                        VFXSpawnerReference.SpawnVFX(HarvestResource.CreativePropToHit, 0.01, VFXToSpawnEnum.eChargedCriticalHitVFX, 1.2, InAgent, false)
                    set TrueCrit = true
                else:
                    set TrueCrit = HitAmountTuple(1)

                FriendsPlayerLevel := bPlayerManager.PlayerStatManager.GetPlayerUpgradeLevel(Player, PlayerUpgradeType.Friends)
                var FriendsPlayerBonus : float = bPlayerManager.PlayerStatManager.UpgradesData.GetFriendsAmount(FriendsPlayerLevel)

                if(FoundCard := SearchForCard(11, Player, bPlayerManager.PlayerStatManager)?):
                    set FriendsPlayerBonus += GetCardBonus(11, FoundCard.Level)

                BNFriendPlayerBonusPer := BigNumber_ConvertFloatIntoBigNumber(FriendsPlayerBonus / 100.0)
                TotalFriendsBonus := BigNumber_MultiplyWithAFloat(BNFriendPlayerBonusPer, FriendsAmount * 1.0)
                CurrentMiningPower := BigNumber_Add(TotalFriendsBonus, HitAmountTuple(0))
                if(Invincible = false):
                    TagView := HarvestResource.CreativePropToHit.GetTags()
                    if(not TagView.Has[tEndGodNode{}]):
                        HarvestResource.TakeDamage(CurrentMiningPower)

                if(NextHitExplode = true):
                    
                    for(i -> Resource : HarvestResourcesProps):
                        ResourcePosition := Resource.GetHarvestProp().GetTransform().Translation
                        MainPosition := HarvestResource.GetHarvestProp().GetTransform().Translation

                        # Check to see if its the harvest resource that is passed in
                        if(ResourcePosition.X <> MainPosition.X and ResourcePosition.Y <> MainPosition.Y):
                            DistanceFromMain := Distance(MainPosition, ResourcePosition)
                            PropAlive := Resource.PropAlive()
                            #Concept here is that it will spread out
                            set YValue = YValue + 5.0
                            set XValue = XValue + 5.0
                            # Check if Explosive Hit Level is 1 which is unlocked
                            ExplosiveHitRangeLevel := bPlayerManager.PlayerStatManager.GetPlayerUpgradeLevel(Player, PlayerUpgradeType.ExplosiveHitRange)
                            ExplosiveHitDamageLevel := bPlayerManager.PlayerStatManager.GetPlayerUpgradeLevel(Player, PlayerUpgradeType.ExplosiveHitDamage)

                            ExplosiveHitRangeAmount := bPlayerManager.PlayerStatManager.UpgradesData.GetExplosiveHitRangeAmount(ExplosiveHitRangeLevel)
                            ExplosiveHitDamageAmount := bPlayerManager.PlayerStatManager.UpgradesData.GetExplosiveHitDamageAmount(ExplosiveHitDamageLevel)
                            if(DistanceFromMain <= ExplosiveHitRangeAmount):    #ExplosiveHitRangeAmount
                                Print("In Range")
                                if(PropAlive = true):
                                    Print("Prop Alive, Explosive hit VFX and Damage Functionality Running")
                                    #TODO: We should get resources etc.. after the explosion not before
                                    spawn:
                                        ArcaneMissile(InAgent, Resource.GetHarvestProp(), YValue, XValue)
                                    set ResourcesHit += array{Resource}
                                    set ExplosiveBonus = BigNumber_MultiplyWithAFloat(HitAmountTuple(0), ExplosiveHitDamageAmount / 100.0)
                                    set TotalExplosiveBonus = BigNumber_Add(TotalExplosiveBonus, ExplosiveBonus)
                                    TagView := Resource.CreativePropToHit.GetTags()
                                    if(not TagView.Has[tEndGodNode{}]):
                                        Resource.TakeDamage(TotalExplosiveBonus)
                                    set ResourceHitCount += 1

                    # Make sure we are only taking into account the other resources when we want to
                    if(NextHitExplode = true):
                        for(HitResource : ResourcesHit):
                            NewPropAlive := HitResource.PropAlive()
                            set PropsAlive += array{NewPropAlive}
                            if(NewPropAlive = false):
                                Print("Dead Resource")
                                spawn:
                                    DestroyAndResetResource(HitResource, InAgent)

                MainPropAlive := HarvestResource.PropAlive()
                set PropsAlive += array{MainPropAlive}
                if(MainPropAlive = false):
                    Print("Main Dead Resource")
                    spawn:
                        DestroyAndResetResource(HarvestResource, InAgent)

                bPlayerManager.CriticalCharge(Player, HitAmountTuple(1), NextHitCrit)
                bPlayerManager.ExplosiveCharge(Player, NextHitExplode)

                FloatTotalExplosiveBonus := BigNumber_ConvertBigNumberIntoFloat(TotalExplosiveBonus)

                OutAmount := BigNumber_Multiply(CurrentMiningPower, BigNumber_ConvertFloatIntoBigNumber(1.0))
                GainResource(Player, HarvestResource, OutAmount, PropsAlive, TrueCrit, FloatTotalExplosiveBonus * ResourceHitCount)
                #Spawn Hit Effects
                HarvestResource.SpawnHitEffects()
                HitAudio.Play(InAgent)


    GainResource(InPlayer : player, InHarvestResource : HarvestResourceNodeClass, InMiningPower : big_number, IsPropsAlive : []logic, IsCrit : logic, InExtraResource : float):void=
        var MiningPower : big_number = InMiningPower
        var ArtifactFinalHitBonus : float = 0.0
        var LastHit : logic = false

        AccoladeForHittingNode.Award(InPlayer)

        if(Invincible = false):
            spawn:
                DropOreResource(InPlayer)
        if(DropsDust = true):
            spawn:
                DropDust(InPlayer, IsPropsAlive)

        if(IsCrit = true):
            Print("Critical Hit")
            #TODO: Critical Hit VFX
            spawn:
                VFXSpawnerReference.SpawnVFX(InHarvestResource.CreativePropToHit, 0.01, VFXToSpawnEnum.eCriticalHitVFX, 1.0, InPlayer, true)

        if (bPlayerManager := PlayerManager?):
            if(PlayerData := PlayerStatsDataMap[InPlayer]):
                for(EquippedArtifact : PlayerData.EquippedPlayerArtifacts):
                    set ArtifactFinalHitBonus += GetArtifactStat(EquippedArtifact, ArtifactStats.FinalStrike)

            for(PropAlive : IsPropsAlive):
                if(PropAlive = false):
                    # Get Player Final Hit Bonus
                    FinalHitLevel := bPlayerManager.PlayerStatManager.GetPlayerUpgradeLevel(InPlayer, PlayerUpgradeType.FinalHit)
                    FinalHitBonus := bPlayerManager.PlayerStatManager.UpgradesData.GetFinalHitAmount(FinalHitLevel)
                    FinalHitTotal := BigNumber_MultiplyWithAFloat(InHarvestResource.PerNodeHarvestPropMaxHealth, (FinalHitBonus / 100.0))
                    BNFinalHitTotal := BigNumber_Add(FinalHitTotal, BigNumber_MultiplyWithAFloat(FinalHitTotal, (ArtifactFinalHitBonus / 100.0)))
                    set MiningPower = BigNumber_Add(InMiningPower, BNFinalHitTotal)
                    if(FoundCard := SearchForCard(9, InPlayer, bPlayerManager.PlayerStatManager)?):
                        BonusAmount := GetCardBonus(9, FoundCard.Level)
                        set MiningPower = GetBonusBigNumber(MiningPower, BonusAmount)
                    set LastHit = true
                    bPlayerManager.PlayerStatManager.UpdateTotalRocksDestroyed(InPlayer, big_number{Value := 1.0, Exponent := 0})
                    RockDestroyedTrigger.Trigger(InPlayer)
                    
                    #Print("FINAL HIT")

            BNExtraResource := BigNumber_ConvertFloatIntoBigNumber(InExtraResource)
            OutAmount := BigNumber_Multiply(BigNumber_Add(MiningPower, BNExtraResource), BigNumber_ConvertFloatIntoBigNumber(DebugMultiplier))
            bPlayerManager.PlayerStatManager.UpdatePlayerOreAmount(InPlayer, ResourceTypeToMine, OutAmount)
            AfterAmount :=  bPlayerManager.PlayerStatManager.GetPlayerOreAmount(InPlayer, ResourceTypeToMine)
            #Print("GAINED VALUE : {OutAmount.Value} | GAINED EXPONENT : {OutAmount.Exponent}")
            RockHitTrigger.Trigger(InPlayer)

            var ResourceName : string = "Stone"
            var ResourceImage : texture = VerseExposed.T_Stone

            bPlayerManager.UIManagerDevice.UIItemStreamDisplayClass.AddToStream(InPlayer, "Stone", ResourceImage, OutAmount, IsCrit, LastHit)
            if(Invincible = false):
                TagView := InHarvestResource.CreativePropToHit.GetTags()
                    if(not TagView.Has[tEndGodNode{}]):
                        UpdateHealthBar(InPlayer, ResourceName, InHarvestResource)

    DropOreResource(InPlayer : player)<suspends>:void=
        # START CHANCE TO GET ANY ORE #
        RandomChance := GetRandomInt(1, 100)
        if (bPlayerManager := PlayerManager?):
            ArtifactMasteryLevel := bPlayerManager.PlayerStatManager.GetPlayerUpgradeLevel(InPlayer, PlayerUpgradeType.Prestige_ArtifactMastery)
            AdditionalChance := bPlayerManager.PlayerStatManager.UpgradesData.GetPrestigeArtifactMasteryAmount(ArtifactMasteryLevel)
            TotalDropChance := AdditionalChance
            if(RandomChance * 1.0 <= TotalDropChance):
                # START CHANCE TO GET SPECIFIC ORE ##
                TypeRandomChance := GetRandomFloat(0.0, 100.0)
                var ResourceGot : PropResourceType = PropResourceType.Null 
                var CardIDToCheck : int = -1
                if(GetOreDropChance(PropResourceType.Obsidian) >= TypeRandomChance):
                    set ResourceGot = PropResourceType.Obsidian
                    set CardIDToCheck = 20
                    Print("Obsidian Got")
                else if(GetOreDropChance(PropResourceType.Spectrolite) >= TypeRandomChance):
                    set ResourceGot = PropResourceType.Spectrolite
                    set CardIDToCheck = 19
                    Print("Spectrolite Got")
                else if(GetOreDropChance(PropResourceType.Malachite) >= TypeRandomChance):
                    set ResourceGot = PropResourceType.Malachite
                    set CardIDToCheck = 18
                    Print("Malachite Got")
                else if(GetOreDropChance(PropResourceType.Silver) >= TypeRandomChance):
                    set ResourceGot = PropResourceType.Silver
                    set CardIDToCheck = 17
                    Print("Silver Got")
                else if(GetOreDropChance(PropResourceType.Copper) >= TypeRandomChance):
                    set ResourceGot = PropResourceType.Copper
                    set CardIDToCheck = 1
                    Print("Copper Got")
                
                if(ResourceGot <> PropResourceType.Null):
                    QuantityLevel := bPlayerManager.PlayerStatManager.GetPlayerUpgradeLevel(InPlayer, PlayerUpgradeType.OreQuantity)
                    Upgrade := bPlayerManager.PlayerStatManager.GetPlayerUpgrade(InPlayer, PlayerUpgradeType.OreQuantity)
                    var OutAmount : big_number = BigNumber_ConvertFloatIntoBigNumber(Upgrade.GetCurrentLevelAmount(QuantityLevel))
                    if(FoundCard := SearchForCard(CardIDToCheck, InPlayer, bPlayerManager.PlayerStatManager)?):
                        BonusAmount := GetCardBonus(CardIDToCheck, FoundCard.Level)
                        set OutAmount = BigNumber_Add(OutAmount, BigNumber_ConvertFloatIntoBigNumber(BonusAmount))

                    bPlayerManager.PlayerStatManager.UpdatePlayerOreAmount(InPlayer, ResourceGot, OutAmount)
                    Sleep(0.25)
                    bPlayerManager.UIManagerDevice.UIItemStreamDisplayClass.AddToStream(InPlayer, "Stone", GetOreTexture(ResourceGot), OutAmount, false, false)

    DropDust(InPlayer : player, PropsAlive : []logic)<suspends>:void=
        if (bPlayerManager := PlayerManager?):
            for(PropAlive : PropsAlive):
                if(PropAlive = false):
                    var OutAmount : big_number = DustDropAmount
                    bPlayerManager.PlayerStatManager.UpdatePlayerOreAmount(InPlayer, PropResourceType.Dust, OutAmount)
                    Sleep(0.12)
                    bPlayerManager.UIManagerDevice.UIItemStreamDisplayClass.AddToStream(InPlayer, "Stone", VerseExposed.T_GlitterDust, OutAmount, false, false)

    #Hide the resource, after a delay reset the health and show it again
    DestroyAndResetResource(HarvestResource : HarvestResourceNodeClass, InAgent : agent)<suspends>:void=
        LastHitAudio.Play(InAgent)
        HarvestResource.KnockBackHarvestProp(InAgent, DistanceToHitGroundAnimationController)
        #This is the time it takes for the knockback animation to complete
        Sleep(1.5)
        HarvestResource.HideProp()
        #This is the time before the prop can come back
        Sleep(2.5)
        HarvestResource.ShowProp()
        HarvestResource.ResetHarvestNodeHealth()

        #Update 1.1 - Setting it to false here after it's been teleported back
        #set HarvestResource.PropBeingAnimated = false

        #Will need to see if we can change this. Previously there were issues with this.
        Sleep(TimeBeforePropReappears)
        #Resets via a Transform Function [Position, Rotation]
        HarvestResource.ResetProp()
        #Need to test if this sleep is required. I had issues previously where the teleporter would mess up
        Sleep(0.1)
        #Animates the scale back up
        HarvestResource.ResetHarvestProp()

    # Update a given player's prop health bar
    UpdateHealthBar(InPlayer : player, InResourceName : string, InHarvestResource : HarvestResourceNodeClass):void=
        BNPropHealth := BigNumber_FormatDisplay(InHarvestResource.GetHarvestPropHealth())
        BNPropMaxHealth := BigNumber_FormatDisplay(InHarvestResource.GetHarvestPropMaxHealth())
        HealthText := "{BNPropHealth} / {BNPropMaxHealth}"
        HealthPercentage := BigNumber_ConvertBigNumberIntoFloat(BigNumber_MultiplyWithAFloat(BigNumber_Divide(InHarvestResource.GetHarvestPropHealth(), InHarvestResource.GetHarvestPropMaxHealth()), 100.0))
        PropPosition := InHarvestResource.GetHarvestProp().GetTransform().Translation
        HealthBarData := MakeNodeHealthBarData(InResourceName, HealthText, HealthPercentage, PropPosition, false)

        if(bPlayerManager := PlayerManager?):
            bPlayerManager.UIManagerDevice.UINodeHealthBar.UpdateUI(InPlayer, HealthBarData)
            bPlayerManager.UIManagerDevice.UINodeHealthBar.AddUI(InPlayer)

    # - - - - - - - - - - EXPLOSIVE HIT VFX METHOD - - - - - - - - - - - # 
    #We get the agent spawning for location and we also pass through the creative prop we spawn earlier.
    #TODO: I believe this will have to occur in a unique/concrete function.
    ArcaneMissile<public>(InAgent : agent, ResourceToHit : creative_prop, YValue : float, XValue : float)<suspends>:void=
        Sleep(GetRandomFloat(0.2, 0.65))
        #This is the VFX we spawn that will move towards the ORE
        var CreativePropSpawned : creative_prop = creative_prop{}
        if(FortChar := InAgent.GetFortCharacter[]):
            SpawnPosition := vector3{X:= FortChar.GetTransform().Translation.X - 20.0, Y:= FortChar.GetTransform().Translation.Y, Z:= FortChar.GetTransform().Translation.Z + 20.0}
            SpawnScale := vector3{X:= 1.0, Y:= 1.0, Z:= 1.0}

            ExplosiveHitTrailVFX := VFXSpawnerReference.VFXToSpawn(VFXToSpawnEnum.eExplosiveHitTrailVFX)
            if(bExplosiveHitTrailVFX := ExplosiveHitTrailVFX?): #ExplosiveHitTrailVFX?
                SpawnedProp := SpawnProp(bExplosiveHitTrailVFX, transform{Translation := SpawnPosition, Rotation := IdentityRotation(), Scale:= SpawnScale})
                if(CreativeProp := SpawnedProp(0)?):
                    set CreativePropSpawned = CreativeProp

            KnockbackDirectionCalculation := ResourceToHit.GetTransform().Translation - CreativePropSpawned.GetTransform().Translation 
            #We get the direction for X,Y and Z to travel
            KnockbackDirection := vector3{X:= KnockbackDirectionCalculation.X, Y:= KnockbackDirectionCalculation.Y, Z:= KnockbackDirectionCalculation.Z}
            #We want it to go half way and go up high so it can then travel down
            HomingDirectionFrame1 := vector3{X:= KnockbackDirectionCalculation.X * 0.1 , Y:= KnockbackDirectionCalculation.Y * 0.1, Z:= 100.0} #+ GetRandomFloat(-80.0, 80.0)
            #Location of the node we want to hit
            MovementFrameDestination := ResourceToHit.GetTransform().Translation
            #Fly up and forward a little
            #Then fly towards
                MovementKeyFrame1 : keyframe_delta = keyframe_delta
                {
                    DeltaLocation := HomingDirectionFrame1 #* 1.0
                    DeltaRotation := MakeRotationFromYawPitchRollDegrees(0.0, 0.0, 0.0)
                    DeltaScale := vector3 {X:=1.0, Y:= 1.0, Z:= 1.0}
                    Time := 0.25
                    Interpolation := Linear    #Start Slower but retain speed at the end
                }
                MovementKeyFrame2 : keyframe_delta = keyframe_delta
                {
                    #Moved onto the ground
                    DeltaLocation := KnockbackDirection * 0.7
                    DeltaRotation := MakeRotationFromYawPitchRollDegrees(0.0, 0.0, 0.0)
                    Time := 0.45
                    Interpolation := EaseOut    #Start fast then ease out for the hit
                }
                if(PropController := CreativePropSpawned.GetAnimationController[]):
                    KnockbackKeyFrames:[]keyframe_delta = array{MovementKeyFrame1, MovementKeyFrame2}  #Can extend the array of keyframes here
                    PropController.SetAnimation(KnockbackKeyFrames, ?Mode:=animation_mode.OneShot)
                    PropController.Play()

        #Impact VFX
        spawn:
            VFXSpawnerReference.SpawnVFX(ResourceToHit, 0.65, VFXToSpawnEnum.eExplosiveHitImpactVFX, 1.0, InAgent, false)
            #SpawnExplosion(ResourceToHit, 0.65)
        #Remove Effects
        spawn:
            DestroySpawnedEffects(CreativePropSpawned, 0.69)
    # - - - - - - - - - - EXPLOSIVE HIT VFX METHOD END - - - - - - - - - - - # 

    SpawnExplosion(PropToSpawnAt : creative_prop, TimeToWait : float)<suspends>:void=
        Sleep(TimeToWait)
        SpawnLocation := vector3{X:= PropToSpawnAt.GetTransform().Translation.X, Y:= PropToSpawnAt.GetTransform().Translation.Y, Z:= PropToSpawnAt.GetTransform().Translation.Z + 20.0}
        
        ExplosiveHitImpactVFX := VFXSpawnerReference.VFXToSpawn(VFXToSpawnEnum.eExplosiveHitImpactVFX)      
        if(bExplosiveHitImpactVFX := ExplosiveHitImpactVFX?):
            SpawnedProp := SpawnProp(bExplosiveHitImpactVFX, transform{Translation := SpawnLocation, Rotation := IdentityRotation(), Scale:= vector3{X:=1.0,Y:=1.0,Z:=1.0}})
            if(CreativeProp := SpawnedProp(0)?):
                DestroySpawnedEffects(CreativeProp, 1.0)

    DestroySpawnedEffects(SpawnedProp : creative_prop, TimeToDispose : float)<suspends>:void=
        Sleep(TimeToDispose)
        SpawnedProp.Dispose()
