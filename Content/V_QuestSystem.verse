
using { /Fortnite.com/Devices }
using { /Verse.org/Simulation }
using { /UnrealEngine.com/Temporary/Diagnostics }

# For Example: Using the save data to save the player's current quest and quest line strings for loading
DataToTrack := enum:
    Strength
    Stone
    Currency
    RareOre
    TotalStone
    Null    #This is used when we want an empty value in some functions. It should always be last

MakeCustomQuest<constructor>(InCustomQuest : CustomQuest, InIncrementEvent : event()) := CustomQuest:
    QuestLineName := InCustomQuest.QuestLineName
    Name := InCustomQuest.Name
    IncrementGoal := InCustomQuest.IncrementGoal
    EndQuestLine := InCustomQuest.EndQuestLine
    ShowCost := InCustomQuest.ShowCost
    Cost := InCustomQuest.Cost
    Increment := InCustomQuest.Increment
    UsePlayerData := InCustomQuest.UsePlayerData
    DataToUse := InCustomQuest.DataToUse
    InvisibleQuest := InCustomQuest.InvisibleQuest
    OptBeacons := InCustomQuest.OptBeacons
    OptIndicatior := InCustomQuest.OptIndicatior
    OptTriggers := InCustomQuest.OptTriggers
    OptPropManipulator := InCustomQuest.OptPropManipulator
    OptSwitches := InCustomQuest.OptSwitches
    OptCreatureManager := InCustomQuest.OptCreatureManager
    OptTrackers := InCustomQuest.OptTrackers
    OptConditionalButtons := InCustomQuest.OptConditionalButtons
    OptSignalRemote := InCustomQuest.OptSignalRemote
    OptTeleporters := InCustomQuest.OptTeleporters
    PlayerStatsManager := InCustomQuest.PlayerStatsManager
    IncrementEvent := InIncrementEvent
    
CustomQuest := class<concrete>():

    var QuestLineName : string = "Quest Line"
    @editable
    Name : string = "Quest Name"
    @editable
    IncrementGoal : float = 1.0
    @editable
    EndQuestLine : logic = false
    @editable
    ShowCost : logic = false
    @editable
    Cost : big_number = big_number{}
    
    var Increment : big_number = big_number{}

    @editable
    UsePlayerData : logic = false
    @editable
    DataToUse : DataToTrack = DataToTrack.Null

    @editable
    InvisibleQuest : logic = true

    @editable
    OptBeacons : []beacon_device = array{}
    @editable
    OptIndicatior : ?map_indicator_device = false
    @editable
    OptTriggers : []trigger_device = array{}
    @editable
    OptPropManipulator : []prop_manipulator_device =  array{}
    @editable
    OptSwitches : []switch_device = array{}
    @editable
    OptCreatureManager : []creature_manager_device = array{}
    @editable
    OptTrackers : []tracker_device = array{}
    @editable
    OptTeleporters : []teleporter_device = array{}
    @editable
    OptConditionalButtons : []conditional_button_device = array{}
    @editable
    OptSignalRemote : ?signal_remote_manager_device = false

    var IncrementEvent : event() = event(){}
    var PlayerReference : []player = array{}
    var Cancelables : []cancelable = array{}

    var PlayerStatsManager : player_stats_manager = player_stats_manager{}

    SetUp(InPlayer : player, InPlayerStatsManager : player_stats_manager):void=
        set PlayerReference += array{InPlayer}
        set PlayerStatsManager = InPlayerStatsManager
        if(UsePlayerData = true):
            IncrementQuest()
        Print("Set Up Subscriptions")
        for(Trigger : OptTriggers):
            set Cancelables += array{Trigger.TriggeredEvent.Subscribe(OnMaybe)}
        for(Manipulator : OptPropManipulator):
            set Cancelables += array{Manipulator.HarvestingEvent.Subscribe(OnAgent)}
        for(CurrentSwitch : OptSwitches):
            set Cancelables += array{CurrentSwitch.TurnedOnEvent.Subscribe(OnAgent)}
        for(CreatureManager : OptCreatureManager):
            set Cancelables += array{CreatureManager.MatchingCreatureTypeEliminatedEvent.Subscribe(OnAgent)}
        for(Tracker : OptTrackers):
            set Cancelables += array{Tracker.CompleteEvent.Subscribe(OnAgent)}
        for(Button : OptConditionalButtons):
            set Cancelables += array{Button.ActivatedEvent.Subscribe(OnAgent)}
        for(Teleporter : OptTeleporters):
            set Cancelables += array{Teleporter.TeleportedEvent.Subscribe(OnAgent)}
        for(Beacon : OptBeacons):
            Beacon.AddToShowList(InPlayer)
        if(Indicatior := OptIndicatior?):
            for(Index := 0..20):
                Indicatior.ActivateObjectivePulse(InPlayer)
        if(SignalRemote := OptSignalRemote?):
            set Cancelables += array{SignalRemote.PrimarySignalEvent.Subscribe(OnAgent)}
        
    
    CancelSubscriptions():void=
        for(Cancelable : Cancelables):
            Cancelable.Cancel()

    IncrementQuest():void=
        Print("Increment Quest: {Name}")
        if(UsePlayerData = true, Player := PlayerReference[0]):
            case(DataToUse):
                DataToTrack.Strength =>
                    NewIncrement := PlayerStatsManager.GetPlayerOreHarvestRatePerHit(Player)
                    set Increment = NewIncrement
                DataToTrack.Stone =>
                    NewIncrement := PlayerStatsManager.GetPlayerStoneOreAmount(Player)
                    set Increment = NewIncrement
                DataToTrack.RareOre =>
                    CopperAmount := PlayerStatsManager.GetPlayerOreAmount(Player, PropResourceType.Copper)
                    SilverAmount := PlayerStatsManager.GetPlayerOreAmount(Player, PropResourceType.Silver)
                    MalachiteAmount := PlayerStatsManager.GetPlayerOreAmount(Player, PropResourceType.Malachite)
                    SpectroliteAmount := PlayerStatsManager.GetPlayerOreAmount(Player, PropResourceType.Spectrolite)
                    ObsidianAmount := PlayerStatsManager.GetPlayerOreAmount(Player, PropResourceType.Obsidian)

                    NewIncrement := BigNumber_Add(CopperAmount, BigNumber_Add(SilverAmount, BigNumber_Add(MalachiteAmount, BigNumber_Add(SpectroliteAmount, ObsidianAmount))))
                    set Increment = NewIncrement
                DataToTrack.TotalStone =>
                    if(PlayerStats := PlayerStatsManager.GetPlayerStats[Player]):
                        NewIncrement := PlayerStats.TotalStoneOreAmount
                        set Increment = NewIncrement
                _ =>
                    Print("DEFAULT SHOULDN'T BE USED")
            Print("PlayerData Used")
        else:
            set Increment = BigNumber_Add(Increment, big_number{Value:=1.0})
            Print("No PlayerData Used")
        IncrementEvent.Signal()

    CheckForCompletion(InPlayer : player):logic=
        ReachedGoal := BigNumber_GreaterThanOrEquals(Increment, BigNumber_ConvertFloatIntoBigNumber(IncrementGoal))
        if(ReachedGoal = true):
            for(Beacon : OptBeacons):
                Beacon.RemoveFromShowList(InPlayer)
            if(Indicatior := OptIndicatior?):
                Indicatior.DeactivateObjectivePulse(InPlayer)
            return true
        if(Switch := OptSwitches[0], Switch.GetCurrentState[InPlayer]):
            for(Beacon : OptBeacons):
                Beacon.RemoveFromShowList(InPlayer)
            if(Indicatior := OptIndicatior?):
                Indicatior.DeactivateObjectivePulse(InPlayer)
            return true
        return false

    OnMaybe(InMaybeAgent : ?agent):void=
        if(Player := player[InMaybeAgent?], PlayerReference = array{Player}):
            IncrementQuest()
        else:
            Print("Player Check Fail")

    OnAgent(InAgent : agent):void=
        if(Player := player[InAgent], PlayerReference = array{Player}):
            IncrementQuest()
        else:
            Print("Player Check Fail")

CustomQuestLine := class<concrete>():
    @editable
    Name : string = "Quest Line Name"
    @editable
    NextQuestLine : string = "Next Quest Line"
    @editable
    CustomQuests : []CustomQuest = array{}

# Hold player data for quests
PerPlayerQuestData := class():
    var CurrentQuest : CustomQuest = CustomQuest{}
    var QuestStart : logic = false

    IncrementEvent : event() = event(){}

# A Verse-authored creative device that can be placed in a level
V_QuestSystem := class(creative_device):
    @editable
    QuestLines : []CustomQuestLine = array{}

    @editable
    UIManager : V_UIManagerDevice = V_UIManagerDevice{}

    @editable
    QuestCompleteAudio : audio_player_device = audio_player_device{}

    @editable
    SkipQuestButton : button_device = button_device{}

    var PerPlayerQuestDataMap : [player]PerPlayerQuestData = map{}

    OnBegin<override>()<suspends>:void=
        SkipQuestButton.InteractedWithEvent.Subscribe(SkipQuestButtonActivated)

    PlayerSetUp(InPlayer : player):void=
        if(PlayerQuestData := PerPlayerQuestDataMap[InPlayer], PlayerQuestData.QuestStart = true):
            return
        else:
            if(set PerPlayerQuestDataMap[InPlayer] = PerPlayerQuestData{}, NewPlayerQuestData := PerPlayerQuestDataMap[InPlayer], CurrentPlayerStats := UIManager.PlayerStatsManager.GetPlayerStats[InPlayer]):
                # Check if player has completed quests before
                if(CurrentPlayerStats.CompletedQuests = array{} and CurrentPlayerStats.CurrentQuest = ""):
                    # If not then assign first quest line
                    QuestLine := SearchForQuestLine("Getting Started")
                    if(FirstQuest := QuestLine.CustomQuests[0]):
                        ConstructedQuest := SearchForQuest(FirstQuest.Name)
                        SetPlayerQuest(InPlayer, ConstructedQuest)

                        CurrentQuestProgress := "{BigNumber_FormatDisplay(NewPlayerQuestData.CurrentQuest.Increment)}/{BigNumber_FormatDisplay(BigNumber_ConvertFloatIntoBigNumber(NewPlayerQuestData.CurrentQuest.IncrementGoal))}"
                        QuestData := MakeQuestDisplayData(NewPlayerQuestData.CurrentQuest.QuestLineName, NewPlayerQuestData.CurrentQuest.Name, CurrentQuestProgress, NewPlayerQuestData.CurrentQuest.ShowCost)

                        UIManager.UIQuestDisplay.InitUI(InPlayer, QuestData)
                        UIManager.UIQuestDisplay.AddUI(InPlayer)
                        spawn {QuestUpdate(InPlayer, NewPlayerQuestData.IncrementEvent)}
                else if(CurrentPlayerStats.CurrentQuest <> ""):
                    # Load current quest and quest line
                    SetPlayerQuest(InPlayer, SearchForQuest(CurrentPlayerStats.CurrentQuest))
                    set NewPlayerQuestData.CurrentQuest.Increment = CurrentPlayerStats.CurrentQuestIncrement
                    #Print("Loaded : {CurrentPlayerStats.CurrentQuestIncrement} | New : {NewPlayerQuestData.CurrentQuest.Increment}")

                    CurrentQuestProgress := "{BigNumber_FormatDisplay(NewPlayerQuestData.CurrentQuest.Increment)}/{BigNumber_FormatDisplay(BigNumber_ConvertFloatIntoBigNumber(NewPlayerQuestData.CurrentQuest.IncrementGoal))}"
                    QuestData := MakeQuestDisplayData(NewPlayerQuestData.CurrentQuest.QuestLineName, NewPlayerQuestData.CurrentQuest.Name, CurrentQuestProgress, NewPlayerQuestData.CurrentQuest.ShowCost)

                    UIManager.UIQuestDisplay.InitUI(InPlayer, QuestData)
                    UIManager.UIQuestDisplay.AddUI(InPlayer)
                    spawn {QuestUpdate(InPlayer, NewPlayerQuestData.IncrementEvent)}
                # Set QuestStart to true so that this only happens when the player joins for the first time and not on restart anything
                # Update 1.2 - Attempt to fix Quest UI showing empty quest
                set NewPlayerQuestData.QuestStart = true

            if(NewPlayerStats := UIManager.PlayerStatsManager.GetPlayerStats[InPlayer]):
                if(NewPlayerStats.CurrentQuest = "" or NewPlayerStats.CurrentQuest = "Quest Name"):
                    spawn{UIManager.UIQuestDisplay.HideUI(InPlayer)}
            
    
    # Responsible for updating the player's quest to next one in quest line and UI
    QuestUpdate(InPlayer : player, IncrementEvent : event())<suspends>:void=
        var CanAfford : logic = false
        loop:
            Sleep(0.25)
            if(CurrentQuest := PerPlayerQuestDataMap[InPlayer].CurrentQuest):
                # Update 1.7
                if(CurrentQuest.InvisibleQuest = true):
                    # HIDE UI
                    UIManager.UIQuestDisplay.HideUI(InPlayer)
                    #Print("HIDE UI")
                else:
                    UIManager.UIQuestDisplay.AddUI(InPlayer)
                    #Print("ADD UI")

                if(CurrentQuest.Name = "" or CurrentQuest.Name = "Quest Name"):
                    UIManager.UIQuestDisplay.HideUI(InPlayer)

                CurrentQuestProgress := "{BigNumber_FormatDisplay(CurrentQuest.Increment)}/{BigNumber_FormatDisplay(BigNumber_ConvertFloatIntoBigNumber(CurrentQuest.IncrementGoal))}"
                UIManager.PlayerStatsManager.ChangeCurrentQuestIncrement(InPlayer, CurrentQuest.Increment)

                # Check if Player can afford the cost
                BNCost := BigNumber_FormatDisplay(CurrentQuest.Cost)

                if(CurrentQuest.ShowCost = true):
                    case(CurrentQuest.DataToUse):
                        DataToTrack.Stone =>
                            PlayerResource := UIManager.PlayerStatsManager.GetPlayerOreAmount(InPlayer, PropResourceType.Stone)
                            set CanAfford = BigNumber_GreaterThanOrEquals(PlayerResource, CurrentQuest.Cost)
                        _ =>
                        
                    
                UIManager.UIQuestDisplay.UpdateUI(InPlayer, CurrentQuest.QuestLineName, CurrentQuest.Name, CurrentQuestProgress, CurrentQuest.ShowCost, BNCost, CanAfford)
                CurrentQuestDone := CurrentQuest.CheckForCompletion(InPlayer)
                if(CurrentQuestDone = true):
                    if(CurrentQuest.InvisibleQuest = false):
                        QuestCompleteAudio.Play(InPlayer)
                    UIManager.PlayerStatsManager.AddCompletedQuest(InPlayer, CurrentQuest.Name)
                    if(CurrentQuest.EndQuestLine = true):
                        Sleep(1.0)
                        UIManager.PlayerStatsManager.AddCompletedQuestLine(InPlayer, CurrentQuest.QuestLineName)

                    Print("Quest Done")
                    # We construct a new quest so that each player's quest is unique
                    SetPlayerQuest(InPlayer, GetNextQuest(InPlayer))

                    if(NewQuest := PerPlayerQuestDataMap[InPlayer].CurrentQuest):
                        # Update 1.7
                        if(NewQuest.InvisibleQuest = true):
                            # HIDE UI
                            UIManager.UIQuestDisplay.HideUI(InPlayer)
                            Print("HIDE UI")
                        else:
                            UIManager.UIQuestDisplay.AddUI(InPlayer)
                            Print("ADD UI")

                        CurrentQuest.CancelSubscriptions()
                        NextQuestProgress := "{BigNumber_FormatDisplay(NewQuest.Increment)}/{BigNumber_FormatDisplay(BigNumber_ConvertFloatIntoBigNumber(NewQuest.IncrementGoal))}"
                        UIManager.PlayerStatsManager.ChangeCurrentQuestIncrement(InPlayer, NewQuest.Increment)
                        var NewCanAfford : logic = false
                        BNNewCost := BigNumber_FormatDisplay(NewQuest.Cost)
                        
                        if(NewQuest.ShowCost = true):
                            case(NewQuest.DataToUse):
                                DataToTrack.Stone =>
                                    PlayerResource := UIManager.PlayerStatsManager.GetPlayerOreAmount(InPlayer, PropResourceType.Stone)
                                    set NewCanAfford = BigNumber_GreaterThanOrEquals(PlayerResource, CurrentQuest.Cost)
                                _ =>
                                    
                                    
                        UIManager.UIQuestDisplay.UpdateUI(InPlayer, NewQuest.QuestLineName, NewQuest.Name, NextQuestProgress, NewQuest.ShowCost, BNNewCost, NewCanAfford)
                        Print("Next Quest Set")

    GetNextQuest(InPlayer : player):CustomQuest=
        if(CurrentQuest := PerPlayerQuestDataMap[InPlayer].CurrentQuest):
            QuestLine := SearchForQuestLine(CurrentQuest.QuestLineName)
            for(Index -> IndexedQuest : QuestLine.CustomQuests):
                if(IndexedQuest.Name = CurrentQuest.Name):
                    if(IndexedQuest.EndQuestLine = true):
                        if(QuestLine.NextQuestLine = "END"):
                            # Clear current quest in player data
                            UIManager.PlayerStatsManager.ChangeCurrentQuest(InPlayer, "")
                            UIManager.PlayerStatsManager.ChangeCurrentQuestIncrement(InPlayer, big_number{})
                            spawn{UIManager.UIQuestDisplay.HideUI(InPlayer)}
                            
                        NextQuestLine := SearchForQuestLine(QuestLine.NextQuestLine)
                        if(FirstQuest := NextQuestLine.CustomQuests[0], NextQuest := SearchForQuest(FirstQuest.Name)):
                            set NextQuest.QuestLineName = NextQuestLine.Name
                            return NextQuest
                    if(NextQuest := QuestLine.CustomQuests[Index + 1]):
                        set NextQuest.QuestLineName = CurrentQuest.QuestLineName
                        return NextQuest
        Print("Next Quest Not Found")
        return CustomQuest{}

    # Call this function to start a new quest
    StartQuest(InPlayer : player, InQuest : CustomQuest):void=
        SetPlayerQuest(InPlayer, InQuest)

        if(PlayerQuestData := PerPlayerQuestDataMap[InPlayer]):
            CurrentQuestProgress := "{BigNumber_FormatDisplay(PlayerQuestData.CurrentQuest.Increment)}/{BigNumber_FormatDisplay(BigNumber_ConvertFloatIntoBigNumber(PlayerQuestData.CurrentQuest.IncrementGoal))}"
            QuestData := MakeQuestDisplayData(PlayerQuestData.CurrentQuest.QuestLineName, PlayerQuestData.CurrentQuest.Name, CurrentQuestProgress, PlayerQuestData.CurrentQuest.ShowCost)
            UIManager.UIQuestDisplay.InitUI(InPlayer, QuestData)
            UIManager.UIQuestDisplay.AddUI(InPlayer)

    # Sets the given player's quest to new one
    SetPlayerQuest(InPlayer : player, InQuest : CustomQuest):void=
        if(PlayerData := PerPlayerQuestDataMap[InPlayer]):
            NewQuest := MakeCustomQuest(InQuest, PlayerData.IncrementEvent)
            set PlayerData.CurrentQuest = NewQuest
            PlayerData.CurrentQuest.SetUp(InPlayer, UIManager.PlayerStatsManager)
            UIManager.PlayerStatsManager.ChangeCurrentQuest(InPlayer, NewQuest.Name)

    # Searches all QuestLines for the quest given a name
    SearchForQuest(InQuestName : string)<transacts>:CustomQuest=
        for(QuestLine : QuestLines):
            for(Quest : QuestLine.CustomQuests):
                if(Quest.Name = InQuestName):
                    Print("{InQuestName} : Quest Found")
                    set Quest.QuestLineName = QuestLine.Name
                    return Quest
        Print("{InQuestName} : Quest Not Found")
        return CustomQuest{}

    # Searches through all QuestLines for that quest line
    SearchForQuestLine(InQuestLineName : string):CustomQuestLine=
        for(QuestLine : QuestLines):
            if(QuestLine.Name = InQuestLineName):
                Print("{InQuestLineName} : Quest Line Found")
                return QuestLine
        Print("{InQuestLineName} : Quest Line Not Found")
        return CustomQuestLine{}

    SkipQuestButtonActivated(InAgent : agent):void=
        Print("Skip Quest")
        if(Player := player[InAgent], CurrentQuest := PerPlayerQuestDataMap[Player].CurrentQuest):
            if(CurrentQuest.Name <> "Quest Name"):
                set CurrentQuest.Increment = BigNumber_ConvertFloatIntoBigNumber(CurrentQuest.IncrementGoal)
