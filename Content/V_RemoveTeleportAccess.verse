
using { /Fortnite.com/Devices }
using { /Verse.org/Simulation }
using { /UnrealEngine.com/Temporary/Diagnostics }

# See https://dev.epicgames.com/documentation/en-us/uefn/create-your-own-device-in-verse for how to create a verse device.

# A Verse-authored creative device that can be placed in a level
V_RemoveTeleportAccess := class(creative_device):

    @editable
    MutatorZones : []mutator_zone_device = array{}
    @editable
    UIManager : V_UIManagerDevice = V_UIManagerDevice{}

    # Runs when the device is started in a running game
    OnBegin<override>()<suspends>:void=
        spawn:
            CheckPlayerLoop()

    CheckPlayerLoop()<suspends>:void=
        loop:
            for(InPlayer : GetPlayspace().GetPlayers()):
                var PlayerChecked : logic = false
                
                for(MutatorZone : MutatorZones):
                    if(PlayerChecked = false):
                        if(MutatorZone.IsInVolume[InPlayer]):
                            UIManager.UIPlayerMenuClass.SetCanTeleportHUB(InPlayer, false)
                            set PlayerChecked = true
                            #Print("No Tele")
                        else:
                            UIManager.UIPlayerMenuClass.SetCanTeleportHUB(InPlayer, true)
                            #Print("Can Tele")
            Sleep(0.1)
            