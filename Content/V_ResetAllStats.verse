
using { /Fortnite.com/Devices }
using { /Verse.org/Simulation }
using { /UnrealEngine.com/Temporary/Diagnostics }

# See https://dev.epicgames.com/documentation/en-us/uefn/create-your-own-device-in-verse for how to create a verse device.

# A Verse-authored creative device that can be placed in a level
V_ResetAllStats := class(creative_device):

    @editable
    ResetButton : button_device = button_device{}

    @editable
    PlayerManager : ?V_PlayerManagerDevice = false
    @editable
    ItemClearer : item_granter_device = item_granter_device{}
    @editable
    SignalGranter : item_granter_device = item_granter_device{}

    var ForceFieldVendors : []V_ForceFieldVendor = array{}

    # Runs when the device is started in a running game
    OnBegin<override>()<suspends>:void=
        ResetButton.InteractedWithEvent.Subscribe(OnButtonInteract)

        TaggedList1 := FindCreativeObjectsWithTag(ForcefieldVendorTag{})
        for(TaggedActor : TaggedList1):
            if(ForceFieldVendorClass := V_ForceFieldVendor[TaggedActor]):
                set ForceFieldVendors += array{ForceFieldVendorClass}

    OnButtonInteract(InAgent : agent):void=
        Reset(InAgent)

    Reset(InAgent : agent):void=
        if(bPlayerManager := PlayerManager?, Player := player[InAgent]):
            ItemClearer.GrantItem(Player)
            if(PropFollower := bPlayerManager.UIManagerDevice.ArtifactManager.PropFollower?, PlayerStatsData := bPlayerManager.PlayerStatManager.GetPlayerStats[InAgent]):
                for(Artifact : PlayerStatsData.EquippedPlayerArtifacts):
                    PropFollower.DisposeProp(Player, Artifact.EquippedSlot)
            bPlayerManager.PlayerStatManager.ResetStatsDEBUG(InAgent)
            #bPlayerManager.QuestSystem.PlayerSetUp(Player)
        
        #Reset all forcefield vendors
        for(ForceFieldVendor : ForceFieldVendors):
            Print("Resetting Forcefield Vendors")
            ForceFieldVendor.ResetVendorOnPrestige(InAgent)