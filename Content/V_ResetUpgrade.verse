
using { /Fortnite.com/Devices }
using { /Verse.org/Simulation }
using { /UnrealEngine.com/Temporary/Diagnostics }

# See https://dev.epicgames.com/documentation/en-us/uefn/create-your-own-device-in-verse for how to create a verse device.

# A Verse-authored creative device that can be placed in a level
V_ResetUpgrade := class(creative_device):
    @editable
    ResetButton : button_device = button_device{}
    @editable
    UpgradeToReset : PlayerUpgradeType = PlayerUpgradeType.Null

    @editable
    PlayerManager : ?V_PlayerManagerDevice = false

    var ForceFieldVendors : []V_ForceFieldVendor = array{}

    # Runs when the device is started in a running game
    OnBegin<override>()<suspends>:void=
        ResetButton.InteractedWithEvent.Subscribe(OnButtonInteract)

        TaggedList1 := FindCreativeObjectsWithTag(ForcefieldVendorTag{})
        for(TaggedActor : TaggedList1):
            if(ForceFieldVendorClass := V_ForceFieldVendor[TaggedActor]):
                set ForceFieldVendors += array{ForceFieldVendorClass}

    OnButtonInteract(InAgent : agent):void=
        if(bPlayerManager := PlayerManager?):
            RestLevel := bPlayerManager.PlayerStatManager.GetPlayerUpgradeLevel(InAgent, UpgradeToReset) * -1
            bPlayerManager.PlayerStatManager.UpdatePlayerUpgrade(InAgent, UpgradeToReset, RestLevel)
