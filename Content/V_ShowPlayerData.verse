
using { /Fortnite.com/Devices }
using { /Verse.org/Simulation }
using { /UnrealEngine.com/Temporary/Diagnostics }

# See https://dev.epicgames.com/documentation/en-us/uefn/create-your-own-device-in-verse for how to create a verse device.

# A Verse-authored creative device that can be placed in a level
V_ShowPlayerData := class(creative_device):

    @editable
    DisplayOreOnHUD : hud_message_device = hud_message_device{}

    @editable
    DisplayCoinOnHUD : hud_message_device = hud_message_device{}

    #Purely For Debugging
    @editable
    DisplayOreStats : hud_message_device = hud_message_device{}

    @editable
    DisplayCoinStats: hud_message_device = hud_message_device{}

    @editable
    PlayerManager : V_PlayerManagerDevice = V_PlayerManagerDevice{}

    @editable
    PlayerVolume : volume_device = volume_device{}

    @editable
    StoneVendor : V_VendorDevice = V_VendorDevice{}

    @editable
    var AlgorithmAndFunctionsClassReference: ?V_AlgorithmsAndFunctions = false

    # Manages and updates stats for each player.
    PlayerStatsManager: player_stats_manager = player_stats_manager{}

    SetOreText<localizes>(FloatValue:float) : message = "Ore Amount = {FloatValue}"
    SetCoinText<localizes>(FloatValue:float) : message = "Gold Amount = {FloatValue}"

    ShowOreStats<localizes>(FloatValue1:float, FloatValue2:float, FloatValue3:float) : message = "Ore Upgrade Level = {FloatValue1}, Ore Per Hit Amount = {FloatValue2}, Ore Cost Next Upgrade = {FloatValue3} "
    ShowCoinStats<localizes>(FloatValue1:float, FloatValue2:float, FloatValue3:float) : message = "Coin Upgrade Level = {FloatValue1}, Coin Per Trade Amount = {FloatValue2}, Coin Cost Next Upgrade = {FloatValue3} "

    # Runs when the device is started in a running game
    OnBegin<override>()<suspends>:void=
        PlayerVolume.AgentEntersEvent.Subscribe(SetupOreHUD)
        
    SetupOreHUD(InAgent : agent):void=
        Print("Setup Ore HUD")
        spawn:
            UpdateOreHUD(InAgent)

    UpdateOreHUD(InAgent : agent)<suspends>:void=  
        loop:
            if(OldData := PlayerStatsDataMap[InAgent]):
                if(CurrentPlayerStats := PlayerStatsManager.GetPlayerStats[InAgent]):  
                    if(bAlgorithmAndFunctionsClass := AlgorithmAndFunctionsClassReference?):
                            OreAmount := PlayerStatsManager.GetPlayerStoneOreAmount(InAgent)
                            #CoinAmount := PlayerStatsManager.GetPlayerCoinAmount(InAgent)
                            #DisplayOreOnHUD.SetText(SetOreText(OreAmount))
                            #DisplayCoinOnHUD.SetText(SetCoinText(CoinAmount))
                            DisplayOreOnHUD.Show(InAgent)
                            DisplayCoinOnHUD.Show(InAgent)

                            #Purely For Debugging Stats                   
                            OreUpgradeLevel := PlayerStatsManager.GetPlayerOreUpgrade1Level(InAgent)
                            OreHarvestRate := PlayerStatsManager.GetPlayerOreHarvestRatePerHit(InAgent)
                            #OreCostOfNextUpgrade := bAlgorithmAndFunctionsClass.AlgorithmAndFunctionClassReference.CalculateCostOfNextUpgrade(10.0,OldData.HRCostIncreaseRate,  OldData.HRUpgrade1Level)
                            #Print("Ore Cost ={OreCostOfNextUpgrade}")
                            StoneVendorUpgradeLevel := PlayerStatsManager.GetStoneVendorUpgradeLevel(InAgent)
                            StoneVendorTradeInGain := PlayerStatsManager.GetStoneVendorCoinPerTrade(InAgent)
                            #StoneVendorCostOfNextUpgrade := bAlgorithmAndFunctionsClass.AlgorithmAndFunctionClassReference.CalculateCostOfNextUpgrade(10.0,StoneVendor.VendorCostIncreaseRate,  CurrentPlayerStats.StoneVendorUpgradeLevel) #  CurrentPlayerStats.CoinCostIncreaseRate

                            #DisplayOreStats.SetText(ShowOreStats(OreUpgradeLevel, OreHarvestRate, OreCostOfNextUpgrade))
                            #DisplayCoinStats.SetText(ShowCoinStats(StoneVendorUpgradeLevel, StoneVendorTradeInGain, StoneVendorCostOfNextUpgrade ))

                            DisplayOreStats.Show(InAgent)
                            DisplayCoinStats.Show(InAgent)
                            #End of Stats Debugging
            Sleep(0.1)