using { /Fortnite.com/Devices }
using { /Verse.org/Simulation }
using { /UnrealEngine.com/Temporary/Diagnostics }

V_TeleporterManager := class(creative_device):

    UITeleporterDisplay : UI_TeleporterDisplay = UI_TeleporterDisplay{}

    @editable
    AltarTrigger : trigger_device = trigger_device{}
    @editable
    ShopsTrigger : trigger_device = trigger_device{}
    @editable
    CrystalTrigger : trigger_device = trigger_device{}
    @editable
    DungeonUpgradesTrigger : trigger_device = trigger_device{}

    #Mushroom Caves
    @editable
    MushroomCave1Trigger : trigger_device = trigger_device{}
    @editable
    MushroomCave2Trigger : trigger_device = trigger_device{}
    @editable
    MushroomCave3Trigger : trigger_device = trigger_device{}
    @editable
    MushroomCave4Trigger : trigger_device = trigger_device{}

    #Mushroom Caves
    @editable
    LavaCave1Trigger : trigger_device = trigger_device{}
    @editable
    LavaCave2Trigger : trigger_device = trigger_device{}
    @editable
    LavaCave3Trigger : trigger_device = trigger_device{}
    @editable
    LavaCave4Trigger : trigger_device = trigger_device{}

    #Mushroom Caves
    @editable
    UnderworldCave1Trigger : trigger_device = trigger_device{}
    @editable
    UnderworldCave2Trigger : trigger_device = trigger_device{}
    @editable
    UnderworldCave3Trigger : trigger_device = trigger_device{}
    @editable
    UnderworldCave4Trigger : trigger_device = trigger_device{}

    @editable
    MagicCave1Trigger : trigger_device = trigger_device{}
    @editable
    MagicCave2Trigger : trigger_device = trigger_device{}
    @editable
    MagicCave3Trigger : trigger_device = trigger_device{}
    @editable
    MagicCave4Trigger : trigger_device = trigger_device{}

    @editable
    TownUITrigger : trigger_device = trigger_device{}
    @editable
    MushroomUITrigger : trigger_device = trigger_device{}
    @editable
    LavaUITrigger : trigger_device = trigger_device{}
    @editable
    UnderworldUITrigger : trigger_device = trigger_device{}
    @editable
    MagicUITrigger : trigger_device = trigger_device{}

    @editable
    CharacterDevices : []character_device = array{}
    @editable
    ForceFieldVendors : []V_ForceFieldVendor = array{}
    var TeleporterItems : []TeleporterItemData = array{}

    #TODO: I can add a show first so we can have different teleporters!?
    # Runs when the device is started in a running game
    OnBegin<override>()<suspends>:void=
        for(Character : CharacterDevices):
            Character.InteractedWithEvent.Subscribe(OnCharacterInteract)

        #Hide UI after trigger event
        AltarTrigger.TriggeredEvent.Subscribe(OnTriggered)
        ShopsTrigger.TriggeredEvent.Subscribe(OnTriggered)
        CrystalTrigger.TriggeredEvent.Subscribe(OnTriggered)
        DungeonUpgradesTrigger.TriggeredEvent.Subscribe(OnTriggered)

        MushroomCave1Trigger.TriggeredEvent.Subscribe(OnTriggered)
        MushroomCave2Trigger.TriggeredEvent.Subscribe(OnTriggered)
        MushroomCave3Trigger.TriggeredEvent.Subscribe(OnTriggered)
        MushroomCave4Trigger.TriggeredEvent.Subscribe(OnTriggered)

        LavaCave1Trigger.TriggeredEvent.Subscribe(OnTriggered)
        LavaCave2Trigger.TriggeredEvent.Subscribe(OnTriggered)
        LavaCave3Trigger.TriggeredEvent.Subscribe(OnTriggered)
        LavaCave4Trigger.TriggeredEvent.Subscribe(OnTriggered)

        UnderworldCave1Trigger.TriggeredEvent.Subscribe(OnTriggered)
        UnderworldCave2Trigger.TriggeredEvent.Subscribe(OnTriggered)
        UnderworldCave3Trigger.TriggeredEvent.Subscribe(OnTriggered)
        UnderworldCave4Trigger.TriggeredEvent.Subscribe(OnTriggered)
        
        MagicCave1Trigger.TriggeredEvent.Subscribe(OnTriggered)
        MagicCave2Trigger.TriggeredEvent.Subscribe(OnTriggered)
        MagicCave3Trigger.TriggeredEvent.Subscribe(OnTriggered)
        MagicCave4Trigger.TriggeredEvent.Subscribe(OnTriggered)

        #Setting the triggers to these so I can use them later rather then passing through a reference
        set UITeleporterDisplay.TownButtonTrigger = TownUITrigger
        set UITeleporterDisplay.MushroomButtonTrigger = MushroomUITrigger
        set UITeleporterDisplay.LavaButtonTrigger = LavaUITrigger
        set UITeleporterDisplay.UnderworldButtonTrigger = UnderworldUITrigger
        set UITeleporterDisplay.MagicButtonTrigger = MagicUITrigger


        #Trigger Events for changing pages
        TownUITrigger.TriggeredEvent.Subscribe(ShowTown)
        MushroomUITrigger.TriggeredEvent.Subscribe(ShowMushroomCaves)
        LavaUITrigger.TriggeredEvent.Subscribe(ShowLavaCaves)
        UnderworldUITrigger.TriggeredEvent.Subscribe(ShowUnderworldCaves)
        MagicUITrigger.TriggeredEvent.Subscribe(ShowMagicCaves)
    
    #This is called everytime the UI is setup via the teleporter shop
    Init(InPlayer : player):void=
        TownSpot1 := MakeTeleporterItem(VerseExposed.T_Altar, "Altar", AltarTrigger)
        TownSpot2 := MakeTeleporterItem(VerseExposed.T_Shops, "Shops", ShopsTrigger)
        TownSpot3 := MakeTeleporterItem(VerseExposed.T_SpawnArea, "Crystal", CrystalTrigger)
        TownSpot4 := MakeTeleporterItem(VerseExposed.T_DungeonUpgrades, "Dungeon Upgrades", DungeonUpgradesTrigger)

        MushroomCave1 := MakeTeleporterItem(VerseExposed.T_MushroomCave1, "Mushroom Cave 1", MushroomCave1Trigger)
        MushroomCave2 := MakeTeleporterItem(VerseExposed.T_MushroomCave2, "Mushroom Cave 2", MushroomCave2Trigger)
        MushroomCave3 := MakeTeleporterItem(VerseExposed.T_MushroomCave3, "Mushroom Cave 3", MushroomCave3Trigger)
        MushroomCave4 := MakeTeleporterItem(VerseExposed.T_MushroomCave4, "Mushroom Cave 4", MushroomCave4Trigger)

        LavaCave1 := MakeTeleporterItem(VerseExposed.T_LavaCave1, "Lava Cave 1", LavaCave1Trigger)
        LavaCave2 := MakeTeleporterItem(VerseExposed.T_LavaCave2, "Lava Cave 2", LavaCave2Trigger)
        LavaCave3 := MakeTeleporterItem(VerseExposed.T_LavaCave3, "Lava Cave 3", LavaCave3Trigger)
        LavaCave4 := MakeTeleporterItem(VerseExposed.T_LavaCave4, "Lava Cave 4", LavaCave4Trigger)

        UnderworldCave1 := MakeTeleporterItem(VerseExposed.T_UnderworldCave1, "Underworld Cave 1", UnderworldCave1Trigger)
        UnderworldCave2 := MakeTeleporterItem(VerseExposed.T_UnderworldCave2, "Underworld Cave 2", UnderworldCave2Trigger)
        UnderworldCave3 := MakeTeleporterItem(VerseExposed.T_UnderworldCave3, "Underworld Cave 3", UnderworldCave3Trigger)
        UnderworldCave4 := MakeTeleporterItem(VerseExposed.T_UnderworldCave4, "Underworld Cave 4", UnderworldCave4Trigger)

        MagicCave1 := MakeTeleporterItem(VerseExposed.T_MagicCave1, "Ancient Library 1", MagicCave1Trigger)
        MagicCave2 := MakeTeleporterItem(VerseExposed.T_MagicCave2, "Ancient Library 2", MagicCave2Trigger)
        MagicCave3 := MakeTeleporterItem(VerseExposed.T_MagicCave3, "Ancient Library 3", MagicCave3Trigger)
        MagicCave4 := MakeTeleporterItem(VerseExposed.T_MagicCave4, "Ancient Library 4", MagicCave4Trigger)

        UITeleporterDisplay.InitUI(InPlayer, 
        array{TownSpot1, TownSpot2, TownSpot3, TownSpot4}, 
        array{MushroomCave1, MushroomCave2, MushroomCave3, MushroomCave4}, 
        array{LavaCave1, LavaCave2, LavaCave3, LavaCave4},
        array{UnderworldCave1, UnderworldCave2, UnderworldCave3, UnderworldCave4},
        array{MagicCave1, MagicCave2, MagicCave3, MagicCave4})

    ShowTown(InAgent : ?agent):void=
        Print("Show Town Caves")
        if(bInAgent := InAgent?):
            if(InPlayer := player[bInAgent]):
                spawn:
                    DelayShowTown(InPlayer)

    DelayShowTown(InPlayer : player)<suspends>:void=
        Sleep(0.0)
        #Add the button
        var BarriersOpen : []logic = array{}
        for(Index -> ForceFieldVendor : ForceFieldVendors):
            if(ForceFieldVendor.Switch.GetCurrentState[InPlayer]):
                set BarriersOpen += array{true}
                Print("{Index} TRUE")
            else:
                set BarriersOpen += array{false}

        OutBarriersOpen := array{true, true, true, true}
        UITeleporterDisplay.UpdateUI(InPlayer, OutBarriersOpen, true, false, false, false, false)
    
    ShowMushroomCaves(InAgent : ?agent):void=
        Print("Show Mushroom Caves")
        if(bInAgent := InAgent?):
            Print("Get the agent")
            if(InPlayer := player[bInAgent]):
                spawn:
                    DelayShowMushroomCaves(InPlayer)

    DelayShowMushroomCaves(InPlayer : player)<suspends>:void=
        Sleep(0.0)

        #Add the button
        var BarriersOpen : []logic = array{}
        for(Index -> ForceFieldVendor : ForceFieldVendors):
            if(ForceFieldVendor.Switch.GetCurrentState[InPlayer]):
                set BarriersOpen += array{true}
                Print("{Index} TRUE")
            else:
                set BarriersOpen += array{false}

        if(OutBarriersOpen := array{BarriersOpen[4], BarriersOpen[5], BarriersOpen[6], BarriersOpen[7]}):
            UITeleporterDisplay.UpdateUI(InPlayer, OutBarriersOpen, false, true, false, false, false)


    ShowLavaCaves(InAgent : ?agent):void=
        Print("Show Lava Caves")
        if(bInAgent := InAgent?):
            if(InPlayer := player[bInAgent]):
                spawn:
                    DelayShowLavaCaves(InPlayer)

    DelayShowLavaCaves(InPlayer : player)<suspends>:void=
        Sleep(0.0)

        #Add the button
        var BarriersOpen : []logic = array{}
        for(Index -> ForceFieldVendor : ForceFieldVendors):
            if(ForceFieldVendor.Switch.GetCurrentState[InPlayer]):
                set BarriersOpen += array{true}
                Print("{Index} TRUE")
            else:
                set BarriersOpen += array{false}

        if(OutBarriersOpen := array{BarriersOpen[8], BarriersOpen[9], BarriersOpen[10], BarriersOpen[11]}):
            UITeleporterDisplay.UpdateUI(InPlayer, OutBarriersOpen, false, false, true, false, false)

    ShowUnderworldCaves(InAgent : ?agent):void=
        Print("Show Underworld Caves")
        if(bInAgent := InAgent?):
            if(InPlayer := player[bInAgent]):
                spawn:
                    DelayShowUnderworld(InPlayer)
                    

    DelayShowUnderworld(InPlayer : player)<suspends>:void=
        Sleep(0.0)
        #Add the button
        var BarriersOpen : []logic = array{}
        for(Index -> ForceFieldVendor : ForceFieldVendors):
            if(ForceFieldVendor.Switch.GetCurrentState[InPlayer]):
                set BarriersOpen += array{true}
                Print("{Index} TRUE")
            else:
                set BarriersOpen += array{false}

        if(OutBarriersOpen := array{BarriersOpen[12], BarriersOpen[13], BarriersOpen[14], BarriersOpen[15]}):
            UITeleporterDisplay.UpdateUI(InPlayer, OutBarriersOpen, false, false, false, true, false)

    ShowMagicCaves(InAgent : ?agent):void=
        Print("Show Magic Caves")
        if(bInAgent := InAgent?):
            if(InPlayer := player[bInAgent]):
                spawn:
                    DelayShowMagic(InPlayer)
                    

    DelayShowMagic(InPlayer : player)<suspends>:void=
        Sleep(0.0)
        #Add the button
        var BarriersOpen : []logic = array{}
        for(Index -> ForceFieldVendor : ForceFieldVendors):
            if(ForceFieldVendor.Switch.GetCurrentState[InPlayer]):
                set BarriersOpen += array{true}
                Print("{Index} TRUE")
            else:
                set BarriersOpen += array{false}

        if(OutBarriersOpen := array{BarriersOpen[16], BarriersOpen[17], BarriersOpen[18], BarriersOpen[19]}):
            UITeleporterDisplay.UpdateUI(InPlayer, OutBarriersOpen, false, false, false, false, true)

    OnCharacterInteract(InAgent : agent):void=
        if(InPlayer := player[InAgent]):
            #Default show town on interaction
            spawn:
                DelayShowTown(InPlayer)
            UITeleporterDisplay.AddUI(InPlayer)

    OnTriggered(InMaybeAgent : ?agent):void=
        if(Player := player[InMaybeAgent?]):
            spawn:
                UITeleporterDisplay.HideUI(Player)