using { /Verse.org/Simulation }
using { /UnrealEngine.com/Temporary/Diagnostics }
using { /UnrealEngine.com/Temporary/UI }
using { /Fortnite.com/UI }
using { /UnrealEngine.com/Temporary/SpatialMath}
using { /Fortnite.com/Devices }
using { /Verse.org/Colors }
using { /Verse.org/Assets }

###################################################################################################################################
############################################################## ENUMS ##############################################################
###################################################################################################################################
ImageBackgroundType := enum:
    DailyRewards
    TimedRewards

###################################################################################################################################
####################################################### CUSTOM BUTTON CLASS #######################################################
###################################################################################################################################
custom_button := class<concrete>():
    _StringToMessage<localizes>(String : string) : message = "{String}"
    _EmptyMessage<localizes>() : message = ""
    
    StringToMessage(String : string) : message = if(String.Length > 0) then _StringToMessage(String) else _EmptyMessage()

    @editable
    DisplayText : string = "Button"
    @editable
    PositionX : float = 0.0
    @editable
    PositionY : float = 0.0
    @editable
    SizeX : float = 250.0
    @editable
    SizeY : float = 50.0
    @editable
    DoesExitUI : logic = false
    @editable
    IsExitButton : logic = false
    @editable
    var Enabled : logic = false

    @editable
    OptEventTrigger : ?trigger_device = false

    var TimedRewardsManager : V_TimedRewardsManager = V_TimedRewardsManager{}

    var Button : text_button_base = button_loud{}
    var CurrentIcon : texture_block = texture_block{DefaultImage := VerseExposed.T_ChestUI_Closed_Red}
    #var CurrentIcon : texture = VerseExposed.T_ChestUI_Closed_Red
    var CurrentText : string = ""
    var HasPressed : logic = false

    var TotalTimeToUnlock : float = 1.0
    var CurrentTimeToUnlock : float = 1.0

    # Set up the events for the button
    ButtonStartUp():void=
        Button.OnClick().Subscribe(OnButtonPressed)
        set CurrentText = DisplayText

    #Called when player presses button
    OnButtonPressed(WidgetMessage : widget_message):void=
        if(Enabled = false):
            return
        set HasPressed = true
        #Button.SetEnabled(false)
        set Enabled = false

        if(IsExitButton = false):
            UpdateSlotValues("Collected", VerseExposed.T_ChestUI_Open_Red)

        #spawn:
            #TimedRewardsManager.RefreshUI(WidgetMessage.Player, WidgetMessage.Source)

        if(Trigger := OptEventTrigger?):
            Print("{DisplayText} Triggered")
            Trigger.Trigger(WidgetMessage.Player)

    #Updates Icon and Text
    UpdateSlotValues(InText : string, InIcon : texture) : void=
        CurrentIcon.SetImage(InIcon)
        set CurrentText = InText
        Button.SetText(StringToMessage(InText))

    #Create Button Widget
    MakeCanvasButton() : canvas_slot=
        # Was still directly setting button to the Custom_Button's Button variable, Needed to create a new one seen below - Charlie
        NewButton := button_loud:
            DefaultText := StringToMessage("Button")
        set Button = NewButton
        Size := vector2{X:=SizeX, Y:=SizeY}
        NewPosition := vector2{X:=PositionX, Y:=PositionY + 125.0}
        Widget := overlay:
            Slots := array:
                overlay_slot:
                    HorizontalAlignment := horizontal_alignment.Fill
                    VerticalAlignment := vertical_alignment.Fill
                    Widget := NewButton
        return MakeCanvasSlot(Widget, NewPosition, ?Size:=Size, ?Alignment := vector2{X:=0.5, Y:=0.5})

    #Create Icon
    MakeIcon(InIcon : texture) : canvas_slot=
        # Passing in the button size to the texture_block's creation - Charlie
        Size := vector2{X:=275.0, Y:=275.0}
        NewPosition := vector2{X:=PositionX, Y:=PositionY}
        Widget := texture_block{DefaultImage := InIcon, DefaultDesiredSize := Size}
        set CurrentIcon = Widget
        return MakeCanvasSlot(Widget, NewPosition, ?Size:=Size, ?Alignment := vector2{X:=0.5, Y:=0.5})

#Constructor for Custom Button
MakeCustomButton<constructor>(OldButtonData : custom_button)<transacts> := custom_button:
    DisplayText := OldButtonData.DisplayText
    PositionX := OldButtonData.PositionX
    PositionY := OldButtonData.PositionY
    SizeX := OldButtonData.SizeX
    SizeY := OldButtonData.SizeY
    DoesExitUI := OldButtonData.DoesExitUI
    IsExitButton := OldButtonData.IsExitButton
    OptEventTrigger := OldButtonData.OptEventTrigger
    TimedRewardsManager := OldButtonData.TimedRewardsManager
    Button := OldButtonData.Button
    CurrentIcon := OldButtonData.CurrentIcon
    CurrentText := OldButtonData.CurrentText
    HasPressed := OldButtonData.HasPressed
    TotalTimeToUnlock := OldButtonData.TotalTimeToUnlock
    CurrentTimeToUnlock := OldButtonData.CurrentTimeToUnlock

###################################################################################################################################
####################################################### CUSTOM CANVAS CLASS #######################################################
###################################################################################################################################
TimedRewardsMenuCanvasData := class:
    RewardCanvas : canvas
    var RewardButtonsToDisplay : []custom_button = array{}
#####################################################################################################################################
####################################################### TIMED REWARDS MANAGER #######################################################
#####################################################################################################################################
V_TimedRewardsManager := class(creative_device):
    _EmptyMessage<localizes>() : message = ""
    _StringToMessage<localizes>(String : string) : message = "{String}"
    StringToMessage(String : string) : message = if(String.Length > 0) then _StringToMessage(String) else _EmptyMessage()

    @editable
    InteractTrigger : trigger_device = trigger_device{}
    @editable
    ImageBackground : ImageBackgroundType = ImageBackgroundType.TimedRewards
    @editable
    ButtonsToDisplay : []custom_button = array{}
    @editable
    UnlockingTimes : []float = array{}

    var TimedRewardsCanvasDataMap : [player]TimedRewardsMenuCanvasData = map{}
    var ShownToPlayer : [player]logic = map{}

    DailyRewards : texture = VerseExposed.T_UIBackground
    TimedRewards : texture = VerseExposed.T_RewardsBackPlate_FullScreen

    # Runs when the device is started in a running game
    OnBegin<override>()<suspends>:void=
        InteractTrigger.TriggeredEvent.Subscribe(OnCharacterInteract)
        GetPlayspace().PlayerAddedEvent().Subscribe(OnPlayerAdded)

        AllPlayers := GetPlayspace().GetPlayers()
        for(Player : AllPlayers):
            PlayerSetup(Player)

    #Sets up when players join
    OnPlayerAdded<private>(Player : player) : void =
        PlayerSetup(Player)

    #Sets up player on join and start
    PlayerSetup(Player : player) : void=
        InitUI(Player)
        if(set ShownToPlayer[Player] = false):
            SetUpRewardButtons(Player)

    #Sets up timers for each reward button
    SetUpRewardButtons(Player : player) : void=
        if(TimedRewardsCanvas := TimedRewardsCanvasDataMap[Player]):
            for(Index -> CustomButton : TimedRewardsCanvas.RewardButtonsToDisplay):
                if(CustomButton.IsExitButton = false):
                    if(CustomButton.HasPressed = false):
                        if(UnlockingTime := UnlockingTimes[Index]):
                            set CustomButton.TotalTimeToUnlock = UnlockingTime
                            spawn:
                                StartTimerReward(Player, CustomButton)
                else:
                    CustomButton.Button.SetText(StringToMessage(CustomButton.CurrentText))

    #Update Timers on each Button
    UpdateTimerUI(CustomButton : custom_button) : void=
        UnlockTime := CustomButton.CurrentTimeToUnlock
        MessageConvert := StringToMessage("{CustomFormattedDuration(UnlockTime, ?HideMinutesIfZero:=true, ?HideHoursIfZero:=true)}")
        CustomButton.Button.SetText(MessageConvert)

    #Starts the Timer for each reward
    StartTimerReward(Player : player, CustomButton : custom_button)<suspends> : void=
        #CustomButton.Button.SetEnabled(false)
        set CustomButton.Enabled = false
        set CustomButton.CurrentTimeToUnlock = CustomButton.TotalTimeToUnlock
        loop:
            set CustomButton.CurrentTimeToUnlock -= 1.0
            UpdateTimerUI(CustomButton)

            if(CustomButton.CurrentTimeToUnlock <= 0.0):
                #CustomButton.Button.SetEnabled(true)
                set CustomButton.Enabled = true
                CustomButton.UpdateSlotValues("Collect", VerseExposed.T_ChestUI_Glowing_Red)
                #spawn:
                    #RefreshUI(Player, CustomButton.Button)
                break

            Sleep(1.0)

    #When player talks to character
    OnCharacterInteract(InMaybeAgent : ?agent):void=
        if(Agent := InMaybeAgent?, Player := player[Agent]):
            AddUI(Player)

    #Create UI
    InitUI(InPlayer : player) : void=
        TimedRewardsCanvas := MakeCanvas()
        for(CustomButton : TimedRewardsCanvas.RewardButtonsToDisplay):
            set CustomButton.TimedRewardsManager = Self
            CustomButton.ButtonStartUp()
            if(CustomButton.DoesExitUI = true):
                CustomButton.Button.OnClick().Subscribe(ExitUI)

        if(set TimedRewardsCanvasDataMap[InPlayer] = TimedRewardsCanvas, set ShownToPlayer[InPlayer] = true):

    #Adds UI from Players Screen
    AddUI(InPlayer : player):void=
        if(PlayerUI := GetPlayerUI[InPlayer], TimedRewardsCanvas := TimedRewardsCanvasDataMap[InPlayer]):
            TimedRewardsCanvas.RewardCanvas.SetVisibility(widget_visibility.Visible)
            PlayerUI.AddWidget(TimedRewardsCanvas.RewardCanvas, player_ui_slot{InputMode := ui_input_mode.All})

            for(CustomButton : TimedRewardsCanvas.RewardButtonsToDisplay):
                if(CustomButton.IsExitButton = true):
                    #CustomButton.Button.SetEnabled(true)
                    set CustomButton.Enabled = false

    #Resets UI on Screen
    RefreshUI(InPlayer : player, InWidget : widget)<suspends> : void=
        if(ShownToPlayer[InPlayer] = true):
            ExitUI(widget_message{Player:=InPlayer,Source:=InWidget})
            AddUI(InPlayer)

    #Removes UI from Players Screen
    ExitUI(WidgetMessage : widget_message):void=
        if(PlayerUI := GetPlayerUI[WidgetMessage.Player], TimedRewardsCanvas := TimedRewardsCanvasDataMap[WidgetMessage.Player]):
            TimedRewardsCanvas.RewardCanvas.SetVisibility(widget_visibility.Hidden)
            PlayerUI.RemoveWidget(TimedRewardsCanvas.RewardCanvas)
            if(set ShownToPlayer[WidgetMessage.Player] = false):

    #Makes UI Canvas
    MakeCanvas():TimedRewardsMenuCanvasData=
        var ImageSize : vector2 = vector2{X := 1920.0, Y := 1080.0}
        var BackgroundImage : texture_block = texture_block{DefaultImage := DailyRewards, DefaultDesiredSize := ImageSize}
        var CanvasSlots : []canvas_slot = array{}
        var Canvas : canvas = canvas{}

        if(ImageBackground = ImageBackgroundType.DailyRewards):
            set BackgroundImage = texture_block{DefaultImage := DailyRewards, DefaultDesiredSize := ImageSize}
        else if(ImageBackground = ImageBackgroundType.TimedRewards):
            set BackgroundImage = texture_block{DefaultImage := TimedRewards, DefaultDesiredSize := ImageSize}

        BackgroundSlot := canvas_slot:
            #Anchors := anchors{Minimum := vector2{X:=0.5, Y:=0.5}, Maximum := vector2{X:=0.5, Y:=0.5}}
            #Alignment := vector2{X:=0.5, Y:=0.5}
            Offsets := margin{Top := 0.0, Left := -10.0, Right := 0.0, Bottom := 0.0}
            Widget := BackgroundImage
            SizeToContent := true

        set CanvasSlots += array{BackgroundSlot}

        var TempButtonsToDisplay : []custom_button = ButtonsToDisplay
        for(Index -> Button : ButtonsToDisplay):
            if(set TempButtonsToDisplay[Index] = MakeCustomButton(Button)):

        for(CustomButton : TempButtonsToDisplay):
            set CanvasSlots += array{CustomButton.MakeCanvasButton()}
            if(CustomButton.IsExitButton = false):
                set CanvasSlots += array{CustomButton.MakeIcon(VerseExposed.T_ChestUI_Closed_Red)}

        TempCanvas := canvas:
            Slots := CanvasSlots
        set Canvas = TempCanvas

        return TimedRewardsMenuCanvasData:
            RewardCanvas := Canvas
            RewardButtonsToDisplay := TempButtonsToDisplay

    CustomFormattedDuration(Duration: float, ?HideMinutesIfZero:logic = false, ?HideHoursIfZero:logic = false)<transacts>:string=
        if(Hours := Int[(Duration / 60.0) / 60.0]):
            if(Minutes := Int[Duration / 60.0] - (Hours * 60)):
                var Seconds : int = 0
                option{ TSeconds := Mod[Floor[Duration], 60], set Seconds = TSeconds}

                if(Hours = 0, HideHoursIfZero? and Minutes = 0, HideMinutesIfZero?):
                    return "{Seconds}S"
                if(Hours = 0, HideHoursIfZero?):
                    ZeroLeadedSeconds := if(Seconds < 10) then "0{Seconds}" else "{Seconds}"
                    return "{Minutes}M {ZeroLeadedSeconds}S"
                else:
                    ZeroLeadedSeconds := if(Seconds < 10) then "0{Seconds}" else "{Seconds}"
                    return "{Hours}H {Minutes}M {ZeroLeadedSeconds}S"
        return ""