
using { /Fortnite.com/Devices }
using { /Verse.org/Simulation }
using { /UnrealEngine.com/Temporary/Diagnostics }

# See https://dev.epicgames.com/documentation/en-us/uefn/create-your-own-device-in-verse for how to create a verse device.

# A Verse-authored creative device that can be placed in a level
V_TokenGranter := class(creative_device):
    @editable
    bPlayerManager : ?V_PlayerManagerDevice = false
    @editable
    AlterGranter : item_granter_device = item_granter_device{}
    @editable
    DungeonGranter : item_granter_device = item_granter_device{}
    @editable
    DungeonTokenCounter : conditional_button_device = conditional_button_device{}
    @editable
    AlterTokenTimer : float = 15.0
    @editable
    DungeonTokenTimer : float = 15.0
    @editable
    SwordGranter : item_granter_device = item_granter_device{}
    @editable
    bUIManager : ?V_UIManagerDevice = false
    @editable
    AlterUpgradeTrigger : trigger_device = trigger_device{}

    # Runs when the device is started in a running game
    OnBegin<override>()<suspends>:void=
        # TODO: Replace this with your code
        AlterUpgradeTrigger.TriggeredEvent.Subscribe(OnAlterUpgrade)
        SwordGranter.ItemGrantedEvent.Subscribe(OnSwordGranted)
        loop:
            Sleep(1.0)
            Players := GetPlayspace().GetPlayers()
            for(Player : Players):
                if(Agent := agent[Player]): 
                    TickPlayerTimer(Agent)
    
    GetDungeonTokenCount(Agent : agent):int=
        return DungeonTokenCounter.GetItemCount(Agent, 1)

    GetAlterTokenCount(Agent : agent):int=
        return DungeonTokenCounter.GetItemCount(Agent, 0)

    # Only called once, when player buys alter token upgrade first time
    OnAlterUpgrade(bAgent : ?agent):void=
        if (Agent := bAgent?, Player := player[Agent], PlayerManager := bPlayerManager?, PlayerStats := PlayerManager.PlayerStatManager.GetPlayerStats[Agent]):
            AlterTokenCount := GetAlterTokenCount(Player)
            if(PlayerStats.AlterMaxAmount < 1): 
                PlayerManager.PlayerStatManager.UpdateAlterMaxAmount(Player)
                if (AlterTokenCount < 1): # Update 1.1 Fix, this is added for player reset debug, granting extra token
                    AlterGranter.GrantItem(Agent)
            if(UIManager := bUIManager?):
                UIManager.ShowAlterTokenUI(Player)

    OnSwordGranted(Agent : agent):void=
        if (Player := player[Agent], PlayerManager := bPlayerManager?, PlayerStats := PlayerManager.PlayerStatManager.GetPlayerStats[Agent]):
            DungeonTokenCount := GetDungeonTokenCount(Player)
            if(PlayerStats.DungeonMaxAmount < 1):
                PlayerManager.PlayerStatManager.UpdateDungeonMaxAmount(Player)
                if (DungeonTokenCount < 1): # Update 1.1 Fix, this is added for player reset debug, granting extra token
                    DungeonGranter.GrantItem(Agent)
                #PlayerManager.PlayerStatManager.UpdateDungeonTimer(Agent, DungeonTokenTimer)
            if(UIManager := bUIManager?):
                UIManager.ShowDungeonTokenUI(Player)
            

    TickPlayerTimer(Agent : agent):void=
        if (PlayerManager := bPlayerManager?, PlayerStats := PlayerManager.PlayerStatManager.GetPlayerStats[Agent]):
            # Alter tick logic
            AlterTokens := GetAlterTokenCount(Agent)
            if (AlterTokens < PlayerStats.AlterMaxAmount):
                TimerReached0 := CalcAlterTimer(Agent)
                if(TimerReached0 = true and PlayerStats.AlterMaxAmount > 0):
                    AlterGranter.GrantItem(Agent)
            else:
                var OutTimer : float = AlterTokenTimer
                if(Player := player[Agent]):
                    if(FoundCard := SearchForCard(4, Player,  PlayerManager.PlayerStatManager)?):
                        CardTimer := ((GetCardBonus(4, FoundCard.Level) / 100.0) * AlterTokenTimer)
                        set OutTimer = AlterTokenTimer - CardTimer
                PlayerManager.PlayerStatManager.UpdateAlterTimer(Agent, OutTimer)

            # Dungeon tick logic
            DungeonTokens := DungeonTokenCounter.GetItemCount(Agent, 1)
            if(DungeonTokens < PlayerStats.DungeonKingTokenSlotAmount + 1):
                TimerReached0 := CalcDungeonTimer(Agent)
                if(TimerReached0 = true and PlayerStats.DungeonMaxAmount > 0):
                    DungeonGranter.GrantItem(Agent)
            else:
                # Calculate token generation timer
                GenerationLevel := PlayerStats.DungeonKingTokenSpeedAmount
                Upgrade := PlayerManager.PlayerStatManager.GetPlayerUpgrade(Agent, PlayerUpgradeType.DungeonKingTokenGenerationSpeed)
                TokenTimer := ((Upgrade.GetCurrentLevelAmount(GenerationLevel) / 100.0) * DungeonTokenTimer) - DungeonTokenTimer
                UpgradedTimer := DungeonTokenTimer - TokenTimer
                var OutTimer : float = UpgradedTimer
                if(Player := player[Agent]):
                    if(FoundCard := SearchForCard(3, Player,  PlayerManager.PlayerStatManager)?):
                        CardTimer := ((GetCardBonus(3, FoundCard.Level) / 100.0) * UpgradedTimer)
                        set OutTimer = UpgradedTimer - CardTimer
                #Print(" GenerationLevel : {Upgrade.GetCurrentLevelAmount(GenerationLevel) / 100.0} | TokenTimer : {TokenTimer} | DungeonTokenTimer : {OutTimer}")

                PlayerManager.PlayerStatManager.UpdateDungeonTimer(Agent, OutTimer)


    CalcAlterTimer(Agent : agent):logic=
        var TimerReached0 : logic = false
        if (PlayerManager := bPlayerManager?):
            if(PlayerStats := PlayerManager.PlayerStatManager.GetPlayerStats[Agent]):
                NewAlterTimer := PlayerStats.AlterTimer - 0.01
                if(IMinutes := Floor[NewAlterTimer]):
                    Seconds := NewAlterTimer - (IMinutes * 1.0)
                    if(var ISeconds : int = Floor[Seconds * 100.0]):
                        if(ISeconds > 59):
                            set ISeconds = 59
                        FSeconds := (ISeconds * 1.0) / 100.0

                        var OutTimer : float = AlterTokenTimer
                        if(Player := player[Agent]):
                            if(FoundCard := SearchForCard(4, Player,  PlayerManager.PlayerStatManager)?):
                                CardTimer := ((GetCardBonus(4, FoundCard.Level) / 100.0) * AlterTokenTimer)
                                set OutTimer = AlterTokenTimer - CardTimer

                        var AlterTimer : float = OutTimer
                        if(IMinutes > 0 or ISeconds > 0):
                            set AlterTimer = (IMinutes * 1.0) + FSeconds
                        else:
                            set TimerReached0 = true
                        PlayerManager.PlayerStatManager.UpdateAlterTimer(Agent, AlterTimer)
        return TimerReached0

    CalcDungeonTimer(Agent : agent):logic=
        var TimerReached0 : logic = false
        if (PlayerManager := bPlayerManager?):
            if(PlayerStats := PlayerManager.PlayerStatManager.GetPlayerStats[Agent]):
                NewDungeonTimer := PlayerStats.DungeonTimer - 0.01
                if(IMinutes := Floor[NewDungeonTimer]):
                    Seconds := NewDungeonTimer - (IMinutes * 1.0)
                    if(var ISeconds : int = Floor[Seconds * 100.0]):
                        if(ISeconds > 59):
                            set ISeconds = 59
                        FSeconds := (ISeconds * 1.0) / 100.0

                        # Calculate token generation timer
                        GenerationLevel := PlayerStats.DungeonKingTokenSpeedAmount
                        Upgrade := PlayerManager.PlayerStatManager.GetPlayerUpgrade(Agent, PlayerUpgradeType.DungeonKingTokenGenerationSpeed)
                        TokenTimer := ((Upgrade.GetCurrentLevelAmount(GenerationLevel) / 100.0) * DungeonTokenTimer) - DungeonTokenTimer
                        UpgradedTimer := DungeonTokenTimer - TokenTimer
                        var OutTimer : float = UpgradedTimer
                        if(Player := player[Agent]):
                            if(FoundCard := SearchForCard(3, Player,  PlayerManager.PlayerStatManager)?):
                                CardTimer := ((GetCardBonus(3, FoundCard.Level) / 100.0) * UpgradedTimer)
                                set OutTimer = UpgradedTimer - CardTimer
                                #Print("NEWTIMER : {OutTimer}, CardTimer : {CardTimer}")
                        #Print(" GenerationLevel : {Upgrade.GetCurrentLevelAmount(GenerationLevel) / 100.0} | TokenTimer : {TokenTimer} | DungeonTokenTimer : {OutTimer}")

                        var DungeonTimer : float = OutTimer # max dungeon time for token granting
                        if(IMinutes > 0 or ISeconds > 0):
                            set DungeonTimer = (IMinutes * 1.0) + FSeconds
                        else:
                            set TimerReached0 = true
                        PlayerManager.PlayerStatManager.UpdateDungeonTimer(Agent, DungeonTimer)
        return TimerReached0
