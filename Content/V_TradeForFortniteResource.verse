
using { /Fortnite.com/Devices }
using { /Verse.org/Simulation }
using { /UnrealEngine.com/Temporary/Diagnostics }

# See https://dev.epicgames.com/documentation/en-us/uefn/create-your-own-device-in-verse for how to create a verse device.

# A Verse-authored creative device that can be placed in a level
V_TradeForFortniteResource := class(creative_device):

    @editable
    Cost : float = 500.0
    @editable
    TradeAmount : int = 1
    @editable
    ResourceType : PropResourceType = PropResourceType.Coin
    @editable
    ItemGranter : item_granter_device = item_granter_device{}
    @editable
    Button : button_device = button_device{}

    @editable
    PlayerManager : ?V_PlayerManagerDevice = false

    # Runs when the device is started in a running game
    OnBegin<override>()<suspends>:void=
        Button.InteractedWithEvent.Subscribe(ButtonActivated)

    ButtonActivated(InAgent : agent):void=
        Print("Attempt Trade")
        if(bPlayerManager := PlayerManager?):
            if(Player := player[InAgent], CurrentPlayerStats := bPlayerManager.PlayerStatManager.GetPlayerStats[Player]):
                #Spend Resources
                if(ResourceType = PropResourceType.Coin):
                    StoneAmount := bPlayerManager.PlayerStatManager.GetPlayerStoneOreAmount(InAgent)
                    BNCost := BigNumber_ConvertFloatIntoBigNumber(Cost)
                    CanAfford := BigNumber_GreaterThanOrEquals(BNCost, BNCost)
                    if(CanAfford = true):
                        bPlayerManager.PlayerStatManager.SpendPlayerOre(InAgent, BNCost, false, PropResourceType.Stone)
                        ItemGranter.GrantItem(Player)
                else:
                    ResourceAmount := bPlayerManager.PlayerStatManager.GetPlayerOreAmount(Player, ResourceType)
                    BNCost := BigNumber_ConvertFloatIntoBigNumber(Cost)
                    CanAfford := BigNumber_GreaterThanOrEquals(ResourceAmount, BNCost)
                    if(CanAfford = true):
                        bPlayerManager.PlayerStatManager.SpendPlayerOre(Player, BNCost, false, ResourceType)
                        ItemGranter.GrantItem(Player)