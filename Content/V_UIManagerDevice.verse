
# See https://dev.epicgames.com/documentation/en-us/uefn/create-your-own-device-in-verse for how to create a verse device.
using { /Fortnite.com/Devices }
using { /Verse.org/Simulation }
using { /UnrealEngine.com/Temporary/Diagnostics }
using { /UnrealEngine.com/Temporary/UI }
using { /Fortnite.com/UI }
using { /UnrealEngine.com/Temporary/SpatialMath}
using { /Verse.org/Colors }
using { /Verse.org/Assets }

FormatFloatTo2DP(InFloat : float):[]char=
    var ReturnArray : []char = array{}
    var StartArray : []char = array{}
    var EndArray : []char = array{}
    FloatString : []char = "{InFloat}"

    var SwitchToEnd : logic = false

    for(Index -> Character : FloatString):
        if(EndArray.Length = 2):
            set ReturnArray += StartArray
            set ReturnArray += "."
            set ReturnArray += EndArray
            return ReturnArray

        if(SwitchToEnd = true):
            set EndArray += array{Character}
        if(array{Character} <> "." and SwitchToEnd = false):
            set StartArray += array{Character}
        
        if (array{Character} = "."):
            set SwitchToEnd = true
            #Print("CHARACTER FOUND")
            
    
    return "EMPTY"

FormatFloatToNDP(InFloat : float, NDP : int):[]char=
    var ReturnArray : []char = array{}
    var StartArray : []char = array{}
    var EndArray : []char = array{}
    FloatString : []char = "{InFloat}"

    var SwitchToEnd : logic = false

    for(Index -> Character : FloatString):
        if(EndArray.Length = NDP):
            set ReturnArray += StartArray
            set ReturnArray += "."
            set ReturnArray += EndArray
            return ReturnArray

        if(SwitchToEnd = true):
            set EndArray += array{Character}
        if(array{Character} <> "." and SwitchToEnd = false):
            set StartArray += array{Character}
        
        if (array{Character} = "."):
            set SwitchToEnd = true
            #Print("CHARACTER FOUND")
            
    
    return "EMPTY"

# A Verse-authored creative device that can be placed in a level
V_UIManagerDevice := class(creative_device):

    @editable
    var AlgorithmAndFunctionsClassReference: V_AlgorithmsAndFunctions = V_AlgorithmsAndFunctions{}

    @editable
    GrantDebugResources : logic = false

    @editable
    HitOreHUDMessageDevice : hud_message_device = hud_message_device{}
    StringToMessage<localizes>(value:string)<computes> : message = "{value}"

    @editable
    OreHitTrigger : trigger_device = trigger_device{}

    @editable
    TeleporterManager : ?V_TeleporterManager = false
    @editable
    ArtifactManager : V_ArtifactManager = V_ArtifactManager{}
    @editable
    Prestige : V_PrestigeFunctionality = V_PrestigeFunctionality{}
    @editable
    CardManager : V_CardManager = V_CardManager{}
    @editable
    UpgradeRemote : signal_remote_manager_device = signal_remote_manager_device{}
    @editable
    AchievementManager : V_AchievementsManager = V_AchievementsManager{}
    @editable
    AltarDevice : V_OfferingAltar = V_OfferingAltar{}

    @editable
    TokenCounter : conditional_button_device = conditional_button_device{}

    @editable
    HUBTeleporter : teleporter_device = teleporter_device{}

    # UI Classes
    UIResourceDisplayClass : UI_ResourceDisplay = UI_ResourceDisplay{}
    UIItemStreamDisplayClass : UI_ItemStreamDisplay = UI_ItemStreamDisplay{}
    UICriticalChargeDisplayClass : UI_CriticalChargeDisplay = UI_CriticalChargeDisplay{}
    UIExplosiveHitDisplayClass : UI_ExplosiveHitDisplay = UI_ExplosiveHitDisplay{}
    UINodeHealthBar : UI_NodeHealthBar = UI_NodeHealthBar{}
    UIQuestDisplay : UI_QuestDisplay = UI_QuestDisplay{}
    UIAltarBuffDisplay : UI_AltarBuffDisplay = UI_AltarBuffDisplay{}
    UIGnomeBuffDisplay : UI_GnomeBuffDisplay = UI_GnomeBuffDisplay{}
    UIPlayerMenuClass : UI_PlayerMenu = UI_PlayerMenu{}
    # Manages and updates stats for each player.
    var PlayerStatsManager: player_stats_manager = player_stats_manager{}

    # Runs when the device is started in a running game
    OnBegin<override>()<suspends>:void=
        OreHitTrigger.TriggeredEvent.Subscribe(ShowOreOnHit)
        UpgradeRemote.SecondarySignalEvent.Subscribe(DisplayPlayerMenuUI)

    InitiatePlayerUI(InAgent : agent):void=
        Print("Initiate UI")
        if(InPlayer := player[InAgent]):
            if(OldData := PlayerStatsDataMap[InAgent]):
                if(CurrentPlayerStats := PlayerStatsManager.GetPlayerStats[InAgent]):  
                    AlterAmount := TokenCounter.GetItemCount(InPlayer, 0)
                    DungeonAmount := TokenCounter.GetItemCount(InPlayer, 1)
                    UIResourceData := MakeResourceDisplayData(OldData.PlayerStoneOreAmount, AlterAmount, DungeonAmount)
                    UIResourceDisplayClass.InitUI(InPlayer, UIResourceData)
                    UIResourceDisplayClass.AddUI(InPlayer)
                    UIItemStreamDisplayClass.InitUI(InPlayer)
                    UINodeHealthBar.InitUI(InPlayer)
                    UICriticalChargeDisplayClass.InitUI(InPlayer)
                    UIExplosiveHitDisplayClass.InitUI(InPlayer)
                    UIAltarBuffDisplay.InitUI(InPlayer)
                    UIGnomeBuffDisplay.InitUI(InPlayer)
                    NewPlayerMenuData := MakePlayerMenuData(PlayerStatsManager)
                    UIPlayerMenuClass.InitUI(InPlayer, NewPlayerMenuData, ArtifactManager, Prestige, CardManager, AchievementManager.AllAchievements, HUBTeleporter, AltarDevice)
                    CardManager.InitiatePlayerUI(InPlayer)
                    spawn{HandleDungeonTokenUIVisibility(InPlayer)}
                    spawn{HandleAlterTokenUIVisibility(InPlayer)}
                    if(bTeleporterManager := TeleporterManager?):
                        bTeleporterManager.Init(InPlayer)
                    spawn:
                        UpdatePlayerUI(InPlayer)
                            
        #PURELY FOR TESTING
        if(GrantDebugResources = true):
            spawn:
                DEBUGGRANTRESOURCES(InAgent)

    RemoveAllPlayerUI(InAgent : agent):void=
        if(InPlayer := player[InAgent]):
            spawn:
                UIResourceDisplayClass.HideUI(InPlayer)
            spawn:
                UIItemStreamDisplayClass.HideUI(InPlayer)
            spawn:
                UICriticalChargeDisplayClass.HideUI(InPlayer)
            spawn:
                UIExplosiveHitDisplayClass.HideUI(InPlayer)
            spawn:
                UIAltarBuffDisplay.HideUI(InPlayer)
            spawn:
                UIGnomeBuffDisplay.HideUI(InPlayer)

    DEBUGGRANTRESOURCES(InAgent : agent)<suspends>:void=
        Sleep(2.0)
        Print("GIVING DEBUG RESOURCES!!")
        if(InPlayer := player[InAgent]):
            if(OldData := PlayerStatsDataMap[InAgent]):
                if(CurrentPlayerStats := PlayerStatsManager.GetPlayerStats[InAgent]):  
                    DebugNumber := BigNumber_ConvertFloatIntoBigNumber(10000000000000000000.0)
                    PlayerStatsManager.GivePlayerStone(InAgent, DebugNumber)
                    PlayerStatsManager.UpdatePlayerOreAmount(InAgent, PropResourceType.Copper, DebugNumber)

    HandleDungeonTokenUIVisibility(Player : player)<suspends>:void=
        Sleep(1.0)
        if(CurrentPlayerStats := PlayerStatsManager.GetPlayerStats[Player]): 
            if (CurrentPlayerStats.DungeonMaxAmount > 0):
                UIResourceDisplayClass.ShowDungeonUI(Player)
            else:
                UIResourceDisplayClass.HideDungeonUI(Player)

    HandleAlterTokenUIVisibility(Player : player)<suspends>:void=
        if(CurrentPlayerStats := PlayerStatsManager.GetPlayerStats[Player]): 
            UIResourceDisplayClass.ShowAlterUI(Player)
            if (CurrentPlayerStats.AltarUpgrade > 0):
                UIResourceDisplayClass.ShowAlterUI(Player)
            else:
                UIResourceDisplayClass.HideAlterUI(Player)

    ShowDungeonTokenUI(Player:player):void=
        UIResourceDisplayClass.ShowDungeonUI(Player)

    ShowAlterTokenUI(Player:player):void=
        UIResourceDisplayClass.ShowAlterUI(Player)

    GetAlterAmount_String(Player : player):string=
        AlterAmount := TokenCounter.GetItemCount(Player, 0)
        return "{AlterAmount}"
    GetDungeonAmount_String(Player : player):string=
        DungeonAmount := TokenCounter.GetItemCount(Player, 1)
        return "{DungeonAmount}"

    GetAlterAmount(Player : player):int=
        AlterAmount := TokenCounter.GetItemCount(Player, 0)
        return AlterAmount

    GetDungeonAmount(Player : player):int=
        DungeonAmount := TokenCounter.GetItemCount(Player, 1)
        return DungeonAmount

    CalcAlterTimer(Player : player):?float=
        AlterAmount := GetAlterAmount(Player)
        if(CurrentPlayerStats := PlayerStatsManager.GetPlayerStats[Player]):  
            if (AlterAmount >= CurrentPlayerStats.AlterMaxAmount):
                return false
            else:
                return option{CurrentPlayerStats.AlterTimer}
        return false
    
    CalcDungeonTimer(Player : player):?float=
        DungeonAmount := GetDungeonAmount(Player)
        if(CurrentPlayerStats := PlayerStatsManager.GetPlayerStats[Player]):  
            if (DungeonAmount >= CurrentPlayerStats.DungeonKingTokenSlotAmount + 1):
                return false
            else:
                return option{CurrentPlayerStats.DungeonTimer}
        return false

    UpdatePlayerUI(InPlayer : player)<suspends>:void=
        Print("Update UI")
        loop:
            if(not InPlayer.IsActive[]):
                break
            if(OldData := PlayerStatsDataMap[InPlayer]):
                if(CurrentPlayerStats := PlayerStatsManager.GetPlayerStats[InPlayer]):  
                    #Convert the displayed information into a big number. We can multiply here to test massive numbers
                    #Format the big number into something we can display to the player
                    BNPlayerStoneString := BigNumber_FormatDisplay(OldData.PlayerStoneOreAmount)
                    BNPlayerPrestigeString := BigNumber_FormatDisplay(OldData.PlayerPrestigeResourceAmount)
                    BNPlayerCardsString := BigNumber_FormatDisplay(OldData.DustAmount)
                    AlterAmount := GetAlterAmount_String(InPlayer)
                    AlterTimer := CalcAlterTimer(InPlayer)
                    DungeonAmount := GetDungeonAmount_String(InPlayer)
                    DungeonTimer := CalcDungeonTimer(InPlayer)
                    UIResourceDisplayClass.UpdateUI(InPlayer, BNPlayerStoneString, BNPlayerCardsString, AlterAmount, DungeonAmount, AlterTimer, DungeonTimer, BNPlayerPrestigeString)    
            Sleep(0.1)

    DisplayPlayerMenuUI(InAgent : agent):void=
        if(InPlayer := player[InAgent]):
            UIPlayerMenuClass.UpdateUI(InPlayer, PlayerStatsManager, AchievementManager.AllAchievements)
            UIPlayerMenuClass.AddUI(InPlayer)

    #Divide by 2 if it's too easy/quick to get gold
    ShowOreOnHit(InAgent : ?agent):void=
        Print("Show Ore On Hit")
        if(Agent := InAgent?):
            spawn:
                ShowOreDelayed(Agent)

    ShowOreDelayed(InAgent : agent)<suspends>:void=
        if(Player := player[InAgent]):
            if(OldData := PlayerStatsDataMap[Player]):
                if(CurrentPlayerStats := PlayerStatsManager.GetPlayerStats[Player]):   
                    Sleep(0.1)
                    BNPlayerOreOnHitString := BigNumber_FormatDisplay(OldData.HarvestRatePerHit)
                    HitOreHUDMessageDevice.SetText(StringToMessage("+ {BNPlayerOreOnHitString} Ore"))