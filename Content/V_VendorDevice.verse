
using { /Fortnite.com/Devices }
using { /Verse.org/Simulation }
using { /UnrealEngine.com/Temporary/Diagnostics }

#Added
using { /UnrealEngine.com/Temporary/SpatialMath}
using { /Verse.org/Assets }
# See https://dev.epicgames.com/documentation/en-us/uefn/create-your-own-device-in-verse for how to create a verse device.

# A Verse-authored creative device that can be placed in a level
V_VendorDevice := class(creative_device):
    #Trigger references
    @editable
    NPCReference : character_device = character_device{}
    @editable
    TriggerSellOreForCoin : trigger_device = trigger_device{}
    @editable 
    TriggerUpgradeVendor : trigger_device = trigger_device{}
    @editable
    TriggerUpgradeMultiplier : trigger_device = trigger_device{}

    @editable
    OpenTrigger : trigger_device = trigger_device{}

    VendorTemplateClass : VendorTemplate_UI = VendorTemplate_UI{}

    #This is a set variable through design so we can change this here and it will change the formula. This is not saved.
    BaseCost : float = 10.0

    #Tier systems, since these are hard set through design we can store them in an array and check against them
    TierSystems : []float = array{10.0,25.0,50.0,100.0,150.0,200.0,250.0,300.0,350.0,400.0,450.0,500.0,550.0,600.0,650.0,700.0,750.0,800.0}

    SetCoinGain<localizes>(FloatValue:float) : message = "Cost = {FloatValue}"
    
    #These are static variables. We will never change these at runtime.
    @editable
    VendorBaseRate: float = 1.0
    @editable
    VendorRateIncrease : float = 1.0
    @editable
    VendorCostIncreaseRate: float = 1.07
    @editable
    ResourceToTrade : VendorUpgradeResourceTypes = VendorUpgradeResourceTypes.Stone
    @editable
    var AlgorithmAndFunctionsClassReference: V_AlgorithmsAndFunctions = V_AlgorithmsAndFunctions{}

    @editable
    X1Trigger : trigger_device = trigger_device{}
    @editable
    X10Trigger : trigger_device = trigger_device{}
    @editable
    X100Trigger : trigger_device = trigger_device{}

    # Manages and updates stats for each player.
    var PlayerStatData: player_stats_manager = player_stats_manager{}

    # Runs when the device is started in a running game
    #OnBegin<override>()<suspends>:void=
        #NPCReference.InteractedWithEvent.Subscribe(InteractWithNPCAndDisplayVendorUI)
        #TriggerSellOreForCoin.TriggeredEvent.Subscribe(TradeOreForCoin)
        #TriggerUpgradeMultiplier.TriggeredEvent.Subscribe(UpgradeTierMultiplier)
        #TriggerUpgradeVendor.TriggeredEvent.Subscribe(PurchaseVendorUpgrade)

        #Purchase Amounts
        #X1Trigger.TriggeredEvent.Subscribe(OnX1Triggered)
        #X10Trigger.TriggeredEvent.Subscribe(OnX10Triggered)
        #X100Trigger.TriggeredEvent.Subscribe(OnX100Triggered)

        #OpenTrigger.TriggeredEvent.Subscribe(TriggerDisplayVendorUI)

    InitiateVendorUI(InPlayer : player):void=
        if(ResourceToTrade = VendorUpgradeResourceTypes.Stone):
            #SetupStoneVendorUIForAllPlayers(InPlayer) 
        #if(ResourceToTrade = VendorUpgradeResourceTypes.Silver):    
            #SetupSilverVendorUIForAllPlayers(InPlayer) 
        #if(ResourceToTrade = VendorUpgradeResourceTypes.Gold):
            #SetupGoldVendorUIForAllPlayers(InPlayer)
        #if(ResourceToTrade = VendorUpgradeResourceTypes.Diamond): 
            #SetupDiamondVendorUIForAllPlayers(InPlayer) 

    <#SetupStoneVendorUIForAllPlayers(InPlayer : player):void=
        if(OldData := PlayerStatsDataMap[InPlayer]):
            if(CurrentPlayerStats := PlayerStatData.GetPlayerStats[InPlayer]):
                #Setting up variables to be used for the "MakeVendorData"
                CostOfNextUpgrade := AlgorithmAndFunctionsClassReference.AlgorithmAndFunctionClassReference.CalculateCostOfNextUpgrade(BaseCost,VendorCostIncreaseRate,  PlayerStatData.GetStoneVendorUpgradeLevel(InPlayer))
                StoneVendorTradeInGain := PlayerStatData.GetStoneVendorCoinPerTrade(InPlayer)     
                CalculatedCoinGainOnTrade := AlgorithmAndFunctionsClassReference.AlgorithmAndFunctionClassReference.CalculateHarvestRateOnHit(VendorRateIncrease, OldData.StoneVendorUpgradeLevel , VendorBaseRate, OldData.StoneVendorTierMultiplier)        
                TotalOreTradeAmount := BigNumber_Multiply(CurrentPlayerStats.PlayerStoneOreAmount, CalculatedCoinGainOnTrade)
                CostOfNextTierMultiplier := VendorRateIncrease * OldData.StoneVendorTierMultiplier

                #Convert into big numbers after performing math functions
                BNPlayerStoneAmount := BigNumber_FormatDisplay(PlayerStatData.GetPlayerStoneOreAmount(InPlayer))
                BNPlayerCoinAmount := BigNumber_FormatDisplay(PlayerStatData.GetPlayerCoinAmount(InPlayer))
                BNStoneVendorTradeInGainString := BigNumber_FormatDisplay(StoneVendorTradeInGain)
                BNTotalOreTradeAmountString := BigNumber_FormatDisplay(BigNumber_Multiply(TotalOreTradeAmount, CalculatedCoinGainOnTrade))
                BNCostOfNextUpgradeString := BigNumber_FormatDisplay(CostOfNextUpgrade)
                BNCostOfNextTierMultiplier := BigNumber_FormatDisplay(BigNumber_ConvertFloatIntoBigNumber(CostOfNextTierMultiplier))

                #Initialise the UI, we can then add it to the player
                #StoneVendorData := MakeVendorData(VerseExposed.T_GoldCoin, "Stone", PlayerStatData.GetPlayerStoneOreAmount(InPlayer), StoneVendorTradeInGain, TotalOreTradeAmount, CostOfNextUpgrade, TriggerSellOreForCoin, TriggerUpgradeVendor, TriggerUpgradeMultiplier)
                StoneVendorData := MakeVendorData(VerseExposed.T_Stone, "Stone", VerseExposed.T_StoneVendorTitle, VerseExposed.T_StoneVendorCharacter, BNPlayerStoneAmount, BNPlayerCoinAmount, BNStoneVendorTradeInGainString, BNTotalOreTradeAmountString, BNCostOfNextUpgradeString, BNCostOfNextTierMultiplier, TriggerSellOreForCoin, TriggerUpgradeVendor, TriggerUpgradeMultiplier)
                VendorTemplateClass.InitUI(InPlayer, StoneVendorData, X1Trigger, X10Trigger, X100Trigger)

    SetupSilverVendorUIForAllPlayers(InPlayer : player):void=
        if(OldData := PlayerStatsDataMap[InPlayer]):
            if(CurrentPlayerStats := PlayerStatData.GetPlayerStats[InPlayer]):
                #Setting up variables to be used for the "MakeVendorData"
                CostOfNextUpgrade := AlgorithmAndFunctionsClassReference.AlgorithmAndFunctionClassReference.CalculateCostOfNextUpgrade(BaseCost,VendorCostIncreaseRate,  PlayerStatData.GetSilverVendorUpgradeLevel(InPlayer))
                SilverVendorTradeInGain := PlayerStatData.GetSilverVendorCoinPerTrade(InPlayer)     
                CalculatedCoinGainOnTrade := AlgorithmAndFunctionsClassReference.AlgorithmAndFunctionClassReference.CalculateHarvestRateOnHit(VendorRateIncrease, OldData.SilverVendorUpgradeLevel , VendorBaseRate, OldData.SilverVendorTierMultiplier)        
                TotalOreTradeAmount := BigNumber_Multiply(CurrentPlayerStats.PlayerSilverOreAmount, CalculatedCoinGainOnTrade)
                CostOfNextTierMultiplier := VendorRateIncrease * OldData.SilverVendorTierMultiplier

                #Convert into big numbers after performing math functions
                BNPlayerSilverAmount := BigNumber_FormatDisplay(PlayerStatData.GetPlayerSilverOreAmount(InPlayer))
                BNPlayerCoinAmount := BigNumber_FormatDisplay(PlayerStatData.GetPlayerCoinAmount(InPlayer))
                BNSilverVendorTradeInGainString := BigNumber_FormatDisplay(SilverVendorTradeInGain)
                BNTotalOreTradeAmountString := BigNumber_FormatDisplay(BigNumber_Multiply(TotalOreTradeAmount, CalculatedCoinGainOnTrade))
                BNCostOfNextUpgradeString := BigNumber_FormatDisplay(CostOfNextUpgrade)
                BNCostOfNextTierMultiplier := BigNumber_FormatDisplay(BigNumber_ConvertFloatIntoBigNumber(CostOfNextTierMultiplier))

                #Initialise the UI, we can then add it to the player
                SilverVendorData := MakeVendorData(VerseExposed.T_SilverOre, "Silver", VerseExposed.T_SilverVendorTitle, VerseExposed.T_StoneVendorCharacter, BNPlayerSilverAmount, BNPlayerCoinAmount, BNSilverVendorTradeInGainString, BNTotalOreTradeAmountString, BNCostOfNextUpgradeString, BNCostOfNextTierMultiplier, TriggerSellOreForCoin, TriggerUpgradeVendor, TriggerUpgradeMultiplier)
                VendorTemplateClass.InitUI(InPlayer, SilverVendorData, X1Trigger, X10Trigger, X100Trigger)

    SetupGoldVendorUIForAllPlayers(InPlayer : player):void=
        if(OldData := PlayerStatsDataMap[InPlayer]):
            if(CurrentPlayerStats := PlayerStatData.GetPlayerStats[InPlayer]):
                #Setting up variables to be used for the "MakeVendorData"
                CostOfNextUpgrade := AlgorithmAndFunctionsClassReference.AlgorithmAndFunctionClassReference.CalculateCostOfNextUpgrade(BaseCost,VendorCostIncreaseRate,  PlayerStatData.GetGoldVendorUpgradeLevel(InPlayer))
                GoldVendorTradeInGain := PlayerStatData.GetGoldVendorCoinPerTrade(InPlayer)     
                CalculatedCoinGainOnTrade := AlgorithmAndFunctionsClassReference.AlgorithmAndFunctionClassReference.CalculateHarvestRateOnHit(VendorRateIncrease, OldData.GoldVendorUpgradeLevel , VendorBaseRate, OldData.GoldVendorTierMultiplier)        
                TotalOreTradeAmount := BigNumber_Multiply(CurrentPlayerStats.PlayerGoldOreAmount, CalculatedCoinGainOnTrade)
                CostOfNextTierMultiplier := VendorRateIncrease * OldData.GoldVendorTierMultiplier

                #Convert into big numbers after performing math functions
                BNPlayerGoldAmount := BigNumber_FormatDisplay(PlayerStatData.GetPlayerGoldOreAmount(InPlayer))
                BNPlayerCoinAmount := BigNumber_FormatDisplay(PlayerStatData.GetPlayerCoinAmount(InPlayer))
                BNGoldVendorTradeInGainString := BigNumber_FormatDisplay(GoldVendorTradeInGain)
                BNTotalOreTradeAmountString := BigNumber_FormatDisplay(BigNumber_Multiply(TotalOreTradeAmount, CalculatedCoinGainOnTrade))
                BNCostOfNextUpgradeString := BigNumber_FormatDisplay(CostOfNextUpgrade)
                BNCostOfNextTierMultiplier := BigNumber_FormatDisplay(BigNumber_ConvertFloatIntoBigNumber(CostOfNextTierMultiplier))

                #Initialise the UI, we can then add it to the player
                GoldVendorData := MakeVendorData(VerseExposed.T_GoldOre, "Gold", VerseExposed.T_GoldVendorTitle, VerseExposed.T_StoneVendorCharacter, BNPlayerGoldAmount, BNPlayerCoinAmount, BNGoldVendorTradeInGainString, BNTotalOreTradeAmountString, BNCostOfNextUpgradeString, BNCostOfNextTierMultiplier, TriggerSellOreForCoin, TriggerUpgradeVendor, TriggerUpgradeMultiplier)
                VendorTemplateClass.InitUI(InPlayer, GoldVendorData, X1Trigger, X10Trigger, X100Trigger)

    SetupDiamondVendorUIForAllPlayers(InPlayer : player):void=
        if(OldData := PlayerStatsDataMap[InPlayer]):
            if(CurrentPlayerStats := PlayerStatData.GetPlayerStats[InPlayer]):
                #Setting up variables to be used for the "MakeVendorData"
                CostOfNextUpgrade := AlgorithmAndFunctionsClassReference.AlgorithmAndFunctionClassReference.CalculateCostOfNextUpgrade(BaseCost,VendorCostIncreaseRate,  PlayerStatData.GetDiamondVendorUpgradeLevel(InPlayer))
                DiamondVendorTradeInGain := PlayerStatData.GetDiamondVendorCoinPerTrade(InPlayer)     
                CalculatedCoinGainOnTrade := AlgorithmAndFunctionsClassReference.AlgorithmAndFunctionClassReference.CalculateHarvestRateOnHit(VendorRateIncrease, OldData.DiamondVendorUpgradeLevel , VendorBaseRate, OldData.DiamondVendorTierMultiplier)        
                TotalOreTradeAmount := BigNumber_Multiply(CurrentPlayerStats.PlayerDiamondOreAmount, CalculatedCoinGainOnTrade)
                CostOfNextTierMultiplier := VendorRateIncrease * OldData.DiamondVendorTierMultiplier

                #Convert into big numbers after performing math functions
                BNPlayerDiamondAmount := BigNumber_FormatDisplay(PlayerStatData.GetPlayerDiamondOreAmount(InPlayer))
                BNPlayerCoinAmount := BigNumber_FormatDisplay(PlayerStatData.GetPlayerCoinAmount(InPlayer))
                BNDiamondVendorTradeInGainString := BigNumber_FormatDisplay(DiamondVendorTradeInGain)
                BNTotalOreTradeAmountString := BigNumber_FormatDisplay(BigNumber_Multiply(TotalOreTradeAmount, CalculatedCoinGainOnTrade))
                BNCostOfNextUpgradeString := BigNumber_FormatDisplay(CostOfNextUpgrade)
                BNCostOfNextTierMultiplier := BigNumber_FormatDisplay(BigNumber_ConvertFloatIntoBigNumber(CostOfNextTierMultiplier))

                #Initialise the UI, we can then add it to the player
                DiamondVendorData := MakeVendorData(VerseExposed.T_Diamond, "Diamond", VerseExposed.T_DiamondVendorTitle, VerseExposed.T_StoneVendorCharacter, BNPlayerDiamondAmount, BNPlayerCoinAmount, BNDiamondVendorTradeInGainString, BNTotalOreTradeAmountString, BNCostOfNextUpgradeString, BNCostOfNextTierMultiplier, TriggerSellOreForCoin, TriggerUpgradeVendor, TriggerUpgradeMultiplier)
                VendorTemplateClass.InitUI(InPlayer, DiamondVendorData, X1Trigger, X10Trigger, X100Trigger)

    TriggerDisplayVendorUI(InMaybeAgent : ?agent):void=
        # Check if the player has the right upgrade
        if(InAgent := InMaybeAgent?, InPlayer := player[InAgent]):
            GlobalVendorLevel := PlayerStatData.GetPlayerUpgradeLevel(InAgent, PlayerUpgradeType.GlobalVendor)
            if(GlobalVendorLevel > 0):
                if(OldData := PlayerStatsDataMap[InPlayer]):
                    if(CurrentPlayerStats := PlayerStatData.GetPlayerStats[InPlayer]):
                        spawn:
                            UpdateVendorUI(InAgent, PlayerStatData.GetPlayerUpgradePurchaseAmount(InPlayer))
                        VendorTemplateClass.AddUI(InPlayer)

    InteractWithNPCAndDisplayVendorUI(InAgent : agent):void=
        Print("Bring up Vendor UI")
            if(InPlayer := player[InAgent]):
                Print("Cast to Player")
                if(OldData := PlayerStatsDataMap[InPlayer]):
                    if(CurrentPlayerStats := PlayerStatData.GetPlayerStats[InPlayer]):
                        Print("All checks have succeded")   
                        #Update UI with the new values
                        spawn:
                            UpdateVendorUI(InAgent, PlayerStatData.GetPlayerUpgradePurchaseAmount(InPlayer))
                        #Display UI to the player
                        VendorTemplateClass.AddUI(InPlayer)

    PurchaseVendorUpgrade(InAgent : ?agent):void=
        if(Agent := InAgent?):
            if(Player := player[Agent]):
                AlgorithmAndFunctionsClassReference.AlgorithmAndFunctionClassReference.TradeOreWithVendor(Agent, BaseCost, TierSystems, VendorBaseRate, VendorRateIncrease, VendorCostIncreaseRate, ResourceToTrade, false, true, false, PlayerStatData.GetPlayerUpgradePurchaseAmount(Player))
                spawn:
                    UpdateVendorUI(Agent, PlayerStatData.GetPlayerUpgradePurchaseAmount(Player))
    
    TradeOreForCoin(InAgent : ?agent):void=
        if(Agent := InAgent?):
            if(Player := player[Agent]):
                AlgorithmAndFunctionsClassReference.AlgorithmAndFunctionClassReference.TradeOreWithVendor(Agent, BaseCost, TierSystems, VendorBaseRate, VendorRateIncrease, VendorCostIncreaseRate, ResourceToTrade, true, false, false, 1)
                spawn:
                    UpdateVendorUI(Agent, PlayerStatData.GetPlayerUpgradePurchaseAmount(Player))

    #This function gives us the cost of the next upgrade
    UpgradeTierMultiplier(InAgent : ?agent):void=
        if(Agent := InAgent?):
            if(Player := player[Agent]):
                AlgorithmAndFunctionsClassReference.AlgorithmAndFunctionClassReference.TradeOreWithVendor(Agent, BaseCost, TierSystems, VendorBaseRate, VendorRateIncrease, VendorCostIncreaseRate, ResourceToTrade, false, false, true, 1)
                spawn:
                    UpdateVendorUI(Agent, PlayerStatData.GetPlayerUpgradePurchaseAmount(Player))

    #If this does not work make sure we intiate all UI in the V_PlayerManagerDevice
    UpdateVendorUI(InAgent : agent, InPurchaseAmount : int)<suspends>:void=
        Sleep(0.01)
        if(InPlayer := player[InAgent]):
            if(OldData := PlayerStatsDataMap[InPlayer]):
                if(CurrentPlayerStats := PlayerStatData.GetPlayerStats[InPlayer]):
                    if(ResourceToTrade = VendorUpgradeResourceTypes.Stone):
                        FormattedCostAndCanAfford := GetFormattedCost(InPlayer, InPurchaseAmount, BaseCost, VendorCostIncreaseRate, PlayerStatData.GetStoneVendorUpgradeLevel(InPlayer))
                        #CostOfNextUpgrade := AlgorithmAndFunctionsClassReference.AlgorithmAndFunctionClassReference.CalculateCostOfNextUpgrade(BaseCost,VendorCostIncreaseRate,  PlayerStatData.GetStoneVendorUpgradeLevel(InPlayer))
                        StoneVendorTradeInGain := PlayerStatData.GetStoneVendorCoinPerTrade(InPlayer)     
                        CalculatedCoinGainOnTrade := AlgorithmAndFunctionsClassReference.AlgorithmAndFunctionClassReference.CalculateHarvestRateOnHit(VendorRateIncrease, OldData.StoneVendorUpgradeLevel , VendorBaseRate, OldData.StoneVendorTierMultiplier)
                        TotalTradeInAmount := BigNumber_Multiply(OldData.StoneOreRatePerTrade, OldData.PlayerStoneOreAmount)
                        #TotalOreTradeAmount := CurrentPlayerStats.PlayerStoneOreAmount * CalculatedCoinGainOnTrade 
                        #Convert into big numbers after performing math functions
                        BNPlayerStoneAmount := BigNumber_FormatDisplay(PlayerStatData.GetPlayerStoneOreAmount(InPlayer))
                        BNPlayerCoinAmount := BigNumber_FormatDisplay(PlayerStatData.GetPlayerCoinAmount(InPlayer))
                        BNStoneVendorTradeInGainString := BigNumber_FormatDisplay(StoneVendorTradeInGain)
                        BNTotalOreTradeAmountString := BigNumber_FormatDisplay(TotalTradeInAmount)
                        BNPurchaseAmount := BigNumber_FormatDisplay(BigNumber_ConvertFloatIntoBigNumber(InPurchaseAmount * 1.0))
                        #Update Vendor UI
                        VendorTemplateClass.UpdateUI(InPlayer, "Stone", BNPlayerStoneAmount, BNPlayerCoinAmount, BNStoneVendorTradeInGainString, BNTotalOreTradeAmountString, FormattedCostAndCanAfford(0), "Money",CurrentPlayerStats, TierSystems, BNPurchaseAmount, FormattedCostAndCanAfford(1))
                        Print("Update Stone UI")
                    
                    if(ResourceToTrade = VendorUpgradeResourceTypes.Silver):
                        Print("Attempt to display UI")
                        FormattedCostAndCanAfford := GetFormattedCost(InPlayer, InPurchaseAmount, BaseCost, VendorCostIncreaseRate, PlayerStatData.GetStoneVendorUpgradeLevel(InPlayer))
                        #CostOfNextUpgrade := AlgorithmAndFunctionsClassReference.AlgorithmAndFunctionClassReference.CalculateCostOfNextUpgrade(BaseCost,VendorCostIncreaseRate,  PlayerStatData.GetSilverVendorUpgradeLevel(InPlayer))
                        SilverVendorTradeInGain := PlayerStatData.GetSilverVendorCoinPerTrade(InPlayer)     
                        CalculatedCoinGainOnTrade := AlgorithmAndFunctionsClassReference.AlgorithmAndFunctionClassReference.CalculateHarvestRateOnHit(VendorRateIncrease, OldData.SilverVendorUpgradeLevel , VendorBaseRate, OldData.SilverVendorTierMultiplier)
                        
                        TotalTradeInAmount := BigNumber_Multiply(OldData.SilverOreRatePerTrade, OldData.PlayerSilverOreAmount )

                        #Convert into big numbers after performing math functions
                        BNPlayerSilverAmount := BigNumber_FormatDisplay(PlayerStatData.GetPlayerSilverOreAmount(InPlayer))
                        BNPlayerCoinAmount := BigNumber_FormatDisplay(PlayerStatData.GetPlayerCoinAmount(InPlayer))
                        BNSilverVendorTradeInGainString := BigNumber_FormatDisplay(SilverVendorTradeInGain)
                        BNTotalOreTradeAmountString := BigNumber_FormatDisplay(TotalTradeInAmount)
                        BNPurchaseAmount := BigNumber_FormatDisplay(BigNumber_ConvertFloatIntoBigNumber(InPurchaseAmount * 1.0))

                        VendorTemplateClass.UpdateUI(InPlayer, "Silver", BNPlayerSilverAmount, BNPlayerCoinAmount, BNSilverVendorTradeInGainString, BNTotalOreTradeAmountString, FormattedCostAndCanAfford(0), "Money", CurrentPlayerStats, TierSystems, BNPurchaseAmount, FormattedCostAndCanAfford(1))

                    if(ResourceToTrade = VendorUpgradeResourceTypes.Gold):
                        FormattedCostAndCanAfford := GetFormattedCost(InPlayer, InPurchaseAmount, BaseCost, VendorCostIncreaseRate, PlayerStatData.GetStoneVendorUpgradeLevel(InPlayer))
                        #CostOfNextUpgrade := AlgorithmAndFunctionsClassReference.AlgorithmAndFunctionClassReference.CalculateCostOfNextUpgrade(BaseCost,VendorCostIncreaseRate,  PlayerStatData.GetGoldVendorUpgradeLevel(InPlayer))
                        GoldVendorTradeInGain := PlayerStatData.GetGoldVendorCoinPerTrade(InPlayer)     
                        CalculatedCoinGainOnTrade := AlgorithmAndFunctionsClassReference.AlgorithmAndFunctionClassReference.CalculateHarvestRateOnHit(VendorRateIncrease, OldData.GoldVendorUpgradeLevel , VendorBaseRate, OldData.GoldVendorTierMultiplier)
                        TotalTradeInAmount := BigNumber_Multiply(OldData.GoldOreRatePerTrade, OldData.PlayerGoldOreAmount )

                        #Convert into big numbers after performing math functions
                        BNPlayerGoldAmount := BigNumber_FormatDisplay(PlayerStatData.GetPlayerGoldOreAmount(InPlayer))
                        BNPlayerCoinAmount := BigNumber_FormatDisplay(PlayerStatData.GetPlayerCoinAmount(InPlayer))
                        BNGoldVendorTradeInGainString := BigNumber_FormatDisplay(GoldVendorTradeInGain)
                        BNTotalOreTradeAmountString := BigNumber_FormatDisplay(TotalTradeInAmount)
                        BNPurchaseAmount := BigNumber_FormatDisplay(BigNumber_ConvertFloatIntoBigNumber(InPurchaseAmount * 1.0))

                        VendorTemplateClass.UpdateUI(InPlayer, "Gold", BNPlayerGoldAmount, BNPlayerCoinAmount, BNGoldVendorTradeInGainString, BNTotalOreTradeAmountString, FormattedCostAndCanAfford(0), "Money", CurrentPlayerStats, TierSystems, BNPurchaseAmount, FormattedCostAndCanAfford(1))
                    
                    if(ResourceToTrade = VendorUpgradeResourceTypes.Diamond):
                        FormattedCostAndCanAfford := GetFormattedCost(InPlayer, InPurchaseAmount, BaseCost, VendorCostIncreaseRate, PlayerStatData.GetStoneVendorUpgradeLevel(InPlayer))
                        DiamondVendorTradeInGain := PlayerStatData.GetDiamondVendorCoinPerTrade(InPlayer)     
                        CalculatedCoinGainOnTrade := AlgorithmAndFunctionsClassReference.AlgorithmAndFunctionClassReference.CalculateHarvestRateOnHit(VendorRateIncrease, OldData.DiamondVendorUpgradeLevel , VendorBaseRate, OldData.DiamondVendorTierMultiplier)
                        TotalTradeInAmount := BigNumber_Multiply(OldData.DiamondOreRatePerTrade, OldData.PlayerDiamondOreAmount )

                        #Convert into big numbers after performing math functions
                        BNPlayerDiamondAmount := BigNumber_FormatDisplay(PlayerStatData.GetPlayerDiamondOreAmount(InPlayer))
                        BNPlayerCoinAmount := BigNumber_FormatDisplay(PlayerStatData.GetPlayerCoinAmount(InPlayer))
                        BNDiamondVendorTradeInGainString := BigNumber_FormatDisplay(DiamondVendorTradeInGain)
                        BNTotalOreTradeAmountString := BigNumber_FormatDisplay(TotalTradeInAmount)
                        BNPurchaseAmount := BigNumber_FormatDisplay(BigNumber_ConvertFloatIntoBigNumber(InPurchaseAmount * 1.0))
                        VendorTemplateClass.UpdateUI(InPlayer, "Diamond", BNPlayerDiamondAmount, BNPlayerCoinAmount, BNDiamondVendorTradeInGainString, BNTotalOreTradeAmountString, FormattedCostAndCanAfford(0), "Money", CurrentPlayerStats, TierSystems, BNPurchaseAmount, FormattedCostAndCanAfford(1))

    GetFormattedCost(InPlayer : player, InPurchaseAmount : int, InBaseCost : float, InHRCostIncrease : float, InUpgradeLevel : float):tuple(string, logic)=
        var RemoteUpgradeCost : big_number = AlgorithmAndFunctionsClassReference.AlgorithmAndFunctionClassReference.CalculateCostOfNextUpgrade(InBaseCost, InHRCostIncrease, InUpgradeLevel)
        var TruePurchaseAmount : int = 1
        var FormattedCost : string = "{BigNumber_FormatDisplay(RemoteUpgradeCost)} (x{TruePurchaseAmount})"
        var CanAfford : logic = false

        PlayerCoinAmount := PlayerStatData.GetPlayerCoinAmount(InPlayer)
        if(InPurchaseAmount > 1):
            for(X := 1..InPurchaseAmount):
                NextRemoteUpgradeCost := AlgorithmAndFunctionsClassReference.AlgorithmAndFunctionClassReference.CalculateCostOfNextUpgrade(InBaseCost, InHRCostIncrease, InUpgradeLevel + (X * 1.0))
                BNCost := BigNumber_Add(RemoteUpgradeCost, NextRemoteUpgradeCost)
                #Check if we have enough coins for the next upgrade.
                InternalCanAfford := BigNumber_GreaterThanOrEquals(PlayerCoinAmount, BNCost)
                if(InternalCanAfford = true):
                    set RemoteUpgradeCost = BigNumber_Add(RemoteUpgradeCost, NextRemoteUpgradeCost)
                    set TruePurchaseAmount = X + 1
                    if(InPurchaseAmount = 10 and X = 10):
                        set TruePurchaseAmount = 10
                    if(InPurchaseAmount = 100 and X = 100):
                        set TruePurchaseAmount = 100
                    set FormattedCost = "{BigNumber_FormatDisplay(RemoteUpgradeCost)} (x{TruePurchaseAmount})"
                    set CanAfford = true
                
        if (TruePurchaseAmount = 1):
            BNCost := RemoteUpgradeCost
            #Check if we have enough coins for the next upgrade.
            InternalCanAfford := BigNumber_GreaterThanOrEquals(PlayerCoinAmount, BNCost)
            if(InternalCanAfford = true):
                set RemoteUpgradeCost = AlgorithmAndFunctionsClassReference.AlgorithmAndFunctionClassReference.CalculateCostOfNextUpgrade(InBaseCost, InHRCostIncrease, InUpgradeLevel)
                set CanAfford = true
        
        return (FormattedCost, CanAfford)

    OnX1Triggered(InMaybeAgent : ?agent):void=
        if(Agent := InMaybeAgent?):
            if(InPlayer := player[Agent]):
                PurchaseAmount := 1
                PlayerStatData.UpdatePlayerUpgradePurchaseAmount(Agent, PurchaseAmount)
                spawn:
                    UpdateVendorUI(InPlayer, PurchaseAmount)
                Print("X1")

    OnX10Triggered(InMaybeAgent : ?agent):void=
        if(Agent := InMaybeAgent?):
            if(InPlayer := player[Agent]):
                PurchaseAmount := 10
                PlayerStatData.UpdatePlayerUpgradePurchaseAmount(Agent, PurchaseAmount)
                spawn:
                    UpdateVendorUI(InPlayer, PurchaseAmount)
                Print("X10")

    OnX100Triggered(InMaybeAgent : ?agent):void=
        if(Agent := InMaybeAgent?):
            if(InPlayer := player[Agent]):
                PurchaseAmount := 100
                PlayerStatData.UpdatePlayerUpgradePurchaseAmount(Agent, PurchaseAmount)
                spawn:
                    UpdateVendorUI(InPlayer, PurchaseAmount)
                Print("X100")#>