
using { /Fortnite.com/Devices }
using { /Verse.org/Simulation }
using { /UnrealEngine.com/Temporary/Diagnostics }

# See https://dev.epicgames.com/documentation/en-us/uefn/create-your-own-device-in-verse for how to create a verse device.

# A Verse-authored creative device that can be placed in a level
V_XPTomePurchase := class(creative_device):

    @editable
    Button : button_device = button_device{}
    @editable
    Switch : switch_device = switch_device{}

    @editable
    Cost : big_number = big_number{}
    
    @editable
    AmountToAddToXP : float = 25.0

    @editable
    EndlessPurchase : logic = false

    PlayerStatManager : player_stats_manager = player_stats_manager{}

    # Runs when the device is started in a running game
    OnBegin<override>()<suspends>:void=
        Button.InteractedWithEvent.Subscribe(PurchaseTome)

    PurchaseTome(InAgent : agent):void=
        if(Player := player[InAgent]):
            DustAmount := PlayerStatManager.GetPlayerOreAmount(Player, PropResourceType.Dust)
            CanAfford := BigNumber_GreaterThanOrEquals(DustAmount, Cost)

            if(CanAfford = true):
                if(Switch.GetCurrentState[InAgent]):
                    #If endless purchase is false return, if not the rest of the logic will continue
                    if(EndlessPurchase = false):
                        return

                Switch.TurnOn(InAgent)
                PlayerStatManager.AddXPBonus(Player, AmountToAddToXP)
                PlayerStatManager.SpendPlayerOre(Player, Cost, false, PropResourceType.Dust)